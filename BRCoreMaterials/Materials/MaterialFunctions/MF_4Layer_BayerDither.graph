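# Material function that builds a temporal screen-space dither value from a 4x4 Bayer pattern and outputs it paired with its one-minus.
# A Custom node makes a mod-5 pixel pattern, shifted each frame by the temporal sample index.
# The static Bayer value is looked up without a texture, using one-hot row and column masks dotted with constant vectors.
# Both patterns are summed, scaled toward 0-1 and clamped just above zero.
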MaterialFunction /BRCoreMaterials/Materials/MaterialFunctions/MF_4Layer_BayerDither
graph {nodes [
  n1 = ViewProperty() {property=MEVP_TemporalSampleIndex}
  MaterialExpressionAdd_2 = Add(A=ScreenPosition().PixelPosition, B=AppendVector(A=@n1.Property, B=@n1.Property))
  n2 = Custom!(p=@MaterialExpressionAdd_2) {
   code="Mod( ((uint)(p.x) + 2 * (uint)(p.y)) , 5 )"
   outputType=CMOT_Float1
   inputs [{inputName=p, input {expression=&self:MaterialExpressionAdd_2}}]
  }
  n3 = Multiply(A=ScreenPosition().PixelPosition) {constB=0.25}
  n4 = Floor(None=Multiply(A=Frac(None=@n3)) {constB=4})
  n5 = Constant4Vector() {constant=<rgba 0 -1 -2 -3>}
  n6 = Abs(None=Add(A=ComponentMask(None=@n4) {r=true}, B=@n5.RGBA) {constB=-1})
  n7 = Add(A=@n5.RGBA, B=ComponentMask(None=@n4) {g=true}) {constB=-1}
  n8 = OneMinus(None=Saturate(None=Abs(None=@n7)))
  n9 = Constant4Vector() {constant=<rgba 0 0.5 0.125 0.625>}
  n10 = Constant4Vector() {constant=<rgba 0.75 0.25 0.875 0.375>}
  n11 = Constant4Vector() {constant=<rgba 0.1875 0.6875 0.0625 0.5625>}
  n12 = Constant4Vector() {constant=<rgba 0.9375 0.4375 0.8125 0.3125>}
  n13 = DotProduct(
   A=OneMinus(None=Saturate(None=@n6))
   B=MaterialFunctionCall(R=DotProduct(A=@n8, B=@n9.RGBA), G=DotProduct(A=@n8, B=@n10.RGBA), B=DotProduct(A=@n8, B=@n11.RGBA), A=DotProduct(A=@n8, B=@n12.RGBA)) {materialFunction=&AppendMany}.RGBA
  )
  n14 = Max(A=Multiply(A=Add(A=@n2, B=@n13) {constB=0.5}) {constB=0.1667}) {constB=0.001}
  FunctionOutput(None=AppendVector(A=@n14, B=OneMinus(None=@n14)))
 ]}
