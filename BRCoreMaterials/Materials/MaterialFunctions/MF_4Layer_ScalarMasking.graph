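# Material function that blends up to four scalar layer values into one output using an RGBA mask.
# L1 to L4 are packed into a four-component vector, zero-padded for unused layers, and dotted with the mask for a weighted sum.
# Nested static switch parameters (Use 2, 3 and 4 Layers) set the layer count at compile time.
# With Use 2 Layers off, L1 passes through unmasked.
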MaterialFunction /BRCoreMaterials/Materials/MaterialFunctions/MF_4Layer_ScalarMasking
graph {nodes [
  n1 = FunctionInput() {inputName=L1, inputType=FunctionInput_Scalar, sortPriority=1}
  n2 = AppendVector(A=@n1, B=FunctionInput() {inputName=L2, inputType=FunctionInput_Scalar, sortPriority=2})
  n3 = AppendVector(A=@n2, B=FunctionInput() {inputName=L3, inputType=FunctionInput_Scalar, sortPriority=3})
  n4 = FunctionInput() {inputName=L4, inputType=FunctionInput_Scalar, sortPriority=4}
  n5 = Constant()
  n6 = StaticSwitchParameter(
   True=StaticSwitchParameter(True=AppendVector(A=@n3, B=@n4), False=AppendVector(A=@n3, B=@n5)) {group=- Layer Switches, parameterName=Use 4 Layers, sortPriority=2}
   False=AppendVector(A=@n2, B=Constant2Vector())
  ) {group=- Layer Switches, parameterName=Use 3 Layers, sortPriority=1}
  n7 = FunctionInput() {inputName=RGBA Mask, inputType=FunctionInput_Vector4}
  FunctionOutput(None=StaticSwitchParameter(True=DotProduct(A=@n6, B=@n7), False=@n1) {group=- Layer Switches, parameterName=Use 2 Layers, sortPriority=0})
 ]}
