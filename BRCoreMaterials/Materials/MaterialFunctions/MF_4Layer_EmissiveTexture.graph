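# Material function that optionally adds a tiled emissive texture to a four-layer material's attributes, behind the "Emissive Texture" static switch.
# The texture is scaled by per-layer intensities, resolved from the RGBA Layer Mask input through MF_4Layer_ScalarMasking.
# The result is added to the incoming Emissive Color, with a RayTracingQualitySwitch applying a separate Lumen boost multiplier on the ray-traced path.
# With the switch off, the input material passes through unchanged.
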MaterialFunction /BRCoreMaterials/Materials/MaterialFunctions/MF_4Layer_EmissiveTexture
graph {nodes [
  n1 = FunctionInput() {inputName=Material, inputType=FunctionInput_MaterialAttributes}
  n2 = GetMaterialAttributes(None=@n1) {attributeGetTypes [#1]}
  n3 = TextureCoordinate()
  n4 = ScalarParameter() {defaultValue=1, group=Emissive Texture, parameterName=EmissiveTexture Tilling, sortPriority=1}
  n5 = Multiply(
   A=TextureSampleParameter2D(UVs=Multiply(A=@n3, B=@n4)) {group=Emissive Texture, parameterName=Emissive, sortPriority=0, texture=/BRCoreMaterials/Materials/Textures/T_Emissive}.RGB
   B=MaterialFunctionCall!(
    RGBA Mask=NamedRerouteUsage()
    L1=ScalarParameter() {defaultValue=1, group=Emissive Texture, parameterName=L1 EmissiveTexture Intensity, sortPriority=3}
    L2=ScalarParameter() {defaultValue=1, group=Emissive Texture, parameterName=L2 EmissiveTexture Intensity, sortPriority=4}
    L3=ScalarParameter() {defaultValue=1, group=Emissive Texture, parameterName=L3 EmissiveTexture Intensity, sortPriority=5}
    L4=ScalarParameter() {defaultValue=1, group=Emissive Texture, parameterName=L4 EmissiveTexture Intensity, sortPriority=6}
   ) {materialFunction=&/MF_4Layer_ScalarMasking}.Result
  )
  n6 = ScalarParameter() {defaultValue=1, group=Emissive Texture, parameterName=EmissiveTexture Lumen Boost, sortPriority=2}
  n7 = Add(A=@n2.EmissiveColor, B=RayTracingQualitySwitch(Normal=@n5, RayTraced=Multiply(A=@n5, B=@n6)))
  FunctionOutput(None=StaticSwitchParameter(
    True=SetMaterialAttributes!(MaterialAttributes=@n2.MaterialAttributes, EmissiveColor=@n7) {attributeSetTypes [#1]}
    False=@n1
   ) {group=- Effect Switches, parameterName=Emissive Texture})
  n8 = FunctionInput() {bUsePreviewValueAsDefault=true, inputName=Layer Mask, inputType=FunctionInput_Vector4, sortPriority=1, previewValue=<1 1 1 1>}
  NamedRerouteDeclaration(None=@n8) {name=Layer Mask, nodeColor=<rgba 0 0.5312 1 1>}
 ]}
