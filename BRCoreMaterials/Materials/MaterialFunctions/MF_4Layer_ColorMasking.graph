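# Material function that blends up to four color layers into one output using a packed four-channel layer mask.
# L1 is the base, and L2, L3 and L4 are lerped over it in sequence by the mask's green, blue and alpha channels; red is unused.
# Nested static switch parameters (Use 2, 3 and 4 Layers) set the layer count at compile time, each higher one requiring those below it.
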MaterialFunction /BRCoreMaterials/Materials/MaterialFunctions/MF_4Layer_ColorMasking
graph {nodes [
  n1 = FunctionInput() {inputName=L1, sortPriority=1}
  n2 = LinearInterpolate(A=@n1, B=FunctionInput() {inputName=L2, sortPriority=2}, Alpha=ComponentMask(None=NamedRerouteUsage()) {g=true})
  n3 = LinearInterpolate(A=@n2, B=FunctionInput() {inputName=L3, sortPriority=3}, Alpha=ComponentMask(None=NamedRerouteUsage()) {b=true})
  n4 = NamedRerouteUsage()
  n5 = StaticSwitchParameter(True=LinearInterpolate(A=@n3, B=FunctionInput() {inputName=L4, sortPriority=4}, Alpha=ComponentMask(None=@n4) {a=true}), False=@n3) {group=- Layer Switches, parameterName=Use 4 Layers, sortPriority=2}
  FunctionOutput(None=StaticSwitchParameter(True=StaticSwitchParameter(True=@n5, False=@n2) {group=- Layer Switches, parameterName=Use 3 Layers, sortPriority=1}, False=@n1) {group=- Layer Switches, parameterName=Use 2 Layers, sortPriority=0})
  NamedRerouteDeclaration(None=FunctionInput() {inputName=Layer Mask, inputType=FunctionInput_Vector4}) {name=Layer Mask, nodeColor=<rgba 0.6875 0 1 1>}
 ]}
