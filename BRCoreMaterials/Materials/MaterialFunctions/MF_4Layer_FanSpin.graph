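# Material function that spins fan-like parts via world position offset, with separate speed, axis and pivot parameters for four mask-selected layers.
# RGBA mask blends per-layer values; time with a five-second period drives the angle, while axis and pivot transform from instance space.
# RotateVertexAboutAxis takes these plus camera-relative position, and its WPO output feeds both World Position Offset and Normal.
# The "Fan Rotation" static switch bypasses the effect; the mask passes through as a second output.
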MaterialFunction /BRCoreMaterials/Materials/MaterialFunctions/MF_4Layer_FanSpin
graph {
 properties {description="This function is used to give assets rotation on the X, Y, and Z-axis. Normalized rotation along with fixed normals. "}
 nodes [
  n1 = FunctionInput() {inputName=Material Attributes, inputType=FunctionInput_MaterialAttributes}
  n2 = GetMaterialAttributes(None=@n1) {attributeGetTypes [#1]}
  n3 = MaterialFunctionCall!(
   RotationAngle=NamedRerouteUsage()
   NormalizedRotationAxis=NamedRerouteUsage()
   PivotPoint=NamedRerouteUsage()
   Position=WorldPosition() {worldPositionShaderOffset=WPT_CameraRelative}.XYZ
   Normal=@n2.Normal
   TangentU
   TangentV
  ) {materialFunction=/Game/Characters/Player/Male/Medium/Bodies/M_MED_Inferno/FX/Mat/RotateVertexAboutAxis}
  FunctionOutput(None=StaticSwitchParameter(
    True=SetMaterialAttributes!(MaterialAttributes=@n2.MaterialAttributes, WorldPositionOffset=@n3.WPO, Normal=@n3.WPO) {attributeSetTypes [#2, #1]}
    False=@n1
   ) {group=- Effect Switches, parameterName=Fan Rotation})
  FunctionOutput(None=NamedRerouteUsage()) {outputName=Mask, sortPriority=1}
  n4 = NamedRerouteUsage()
  n5 = ScalarParameter() {group=Fan Rotation, parameterName=L1 FanRotation Speed, sortPriority=0}
  n6 = ScalarParameter() {group=Fan Rotation, parameterName=L2 FanRotation Speed, sortPriority=3}
  n7 = ScalarParameter() {group=Fan Rotation, parameterName=L3 FanRotation Speed, sortPriority=6}
  n8 = ScalarParameter() {group=Fan Rotation, parameterName=L4 FanRotation Speed, sortPriority=9}
  NamedRerouteDeclaration(None=Multiply(
    A=MaterialFunctionCall!(RGBA Mask=@n4, L1=@n5, L2=@n6, L3=@n7, L4=@n8) {materialFunction=&/MF_4Layer_ScalarMasking}.Result
    B=Time() {period=5}
   )) {name=RotationAmount, nodeColor=<rgba 0 1 0.2031 1>}
  n9 = VectorParameter() {group=Fan Rotation, parameterName=L1 FanRotation RotationAxis, sortPriority=1, defaultValue=<rgba 1 0 0 1>}
  n10 = VectorParameter() {group=Fan Rotation, parameterName=L2 FanRotation RotationAxis, sortPriority=4, defaultValue=<rgba 1 0 0 1>}
  n11 = VectorParameter() {group=Fan Rotation, parameterName=L3 FanRotation RotationAxis, sortPriority=7, defaultValue=<rgba 1 0 0 1>}
  n12 = VectorParameter() {group=Fan Rotation, parameterName=L4 FanRotation RotationAxis, sortPriority=10, defaultValue=<rgba 1 0 0 1>}
  n13 = MaterialFunctionCall!(RGBA Mask=NamedRerouteUsage(), L1=@n9.RGB, L2=@n10.RGB, L3=@n11.RGB, L4=@n12.RGB) {materialFunction=&/MF_4Layer_ColorMasking}
  NamedRerouteDeclaration(None=Normalize(VectorInput=Transform(None=@n13.Result) {transformSourceType=TRANSFORMSOURCE_Instance})) {name=RotationAxis, nodeColor=<rgba 0.5469 0 1 1>}
  NamedRerouteDeclaration(None=FunctionInput() {inputName=Mask, inputType=FunctionInput_Vector4, sortPriority=1}) {name=RGBA Mask, nodeColor=<rgba 0 1 0.3438 1>}
  n14 = NamedRerouteUsage()
  n15 = VectorParameter() {group=Fan Rotation, parameterName=L1 FanRotation Pivot Offset, sortPriority=2, defaultValue=<rgba 0 0 0 1>}
  n16 = VectorParameter() {group=Fan Rotation, parameterName=L2 FanRotation Pivot Offset, sortPriority=5, defaultValue=<rgba 0 0 0 1>}
  n17 = VectorParameter() {group=Fan Rotation, parameterName=L3 FanRotation Pivot Offset, sortPriority=8, defaultValue=<rgba 0 0 0 1>}
  n18 = VectorParameter() {group=Fan Rotation, parameterName=L4 FanRotation Pivot Offset, sortPriority=11, defaultValue=<rgba 0 0 0 1>}
  NamedRerouteDeclaration(None=TransformPosition(None=MaterialFunctionCall(RGBA Mask=@n14, L1=@n15.RGB, L2=@n16.RGB, L3=@n17.RGB, L4=@n18.RGB) {materialFunction=&MF_4Layer_ColorMasking}.Result) {transformSourceType=TRANSFORMPOSSOURCE_Instance, transformType=TRANSFORMPOSSOURCE_TranslatedWorld}) {name=Pivot Offset, nodeColor=<rgba 0.2891 1 0 1>}
 ]
}
