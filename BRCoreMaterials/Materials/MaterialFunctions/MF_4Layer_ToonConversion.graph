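# Material function that converts a Material Attributes input into a Substrate Toon BSDF, named for use with a four-layer material setup.
# It breaks the attributes apart to feed the toon surface inputs, with a toon profile parameter and texture coordinates for pattern UVs.
# World position offset, displacement and temporal responsiveness are repacked into a separate Material Attributes output alongside the Material Toon output.
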MaterialFunction /BRCoreMaterials/Materials/MaterialFunctions/MF_4Layer_ToonConversion
graph {nodes [
  n1 = GetMaterialAttributes(None=FunctionInput() {inputName=Material, inputType=FunctionInput_MaterialAttributes}) {attributeGetTypes [
    #1
    #2
    #3
    #4
    #5
    #6
    #7
    #8
    #9
   ]}
  FunctionOutput(None=SubstrateToonBSDF!(
    BaseColor=@n1.BaseColor
    Metallic=@n1.BaseColor
    Specular=@n1.BaseColor
    Roughness=@n1.BaseColor
    Normal=@n1.BaseColor
    EmissiveColor=@n1.BaseColor
    PatternUVs=TextureCoordinate()
    Anisotropy
    Tangent
    ToonProfile=SubstrateToonProfileParameter()
   )) {outputName=Material Toon, sortPriority=1}
  FunctionOutput(None=SetMaterialAttributes!(
    MaterialAttributes
    WorldPositionOffset=@n1.WorldPositionOffset
    Displacement=@n1.WorldPositionOffset
    TemporalResponsiveness=@n1.WorldPositionOffset
   ) {attributeSetTypes [#7, #8, #9]}) {outputName=Material Attributes}
 ]}
