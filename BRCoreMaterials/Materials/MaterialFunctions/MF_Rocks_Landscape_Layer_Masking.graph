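# Material function that builds a layer blend mask, named for rock-to-landscape layering, with a separate Low/Mobile output.
# Toggleable masks include world Z normal, vertex color, DFAO, Z gradient, baked AO/curvature/painted, waterline, world-space noise and height.
# Masks have bias and intensity controls, many using channel weights, inversion or min/max blend selectors.
# If nodes stand in for lerps as cheaper selectors, and a multiplier compensates for sparser mobile vertex gradients.
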MaterialFunction /BRCoreMaterials/Materials/MaterialFunctions/MF_Rocks_Landscape_Layer_Masking
graph {expressions {
  MaterialExpressionAdd_0 {
   class=MaterialExpressionAdd
   inputs [
    A {from=MaterialExpressionIf_6}
    B {from=MaterialExpressionNamedRerouteUsage_11}
   ]
  }
  MaterialExpressionAdd_1 {
   class=MaterialExpressionAdd
   inputs [
    A {from=MaterialExpressionMultiply_0}
    B {from=MaterialExpressionSubtract_0}
   ]
  }
  MaterialExpressionAdd_18 {
   class=MaterialExpressionAdd
   inputs [
    A {from=MaterialExpressionStaticSwitch_39}
    B {from=MaterialExpressionNamedRerouteUsage_271}
   ]
  }
  MaterialExpressionAdd_19 {
   class=MaterialExpressionAdd
   inputs [
    A {from=MaterialExpressionIf_7}
    B {from=MaterialExpressionNamedRerouteUsage_282}
   ]
  }
  MaterialExpressionAdd_2 {
   class=MaterialExpressionAdd
   inputs [
    A {from=MaterialExpressionIf_5}
    B {from=MaterialExpressionNamedRerouteUsage_14}
   ]
  }
  MaterialExpressionAdd_20 {
   class=MaterialExpressionAdd
   inputs [
    A {from=MaterialExpressionStaticSwitch_40}
    B {from=MaterialExpressionNamedRerouteUsage_285}
   ]
  }
  MaterialExpressionAdd_21 {
   class=MaterialExpressionAdd
   inputs [
    A {from=MaterialExpressionIf_5}
    B {from=MaterialExpressionNamedRerouteUsage_281}
   ]
  }
  MaterialExpressionAdd_22 {
   class=MaterialExpressionAdd
   inputs [
    A {from=MaterialExpressionIf_8}
    B {from=MaterialExpressionNamedRerouteUsage_294}
   ]
  }
  MaterialExpressionAdd_23 {
   class=MaterialExpressionAdd
   inputs [
    A {from=MaterialExpressionWorldPosition_3, output=Z}
    B {from=MaterialExpressionAdd_24}
   ]
   properties {constB=30}
  }
  MaterialExpressionAdd_24 {
   class=MaterialExpressionAdd
   inputs [
    A {from=MaterialExpressionScalarParameter_51}
    B {from=MaterialExpressionNamedRerouteUsage_309}
   ]
   properties {constB=30}
  }
  MaterialExpressionAdd_51 {
   class=MaterialExpressionAdd
   inputs [
    A {from=MaterialExpressionMultiply_78}
    B {from=MaterialExpressionSubtract_19}
   ]
  }
  MaterialExpressionComponentMask_4 {class=MaterialExpressionComponentMask, inputs [None {from=MaterialExpressionGetMaterialAttributes_7, output=Normal}], properties {b=true}}
  MaterialExpressionComponentMask_5 {class=MaterialExpressionComponentMask, inputs [None {from=MaterialExpressionVertexNormalWS_3}], properties {b=true}}
  MaterialExpressionConstant_0 {class=MaterialExpressionConstant}
  MaterialExpressionConstant_1 {class=MaterialExpressionConstant, properties {r=1}}
  MaterialExpressionConstant_2 {class=MaterialExpressionConstant}
  MaterialExpressionConstant_3 {class=MaterialExpressionConstant}
  MaterialExpressionConstant_4 {class=MaterialExpressionConstant, properties {r=1}}
  MaterialExpressionConstant_5 {class=MaterialExpressionConstant}
  MaterialExpressionConstant_65 {class=MaterialExpressionConstant}
  MaterialExpressionConstant_66 {class=MaterialExpressionConstant, properties {r=2}}
  MaterialExpressionConstant_67 {class=MaterialExpressionConstant, properties {r=1}}
  MaterialExpressionConstant_68 {class=MaterialExpressionConstant, properties {r=-0.2}}
  MaterialExpressionConstant_69 {class=MaterialExpressionConstant, properties {r=1.5}}
  MaterialExpressionConstant_70 {class=MaterialExpressionConstant}
  MaterialExpressionConstant_71 {class=MaterialExpressionConstant, properties {r=1}}
  MaterialExpressionConstant_72 {class=MaterialExpressionConstant}
  MaterialExpressionConstant_73 {class=MaterialExpressionConstant, properties {r=1}}
  MaterialExpressionConstant_74 {class=MaterialExpressionConstant}
  MaterialExpressionConstant_75 {class=MaterialExpressionConstant, properties {r=1}}
  MaterialExpressionConstant_76 {class=MaterialExpressionConstant, properties {r=1}}
  MaterialExpressionConstant_77 {class=MaterialExpressionConstant}
  MaterialExpressionConstant_78 {class=MaterialExpressionConstant, properties {r=1}}
  MaterialExpressionConstant_79 {class=MaterialExpressionConstant, properties {r=1}}
  MaterialExpressionConstant_80 {class=MaterialExpressionConstant}
  MaterialExpressionConstant_81 {class=MaterialExpressionConstant, properties {r=1}}
  MaterialExpressionConstant_82 {class=MaterialExpressionConstant, properties {r=1}}
  MaterialExpressionConstant_83 {class=MaterialExpressionConstant, properties {r=1}}
  MaterialExpressionConstant_84 {class=MaterialExpressionConstant, properties {r=1}}
  MaterialExpressionConstant_85 {class=MaterialExpressionConstant, properties {r=1}}
  MaterialExpressionConstant_86 {class=MaterialExpressionConstant, properties {r=1}}
  MaterialExpressionConstant_87 {class=MaterialExpressionConstant, properties {r=1}}
  MaterialExpressionConstant_88 {class=MaterialExpressionConstant}
  MaterialExpressionConstant_89 {class=MaterialExpressionConstant, properties {r=1}}
  MaterialExpressionConstant_90 {class=MaterialExpressionConstant, properties {r=1}}
  MaterialExpressionConstant_91 {class=MaterialExpressionConstant, properties {r=1}}
  MaterialExpressionConstant_92 {class=MaterialExpressionConstant}
  MaterialExpressionConstant_93 {class=MaterialExpressionConstant}
  MaterialExpressionConstant_94 {class=MaterialExpressionConstant, properties {r=1}}
  MaterialExpressionConstant_95 {class=MaterialExpressionConstant, properties {r=1}}
  MaterialExpressionConstant_96 {class=MaterialExpressionConstant, properties {r=1}}
  MaterialExpressionConstant_97 {class=MaterialExpressionConstant, properties {r=1}}
  MaterialExpressionConvert_1 {
   class=MaterialExpressionConvert
   inputs [
    None {from=MaterialExpressionNamedRerouteUsage_12}
    None {from=MaterialExpressionNamedRerouteUsage_13}
    None {from=MaterialExpressionNamedRerouteUsage_18}
   ]
   properties {
    nodeName=Make Vector3
    convertInputs [
     {
      type=Scalar
      expressionInput {
       expression=&self:MaterialExpressionNamedRerouteUsage_12
       inputName=None
       mask=0
       maskA=0
       maskB=0
       maskG=0
       maskR=0
       outputIndex=0
      }
     }
     {
      type=Scalar
      expressionInput {
       expression=&self:MaterialExpressionNamedRerouteUsage_13
       inputName=None
       mask=0
       maskA=0
       maskB=0
       maskG=0
       maskR=0
       outputIndex=0
      }
     }
     {
      type=Scalar
      expressionInput {
       expression=&self:MaterialExpressionNamedRerouteUsage_18
       inputName=None
       mask=0
       maskA=0
       maskB=0
       maskG=0
       maskR=0
       outputIndex=0
      }
     }
    ]
    convertOutputs [{type=Vector3}]
   }
  }
  MaterialExpressionDotProduct_10 {
   class=MaterialExpressionDotProduct
   inputs [
    A {from=MaterialExpressionNamedRerouteUsage_300, output=""}
    B {from=MaterialExpressionNamedRerouteUsage_313, output=""}
   ]
   properties {}
  }
  MaterialExpressionDotProduct_11 {
   class=MaterialExpressionDotProduct
   inputs [
    A {from=MaterialExpressionNamedRerouteUsage_274, output=""}
    B {from=MaterialExpressionVertexColor_3, output=""}
   ]
   properties {}
  }
  MaterialExpressionDotProduct_9 {
   class=MaterialExpressionDotProduct
   inputs [
    A {from=MaterialExpressionNamedRerouteUsage_292, output=""}
    B {from=MaterialExpressionNamedRerouteUsage_314, output=""}
   ]
   properties {}
  }
  MaterialExpressionFunctionInput_0 {
   class=MaterialExpressionFunctionInput
   inputs [Preview {from="", output=""}]
   properties {inputName=TerrainMasks_QuadB, inputType=FunctionInput_Scalar, sortPriority=1013}
  }
  MaterialExpressionFunctionInput_1 {
   class=MaterialExpressionFunctionInput
   inputs [Preview {from="", output=""}]
   properties {inputName=TerrainMasks_QuadA, inputType=FunctionInput_Scalar, sortPriority=1012}
  }
  MaterialExpressionFunctionInput_10 {
   class=MaterialExpressionFunctionInput
   inputs [Preview {from=MaterialExpressionConstant_2, output=""}]
   properties {bUsePreviewValueAsDefault=true, inputName="• VTX Color - Bias", inputType=FunctionInput_Scalar, sortPriority=175}
  }
  MaterialExpressionFunctionInput_100 {
   class=MaterialExpressionFunctionInput
   inputs [Preview {from=MaterialExpressionConstant_68, output=""}]
   properties {bUsePreviewValueAsDefault=true, inputName=* World Z Normal - Bias, inputType=FunctionInput_Scalar, sortPriority=120}
  }
  MaterialExpressionFunctionInput_101 {
   class=MaterialExpressionFunctionInput
   inputs [Preview {from=MaterialExpressionStaticBool_18, output=""}]
   properties {bUsePreviewValueAsDefault=true, inputName="• World Z Normal - Use VTX Normals", inputType=FunctionInput_StaticBool, sortPriority=110}
  }
  MaterialExpressionFunctionInput_102 {
   class=MaterialExpressionFunctionInput
   inputs [Preview {from=MaterialExpressionStaticBool_21, output=""}]
   properties {bUsePreviewValueAsDefault=true, inputName=--- Use WORLD Z NORMAL MASKING ---, inputType=FunctionInput_StaticBool, sortPriority=100}
  }
  MaterialExpressionFunctionInput_103 {
   class=MaterialExpressionFunctionInput
   inputs [Preview {from=MaterialExpressionConstant_69, output=""}]
   properties {bUsePreviewValueAsDefault=true, inputName=* World Z Normal - Intensity, inputType=FunctionInput_Scalar, sortPriority=121}
  }
  MaterialExpressionFunctionInput_104 {
   class=MaterialExpressionFunctionInput
   inputs [Preview {from=MaterialExpressionStaticBool_22, output=""}]
   properties {bUsePreviewValueAsDefault=true, inputName=* DFAO - Force Baked AO on Nanite, inputType=FunctionInput_StaticBool, sortPriority=210}
  }
  MaterialExpressionFunctionInput_105 {
   class=MaterialExpressionFunctionInput
   inputs [Preview {from=MaterialExpressionStaticBool_23, output=""}]
   properties {bUsePreviewValueAsDefault=true, inputName=--- Use DFAO MASKING ---, inputType=FunctionInput_StaticBool, sortPriority=200}
  }
  MaterialExpressionFunctionInput_106 {
   class=MaterialExpressionFunctionInput
   inputs [Preview {from=MaterialExpressionConstant_70, output=""}]
   properties {bUsePreviewValueAsDefault=true, inputName="• DFAO - Bias", inputType=FunctionInput_Scalar, sortPriority=225}
  }
  MaterialExpressionFunctionInput_107 {
   class=MaterialExpressionFunctionInput
   inputs [Preview {from=MaterialExpressionConstant_71, output=""}]
   properties {bUsePreviewValueAsDefault=true, inputName="• DFAO - Intensity", inputType=FunctionInput_Scalar, sortPriority=226}
  }
  MaterialExpressionFunctionInput_108 {
   class=MaterialExpressionFunctionInput
   inputs [Preview {from=MaterialExpressionStaticBool_24, output=""}]
   properties {bUsePreviewValueAsDefault=true, inputName=--- Use Z GRADIENT MASKING ---, inputType=FunctionInput_StaticBool, sortPriority=250}
  }
  MaterialExpressionFunctionInput_109 {
   class=MaterialExpressionFunctionInput
   inputs [Preview {from=MaterialExpressionConstant_72, output=""}]
   properties {bUsePreviewValueAsDefault=true, inputName="• Z Gradient - Height - Bottom", inputType=FunctionInput_Scalar, sortPriority=265}
  }
  MaterialExpressionFunctionInput_11 {
   class=MaterialExpressionFunctionInput
   inputs [Preview {from=MaterialExpressionSetMaterialAttributes_0, output=""}]
   properties {
    description=MaterialAttributes from previous layers
    inputName=Layer A MaterialAttributes
    inputType=FunctionInput_MaterialAttributes
    sortPriority=-100
   }
  }
  MaterialExpressionFunctionInput_110 {
   class=MaterialExpressionFunctionInput
   inputs [Preview {from=MaterialExpressionConstant_73, output=""}]
   properties {bUsePreviewValueAsDefault=true, inputName="• Z Gradient - Height - Top", inputType=FunctionInput_Scalar, sortPriority=266}
  }
  MaterialExpressionFunctionInput_111 {
   class=MaterialExpressionFunctionInput
   inputs [Preview {from=MaterialExpressionStaticBool_26, output=""}]
   properties {
    bUsePreviewValueAsDefault=true
    inputName=--- Use BAKED AO/CURVATURE/PAINTED MASKING ---
    inputType=FunctionInput_StaticBool
    sortPriority=350
   }
  }
  MaterialExpressionFunctionInput_112 {
   class=MaterialExpressionFunctionInput
   inputs [Preview {from=MaterialExpressionConstant_74, output=""}]
   properties {bUsePreviewValueAsDefault=true, inputName="• Baked AO/Curvature/Painted - Bias", inputType=FunctionInput_Scalar, sortPriority=375}
  }
  MaterialExpressionFunctionInput_113 {
   class=MaterialExpressionFunctionInput
   inputs [Preview {from=MaterialExpressionConstant_75, output=""}]
   properties {bUsePreviewValueAsDefault=true, inputName="• Baked AO/Curvature/Painted - Intensity", inputType=FunctionInput_Scalar, sortPriority=376}
  }
  MaterialExpressionFunctionInput_114 {
   class=MaterialExpressionFunctionInput
   inputs [Preview {from=MaterialExpressionConstant_76, output=""}]
   properties {bUsePreviewValueAsDefault=true, inputName="• Baked AO/Curvature/Painted - Channel Weights", sortPriority=360}
  }
  MaterialExpressionFunctionInput_115 {
   class=MaterialExpressionFunctionInput
   inputs [Preview {from=MaterialExpressionStaticBool_27, output=""}]
   properties {bUsePreviewValueAsDefault=true, inputName=--- Use WS NOISE MASKING ---, inputType=FunctionInput_StaticBool, sortPriority=450}
  }
  MaterialExpressionFunctionInput_116 {
   class=MaterialExpressionFunctionInput
   inputs [Preview {from=MaterialExpressionConstant_77, output=""}]
   properties {bUsePreviewValueAsDefault=true, inputName="• WS Noise - Bias", inputType=FunctionInput_Scalar, sortPriority=475}
  }
  MaterialExpressionFunctionInput_117 {
   class=MaterialExpressionFunctionInput
   inputs [Preview {from=MaterialExpressionConstant_78, output=""}]
   properties {bUsePreviewValueAsDefault=true, inputName="• WS Noise - Intensity", inputType=FunctionInput_Scalar, sortPriority=476}
  }
  MaterialExpressionFunctionInput_118 {
   class=MaterialExpressionFunctionInput
   inputs [Preview {from=MaterialExpressionConstant_79, output=""}]
   properties {bUsePreviewValueAsDefault=true, inputName="• Masking - Intensity", inputType=FunctionInput_Scalar, sortPriority=61}
  }
  MaterialExpressionFunctionInput_119 {
   class=MaterialExpressionFunctionInput
   inputs [Preview {from=MaterialExpressionConstant_80, output=""}]
   properties {bUsePreviewValueAsDefault=true, inputName="• Masking - Bias", inputType=FunctionInput_Scalar, sortPriority=60}
  }
  MaterialExpressionFunctionInput_12 {
   class=MaterialExpressionFunctionInput
   inputs [Preview {from=MaterialExpressionConstant_4, output=""}]
   properties {bUsePreviewValueAsDefault=true, inputName="• Masking - Intensity (Low/Mobile)", inputType=FunctionInput_Scalar, sortPriority=61}
  }
  MaterialExpressionFunctionInput_120 {
   class=MaterialExpressionFunctionInput
   inputs [Preview {from=MaterialExpressionConstant_3, output=""}]
   properties {bUsePreviewValueAsDefault=true, inputName=* Masking - Invert, inputType=FunctionInput_Scalar, sortPriority=50}
  }
  MaterialExpressionFunctionInput_121 {
   class=MaterialExpressionFunctionInput
   inputs [Preview {from=MaterialExpressionConstant_81, output=""}]
   properties {bUsePreviewValueAsDefault=true, inputName="• VTX Color - Blend (1=Max 0=Min)", inputType=FunctionInput_Scalar, sortPriority=170}
  }
  MaterialExpressionFunctionInput_122 {
   class=MaterialExpressionFunctionInput
   inputs [Preview {from=MaterialExpressionConstant_82, output=""}]
   properties {bUsePreviewValueAsDefault=true, inputName="• DFAO - Blend (1=Max 0=Min)", inputType=FunctionInput_Scalar, sortPriority=220}
  }
  MaterialExpressionFunctionInput_123 {
   class=MaterialExpressionFunctionInput
   inputs [Preview {from=MaterialExpressionConstant_83, output=""}]
   properties {bUsePreviewValueAsDefault=true, inputName="• Z Gradient - Blend (1=Max 0=Min)", inputType=FunctionInput_Scalar, sortPriority=260}
  }
  MaterialExpressionFunctionInput_124 {
   class=MaterialExpressionFunctionInput
   inputs [Preview {from=MaterialExpressionConstant_84, output=""}]
   properties {
    bUsePreviewValueAsDefault=true
    inputName="• Baked AO/Curvature/Painted - Blend (1=Max 0=Min)"
    inputType=FunctionInput_Scalar
    sortPriority=370
   }
  }
  MaterialExpressionFunctionInput_125 {
   class=MaterialExpressionFunctionInput
   inputs [Preview {from=MaterialExpressionConstant_85, output=""}]
   properties {bUsePreviewValueAsDefault=true, inputName="• WS Noise - Blend (1=Blend 0=Min)", inputType=FunctionInput_Scalar, sortPriority=470}
  }
  MaterialExpressionFunctionInput_126 {
   class=MaterialExpressionFunctionInput
   inputs [Preview {from=MaterialExpressionConstant_86, output=""}]
   properties {bUsePreviewValueAsDefault=true, inputName="• WS Noise - Channel Weights", sortPriority=460}
  }
  MaterialExpressionFunctionInput_127 {
   class=MaterialExpressionFunctionInput
   inputs [Preview {from=MaterialExpressionStaticBool_30, output=""}]
   properties {bUsePreviewValueAsDefault=true, inputName=--- Use HEIGHT MASKING ---, inputType=FunctionInput_StaticBool, sortPriority=500}
  }
  MaterialExpressionFunctionInput_128 {
   class=MaterialExpressionFunctionInput
   inputs [Preview {from=MaterialExpressionConstant_88, output=""}]
   properties {bUsePreviewValueAsDefault=true, inputName="• Height - Bias", inputType=FunctionInput_Scalar, sortPriority=515}
  }
  MaterialExpressionFunctionInput_129 {
   class=MaterialExpressionFunctionInput
   inputs [Preview {from=MaterialExpressionConstant_89, output=""}]
   properties {bUsePreviewValueAsDefault=true, inputName="• Height - Intensity", inputType=FunctionInput_Scalar, sortPriority=516}
  }
  MaterialExpressionFunctionInput_13 {
   class=MaterialExpressionFunctionInput
   inputs [Preview {from=MaterialExpressionConstant_5, output=""}]
   properties {bUsePreviewValueAsDefault=true, inputName="• Masking - Bias (Low/Mobile)", inputType=FunctionInput_Scalar, sortPriority=60}
  }
  MaterialExpressionFunctionInput_130 {
   class=MaterialExpressionFunctionInput
   inputs [Preview {from=MaterialExpressionConstant_90, output=""}]
   properties {bUsePreviewValueAsDefault=true, inputName="• Height - Blend (1=Blend 0=Min)", inputType=FunctionInput_Scalar, sortPriority=510}
  }
  MaterialExpressionFunctionInput_131 {
   class=MaterialExpressionFunctionInput
   inputs [Preview {from=MaterialExpressionConstant_91, output=""}]
   properties {
    bUsePreviewValueAsDefault=true
    inputName="• Height - Selection (1=Current Layer 0=Previous Layers)"
    inputType=FunctionInput_Scalar
    sortPriority=509
   }
  }
  MaterialExpressionFunctionInput_132 {
   class=MaterialExpressionFunctionInput
   inputs [Preview {from=MaterialExpressionStaticBool_31, output=""}]
   properties {bUsePreviewValueAsDefault=true, inputName=--- Use WATERLINE MASKING ---, inputType=FunctionInput_StaticBool, sortPriority=400}
  }
  MaterialExpressionFunctionInput_133 {
   class=MaterialExpressionFunctionInput
   inputs [Preview {from=MaterialExpressionConstant_92, output=""}]
   properties {bUsePreviewValueAsDefault=true, inputName=* Waterline - Bias, inputType=FunctionInput_Scalar, sortPriority=415}
  }
  MaterialExpressionFunctionInput_134 {
   class=MaterialExpressionFunctionInput
   inputs [Preview {from=MaterialExpressionConstant_93, output=""}]
   properties {bUsePreviewValueAsDefault=true, inputName="* Waterline - Blend (1=Max 0=Min)", inputType=FunctionInput_Scalar, sortPriority=410}
  }
  MaterialExpressionFunctionInput_135 {
   class=MaterialExpressionFunctionInput
   inputs [Preview {from=MaterialExpressionConstant_94, output=""}]
   properties {bUsePreviewValueAsDefault=true, inputName=* Waterline - Intensity, inputType=FunctionInput_Scalar, sortPriority=416}
  }
  MaterialExpressionFunctionInput_136 {
   class=MaterialExpressionFunctionInput
   inputs [Preview {from=MaterialExpressionConstant_95, output=""}]
   properties {bUsePreviewValueAsDefault=true, inputName="• Baked AO/Curvature/Painted - Invert", inputType=FunctionInput_Scalar, sortPriority=361}
  }
  MaterialExpressionFunctionInput_137 {
   class=MaterialExpressionFunctionInput
   inputs [Preview {from=MaterialExpressionConstant_96, output=""}]
   properties {bUsePreviewValueAsDefault=true, inputName="• WS Noise - Invert", inputType=FunctionInput_Scalar, sortPriority=461}
  }
  MaterialExpressionFunctionInput_138 {
   class=MaterialExpressionFunctionInput
   inputs [Preview {from=MaterialExpressionConstant_97, output=""}]
   properties {bUsePreviewValueAsDefault=true, inputName="• VTX Color - Invert", inputType=FunctionInput_Scalar, sortPriority=161}
  }
  MaterialExpressionFunctionInput_17 {
   class=MaterialExpressionFunctionInput
   inputs [Preview {from="", output=""}]
   properties {inputName=TerrainMasks_QuadD, inputType=FunctionInput_Scalar, sortPriority=1015}
  }
  MaterialExpressionFunctionInput_18 {
   class=MaterialExpressionFunctionInput
   inputs [Preview {from="", output=""}]
   properties {inputName="\\ --- Dependencies --- /", inputType=FunctionInput_Scalar, sortPriority=800}
  }
  MaterialExpressionFunctionInput_2 {
   class=MaterialExpressionFunctionInput
   inputs [Preview {from="", output=""}]
   properties {inputName=TerrainMasks_QuadC, inputType=FunctionInput_Scalar, sortPriority=1014}
  }
  MaterialExpressionFunctionInput_20 {
   class=MaterialExpressionFunctionInput
   inputs [Preview {from="", output=""}]
   properties {inputName="\\ --- Landscape/Biome Masks --- /", inputType=FunctionInput_Scalar, sortPriority=1000}
  }
  MaterialExpressionFunctionInput_24 {
   class=MaterialExpressionFunctionInput
   inputs [Preview {from="", output=""}]
   properties {inputName=Landscape Layers, inputType=FunctionInput_MaterialAttributes, sortPriority=1005}
  }
  MaterialExpressionFunctionInput_25 {
   class=MaterialExpressionFunctionInput
   inputs [Preview {from="", output=""}]
   properties {inputName=Landscape Height, inputType=FunctionInput_Scalar, sortPriority=1006}
  }
  MaterialExpressionFunctionInput_26 {
   class=MaterialExpressionFunctionInput
   inputs [Preview {from="", output=""}]
   properties {inputName=WS Noise Masks, inputType=FunctionInput_Scalar, sortPriority=815}
  }
  MaterialExpressionFunctionInput_3 {
   class=MaterialExpressionFunctionInput
   inputs [Preview {from="", output=""}]
   properties {inputName=Baked Curvature, inputType=FunctionInput_Scalar, sortPriority=810}
  }
  MaterialExpressionFunctionInput_4 {
   class=MaterialExpressionFunctionInput
   inputs [Preview {from="", output=""}]
   properties {inputName=DFAO, inputType=FunctionInput_Scalar, sortPriority=813}
  }
  MaterialExpressionFunctionInput_5 {
   class=MaterialExpressionFunctionInput
   inputs [Preview {from="", output=""}]
   properties {inputName=Z Gradient, inputType=FunctionInput_Scalar, sortPriority=814}
  }
  MaterialExpressionFunctionInput_6 {
   class=MaterialExpressionFunctionInput
   inputs [Preview {from="", output=""}]
   properties {inputName=Baked AO, inputType=FunctionInput_Scalar, sortPriority=811}
  }
  MaterialExpressionFunctionInput_7 {
   class=MaterialExpressionFunctionInput
   inputs [Preview {from="", output=""}]
   properties {inputName=Painted Mask, inputType=FunctionInput_Scalar, sortPriority=812}
  }
  MaterialExpressionFunctionInput_8 {
   class=MaterialExpressionFunctionInput
   inputs [Preview {from=MaterialExpressionConstant_0, output=""}]
   properties {bUsePreviewValueAsDefault=true, inputName="• Height - Min", inputType=FunctionInput_Scalar, sortPriority=512}
  }
  MaterialExpressionFunctionInput_9 {
   class=MaterialExpressionFunctionInput
   inputs [Preview {from=MaterialExpressionConstant_1, output=""}]
   properties {bUsePreviewValueAsDefault=true, inputName="• Height - Max", inputType=FunctionInput_Scalar, sortPriority=513}
  }
  MaterialExpressionFunctionInput_97 {
   class=MaterialExpressionFunctionInput
   inputs [Preview {from=MaterialExpressionConstant_87, output=""}]
   properties {bUsePreviewValueAsDefault=true, inputName="• VTX Color - Channel Weights", sortPriority=160}
  }
  MaterialExpressionFunctionInput_98 {
   class=MaterialExpressionFunctionInput
   inputs [Preview {from=MaterialExpressionConstant_67, output=""}]
   properties {bUsePreviewValueAsDefault=true, inputName="• VTX Color - Intensity", inputType=FunctionInput_Scalar, sortPriority=176}
  }
  MaterialExpressionFunctionInput_99 {
   class=MaterialExpressionFunctionInput
   inputs [Preview {from=MaterialExpressionStaticBool_17, output=""}]
   properties {bUsePreviewValueAsDefault=true, inputName=--- Use VTX COLOR MASKING ---, inputType=FunctionInput_StaticBool, sortPriority=150}
  }
  MaterialExpressionFunctionOutput_0 {
   class=MaterialExpressionFunctionOutput
   inputs [None {from=MaterialExpressionNamedRerouteUsage_16, output=""}]
   properties {outputName="Combined Blend Mask (Low/Mobile)"}
  }
  MaterialExpressionFunctionOutput_1 {
   class=MaterialExpressionFunctionOutput
   inputs [None {from=MaterialExpressionNamedRerouteUsage_297, output=""}]
   properties {outputName=Combined Blend Mask}
  }
  MaterialExpressionGetMaterialAttributes_7 {
   class=MaterialExpressionGetMaterialAttributes
   inputs [None {from=MaterialExpressionNamedRerouteUsage_265, output=""}]
   properties {attributeGetTypes [#1]}
  }
  MaterialExpressionGetMaterialAttributes_9 {
   class=MaterialExpressionGetMaterialAttributes
   inputs [None {from=MaterialExpressionNamedRerouteUsage_303, output=""}]
   properties {attributeGetTypes [#2]}
  }
  MaterialExpressionIf_0 {
   class=MaterialExpressionIf
   inputs [
    A {from=MaterialExpressionNamedRerouteUsage_308, output=""}
    B {from="", output=""}
    "A > B" {from=MaterialExpressionMax_27, output=""}
    "A == B" {from="", output=""}
    "A < B" {from=MaterialExpressionMin_27, output=""}
   ]
   properties {constB=0.5, desc="cheaper than lerp without \"==\" condition", equalsThreshold=0.001}
  }
  MaterialExpressionIf_1 {
   class=MaterialExpressionIf
   inputs [
    A {from=MaterialExpressionNamedRerouteUsage_290, output=""}
    B {from="", output=""}
    "A > B" {from=MaterialExpressionMax_21, output=""}
    "A == B" {from="", output=""}
    "A < B" {from=MaterialExpressionMin_22, output=""}
   ]
   properties {constB=0.5, desc="cheaper than lerp without \"==\" condition", equalsThreshold=0.001}
  }
  MaterialExpressionIf_10 {
   class=MaterialExpressionIf
   inputs [
    A {from=MaterialExpressionNamedRerouteUsage_306, output=""}
    B {from="", output=""}
    "A > B" {from=MaterialExpressionAdd_51, output=""}
    "A == B" {from="", output=""}
    "A < B" {from=MaterialExpressionMin_26, output=""}
   ]
   properties {constB=0.5, desc="cheaper than lerp without \"==\" condition", equalsThreshold=0.001}
  }
  MaterialExpressionIf_2 {
   class=MaterialExpressionIf
   inputs [
    A {from=MaterialExpressionNamedRerouteUsage_289, output=""}
    B {from="", output=""}
    "A > B" {from=MaterialExpressionMax_18, output=""}
    "A == B" {from="", output=""}
    "A < B" {from=MaterialExpressionMin_20, output=""}
   ]
   properties {constB=0.5, desc="cheaper than lerp without \"==\" condition", equalsThreshold=0.001}
  }
  MaterialExpressionIf_3 {
   class=MaterialExpressionIf
   inputs [
    A {from=MaterialExpressionNamedRerouteUsage_291, output=""}
    B {from="", output=""}
    "A > B" {from=MaterialExpressionMax_24, output=""}
    "A == B" {from="", output=""}
    "A < B" {from=MaterialExpressionMin_25, output=""}
   ]
   properties {constB=0.5, desc="cheaper than lerp without \"==\" condition", equalsThreshold=0.001}
  }
  MaterialExpressionIf_4 {
   class=MaterialExpressionIf
   inputs [
    A {from=MaterialExpressionNamedRerouteUsage_287, output=""}
    B {from="", output=""}
    "A > B" {from=MaterialExpressionMax_17, output=""}
    "A == B" {from="", output=""}
    "A < B" {from=MaterialExpressionMin_17, output=""}
   ]
   properties {constB=0.5, desc="cheaper than lerp without \"==\" condition", equalsThreshold=0.001}
  }
  MaterialExpressionIf_5 {
   class=MaterialExpressionIf
   inputs [
    A {from=MaterialExpressionNamedRerouteUsage_299, output=""}
    B {from="", output=""}
    "A > B" {from=MaterialExpressionOneMinus_12, output=""}
    "A == B" {from="", output=""}
    "A < B" {from=MaterialExpressionReroute_183, output=""}
   ]
   properties {constB=0.5, desc="cheaper than lerp without \"==\" condition", equalsThreshold=0.001}
  }
  MaterialExpressionIf_6 {
   class=MaterialExpressionIf
   inputs [
    A {from=MaterialExpressionNamedRerouteUsage_317, output=""}
    B {from="", output=""}
    "A > B" {from=MaterialExpressionOneMinus_16, output=""}
    "A == B" {from="", output=""}
    "A < B" {from=MaterialExpressionDotProduct_11, output=""}
   ]
   properties {constB=0.5, desc="cheaper than lerp without \"==\" condition", equalsThreshold=0.001}
  }
  MaterialExpressionIf_7 {
   class=MaterialExpressionIf
   inputs [
    A {from=MaterialExpressionNamedRerouteUsage_315, output=""}
    B {from="", output=""}
    "A > B" {from=MaterialExpressionOneMinus_14, output=""}
    "A == B" {from="", output=""}
    "A < B" {from=MaterialExpressionDotProduct_9, output=""}
   ]
   properties {constB=0.5, desc="cheaper than lerp without \"==\" condition", equalsThreshold=0.001}
  }
  MaterialExpressionIf_8 {
   class=MaterialExpressionIf
   inputs [
    A {from=MaterialExpressionNamedRerouteUsage_316, output=""}
    B {from="", output=""}
    "A > B" {from=MaterialExpressionOneMinus_15, output=""}
    "A == B" {from="", output=""}
    "A < B" {from=MaterialExpressionDotProduct_10, output=""}
   ]
   properties {constB=0.5, desc="cheaper than lerp without \"==\" condition", equalsThreshold=0.001}
  }
  MaterialExpressionIf_9 {
   class=MaterialExpressionIf
   inputs [
    A {from=MaterialExpressionNamedRerouteUsage_288, output=""}
    B {from="", output=""}
    "A > B" {from=MaterialExpressionAdd_1, output=""}
    "A == B" {from="", output=""}
    "A < B" {from=MaterialExpressionMin_23, output=""}
   ]
   properties {constB=0.5, desc="cheaper than lerp without \"==\" condition", equalsThreshold=0.001}
  }
  MaterialExpressionLinearInterpolate_31 {
   class=MaterialExpressionLinearInterpolate
   inputs [
    A {from=MaterialExpressionNamedRerouteUsage_1}
    B {from=MaterialExpressionNamedRerouteUsage_6}
    Alpha {from=MaterialExpressionLinearInterpolate_37}
   ]
  }
  MaterialExpressionLinearInterpolate_37 {
   class=MaterialExpressionLinearInterpolate
   inputs [
    A {from=MaterialExpressionReroute_214}
    B {from=MaterialExpressionReroute_213}
    Alpha {from=MaterialExpressionNamedRerouteUsage_304}
   ]
  }
  MaterialExpressionMax_17 {
   class=MaterialExpressionMax
   inputs [
    A {from=MaterialExpressionReroute_185}
    B {from=MaterialExpressionReroute_186}
   ]
  }
  MaterialExpressionMax_18 {
   class=MaterialExpressionMax
   inputs [
    A {from=MaterialExpressionReroute_187}
    B {from=MaterialExpressionReroute_188}
   ]
  }
  MaterialExpressionMax_21 {
   class=MaterialExpressionMax
   inputs [
    A {from=MaterialExpressionReroute_192}
    B {from=MaterialExpressionReroute_193}
   ]
  }
  MaterialExpressionMax_24 {
   class=MaterialExpressionMax
   inputs [
    A {from=MaterialExpressionReroute_199}
    B {from=MaterialExpressionReroute_200}
   ]
  }
  MaterialExpressionMax_27 {
   class=MaterialExpressionMax
   inputs [
    A {from=MaterialExpressionReroute_216}
    B {from=MaterialExpressionReroute_217}
   ]
  }
  MaterialExpressionMin_17 {
   class=MaterialExpressionMin
   inputs [
    A {from=MaterialExpressionReroute_185, output=""}
    B {from=MaterialExpressionReroute_186, output=""}
   ]
   properties {}
  }
  MaterialExpressionMin_20 {
   class=MaterialExpressionMin
   inputs [
    A {from=MaterialExpressionReroute_187, output=""}
    B {from=MaterialExpressionReroute_188, output=""}
   ]
   properties {}
  }
  MaterialExpressionMin_22 {
   class=MaterialExpressionMin
   inputs [
    A {from=MaterialExpressionReroute_192, output=""}
    B {from=MaterialExpressionReroute_193, output=""}
   ]
   properties {}
  }
  MaterialExpressionMin_23 {
   class=MaterialExpressionMin
   inputs [
    A {from=MaterialExpressionReroute_196, output=""}
    B {from=MaterialExpressionReroute_197, output=""}
   ]
   properties {}
  }
  MaterialExpressionMin_25 {
   class=MaterialExpressionMin
   inputs [
    A {from=MaterialExpressionReroute_199, output=""}
    B {from=MaterialExpressionReroute_200, output=""}
   ]
   properties {}
  }
  MaterialExpressionMin_26 {
   class=MaterialExpressionMin
   inputs [
    A {from=MaterialExpressionReroute_210, output=""}
    B {from=MaterialExpressionReroute_211, output=""}
   ]
   properties {}
  }
  MaterialExpressionMin_27 {
   class=MaterialExpressionMin
   inputs [
    A {from=MaterialExpressionReroute_216, output=""}
    B {from=MaterialExpressionReroute_217, output=""}
   ]
   properties {}
  }
  MaterialExpressionMultiply_0 {
   class=MaterialExpressionMultiply
   inputs [
    A {from=MaterialExpressionReroute_196}
    B {from=""}
   ]
   properties {constB=2, desc=2}
  }
  MaterialExpressionMultiply_1 {
   class=MaterialExpressionMultiply
   inputs [
    A {from=MaterialExpressionAdd_2}
    B {from=MaterialExpressionNamedRerouteUsage_15}
   ]
  }
  MaterialExpressionMultiply_22 {
   class=MaterialExpressionMultiply
   inputs [
    A {from=MaterialExpressionAdd_18}
    B {from=MaterialExpressionNamedRerouteUsage_272}
   ]
  }
  MaterialExpressionMultiply_23 {
   class=MaterialExpressionMultiply
   inputs [
    A {from=MaterialExpressionReroute_189}
    B {from=MaterialExpressionConstant_66}
   ]
   properties {desc=compensates for the less dense meshes on mobile creating bigger vertex gradients}
  }
  MaterialExpressionMultiply_24 {
   class=MaterialExpressionMultiply
   inputs [
    A {from=MaterialExpressionAdd_0}
    B {from=""}
   ]
   properties {constB=3}
  }
  MaterialExpressionMultiply_25 {
   class=MaterialExpressionMultiply
   inputs [
    A {from=MaterialExpressionAdd_19}
    B {from=MaterialExpressionNamedRerouteUsage_293}
   ]
  }
  MaterialExpressionMultiply_26 {
   class=MaterialExpressionMultiply
   inputs [
    A {from=MaterialExpressionAdd_20}
    B {from=MaterialExpressionNamedRerouteUsage_286}
   ]
  }
  MaterialExpressionMultiply_27 {
   class=MaterialExpressionMultiply
   inputs [
    A {from=MaterialExpressionAdd_21}
    B {from=MaterialExpressionNamedRerouteUsage_280}
   ]
  }
  MaterialExpressionMultiply_28 {
   class=MaterialExpressionMultiply
   inputs [
    A {from=MaterialExpressionMultiply_24}
    B {from=MaterialExpressionShadingPathSwitch_2}
   ]
  }
  MaterialExpressionMultiply_29 {
   class=MaterialExpressionMultiply
   inputs [
    A {from=MaterialExpressionAdd_22}
    B {from=MaterialExpressionNamedRerouteUsage_360}
   ]
  }
  MaterialExpressionMultiply_30 {
   class=MaterialExpressionMultiply
   inputs [
    A {from=MaterialExpressionAdd_23}
    B {from=MaterialExpressionNamedRerouteUsage_311}
   ]
   properties {constB=0.015}
  }
  MaterialExpressionMultiply_78 {
   class=MaterialExpressionMultiply
   inputs [
    A {from=MaterialExpressionReroute_210}
    B {from=""}
   ]
   properties {constB=2, desc=2}
  }
  MaterialExpressionNamedRerouteDeclaration_0 {
   class=MaterialExpressionNamedRerouteDeclaration
   inputs [None {from=MaterialExpressionFunctionInput_26}]
   properties {name=_WS Noise Masks, nodeColor=<rgba 1 0 0.0525 1>}
  }
  MaterialExpressionNamedRerouteDeclaration_1 {
   class=MaterialExpressionNamedRerouteDeclaration
   inputs [None {from=MaterialExpressionFunctionInput_8}]
   properties {name="• Height - Min", nodeColor=<rgba 0.5043 1 0 1>}
  }
  MaterialExpressionNamedRerouteDeclaration_101 {
   class=MaterialExpressionNamedRerouteDeclaration
   inputs [None {from=MaterialExpressionFunctionInput_0}]
   properties {name=TerrainMasks_QuadB, nodeColor=<rgba 1 0.0434 0.0403 1>}
  }
  MaterialExpressionNamedRerouteDeclaration_102 {
   class=MaterialExpressionNamedRerouteDeclaration
   inputs [None {from=MaterialExpressionFunctionInput_1}]
   properties {name=TerrainMasks_QuadA, nodeColor=<rgba 1 0.0434 0.0403 1>}
  }
  MaterialExpressionNamedRerouteDeclaration_103 {
   class=MaterialExpressionNamedRerouteDeclaration
   inputs [None {from=MaterialExpressionFunctionInput_2}]
   properties {name=TerrainMasks_QuadC, nodeColor=<rgba 1 0.0434 0.0403 1>}
  }
  MaterialExpressionNamedRerouteDeclaration_104 {
   class=MaterialExpressionNamedRerouteDeclaration
   inputs [None {from=MaterialExpressionFunctionInput_17}]
   properties {name=TerrainMasks_QuadD, nodeColor=<rgba 1 0.0434 0.0403 1>}
  }
  MaterialExpressionNamedRerouteDeclaration_11 {
   class=MaterialExpressionNamedRerouteDeclaration
   inputs [None {from=MaterialExpressionFunctionInput_4}]
   properties {name=_DFAO, nodeColor=<rgba 1 0 0.0525 1>}
  }
  MaterialExpressionNamedRerouteDeclaration_12 {
   class=MaterialExpressionNamedRerouteDeclaration
   inputs [None {from=MaterialExpressionFunctionInput_3}]
   properties {name=_Baked Curvature, nodeColor=<rgba 1 0 0.0525 1>}
  }
  MaterialExpressionNamedRerouteDeclaration_125 {
   class=MaterialExpressionNamedRerouteDeclaration
   inputs [None {from=MaterialExpressionReroute_190}]
   properties {name=Combined Mask, nodeColor=<rgba 0.5 0 1 1>}
  }
  MaterialExpressionNamedRerouteDeclaration_126 {
   class=MaterialExpressionNamedRerouteDeclaration
   inputs [None {from=MaterialExpressionSaturate_26}]
   properties {name=Normal Z, nodeColor=<rgba 0.5 0 1 1>}
  }
  MaterialExpressionNamedRerouteDeclaration_127 {
   class=MaterialExpressionNamedRerouteDeclaration
   inputs [None {from=MaterialExpressionVertexInterpolator_2, output=PS}]
   properties {name=Vertex Color, nodeColor=<rgba 0.5 0 1 1>}
  }
  MaterialExpressionNamedRerouteDeclaration_128 {
   class=MaterialExpressionNamedRerouteDeclaration
   inputs [None {from=MaterialExpressionSmoothStep_3}]
   properties {name=Z Gradient, nodeColor=<rgba 0.3828 0 1 1>}
  }
  MaterialExpressionNamedRerouteDeclaration_129 {
   class=MaterialExpressionNamedRerouteDeclaration
   inputs [None {from=MaterialExpressionSaturate_30}]
   properties {name=Baked AO/Curvature/Painted, nodeColor=<rgba 0.5 0 1 1>}
  }
  MaterialExpressionNamedRerouteDeclaration_13 {
   class=MaterialExpressionNamedRerouteDeclaration
   inputs [None {from=MaterialExpressionFunctionInput_7}]
   properties {name=_Painted Mask, nodeColor=<rgba 1 0 0.0525 1>}
  }
  MaterialExpressionNamedRerouteDeclaration_130 {
   class=MaterialExpressionNamedRerouteDeclaration
   inputs [None {from=MaterialExpressionSaturate_31}]
   properties {name=DFAO - Mask, nodeColor=<rgba 0.5 0 1 1>}
  }
  MaterialExpressionNamedRerouteDeclaration_131 {
   class=MaterialExpressionNamedRerouteDeclaration
   inputs [None {from=MaterialExpressionSaturate_35}]
   properties {name=WS Noise, nodeColor=<rgba 0.5 0 1 1>}
  }
  MaterialExpressionNamedRerouteDeclaration_132 {
   class=MaterialExpressionNamedRerouteDeclaration
   inputs [None {from=MaterialExpressionLinearInterpolate_31}]
   properties {name=Height, nodeColor=<rgba 0.5 0 1 1>}
  }
  MaterialExpressionNamedRerouteDeclaration_133 {
   class=MaterialExpressionNamedRerouteDeclaration
   inputs [None {from=MaterialExpressionSaturate_37}]
   properties {name=Waterline, nodeColor=<rgba 1 0.8438 0 1>}
  }
  MaterialExpressionNamedRerouteDeclaration_134 {
   class=MaterialExpressionNamedRerouteDeclaration
   inputs [None {from=MaterialExpressionFunctionInput_100}]
   properties {name=* World Z Normal - Bias, nodeColor=<rgba 0.5043 1 0 1>}
  }
  MaterialExpressionNamedRerouteDeclaration_135 {
   class=MaterialExpressionNamedRerouteDeclaration
   inputs [None {from=MaterialExpressionFunctionInput_102}]
   properties {name=--- Use WORLD Z NORMAL MASKING ---, nodeColor=<rgba 0.1979 0.0084 0 1>}
  }
  MaterialExpressionNamedRerouteDeclaration_136 {
   class=MaterialExpressionNamedRerouteDeclaration
   inputs [None {from=MaterialExpressionFunctionInput_103}]
   properties {name=* World Z Normal - Intensity, nodeColor=<rgba 0.5043 1 0 1>}
  }
  MaterialExpressionNamedRerouteDeclaration_137 {
   class=MaterialExpressionNamedRerouteDeclaration
   inputs [None {from=MaterialExpressionFunctionInput_101}]
   properties {name="• World Z Normal - Use VTX Normals", desc="T= VTX Normals only\r\nF= Combined Material Normals", nodeColor=<rgba 0.1979 0.0084 0 1>}
  }
  MaterialExpressionNamedRerouteDeclaration_138 {
   class=MaterialExpressionNamedRerouteDeclaration
   inputs [None {from=MaterialExpressionFunctionInput_98}]
   properties {name="• VTX Color - Intensity", nodeColor=<rgba 0.5043 1 0 1>}
  }
  MaterialExpressionNamedRerouteDeclaration_139 {
   class=MaterialExpressionNamedRerouteDeclaration
   inputs [None {from=MaterialExpressionFunctionInput_99}]
   properties {name=--- Use VTX COLOR MASKING ---, nodeColor=<rgba 0.1979 0.0084 0 1>}
  }
  MaterialExpressionNamedRerouteDeclaration_140 {
   class=MaterialExpressionNamedRerouteDeclaration
   inputs [None {from=MaterialExpressionFunctionInput_97}]
   properties {name="• VTX Color - Channel Weights", nodeColor=<rgba 0.5043 1 0 1>}
  }
  MaterialExpressionNamedRerouteDeclaration_141 {
   class=MaterialExpressionNamedRerouteDeclaration
   inputs [None {from=MaterialExpressionFunctionInput_105}]
   properties {name=--- Use DFAO MASKING ---, nodeColor=<rgba 0.1979 0.0084 0 1>}
  }
  MaterialExpressionNamedRerouteDeclaration_142 {
   class=MaterialExpressionNamedRerouteDeclaration
   inputs [None {from=MaterialExpressionFunctionInput_108}]
   properties {name=--- Use Z GRADIENT MASKING ---, nodeColor=<rgba 0.1979 0.0084 0 1>}
  }
  MaterialExpressionNamedRerouteDeclaration_143 {
   class=MaterialExpressionNamedRerouteDeclaration
   inputs [None {from=MaterialExpressionFunctionInput_111}]
   properties {name=--- Use BAKED AO/CURVATURE/PAINTED MASKING ---, nodeColor=<rgba 0.1979 0.0084 0 1>}
  }
  MaterialExpressionNamedRerouteDeclaration_144 {
   class=MaterialExpressionNamedRerouteDeclaration
   inputs [None {from=MaterialExpressionFunctionInput_115}]
   properties {name=--- Use WS NOISE MASKING ---, nodeColor=<rgba 0.1979 0.0084 0 1>}
  }
  MaterialExpressionNamedRerouteDeclaration_145 {
   class=MaterialExpressionNamedRerouteDeclaration
   inputs [None {from=MaterialExpressionFunctionInput_120}]
   properties {name=* Masking - Invert, nodeColor=<rgba 0.5043 1 0 1>}
  }
  MaterialExpressionNamedRerouteDeclaration_146 {
   class=MaterialExpressionNamedRerouteDeclaration
   inputs [None {from=MaterialExpressionFunctionInput_119}]
   properties {name="• Masking - Bias", nodeColor=<rgba 0.5043 1 0 1>}
  }
  MaterialExpressionNamedRerouteDeclaration_147 {
   class=MaterialExpressionNamedRerouteDeclaration
   inputs [None {from=MaterialExpressionFunctionInput_118}]
   properties {name="• Masking - Intensity", nodeColor=<rgba 0.5043 1 0 1>}
  }
  MaterialExpressionNamedRerouteDeclaration_148 {
   class=MaterialExpressionNamedRerouteDeclaration
   inputs [None {from=MaterialExpressionFunctionInput_117}]
   properties {name="• WS Noise - Intensity", nodeColor=<rgba 0.5043 1 0 1>}
  }
  MaterialExpressionNamedRerouteDeclaration_149 {
   class=MaterialExpressionNamedRerouteDeclaration
   inputs [None {from=MaterialExpressionFunctionInput_116}]
   properties {name="• WS Noise - Bias", nodeColor=<rgba 0.5043 1 0 1>}
  }
  MaterialExpressionNamedRerouteDeclaration_150 {
   class=MaterialExpressionNamedRerouteDeclaration
   inputs [None {from=MaterialExpressionFunctionInput_113}]
   properties {name="• Baked AO/Curvature/Painted - Intensity", nodeColor=<rgba 0.5043 1 0 1>}
  }
  MaterialExpressionNamedRerouteDeclaration_151 {
   class=MaterialExpressionNamedRerouteDeclaration
   inputs [None {from=MaterialExpressionFunctionInput_112}]
   properties {name="• Baked AO/Curvature/Painted - Bias", nodeColor=<rgba 0.5043 1 0 1>}
  }
  MaterialExpressionNamedRerouteDeclaration_152 {
   class=MaterialExpressionNamedRerouteDeclaration
   inputs [None {from=MaterialExpressionFunctionInput_110}]
   properties {name="• Z Gradient - Height - Top", nodeColor=<rgba 0.5043 1 0 1>}
  }
  MaterialExpressionNamedRerouteDeclaration_153 {
   class=MaterialExpressionNamedRerouteDeclaration
   inputs [None {from=MaterialExpressionFunctionInput_109}]
   properties {name="• Z Gradient - Height - Bottom", nodeColor=<rgba 0.5043 1 0 1>}
  }
  MaterialExpressionNamedRerouteDeclaration_154 {
   class=MaterialExpressionNamedRerouteDeclaration
   inputs [None {from=MaterialExpressionFunctionInput_107}]
   properties {name="• DFAO - Intensity", nodeColor=<rgba 0.5043 1 0 1>}
  }
  MaterialExpressionNamedRerouteDeclaration_155 {
   class=MaterialExpressionNamedRerouteDeclaration
   inputs [None {from=MaterialExpressionFunctionInput_106}]
   properties {name="• DFAO - Bias", nodeColor=<rgba 0.5043 1 0 1>}
  }
  MaterialExpressionNamedRerouteDeclaration_156 {
   class=MaterialExpressionNamedRerouteDeclaration
   inputs [None {from=MaterialExpressionFunctionInput_104}]
   properties {name=* DFAO - Force Baked AO on Nanite, nodeColor=<rgba 0.1979 0.0084 0 1>}
  }
  MaterialExpressionNamedRerouteDeclaration_157 {
   class=MaterialExpressionNamedRerouteDeclaration
   inputs [None {from=MaterialExpressionFunctionInput_114}]
   properties {name="• Baked AO/Curvature/Painted - Channel Weights", nodeColor=<rgba 0.5043 1 0 1>}
  }
  MaterialExpressionNamedRerouteDeclaration_158 {
   class=MaterialExpressionNamedRerouteDeclaration
   inputs [None {from=MaterialExpressionFunctionInput_121}]
   properties {name="• VTX Color - Blend (1=Max 0=Min)", nodeColor=<rgba 0.5043 1 0 1>}
  }
  MaterialExpressionNamedRerouteDeclaration_159 {
   class=MaterialExpressionNamedRerouteDeclaration
   inputs [None {from=MaterialExpressionFunctionInput_122}]
   properties {name="• DFAO - Blend (1=Max 0=Min)", nodeColor=<rgba 0.5043 1 0 1>}
  }
  MaterialExpressionNamedRerouteDeclaration_160 {
   class=MaterialExpressionNamedRerouteDeclaration
   inputs [None {from=MaterialExpressionFunctionInput_123}]
   properties {name="• Z Gradient - Blend (1=Max 0=Min)", nodeColor=<rgba 0.5043 1 0 1>}
  }
  MaterialExpressionNamedRerouteDeclaration_161 {
   class=MaterialExpressionNamedRerouteDeclaration
   inputs [None {from=MaterialExpressionFunctionInput_124}]
   properties {name="• Baked AO/Curvature/Painted - Blend (1=Max 0=Min)", nodeColor=<rgba 0.5043 1 0 1>}
  }
  MaterialExpressionNamedRerouteDeclaration_162 {
   class=MaterialExpressionNamedRerouteDeclaration
   inputs [None {from=MaterialExpressionFunctionInput_125}]
   properties {name="• WS Noise - Blend (1=Blend 0=Min)", nodeColor=<rgba 0.5043 1 0 1>}
  }
  MaterialExpressionNamedRerouteDeclaration_163 {
   class=MaterialExpressionNamedRerouteDeclaration
   inputs [None {from=MaterialExpressionFunctionInput_126}]
   properties {name="• WS Noise - Channel Weights", nodeColor=<rgba 0.5043 1 0 1>}
  }
  MaterialExpressionNamedRerouteDeclaration_164 {
   class=MaterialExpressionNamedRerouteDeclaration
   inputs [None {from=MaterialExpressionStaticBoolParameter_28}]
   properties {name="Normals - Add Normals(SB)", nodeColor=<rgba 0.1979 0.0084 0 1>}
  }
  MaterialExpressionNamedRerouteDeclaration_165 {
   class=MaterialExpressionNamedRerouteDeclaration
   inputs [None {from=MaterialExpressionFunctionInput_127}]
   properties {name=--- Use HEIGHT MASKING ---, nodeColor=<rgba 0.1979 0.0084 0 1>}
  }
  MaterialExpressionNamedRerouteDeclaration_166 {
   class=MaterialExpressionNamedRerouteDeclaration
   inputs [None {from=MaterialExpressionFunctionInput_129}]
   properties {name="• Height - Intensity", nodeColor=<rgba 0.5043 1 0 1>}
  }
  MaterialExpressionNamedRerouteDeclaration_167 {
   class=MaterialExpressionNamedRerouteDeclaration
   inputs [None {from=MaterialExpressionFunctionInput_128}]
   properties {name="• Height - Bias", nodeColor=<rgba 0.5043 1 0 1>}
  }
  MaterialExpressionNamedRerouteDeclaration_168 {
   class=MaterialExpressionNamedRerouteDeclaration
   inputs [None {from=MaterialExpressionFunctionInput_130}]
   properties {name="• Height - Blend (1=Blend 0=Min)", nodeColor=<rgba 0.5043 1 0 1>}
  }
  MaterialExpressionNamedRerouteDeclaration_169 {
   class=MaterialExpressionNamedRerouteDeclaration
   inputs [None {from=MaterialExpressionFunctionInput_131}]
   properties {name="• Height - Selection (1=Current Layer 0=Previous Layers)", nodeColor=<rgba 0.5043 1 0 1>}
  }
  MaterialExpressionNamedRerouteDeclaration_170 {
   class=MaterialExpressionNamedRerouteDeclaration
   inputs [None {from=MaterialExpressionFunctionInput_132}]
   properties {name=--- Use WATERLINE MASKING ---, nodeColor=<rgba 0.1979 0.0084 0 1>}
  }
  MaterialExpressionNamedRerouteDeclaration_171 {
   class=MaterialExpressionNamedRerouteDeclaration
   inputs [None {from=MaterialExpressionFunctionInput_133}]
   properties {name=* Waterline - Bias, nodeColor=<rgba 0.5043 1 0 1>}
  }
  MaterialExpressionNamedRerouteDeclaration_172 {
   class=MaterialExpressionNamedRerouteDeclaration
   inputs [None {from=MaterialExpressionFunctionInput_134}]
   properties {name="* Waterline - Blend (1=Max 0=Min)", nodeColor=<rgba 0.5043 1 0 1>}
  }
  MaterialExpressionNamedRerouteDeclaration_173 {
   class=MaterialExpressionNamedRerouteDeclaration
   inputs [None {from=MaterialExpressionFunctionInput_135}]
   properties {name=* Waterline - Intensity, nodeColor=<rgba 0.5043 1 0 1>}
  }
  MaterialExpressionNamedRerouteDeclaration_174 {
   class=MaterialExpressionNamedRerouteDeclaration
   inputs [None {from=MaterialExpressionFunctionInput_136}]
   properties {name="• Baked AO/Curvature/Painted - Invert", nodeColor=<rgba 0.5043 1 0 1>}
  }
  MaterialExpressionNamedRerouteDeclaration_175 {
   class=MaterialExpressionNamedRerouteDeclaration
   inputs [None {from=MaterialExpressionFunctionInput_137}]
   properties {name="• WS Noise - Invert", nodeColor=<rgba 0.5043 1 0 1>}
  }
  MaterialExpressionNamedRerouteDeclaration_176 {
   class=MaterialExpressionNamedRerouteDeclaration
   inputs [None {from=MaterialExpressionFunctionInput_138}]
   properties {name="• VTX Color - Invert", nodeColor=<rgba 0.5043 1 0 1>}
  }
  MaterialExpressionNamedRerouteDeclaration_18 {
   class=MaterialExpressionNamedRerouteDeclaration
   inputs [None {from=MaterialExpressionFunctionInput_24}]
   properties {name=Landscape Material, nodeColor=<rgba 1 0.0434 0.0403 1>}
  }
  MaterialExpressionNamedRerouteDeclaration_19 {
   class=MaterialExpressionNamedRerouteDeclaration
   inputs [None {from=MaterialExpressionFunctionInput_25}]
   properties {name=Landscape Height, nodeColor=<rgba 1 0.0434 0.0403 1>}
  }
  MaterialExpressionNamedRerouteDeclaration_2 {
   class=MaterialExpressionNamedRerouteDeclaration
   inputs [None {from=MaterialExpressionFunctionInput_9}]
   properties {name="• Height - Max", nodeColor=<rgba 0.5043 1 0 1>}
  }
  MaterialExpressionNamedRerouteDeclaration_23 {
   class=MaterialExpressionNamedRerouteDeclaration
   inputs [None {from=MaterialExpressionConvert_1}]
   properties {name=Baked AO/Curvature/Painted Texture, nodeColor=<rgba 1 1 1 1>}
  }
  MaterialExpressionNamedRerouteDeclaration_3 {
   class=MaterialExpressionNamedRerouteDeclaration
   inputs [None {from=MaterialExpressionFunctionInput_10}]
   properties {name="• VTX Color - Bias", nodeColor=<rgba 0.5043 1 0 1>}
  }
  MaterialExpressionNamedRerouteDeclaration_4 {
   class=MaterialExpressionNamedRerouteDeclaration
   inputs [None {from=MaterialExpressionFunctionInput_13}]
   properties {name="• Masking - Bias (Low/Mobile)", nodeColor=<rgba 0.5043 1 0 1>}
  }
  MaterialExpressionNamedRerouteDeclaration_5 {
   class=MaterialExpressionNamedRerouteDeclaration
   inputs [None {from=MaterialExpressionFunctionInput_12}]
   properties {name="• Masking - Intensity (Low/Mobile)", nodeColor=<rgba 0.5043 1 0 1>}
  }
  MaterialExpressionNamedRerouteDeclaration_6 {
   class=MaterialExpressionNamedRerouteDeclaration
   inputs [None {from=MaterialExpressionSaturate_0}]
   properties {name="Combined Mask (Low/Mobile)", nodeColor=<rgba 0.5 0 1 1>}
  }
  MaterialExpressionNamedRerouteDeclaration_7 {
   class=MaterialExpressionNamedRerouteDeclaration
   inputs [None {from=MaterialExpressionFunctionInput_5}]
   properties {name=_Z Gradient, nodeColor=<rgba 1 0 0.0525 1>}
  }
  MaterialExpressionNamedRerouteDeclaration_8 {
   class=MaterialExpressionNamedRerouteDeclaration
   inputs [None {from=MaterialExpressionFunctionInput_6}]
   properties {name=_Baked AO, nodeColor=<rgba 1 0 0.0525 1>}
  }
  MaterialExpressionNamedRerouteDeclaration_9 {
   class=MaterialExpressionNamedRerouteDeclaration
   inputs [None {from=MaterialExpressionFunctionInput_11}]
   properties {name=_Input A MaterialAttributes, desc=MaterialAttributes from previous layers, nodeColor=<rgba 0.1016 0 1 1>}
  }
  MaterialExpressionNamedRerouteUsage_1 {class=MaterialExpressionNamedRerouteUsage}
  MaterialExpressionNamedRerouteUsage_10 {class=MaterialExpressionNamedRerouteUsage}
  MaterialExpressionNamedRerouteUsage_11 {class=MaterialExpressionNamedRerouteUsage}
  MaterialExpressionNamedRerouteUsage_12 {class=MaterialExpressionNamedRerouteUsage}
  MaterialExpressionNamedRerouteUsage_13 {class=MaterialExpressionNamedRerouteUsage}
  MaterialExpressionNamedRerouteUsage_14 {class=MaterialExpressionNamedRerouteUsage}
  MaterialExpressionNamedRerouteUsage_15 {class=MaterialExpressionNamedRerouteUsage}
  MaterialExpressionNamedRerouteUsage_16 {class=MaterialExpressionNamedRerouteUsage}
  MaterialExpressionNamedRerouteUsage_160 {class=MaterialExpressionNamedRerouteUsage}
  MaterialExpressionNamedRerouteUsage_162 {class=MaterialExpressionNamedRerouteUsage}
  MaterialExpressionNamedRerouteUsage_18 {class=MaterialExpressionNamedRerouteUsage}
  MaterialExpressionNamedRerouteUsage_2 {class=MaterialExpressionNamedRerouteUsage}
  MaterialExpressionNamedRerouteUsage_261 {class=MaterialExpressionNamedRerouteUsage}
  MaterialExpressionNamedRerouteUsage_262 {class=MaterialExpressionNamedRerouteUsage}
  MaterialExpressionNamedRerouteUsage_263 {class=MaterialExpressionNamedRerouteUsage}
  MaterialExpressionNamedRerouteUsage_264 {class=MaterialExpressionNamedRerouteUsage}
  MaterialExpressionNamedRerouteUsage_265 {class=MaterialExpressionNamedRerouteUsage}
  MaterialExpressionNamedRerouteUsage_266 {class=MaterialExpressionNamedRerouteUsage}
  MaterialExpressionNamedRerouteUsage_267 {class=MaterialExpressionNamedRerouteUsage}
  MaterialExpressionNamedRerouteUsage_268 {class=MaterialExpressionNamedRerouteUsage}
  MaterialExpressionNamedRerouteUsage_269 {class=MaterialExpressionNamedRerouteUsage}
  MaterialExpressionNamedRerouteUsage_270 {class=MaterialExpressionNamedRerouteUsage}
  MaterialExpressionNamedRerouteUsage_271 {class=MaterialExpressionNamedRerouteUsage}
  MaterialExpressionNamedRerouteUsage_272 {class=MaterialExpressionNamedRerouteUsage}
  MaterialExpressionNamedRerouteUsage_273 {class=MaterialExpressionNamedRerouteUsage}
  MaterialExpressionNamedRerouteUsage_274 {class=MaterialExpressionNamedRerouteUsage}
  MaterialExpressionNamedRerouteUsage_275 {class=MaterialExpressionNamedRerouteUsage, properties {desc=Modulates VTX color transition intensity}}
  MaterialExpressionNamedRerouteUsage_276 {class=MaterialExpressionNamedRerouteUsage}
  MaterialExpressionNamedRerouteUsage_277 {class=MaterialExpressionNamedRerouteUsage}
  MaterialExpressionNamedRerouteUsage_278 {class=MaterialExpressionNamedRerouteUsage}
  MaterialExpressionNamedRerouteUsage_279 {class=MaterialExpressionNamedRerouteUsage}
  MaterialExpressionNamedRerouteUsage_280 {class=MaterialExpressionNamedRerouteUsage}
  MaterialExpressionNamedRerouteUsage_281 {class=MaterialExpressionNamedRerouteUsage}
  MaterialExpressionNamedRerouteUsage_282 {class=MaterialExpressionNamedRerouteUsage}
  MaterialExpressionNamedRerouteUsage_283 {class=MaterialExpressionNamedRerouteUsage}
  MaterialExpressionNamedRerouteUsage_284 {class=MaterialExpressionNamedRerouteUsage}
  MaterialExpressionNamedRerouteUsage_285 {class=MaterialExpressionNamedRerouteUsage}
  MaterialExpressionNamedRerouteUsage_286 {class=MaterialExpressionNamedRerouteUsage}
  MaterialExpressionNamedRerouteUsage_287 {class=MaterialExpressionNamedRerouteUsage}
  MaterialExpressionNamedRerouteUsage_288 {class=MaterialExpressionNamedRerouteUsage}
  MaterialExpressionNamedRerouteUsage_289 {class=MaterialExpressionNamedRerouteUsage}
  MaterialExpressionNamedRerouteUsage_290 {class=MaterialExpressionNamedRerouteUsage}
  MaterialExpressionNamedRerouteUsage_291 {class=MaterialExpressionNamedRerouteUsage}
  MaterialExpressionNamedRerouteUsage_292 {class=MaterialExpressionNamedRerouteUsage}
  MaterialExpressionNamedRerouteUsage_293 {class=MaterialExpressionNamedRerouteUsage}
  MaterialExpressionNamedRerouteUsage_294 {class=MaterialExpressionNamedRerouteUsage}
  MaterialExpressionNamedRerouteUsage_296 {class=MaterialExpressionNamedRerouteUsage}
  MaterialExpressionNamedRerouteUsage_297 {class=MaterialExpressionNamedRerouteUsage}
  MaterialExpressionNamedRerouteUsage_298 {class=MaterialExpressionNamedRerouteUsage}
  MaterialExpressionNamedRerouteUsage_299 {class=MaterialExpressionNamedRerouteUsage}
  MaterialExpressionNamedRerouteUsage_3 {class=MaterialExpressionNamedRerouteUsage}
  MaterialExpressionNamedRerouteUsage_300 {class=MaterialExpressionNamedRerouteUsage}
  MaterialExpressionNamedRerouteUsage_301 {class=MaterialExpressionNamedRerouteUsage}
  MaterialExpressionNamedRerouteUsage_303 {class=MaterialExpressionNamedRerouteUsage}
  MaterialExpressionNamedRerouteUsage_304 {class=MaterialExpressionNamedRerouteUsage}
  MaterialExpressionNamedRerouteUsage_305 {class=MaterialExpressionNamedRerouteUsage}
  MaterialExpressionNamedRerouteUsage_306 {class=MaterialExpressionNamedRerouteUsage}
  MaterialExpressionNamedRerouteUsage_307 {class=MaterialExpressionNamedRerouteUsage}
  MaterialExpressionNamedRerouteUsage_308 {class=MaterialExpressionNamedRerouteUsage}
  MaterialExpressionNamedRerouteUsage_309 {class=MaterialExpressionNamedRerouteUsage}
  MaterialExpressionNamedRerouteUsage_310 {class=MaterialExpressionNamedRerouteUsage}
  MaterialExpressionNamedRerouteUsage_311 {class=MaterialExpressionNamedRerouteUsage}
  MaterialExpressionNamedRerouteUsage_312 {class=MaterialExpressionNamedRerouteUsage}
  MaterialExpressionNamedRerouteUsage_313 {class=MaterialExpressionNamedRerouteUsage}
  MaterialExpressionNamedRerouteUsage_314 {class=MaterialExpressionNamedRerouteUsage}
  MaterialExpressionNamedRerouteUsage_315 {class=MaterialExpressionNamedRerouteUsage}
  MaterialExpressionNamedRerouteUsage_316 {class=MaterialExpressionNamedRerouteUsage}
  MaterialExpressionNamedRerouteUsage_317 {class=MaterialExpressionNamedRerouteUsage}
  MaterialExpressionNamedRerouteUsage_318 {class=MaterialExpressionNamedRerouteUsage}
  MaterialExpressionNamedRerouteUsage_319 {class=MaterialExpressionNamedRerouteUsage}
  MaterialExpressionNamedRerouteUsage_320 {class=MaterialExpressionNamedRerouteUsage}
  MaterialExpressionNamedRerouteUsage_321 {class=MaterialExpressionNamedRerouteUsage}
  MaterialExpressionNamedRerouteUsage_322 {class=MaterialExpressionNamedRerouteUsage}
  MaterialExpressionNamedRerouteUsage_323 {class=MaterialExpressionNamedRerouteUsage}
  MaterialExpressionNamedRerouteUsage_324 {class=MaterialExpressionNamedRerouteUsage}
  MaterialExpressionNamedRerouteUsage_325 {class=MaterialExpressionNamedRerouteUsage}
  MaterialExpressionNamedRerouteUsage_326 {class=MaterialExpressionNamedRerouteUsage}
  MaterialExpressionNamedRerouteUsage_327 {class=MaterialExpressionNamedRerouteUsage}
  MaterialExpressionNamedRerouteUsage_328 {class=MaterialExpressionNamedRerouteUsage}
  MaterialExpressionNamedRerouteUsage_329 {class=MaterialExpressionNamedRerouteUsage}
  MaterialExpressionNamedRerouteUsage_330 {class=MaterialExpressionNamedRerouteUsage}
  MaterialExpressionNamedRerouteUsage_331 {class=MaterialExpressionNamedRerouteUsage}
  MaterialExpressionNamedRerouteUsage_332 {class=MaterialExpressionNamedRerouteUsage}
  MaterialExpressionNamedRerouteUsage_333 {class=MaterialExpressionNamedRerouteUsage}
  MaterialExpressionNamedRerouteUsage_334 {class=MaterialExpressionNamedRerouteUsage}
  MaterialExpressionNamedRerouteUsage_335 {class=MaterialExpressionNamedRerouteUsage}
  MaterialExpressionNamedRerouteUsage_336 {class=MaterialExpressionNamedRerouteUsage}
  MaterialExpressionNamedRerouteUsage_337 {class=MaterialExpressionNamedRerouteUsage}
  MaterialExpressionNamedRerouteUsage_338 {class=MaterialExpressionNamedRerouteUsage}
  MaterialExpressionNamedRerouteUsage_339 {class=MaterialExpressionNamedRerouteUsage}
  MaterialExpressionNamedRerouteUsage_340 {class=MaterialExpressionNamedRerouteUsage}
  MaterialExpressionNamedRerouteUsage_341 {class=MaterialExpressionNamedRerouteUsage}
  MaterialExpressionNamedRerouteUsage_342 {class=MaterialExpressionNamedRerouteUsage}
  MaterialExpressionNamedRerouteUsage_343 {class=MaterialExpressionNamedRerouteUsage}
  MaterialExpressionNamedRerouteUsage_344 {class=MaterialExpressionNamedRerouteUsage}
  MaterialExpressionNamedRerouteUsage_345 {class=MaterialExpressionNamedRerouteUsage}
  MaterialExpressionNamedRerouteUsage_346 {class=MaterialExpressionNamedRerouteUsage}
  MaterialExpressionNamedRerouteUsage_347 {class=MaterialExpressionNamedRerouteUsage}
  MaterialExpressionNamedRerouteUsage_348 {class=MaterialExpressionNamedRerouteUsage}
  MaterialExpressionNamedRerouteUsage_349 {class=MaterialExpressionNamedRerouteUsage}
  MaterialExpressionNamedRerouteUsage_350 {class=MaterialExpressionNamedRerouteUsage}
  MaterialExpressionNamedRerouteUsage_351 {class=MaterialExpressionNamedRerouteUsage}
  MaterialExpressionNamedRerouteUsage_352 {class=MaterialExpressionNamedRerouteUsage}
  MaterialExpressionNamedRerouteUsage_353 {class=MaterialExpressionNamedRerouteUsage}
  MaterialExpressionNamedRerouteUsage_354 {class=MaterialExpressionNamedRerouteUsage}
  MaterialExpressionNamedRerouteUsage_355 {class=MaterialExpressionNamedRerouteUsage}
  MaterialExpressionNamedRerouteUsage_356 {class=MaterialExpressionNamedRerouteUsage}
  MaterialExpressionNamedRerouteUsage_357 {class=MaterialExpressionNamedRerouteUsage}
  MaterialExpressionNamedRerouteUsage_358 {class=MaterialExpressionNamedRerouteUsage}
  MaterialExpressionNamedRerouteUsage_359 {class=MaterialExpressionNamedRerouteUsage}
  MaterialExpressionNamedRerouteUsage_360 {class=MaterialExpressionNamedRerouteUsage}
  MaterialExpressionNamedRerouteUsage_4 {class=MaterialExpressionNamedRerouteUsage}
  MaterialExpressionNamedRerouteUsage_5 {class=MaterialExpressionNamedRerouteUsage}
  MaterialExpressionNamedRerouteUsage_6 {class=MaterialExpressionNamedRerouteUsage}
  MaterialExpressionNamedRerouteUsage_7 {class=MaterialExpressionNamedRerouteUsage}
  MaterialExpressionNamedRerouteUsage_8 {class=MaterialExpressionNamedRerouteUsage}
  MaterialExpressionOneMinus_12 {class=MaterialExpressionOneMinus, inputs [None {from=MaterialExpressionReroute_183}]}
  MaterialExpressionOneMinus_13 {
   class=MaterialExpressionOneMinus
   inputs [None {from=MaterialExpressionWorldPosition_3, output=Z}]
   properties {desc=flip so positive values raise the waterline}
  }
  MaterialExpressionOneMinus_14 {class=MaterialExpressionOneMinus, inputs [None {from=MaterialExpressionDotProduct_9}]}
  MaterialExpressionOneMinus_15 {class=MaterialExpressionOneMinus, inputs [None {from=MaterialExpressionDotProduct_10}]}
  MaterialExpressionOneMinus_16 {class=MaterialExpressionOneMinus, inputs [None {from=MaterialExpressionDotProduct_11}]}
  MaterialExpressionReroute_1 {class=MaterialExpressionReroute, inputs [None {from=MaterialExpressionNamedRerouteUsage_8}]}
  MaterialExpressionReroute_181 {class=MaterialExpressionReroute, inputs [None {from=MaterialExpressionReroute_185}]}
  MaterialExpressionReroute_182 {class=MaterialExpressionReroute, inputs [None {from=MaterialExpressionReroute_187}]}
  MaterialExpressionReroute_183 {class=MaterialExpressionReroute, inputs [None {from=MaterialExpressionReroute_212}]}
  MaterialExpressionReroute_185 {class=MaterialExpressionReroute, inputs [None {from=MaterialExpressionReroute_201}]}
  MaterialExpressionReroute_186 {class=MaterialExpressionReroute, inputs [None {from=MaterialExpressionNamedRerouteUsage_262}]}
  MaterialExpressionReroute_187 {class=MaterialExpressionReroute, inputs [None {from=MaterialExpressionReroute_208}]}
  MaterialExpressionReroute_188 {class=MaterialExpressionReroute, inputs [None {from=MaterialExpressionNamedRerouteUsage_263}]}
  MaterialExpressionReroute_189 {class=MaterialExpressionReroute, inputs [None {from=MaterialExpressionNamedRerouteUsage_275}]}
  MaterialExpressionReroute_190 {class=MaterialExpressionReroute, inputs [None {from=MaterialExpressionSaturate_28}]}
  MaterialExpressionReroute_191 {class=MaterialExpressionReroute, inputs [None {from=MaterialExpressionReroute_192}]}
  MaterialExpressionReroute_192 {class=MaterialExpressionReroute, inputs [None {from=MaterialExpressionReroute_194}]}
  MaterialExpressionReroute_193 {class=MaterialExpressionReroute, inputs [None {from=MaterialExpressionNamedRerouteUsage_264}]}
  MaterialExpressionReroute_194 {class=MaterialExpressionReroute, inputs [None {from=MaterialExpressionReroute_204}]}
  MaterialExpressionReroute_195 {class=MaterialExpressionReroute, inputs [None {from=MaterialExpressionReroute_196}]}
  MaterialExpressionReroute_196 {class=MaterialExpressionReroute, inputs [None {from=MaterialExpressionReroute_218}]}
  MaterialExpressionReroute_197 {class=MaterialExpressionReroute, inputs [None {from=MaterialExpressionNamedRerouteUsage_296}]}
  MaterialExpressionReroute_198 {class=MaterialExpressionReroute, inputs [None {from=MaterialExpressionReroute_199}]}
  MaterialExpressionReroute_199 {class=MaterialExpressionReroute, inputs [None {from=MaterialExpressionReroute_202}]}
  MaterialExpressionReroute_2 {class=MaterialExpressionReroute, inputs [None {from=MaterialExpressionNamedRerouteUsage_7}]}
  MaterialExpressionReroute_200 {class=MaterialExpressionReroute, inputs [None {from=MaterialExpressionNamedRerouteUsage_267}]}
  MaterialExpressionReroute_201 {class=MaterialExpressionReroute, inputs [None {from=MaterialExpressionStaticSwitch_33}]}
  MaterialExpressionReroute_202 {class=MaterialExpressionReroute, inputs [None {from=MaterialExpressionStaticSwitch_34}]}
  MaterialExpressionReroute_203 {class=MaterialExpressionReroute, inputs [None {from=MaterialExpressionStaticSwitch_35}]}
  MaterialExpressionReroute_204 {class=MaterialExpressionReroute, inputs [None {from=MaterialExpressionStaticSwitch_36}]}
  MaterialExpressionReroute_205 {class=MaterialExpressionReroute, inputs [None {from=MaterialExpressionStaticSwitch_37}]}
  MaterialExpressionReroute_206 {class=MaterialExpressionReroute, inputs [None {from=MaterialExpressionStaticSwitch_38}]}
  MaterialExpressionReroute_207 {class=MaterialExpressionReroute, inputs [None {from=MaterialExpressionReroute_203}]}
  MaterialExpressionReroute_208 {class=MaterialExpressionReroute, inputs [None {from=MaterialExpressionReroute_207}]}
  MaterialExpressionReroute_209 {class=MaterialExpressionReroute, inputs [None {from=MaterialExpressionReroute_210}]}
  MaterialExpressionReroute_210 {class=MaterialExpressionReroute, inputs [None {from=MaterialExpressionReroute_206}]}
  MaterialExpressionReroute_211 {class=MaterialExpressionReroute, inputs [None {from=MaterialExpressionNamedRerouteUsage_301}]}
  MaterialExpressionReroute_212 {class=MaterialExpressionReroute, inputs [None {from=MaterialExpressionStaticSwitch_42}]}
  MaterialExpressionReroute_213 {class=MaterialExpressionReroute, inputs [None {from=MaterialExpressionNamedRerouteUsage_160}], properties {desc=Current Layer Height}}
  MaterialExpressionReroute_214 {
   class=MaterialExpressionReroute
   inputs [None {from=MaterialExpressionGetMaterialAttributes_9, output=Opacity}]
   properties {desc=Previous Layer Height}
  }
  MaterialExpressionReroute_215 {class=MaterialExpressionReroute, inputs [None {from=MaterialExpressionReroute_216}]}
  MaterialExpressionReroute_216 {class=MaterialExpressionReroute, inputs [None {from=MaterialExpressionReroute_205}]}
  MaterialExpressionReroute_217 {class=MaterialExpressionReroute, inputs [None {from=MaterialExpressionNamedRerouteUsage_310}]}
  MaterialExpressionReroute_218 {class=MaterialExpressionReroute, inputs [None {from=MaterialExpressionStaticSwitch_43}]}
  MaterialExpressionReroute_219 {class=MaterialExpressionReroute, inputs [None {from=MaterialExpressionNamedRerouteUsage_345}]}
  MaterialExpressionReroute_220 {class=MaterialExpressionReroute, inputs [None {from=MaterialExpressionNamedRerouteUsage_344}]}
  MaterialExpressionReroute_221 {class=MaterialExpressionReroute, inputs [None {from=MaterialExpressionNamedRerouteUsage_342}]}
  MaterialExpressionReroute_222 {class=MaterialExpressionReroute, inputs [None {from=MaterialExpressionNamedRerouteUsage_343}]}
  MaterialExpressionReroute_223 {class=MaterialExpressionReroute, inputs [None {from=MaterialExpressionNamedRerouteUsage_356}]}
  MaterialExpressionReroute_224 {class=MaterialExpressionReroute, inputs [None {from=MaterialExpressionNamedRerouteUsage_341}]}
  MaterialExpressionReroute_225 {class=MaterialExpressionReroute, inputs [None {from=MaterialExpressionNamedRerouteUsage_339}]}
  MaterialExpressionReroute_226 {class=MaterialExpressionReroute, inputs [None {from=MaterialExpressionNamedRerouteUsage_340}]}
  MaterialExpressionReroute_227 {class=MaterialExpressionReroute, inputs [None {from=MaterialExpressionNamedRerouteUsage_337}]}
  MaterialExpressionReroute_232 {class=MaterialExpressionReroute, inputs [None {from=MaterialExpressionNamedRerouteUsage_336}]}
  MaterialExpressionReroute_233 {class=MaterialExpressionReroute, inputs [None {from=MaterialExpressionNamedRerouteUsage_338}]}
  MaterialExpressionReroute_234 {class=MaterialExpressionReroute, inputs [None {from=MaterialExpressionNamedRerouteUsage_334}]}
  MaterialExpressionReroute_236 {class=MaterialExpressionReroute, inputs [None {from=MaterialExpressionNamedRerouteUsage_335}]}
  MaterialExpressionReroute_238 {class=MaterialExpressionReroute, inputs [None {from=MaterialExpressionNamedRerouteUsage_332}]}
  MaterialExpressionReroute_240 {class=MaterialExpressionReroute, inputs [None {from=MaterialExpressionNamedRerouteUsage_333}]}
  MaterialExpressionReroute_243 {class=MaterialExpressionReroute, inputs [None {from=MaterialExpressionNamedRerouteUsage_330}]}
  MaterialExpressionReroute_244 {class=MaterialExpressionReroute, inputs [None {from=MaterialExpressionNamedRerouteUsage_331}]}
  MaterialExpressionReroute_245 {class=MaterialExpressionReroute, inputs [None {from=MaterialExpressionNamedRerouteUsage_328}]}
  MaterialExpressionReroute_246 {class=MaterialExpressionReroute, inputs [None {from=MaterialExpressionNamedRerouteUsage_327}]}
  MaterialExpressionReroute_247 {class=MaterialExpressionReroute, inputs [None {from=MaterialExpressionNamedRerouteUsage_329}]}
  MaterialExpressionReroute_248 {class=MaterialExpressionReroute, inputs [None {from=MaterialExpressionNamedRerouteUsage_325}]}
  MaterialExpressionReroute_251 {class=MaterialExpressionReroute, inputs [None {from=MaterialExpressionNamedRerouteUsage_326}]}
  MaterialExpressionReroute_252 {class=MaterialExpressionReroute, inputs [None {from=MaterialExpressionNamedRerouteUsage_321}]}
  MaterialExpressionReroute_253 {class=MaterialExpressionReroute, inputs [None {from=MaterialExpressionNamedRerouteUsage_324}]}
  MaterialExpressionReroute_254 {class=MaterialExpressionReroute, inputs [None {from=MaterialExpressionNamedRerouteUsage_322}]}
  MaterialExpressionReroute_255 {class=MaterialExpressionReroute, inputs [None {from=MaterialExpressionNamedRerouteUsage_323}]}
  MaterialExpressionReroute_256 {class=MaterialExpressionReroute, inputs [None {from=MaterialExpressionNamedRerouteUsage_346}]}
  MaterialExpressionReroute_257 {class=MaterialExpressionReroute, inputs [None {from=MaterialExpressionNamedRerouteUsage_320}]}
  MaterialExpressionReroute_258 {class=MaterialExpressionReroute, inputs [None {from=MaterialExpressionNamedRerouteUsage_319}]}
  MaterialExpressionReroute_259 {class=MaterialExpressionReroute, inputs [None {from=MaterialExpressionNamedRerouteUsage_318}]}
  MaterialExpressionReroute_260 {class=MaterialExpressionReroute, inputs [None {from=MaterialExpressionNamedRerouteUsage_350}]}
  MaterialExpressionReroute_263 {class=MaterialExpressionReroute, inputs [None {from=MaterialExpressionNamedRerouteUsage_347}]}
  MaterialExpressionReroute_264 {class=MaterialExpressionReroute, inputs [None {from=MaterialExpressionNamedRerouteUsage_349}]}
  MaterialExpressionReroute_265 {class=MaterialExpressionReroute, inputs [None {from=MaterialExpressionNamedRerouteUsage_348}]}
  MaterialExpressionReroute_266 {class=MaterialExpressionReroute, inputs [None {from=MaterialExpressionNamedRerouteUsage_351}]}
  MaterialExpressionReroute_267 {class=MaterialExpressionReroute, inputs [None {from=MaterialExpressionNamedRerouteUsage_353}]}
  MaterialExpressionReroute_268 {class=MaterialExpressionReroute, inputs [None {from=MaterialExpressionNamedRerouteUsage_354}]}
  MaterialExpressionReroute_269 {class=MaterialExpressionReroute, inputs [None {from=MaterialExpressionNamedRerouteUsage_355}]}
  MaterialExpressionReroute_270 {class=MaterialExpressionReroute, inputs [None {from=MaterialExpressionNamedRerouteUsage_352}]}
  MaterialExpressionReroute_271 {class=MaterialExpressionReroute, inputs [None {from=MaterialExpressionNamedRerouteUsage_357}]}
  MaterialExpressionReroute_272 {class=MaterialExpressionReroute, inputs [None {from=MaterialExpressionNamedRerouteUsage_358}]}
  MaterialExpressionReroute_273 {class=MaterialExpressionReroute, inputs [None {from=MaterialExpressionNamedRerouteUsage_359}]}
  MaterialExpressionReroute_3 {class=MaterialExpressionReroute, inputs [None {from=MaterialExpressionNamedRerouteUsage_10}]}
  MaterialExpressionReroute_52 {class=MaterialExpressionReroute, inputs [None {from=""}]}
  MaterialExpressionSaturate_0 {class=MaterialExpressionSaturate, inputs [None {from=MaterialExpressionMultiply_1}]}
  MaterialExpressionSaturate_25 {class=MaterialExpressionSaturate, inputs [None {from=MaterialExpressionComponentMask_4}]}
  MaterialExpressionSaturate_26 {class=MaterialExpressionSaturate, inputs [None {from=MaterialExpressionMultiply_22}]}
  MaterialExpressionSaturate_27 {class=MaterialExpressionSaturate, inputs [None {from=MaterialExpressionMultiply_28}]}
  MaterialExpressionSaturate_28 {class=MaterialExpressionSaturate, inputs [None {from=MaterialExpressionShadingPathSwitch_0}]}
  MaterialExpressionSaturate_29 {class=MaterialExpressionSaturate, inputs [None {from=MaterialExpressionComponentMask_5}]}
  MaterialExpressionSaturate_30 {class=MaterialExpressionSaturate, inputs [None {from=MaterialExpressionMultiply_25}]}
  MaterialExpressionSaturate_31 {class=MaterialExpressionSaturate, inputs [None {from=MaterialExpressionMultiply_26}]}
  MaterialExpressionSaturate_35 {class=MaterialExpressionSaturate, inputs [None {from=MaterialExpressionMultiply_29}]}
  MaterialExpressionSaturate_37 {class=MaterialExpressionSaturate, inputs [None {from=MaterialExpressionMultiply_30}]}
  MaterialExpressionScalarParameter_0 {class=MaterialExpressionScalarParameter, properties {group=Mask, parameterName=* Masking - Invert, sortPriority=50}}
  MaterialExpressionScalarParameter_1 {class=MaterialExpressionScalarParameter, properties {group=Mask, parameterName="• Masking - Bias (Low/Mobile)", sortPriority=60}}
  MaterialExpressionScalarParameter_10 {class=MaterialExpressionScalarParameter, properties {defaultValue=1, group=Mask, parameterName=* Waterline - Intensity, sortPriority=416}}
  MaterialExpressionScalarParameter_2 {
   class=MaterialExpressionScalarParameter
   properties {defaultValue=1, group=Mask, parameterName="• Masking - Intensity (Low/Mobile)", sortPriority=61}
  }
  MaterialExpressionScalarParameter_22 {class=MaterialExpressionScalarParameter, properties {defaultValue=1, group=Mask, parameterName="• Height - Max", sortPriority=513}}
  MaterialExpressionScalarParameter_23 {class=MaterialExpressionScalarParameter, properties {group=Mask, parameterName="• Height - Min", sortPriority=512}}
  MaterialExpressionScalarParameter_3 {class=MaterialExpressionScalarParameter, properties {group=Mask, parameterName="• VTX Color - Bias", sortPriority=175}}
  MaterialExpressionScalarParameter_4 {class=MaterialExpressionScalarParameter, properties {defaultValue=1, group=Mask, parameterName="• Height - Intensity", sortPriority=516}}
  MaterialExpressionScalarParameter_5 {
   class=MaterialExpressionScalarParameter
   properties {defaultValue=1, group=Mask, parameterName="• Height - Blend (1=Max 0=Min)", sortPriority=510}
  }
  MaterialExpressionScalarParameter_51 {
   class=MaterialExpressionScalarParameter
   properties {
    bUseCustomPrimitiveData=true
    desc="Primitive Data Only, Enable per instance in the details panel."
    group=Waterline
    parameterName=PD_WaterlineHeightOffset
    sortPriority=0
   }
  }
  MaterialExpressionScalarParameter_52 {class=MaterialExpressionScalarParameter, properties {defaultValue=-0.2, group=Mask, parameterName=* World Z Normal - Bias, sortPriority=120}}
  MaterialExpressionScalarParameter_53 {
   class=MaterialExpressionScalarParameter
   properties {defaultValue=1.5, group=Mask, parameterName=* World Z Normal - Intensity, sortPriority=121}
  }
  MaterialExpressionScalarParameter_54 {class=MaterialExpressionScalarParameter, properties {group=Mask, parameterName="• Baked AO/Curvature/Painted - Intensity", sortPriority=376}}
  MaterialExpressionScalarParameter_55 {class=MaterialExpressionScalarParameter, properties {group=Mask, parameterName="• Baked AO/Curvature/Painted - Bias", sortPriority=375}}
  MaterialExpressionScalarParameter_56 {class=MaterialExpressionScalarParameter, properties {group=Mask, parameterName="• WS Noise - Bias", sortPriority=475}}
  MaterialExpressionScalarParameter_57 {class=MaterialExpressionScalarParameter, properties {defaultValue=1, group=Mask, parameterName="• WS Noise - Intensity", sortPriority=476}}
  MaterialExpressionScalarParameter_58 {class=MaterialExpressionScalarParameter, properties {defaultValue=1, group=Mask, parameterName="• DFAO - Intensity", sortPriority=226}}
  MaterialExpressionScalarParameter_59 {class=MaterialExpressionScalarParameter, properties {group=Mask, parameterName="• DFAO - Bias", sortPriority=225}}
  MaterialExpressionScalarParameter_6 {class=MaterialExpressionScalarParameter, properties {group=Mask, parameterName="• Height - Bias", sortPriority=515}}
  MaterialExpressionScalarParameter_60 {
   class=MaterialExpressionScalarParameter
   properties {defaultValue=1, desc=Modulates VTX color transition intensity, group=Mask, parameterName="• VTX Color - Intensity", sortPriority=176}
  }
  MaterialExpressionScalarParameter_61 {class=MaterialExpressionScalarParameter, properties {group=Mask, parameterName="• Masking - Bias", sortPriority=60}}
  MaterialExpressionScalarParameter_62 {class=MaterialExpressionScalarParameter, properties {defaultValue=1, group=Mask, parameterName="• Masking - Intensity", sortPriority=61}}
  MaterialExpressionScalarParameter_63 {
   class=MaterialExpressionScalarParameter
   properties {
    defaultValue=1
    desc="1= object top bound\r\n0= object bottom bound"
    group=Mask
    parameterName="• Z Gradient - Height - Top"
    sortPriority=266
   }
  }
  MaterialExpressionScalarParameter_64 {
   class=MaterialExpressionScalarParameter
   properties {desc="1= object top bound\r\n0= object bottom bound", group=Mask, parameterName="• Z Gradient - Height - Bottom", sortPriority=265}
  }
  MaterialExpressionScalarParameter_65 {
   class=MaterialExpressionScalarParameter
   properties {defaultValue=1, group=Mask, parameterName="• VTX Color - Blend (1=Max 0=Min)", sortPriority=170}
  }
  MaterialExpressionScalarParameter_66 {
   class=MaterialExpressionScalarParameter
   properties {defaultValue=1, group=Mask, parameterName="• DFAO - Blend (1=Max 0=Min)", sortPriority=220}
  }
  MaterialExpressionScalarParameter_67 {
   class=MaterialExpressionScalarParameter
   properties {defaultValue=1, group=Mask, parameterName="• Z Gradient - Blend (1=Max 0=Min)", sortPriority=260}
  }
  MaterialExpressionScalarParameter_68 {
   class=MaterialExpressionScalarParameter
   properties {defaultValue=1, group=Mask, parameterName="• Baked AO/Curvature/Painted - Blend (1=Max 0=Min)", sortPriority=370}
  }
  MaterialExpressionScalarParameter_69 {
   class=MaterialExpressionScalarParameter
   properties {defaultValue=1, group=Mask, parameterName="• WS Noise - Blend (1=Blend 0=Min)", sortPriority=470}
  }
  MaterialExpressionScalarParameter_7 {
   class=MaterialExpressionScalarParameter
   properties {defaultValue=1, group=Mask, parameterName="• Height - Selection (1=Current Layer 0=Previous Layers)", sortPriority=509}
  }
  MaterialExpressionScalarParameter_77 {class=MaterialExpressionScalarParameter, properties {group=Mask, parameterName="• Baked AO/Curvature/Painted - Invert", sortPriority=361}}
  MaterialExpressionScalarParameter_78 {class=MaterialExpressionScalarParameter, properties {group=Mask, parameterName="• WS Noise - Invert", sortPriority=461}}
  MaterialExpressionScalarParameter_79 {class=MaterialExpressionScalarParameter, properties {defaultValue=1, group=Mask, parameterName="• VTX Color - Invert", sortPriority=161}}
  MaterialExpressionScalarParameter_8 {class=MaterialExpressionScalarParameter, properties {group=Mask, parameterName="* Waterline - Blend (1=Max 0=Min)", sortPriority=410}}
  MaterialExpressionScalarParameter_9 {class=MaterialExpressionScalarParameter, properties {defaultValue=1, group=Mask, parameterName=* Waterline - Bias, sortPriority=415}}
  MaterialExpressionSetMaterialAttributes_0 {class=MaterialExpressionSetMaterialAttributes, inputs [MaterialAttributes {from="", output=""}]}
  MaterialExpressionShadingPathSwitch_0 {
   class=MaterialExpressionShadingPathSwitch
   inputs [
    Default {from=MaterialExpressionMultiply_27, output=""}
    Deferred {from="", output=""}
    Forward {from="", output=""}
    Mobile {from=MaterialExpressionMultiply_1, output=""}
   ]
   properties {}
  }
  MaterialExpressionShadingPathSwitch_2 {
   class=MaterialExpressionShadingPathSwitch
   inputs [
    Default {from=MaterialExpressionReroute_189, output=""}
    Deferred {from="", output=""}
    Forward {from="", output=""}
    Mobile {from=MaterialExpressionMultiply_23, output=""}
   ]
   properties {}
  }
  MaterialExpressionSmoothStep_3 {
   class=MaterialExpressionSmoothStep
   inputs [
    Min {from=MaterialExpressionNamedRerouteUsage_283, output=""}
    Max {from=MaterialExpressionNamedRerouteUsage_284, output=""}
    Value {from=MaterialExpressionNamedRerouteUsage_266, output=""}
   ]
   properties {}
  }
  MaterialExpressionStaticBoolParameter_0 {class=MaterialExpressionStaticBoolParameter, inputs [], properties {group=Mask, parameterName=--- Use HEIGHT MASKING ---, sortPriority=500}}
  MaterialExpressionStaticBoolParameter_1 {
   class=MaterialExpressionStaticBoolParameter
   inputs []
   properties {group=Mask, parameterName=--- Use WATERLINE MASKING ---, sortPriority=400}
  }
  MaterialExpressionStaticBoolParameter_15 {
   class=MaterialExpressionStaticBoolParameter
   inputs []
   properties {desc="T= VTX Normals only\r\nF= Combined Material Normals", group=Mask, parameterName="• World Z Normal - Use VTX Normals", sortPriority=110}
  }
  MaterialExpressionStaticBoolParameter_16 {
   class=MaterialExpressionStaticBoolParameter
   inputs []
   properties {group=Mask, parameterName=--- Use Z GRADIENT MASKING ---, sortPriority=250}
  }
  MaterialExpressionStaticBoolParameter_17 {
   class=MaterialExpressionStaticBoolParameter
   inputs []
   properties {group=Mask, parameterName=--- Use WS NOISE MASKING ---, sortPriority=450}
  }
  MaterialExpressionStaticBoolParameter_22 {
   class=MaterialExpressionStaticBoolParameter
   inputs []
   properties {defaultValue=true, group=Mask, parameterName=* DFAO - Force Baked AO on Nanite, sortPriority=210}
  }
  MaterialExpressionStaticBoolParameter_23 {
   class=MaterialExpressionStaticBoolParameter
   inputs []
   properties {group=Mask, parameterName=--- Use WORLD Z NORMAL MASKING ---, sortPriority=100}
  }
  MaterialExpressionStaticBoolParameter_24 {
   class=MaterialExpressionStaticBoolParameter
   inputs []
   properties {group=Mask, parameterName=--- Use BAKED AO/CURVATURE/PAINTED MASKING ---, sortPriority=350}
  }
  MaterialExpressionStaticBoolParameter_26 {
   class=MaterialExpressionStaticBoolParameter
   inputs []
   properties {group=Mask, parameterName=--- Use VTX COLOR MASKING ---, sortPriority=150}
  }
  MaterialExpressionStaticBoolParameter_27 {class=MaterialExpressionStaticBoolParameter, inputs [], properties {group=Mask, parameterName=--- Use DFAO MASKING ---, sortPriority=200}}
  MaterialExpressionStaticBoolParameter_28 {class=MaterialExpressionStaticBoolParameter, inputs [], properties {group=Mask, parameterName=Normals - Add Normals, sortPriority=325}}
  MaterialExpressionStaticBool_17 {class=MaterialExpressionStaticBool}
  MaterialExpressionStaticBool_18 {class=MaterialExpressionStaticBool}
  MaterialExpressionStaticBool_21 {class=MaterialExpressionStaticBool}
  MaterialExpressionStaticBool_22 {class=MaterialExpressionStaticBool}
  MaterialExpressionStaticBool_23 {class=MaterialExpressionStaticBool}
  MaterialExpressionStaticBool_24 {class=MaterialExpressionStaticBool}
  MaterialExpressionStaticBool_26 {class=MaterialExpressionStaticBool}
  MaterialExpressionStaticBool_27 {class=MaterialExpressionStaticBool}
  MaterialExpressionStaticBool_30 {class=MaterialExpressionStaticBool}
  MaterialExpressionStaticBool_31 {class=MaterialExpressionStaticBool}
  MaterialExpressionStaticSwitch_33 {
   class=MaterialExpressionStaticSwitch
   inputs [
    True {from=MaterialExpressionNamedRerouteUsage_261, output=""}
    False {from=MaterialExpressionConstant_65, output=""}
    Value {from=MaterialExpressionNamedRerouteUsage_270, output=""}
   ]
   properties {}
  }
  MaterialExpressionStaticSwitch_34 {
   class=MaterialExpressionStaticSwitch
   inputs [
    True {from=MaterialExpressionIf_4, output=""}
    False {from=MaterialExpressionReroute_181, output=""}
    Value {from=MaterialExpressionNamedRerouteUsage_312, output=""}
   ]
   properties {}
  }
  MaterialExpressionStaticSwitch_35 {
   class=MaterialExpressionStaticSwitch
   inputs [
    True {from=MaterialExpressionIf_3, output=""}
    False {from=MaterialExpressionReroute_198, output=""}
    Value {from=MaterialExpressionNamedRerouteUsage_276, output=""}
   ]
   properties {}
  }
  MaterialExpressionStaticSwitch_36 {
   class=MaterialExpressionStaticSwitch
   inputs [
    True {from=MaterialExpressionIf_2, output=""}
    False {from=MaterialExpressionReroute_182, output=""}
    Value {from=MaterialExpressionNamedRerouteUsage_277, output=""}
   ]
   properties {}
  }
  MaterialExpressionStaticSwitch_37 {
   class=MaterialExpressionStaticSwitch
   inputs [
    True {from=MaterialExpressionIf_1, output=""}
    False {from=MaterialExpressionReroute_191, output=""}
    Value {from=MaterialExpressionNamedRerouteUsage_278, output=""}
   ]
   properties {}
  }
  MaterialExpressionStaticSwitch_38 {
   class=MaterialExpressionStaticSwitch
   inputs [
    True {from=MaterialExpressionIf_9, output=""}
    False {from=MaterialExpressionReroute_195, output=""}
    Value {from=MaterialExpressionNamedRerouteUsage_279, output=""}
   ]
   properties {}
  }
  MaterialExpressionStaticSwitch_39 {
   class=MaterialExpressionStaticSwitch
   inputs [
    True {from=MaterialExpressionSaturate_29, output=""}
    False {from=MaterialExpressionSaturate_25, output=""}
    Value {from=MaterialExpressionNamedRerouteUsage_273, output=""}
   ]
   properties {desc="T= VTX Normals only\r\nF= Combined Material Normals"}
  }
  MaterialExpressionStaticSwitch_40 {
   class=MaterialExpressionStaticSwitch
   inputs [
    True {from=MaterialExpressionNamedRerouteUsage_269, output=""}
    False {from=MaterialExpressionNamedRerouteUsage_268, output=""}
    Value {from=MaterialExpressionNamedRerouteUsage_298, output=""}
   ]
   properties {desc="** DF AO is for Epic Settings Only!\r\nTrue = AO Baked from VTX or Texture\r\nFalse = AO Distance Field"}
  }
  MaterialExpressionStaticSwitch_42 {
   class=MaterialExpressionStaticSwitch
   inputs [
    True {from=MaterialExpressionIf_10, output=""}
    False {from=MaterialExpressionReroute_209, output=""}
    Value {from=MaterialExpressionNamedRerouteUsage_305, output=""}
   ]
   properties {}
  }
  MaterialExpressionStaticSwitch_43 {
   class=MaterialExpressionStaticSwitch
   inputs [
    True {from=MaterialExpressionIf_0, output=""}
    False {from=MaterialExpressionReroute_215, output=""}
    Value {from=MaterialExpressionNamedRerouteUsage_307, output=""}
   ]
   properties {}
  }
  MaterialExpressionSubtract_0 {
   class=MaterialExpressionSubtract
   inputs [
    A {from=MaterialExpressionReroute_197}
    B {from=""}
   ]
   properties {desc=1}
  }
  MaterialExpressionSubtract_19 {
   class=MaterialExpressionSubtract
   inputs [
    A {from=MaterialExpressionReroute_211}
    B {from=""}
   ]
   properties {desc=1}
  }
  MaterialExpressionTransform_2 {class=MaterialExpressionTransform, inputs [None {from=MaterialExpressionGetMaterialAttributes_7, output=Normal}]}
  MaterialExpressionVectorParameter_5 {
   class=MaterialExpressionVectorParameter
   properties {
    group=Mask
    parameterName=* Baked AO/Curvature/Painted - Channel Weights
    sortPriority=360
    channelNames {a="", b=Blue, g=Green, r=Red}
    defaultValue=<rgba 1 0 0 0>
   }
  }
  MaterialExpressionVectorParameter_6 {
   class=MaterialExpressionVectorParameter
   properties {
    group=Mask
    parameterName=* WS Noise - Channel Weights
    sortPriority=460
    channelNames {a="", b=Blue, g=Green, r=Red}
    defaultValue=<rgba 1 0 0 0>
   }
  }
  MaterialExpressionVectorParameter_7 {
   class=MaterialExpressionVectorParameter
   properties {
    group=Mask
    parameterName="• VTX Color - Channel Weights"
    sortPriority=160
    channelNames {a="", b=Blue, g=Green, r=Red}
    defaultValue=<rgba 1 0 0 0>
   }
  }
  MaterialExpressionVertexColor_3 {class=MaterialExpressionVertexColor, inputs [], properties {}}
  MaterialExpressionVertexInterpolator_2 {class=MaterialExpressionVertexInterpolator, inputs [VS {from=MaterialExpressionSaturate_27, output=""}], properties {}}
  MaterialExpressionVertexNormalWS_3 {class=MaterialExpressionVertexNormalWS, inputs [], properties {}}
  MaterialExpressionWorldPosition_3 {class=MaterialExpressionWorldPosition, inputs [], properties {}}
 }}
