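# Material function that blends a static-switch-selected landscape layer (snow, sand, grass, gravel and others) onto rock material attributes.
# MF_Rocks_Landscape_Layer_Masking builds the blend mask from switchable world-normal, vertex color, DFAO, gradient, baked/painted, waterline, noise and height masks.
# OpacityMask carries height, remapped for displacement or set neutral.
# Static switches add landscape-only, mask debug and additive-normal options; the function outputs blended attributes and the mask.
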MaterialFunction /BRCoreMaterials/Materials/MaterialFunctions/MF_Rocks_Landscape_Layer
graph {nodes [
  FunctionInput() {inputName="\\ --- Dependencies --- /", inputType=FunctionInput_Scalar, sortPriority=800}
  FunctionInput() {inputName="\\ --- Landscape/Biome Masks --- /", inputType=FunctionInput_Scalar, sortPriority=1000}
  n1 = NamedRerouteUsage()
  n2 = GetMaterialAttributes(None=@n1) {attributeGetTypes [#1]}
  n3 = SetMaterialAttributes!(
   MaterialAttributes=@n2.MaterialAttributes
   OpacityMask=LinearInterpolate(
    A=ScalarParameter() {group=--- F3 | Landscape, parameterName=* Landscape - Realtime Displacement - Min, sortPriority=4}
    B=ScalarParameter() {defaultValue=1, group=--- F3 | Landscape, parameterName=* Landscape - Realtime Displacement - Max, sortPriority=4}
    Alpha=@n2.OpacityMask
   )
  ) {attributeSetTypes [#1]}
  n4 = SetMaterialAttributes(MaterialAttributes=@n1, OpacityMask=Constant() {desc="No Displacmenet, Neutral Height", r=0.5}) {attributeSetTypes [#1]}
  n5 = NamedRerouteUsage()
  n6 = NamedRerouteUsage()
  n7 = StaticSwitch(
   True=NamedRerouteUsage()
   False=BlendMaterialAttributes(
    A=NamedRerouteUsage()
    B=StaticSwitch(True=@n3, False=@n4, Value=@n5)
    Alpha=FunctionOutput(None=@n6) {outputName=Landscape Layer Blend Mask, sortPriority=-50}
   ) {vertexAttributeBlendType=UseB}
   Value=NamedRerouteUsage()
  )
  n8 = SetMaterialAttributes!(MaterialAttributes=@n7, BaseColor=NamedRerouteUsage()) {attributeSetTypes [#2]}
  n9 = NamedRerouteUsage()
  FunctionOutput(None=StaticSwitch(True=StaticSwitch(True=@n8, False=@n7, Value=@n9), False=NamedRerouteUsage(), Value=NamedRerouteUsage())) {outputName=Blended MaterialAttributes, sortPriority=-100}
  NamedRerouteDeclaration(None=FunctionInput() {inputName=Z Gradient, inputType=FunctionInput_Scalar, sortPriority=814}) {name=_Z Gradient, nodeColor=<rgba 1 0 0.0525 1>}
  NamedRerouteDeclaration(None=FunctionInput() {inputName=Baked AO, inputType=FunctionInput_Scalar, sortPriority=811}) {name=_Baked AO, nodeColor=<rgba 1 0 0.0525 1>}
  NamedRerouteDeclaration(None=StaticBoolParameter() {defaultValue=true, group=--- F3 | Landscape, parameterName=* Landscape - Normals - Add Normals, sortPriority=0}) {name="* Blending - Normals - Add Normals(SB)", nodeColor=<rgba 0.1979 0.0084 0 1>}
  NamedRerouteDeclaration(None=FunctionInput() {inputName=TerrainMasks_QuadB, inputType=FunctionInput_Scalar, sortPriority=1013}) {name=TerrainMasks_QuadB, nodeColor=<rgba 1 0.0434 0.0403 1>}
  NamedRerouteDeclaration(None=FunctionInput() {inputName=TerrainMasks_QuadA, inputType=FunctionInput_Scalar, sortPriority=1012}) {name=TerrainMasks_QuadA, nodeColor=<rgba 1 0.0434 0.0403 1>}
  NamedRerouteDeclaration(None=FunctionInput() {inputName=TerrainMasks_QuadC, inputType=FunctionInput_Scalar, sortPriority=1014}) {name=TerrainMasks_QuadC, nodeColor=<rgba 1 0.0434 0.0403 1>}
  NamedRerouteDeclaration(None=FunctionInput() {inputName=TerrainMasks_QuadD, inputType=FunctionInput_Scalar, sortPriority=1015}) {name=TerrainMasks_QuadD, nodeColor=<rgba 1 0.0434 0.0403 1>}
  n10 = NamedRerouteUsage()
  n11 = NamedRerouteUsage()
  n12 = NamedRerouteUsage()
  n13 = NamedRerouteUsage()
  n14 = StaticBoolParameter() {group=--- F3 | Landscape, parameterName=* Landscape - Use Chaparral Grass, sortPriority=-4}
  n15 = StaticSwitch(
   True=NamedRerouteUsage()
   False=StaticSwitch(
    True=NamedRerouteUsage()
    False=StaticSwitch(True=@n12, False=@n13, Value=@n14)
    Value=StaticBoolParameter() {group=--- F3 | Landscape, parameterName=* Landscape - Use Desert Grass, sortPriority=-4}
   )
   Value=StaticBoolParameter() {group=--- F3 | Landscape, parameterName=* Landscape - Use Mossy Forest, sortPriority=-4}
  )
  n16 = StaticBoolParameter() {group=--- F3 | Landscape, parameterName=* Landscape - Use Gravel, sortPriority=-4}
  n17 = StaticSwitch(
   True=NamedRerouteUsage()
   False=StaticSwitch(
    True=NamedRerouteUsage()
    False=StaticSwitch(True=@n11, False=@n15, Value=@n16)
    Value=StaticBoolParameter() {group=--- F3 | Landscape, parameterName=* Landscape - Use Mud, sortPriority=-4}
   )
   Value=StaticBoolParameter() {group=--- F3 | Landscape, parameterName=* Landscape - Use Sand, sortPriority=-4}
  )
  n18 = StaticBoolParameter() {group=--- F3 | Landscape, parameterName=* Landscape - Use Asphalt, sortPriority=-4}
  n19 = StaticSwitch(
   True=NamedRerouteUsage()
   False=StaticSwitch(
    True=NamedRerouteUsage()
    False=StaticSwitch(True=@n10, False=@n17, Value=@n18)
    Value=StaticBoolParameter() {group=--- F3 | Landscape, parameterName=* Landscape - Use Scree, sortPriority=-4}
   )
   Value=NamedRerouteDeclaration(None=StaticBoolParameter() {desc="If set to True, don't use L3", group=--- F3 | Landscape, parameterName=* Landscape - Use Snow, sortPriority=-4}) {name=Use Snow, nodeColor=<rgba 0.3594 0 1 1>}
  )
  n20 = GetMaterialAttributes(None=@n19) {attributeGetTypes [#1]}
  NamedRerouteDeclaration(None=SetMaterialAttributes!(MaterialAttributes=@n19, Opacity=@n20.OpacityMask) {attributeSetTypes [#3]}) {name=Landscape Layers, nodeColor=<rgba 0.0312 0 1 1>}
  NamedRerouteDeclaration(None=@n20.OpacityMask) {name=Landscape Height, nodeColor=<rgba 0.0312 0 1 1>}
  NamedRerouteDeclaration(None=StaticBoolParameter() {group=--- F3 | Landscape, parameterName=* Landscape --- Use HEIGHT MASKING ---, sortPriority=500}) {name=--- Use HEIGHT MASKING ---, nodeColor=<rgba 0.1979 0.0084 0 1>}
  NamedRerouteDeclaration(None=ScalarParameter() {defaultValue=1, group=--- F3 | Landscape, parameterName=* Landscape - Height - Intensity, sortPriority=516}) {name="• Height - Intensity", nodeColor=<rgba 0.5043 1 0 1>}
  NamedRerouteDeclaration(None=StaticBoolParameter() {group=--- F3 | Landscape, parameterName=* Landscape - Use Landscape Only, sortPriority=10}) {name="Use Landscape Only(SB)", nodeColor=<rgba 0.1979 0.0084 0 1>}
  NamedRerouteDeclaration(None=StaticBoolParameter() {group=--- F3 | Landscape, parameterName=* Landscape - DEBUG Landscape Mask, sortPriority=-1}) {name=Landscape - DEBUG Landscape Mask, nodeColor=<rgba 0 0 1 1>}
  NamedRerouteDeclaration(None=ScalarParameter() {group=--- F3 | Landscape, parameterName=* Landscape - Baked AO/Curvature/Painted - Invert, sortPriority=361}) {name="• Baked AO/Curvature/Painted - Invert", nodeColor=<rgba 0.5043 1 0 1>}
  NamedRerouteDeclaration(None=ScalarParameter() {defaultValue=-0.2, group=--- F3 | Landscape, parameterName=* Landscape - World Z Normal - Bias, sortPriority=120}) {name=* World Z Normal - Bias, nodeColor=<rgba 0.5043 1 0 1>}
  NamedRerouteDeclaration(None=ScalarParameter() {group=--- F3 | Landscape, parameterName=* Landscape - Height - Bias, sortPriority=515}) {name="• Height - Bias", nodeColor=<rgba 0.5043 1 0 1>}
  NamedRerouteDeclaration(None=StaticBoolParameter() {group=--- F3 | Landscape, parameterName=* Landscape --- Use WORLD Z NORMAL MASKING ---, sortPriority=100}) {name=--- Use WORLD Z NORMAL MASKING ---, nodeColor=<rgba 0.1979 0.0084 0 1>}
  NamedRerouteDeclaration(None=ScalarParameter() {defaultValue=1.5, group=--- F3 | Landscape, parameterName=* Landscape - World Z Normal - Intensity, sortPriority=121}) {name=* World Z Normal - Intensity, nodeColor=<rgba 0.5043 1 0 1>}
  n21 = StaticBoolParameter() {
   desc="T= VTX Normals only\r\nF= Combined Material Normals"
   group=--- F3 | Landscape
   parameterName=* Landscape - World Z Normal - Use VTX Normals
   sortPriority=110
  }
  NamedRerouteDeclaration(None=@n21) {name="• World Z Normal - Use VTX Normals", desc="T= VTX Normals only\r\nF= Combined Material Normals", nodeColor=<rgba 0.1979 0.0084 0 1>}
  n22 = ScalarParameter() {
   defaultValue=1
   desc=Modulates VTX color transition intensity
   group=--- F3 | Landscape
   parameterName=* Landscape - VTX Color - Intensity
   sortPriority=176
  }
  NamedRerouteDeclaration(None=@n22) {name="• VTX Color - Intensity", desc=Modulates VTX color transition intensity, nodeColor=<rgba 0.5043 1 0 1>}
  NamedRerouteDeclaration(None=StaticBoolParameter() {group=--- F3 | Landscape, parameterName=* Landscape --- Use VTX COLOR MASKING ---, sortPriority=150}) {name=--- Use VTX COLOR MASKING ---, nodeColor=<rgba 0.1979 0.0084 0 1>}
  NamedRerouteDeclaration(None=StaticBoolParameter() {group=--- F3 | Landscape, parameterName=* Landscape --- Use DFAO MASKING ---, sortPriority=200}) {name=--- Use DFAO MASKING ---, nodeColor=<rgba 0.1979 0.0084 0 1>}
  NamedRerouteDeclaration(None=StaticBoolParameter() {group=--- F3 | Landscape, parameterName=* Landscape --- Use Z GRADIENT MASKING ---, sortPriority=250}) {name=--- Use Z GRADIENT MASKING ---, nodeColor=<rgba 0.1979 0.0084 0 1>}
  NamedRerouteDeclaration(None=StaticBoolParameter() {group=--- F3 | Landscape, parameterName=* Landscape --- Use BAKED AO/CURVATURE/PAINTED MASKING ---, sortPriority=350}) {name=--- Use BAKED AO/CURVATURE/PAINTED MASKING ---, nodeColor=<rgba 0.1979 0.0084 0 1>}
  NamedRerouteDeclaration(None=StaticBoolParameter() {group=--- F3 | Landscape, parameterName=* Landscape --- Use WS NOISE MASKING ---, sortPriority=450}) {name=--- Use WS NOISE MASKING ---, nodeColor=<rgba 0.1979 0.0084 0 1>}
  n23 = VectorParameter() {
   group=--- F3 | Landscape
   parameterName=* Landscape - WS Noise - Channel Weights
   sortPriority=460
   channelNames {a="", b=Blue, g=Green, r=Red}
   defaultValue=<rgba 1 0 0 0>
  }
  NamedRerouteDeclaration(None=@n23.RGB) {name="• WS Noise - Channel Weights", nodeColor=<rgba 0.5043 1 0 1>}
  NamedRerouteDeclaration(None=ScalarParameter() {group=--- F3 | Landscape, parameterName=* Landscape - Masking - Invert, sortPriority=50}) {name=* Masking - Invert, nodeColor=<rgba 0.5043 1 0 1>}
  NamedRerouteDeclaration(None=ScalarParameter() {group=--- F3 | Landscape, parameterName=* Landscape - Masking - Bias, sortPriority=61}) {name="• Masking - Bias", nodeColor=<rgba 0.5043 1 0 1>}
  NamedRerouteDeclaration(None=ScalarParameter() {defaultValue=1, group=--- F3 | Landscape, parameterName=* Landscape - Masking - Intensity, sortPriority=60}) {name="• Masking - Intensity", nodeColor=<rgba 0.5043 1 0 1>}
  NamedRerouteDeclaration(None=ScalarParameter() {defaultValue=1, group=--- F3 | Landscape, parameterName=* Landscape - WS Noise - Intensity, sortPriority=476}) {name="• WS Noise - Intensity", nodeColor=<rgba 0.5043 1 0 1>}
  NamedRerouteDeclaration(None=ScalarParameter() {group=--- F3 | Landscape, parameterName=* Landscape - WS Noise - Bias, sortPriority=475}) {name="• WS Noise - Bias", nodeColor=<rgba 0.5043 1 0 1>}
  n24 = ScalarParameter() {defaultValue=1, group=--- F3 | Landscape, parameterName=* Landscape - Baked AO/Curvature/Painted - Intensity, sortPriority=376}
  NamedRerouteDeclaration(None=@n24) {name="• Baked AO/Curvature/Painted - Intensity", nodeColor=<rgba 0.5043 1 0 1>}
  NamedRerouteDeclaration(None=ScalarParameter() {group=--- F3 | Landscape, parameterName=* Landscape - Baked AO/Curvature/Painted - Bias, sortPriority=375}) {name="• Baked AO/Curvature/Painted - Bias", nodeColor=<rgba 0.5043 1 0 1>}
  n25 = ScalarParameter() {
   defaultValue=1
   desc="1= object top bound\r\n0= object bottom bound"
   group=--- F3 | Landscape
   parameterName=* Landscape - Z Gradient - Height - Top
   sortPriority=266
  }
  NamedRerouteDeclaration(None=@n25) {name="• Z Gradient - Height - Top", nodeColor=<rgba 0.5043 1 0 1>}
  n26 = ScalarParameter() {
   desc="1= object top bound\r\n0= object bottom bound"
   group=--- F3 | Landscape
   parameterName=* Landscape - Z Gradient - Height - Bottom
   sortPriority=265
  }
  NamedRerouteDeclaration(None=@n26) {name="• Z Gradient - Height - Bottom", nodeColor=<rgba 0.5043 1 0 1>}
  NamedRerouteDeclaration(None=ScalarParameter() {defaultValue=1, group=--- F3 | Landscape, parameterName=* Landscape - DFAO - Intensity, sortPriority=226}) {name="• DFAO - Intensity", nodeColor=<rgba 0.5043 1 0 1>}
  NamedRerouteDeclaration(None=FunctionInput() {inputName=DFAO, inputType=FunctionInput_Scalar, sortPriority=813}) {name=_DFAO, nodeColor=<rgba 1 0 0.0525 1>}
  NamedRerouteDeclaration(None=ScalarParameter() {group=--- F3 | Landscape, parameterName=* Landscape - WS Noise - Invert, sortPriority=461}) {name="• WS Noise - Invert", nodeColor=<rgba 0.5043 1 0 1>}
  NamedRerouteDeclaration(None=ScalarParameter() {group=--- F3 | Landscape, parameterName=* Landscape - DFAO - Bias, sortPriority=225}) {name="• DFAO - Bias", nodeColor=<rgba 0.5043 1 0 1>}
  n27 = StaticBoolParameter() {defaultValue=true, group=--- F3 | Landscape, parameterName=* Landscape - DFAO - Force Baked AO on Nanite, sortPriority=210}
  NamedRerouteDeclaration(None=@n27) {name=* DFAO - Force Baked AO on Nanite, nodeColor=<rgba 0.1979 0.0084 0 1>}
  n28 = ScalarParameter() {defaultValue=1, group=--- F3 | Landscape, parameterName="* Landscape - VTX Color - Blend (1=Max 0=Min)", sliderMax=1, sortPriority=170}
  NamedRerouteDeclaration(None=@n28) {name="• VTX Color - Blend (1=Max 0=Min)", nodeColor=<rgba 0.5043 1 0 1>}
  n29 = ScalarParameter() {defaultValue=1, group=--- F3 | Landscape, parameterName="* Landscape - DFAO - Blend (1=Max 0=Min)", sliderMax=1, sortPriority=220}
  NamedRerouteDeclaration(None=@n29) {name="• DFAO - Blend (1=Max 0=Min)", nodeColor=<rgba 0.5043 1 0 1>}
  n30 = ScalarParameter() {defaultValue=1, group=--- F3 | Landscape, parameterName="* Landscape - Z Gradient - Blend (1=Max 0=Min)", sliderMax=1, sortPriority=260}
  NamedRerouteDeclaration(None=@n30) {name="• Z Gradient - Blend (1=Max 0=Min)", nodeColor=<rgba 0.5043 1 0 1>}
  n31 = ScalarParameter() {
   defaultValue=1
   group=--- F3 | Landscape
   parameterName="* Landscape - Baked AO/Curvature/Painted - Blend (1=Max 0=Min)"
   sliderMax=1
   sortPriority=370
  }
  NamedRerouteDeclaration(None=@n31) {name="• Baked AO/Curvature/Painted - Blend (1=Max 0=Min)", nodeColor=<rgba 0.5043 1 0 1>}
  n32 = ScalarParameter() {defaultValue=1, group=--- F3 | Landscape, parameterName="* Landscape - WS Noise - Blend (1=Blend 0=Min)", sliderMax=1, sortPriority=470}
  NamedRerouteDeclaration(None=@n32) {name="• WS Noise - Blend (1=Max 0=Min)", nodeColor=<rgba 0.5043 1 0 1>}
  NamedRerouteDeclaration(None=ScalarParameter() {group=--- F3 | Landscape, parameterName=* Landscape - Height - Min, sortPriority=512}) {name="• Height - Min", nodeColor=<rgba 0.5043 1 0 1>}
  NamedRerouteDeclaration(None=ScalarParameter() {defaultValue=1, group=--- F3 | Landscape, parameterName=* Landscape - Height - Max, sortPriority=513}) {name="• Height - Max", nodeColor=<rgba 0.5043 1 0 1>}
  NamedRerouteDeclaration(None=ScalarParameter() {group=--- F3 | Landscape, parameterName=* Landscape - VTX Color - Bias, sortPriority=175}) {name="• VTX Color - Bias", nodeColor=<rgba 0.5043 1 0 1>}
  NamedRerouteDeclaration(None=FunctionInput() {inputName=Landscape - Mud, inputType=FunctionInput_MaterialAttributes, sortPriority=1032}) {name=Landscape - Mud, nodeColor=<rgba 0.5723 0.25 1 1>}
  n33 = ScalarParameter() {defaultValue=1, group=--- F3 | Landscape, parameterName="* Landscape - Height - Blend (1=Blend 0=Min)", sliderMax=1, sortPriority=510}
  NamedRerouteDeclaration(None=@n33) {name="• Height - Blend (1=Blend 0=Min)", nodeColor=<rgba 0.5043 1 0 1>}
  NamedRerouteDeclaration(None=ScalarParameter() {defaultValue=1, group=--- F3 | Landscape, parameterName=* Landscape - VTX Color - Invert, sortPriority=161}) {name="• VTX Color - Invert", nodeColor=<rgba 0.5043 1 0 1>}
  n34 = ScalarParameter() {
   defaultValue=1
   group=--- F3 | Landscape
   parameterName="* Landscape - Height - Selection (1=Current Layer 0=Previous Layers)"
   sortPriority=509
  }
  NamedRerouteDeclaration(None=@n34) {name="• Height - Selection (1=Current Layer 0=Previous Layers)", nodeColor=<rgba 0.5043 1 0 1>}
  NamedRerouteDeclaration(None=FunctionInput() {inputName=Landscape - Chaparral Grass, inputType=FunctionInput_MaterialAttributes, sortPriority=1024}) {name=Landscape - Chaparral Grass, nodeColor=<rgba 0.5723 0.25 1 1>}
  NamedRerouteDeclaration(None=StaticBoolParameter() {group=--- F3 | Landscape, parameterName=* Landscape --- Use WATERLINE MASKING ---, sortPriority=400}) {name=--- Use WATERLINE MASKING ---, nodeColor=<rgba 0.1979 0.0084 0 1>}
  NamedRerouteDeclaration(None=FunctionInput() {inputName=Baked Curvature, inputType=FunctionInput_Scalar, sortPriority=810}) {name=_Baked Curvature, nodeColor=<rgba 1 0 0.0525 1>}
  NamedRerouteDeclaration(None=FunctionInput() {inputName=Landscape - Desert Grass, inputType=FunctionInput_MaterialAttributes, sortPriority=1025}) {name=Landscape - Desert Grass, nodeColor=<rgba 0.5723 0.25 1 1>}
  NamedRerouteDeclaration(None=FunctionInput() {inputName=Landscape - Desert Grass Noise, inputType=FunctionInput_Scalar, sortPriority=1026}) {name=Landscape - Desert Grass Noise, nodeColor=<rgba 0.5723 0.25 1 1>}
  n35 = StaticBoolParameter() {defaultValue=true, group=--- F3 | Landscape, parameterName=* Landscape - Layer Contributes to Displacement, sortPriority=2}
  NamedRerouteDeclaration(None=@n35) {name="Layer Contributes to Displacement(SB)", nodeColor=<rgba 0.1979 0.0084 0 1>}
  n36 = VectorParameter() {
   group=--- F3 | Landscape
   parameterName=* Landscape - VTX Color - Channel Weights
   sortPriority=160
   channelNames {a="", b=Blue, g=Green, r=Red}
   defaultValue=<rgba 1 0 0 0>
  }
  NamedRerouteDeclaration(None=@n36.RGB) {name="• VTX Color - Channel Weights", nodeColor=<rgba 0.5043 1 0 1>}
  n37 = FunctionInput(Preview=SetMaterialAttributes!(MaterialAttributes)) {bUsePreviewValueAsDefault=true, inputName=Input MaterialAttributes, inputType=FunctionInput_MaterialAttributes, sortPriority=-100}
  NamedRerouteDeclaration(None=@n37) {name=_Input MaterialAttributes, nodeColor=<rgba 0.1016 0 1 1>}
  NamedRerouteDeclaration(None=ScalarParameter() {group=--- F3 | Landscape, parameterName=* Landscape - Waterline - Bias, sortPriority=415}) {name=* Waterline - Bias, nodeColor=<rgba 0.5043 1 0 1>}
  NamedRerouteDeclaration(None=ScalarParameter() {group=--- F3 | Landscape, parameterName="* Landscape - Waterline - Blend (1=Max 0=Min)", sliderMax=1, sortPriority=410}) {name="* Waterline - Blend (1=Max 0=Min)", nodeColor=<rgba 0.5043 1 0 1>}
  NamedRerouteDeclaration(None=ScalarParameter() {defaultValue=1, group=--- F3 | Landscape, parameterName=* Landscape - Waterline - Intensity, sortPriority=416}) {name=* Waterline - Intensity, nodeColor=<rgba 0.5043 1 0 1>}
  NamedRerouteDeclaration(None=FunctionInput() {inputName=Painted Mask, inputType=FunctionInput_Scalar, sortPriority=812}) {name=_Painted Mask, nodeColor=<rgba 1 0 0.0525 1>}
  NamedRerouteDeclaration(None=MaterialFunctionCall!(
    Layer A MaterialAttributes=NamedRerouteUsage()
    * Masking - Invert=NamedRerouteUsage()
    "• Masking - Bias"=NamedRerouteUsage()
    "• Masking - Intensity"=NamedRerouteUsage()
    --- Use WORLD Z NORMAL MASKING ---=NamedRerouteUsage()
    "• World Z Normal - Use VTX Normals"=NamedRerouteUsage()
    * World Z Normal - Bias=NamedRerouteUsage()
    * World Z Normal - Intensity=NamedRerouteUsage()
    --- Use VTX COLOR MASKING ---=NamedRerouteUsage()
    "• VTX Color - Channel Weights"=NamedRerouteUsage()
    "• VTX Color - Invert"=NamedRerouteUsage()
    "• VTX Color - Blend (1=Max 0=Min)"=NamedRerouteUsage()
    "• VTX Color - Bias"=NamedRerouteUsage()
    "• VTX Color - Intensity"=NamedRerouteUsage()
    --- Use DFAO MASKING ---=NamedRerouteUsage()
    * DFAO - Force Baked AO on Nanite=NamedRerouteUsage()
    "• DFAO - Blend (1=Max 0=Min)"=NamedRerouteUsage()
    "• DFAO - Bias"=NamedRerouteUsage()
    "• DFAO - Intensity"=NamedRerouteUsage()
    --- Use Z GRADIENT MASKING ---=NamedRerouteUsage()
    "• Z Gradient - Blend (1=Max 0=Min)"=NamedRerouteUsage()
    "• Z Gradient - Height - Bottom"=NamedRerouteUsage()
    "• Z Gradient - Height - Top"=NamedRerouteUsage()
    --- Use BAKED AO/CURVATURE/PAINTED MASKING ---=NamedRerouteUsage()
    "• Baked AO/Curvature/Painted - Channel Weights"=NamedRerouteUsage()
    "• Baked AO/Curvature/Painted - Invert"=NamedRerouteUsage()
    "• Baked AO/Curvature/Painted - Blend (1=Max 0=Min)"=NamedRerouteUsage()
    "• Baked AO/Curvature/Painted - Bias"=NamedRerouteUsage()
    "• Baked AO/Curvature/Painted - Intensity"=NamedRerouteUsage()
    --- Use WATERLINE MASKING ---=NamedRerouteUsage()
    "* Waterline - Blend (1=Max 0=Min)"=NamedRerouteUsage()
    * Waterline - Bias=NamedRerouteUsage()
    * Waterline - Intensity=NamedRerouteUsage()
    --- Use WS NOISE MASKING ---=NamedRerouteUsage()
    "• WS Noise - Channel Weights"=NamedRerouteUsage()
    "• WS Noise - Invert"=NamedRerouteUsage()
    "• WS Noise - Blend (1=Blend 0=Min)"=NamedRerouteUsage()
    "• WS Noise - Bias"=NamedRerouteUsage()
    "• WS Noise - Intensity"=NamedRerouteUsage()
    --- Use HEIGHT MASKING ---=NamedRerouteUsage()
    "• Height - Selection (1=Current Layer 0=Previous Layers)"=NamedRerouteUsage()
    "• Height - Blend (1=Blend 0=Min)"=NamedRerouteUsage()
    "• Height - Min"=NamedRerouteUsage()
    "• Height - Max"=NamedRerouteUsage()
    "• Height - Bias"=NamedRerouteUsage()
    "• Height - Intensity"=NamedRerouteUsage()
    "\\ --- Dependencies --- /"
    Baked Curvature=NamedRerouteUsage()
    Baked AO=NamedRerouteUsage()
    Painted Mask=NamedRerouteUsage()
    DFAO=NamedRerouteUsage()
    Z Gradient=NamedRerouteUsage()
    WS Noise Masks=NamedRerouteUsage()
    "\\ --- Landscape/Biome Masks --- /"
    Landscape Layers=NamedRerouteUsage()
    Landscape Height=NamedRerouteUsage()
    TerrainMasks_QuadA=NamedRerouteUsage()
    TerrainMasks_QuadB=NamedRerouteUsage()
    TerrainMasks_QuadC=NamedRerouteUsage()
    TerrainMasks_QuadD=NamedRerouteUsage()
   ) {materialFunction=&/MF_Rocks_Landscape_Layer_Masking}.Combined Blend Mask) {name=Landscape Combined Blend Mask, nodeColor=<rgba 0.5 0 1 1>}
  n38 = GetMaterialAttributes(None=NamedRerouteUsage()) {attributeGetTypes [#4]}
  n39 = GetMaterialAttributes(None=NamedRerouteUsage()) {attributeGetTypes [#4]}
  n40 = Multiply(A=@n39.Normal, B=Constant3Vector() {constant=<rgba 1 1 0 0>})
  n41 = ScalarParameter() {defaultValue=0.5, group=--- F3 | Landscape, parameterName=* Landscape - Add Normals Value, sortPriority=24}
  n42 = Normalize(VectorInput=LinearInterpolate(A=@n38.Normal, B=Add(A=@n38.Normal, B=@n40), Alpha=Saturate(None=@n41)))
  n43 = StaticBoolParameter() {
   desc="If set to True, set TopLayer to False"
   group=--- F3 | Landscape
   parameterName=* Landscape - Add Landscape Layer Normals
   sortPriority=20
  }
  NamedRerouteDeclaration(None=SetMaterialAttributes!(MaterialAttributes=@n38.MaterialAttributes, Normal=StaticSwitch(True=@n42, False=@n38.Normal, Value=@n43)) {attributeSetTypes [#4]}) {name=Landscape Layer Blended MaterialAttibutes, nodeColor=<rgba 0.5 0 1 1>}
  n44 = VectorParameter() {
   group=--- F3 | Landscape
   parameterName=* Landscape - Baked AO/Curvature/Painted - Channel Weights
   sortPriority=360
   channelNames {a="", b=Blue, g=Green, r=Red}
   defaultValue=<rgba 1 0 0 0>
  }
  NamedRerouteDeclaration(None=@n44.RGB) {name="• Baked AO/Curvature/Painted - Channel Weights", nodeColor=<rgba 0.5043 1 0 1>}
  NamedRerouteDeclaration(None=FunctionInput(Preview=StaticBool()) {bUsePreviewValueAsDefault=true, inputName=Use Landscape Layer, inputType=FunctionInput_StaticBool, sortPriority=-111}) {name=Use Landscape, nodeColor=<rgba 1 0.4219 0 1>}
  NamedRerouteDeclaration(None=FunctionInput() {inputName=Landscape - Grass, inputType=FunctionInput_MaterialAttributes, sortPriority=1020}) {name=Landscape - Grass, nodeColor=<rgba 0.5723 0.25 1 1>}
  NamedRerouteDeclaration(None=FunctionInput() {inputName=Landscape - Gravel, inputType=FunctionInput_MaterialAttributes, sortPriority=1031}) {name=Landscape - Gravel, nodeColor=<rgba 0.5723 0.25 1 1>}
  NamedRerouteDeclaration(None=FunctionInput() {inputName=Landscape - Sand, inputType=FunctionInput_MaterialAttributes, sortPriority=1033}) {name=Landscape - Sand, nodeColor=<rgba 0.5723 0.25 1 1>}
  NamedRerouteDeclaration(None=FunctionInput() {inputName=Landscape - Asphalt, inputType=FunctionInput_MaterialAttributes, sortPriority=1034}) {name=Landscape - Asphalt, nodeColor=<rgba 0.5723 0.25 1 1>}
  NamedRerouteDeclaration(None=FunctionInput() {inputName=Landscape - Scree, inputType=FunctionInput_MaterialAttributes, sortPriority=1035}) {name=Landscape - Scree, nodeColor=<rgba 0.5723 0.25 1 1>}
  NamedRerouteDeclaration(None=FunctionInput() {inputName=Landscape - Snow, inputType=FunctionInput_MaterialAttributes, sortPriority=1036}) {name=Landscape - Snow, nodeColor=<rgba 0.4297 0 1 1>}
  NamedRerouteDeclaration(None=FunctionInput() {inputName=WS Noise Masks, inputType=FunctionInput_Scalar, sortPriority=815}) {name=_WS Noise Masks, nodeColor=<rgba 1 0 0.0525 1>}
  NamedRerouteDeclaration() {name=Landscape - MossyForest, nodeColor=<rgba 0.5723 0.25 1 1>}
  NamedRerouteUsage()
  n45 = GetMaterialAttributes() {attributeGetTypes [#5]}
  SetMaterialAttributes!(
   MaterialAttributes=@n45.MaterialAttributes
   Displacement=LinearInterpolate(B=@n45.Displacement, Alpha=Constant() {desc=Param}) {constA=0.5, constB=0.5}
  ) {attributeSetTypes [#5]}
  n46 = GetMaterialAttributes(None=NamedRerouteUsage()) {attributeGetTypes [#4]}
  SetMaterialAttributes!(MaterialAttributes=@n46.MaterialAttributes, Normal=Transform(None=@n46.Normal)) {attributeSetTypes [#4]}
  StaticBool()
  Transform(None=@n39.Normal)
 ]}
