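# Material function for 4-layer material debugging that switches between the input material and one of five scalar debug views by an Enum input.
# At zero, it darkens BaseColor with a screen-tiled T_Debug texture mask.
# Higher values select a debug input through nested Ifs, shown as emissive on flattened attributes with vertex normals.
# EyeAdaptationInverse counters auto-exposure on that emissive value.
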MaterialFunction /BRCoreMaterials/Materials/MaterialFunctions/MF_4Layer_Debug
graph {nodes [
  n1 = Constant()
  n2 = SetMaterialAttributes!(MaterialAttributes, Normal=VertexNormalWS(), BaseColor=@n1, Specular=@n1, Roughness=@n1, Metallic=@n1, SubsurfaceColor=@n1) {attributeSetTypes [
    #1
    #2
    #3
    #4
    #5
    #6
   ]}
  n3 = NamedRerouteUsage()
  n4 = If(
   A=NamedRerouteUsage()
   "A > B"=If(
    A=NamedRerouteUsage()
    "A > B"=FunctionInput() {bUsePreviewValueAsDefault=true, inputName=Debug 5, inputType=FunctionInput_Scalar, sortPriority=6}
    "A < B"=FunctionInput() {bUsePreviewValueAsDefault=true, inputName=Debug 4, inputType=FunctionInput_Scalar, sortPriority=5}
   ) {constB=4.01}
   "A < B"=FunctionInput() {bUsePreviewValueAsDefault=true, inputName=Debug 3, inputType=FunctionInput_Scalar, sortPriority=4}
  ) {constB=3.01}
  n5 = FunctionInput() {bUsePreviewValueAsDefault=true, inputName=Debug 2, inputType=FunctionInput_Scalar, sortPriority=3}
  n6 = EyeAdaptationInverse(
   LightValueInput=If(
    A=NamedRerouteUsage()
    "A > B"=If(A=@n3, "A > B"=@n4, "A < B"=@n5) {constB=2.01}
    "A < B"=FunctionInput() {bUsePreviewValueAsDefault=true, inputName=Debug 1, inputType=FunctionInput_Scalar, sortPriority=2}
   ) {constB=1.01}
   AlphaInput=Constant() {r=1}
  )
  n7 = GetMaterialAttributes(None=FunctionInput() {bUsePreviewValueAsDefault=true, inputName=Material, inputType=FunctionInput_MaterialAttributes, sortPriority=1}) {attributeGetTypes [#2]}
  n8 = ScreenPosition()
  n9 = Constant2Vector() {g=50, r=150}
  n10 = Add(A=Frac(None=Divide(A=@n8.PixelPosition, B=@n9))) {constB=-0.5}
  n11 = Constant2Vector() {g=1.5, r=1.5}
  n12 = TextureSample(UVs=Add(A=Multiply(A=@n10, B=@n11)) {constB=0.5}) {samplerSource=SSM_Clamp_WorldGroupSettings, samplerType=SAMPLERTYPE_LinearColor, texture=/BRCoreMaterials/Materials/Textures/T_Debug}
  n13 = LinearInterpolate(A=@n7.BaseColor, Alpha=@n12.R) {constB=0}
  FunctionOutput(None=If(
    A=NamedRerouteUsage()
    "A > B"=SetMaterialAttributes!(MaterialAttributes=@n2, EmissiveColor=@n6.EyeAdaptationInverse) {attributeSetTypes [#7]}
    "A < B"=SetMaterialAttributes!(MaterialAttributes=@n7.MaterialAttributes, BaseColor=@n13) {attributeSetTypes [#2]}
   ) {constB=0.01})
  NamedRerouteDeclaration(None=FunctionInput() {inputName=Enum, inputType=FunctionInput_Scalar}) {name=Enum, nodeColor=<rgba 0.2656 1 0 1>}
 ]}
