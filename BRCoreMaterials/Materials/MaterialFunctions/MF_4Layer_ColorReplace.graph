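# Material function that applies a per-layer color replace to BaseColor in a four-layer material, apparently swapping key-colored areas for a desaturated "New Color".
# Layer-masked key colors and an input color are brightness-normalized, with dark values flattened toward gray; the key side runs through a VertexInterpolator.
# Their distance, shaped by per-layer range, contrast and toggles, forms the replace mask, output alongside material and toggle.
# Static switches bypass the effect or select a debug path.
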MaterialFunction /BRCoreMaterials/Materials/MaterialFunctions/MF_4Layer_ColorReplace
graph {nodes [
  n1 = NamedRerouteUsage()
  n2 = FunctionInput() {inputName=Material, inputType=FunctionInput_MaterialAttributes}
  n3 = GetMaterialAttributes(None=@n2) {attributeGetTypes [#1]}
  n4 = SetMaterialAttributes!(
   MaterialAttributes=@n3.MaterialAttributes
   BaseColor=LinearInterpolate(
    A=NamedRerouteDeclaration(None=@n3.BaseColor) {name=BaseColor Input, nodeColor=<rgba 0.8984 1 0 1>}
    B=NamedRerouteUsage()
    Alpha=NamedRerouteUsage()
   )
  ) {attributeSetTypes [#1]}
  n5 = NamedRerouteDeclaration(None=@n4) {name=Final Color Replace, nodeColor=<rgba 0 0.0625 1 1>}
  FunctionOutput(None=StaticSwitchParameter(True=StaticSwitchParameter(True=@n1, False=@n5) {group=Color Replace, parameterName=ColorReplace Debug, sortPriority=31}, False=@n2) {group=- Effect Switches, parameterName=Color Replace}) {outputName=Material}
  n6 = ScalarParameter() {
   controlType=Enumeration
   enumeration=/BRCoreMaterials/Materials/Enums/EN_4Layer_GenericToggle
   group=Color Replace
   parameterName=L1 ColorReplace
   sliderMax=1
   sortPriority=0
  }
  n7 = ScalarParameter() {
   controlType=Enumeration
   enumeration=&EN_4Layer_GenericToggle
   group=Color Replace
   parameterName=L2 ColorReplace
   sliderMax=1
   sortPriority=10
  }
  n8 = ScalarParameter() {
   controlType=Enumeration
   enumeration=&EN_4Layer_GenericToggle
   group=Color Replace
   parameterName=L3 ColorReplace
   sliderMax=1
   sortPriority=20
  }
  n9 = ScalarParameter() {
   controlType=Enumeration
   enumeration=&EN_4Layer_GenericToggle
   group=Color Replace
   parameterName=L4 ColorReplace
   sliderMax=1
   sortPriority=30
  }
  n10 = MaterialFunctionCall!(RGBA Mask=NamedRerouteUsage(), L1=@n6, L2=@n7, L3=@n8, L4=@n9) {materialFunction=&/MF_4Layer_ScalarMasking}
  FunctionOutput(None=@n10.Result) {outputName=Color Replace Toggle, sortPriority=2}
  n11 = VectorParameter() {group=Color Replace, parameterName=L1 ColorReplace Key Color, sortPriority=1, defaultValue=<rgba 0.4233 0.0999 0.0865 1>}
  n12 = VectorParameter() {group=Color Replace, parameterName=L2 ColorReplace Key Color, sortPriority=11, defaultValue=<rgba 0.4233 0.0999 0.0865 1>}
  n13 = VectorParameter() {group=Color Replace, parameterName=L3 ColorReplace Key Color, sortPriority=21, defaultValue=<rgba 0.4233 0.0999 0.0865 1>}
  n14 = VectorParameter() {group=Color Replace, parameterName=L4 ColorReplace Key Color, sortPriority=31, defaultValue=<rgba 0.4233 0.0999 0.0865 1>}
  n15 = MaterialFunctionCall!(Layer Mask=NamedRerouteUsage(), L1=@n11.RGB, L2=@n12.RGB, L3=@n13.RGB, L4=@n14.RGB) {materialFunction=&/MF_4Layer_ColorMasking}
  n16 = DotProduct(A=@n15.Result, B=Constant() {r=1})
  n17 = LinearInterpolate(A=@n15.Result, B=@n16, Alpha=Power(Base=Constant() {r=2.7183}, Exp=Multiply(A=@n16) {constB=-6.28}))
  n18 = Constant() {r=2}
  n19 = VertexInterpolator(VS=Divide(A=@n17, B=DotProduct(A=@n17, B=@n18)))
  n20 = NamedRerouteUsage()
  n21 = DotProduct(A=@n20, B=Constant() {r=1})
  n22 = LinearInterpolate(A=@n20, B=@n21, Alpha=Power(Base=Constant() {r=2.7183}, Exp=Multiply(A=@n21) {constB=-6.28}))
  n23 = Divide(A=@n22, B=DotProduct(A=@n22, B=Constant() {r=2}))
  n24 = OneMinus(None=Length(None=Subtract(A=@n19.PS, B=@n23)) {is=class MaterialExpressionLength})
  n25 = Add(A=MaterialFunctionCall(
    RGBA Mask=NamedRerouteUsage()
    L1=ScalarParameter() {defaultValue=1, group=Color Replace, parameterName=L1 ColorReplace Range, sortPriority=3}
    L2=ScalarParameter() {defaultValue=1, group=Color Replace, parameterName=L2 ColorReplace Range, sortPriority=13}
    L3=ScalarParameter() {defaultValue=1, group=Color Replace, parameterName=L3 ColorReplace Range, sortPriority=23}
    L4=ScalarParameter() {defaultValue=1, group=Color Replace, parameterName=L4 ColorReplace Range, sortPriority=33}
   ) {materialFunction=&MF_4Layer_ScalarMasking}.Result) {constB=-1.45}
  n26 = NamedRerouteUsage()
  n27 = ScalarParameter() {defaultValue=5, group=Color Replace, parameterName=L1 ColorReplace Contrast, sortPriority=4}
  n28 = ScalarParameter() {defaultValue=5, group=Color Replace, parameterName=L2 ColorReplace Contrast, sortPriority=14}
  n29 = ScalarParameter() {defaultValue=5, group=Color Replace, parameterName=L3 ColorReplace Contrast, sortPriority=24}
  n30 = ScalarParameter() {defaultValue=5, group=Color Replace, parameterName=L4 ColorReplace Contrast, sortPriority=34}
  n31 = Saturate(None=MaterialFunctionCall(
    In=Add(A=@n24, B=@n25)
    Contrast=MaterialFunctionCall(RGBA Mask=@n26, L1=@n27, L2=@n28, L3=@n29, L4=@n30) {materialFunction=&MF_4Layer_ScalarMasking}.Result
   ) {materialFunction=&CheapContrast}.Result)
  FunctionOutput(None=NamedRerouteDeclaration(None=Multiply(A=@n31, B=@n10.Result)) {name=Color Replace Mask, nodeColor=<rgba 1 1 1 1>}) {outputName=Color Replace Mask, sortPriority=1}
  NamedRerouteDeclaration(None=FunctionInput() {inputName=Layer Mask, inputType=FunctionInput_Vector4, sortPriority=1}) {name=Layer Mask, nodeColor=<rgba 1 0 0.4219 1>}
  NamedRerouteDeclaration(None=Desaturation(None=NamedRerouteUsage(), Fraction=Constant() {r=1})) {name=New Color, nodeColor=<rgba 0.2891 0 1 1>}
  n32 = ScalarParameter() {
   controlType=Enumeration
   enumeration=/BRCoreMaterials/Materials/Enums/EN_4Layer_ColorReplace_Debug
   group=Color Replace
   parameterName=ColorReplace Debug Mode
   sliderMax=1
  }
  NamedRerouteDeclaration(None=MaterialFunctionCall!(Enum=@n32, Material=NamedRerouteUsage(), Debug 1=NamedRerouteUsage(), Debug 2, Debug 3, Debug 4, Debug 5) {materialFunction=&/MF_4Layer_Debug}.Result) {name=Debug Masks, nodeColor=<rgba 0.0312 0 1 1>}
  Reroute(None=@n4)
 ]}
