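# Material function that builds a height-based blend mask between a foreground and a background layer.
# Each height input is remapped between its Low and High values and saturated, then their offset difference is combined with the recentred Mask input.
# Contrast drives a lerp toward that combined value, extrapolating above 1, and the saturated result passes through a squared ease curve.
# Both remapped heightmaps and the pre-contrast mask are also outputs.
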MaterialFunction /BRCoreMaterials/Materials/MaterialFunctions/MF_4Layer_HeightBlend
graph {nodes [
  n1 = NamedRerouteUsage()
  n2 = FunctionInput() {bUsePreviewValueAsDefault=true, inputName=Foreground High, inputType=FunctionInput_Scalar, sortPriority=5, previewValue=<1 1 1 1>}
  n3 = FunctionInput() {bUsePreviewValueAsDefault=true, inputName=Foreground, inputType=FunctionInput_Scalar, sortPriority=3, previewValue=<1 1 1 1>}
  n4 = LinearInterpolate(
   A=FunctionInput() {bUsePreviewValueAsDefault=true, inputName=Foreground Low, inputType=FunctionInput_Scalar, sortPriority=4}
   B=@n2
   Alpha=@n3
  )
  n5 = FunctionInput() {bUsePreviewValueAsDefault=true, inputName=Background Low, inputType=FunctionInput_Scalar, sortPriority=1, previewValue=<0.5 0.5 0.5 0.5>}
  n6 = FunctionInput() {bUsePreviewValueAsDefault=true, inputName=Background High, inputType=FunctionInput_Scalar, sortPriority=2, previewValue=<0.5 0.5 0.5 1>}
  n7 = FunctionInput() {bUsePreviewValueAsDefault=true, inputName=Background, inputType=FunctionInput_Scalar, previewValue=<0.5 0.5 0.5 1>}
  n8 = Saturate(None=LinearInterpolate(A=@n5, B=@n6, Alpha=@n7))
  n9 = Subtract(
   A=FunctionOutput(None=Saturate(None=@n4)) {outputName=Foreground Heightmap, sortPriority=4}
   B=Add(A=FunctionOutput(None=@n8) {outputName=Background Heightmap, sortPriority=3}) {constB=-0.5}
  )
  n10 = FunctionInput() {bUsePreviewValueAsDefault=true, inputName=Mask, inputType=FunctionInput_Scalar, sortPriority=8, previewValue=<0.5 0.5 0.5 1>}
  n11 = Add(A=NamedRerouteDeclaration(None=@n10) {name=Mask, nodeColor=<rgba 1 0.3047 0 1>}) {constB=-0.5}
  n12 = FunctionOutput(None=Add(A=Saturate(None=@n9), B=Multiply(A=@n11) {constB=2})) {outputName=Mask Before Contrast, sortPriority=2}
  n13 = FunctionInput() {bUsePreviewValueAsDefault=true, inputName=Contrast, inputType=FunctionInput_Scalar, sortPriority=6, previewValue=<1.5 1.5 1.5 1>}
  n14 = OneMinus(None=Saturate(None=LinearInterpolate(A=@n1, B=@n12, Alpha=@n13)))
  n15 = OneMinus(None=Multiply(A=@n14, B=@n14))
  FunctionOutput(None=Multiply(A=@n15, B=@n15)) {outputName=Mask}
 ]}
