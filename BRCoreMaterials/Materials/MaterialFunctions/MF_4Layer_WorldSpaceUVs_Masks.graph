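# Material function that builds per-axis projection masks from absolute world vertex normals, apparently for world-space UV mapping in a four-layer material.
# Axis dominance differences yield three outputs: Flat hard masks, Smooth squared, normalized blends widened by Blend Radius, and Dithered masks.
# Dithered rounds the smooth weights using a temporally offset screen-pixel grid plus 4x4 Bayer pattern.
# Mode thresholds zero one axis or force all outputs to Z only.
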MaterialFunction /BRCoreMaterials/Materials/MaterialFunctions/MF_4Layer_WorldSpaceUVs_Masks
graph {nodes [
  MaterialExpressionAbs_5 = Abs(None=VertexNormalWS())
  n1 = Convert(None=@MaterialExpressionAbs_5) {
   nodeName=Break Float3
   convertInputs [{
     type=Vector3
     expressionInput {
      expression=&self:MaterialExpressionAbs_5
      inputName=None
      mask=0
      maskA=0
      maskB=0
      maskG=0
      maskR=0
      outputIndex=0
     }
    }]
   convertOutputs [{type=Scalar}, {type=Scalar}, {type=Scalar}]
  }
  n2 = If(A=NamedRerouteUsage(), "A > B"=Constant(), "A < B"=@n1) {constB=1.5}
  n3 = Max(A=@n1, B=@n2)
  n4 = Max(A=@n1, B=@n2)
  n5 = AppendVector(A=AppendVector(A=Subtract(A=@n1, B=@n3), B=Subtract(A=@n1, B=@n4)), B=Subtract(A=@n2, B=Max(A=@n1, B=@n1)))
  n6 = FunctionInput() {bUsePreviewValueAsDefault=true, inputName=Blend Radius, inputType=FunctionInput_Scalar, sortPriority=1, previewValue=<0.1 0.1 0.1 1>}
  n7 = Saturate(None=Add(A=@n5, B=@n6) {constB=0.2})
  MaterialExpressionMultiply_2 = Multiply(A=@n7, B=@n7)
  n8 = Convert(None=@MaterialExpressionMultiply_2) {
   nodeName=Break Float3
   convertInputs [{
     type=Vector3
     expressionInput {
      expression=&self:MaterialExpressionMultiply_2
      inputName=None
      mask=0
      maskA=0
      maskB=0
      maskG=0
      maskR=0
      outputIndex=0
     }
    }]
   convertOutputs [{type=Scalar}, {type=Scalar}, {type=Scalar}]
  }
  n9 = Divide(A=@MaterialExpressionMultiply_2, B=Add(A=Add(A=@n8, B=@n8), B=@n8))
  FunctionOutput(None=If(A=NamedRerouteUsage(), "A > B"=Constant3Vector() {constant=<rgba 0 0 1 0>}, "A < B"=@n9) {constB=2.5}) {outputName=Smooth, sortPriority=2}
  n10 = Saturate(None=@n5)
  FunctionOutput(None=If(A=NamedRerouteUsage(), "A > B"=Constant3Vector() {constant=<rgba 0 0 1 0>}, "A < B"=Ceil(None=@n10)) {constB=2.5}) {outputName=Flat, sortPriority=1}
  MaterialExpressionAdd_15 = Add(A=ScreenPosition().PixelPosition, B=ViewProperty() {property=MEVP_TemporalSampleIndex}.Property)
  n11 = Custom!(None=@MaterialExpressionAdd_15) {
   code="Mod(((uint)(Input.x) + 2 * (uint)(Input.y)), 5);"
   description=Grid
   outputType=CMOT_Float1
   showCode=true
   inputs [{inputName=Input, input {expression=&self:MaterialExpressionAdd_15}}]
  }
  n12 = Custom!(PixelPosition=@MaterialExpressionAdd_15) {
   code="float4x4 bayerMatrix4x4 = float4x4(\r\n    0.1, 8.0, 2.0, 10.0,\r\n    12.0, 4.0, 14.0, 6.0,\r\n    3.0, 11.0, 1.0, 9.0,\r\n    15.0, 7.0, 13.0, 5.0\r\n) / 16.0f;\r\n\r\n    // Get the pixel's integer coordinates\r\n    int2 coords = Mod((int2)(PixelPosition.xy),4);\r\n\r\n    // HLSL accesses matrices using [row][column] syntax\r\n    float dither = bayerMatrix4x4[coords.x][coords.y] * 2.0;\r\n\r\n    return dither;"
   description=Bayer Matrix 4x4
   showCode=true
   inputs [{
     inputName=PixelPosition
     input {expression=&self:MaterialExpressionAdd_15}
    }]
  }
  n13 = Add(A=@n11, B=@n12) {constB=0.5}
  n14 = ComponentMask(None=Add(A=Divide(A=@n13) {constB=6}) {constB=-0.5}) {r=true}
  n15 = Multiply(A=@n14) {constB=-1}
  n16 = Round(None=Add(A=ComponentMask(None=@n9) {g=true, r=true}, B=AppendVector(A=@n14, B=@n15)) {constB=-0.5})
  n17 = OneMinus(None=Max(A=ComponentMask(None=@n16) {r=true}, B=ComponentMask(None=@n16) {g=true}))
  FunctionOutput(None=If(A=NamedRerouteUsage(), "A > B"=Constant3Vector() {constant=<rgba 0 0 1 0>}, "A < B"=AppendVector(A=@n16, B=@n17)) {constB=2.5}) {outputName=Dithered}
  n18 = FunctionInput() {bUsePreviewValueAsDefault=true, inputName=Mode, inputType=FunctionInput_Scalar, previewValue=<1 1 1 1>}
  NamedRerouteDeclaration(None=@n18) {name=Mode, nodeColor=<rgba 0 1 0.625 1>}
 ]}
