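# Material function for 4-layer world-space UV mapping that converts tangent-space normals into world-space normals without using mesh tangents.
# It builds a reference vector (0, Mask blue, -1) and crosses it with the vertex normal, normalized, to get the X basis.
# The normal crossed with X gives Y, and the vertex normal is Z.
# Transform3x3Matrix applies this basis to the TangentNormals input to output World Normals.
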MaterialFunction /BRCoreMaterials/Materials/MaterialFunctions/MF_4Layer_WorldSpaceUVs_TangentToWorld
graph {nodes [
  n1 = Constant()
  n2 = ComponentMask(None=FunctionInput() {inputName=Mask, sortPriority=1}) {b=true}
  n3 = Constant() {r=-1}
  n4 = VertexNormalWS()
  n5 = Normalize(VectorInput=CrossProduct(A=MaterialFunctionCall(R=@n1, G=@n2, B=@n3) {materialFunction=&AppendMany}.RGB, B=@n4))
  FunctionOutput(None=MaterialFunctionCall(VectorToTransform=FunctionInput() {inputName=TangentNormals}, BasisX=@n5, BasisY=CrossProduct(A=@n4, B=@n5), BasisZ=VertexNormalWS()) {materialFunction=&Transform3x3Matrix}.Transformed Vector) {outputName=World Normals}
 ]}
