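# Material function that builds world-space projected UVs along X, Y and Z, outputting each set plus a mask-selected combination.
# Periodic-world position, tiled and divided by UV Scale, supplies the coordinates, with vertex normal signs flipping the horizontal axis per facing.
# The Mask's green channel blends Z toward Y, then red blends toward X.
# A separate UV Derivatives output rescales mesh texture coordinates by a constant divided through a reroute.
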MaterialFunction /BRCoreMaterials/Materials/MaterialFunctions/MF_4Layer_WorldSpaceUVs_UVs
graph {nodes [
  n1 = WorldPosition() {worldPositionShaderOffset=WPT_CameraRelativeNoOffsets}
  n2 = FunctionInput() {bUsePreviewValueAsDefault=true, inputName=UV Scale, inputType=FunctionInput_Scalar, sortPriority=1, previewValue=<650 650 650 650>}
  n3 = NamedRerouteDeclaration(None=@n2) {name=UV Scale, nodeColor=<rgba 0.9688 1 0 1>}
  n4 = Multiply(
   A=Divide(
    A=TransformPosition!(None=@n1.XYZ, Periodic World Tile Size=@n3) {transformSourceType=TRANSFORMPOSSOURCE_TranslatedWorld, transformType=TRANSFORMPOSSOURCE_PeriodicWorld}
    B=@n3
   )
   B=Constant3Vector() {constant=<rgba 1 1 -1 0>}
  )
  n5 = ComponentMask(None=@n4) {r=true}
  n6 = Sign(None=VertexNormalWS())
  n7 = ComponentMask(None=@n6) {b=true}
  n8 = ComponentMask(None=@n4) {g=true}
  n9 = FunctionOutput(None=AppendVector(A=Multiply(A=@n5, B=@n7), B=@n8)) {outputName=UV Z, sortPriority=4}
  n10 = ComponentMask(None=@n6) {g=true}
  n11 = ComponentMask(None=@n4) {b=true}
  n12 = FunctionOutput(None=AppendVector(A=Multiply(A=@n5, B=@n10), B=@n11)) {outputName=UV Y, sortPriority=3}
  n13 = FunctionInput() {bUsePreviewValueAsDefault=true, inputName=Mask, sortPriority=3, previewValue=<1 1 1 1>}
  n14 = ComponentMask(None=@n13) {g=true}
  n15 = ComponentMask(None=@n6) {r=true}
  n16 = AppendVector(A=Multiply(A=Multiply(A=@n8, B=@n15)) {constB=-1}, B=@n11)
  FunctionOutput(None=LinearInterpolate(
    A=LinearInterpolate(A=@n9, B=@n12, Alpha=@n14)
    B=FunctionOutput(None=@n16) {outputName=UV X, sortPriority=2}
    Alpha=ComponentMask(None=@n13) {r=true}
   )) {outputName=UVs}
  n17 = NamedRerouteUsage()
  FunctionOutput(None=Multiply(A=Divide(B=@n17) {constA=384}, B=TextureCoordinate())) {outputName=UV Derivatives, sortPriority=1}
 ]}
