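# Material function that adds a cover layer over input material attributes in the 4-layer material system, with a static switch at the output.
# Coverage masks draw on directional normals, vertex color, local bounds gradients, distance-field AO and per-layer masking, refined by height blends against displacement.
# Cover textures use world-space UVs with dithered axis masks, selecting samples per axis, with tangent normals converted to world space.
# Blended attributes and intermediate masks feed enum-driven debug views.
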MaterialFunction /BRCoreMaterials/Materials/MaterialFunctions/MF_4Layer_Cover
graph {expressions {
  MaterialExpressionAbs_1 {class=MaterialExpressionAbs, inputs [None {from=MaterialExpressionNamedRerouteUsage_85, output=""}], properties {}}
  MaterialExpressionAdd_0 {
   class=MaterialExpressionAdd
   inputs [
    A {from=MaterialExpressionScalarParameter_0}
    B {from=""}
   ]
  }
  MaterialExpressionAdd_1 {
   class=MaterialExpressionAdd
   inputs [
    A {from=MaterialExpressionIfThenElse_0}
    B {from=MaterialExpressionIfThenElse_1}
   ]
  }
  MaterialExpressionAdd_10 {
   class=MaterialExpressionAdd
   inputs [
    A {from=MaterialExpressionNamedRerouteUsage_45}
    B {from=MaterialExpressionNamedRerouteUsage_72}
   ]
  }
  MaterialExpressionAdd_11 {
   class=MaterialExpressionAdd
   inputs [
    A {from=MaterialExpressionAdd_10}
    B {from=MaterialExpressionNamedRerouteUsage_73}
   ]
  }
  MaterialExpressionAdd_12 {
   class=MaterialExpressionAdd
   inputs [
    A {from=MaterialExpressionAdd_24}
    B {from=MaterialExpressionMultiply_29}
   ]
  }
  MaterialExpressionAdd_13 {
   class=MaterialExpressionAdd
   inputs [
    A {from=MaterialExpressionDotProduct_0}
    B {from=MaterialExpressionMultiply_17}
   ]
  }
  MaterialExpressionAdd_14 {
   class=MaterialExpressionAdd
   inputs [
    A {from=MaterialExpressionDotProduct_1}
    B {from=MaterialExpressionMultiply_24}
   ]
  }
  MaterialExpressionAdd_15 {
   class=MaterialExpressionAdd
   inputs [
    A {from=MaterialExpressionNamedRerouteUsage_68}
    B {from=MaterialExpressionAdd_30}
   ]
  }
  MaterialExpressionAdd_16 {
   class=MaterialExpressionAdd
   inputs [
    A {from=MaterialExpressionConvert_4}
    B {from=MaterialExpressionConvert_4}
   ]
  }
  MaterialExpressionAdd_17 {
   class=MaterialExpressionAdd
   inputs [
    A {from=MaterialExpressionNamedRerouteUsage_74}
    B {from=MaterialExpressionAdd_34}
   ]
  }
  MaterialExpressionAdd_18 {
   class=MaterialExpressionAdd
   inputs [
    A {from=MaterialExpressionNamedRerouteUsage_46}
    B {from=MaterialExpressionNamedRerouteUsage_84}
   ]
  }
  MaterialExpressionAdd_19 {
   class=MaterialExpressionAdd
   inputs [
    A {from=MaterialExpressionAdd_11}
    B {from=MaterialExpressionNamedRerouteUsage_1}
   ]
  }
  MaterialExpressionAdd_2 {
   class=MaterialExpressionAdd
   inputs [
    A {from=MaterialExpressionAdd_1}
    B {from=MaterialExpressionIfThenElse_2}
   ]
  }
  MaterialExpressionAdd_20 {
   class=MaterialExpressionAdd
   inputs [
    A {from=MaterialExpressionNamedRerouteUsage_75}
    B {from=MaterialExpressionAdd_36}
   ]
  }
  MaterialExpressionAdd_21 {
   class=MaterialExpressionAdd
   inputs [
    A {from=MaterialExpressionPower_3}
    B {from=MaterialExpressionMultiply_16}
   ]
  }
  MaterialExpressionAdd_22 {
   class=MaterialExpressionAdd
   inputs [
    A {from=MaterialExpressionMultiply_33}
    B {from=MaterialExpressionMultiply_34}
   ]
  }
  MaterialExpressionAdd_23 {
   class=MaterialExpressionAdd
   inputs [
    A {from=MaterialExpressionConvert_2}
    B {from=MaterialExpressionConvert_2}
   ]
  }
  MaterialExpressionAdd_24 {
   class=MaterialExpressionAdd
   inputs [
    A {from=MaterialExpressionAdd_23}
    B {from=MaterialExpressionConvert_2}
   ]
  }
  MaterialExpressionAdd_25 {
   class=MaterialExpressionAdd
   inputs [
    A {from=MaterialExpressionDivide_0}
    B {from=MaterialExpressionConvert_8}
   ]
  }
  MaterialExpressionAdd_26 {
   class=MaterialExpressionAdd
   inputs [
    A {from=MaterialExpressionMultiply_22}
    B {from=MaterialExpressionConvert_9}
   ]
  }
  MaterialExpressionAdd_27 {
   class=MaterialExpressionAdd
   inputs [
    A {from=MaterialExpressionBounds_0, output=Half Extents}
    B {from=""}
   ]
   properties {constB=0.01}
  }
  MaterialExpressionAdd_28 {
   class=MaterialExpressionAdd
   inputs [
    A {from=MaterialExpressionAdd_18}
    B {from=MaterialExpressionNamedRerouteUsage_78}
   ]
  }
  MaterialExpressionAdd_29 {
   class=MaterialExpressionAdd
   inputs [
    A {from=MaterialExpressionNamedRerouteUsage_69}
    B {from=MaterialExpressionAdd_31}
   ]
  }
  MaterialExpressionAdd_3 {
   class=MaterialExpressionAdd
   inputs [
    A {from=MaterialExpressionMultiply_2}
    B {from=MaterialExpressionMultiply_5}
   ]
  }
  MaterialExpressionAdd_30 {
   class=MaterialExpressionAdd
   inputs [
    A {from=MaterialExpressionScalarParameter_36}
    B {from=""}
   ]
   properties {constB=-0.5}
  }
  MaterialExpressionAdd_31 {
   class=MaterialExpressionAdd
   inputs [
    A {from=MaterialExpressionScalarParameter_66}
    B {from=""}
   ]
   properties {constB=-0.5}
  }
  MaterialExpressionAdd_32 {
   class=MaterialExpressionAdd
   inputs [
    A {from=MaterialExpressionNamedRerouteUsage_83}
    B {from=MaterialExpressionAdd_35}
   ]
  }
  MaterialExpressionAdd_33 {
   class=MaterialExpressionAdd
   inputs [
    A {from=MaterialExpressionAdd_24}
    B {from=MaterialExpressionMultiply_35}
   ]
  }
  MaterialExpressionAdd_34 {
   class=MaterialExpressionAdd
   inputs [
    A {from=MaterialExpressionScalarParameter_54}
    B {from=""}
   ]
   properties {constB=-0.5}
  }
  MaterialExpressionAdd_35 {
   class=MaterialExpressionAdd
   inputs [
    A {from=MaterialExpressionScalarParameter_81}
    B {from=""}
   ]
   properties {constB=-0.5}
  }
  MaterialExpressionAdd_36 {
   class=MaterialExpressionAdd
   inputs [
    A {from=MaterialExpressionScalarParameter_60}
    B {from=""}
   ]
   properties {constB=-0.5}
  }
  MaterialExpressionAdd_37 {
   class=MaterialExpressionAdd
   inputs [
    A {from=MaterialExpressionScalarParameter_31}
    B {from=""}
   ]
   properties {constB=-0.5}
  }
  MaterialExpressionAdd_38 {
   class=MaterialExpressionAdd
   inputs [
    A {from=MaterialExpressionAdd_16}
    B {from=MaterialExpressionConvert_4}
   ]
  }
  MaterialExpressionAdd_4 {
   class=MaterialExpressionAdd
   inputs [
    A {from=MaterialExpressionMultiply_3}
    B {from=MaterialExpressionMultiply_8}
   ]
  }
  MaterialExpressionAdd_5 {
   class=MaterialExpressionAdd
   inputs [
    A {from=MaterialExpressionNamedRerouteUsage_56}
    B {from=MaterialExpressionScalarParameter_29}
   ]
  }
  MaterialExpressionAdd_6 {
   class=MaterialExpressionAdd
   inputs [
    A {from=MaterialExpressionIfThenElse_3}
    B {from=MaterialExpressionIfThenElse_4}
   ]
  }
  MaterialExpressionAdd_7 {
   class=MaterialExpressionAdd
   inputs [
    A {from=MaterialExpressionAdd_6}
    B {from=MaterialExpressionIfThenElse_5}
   ]
  }
  MaterialExpressionAdd_8 {
   class=MaterialExpressionAdd
   inputs [
    A {from=MaterialExpressionLinearInterpolate_4}
    B {from=MaterialExpressionMultiply_20}
   ]
  }
  MaterialExpressionAdd_9 {
   class=MaterialExpressionAdd
   inputs [
    A {from=MaterialExpressionNamedRerouteUsage_58}
    B {from=MaterialExpressionAdd_37}
   ]
  }
  MaterialExpressionBlendMaterialAttributes_0 {
   class=MaterialExpressionBlendMaterialAttributes
   inputs [
    A {from=MaterialExpressionBlendMaterialAttributes_1, output=""}
    B {from=MaterialExpressionSetMaterialAttributes_1, output=""}
    Alpha {from=MaterialExpressionNamedRerouteUsage_28, output=""}
   ]
   properties {}
  }
  MaterialExpressionBlendMaterialAttributes_1 {
   class=MaterialExpressionBlendMaterialAttributes
   inputs [
    A {from=MaterialExpressionNamedRerouteUsage_55, output=""}
    B {from=MaterialExpressionSetMaterialAttributes_0, output=""}
    Alpha {from=MaterialExpressionNamedRerouteUsage_27, output=""}
   ]
   properties {}
  }
  MaterialExpressionBlendMaterialAttributes_2 {
   class=MaterialExpressionBlendMaterialAttributes
   inputs [
    A {from=MaterialExpressionNamedRerouteUsage_54, output=""}
    B {from=MaterialExpressionSetMaterialAttributes_2, output=""}
    Alpha {from=MaterialExpressionNamedRerouteDeclaration_46, output=""}
   ]
   properties {}
  }
  MaterialExpressionBounds_0 {class=MaterialExpressionBounds, inputs [], properties {}}
  MaterialExpressionComponentMask_0 {class=MaterialExpressionComponentMask, inputs [None {from=MaterialExpressionNamedRerouteUsage_61}], properties {b=true}}
  MaterialExpressionComponentMask_1 {class=MaterialExpressionComponentMask, inputs [None {from=MaterialExpressionNamedRerouteUsage_61}], properties {r=true}}
  MaterialExpressionComponentMask_2 {class=MaterialExpressionComponentMask, inputs [None {from=MaterialExpressionNamedRerouteUsage_61}], properties {g=true}}
  MaterialExpressionComponentMask_4 {class=MaterialExpressionComponentMask, inputs [None {from=MaterialExpressionNamedRerouteUsage_6}], properties {b=true}}
  MaterialExpressionComponentMask_5 {class=MaterialExpressionComponentMask, inputs [None {from=MaterialExpressionNamedRerouteUsage_6}], properties {r=true}}
  MaterialExpressionComponentMask_6 {class=MaterialExpressionComponentMask, inputs [None {from=MaterialExpressionNamedRerouteUsage_6}], properties {g=true}}
  MaterialExpressionConstant3Vector_1 {class=MaterialExpressionConstant3Vector, properties {constant=<rgba 0 0 1 0>}}
  MaterialExpressionConstant_0 {class=MaterialExpressionConstant}
  MaterialExpressionConstant_1 {class=MaterialExpressionConstant}
  MaterialExpressionConstant_2 {class=MaterialExpressionConstant}
  MaterialExpressionConstant_3 {class=MaterialExpressionConstant}
  MaterialExpressionConstant_4 {class=MaterialExpressionConstant, properties {r=1}}
  MaterialExpressionConstant_5 {class=MaterialExpressionConstant, properties {r=1}}
  MaterialExpressionConstant_6 {class=MaterialExpressionConstant}
  MaterialExpressionConstant_7 {class=MaterialExpressionConstant}
  MaterialExpressionConstant_8 {class=MaterialExpressionConstant}
  MaterialExpressionConvert_0 {
   class=MaterialExpressionConvert
   inputs [
    None {from=MaterialExpressionScalarParameter_18}
    None {from=MaterialExpressionScalarParameter_19}
    None {from=MaterialExpressionScalarParameter_17}
   ]
   properties {
    nodeName=Append
    convertInputs [
     {
      type=Scalar
      expressionInput {
       expression=&self:MaterialExpressionScalarParameter_18
       inputName=None
       mask=0
       maskA=0
       maskB=0
       maskG=0
       maskR=0
       outputIndex=0
      }
     }
     {
      type=Scalar
      expressionInput {
       expression=&self:MaterialExpressionScalarParameter_19
       inputName=None
       mask=0
       maskA=0
       maskB=0
       maskG=0
       maskR=0
       outputIndex=0
      }
     }
     {
      type=Scalar
      expressionInput {
       expression=&self:MaterialExpressionScalarParameter_17
       inputName=None
       mask=0
       maskA=0
       maskB=0
       maskG=0
       maskR=0
       outputIndex=0
      }
     }
    ]
    convertOutputs [{type=Vector3}]
   }
  }
  MaterialExpressionConvert_1 {
   class=MaterialExpressionConvert
   inputs [
    None {from=MaterialExpressionScalarParameter_23}
    None {from=MaterialExpressionScalarParameter_24}
    None {from=MaterialExpressionScalarParameter_22}
   ]
   properties {
    nodeName=Append
    convertInputs [
     {
      type=Scalar
      expressionInput {
       expression=&self:MaterialExpressionScalarParameter_23
       inputName=None
       mask=0
       maskA=0
       maskB=0
       maskG=0
       maskR=0
       outputIndex=0
      }
     }
     {
      type=Scalar
      expressionInput {
       expression=&self:MaterialExpressionScalarParameter_24
       inputName=None
       mask=0
       maskA=0
       maskB=0
       maskG=0
       maskR=0
       outputIndex=0
      }
     }
     {
      type=Scalar
      expressionInput {
       expression=&self:MaterialExpressionScalarParameter_22
       inputName=None
       mask=0
       maskA=0
       maskB=0
       maskG=0
       maskR=0
       outputIndex=0
      }
     }
    ]
    convertOutputs [{type=Vector3}]
   }
  }
  MaterialExpressionConvert_2 {
   class=MaterialExpressionConvert
   inputs [None {from=MaterialExpressionAdd_22}]
   properties {
    nodeName=Break Float3
    convertInputs [{
      type=Vector3
      expressionInput {
       expression=&self:MaterialExpressionAdd_22
       inputName=None
       mask=0
       maskA=0
       maskB=0
       maskG=0
       maskR=0
       outputIndex=0
      }
     }]
    convertOutputs [{type=Scalar}, {type=Scalar}, {type=Scalar}]
   }
  }
  MaterialExpressionConvert_3 {
   class=MaterialExpressionConvert
   inputs [
    None {from=MaterialExpressionScalarParameter_21}
    None {from=MaterialExpressionScalarParameter_45}
    None {from=MaterialExpressionScalarParameter_52}
   ]
   properties {
    nodeName=Append
    convertInputs [
     {
      type=Scalar
      expressionInput {
       expression=&self:MaterialExpressionScalarParameter_21
       inputName=None
       mask=0
       maskA=0
       maskB=0
       maskG=0
       maskR=0
       outputIndex=0
      }
     }
     {
      type=Scalar
      expressionInput {
       expression=&self:MaterialExpressionScalarParameter_45
       inputName=None
       mask=0
       maskA=0
       maskB=0
       maskG=0
       maskR=0
       outputIndex=0
      }
     }
     {
      type=Scalar
      expressionInput {
       expression=&self:MaterialExpressionScalarParameter_52
       inputName=None
       mask=0
       maskA=0
       maskB=0
       maskG=0
       maskR=0
       outputIndex=0
      }
     }
    ]
    convertOutputs [{type=Vector3}]
   }
  }
  MaterialExpressionConvert_4 {
   class=MaterialExpressionConvert
   inputs [None {from=MaterialExpressionMultiply_21}]
   properties {
    nodeName=Break Float3
    convertInputs [{
      type=Vector3
      expressionInput {
       expression=&self:MaterialExpressionMultiply_21
       inputName=None
       mask=0
       maskA=0
       maskB=0
       maskG=0
       maskR=0
       outputIndex=0
      }
     }]
    convertOutputs [{type=Scalar}, {type=Scalar}, {type=Scalar}]
   }
  }
  MaterialExpressionConvert_5 {
   class=MaterialExpressionConvert
   inputs [
    None {from=MaterialExpressionScalarParameter_47}
    None {from=MaterialExpressionScalarParameter_48}
    None {from=MaterialExpressionScalarParameter_46}
   ]
   properties {
    convertInputs [
     {
      type=Scalar
      expressionInput {
       expression=&self:MaterialExpressionScalarParameter_47
       inputName=None
       mask=0
       maskA=0
       maskB=0
       maskG=0
       maskR=0
       outputIndex=0
      }
     }
     {
      type=Scalar
      expressionInput {
       expression=&self:MaterialExpressionScalarParameter_48
       inputName=None
       mask=0
       maskA=0
       maskB=0
       maskG=0
       maskR=0
       outputIndex=0
      }
     }
     {
      type=Scalar
      expressionInput {
       expression=&self:MaterialExpressionScalarParameter_46
       inputName=None
       mask=0
       maskA=0
       maskB=0
       maskG=0
       maskR=0
       outputIndex=0
      }
     }
    ]
    convertOutputs [{type=Vector3}]
   }
  }
  MaterialExpressionConvert_6 {
   class=MaterialExpressionConvert
   inputs [
    None {from=MaterialExpressionScalarParameter_50}
    None {from=MaterialExpressionScalarParameter_51}
    None {from=MaterialExpressionScalarParameter_49}
   ]
   properties {
    nodeName=Append
    convertInputs [
     {
      type=Scalar
      expressionInput {
       expression=&self:MaterialExpressionScalarParameter_50
       inputName=None
       mask=0
       maskA=0
       maskB=0
       maskG=0
       maskR=0
       outputIndex=0
      }
     }
     {
      type=Scalar
      expressionInput {
       expression=&self:MaterialExpressionScalarParameter_51
       inputName=None
       mask=0
       maskA=0
       maskB=0
       maskG=0
       maskR=0
       outputIndex=0
      }
     }
     {
      type=Scalar
      expressionInput {
       expression=&self:MaterialExpressionScalarParameter_49
       inputName=None
       mask=0
       maskA=0
       maskB=0
       maskG=0
       maskR=0
       outputIndex=0
      }
     }
    ]
    convertOutputs [{type=Vector3}]
   }
  }
  MaterialExpressionConvert_8 {
   class=MaterialExpressionConvert
   inputs [
    None {from=MaterialExpressionScalarParameter_74}
    None {from=MaterialExpressionScalarParameter_75}
    None {from=MaterialExpressionScalarParameter_73}
   ]
   properties {
    nodeName=Append
    convertInputs [
     {
      type=Scalar
      expressionInput {
       expression=&self:MaterialExpressionScalarParameter_74
       inputName=None
       mask=0
       maskA=0
       maskB=0
       maskG=0
       maskR=0
       outputIndex=0
      }
     }
     {
      type=Scalar
      expressionInput {
       expression=&self:MaterialExpressionScalarParameter_75
       inputName=None
       mask=0
       maskA=0
       maskB=0
       maskG=0
       maskR=0
       outputIndex=0
      }
     }
     {
      type=Scalar
      expressionInput {
       expression=&self:MaterialExpressionScalarParameter_73
       inputName=None
       mask=0
       maskA=0
       maskB=0
       maskG=0
       maskR=0
       outputIndex=0
      }
     }
    ]
    convertOutputs [{type=Vector3}]
   }
  }
  MaterialExpressionConvert_9 {
   class=MaterialExpressionConvert
   inputs [
    None {from=MaterialExpressionScalarParameter_77}
    None {from=MaterialExpressionScalarParameter_78}
    None {from=MaterialExpressionScalarParameter_76}
   ]
   properties {
    nodeName=Append
    convertInputs [
     {
      type=Scalar
      expressionInput {
       expression=&self:MaterialExpressionScalarParameter_77
       inputName=None
       mask=0
       maskA=0
       maskB=0
       maskG=0
       maskR=0
       outputIndex=0
      }
     }
     {
      type=Scalar
      expressionInput {
       expression=&self:MaterialExpressionScalarParameter_78
       inputName=None
       mask=0
       maskA=0
       maskB=0
       maskG=0
       maskR=0
       outputIndex=0
      }
     }
     {
      type=Scalar
      expressionInput {
       expression=&self:MaterialExpressionScalarParameter_76
       inputName=None
       mask=0
       maskA=0
       maskB=0
       maskG=0
       maskR=0
       outputIndex=0
      }
     }
    ]
    convertOutputs [{type=Vector3}]
   }
  }
  MaterialExpressionDDX_0 {class=MaterialExpressionDDX, inputs [Value {from=MaterialExpressionMaterialFunctionCall_3, output=UV Derivatives}], properties {}}
  MaterialExpressionDDX_1 {class=MaterialExpressionDDX, inputs [Value {from=MaterialExpressionMaterialFunctionCall_10, output=UV Derivatives}], properties {}}
  MaterialExpressionDDY_0 {class=MaterialExpressionDDY, inputs [Value {from=MaterialExpressionMaterialFunctionCall_3, output=UV Derivatives}], properties {}}
  MaterialExpressionDDY_1 {class=MaterialExpressionDDY, inputs [Value {from=MaterialExpressionMaterialFunctionCall_10, output=UV Derivatives}], properties {}}
  MaterialExpressionDistanceFieldApproxAO_0 {
   class=MaterialExpressionDistanceFieldApproxAO
   inputs [
    World Position {from="", output=""}
    Normal {from=MaterialExpressionLinearInterpolate_7, output=""}
    BaseDistance {from=MaterialExpressionScalarParameter_35, output=""}
    Radius {from="", output=""}
   ]
   properties {}
  }
  MaterialExpressionDivide_0 {
   class=MaterialExpressionDivide
   inputs [
    A {from=MaterialExpressionLocalPosition_0, output=XYZ}
    B {from=MaterialExpressionAdd_27}
   ]
  }
  MaterialExpressionDotProduct_0 {
   class=MaterialExpressionDotProduct
   inputs [
    A {from=MaterialExpressionAdd_3, output=""}
    B {from=MaterialExpressionConstant_4, output=""}
   ]
   properties {}
  }
  MaterialExpressionDotProduct_1 {
   class=MaterialExpressionDotProduct
   inputs [
    A {from=MaterialExpressionAdd_4, output=""}
    B {from=MaterialExpressionConstant_5, output=""}
   ]
   properties {}
  }
  MaterialExpressionFunctionInput_0 {
   class=MaterialExpressionFunctionInput
   inputs [Preview {from="", output=""}]
   properties {inputName=Material, inputType=FunctionInput_MaterialAttributes}
  }
  MaterialExpressionFunctionInput_1 {
   class=MaterialExpressionFunctionInput
   inputs [Preview {from="", output=""}]
   properties {bUsePreviewValueAsDefault=true, inputName=Displacement Texture, inputType=FunctionInput_Scalar, sortPriority=1, previewValue=<1 1 1 1>}
  }
  MaterialExpressionFunctionInput_2 {
   class=MaterialExpressionFunctionInput
   inputs [Preview {from="", output=""}]
   properties {bUsePreviewValueAsDefault=true, inputName=Layer Mask, inputType=FunctionInput_Vector4, sortPriority=2, previewValue=<1 1 1 1>}
  }
  MaterialExpressionFunctionInput_3 {
   class=MaterialExpressionFunctionInput
   inputs [Preview {from="", output=""}]
   properties {bUsePreviewValueAsDefault=true, inputName=Layer Texture Mask, inputType=FunctionInput_Scalar, sortPriority=3, previewValue=<1 1 1 1>}
  }
  MaterialExpressionFunctionOutput_0 {class=MaterialExpressionFunctionOutput, inputs [None {from=MaterialExpressionStaticSwitchParameter_0, output=""}]}
  MaterialExpressionGetMaterialAttributes_0 {
   class=MaterialExpressionGetMaterialAttributes
   inputs [None {from=MaterialExpressionFunctionInput_0, output=""}]
   properties {attributeGetTypes [
     #1
     #2
     #3
     #4
     #5
    ]}
  }
  MaterialExpressionGetMaterialAttributes_1 {
   class=MaterialExpressionGetMaterialAttributes
   inputs [None {from=MaterialExpressionNamedRerouteUsage_38, output=""}]
   properties {attributeGetTypes [#3, #6]}
  }
  MaterialExpressionGetMaterialAttributes_2 {
   class=MaterialExpressionGetMaterialAttributes
   inputs [None {from=MaterialExpressionNamedRerouteUsage_67, output=""}]
   properties {attributeGetTypes [#3, #6]}
  }
  MaterialExpressionIfThenElse_0 {
   class=MaterialExpressionIfThenElse
   inputs [
    Condition {from=MaterialExpressionComponentMask_5, output=""}
    True {from=MaterialExpressionMultiply_7, output=""}
    False {from=MaterialExpressionConstant_0, output=""}
   ]
   properties {}
  }
  MaterialExpressionIfThenElse_1 {
   class=MaterialExpressionIfThenElse
   inputs [
    Condition {from=MaterialExpressionComponentMask_6, output=""}
    True {from=MaterialExpressionMultiply_9, output=""}
    False {from=MaterialExpressionConstant_1, output=""}
   ]
   properties {}
  }
  MaterialExpressionIfThenElse_2 {
   class=MaterialExpressionIfThenElse
   inputs [
    Condition {from=MaterialExpressionComponentMask_4, output=""}
    True {from=MaterialExpressionMultiply_11, output=""}
    False {from=MaterialExpressionConstant_2, output=""}
   ]
   properties {}
  }
  MaterialExpressionIfThenElse_3 {
   class=MaterialExpressionIfThenElse
   inputs [
    Condition {from=MaterialExpressionComponentMask_1, output=""}
    True {from=MaterialExpressionMultiply_12, output=""}
    False {from=MaterialExpressionConstant_7, output=""}
   ]
   properties {}
  }
  MaterialExpressionIfThenElse_4 {
   class=MaterialExpressionIfThenElse
   inputs [
    Condition {from=MaterialExpressionComponentMask_2, output=""}
    True {from=MaterialExpressionMultiply_13, output=""}
    False {from=MaterialExpressionConstant_6, output=""}
   ]
   properties {}
  }
  MaterialExpressionIfThenElse_5 {
   class=MaterialExpressionIfThenElse
   inputs [
    Condition {from=MaterialExpressionComponentMask_0, output=""}
    True {from=MaterialExpressionMultiply_14, output=""}
    False {from=MaterialExpressionConstant_3, output=""}
   ]
   properties {}
  }
  MaterialExpressionLinearInterpolate_0 {
   class=MaterialExpressionLinearInterpolate
   inputs [
    A {from=MaterialExpressionConstant3Vector_1}
    B {from=MaterialExpressionTextureSampleParameter2D_1, output=RGB}
    Alpha {from=MaterialExpressionScalarParameter_3}
   ]
  }
  MaterialExpressionLinearInterpolate_1 {
   class=MaterialExpressionLinearInterpolate
   inputs [
    A {from=MaterialExpressionScalarParameter_4}
    B {from=MaterialExpressionScalarParameter_5}
    Alpha {from=MaterialExpressionTextureSampleParameter2D_2, output=Specular}
   ]
  }
  MaterialExpressionLinearInterpolate_2 {
   class=MaterialExpressionLinearInterpolate
   inputs [
    A {from=MaterialExpressionScalarParameter_7}
    B {from=MaterialExpressionScalarParameter_8}
    Alpha {from=MaterialExpressionTextureSampleParameter2D_2, output=Roughness}
   ]
  }
  MaterialExpressionLinearInterpolate_3 {
   class=MaterialExpressionLinearInterpolate
   inputs [
    A {from=MaterialExpressionVertexNormalWS_1}
    B {from=MaterialExpressionNamedRerouteUsage_42}
    Alpha {from=MaterialExpressionScalarParameter_16}
   ]
  }
  MaterialExpressionLinearInterpolate_4 {
   class=MaterialExpressionLinearInterpolate
   inputs [
    A {from=MaterialExpressionNamedRerouteUsage_57}
    B {from=MaterialExpressionAdd_5}
    Alpha {from=MaterialExpressionScalarParameter_28}
   ]
  }
  MaterialExpressionLinearInterpolate_5 {
   class=MaterialExpressionLinearInterpolate
   inputs [
    A {from=MaterialExpressionScalarParameter_40}
    B {from=MaterialExpressionScalarParameter_41}
    Alpha {from=MaterialExpressionGetMaterialAttributes_1, output=Specular}
   ]
  }
  MaterialExpressionLinearInterpolate_6 {
   class=MaterialExpressionLinearInterpolate
   inputs [
    A {from=MaterialExpressionScalarParameter_42}
    B {from=MaterialExpressionScalarParameter_43}
    Alpha {from=MaterialExpressionGetMaterialAttributes_1, output=Roughness}
   ]
  }
  MaterialExpressionLinearInterpolate_7 {
   class=MaterialExpressionLinearInterpolate
   inputs [
    A {from=MaterialExpressionVertexNormalWS_0}
    B {from=MaterialExpressionNamedRerouteUsage_53}
    Alpha {from=MaterialExpressionScalarParameter_58}
   ]
  }
  MaterialExpressionLocalPosition_0 {class=MaterialExpressionLocalPosition, inputs [], properties {}}
  MaterialExpressionMaterialFunctionCall_0 {
   class=MaterialExpressionMaterialFunctionCall
   inputs [
    Background {from=MaterialExpressionNamedRerouteUsage_79}
    Background Low {from=MaterialExpressionScalarParameter_10}
    Background High {from=MaterialExpressionScalarParameter_9}
    Foreground {from=MaterialExpressionNamedRerouteUsage_26}
    Foreground Low {from=MaterialExpressionScalarParameter_14}
    Foreground High {from=MaterialExpressionScalarParameter_13}
    Contrast {from=MaterialExpressionScalarParameter_15}
    Mask {from=MaterialExpressionNamedRerouteUsage_32}
   ]
   properties {materialFunction=&/MF_4Layer_HeightBlend}
  }
  MaterialExpressionMaterialFunctionCall_1 {
   class=MaterialExpressionMaterialFunctionCall
   inputs [
    Background {from=MaterialExpressionNamedRerouteUsage_35}
    Background Low {from=MaterialExpressionScalarParameter_10}
    Background High {from=MaterialExpressionScalarParameter_9}
    Foreground {from=MaterialExpressionNamedRerouteUsage_34}
    Foreground Low {from=MaterialExpressionScalarParameter_14}
    Foreground High {from=MaterialExpressionScalarParameter_13}
    Contrast {from=MaterialExpressionScalarParameter_11}
    Mask {from=MaterialExpressionNamedRerouteUsage_43}
   ]
   properties {materialFunction=&MF_4Layer_HeightBlend}
  }
  MaterialExpressionMaterialFunctionCall_10 {
   class=MaterialExpressionMaterialFunctionCall
   inputs [
    UV Scale {from=MaterialExpressionScalarParameter_39}
    Mask {from=MaterialExpressionNamedRerouteUsage_49}
   ]
   properties {materialFunction=&/MF_4Layer_WorldSpaceUVs_UVs}
  }
  MaterialExpressionMaterialFunctionCall_11 {
   class=MaterialExpressionMaterialFunctionCall
   inputs [
    Enum {from=MaterialExpressionScalarParameter_79}
    Material {from=MaterialExpressionNamedRerouteUsage_77}
    Debug 1 {from=MaterialExpressionNamedRerouteUsage_81}
    Debug 2 {from=""}
    Debug 3 {from=""}
    Debug 4 {from=""}
    Debug 5 {from=""}
   ]
   properties {materialFunction=&/MF_4Layer_Debug}
  }
  MaterialExpressionMaterialFunctionCall_13 {
   class=MaterialExpressionMaterialFunctionCall
   inputs [
    Enum {from=MaterialExpressionScalarParameter_56}
    Material {from=MaterialExpressionNamedRerouteUsage_80}
    Debug 1 {from=MaterialExpressionAbs_1}
    Debug 2 {from=""}
    Debug 3 {from=""}
    Debug 4 {from=""}
    Debug 5 {from=""}
   ]
   properties {materialFunction=&MF_4Layer_Debug}
  }
  MaterialExpressionMaterialFunctionCall_2 {
   class=MaterialExpressionMaterialFunctionCall
   inputs [
    Enum {from=MaterialExpressionScalarParameter_2}
    Material {from=MaterialExpressionNamedRerouteUsage_19}
    Debug 1 {from=MaterialExpressionNamedRerouteUsage_30}
    Debug 2 {from=MaterialExpressionNamedRerouteUsage_39}
    Debug 3 {from=MaterialExpressionNamedRerouteUsage_33}
    Debug 4 {from=MaterialExpressionNamedRerouteUsage_36}
    Debug 5 {from=MaterialExpressionNamedRerouteUsage_40}
   ]
   properties {materialFunction=&MF_4Layer_Debug}
  }
  MaterialExpressionMaterialFunctionCall_3 {
   class=MaterialExpressionMaterialFunctionCall
   inputs [
    UV Scale {from=MaterialExpressionScalarParameter_33}
    Mask {from=MaterialExpressionNamedRerouteDeclaration_11}
   ]
   properties {materialFunction=&MF_4Layer_WorldSpaceUVs_UVs}
  }
  MaterialExpressionMaterialFunctionCall_4 {
   class=MaterialExpressionMaterialFunctionCall
   inputs [
    RGBA Mask {from=MaterialExpressionNamedRerouteUsage_9}
    L1 {from=MaterialExpressionScalarParameter_12}
    L2 {from=MaterialExpressionScalarParameter_25}
    L3 {from=MaterialExpressionScalarParameter_26}
    L4 {from=MaterialExpressionScalarParameter_27}
   ]
   properties {materialFunction=&/MF_4Layer_ScalarMasking}
  }
  MaterialExpressionMaterialFunctionCall_5 {
   class=MaterialExpressionMaterialFunctionCall
   inputs [
    Mode {from=MaterialExpressionAdd_0}
    Blend Radius {from=MaterialExpressionScalarParameter_6}
   ]
   properties {materialFunction=&/MF_4Layer_WorldSpaceUVs_Masks}
  }
  MaterialExpressionMaterialFunctionCall_6 {
   class=MaterialExpressionMaterialFunctionCall
   inputs [
    TangentNormals {from=MaterialExpressionLinearInterpolate_0}
    Mask {from=MaterialExpressionNamedRerouteUsage_22}
   ]
   properties {materialFunction=&/MF_4Layer_WorldSpaceUVs_TangentToWorld}
  }
  MaterialExpressionMaterialFunctionCall_7 {
   class=MaterialExpressionMaterialFunctionCall
   inputs [
    Enum {from=MaterialExpressionScalarParameter_44}
    Material {from=MaterialExpressionNamedRerouteUsage_60}
    Debug 1 {from=MaterialExpressionGetMaterialAttributes_2, output=Specular}
    Debug 2 {from=MaterialExpressionGetMaterialAttributes_2, output=Specular}
    Debug 3 {from=MaterialExpressionMultiply_23}
    Debug 4 {from=""}
    Debug 5 {from=""}
   ]
   properties {materialFunction=&MF_4Layer_Debug}
  }
  MaterialExpressionMaterialFunctionCall_8 {
   class=MaterialExpressionMaterialFunctionCall
   inputs [
    Enum {from=MaterialExpressionScalarParameter_55}
    Material {from=MaterialExpressionNamedRerouteUsage_76}
    Debug 1 {from=MaterialExpressionNamedRerouteUsage_65}
    Debug 2 {from=""}
    Debug 3 {from=""}
    Debug 4 {from=""}
    Debug 5 {from=""}
   ]
   properties {materialFunction=&MF_4Layer_Debug}
  }
  MaterialExpressionMaterialFunctionCall_9 {
   class=MaterialExpressionMaterialFunctionCall
   inputs [
    Enum {from=MaterialExpressionScalarParameter_20}
    Material {from=MaterialExpressionNamedRerouteUsage_44}
    Debug 1 {from=MaterialExpressionNamedRerouteUsage_47}
    Debug 2 {from=""}
    Debug 3 {from=""}
    Debug 4 {from=""}
    Debug 5 {from=""}
   ]
   properties {materialFunction=&MF_4Layer_Debug}
  }
  MaterialExpressionMultiply_0 {
   class=MaterialExpressionMultiply
   inputs [
    A {from=MaterialExpressionNamedRerouteUsage_7}
    B {from=MaterialExpressionScalarParameter_1}
   ]
  }
  MaterialExpressionMultiply_1 {
   class=MaterialExpressionMultiply
   inputs [
    A {from=MaterialExpressionTextureSampleParameter2D_0, output=RGB}
    B {from=MaterialExpressionVectorParameter_0, output=RGB}
   ]
  }
  MaterialExpressionMultiply_10 {
   class=MaterialExpressionMultiply
   inputs [
    A {from=MaterialExpressionMaterialFunctionCall_0, output=Mask}
    B {from=MaterialExpressionNamedRerouteUsage_31}
   ]
  }
  MaterialExpressionMultiply_11 {
   class=MaterialExpressionMultiply
   inputs [
    A {from=MaterialExpressionTextureSample_3, output=R}
    B {from=MaterialExpressionComponentMask_4}
   ]
  }
  MaterialExpressionMultiply_12 {
   class=MaterialExpressionMultiply
   inputs [
    A {from=MaterialExpressionTextureSample_4, output=R}
    B {from=MaterialExpressionComponentMask_1}
   ]
  }
  MaterialExpressionMultiply_13 {
   class=MaterialExpressionMultiply
   inputs [
    A {from=MaterialExpressionTextureSample_5, output=R}
    B {from=MaterialExpressionComponentMask_2}
   ]
  }
  MaterialExpressionMultiply_14 {
   class=MaterialExpressionMultiply
   inputs [
    A {from=MaterialExpressionTextureSample_6, output=R}
    B {from=MaterialExpressionComponentMask_0}
   ]
  }
  MaterialExpressionMultiply_15 {
   class=MaterialExpressionMultiply
   inputs [
    A {from=MaterialExpressionMaterialFunctionCall_1, output=Mask}
    B {from=MaterialExpressionNamedRerouteUsage_37}
   ]
  }
  MaterialExpressionMultiply_16 {
   class=MaterialExpressionMultiply
   inputs [
    A {from=MaterialExpressionAdd_20}
    B {from=MaterialExpressionScalarParameter_59}
   ]
  }
  MaterialExpressionMultiply_17 {
   class=MaterialExpressionMultiply
   inputs [
    A {from=MaterialExpressionAdd_15}
    B {from=MaterialExpressionScalarParameter_34}
   ]
  }
  MaterialExpressionMultiply_18 {
   class=MaterialExpressionMultiply
   inputs [
    A {from=MaterialExpressionSaturate_3}
    B {from=MaterialExpressionScalarParameter_30}
   ]
  }
  MaterialExpressionMultiply_2 {
   class=MaterialExpressionMultiply
   inputs [
    A {from=MaterialExpressionSaturate_0}
    B {from=MaterialExpressionConvert_0}
   ]
  }
  MaterialExpressionMultiply_20 {
   class=MaterialExpressionMultiply
   inputs [
    A {from=MaterialExpressionAdd_9}
    B {from=MaterialExpressionScalarParameter_32}
   ]
  }
  MaterialExpressionMultiply_21 {
   class=MaterialExpressionMultiply
   inputs [
    A {from=MaterialExpressionVertexColor_0}
    B {from=MaterialExpressionConvert_3}
   ]
  }
  MaterialExpressionMultiply_22 {
   class=MaterialExpressionMultiply
   inputs [
    A {from=MaterialExpressionDivide_0}
    B {from=""}
   ]
   properties {constB=-1}
  }
  MaterialExpressionMultiply_23 {
   class=MaterialExpressionMultiply
   inputs [
    A {from=MaterialExpressionNamedRerouteUsage_70}
    B {from=MaterialExpressionOneMinus_1}
   ]
  }
  MaterialExpressionMultiply_24 {
   class=MaterialExpressionMultiply
   inputs [
    A {from=MaterialExpressionAdd_29}
    B {from=MaterialExpressionScalarParameter_65}
   ]
  }
  MaterialExpressionMultiply_29 {
   class=MaterialExpressionMultiply
   inputs [
    A {from=MaterialExpressionAdd_17}
    B {from=MaterialExpressionScalarParameter_53}
   ]
  }
  MaterialExpressionMultiply_3 {
   class=MaterialExpressionMultiply
   inputs [
    A {from=MaterialExpressionSaturate_1}
    B {from=MaterialExpressionConvert_0}
   ]
  }
  MaterialExpressionMultiply_33 {
   class=MaterialExpressionMultiply
   inputs [
    A {from=MaterialExpressionSaturate_12}
    B {from=MaterialExpressionConvert_5}
   ]
  }
  MaterialExpressionMultiply_34 {
   class=MaterialExpressionMultiply
   inputs [
    A {from=MaterialExpressionSaturate_13}
    B {from=MaterialExpressionConvert_6}
   ]
  }
  MaterialExpressionMultiply_35 {
   class=MaterialExpressionMultiply
   inputs [
    A {from=MaterialExpressionAdd_32}
    B {from=MaterialExpressionScalarParameter_80}
   ]
  }
  MaterialExpressionMultiply_36 {
   class=MaterialExpressionMultiply
   inputs [
    A {from=MaterialExpressionAdd_21}
    B {from=MaterialExpressionScalarParameter_82}
   ]
  }
  MaterialExpressionMultiply_4 {
   class=MaterialExpressionMultiply
   inputs [
    A {from=MaterialExpressionLinearInterpolate_3}
    B {from=""}
   ]
   properties {constB=-1}
  }
  MaterialExpressionMultiply_5 {
   class=MaterialExpressionMultiply
   inputs [
    A {from=MaterialExpressionSaturate_4}
    B {from=MaterialExpressionConvert_1}
   ]
  }
  MaterialExpressionMultiply_6 {
   class=MaterialExpressionMultiply
   inputs [
    A {from=MaterialExpressionVertexNormalWS_2}
    B {from=""}
   ]
   properties {constB=-1}
  }
  MaterialExpressionMultiply_7 {
   class=MaterialExpressionMultiply
   inputs [
    A {from=MaterialExpressionTextureSample_1, output=R}
    B {from=MaterialExpressionComponentMask_5}
   ]
  }
  MaterialExpressionMultiply_8 {
   class=MaterialExpressionMultiply
   inputs [
    A {from=MaterialExpressionSaturate_5}
    B {from=MaterialExpressionConvert_1}
   ]
  }
  MaterialExpressionMultiply_9 {
   class=MaterialExpressionMultiply
   inputs [
    A {from=MaterialExpressionTextureSample_2, output=R}
    B {from=MaterialExpressionComponentMask_6}
   ]
  }
  MaterialExpressionNamedRerouteDeclaration_0 {
   class=MaterialExpressionNamedRerouteDeclaration
   inputs [None {from=MaterialExpressionGetMaterialAttributes_0, output=MaterialAttributes}]
   properties {name=Material Input, nodeColor=<rgba 1 0.4922 0 1>}
  }
  MaterialExpressionNamedRerouteDeclaration_1 {
   class=MaterialExpressionNamedRerouteDeclaration
   inputs [None {from=MaterialExpressionFunctionInput_2}]
   properties {name=Layer Mask, nodeColor=<rgba 1 0.001 0.4767 1>}
  }
  MaterialExpressionNamedRerouteDeclaration_10 {
   class=MaterialExpressionNamedRerouteDeclaration
   inputs [None {from=MaterialExpressionMaterialFunctionCall_3, output=UVs}]
   properties {name=Texture UVs, nodeColor=<rgba 0 1 0.5547 1>}
  }
  MaterialExpressionNamedRerouteDeclaration_11 {
   class=MaterialExpressionNamedRerouteDeclaration
   inputs [None {from=MaterialExpressionMaterialFunctionCall_5, output=Dithered}]
   properties {name=Dithered Mask, nodeColor=<rgba 0 1 0.3672 1>}
  }
  MaterialExpressionNamedRerouteDeclaration_12 {
   class=MaterialExpressionNamedRerouteDeclaration
   inputs [None {from=MaterialExpressionMaterialFunctionCall_5, output=Smooth}]
   properties {name=WorldSpace Smooth Mask, nodeColor=<rgba 1 0 0.3047 1>}
  }
  MaterialExpressionNamedRerouteDeclaration_13 {
   class=MaterialExpressionNamedRerouteDeclaration
   inputs [None {from=MaterialExpressionMultiply_1}]
   properties {name=Cover Base Color, nodeColor=<rgba 0.5703 1 0 1>}
  }
  MaterialExpressionNamedRerouteDeclaration_14 {
   class=MaterialExpressionNamedRerouteDeclaration
   inputs [None {from=MaterialExpressionMaterialFunctionCall_6, output=World Normals}]
   properties {name=Cover Normal, nodeColor=<rgba 1 0 0.8906 1>}
  }
  MaterialExpressionNamedRerouteDeclaration_15 {
   class=MaterialExpressionNamedRerouteDeclaration
   inputs [None {from=MaterialExpressionLinearInterpolate_1}]
   properties {name=Cover Specular, nodeColor=<rgba 0 1 0.7891 1>}
  }
  MaterialExpressionNamedRerouteDeclaration_16 {
   class=MaterialExpressionNamedRerouteDeclaration
   inputs [None {from=MaterialExpressionLinearInterpolate_2}]
   properties {name=Cover Roughness, nodeColor=<rgba 1 0.4688 0 1>}
  }
  MaterialExpressionNamedRerouteDeclaration_17 {
   class=MaterialExpressionNamedRerouteDeclaration
   inputs [None {from=MaterialExpressionTextureSampleParameter2D_2, output=Metal}]
   properties {name=Cover Metallic, nodeColor=<rgba 0.1484 0 1 1>}
  }
  MaterialExpressionNamedRerouteDeclaration_18 {
   class=MaterialExpressionNamedRerouteDeclaration
   inputs [None {from=MaterialExpressionGetMaterialAttributes_0, output=Opacity}]
   properties {name=Opacity Input, nodeColor=<rgba 1 0.0107 0.3163 1>}
  }
  MaterialExpressionNamedRerouteDeclaration_19 {
   class=MaterialExpressionNamedRerouteDeclaration
   inputs [None {from=MaterialExpressionAdd_2}]
   properties {name=Cover Displacement, nodeColor=<rgba 1 0 0.2109 1>}
  }
  MaterialExpressionNamedRerouteDeclaration_2 {
   class=MaterialExpressionNamedRerouteDeclaration
   inputs [None {from=MaterialExpressionGetMaterialAttributes_0, output=BaseColor}]
   properties {name=BaseColor Input, nodeColor=<rgba 1 0 0.2109 1>}
  }
  MaterialExpressionNamedRerouteDeclaration_20 {
   class=MaterialExpressionNamedRerouteDeclaration
   inputs [None {from=MaterialExpressionTextureSampleParameter2D_0, output=Mask}]
   properties {name=Cover Opacity, nodeColor=<rgba 1 0.4205 0 1>}
  }
  MaterialExpressionNamedRerouteDeclaration_21 {
   class=MaterialExpressionNamedRerouteDeclaration
   inputs [None {from=MaterialExpressionMultiply_10}]
   properties {name=Cover Blend Mask, nodeColor=<rgba 0.2891 0 1 1>}
  }
  MaterialExpressionNamedRerouteDeclaration_22 {
   class=MaterialExpressionNamedRerouteDeclaration
   inputs [None {from=MaterialExpressionMultiply_15}]
   properties {name=Cover Blend Mask Displacement, nodeColor=<rgba 0.8281 0 1 1>}
  }
  MaterialExpressionNamedRerouteDeclaration_23 {
   class=MaterialExpressionNamedRerouteDeclaration
   inputs [None {from=MaterialExpressionMaterialFunctionCall_4, output=Result}]
   properties {name=Cover Layer Mask, nodeColor=<rgba 1 0 0.375 1>}
  }
  MaterialExpressionNamedRerouteDeclaration_24 {
   class=MaterialExpressionNamedRerouteDeclaration
   inputs [None {from=MaterialExpressionGetMaterialAttributes_0, output=Displacement}]
   properties {name=Displacement Input, nodeColor=<rgba 0 0.5781 1 1>}
  }
  MaterialExpressionNamedRerouteDeclaration_25 {
   class=MaterialExpressionNamedRerouteDeclaration
   inputs [None {from=MaterialExpressionMaterialFunctionCall_2, output=Result}]
   properties {name=Debug Blend, nodeColor=<rgba 0.0312 0 1 1>}
  }
  MaterialExpressionNamedRerouteDeclaration_26 {
   class=MaterialExpressionNamedRerouteDeclaration
   inputs [None {from=MaterialExpressionFunctionInput_1}]
   properties {name=Displacement Texture, nodeColor=<rgba 0.4766 0 1 1>}
  }
  MaterialExpressionNamedRerouteDeclaration_27 {
   class=MaterialExpressionNamedRerouteDeclaration
   inputs [None {from=MaterialExpressionMaterialFunctionCall_0, output=Background Heightmap}]
   properties {name=Background Heightmap, nodeColor=<rgba 0.6641 0 1 1>}
  }
  MaterialExpressionNamedRerouteDeclaration_28 {
   class=MaterialExpressionNamedRerouteDeclaration
   inputs [None {from=MaterialExpressionMaterialFunctionCall_0, output=Foreground Heightmap}]
   properties {name=Foreground HeightMap, nodeColor=<rgba 0 1 0.3438 1>}
  }
  MaterialExpressionNamedRerouteDeclaration_29 {
   class=MaterialExpressionNamedRerouteDeclaration
   inputs [None {from=MaterialExpressionStaticSwitch_0}]
   properties {name=Cover Result, nodeColor=<rgba 1 0 0.8672 1>}
  }
  MaterialExpressionNamedRerouteDeclaration_3 {
   class=MaterialExpressionNamedRerouteDeclaration
   inputs [None {from=MaterialExpressionGetMaterialAttributes_0, output=Normal}]
   properties {name=Normal Input, nodeColor=<rgba 0 0.5312 1 1>}
  }
  MaterialExpressionNamedRerouteDeclaration_30 {
   class=MaterialExpressionNamedRerouteDeclaration
   inputs [None {from=MaterialExpressionAdd_19}]
   properties {name=Mask Before Blend, nodeColor=<rgba 0 0.0156 1 1>}
  }
  MaterialExpressionNamedRerouteDeclaration_31 {
   class=MaterialExpressionNamedRerouteDeclaration
   inputs [None {from=MaterialExpressionAdd_28}]
   properties {name=Mask Before Blend Displacement, nodeColor=<rgba 0 0.0156 1 1>}
  }
  MaterialExpressionNamedRerouteDeclaration_32 {
   class=MaterialExpressionNamedRerouteDeclaration
   inputs [None {from=MaterialExpressionMaterialFunctionCall_9, output=Result}]
   properties {name=Debug Mask Normals, nodeColor=<rgba 0.0312 0 1 1>}
  }
  MaterialExpressionNamedRerouteDeclaration_33 {
   class=MaterialExpressionNamedRerouteDeclaration
   inputs [None {from=MaterialExpressionAdd_13}]
   properties {name=Mask Normals, nodeColor=<rgba 1 0 0.7969 1>}
  }
  MaterialExpressionNamedRerouteDeclaration_34 {
   class=MaterialExpressionNamedRerouteDeclaration
   inputs [None {from=MaterialExpressionAdd_14}]
   properties {name=Mask Normals Displacement, nodeColor=<rgba 1 0 0.7969 1>}
  }
  MaterialExpressionNamedRerouteDeclaration_35 {
   class=MaterialExpressionNamedRerouteDeclaration
   inputs [None {from=MaterialExpressionDDX_1}]
   properties {name=Mask DDX, nodeColor=<rgba 0.8047 1 0 1>}
  }
  MaterialExpressionNamedRerouteDeclaration_36 {
   class=MaterialExpressionNamedRerouteDeclaration
   inputs [None {from=MaterialExpressionDDY_1}]
   properties {name=Mask DDY, nodeColor=<rgba 0.875 0 1 1>}
  }
  MaterialExpressionNamedRerouteDeclaration_37 {
   class=MaterialExpressionNamedRerouteDeclaration
   inputs [None {from=MaterialExpressionMaterialFunctionCall_10, output=UVs}]
   properties {name=Mask UVs, nodeColor=<rgba 0 1 0.5547 1>}
  }
  MaterialExpressionNamedRerouteDeclaration_38 {
   class=MaterialExpressionNamedRerouteDeclaration
   inputs [None {from=MaterialExpressionTextureSample_0, output=R}]
   properties {name=Mask Texture, nodeColor=<rgba 1 0.7734 0 1>}
  }
  MaterialExpressionNamedRerouteDeclaration_39 {
   class=MaterialExpressionNamedRerouteDeclaration
   inputs [None {from=MaterialExpressionMaterialFunctionCall_0, output=Mask Before Contrast}]
   properties {name=Mask Before Contrast, nodeColor=<rgba 0 1 0.8828 1>}
  }
  MaterialExpressionNamedRerouteDeclaration_4 {
   class=MaterialExpressionNamedRerouteDeclaration
   inputs [None {from=MaterialExpressionDDX_0}]
   properties {name=DDX, nodeColor=<rgba 0.8047 1 0 1>}
  }
  MaterialExpressionNamedRerouteDeclaration_40 {
   class=MaterialExpressionNamedRerouteDeclaration
   inputs [None {from=MaterialExpressionBlendMaterialAttributes_2}]
   properties {name=Border Result, nodeColor=<rgba 1 0.5625 0 1>}
  }
  MaterialExpressionNamedRerouteDeclaration_41 {
   class=MaterialExpressionNamedRerouteDeclaration
   inputs [None {from=MaterialExpressionMaterialFunctionCall_10, output=UV X}]
   properties {name=Mask UVX, nodeColor=<rgba 1 0 0 1>}
  }
  MaterialExpressionNamedRerouteDeclaration_42 {
   class=MaterialExpressionNamedRerouteDeclaration
   inputs [None {from=MaterialExpressionMaterialFunctionCall_10, output=UV Y}]
   properties {name=Mask UVY, nodeColor=<rgba 0 1 0 1>}
  }
  MaterialExpressionNamedRerouteDeclaration_43 {
   class=MaterialExpressionNamedRerouteDeclaration
   inputs [None {from=MaterialExpressionMaterialFunctionCall_10, output=UV Z}]
   properties {name=Mask UVZ, nodeColor=<rgba 0 0 1 1>}
  }
  MaterialExpressionNamedRerouteDeclaration_44 {
   class=MaterialExpressionNamedRerouteDeclaration
   inputs [None {from=MaterialExpressionAdd_7}]
   properties {name=Mask Texture Displacement, nodeColor=<rgba 1 0.7734 0 1>}
  }
  MaterialExpressionNamedRerouteDeclaration_45 {
   class=MaterialExpressionNamedRerouteDeclaration
   inputs [None {from=MaterialExpressionMaterialFunctionCall_7, output=Result}]
   properties {name=Debug Border, nodeColor=<rgba 0.0312 0 1 1>}
  }
  MaterialExpressionNamedRerouteDeclaration_46 {
   class=MaterialExpressionNamedRerouteDeclaration
   inputs [None {from=MaterialExpressionMultiply_18}]
   properties {name=Border Mask, nodeColor=<rgba 1 0.5391 0 1>}
  }
  MaterialExpressionNamedRerouteDeclaration_47 {
   class=MaterialExpressionNamedRerouteDeclaration
   inputs [None {from=MaterialExpressionAdd_12}]
   properties {name=Bounds Mask, nodeColor=<rgba 1 0.3047 0 1>}
  }
  MaterialExpressionNamedRerouteDeclaration_48 {
   class=MaterialExpressionNamedRerouteDeclaration
   inputs [None {from=MaterialExpressionMaterialFunctionCall_8, output=Result}]
   properties {name=Debug DFAO Mask, nodeColor=<rgba 0.0312 0 1 1>}
  }
  MaterialExpressionNamedRerouteDeclaration_49 {
   class=MaterialExpressionNamedRerouteDeclaration
   inputs [None {from=MaterialExpressionShadingPathSwitch_0}]
   properties {name=DFAO Mask, nodeColor=<rgba 0.5469 1 0 1>}
  }
  MaterialExpressionNamedRerouteDeclaration_5 {
   class=MaterialExpressionNamedRerouteDeclaration
   inputs [None {from=MaterialExpressionGetMaterialAttributes_0, output=Specular}]
   properties {name=Specular Input, nodeColor=<rgba 0 0.0859 1 1>}
  }
  MaterialExpressionNamedRerouteDeclaration_50 {
   class=MaterialExpressionNamedRerouteDeclaration
   inputs [None {from=MaterialExpressionMaterialFunctionCall_11, output=Result}]
   properties {name=Debug Bounds Mask, nodeColor=<rgba 0.0312 0 1 1>}
  }
  MaterialExpressionNamedRerouteDeclaration_51 {
   class=MaterialExpressionNamedRerouteDeclaration
   inputs [None {from=MaterialExpressionAdd_33}]
   properties {name=Bounds Mask Displacement, nodeColor=<rgba 1 0.3047 0 1>}
  }
  MaterialExpressionNamedRerouteDeclaration_52 {
   class=MaterialExpressionNamedRerouteDeclaration
   inputs [None {from=MaterialExpressionAdd_38}]
   properties {name=Vertex Color Mask, nodeColor=<rgba 0.5938 0 1 1>}
  }
  MaterialExpressionNamedRerouteDeclaration_53 {
   class=MaterialExpressionNamedRerouteDeclaration
   inputs [None {from=MaterialExpressionMaterialFunctionCall_13, output=Result}]
   properties {name=Debug Vertex Color Mask, nodeColor=<rgba 0.0312 0 1 1>}
  }
  MaterialExpressionNamedRerouteDeclaration_54 {
   class=MaterialExpressionNamedRerouteDeclaration
   inputs [None {from=MaterialExpressionFunctionInput_3}]
   properties {name=Layer Texture Mask, nodeColor=<rgba 0 1 0.5078 1>}
  }
  MaterialExpressionNamedRerouteDeclaration_6 {
   class=MaterialExpressionNamedRerouteDeclaration
   inputs [None {from=MaterialExpressionMaterialFunctionCall_3, output=UV X}]
   properties {name=UVX, nodeColor=<rgba 1 0 0 1>}
  }
  MaterialExpressionNamedRerouteDeclaration_7 {
   class=MaterialExpressionNamedRerouteDeclaration
   inputs [None {from=MaterialExpressionMaterialFunctionCall_3, output=UV Y}]
   properties {name=UVY, nodeColor=<rgba 0 1 0 1>}
  }
  MaterialExpressionNamedRerouteDeclaration_8 {
   class=MaterialExpressionNamedRerouteDeclaration
   inputs [None {from=MaterialExpressionMaterialFunctionCall_3, output=UV Z}]
   properties {name=UVZ, nodeColor=<rgba 0 0 1 1>}
  }
  MaterialExpressionNamedRerouteDeclaration_9 {
   class=MaterialExpressionNamedRerouteDeclaration
   inputs [None {from=MaterialExpressionDDY_0}]
   properties {name=DDY, nodeColor=<rgba 0.875 0 1 1>}
  }
  MaterialExpressionNamedRerouteUsage_0 {class=MaterialExpressionNamedRerouteUsage}
  MaterialExpressionNamedRerouteUsage_1 {class=MaterialExpressionNamedRerouteUsage}
  MaterialExpressionNamedRerouteUsage_10 {class=MaterialExpressionNamedRerouteUsage}
  MaterialExpressionNamedRerouteUsage_11 {class=MaterialExpressionNamedRerouteUsage}
  MaterialExpressionNamedRerouteUsage_12 {class=MaterialExpressionNamedRerouteUsage}
  MaterialExpressionNamedRerouteUsage_13 {class=MaterialExpressionNamedRerouteUsage}
  MaterialExpressionNamedRerouteUsage_14 {class=MaterialExpressionNamedRerouteUsage}
  MaterialExpressionNamedRerouteUsage_15 {class=MaterialExpressionNamedRerouteUsage}
  MaterialExpressionNamedRerouteUsage_16 {class=MaterialExpressionNamedRerouteUsage}
  MaterialExpressionNamedRerouteUsage_17 {class=MaterialExpressionNamedRerouteUsage}
  MaterialExpressionNamedRerouteUsage_18 {class=MaterialExpressionNamedRerouteUsage}
  MaterialExpressionNamedRerouteUsage_19 {class=MaterialExpressionNamedRerouteUsage}
  MaterialExpressionNamedRerouteUsage_2 {class=MaterialExpressionNamedRerouteUsage}
  MaterialExpressionNamedRerouteUsage_20 {class=MaterialExpressionNamedRerouteUsage}
  MaterialExpressionNamedRerouteUsage_21 {class=MaterialExpressionNamedRerouteUsage}
  MaterialExpressionNamedRerouteUsage_22 {class=MaterialExpressionNamedRerouteUsage}
  MaterialExpressionNamedRerouteUsage_23 {class=MaterialExpressionNamedRerouteUsage}
  MaterialExpressionNamedRerouteUsage_25 {class=MaterialExpressionNamedRerouteUsage}
  MaterialExpressionNamedRerouteUsage_26 {class=MaterialExpressionNamedRerouteUsage}
  MaterialExpressionNamedRerouteUsage_27 {class=MaterialExpressionNamedRerouteUsage}
  MaterialExpressionNamedRerouteUsage_28 {class=MaterialExpressionNamedRerouteUsage}
  MaterialExpressionNamedRerouteUsage_29 {class=MaterialExpressionNamedRerouteUsage}
  MaterialExpressionNamedRerouteUsage_3 {class=MaterialExpressionNamedRerouteUsage}
  MaterialExpressionNamedRerouteUsage_30 {class=MaterialExpressionNamedRerouteUsage}
  MaterialExpressionNamedRerouteUsage_31 {class=MaterialExpressionNamedRerouteUsage}
  MaterialExpressionNamedRerouteUsage_32 {class=MaterialExpressionNamedRerouteUsage}
  MaterialExpressionNamedRerouteUsage_33 {class=MaterialExpressionNamedRerouteUsage}
  MaterialExpressionNamedRerouteUsage_34 {class=MaterialExpressionNamedRerouteUsage}
  MaterialExpressionNamedRerouteUsage_35 {class=MaterialExpressionNamedRerouteUsage}
  MaterialExpressionNamedRerouteUsage_36 {class=MaterialExpressionNamedRerouteUsage}
  MaterialExpressionNamedRerouteUsage_37 {class=MaterialExpressionNamedRerouteUsage}
  MaterialExpressionNamedRerouteUsage_38 {class=MaterialExpressionNamedRerouteUsage}
  MaterialExpressionNamedRerouteUsage_39 {class=MaterialExpressionNamedRerouteUsage}
  MaterialExpressionNamedRerouteUsage_4 {class=MaterialExpressionNamedRerouteUsage}
  MaterialExpressionNamedRerouteUsage_40 {class=MaterialExpressionNamedRerouteUsage}
  MaterialExpressionNamedRerouteUsage_41 {class=MaterialExpressionNamedRerouteUsage}
  MaterialExpressionNamedRerouteUsage_42 {class=MaterialExpressionNamedRerouteUsage}
  MaterialExpressionNamedRerouteUsage_43 {class=MaterialExpressionNamedRerouteUsage}
  MaterialExpressionNamedRerouteUsage_44 {class=MaterialExpressionNamedRerouteUsage}
  MaterialExpressionNamedRerouteUsage_45 {class=MaterialExpressionNamedRerouteUsage}
  MaterialExpressionNamedRerouteUsage_46 {class=MaterialExpressionNamedRerouteUsage}
  MaterialExpressionNamedRerouteUsage_47 {class=MaterialExpressionNamedRerouteUsage}
  MaterialExpressionNamedRerouteUsage_48 {class=MaterialExpressionNamedRerouteUsage}
  MaterialExpressionNamedRerouteUsage_49 {class=MaterialExpressionNamedRerouteUsage}
  MaterialExpressionNamedRerouteUsage_5 {class=MaterialExpressionNamedRerouteUsage}
  MaterialExpressionNamedRerouteUsage_50 {class=MaterialExpressionNamedRerouteUsage}
  MaterialExpressionNamedRerouteUsage_51 {class=MaterialExpressionNamedRerouteUsage}
  MaterialExpressionNamedRerouteUsage_52 {class=MaterialExpressionNamedRerouteUsage}
  MaterialExpressionNamedRerouteUsage_53 {class=MaterialExpressionNamedRerouteUsage}
  MaterialExpressionNamedRerouteUsage_54 {class=MaterialExpressionNamedRerouteUsage}
  MaterialExpressionNamedRerouteUsage_55 {class=MaterialExpressionNamedRerouteUsage}
  MaterialExpressionNamedRerouteUsage_56 {class=MaterialExpressionNamedRerouteUsage}
  MaterialExpressionNamedRerouteUsage_57 {class=MaterialExpressionNamedRerouteUsage}
  MaterialExpressionNamedRerouteUsage_58 {class=MaterialExpressionNamedRerouteUsage}
  MaterialExpressionNamedRerouteUsage_59 {class=MaterialExpressionNamedRerouteUsage}
  MaterialExpressionNamedRerouteUsage_6 {class=MaterialExpressionNamedRerouteUsage}
  MaterialExpressionNamedRerouteUsage_60 {class=MaterialExpressionNamedRerouteUsage}
  MaterialExpressionNamedRerouteUsage_61 {class=MaterialExpressionNamedRerouteUsage}
  MaterialExpressionNamedRerouteUsage_62 {class=MaterialExpressionNamedRerouteUsage}
  MaterialExpressionNamedRerouteUsage_63 {class=MaterialExpressionNamedRerouteUsage}
  MaterialExpressionNamedRerouteUsage_64 {class=MaterialExpressionNamedRerouteUsage}
  MaterialExpressionNamedRerouteUsage_65 {class=MaterialExpressionNamedRerouteUsage}
  MaterialExpressionNamedRerouteUsage_66 {class=MaterialExpressionNamedRerouteUsage}
  MaterialExpressionNamedRerouteUsage_67 {class=MaterialExpressionNamedRerouteUsage}
  MaterialExpressionNamedRerouteUsage_68 {class=MaterialExpressionNamedRerouteUsage}
  MaterialExpressionNamedRerouteUsage_69 {class=MaterialExpressionNamedRerouteUsage}
  MaterialExpressionNamedRerouteUsage_7 {class=MaterialExpressionNamedRerouteUsage}
  MaterialExpressionNamedRerouteUsage_70 {class=MaterialExpressionNamedRerouteUsage}
  MaterialExpressionNamedRerouteUsage_71 {class=MaterialExpressionNamedRerouteUsage}
  MaterialExpressionNamedRerouteUsage_72 {class=MaterialExpressionNamedRerouteUsage}
  MaterialExpressionNamedRerouteUsage_73 {class=MaterialExpressionNamedRerouteUsage}
  MaterialExpressionNamedRerouteUsage_74 {class=MaterialExpressionNamedRerouteUsage}
  MaterialExpressionNamedRerouteUsage_75 {class=MaterialExpressionNamedRerouteUsage}
  MaterialExpressionNamedRerouteUsage_76 {class=MaterialExpressionNamedRerouteUsage}
  MaterialExpressionNamedRerouteUsage_77 {class=MaterialExpressionNamedRerouteUsage}
  MaterialExpressionNamedRerouteUsage_78 {class=MaterialExpressionNamedRerouteUsage}
  MaterialExpressionNamedRerouteUsage_79 {class=MaterialExpressionNamedRerouteUsage}
  MaterialExpressionNamedRerouteUsage_8 {class=MaterialExpressionNamedRerouteUsage}
  MaterialExpressionNamedRerouteUsage_80 {class=MaterialExpressionNamedRerouteUsage}
  MaterialExpressionNamedRerouteUsage_81 {class=MaterialExpressionNamedRerouteUsage}
  MaterialExpressionNamedRerouteUsage_82 {class=MaterialExpressionNamedRerouteUsage}
  MaterialExpressionNamedRerouteUsage_83 {class=MaterialExpressionNamedRerouteUsage}
  MaterialExpressionNamedRerouteUsage_84 {class=MaterialExpressionNamedRerouteUsage}
  MaterialExpressionNamedRerouteUsage_85 {class=MaterialExpressionNamedRerouteUsage}
  MaterialExpressionNamedRerouteUsage_87 {class=MaterialExpressionNamedRerouteUsage}
  MaterialExpressionNamedRerouteUsage_9 {class=MaterialExpressionNamedRerouteUsage}
  MaterialExpressionOneMinus_0 {class=MaterialExpressionOneMinus, inputs [None {from=MaterialExpressionSaturate_7}]}
  MaterialExpressionOneMinus_1 {class=MaterialExpressionOneMinus, inputs [None {from=MaterialExpressionNamedRerouteUsage_71}]}
  MaterialExpressionPower_3 {
   class=MaterialExpressionPower
   inputs [
    Base {from=MaterialExpressionOneMinus_0}
    Exp {from=MaterialExpressionScalarParameter_57}
   ]
  }
  MaterialExpressionQualitySwitch_1 {
   class=MaterialExpressionQualitySwitch
   inputs [
    Default {from=MaterialExpressionConstant_8, output=""}
    Low {from="", output=""}
    High {from="", output=""}
    Medium {from="", output=""}
    Epic {from=MaterialExpressionMultiply_36, output=""}
   ]
   properties {}
  }
  MaterialExpressionReroute_0 {class=MaterialExpressionReroute, inputs [None {from=MaterialExpressionConstant_8}]}
  MaterialExpressionReroute_1 {class=MaterialExpressionReroute, inputs [None {from=MaterialExpressionMultiply_21}]}
  MaterialExpressionSaturate_0 {class=MaterialExpressionSaturate, inputs [None {from=MaterialExpressionLinearInterpolate_3}]}
  MaterialExpressionSaturate_1 {class=MaterialExpressionSaturate, inputs [None {from=MaterialExpressionVertexNormalWS_2}]}
  MaterialExpressionSaturate_12 {class=MaterialExpressionSaturate, inputs [None {from=MaterialExpressionAdd_25}]}
  MaterialExpressionSaturate_13 {class=MaterialExpressionSaturate, inputs [None {from=MaterialExpressionAdd_26}]}
  MaterialExpressionSaturate_3 {class=MaterialExpressionSaturate, inputs [None {from=MaterialExpressionAdd_8}]}
  MaterialExpressionSaturate_4 {class=MaterialExpressionSaturate, inputs [None {from=MaterialExpressionMultiply_4}]}
  MaterialExpressionSaturate_5 {class=MaterialExpressionSaturate, inputs [None {from=MaterialExpressionMultiply_6}]}
  MaterialExpressionSaturate_7 {class=MaterialExpressionSaturate, inputs [None {from=MaterialExpressionDistanceFieldApproxAO_0}]}
  MaterialExpressionScalarParameter_0 {
   class=MaterialExpressionScalarParameter
   properties {
    controlType=Enumeration
    defaultValue=0.15
    enumeration=/BRCoreMaterials/Materials/Enums/EN_4Layer_Cover
    group=Cover | Cover World Space UVs
    parameterName=Cover WorldSpaceUVs Mode
    sliderMax=1
    sortPriority=0
   }
  }
  MaterialExpressionScalarParameter_1 {
   class=MaterialExpressionScalarParameter
   properties {
    defaultValue=1
    enumeration=/BRCoreMaterials/Materials/Enums/EN_4Layer_WorldSpaceUVs_Toggle
    group=Cover
    parameterName=Cover Displacement Magnitude
    sliderMax=1
    sortPriority=11
   }
  }
  MaterialExpressionScalarParameter_10 {class=MaterialExpressionScalarParameter, properties {group=Cover | Cover Blend, parameterName=Cover Blend Background Low, sortPriority=0}}
  MaterialExpressionScalarParameter_11 {
   class=MaterialExpressionScalarParameter
   properties {defaultValue=1, group=Cover | Cover Blend, parameterName=Cover Blend Contrast Displacement, sortPriority=7}
  }
  MaterialExpressionScalarParameter_12 {
   class=MaterialExpressionScalarParameter
   properties {defaultValue=1, group=Cover | Cover Layer Mask, parameterName=L1 Cover, sliderMax=1, sortPriority=0}
  }
  MaterialExpressionScalarParameter_13 {
   class=MaterialExpressionScalarParameter
   properties {defaultValue=1, group=Cover | Cover Blend, parameterName=Cover Blend Foreground High, sortPriority=4}
  }
  MaterialExpressionScalarParameter_14 {class=MaterialExpressionScalarParameter, properties {group=Cover | Cover Blend, parameterName=Cover Blend Foreground Low, sortPriority=3}}
  MaterialExpressionScalarParameter_15 {
   class=MaterialExpressionScalarParameter
   properties {defaultValue=3, group=Cover | Cover Blend, parameterName=Cover Blend Contrast, sortPriority=6}
  }
  MaterialExpressionScalarParameter_16 {
   class=MaterialExpressionScalarParameter
   properties {defaultValue=0.5, group=Cover | Cover Normal Mask, parameterName=Cover Normal Mask Per Pixel, sliderMax=1, sortPriority=8}
  }
  MaterialExpressionScalarParameter_17 {
   class=MaterialExpressionScalarParameter
   properties {defaultValue=0.5, group=Cover | Cover Normal Mask, parameterName=Cover Normal Mask Z +, sortPriority=0}
  }
  MaterialExpressionScalarParameter_18 {class=MaterialExpressionScalarParameter, properties {group=Cover | Cover Normal Mask, parameterName=Cover Normal Mask X +, sortPriority=2}}
  MaterialExpressionScalarParameter_19 {class=MaterialExpressionScalarParameter, properties {group=Cover | Cover Normal Mask, parameterName=Cover Normal Mask Y +, sortPriority=4}}
  MaterialExpressionScalarParameter_2 {
   class=MaterialExpressionScalarParameter
   properties {
    controlType=Enumeration
    enumeration=/BRCoreMaterials/Materials/Enums/EN_4Layer_CoverBlend_Debug
    group=Cover | Cover Blend
    parameterName=Cover Blend Debug Mode
   }
  }
  MaterialExpressionScalarParameter_20 {
   class=MaterialExpressionScalarParameter
   properties {
    controlType=Enumeration
    enumeration=/BRCoreMaterials/Materials/Enums/EN_4Layer_GenericToggle
    group=Cover | Cover Normal Mask
    parameterName=Cover Normal Mask Debug Mode
    sliderMax=1
   }
  }
  MaterialExpressionScalarParameter_21 {
   class=MaterialExpressionScalarParameter
   properties {group=Cover | Cover Vertex Color Mask, parameterName=Cover Vertex Color Mask Red, sortPriority=0}
  }
  MaterialExpressionScalarParameter_22 {class=MaterialExpressionScalarParameter, properties {group=Cover | Cover Normal Mask, parameterName=Cover Normal Mask Z -, sortPriority=1}}
  MaterialExpressionScalarParameter_23 {class=MaterialExpressionScalarParameter, properties {group=Cover | Cover Normal Mask, parameterName=Cover Normal Mask X -, sortPriority=3}}
  MaterialExpressionScalarParameter_24 {class=MaterialExpressionScalarParameter, properties {group=Cover | Cover Normal Mask, parameterName=Cover Normal Mask Y -, sortPriority=5}}
  MaterialExpressionScalarParameter_25 {
   class=MaterialExpressionScalarParameter
   properties {defaultValue=1, group=Cover | Cover Layer Mask, parameterName=L2 Cover, sliderMax=1, sortPriority=0}
  }
  MaterialExpressionScalarParameter_26 {
   class=MaterialExpressionScalarParameter
   properties {defaultValue=1, group=Cover | Cover Layer Mask, parameterName=L3 Cover, sliderMax=1, sortPriority=0}
  }
  MaterialExpressionScalarParameter_27 {
   class=MaterialExpressionScalarParameter
   properties {defaultValue=1, group=Cover | Cover Layer Mask, parameterName=L4 Cover, sliderMax=1, sortPriority=0}
  }
  MaterialExpressionScalarParameter_28 {
   class=MaterialExpressionScalarParameter
   properties {defaultValue=3, group=Cover | Cover Border, parameterName=Cover Border Contrast, sortPriority=6}
  }
  MaterialExpressionScalarParameter_29 {
   class=MaterialExpressionScalarParameter
   properties {defaultValue=0.2, group=Cover | Cover Border, parameterName=Cover Border Range, sortPriority=7}
  }
  MaterialExpressionScalarParameter_3 {
   class=MaterialExpressionScalarParameter
   properties {
    defaultValue=1
    enumeration=&EN_4Layer_WorldSpaceUVs_Toggle
    group=Cover
    parameterName=Cover Normal Intensity
    sortPriority=6
   }
  }
  MaterialExpressionScalarParameter_30 {
   class=MaterialExpressionScalarParameter
   properties {group=Cover | Cover Border, parameterName=Cover Border Weight, sliderMax=1, sortPriority=0}
  }
  MaterialExpressionScalarParameter_31 {class=MaterialExpressionScalarParameter, properties {group=Cover | Cover Border, parameterName=Cover Border Pattern Offset, sortPriority=9}}
  MaterialExpressionScalarParameter_32 {
   class=MaterialExpressionScalarParameter
   properties {group=Cover | Cover Border, parameterName=Cover Border Pattern Intensity, sortPriority=8}
  }
  MaterialExpressionScalarParameter_33 {
   class=MaterialExpressionScalarParameter
   properties {
    defaultValue=384
    enumeration=&EN_4Layer_WorldSpaceUVs_Toggle
    group=Cover | Cover World Space UVs
    parameterName=Cover WorldSpaceUVs Scale
    sortPriority=1
   }
  }
  MaterialExpressionScalarParameter_34 {
   class=MaterialExpressionScalarParameter
   properties {group=Cover | Cover Normal Mask, parameterName=Cover Normal Mask Pattern Weight, sortPriority=10}
  }
  MaterialExpressionScalarParameter_35 {
   class=MaterialExpressionScalarParameter
   properties {defaultValue=150, group=Cover | Cover DFAO Mask, parameterName=Cover DFAO Mask Base Distance, sortPriority=0}
  }
  MaterialExpressionScalarParameter_36 {
   class=MaterialExpressionScalarParameter
   properties {group=Cover | Cover Normal Mask, parameterName=Cover Normal Mask Pattern Offset, sortPriority=11}
  }
  MaterialExpressionScalarParameter_39 {
   class=MaterialExpressionScalarParameter
   properties {
    defaultValue=1500
    enumeration=&EN_4Layer_WorldSpaceUVs_Toggle
    group=Cover | Cover World Space UVs
    parameterName=Cover WorldSpaceUVs Scale Pattern
    sortPriority=2
   }
  }
  MaterialExpressionScalarParameter_4 {
   class=MaterialExpressionScalarParameter
   properties {
    enumeration=&EN_4Layer_WorldSpaceUVs_Toggle
    group=Cover
    parameterName=Cover Specular Range Low
    sortPriority=7
   }
  }
  MaterialExpressionScalarParameter_40 {
   class=MaterialExpressionScalarParameter
   properties {
    enumeration=&EN_4Layer_WorldSpaceUVs_Toggle
    group=Cover | Cover Border
    parameterName=Cover Border Specular Range Low
    sortPriority=2
   }
  }
  MaterialExpressionScalarParameter_41 {
   class=MaterialExpressionScalarParameter
   properties {
    defaultValue=1
    enumeration=&EN_4Layer_WorldSpaceUVs_Toggle
    group=Cover | Cover Border
    parameterName=Cover Border Specular Range High
    sortPriority=3
   }
  }
  MaterialExpressionScalarParameter_42 {
   class=MaterialExpressionScalarParameter
   properties {
    enumeration=&EN_4Layer_WorldSpaceUVs_Toggle
    group=Cover | Cover Border
    parameterName=Cover Border Roughness Range Low
    sortPriority=4
   }
  }
  MaterialExpressionScalarParameter_43 {
   class=MaterialExpressionScalarParameter
   properties {
    defaultValue=1
    enumeration=&EN_4Layer_WorldSpaceUVs_Toggle
    group=Cover | Cover Border
    parameterName=Cover Border Roughness Range High
    sortPriority=5
   }
  }
  MaterialExpressionScalarParameter_44 {
   class=MaterialExpressionScalarParameter
   properties {
    controlType=Enumeration
    enumeration=/BRCoreMaterials/Materials/Enums/EN_4Layer_CoverBorder_Debug
    group=Cover | Cover Border
    parameterName=Cover Border Debug Mode
    sliderMax=1
   }
  }
  MaterialExpressionScalarParameter_45 {
   class=MaterialExpressionScalarParameter
   properties {group=Cover | Cover Vertex Color Mask, parameterName=Cover Vertex Color Mask Green, sortPriority=1}
  }
  MaterialExpressionScalarParameter_46 {class=MaterialExpressionScalarParameter, properties {group=Cover | Cover Bounds Mask, parameterName=Cover Bounds Mask Z +, sortPriority=0}}
  MaterialExpressionScalarParameter_47 {class=MaterialExpressionScalarParameter, properties {group=Cover | Cover Bounds Mask, parameterName=Cover Bounds Mask X +, sortPriority=2}}
  MaterialExpressionScalarParameter_48 {class=MaterialExpressionScalarParameter, properties {group=Cover | Cover Bounds Mask, parameterName=Cover Bounds Mask Y +, sortPriority=4}}
  MaterialExpressionScalarParameter_49 {class=MaterialExpressionScalarParameter, properties {group=Cover | Cover Bounds Mask, parameterName=Cover Bounds Mask Z -, sortPriority=1}}
  MaterialExpressionScalarParameter_5 {
   class=MaterialExpressionScalarParameter
   properties {
    defaultValue=1
    enumeration=&EN_4Layer_WorldSpaceUVs_Toggle
    group=Cover
    parameterName=Cover Specular Range High
    sortPriority=8
   }
  }
  MaterialExpressionScalarParameter_50 {class=MaterialExpressionScalarParameter, properties {group=Cover | Cover Bounds Mask, parameterName=Cover Bounds Mask X-, sortPriority=3}}
  MaterialExpressionScalarParameter_51 {class=MaterialExpressionScalarParameter, properties {group=Cover | Cover Bounds Mask, parameterName=Cover Bounds Mask Y -, sortPriority=5}}
  MaterialExpressionScalarParameter_52 {
   class=MaterialExpressionScalarParameter
   properties {group=Cover | Cover Vertex Color Mask, parameterName=Cover Vertex Color Mask Blue, sortPriority=2}
  }
  MaterialExpressionScalarParameter_53 {
   class=MaterialExpressionScalarParameter
   properties {group=Cover | Cover Bounds Mask, parameterName=Cover Bounds Mask Pattern Weight, sortPriority=12}
  }
  MaterialExpressionScalarParameter_54 {
   class=MaterialExpressionScalarParameter
   properties {group=Cover | Cover Bounds Mask, parameterName=Cover Bounds Mask Pattern Offset, sortPriority=13}
  }
  MaterialExpressionScalarParameter_55 {
   class=MaterialExpressionScalarParameter
   properties {
    controlType=Enumeration
    enumeration=&EN_4Layer_GenericToggle
    group=Cover | Cover DFAO Mask
    parameterName=Cover DFAO Mask Debug Mode
    sliderMax=1
   }
  }
  MaterialExpressionScalarParameter_56 {
   class=MaterialExpressionScalarParameter
   properties {
    controlType=Enumeration
    enumeration=&EN_4Layer_GenericToggle
    group=Cover | Cover Vertex Color Mask
    parameterName=Cover Vertex Color Mask Debug Mode
    sliderMax=1
   }
  }
  MaterialExpressionScalarParameter_57 {
   class=MaterialExpressionScalarParameter
   properties {defaultValue=1, group=Cover | Cover DFAO Mask, parameterName=Cover DFAO Falloff, sortPriority=2}
  }
  MaterialExpressionScalarParameter_58 {
   class=MaterialExpressionScalarParameter
   properties {group=Cover | Cover DFAO Mask, parameterName=Cover DFAO Mask Pixel Normals, sliderMax=1, sortPriority=1}
  }
  MaterialExpressionScalarParameter_59 {
   class=MaterialExpressionScalarParameter
   properties {group=Cover | Cover DFAO Mask, parameterName=Cover DFAO Mask Pattern Weight, sortPriority=3}
  }
  MaterialExpressionScalarParameter_6 {
   class=MaterialExpressionScalarParameter
   properties {
    defaultValue=0.1
    enumeration=&EN_4Layer_WorldSpaceUVs_Toggle
    group=Cover | Cover World Space UVs
    parameterName=Cover WorldSpaceUVs Blend
    sliderMax=1
    sortPriority=3
   }
  }
  MaterialExpressionScalarParameter_60 {
   class=MaterialExpressionScalarParameter
   properties {group=Cover | Cover DFAO Mask, parameterName=Cover DFAO Mask Pattern Offset, sortPriority=4}
  }
  MaterialExpressionScalarParameter_65 {
   class=MaterialExpressionScalarParameter
   properties {group=Cover | Cover Normal Mask, parameterName=Cover Normal Mask Pattern Weight, sortPriority=10}
  }
  MaterialExpressionScalarParameter_66 {
   class=MaterialExpressionScalarParameter
   properties {group=Cover | Cover Normal Mask, parameterName=Cover Normal Mask Pattern Offset, sortPriority=11}
  }
  MaterialExpressionScalarParameter_7 {
   class=MaterialExpressionScalarParameter
   properties {
    enumeration=&EN_4Layer_WorldSpaceUVs_Toggle
    group=Cover
    parameterName=Cover Roughness Range Low
    sortPriority=9
   }
  }
  MaterialExpressionScalarParameter_73 {
   class=MaterialExpressionScalarParameter
   properties {group=Cover | Cover Bounds Mask, parameterName=Cover Bounds Mask Offset Z +, sortPriority=6}
  }
  MaterialExpressionScalarParameter_74 {
   class=MaterialExpressionScalarParameter
   properties {group=Cover | Cover Bounds Mask, parameterName=Cover Bounds Mask Offset X +, sortPriority=8}
  }
  MaterialExpressionScalarParameter_75 {
   class=MaterialExpressionScalarParameter
   properties {group=Cover | Cover Bounds Mask, parameterName=Cover Bounds Mask Offset Y +, sortPriority=10}
  }
  MaterialExpressionScalarParameter_76 {
   class=MaterialExpressionScalarParameter
   properties {group=Cover | Cover Bounds Mask, parameterName=Cover Bounds Mask Offset Z -, sortPriority=7}
  }
  MaterialExpressionScalarParameter_77 {
   class=MaterialExpressionScalarParameter
   properties {group=Cover | Cover Bounds Mask, parameterName=Cover Bounds Mask Offset X-, sortPriority=9}
  }
  MaterialExpressionScalarParameter_78 {
   class=MaterialExpressionScalarParameter
   properties {group=Cover | Cover Bounds Mask, parameterName=Cover Bounds Mask Offset Y -, sortPriority=11}
  }
  MaterialExpressionScalarParameter_79 {
   class=MaterialExpressionScalarParameter
   properties {
    controlType=Enumeration
    enumeration=&EN_4Layer_GenericToggle
    group=Cover | Cover Bounds Mask
    parameterName=Cover Bounds Mask Debug Mode
    sliderMax=1
   }
  }
  MaterialExpressionScalarParameter_8 {
   class=MaterialExpressionScalarParameter
   properties {
    defaultValue=1
    enumeration=&EN_4Layer_WorldSpaceUVs_Toggle
    group=Cover
    parameterName=Cover Roughness Range High
    sortPriority=10
   }
  }
  MaterialExpressionScalarParameter_80 {
   class=MaterialExpressionScalarParameter
   properties {group=Cover | Cover Bounds Mask, parameterName=Cover Bounds Mask Pattern Weight, sortPriority=13}
  }
  MaterialExpressionScalarParameter_81 {
   class=MaterialExpressionScalarParameter
   properties {group=Cover | Cover Bounds Mask, parameterName=Cover Bounds Mask Pattern Offset, sortPriority=13}
  }
  MaterialExpressionScalarParameter_82 {class=MaterialExpressionScalarParameter, properties {group=Cover | Cover DFAO Mask, parameterName=Cover DFAO Mask Weight, sortPriority=-1}}
  MaterialExpressionScalarParameter_9 {
   class=MaterialExpressionScalarParameter
   properties {defaultValue=1, group=Cover | Cover Blend, parameterName=Cover Blend Background High, sortPriority=1}
  }
  MaterialExpressionSetMaterialAttributes_0 {
   class=MaterialExpressionSetMaterialAttributes
   inputs [
    MaterialAttributes {from=MaterialExpressionNamedRerouteUsage_2, output=""}
    BaseColor {from=MaterialExpressionNamedRerouteUsage_8, output=""}
    Normal {from=MaterialExpressionNamedRerouteUsage_10, output=""}
    Specular {from=MaterialExpressionNamedRerouteUsage_23, output=""}
    Metallic {from=MaterialExpressionNamedRerouteUsage_21, output=""}
    Roughness {from=MaterialExpressionNamedRerouteUsage_20, output=""}
    Opacity {from=MaterialExpressionNamedRerouteUsage_25, output=""}
   ]
   properties {attributeSetTypes [
     #1
     #2
     #3
     #7
     #6
     #4
    ]}
  }
  MaterialExpressionSetMaterialAttributes_1 {
   class=MaterialExpressionSetMaterialAttributes
   inputs [
    MaterialAttributes {from=MaterialExpressionBlendMaterialAttributes_1, output=""}
    Displacement {from=MaterialExpressionMultiply_0, output=""}
   ]
   properties {attributeSetTypes [#5]}
  }
  MaterialExpressionSetMaterialAttributes_2 {
   class=MaterialExpressionSetMaterialAttributes
   inputs [
    MaterialAttributes {from=MaterialExpressionGetMaterialAttributes_1, output=MaterialAttributes}
    BaseColor {from=MaterialExpressionVectorParameter_1, output=RGB}
    Specular {from=MaterialExpressionLinearInterpolate_5, output=""}
    Roughness {from=MaterialExpressionLinearInterpolate_6, output=""}
   ]
   properties {attributeSetTypes [#1, #3, #6]}
  }
  MaterialExpressionShadingPathSwitch_0 {
   class=MaterialExpressionShadingPathSwitch
   inputs [
    Default {from=MaterialExpressionQualitySwitch_1, output=""}
    Deferred {from="", output=""}
    Forward {from="", output=""}
    Mobile {from=MaterialExpressionReroute_0, output=""}
   ]
   properties {}
  }
  MaterialExpressionStaticBoolParameter_6 {class=MaterialExpressionStaticBoolParameter, inputs [], properties {group=- Effect Switches, parameterName=Displacement}}
  MaterialExpressionStaticSwitchParameter_0 {
   class=MaterialExpressionStaticSwitchParameter
   inputs [
    True {from=MaterialExpressionStaticSwitchParameter_1, output=""}
    False {from=MaterialExpressionNamedRerouteUsage_29, output=""}
   ]
   properties {group=- Effect Switches, parameterName=Cover}
  }
  MaterialExpressionStaticSwitchParameter_1 {
   class=MaterialExpressionStaticSwitchParameter
   inputs [
    True {from=MaterialExpressionNamedRerouteUsage_41, output=""}
    False {from=MaterialExpressionStaticSwitchParameter_2, output=""}
   ]
   properties {group=Cover | Cover Blend, parameterName=Cover Blend Debug, sortPriority=31}
  }
  MaterialExpressionStaticSwitchParameter_2 {
   class=MaterialExpressionStaticSwitchParameter
   inputs [
    True {from=MaterialExpressionNamedRerouteUsage_48, output=""}
    False {from=MaterialExpressionStaticSwitchParameter_3, output=""}
   ]
   properties {group=Cover | Cover Normal Mask, parameterName=Cover Normal Mask Debug, sortPriority=31}
  }
  MaterialExpressionStaticSwitchParameter_3 {
   class=MaterialExpressionStaticSwitchParameter
   inputs [
    True {from=MaterialExpressionNamedRerouteUsage_66, output=""}
    False {from=MaterialExpressionStaticSwitchParameter_4, output=""}
   ]
   properties {group=Cover | Cover Border, parameterName=Cover Border Debug, sortPriority=31}
  }
  MaterialExpressionStaticSwitchParameter_4 {
   class=MaterialExpressionStaticSwitchParameter
   inputs [
    True {from=MaterialExpressionNamedRerouteUsage_82, output=""}
    False {from=MaterialExpressionStaticSwitchParameter_5, output=""}
   ]
   properties {group=Cover | Cover Bounds Mask, parameterName=Cover Bounds Mask Debug, sortPriority=31}
  }
  MaterialExpressionStaticSwitchParameter_5 {
   class=MaterialExpressionStaticSwitchParameter
   inputs [
    True {from=MaterialExpressionNamedRerouteUsage_59, output=""}
    False {from=MaterialExpressionStaticSwitchParameter_6, output=""}
   ]
   properties {group=Cover | Cover DFAO Mask, parameterName=Cover DFAO Mask Debug, sortPriority=31}
  }
  MaterialExpressionStaticSwitchParameter_6 {
   class=MaterialExpressionStaticSwitchParameter
   inputs [
    True {from=MaterialExpressionNamedRerouteUsage_87, output=""}
    False {from=MaterialExpressionNamedRerouteDeclaration_29, output=""}
   ]
   properties {group=Cover | Cover Vertex Color Mask, parameterName=Cover Vertex Color Mask Debug, sortPriority=31}
  }
  MaterialExpressionStaticSwitch_0 {
   class=MaterialExpressionStaticSwitch
   inputs [
    True {from=MaterialExpressionBlendMaterialAttributes_0, output=""}
    False {from=MaterialExpressionBlendMaterialAttributes_1, output=""}
    Value {from=MaterialExpressionStaticBoolParameter_6, output=""}
   ]
   properties {}
  }
  MaterialExpressionTextureObjectParameter_0 {
   class=MaterialExpressionTextureObjectParameter
   inputs [
    UVs {from="", output=""}
    Apply View MipBias {from="", output=""}
   ]
   properties {
    group=Cover
    parameterName=Cover Displacement
    samplerType=SAMPLERTYPE_LinearGrayscale
    sortPriority=3
    texture=/BRCoreMaterials/Materials/Textures/T_Displacement
   }
  }
  MaterialExpressionTextureObjectParameter_2 {
   class=MaterialExpressionTextureObjectParameter
   inputs [
    UVs {from="", output=""}
    Apply View MipBias {from="", output=""}
   ]
   properties {group=Cover, parameterName=Cover Pattern, samplerType=SAMPLERTYPE_Masks, sortPriority=4, texture=/BRCoreMaterials/Materials/Textures/T_Mask}
  }
  MaterialExpressionTextureSampleParameter2D_0 {
   class=MaterialExpressionTextureSampleParameter2D
   inputs [
    UVs {from=MaterialExpressionNamedRerouteUsage_15, output=""}
    "DDX(UVs)" {from=MaterialExpressionNamedRerouteUsage_3, output=""}
    "DDY(UVs)" {from=MaterialExpressionNamedRerouteUsage_13, output=""}
    Apply View MipBias {from="", output=""}
   ]
   properties {
    group=Cover
    mipValueMode=TMVM_Derivative
    parameterName=Cover Diffuse
    samplerSource=SSM_Wrap_WorldGroupSettings
    sortPriority=0
    texture=/BRCoreMaterials/Materials/Textures/T_BaseColor
    channelNames {a=Mask, b="", g="", r=""}
   }
  }
  MaterialExpressionTextureSampleParameter2D_1 {
   class=MaterialExpressionTextureSampleParameter2D
   inputs [
    UVs {from=MaterialExpressionNamedRerouteUsage_16, output=""}
    "DDX(UVs)" {from=MaterialExpressionNamedRerouteUsage_11, output=""}
    "DDY(UVs)" {from=MaterialExpressionNamedRerouteUsage_14, output=""}
    Apply View MipBias {from="", output=""}
   ]
   properties {
    group=Cover
    mipValueMode=TMVM_Derivative
    parameterName=Cover Normal
    samplerSource=SSM_Wrap_WorldGroupSettings
    samplerType=SAMPLERTYPE_Normal
    sortPriority=1
    texture=/BRCoreMaterials/Materials/Textures/T_Normal
   }
  }
  MaterialExpressionTextureSampleParameter2D_2 {
   class=MaterialExpressionTextureSampleParameter2D
   inputs [
    UVs {from=MaterialExpressionNamedRerouteUsage_17, output=""}
    "DDX(UVs)" {from=MaterialExpressionNamedRerouteUsage_12, output=""}
    "DDY(UVs)" {from=MaterialExpressionNamedRerouteUsage_18, output=""}
    Apply View MipBias {from="", output=""}
   ]
   properties {
    group=Cover
    mipValueMode=TMVM_Derivative
    parameterName=Cover Specular
    samplerSource=SSM_Wrap_WorldGroupSettings
    samplerType=SAMPLERTYPE_Masks
    sortPriority=2
    texture=/BRCoreMaterials/Materials/Textures/T_Specular
    channelNames {a="", b=Roughness, g=Metal, r=Specular}
   }
  }
  MaterialExpressionTextureSample_0 {
   class=MaterialExpressionTextureSample
   inputs [
    UVs {from=MaterialExpressionNamedRerouteUsage_50}
    Tex {from=MaterialExpressionTextureObjectParameter_2}
    "DDX(UVs)" {from=MaterialExpressionNamedRerouteUsage_51}
    "DDY(UVs)" {from=MaterialExpressionNamedRerouteUsage_52}
    Apply View MipBias {from=""}
   ]
   properties {mipValueMode=TMVM_Derivative, samplerType=SAMPLERTYPE_Masks, texture=&T_Mask}
  }
  MaterialExpressionTextureSample_1 {
   class=MaterialExpressionTextureSample
   inputs [
    UVs {from=MaterialExpressionNamedRerouteUsage_0}
    Tex {from=MaterialExpressionTextureObjectParameter_0}
    Apply View MipBias {from=""}
   ]
   properties {samplerType=SAMPLERTYPE_LinearGrayscale, texture=&T_Displacement}
  }
  MaterialExpressionTextureSample_2 {
   class=MaterialExpressionTextureSample
   inputs [
    UVs {from=MaterialExpressionNamedRerouteUsage_4}
    Tex {from=MaterialExpressionTextureObjectParameter_0}
    Apply View MipBias {from=""}
   ]
   properties {samplerType=SAMPLERTYPE_LinearGrayscale, texture=&T_Displacement}
  }
  MaterialExpressionTextureSample_3 {
   class=MaterialExpressionTextureSample
   inputs [
    UVs {from=MaterialExpressionNamedRerouteUsage_5}
    Tex {from=MaterialExpressionTextureObjectParameter_0}
    Apply View MipBias {from=""}
   ]
   properties {samplerType=SAMPLERTYPE_LinearGrayscale, texture=&T_Displacement}
  }
  MaterialExpressionTextureSample_4 {
   class=MaterialExpressionTextureSample
   inputs [
    UVs {from=MaterialExpressionNamedRerouteUsage_62}
    Tex {from=MaterialExpressionTextureObjectParameter_2}
    Apply View MipBias {from=""}
   ]
   properties {samplerType=SAMPLERTYPE_Masks, texture=&T_Mask}
  }
  MaterialExpressionTextureSample_5 {
   class=MaterialExpressionTextureSample
   inputs [
    UVs {from=MaterialExpressionNamedRerouteUsage_63}
    Tex {from=MaterialExpressionTextureObjectParameter_2}
    Apply View MipBias {from=""}
   ]
   properties {samplerType=SAMPLERTYPE_Masks, texture=&T_Mask}
  }
  MaterialExpressionTextureSample_6 {
   class=MaterialExpressionTextureSample
   inputs [
    UVs {from=MaterialExpressionNamedRerouteUsage_64}
    Tex {from=MaterialExpressionTextureObjectParameter_2}
    Apply View MipBias {from=""}
   ]
   properties {samplerType=SAMPLERTYPE_Masks, texture=&T_Mask}
  }
  MaterialExpressionVectorParameter_0 {
   class=MaterialExpressionVectorParameter
   properties {group=Cover, parameterName=Cover Tint Color, sortPriority=5, defaultValue=<rgba 1 1 1 1>}
  }
  MaterialExpressionVectorParameter_1 {
   class=MaterialExpressionVectorParameter
   properties {group=Cover | Cover Border, parameterName=Cover Border Color, sortPriority=1, defaultValue=<rgba 0 0 0 1>}
  }
  MaterialExpressionVertexColor_0 {class=MaterialExpressionVertexColor, inputs [], properties {}}
  MaterialExpressionVertexNormalWS_0 {class=MaterialExpressionVertexNormalWS, inputs [], properties {}}
  MaterialExpressionVertexNormalWS_1 {class=MaterialExpressionVertexNormalWS, inputs [], properties {}}
  MaterialExpressionVertexNormalWS_2 {class=MaterialExpressionVertexNormalWS, inputs [], properties {}}
 }}
