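# DirectFX material function that adds animated, tinted caustics to the emissive of incoming material attributes, behind a static switch and low-quality bypass.
# Two panning samples of one texture's alpha build a UV warp for two further panning samples, which are summed, desaturated and power-sharpened.
# Intensity, an alpha input, inverse-square distance falloff from a pivot and a UV sphere mask scale the result.
# A caustic opacity chain is left unconnected.
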MaterialFunction /Game/ArtTools/Tech_Art/DirectFX/Assets/MaterialFunctions/MF_DFX_Caustics
graph {nodes [
  n1 = NamedRerouteUsage()
  n2 = NamedRerouteUsage()
  FunctionOutput(None=QualitySwitch(Default=StaticSwitchParameter(True=@n1, False=@n2) {group=DFX, parameterName=Enable Caustics}, Low=@n2))
  n3 = Divide(A=NamedRerouteUsage(), B=ScalarParameter() {defaultValue=1, group=Caustics, parameterName=CausticScaleOffset})
  n4 = ComponentMask(None=@n3) {g=true, r=true}
  n5 = NamedRerouteUsage()
  n6 = NamedRerouteUsage()
  n7 = TextureSample(UVs=Panner(Coordinate=Add(A=@n4, B=@n5), Speed=NamedRerouteUsage()) {speedX=-0.05, speedY=0.02}, Tex=@n6) {samplerSource=SSM_Wrap_WorldGroupSettings, samplerType=SAMPLERTYPE_LinearColor}
  n8 = Add(A=@n5, B=ComponentMask(None=@n3) {g=true, r=true})
  n9 = NamedRerouteUsage()
  n10 = TextureSample(UVs=OneMinus(None=Panner(Coordinate=@n8, Speed=@n9) {speedX=0.02, speedY=-0.02}), Tex=@n6) {constCoordinate=1, samplerSource=SSM_Wrap_WorldGroupSettings, samplerType=SAMPLERTYPE_LinearColor}
  n11 = Add(A=@n7.RGB, B=@n10.RGB)
  Multiply(A=@n11, B=Multiply(B=NamedRerouteUsage()))
  NamedRerouteDeclaration(None=FunctionInput() {bUsePreviewValueAsDefault=true, inputType=FunctionInput_MaterialAttributes}) {name=Original Material, nodeColor=<rgba 1 0 0.375 1>}
  n12 = GetMaterialAttributes(None=NamedRerouteUsage()) {attributeGetTypes [#1]}
  n13 = NamedRerouteUsage()
  n14 = ScalarParameter() {defaultValue=1, group=DFX, parameterName=DFX_Intensity}
  n15 = Divide(A=Distance(A=FunctionInput() {bUsePreviewValueAsDefault=true, inputName=Falloff Pivot, sortPriority=30}, B=WorldPosition().XYZ)) {constB=100, desc=CM to M}
  n16 = Divide(B=Multiply(A=@n15, B=@n15))
  n17 = TextureCoordinate()
  n18 = Multiply(
   A=LinearInterpolate(
    B=NamedRerouteDeclaration(None=@n16) {name=Inverse Square Law, nodeColor=<rgba 0.7812 1 0 1>}
    Alpha=ScalarParameter() {defaultValue=1, group=DFX, parameterName=DFX_Falloff}
   ) {constA=1}
   B=SphereMask(
    A=FunctionInput(Preview=@n17) {bUsePreviewValueAsDefault=true, inputName=UV Unscaled, inputType=FunctionInput_Vector2, sortPriority=15}
    B=Constant2Vector() {g=0.5, r=0.5}
    Radius=ScalarParameter() {defaultValue=1, group=DFX, parameterName=DFX_FalloffRadius}
   ) {hardnessPercent=0}
  )
  n19 = FunctionInput() {bUsePreviewValueAsDefault=true, inputName=Alpha, inputType=FunctionInput_Scalar, sortPriority=20}
  n20 = Multiply(
   A=Multiply(A=Multiply(A=@n13, B=@n14), B=Multiply(A=@n18, B=@n19))
   B=VectorParameter() {group=DFX, parameterName=DFX_TintColor, defaultValue=<rgba 1 1 1 0>}.RGB
  )
  NamedRerouteDeclaration(None=SetMaterialAttributes!(MaterialAttributes=@n12.MaterialAttributes, EmissiveColor=Add(A=@n12.EmissiveColor, B=@n20)) {attributeSetTypes [#1]}) {name=Caustic Material, nodeColor=<rgba 0 0.8934 1 1>}
  NamedRerouteDeclaration(None=FunctionInput(Preview=TextureCoordinate()) {bUsePreviewValueAsDefault=true, inputName=UV, inputType=FunctionInput_Vector2, sortPriority=10}) {name=Coordinates, nodeColor=<rgba 1 0.8906 0 1>}
  NamedRerouteDeclaration(None=TextureObjectParameter() {group=DFX, parameterName=DFX_Texture, texture=/Game/Global/Base4Layers_2022/Textures/T-GraffitiSample_D}) {name=Texture Object, nodeColor=<rgba 0 0.0859 1 1>}
  n21 = VectorParameter() {
   group=DFX
   parameterName=DFX_PanningSpeed
   channelNames {a=Panning Y Tex2, b=Panning X Tex2, g=Panning Y Tex1, r=Panning X Tex1}
   defaultValue=<rgba -0.05 0.02 0.02 0.05>
  }
  NamedRerouteDeclaration(None=MaterialFunctionCall(X=@n21.Panning X Tex1, Y=@n21.Panning X Tex1) {materialFunction=&MakeFloat2}.Result) {name=Panning Tex1, nodeColor=<rgba 0 0.2266 1 1>}
  NamedRerouteDeclaration(None=MaterialFunctionCall(X=@n21.Panning X Tex2, Y=@n21.Panning X Tex2) {materialFunction=&MakeFloat2}.Result) {name=Panning Tex2, nodeColor=<rgba 0 0.2266 1 1>}
  n22 = NamedRerouteUsage()
  n23 = ComponentMask(None=@n22) {g=true, r=true}
  n24 = NamedRerouteUsage()
  n25 = NamedRerouteUsage()
  n26 = ConstantBiasScale(None=TextureSample(UVs=Panner(Coordinate=@n23, Speed=@n24) {speedX=-0.05, speedY=0.02}, Tex=@n25) {samplerSource=SSM_Wrap_WorldGroupSettings, samplerType=SAMPLERTYPE_LinearColor}.A) {bias=-0.5, scale=2}
  n27 = ScalarParameter() {defaultValue=0.0338, group=Caustics, parameterName=Caustic Warp Strength, sliderMax=0.1}
  n28 = ComponentMask(None=Divide(A=@n22, B=Constant() {r=3})) {g=true, r=true}
  n29 = NamedRerouteUsage()
  n30 = ConstantBiasScale(None=TextureSample(UVs=Panner(Coordinate=@n28, Speed=@n29) {speedX=0.02, speedY=0.05}, Tex=@n25) {samplerSource=SSM_Wrap_WorldGroupSettings, samplerType=SAMPLERTYPE_LinearColor}.A) {bias=-0.5, scale=2}
  n31 = ScalarParameter() {defaultValue=0.0552, group=Caustics, parameterName=Caustic Warp Strength 2, sliderMax=0.1}
  NamedRerouteDeclaration(None=Add(A=Multiply(A=@n26, B=@n27), B=Multiply(A=@n30, B=@n31))) {name=Warp, nodeColor=<rgba 1 0 0.3281 1>}
  n32 = OneMinus(None=ScalarParameter() {defaultValue=0.2, parameterName=Saturation})
  NamedRerouteDeclaration(None=Power(Base=Desaturation(None=@n11, Fraction=@n32), Exp=ScalarParameter() {defaultValue=2.4, parameterName=PowerTightness})) {name=Caustics, nodeColor=<rgba 0 1 0.4609 1>}
  n33 = Multiply(A=Multiply(A=@n7.G, B=@n10.G), B=Multiply(A=@n7.B, B=@n10.B))
  n34 = Add(A=Multiply(A=Add(A=@n7.R, B=@n10.R), B=@n33), B=@n33)
  n35 = ScalarParameter() {defaultValue=3, group=Caustics, parameterName=CausticOpacityMult}
  Saturate(None=Power(Base=Multiply(A=@n34, B=@n35), Exp=ScalarParameter() {defaultValue=0.6, group=Caustics, parameterName=CausticOpacityPower}) {constExponent=1.1532})
 ]}
