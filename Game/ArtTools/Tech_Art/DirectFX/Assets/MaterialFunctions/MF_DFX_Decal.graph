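# Material function that blends a DFX decal over incoming material attributes, toggled by the Enable Decal static switch.
# A texture-derived alpha with invert, contrast, opacity and a UV-based SphereMask drives the tinted base color blend.
# An optional packed NMRE texture adds roughness, metallic or emissive, specular override and normals oriented by DFX_RightVector, partly tuned through custom primitive data.
# A platform switch keeps only base color on Switch, Android and PC-Perf.
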MaterialFunction /Game/ArtTools/Tech_Art/DirectFX/Assets/MaterialFunctions/MF_DFX_Decal
graph {nodes [
  FunctionOutput(None=StaticSwitchParameter(True=NamedRerouteUsage(), False=NamedRerouteUsage()) {group=DFX, parameterName=Enable Decal})
  NamedRerouteDeclaration(None=FunctionInput() {bUsePreviewValueAsDefault=true, inputType=FunctionInput_MaterialAttributes}) {name=Original Material, nodeColor=<rgba 0.0547 0 1 1>}
  n1 = GetMaterialAttributes(None=NamedRerouteUsage()) {attributeGetTypes [
    #1
    #2
    #3
    #4
    #5
    #6
   ]}
  n2 = FunctionInput(Preview=TextureCoordinate()) {bUsePreviewValueAsDefault=true, inputName=UVs, inputType=FunctionInput_Vector2, sortPriority=10}
  n3 = TextureSampleParameter2D(UVs=@n2) {
   group=DFX
   parameterName=DFX_NMRE
   texture=/Game/Global/Base4Layers_2022/Textures/T-GraffitiSample_D
   channelNames {a=Roughness, b=Metallic/Emissive, g=NormalY, r=NormalX}
  }
  n4 = TextureSampleParameter2D(UVs=@n2) {parameterName=DFX_Texture, texture=&T-GraffitiSample_D}
  n5 = VectorParameter() {parameterName=AlphaSource, defaultValue=<rgba 0 0 0 1>}
  n6 = DotProduct(
   A=@n4.RGBA
   B=MaterialFunctionCall(R=@n5.R, G=@n5.R, B=@n5.R, A=@n5.R) {materialFunction=&AppendMany}.RGBA
  )
  n7 = LinearInterpolate(A=@n6, B=OneMinus(None=@n6), Alpha=ScalarParameter() {parameterName=InvertAlpha})
  n8 = ScalarParameter() {defaultValue=1, parameterName=AlphaContrast}
  n9 = FunctionInput() {bUsePreviewValueAsDefault=true, inputName=Alpha, inputType=FunctionInput_Scalar, sortPriority=20}
  n10 = Saturate(None=ScalarParameter() {defaultValue=1, parameterName=DFX_Opacity})
  n11 = TextureCoordinate()
  n12 = Multiply(A=Subtract(A=FunctionInput(Preview=@n11) {bUsePreviewValueAsDefault=true, inputName=UV Unscaled, inputType=FunctionInput_Vector2, sortPriority=15}) {constB=0.5}) {constB=2}
  n13 = Power(Base=ScalarParameter() {defaultValue=1, parameterName=ToCircle, sliderMax=1})
  n14 = MaterialFunctionCall(Float2=Power(Base=Abs(None=@n12), Exp=LinearInterpolate(Alpha=@n13) {constA=2, constB=100})) {materialFunction=&BreakOutFloat2Components}
  n15 = Multiply(
   A=Multiply(A=Power(Base=@n7, Exp=@n8) {constExponent=1}, B=Multiply(A=@n9, B=@n10))
   B=SphereMask(A=Add(A=@n14.R, B=@n14.R), B=Constant(), Hardness=ScalarParameter() {parameterName=DFX_Decal_Hardness}) {attenuationRadius=1}
  )
  n16 = Multiply(A=@n15, B=ScalarParameter() {parameterName=DFX_Metallic, primitiveDataIndex=17})
  n17 = ScalarParameter() {group=DFX, parameterName=DFX_UseEmissive}
  n18 = LinearInterpolate(
   A=@n1.BaseColor
   B=Multiply(A=VectorParameter() {group=DFX, parameterName=DFX_TintColor, defaultValue=<rgba 1 1 1 0>}.RGB, B=@n4.RGB)
   Alpha=@n15
  )
  n19 = ScalarParameter() {parameterName=DFX_Roughness, primitiveDataIndex=18}
  n20 = AppendVector(A=@n3.NormalX, B=@n3.NormalX)
  n21 = DeriveNormalZ(InXY=Multiply(A=Subtract(A=@n20) {constB=0.5}) {constB=2})
  n22 = VertexNormalWS()
  n23 = VectorParameter() {group=DFX, parameterName=DFX_RightVector, primitiveDataIndex=4, defaultValue=<rgba 1 1 1 0>}
  n24 = Normalize(VectorInput=CrossProduct(A=@n22, B=Add(A=@n23.RGB) {constB=0.0001}))
  n25 = CrossProduct(A=@n24, B=@n22)
  n26 = VertexNormalWS()
  n27 = ScalarParameter() {parameterName=DFX_Normal, primitiveDataIndex=19}
  n28 = LinearInterpolate(
   A=@n1.EmissiveColor
   B=Multiply(A=@n15, B=ScalarParameter() {parameterName=DFX_Emissive, primitiveDataIndex=16})
   Alpha=@n3.Metallic/Emissive
  ) {constB=0}
  n29 = SetMaterialAttributes!(
   MaterialAttributes=@n1.MaterialAttributes
   BaseColor=@n18
   Metallic=LinearInterpolate(A=LinearInterpolate(A=@n1.Metallic, B=@n3.Metallic/Emissive, Alpha=@n16), B=@n1.Metallic, Alpha=@n17)
   Roughness=LinearInterpolate(A=@n1.Roughness, B=@n3.Roughness, Alpha=Multiply(A=@n15, B=@n19))
   Normal=LinearInterpolate(
    A=@n1.Normal
    B=MaterialFunctionCall(VectorToTransform=@n21, BasisX=@n25, BasisY=@n24, BasisZ=@n26) {materialFunction=&Transform3x3Matrix}.Transformed Vector
    Alpha=Multiply(A=@n15, B=@n27)
   )
   EmissiveColor=LinearInterpolate(A=@n1.EmissiveColor, B=Multiply(A=@n18, B=@n28), Alpha=@n17) {constB=0}
   Specular=LinearInterpolate(A=@n1.Specular, B=ScalarParameter() {defaultValue=0.5, parameterName=DFX_SpecularOverride, primitiveDataIndex=15}, Alpha=@n15) {constB=0}
  ) {attributeSetTypes [
    #1
    #2
    #3
    #4
    #5
    #6
   ]}
  n30 = SetMaterialAttributes!(MaterialAttributes=@n1.MaterialAttributes, BaseColor=@n18) {attributeSetTypes [#1]}
  n31 = StaticSwitchParameter(True=@n29, False=@n30) {group=DFX, parameterName=UseNSRTexture}
  NamedRerouteDeclaration(None=MaterialFunctionCall!("Xb1,PS4"=@n31, "PC-High&&Epic,XSX,PS5"=@n31, "PC-Perf,Android"=@n30, Switch=@n30, "PC-Low(NotPerf)&&Medium"=@n31) {materialFunction=/Game/Effects/Niagara/MasterMaterials/Functions/Common/PlatformSwitch_MA}.Result) {name=Decal Material, nodeColor=<rgba 0 0.8934 1 1>}
 ]}
