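# MF_4Layer_DFX is a material function that layers DirectFX projected decal and caustics effects onto incoming Material Attributes.
# The attributes pass through MF_DFX_Decal and then MF_DFX_Caustics, fed UVs, alpha and, for caustics, a falloff pivot through named reroutes.
# MF_ProjectionCoordinates supplies the projection UV variants, fresnel alpha, location and ratio.
# The "Enable DirectFX" static switch selects the effect result or a rerouted alternative at compile time.
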MaterialFunction /BRCoreMaterials/Materials/MaterialFunctions/MF_4Layer_DFX
graph {nodes [
  AppendVector(A=NamedRerouteUsage(), B=NamedRerouteUsage())
  n1 = FunctionInput() {inputType=FunctionInput_MaterialAttributes}
  n2 = MaterialFunctionCall!(
   In=MaterialFunctionCall!(
    In=NamedRerouteDeclaration(None=@n1) {name=Original Material, nodeColor=<rgba 0.2188 0 1 1>}
    UVs=NamedRerouteUsage()
    UV Unscaled=NamedRerouteUsage()
    Alpha=NamedRerouteUsage()
   ) {materialFunction=/Game/ArtTools/Tech_Art/DirectFX/Assets/MaterialFunctions/MF_DFX_Decal}.Result
   UV=NamedRerouteUsage()
   UV Unscaled=NamedRerouteUsage()
   Alpha=NamedRerouteUsage()
   Falloff Pivot=NamedRerouteUsage()
  ) {materialFunction=/Game/ArtTools/Tech_Art/DirectFX/Assets/MaterialFunctions/MF_DFX_Caustics}
  FunctionOutput(None=StaticSwitchParameter(True=NamedRerouteDeclaration(None=@n2.Result) {name=DFX Material, nodeColor=<rgba 1 0 0.75 1>}, False=NamedRerouteUsage()) {group=- Effect Switches, parameterName=Enable DirectFX})
  n3 = MaterialFunctionCall!(World Position) {materialFunction=/Game/ArtTools/Tech_Art/DirectFX/Assets/MaterialFunctions/MF_ProjectionCoordinates}
  NamedRerouteDeclaration(None=@n3.UVCoord) {name=UVCoord, nodeColor=<rgba 1 0 0.9844 1>}
  NamedRerouteDeclaration(None=@n3.Projection Fresnel) {name=Projection Alpha, nodeColor=<rgba 1 0 0.0937 1>}
  NamedRerouteDeclaration(None=@n3.Location) {name=Projection Location, nodeColor=<rgba 1 0 0.7266 1>}
  NamedRerouteDeclaration(None=@n3.UVCoord Unscaled) {name=UVCoord Unscaled, nodeColor=<rgba 1 0 0.9844 1>}
  NamedRerouteDeclaration(None=@n3.X/Y Ratio) {name=Ratio, nodeColor=<rgba 0 0.2734 1 1>}
  NamedRerouteDeclaration(None=@n3.UVCoord Const Ratio) {name=UVCoord Const Ratio, nodeColor=<rgba 1 0 0.9844 1>}
  TextureCoordinate()
 ]}
