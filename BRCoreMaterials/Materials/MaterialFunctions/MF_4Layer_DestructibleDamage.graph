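# Material function that adds gameplay-driven destructible damage, cracked surface visuals and a hit wobble, to an incoming material.
# Crack mask and normal textures use local-space axis projection chosen by the dominant world normal, altering BaseColor and Normal.
# World position offset fades radially around a deformation center, with a camera distance cutoff input.
# Damage values default to custom primitive data, static switches gate the effect, and quality or mobile switches simplify it.
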MaterialFunction /BRCoreMaterials/Materials/MaterialFunctions/MF_4Layer_DestructibleDamage
graph {nodes [
  ComponentMask(None=TextureCoordinate()) {r=true}
  n1 = If(
   A=NamedRerouteUsage()
   "A > B"=SetMaterialAttributes!(MaterialAttributes=NamedRerouteUsage(), BaseColor=NamedRerouteUsage(), Normal=NamedRerouteUsage()) {attributeSetTypes [#1, #2]}
   "A < B"=NamedRerouteUsage()
  )
  n2 = SetMaterialAttributes!(
   MaterialAttributes=@n1
   WorldPositionOffset=If(A=NamedRerouteUsage(), B=NamedRerouteUsage(), "A > B"=NamedRerouteUsage(), "A < B"=NamedRerouteUsage())
  ) {attributeSetTypes [#3]}
  n3 = NamedRerouteUsage()
  FunctionOutput(None=StaticSwitch(True=StaticSwitch(True=@n2, False=@n1, Value=@n3), False=NamedRerouteUsage(), Value=NamedRerouteUsage())) {outputName=Material}
  n4 = VectorParameter() {
   bUseCustomPrimitiveData=true
   desc=Destructible Damage. This parameter is controlled by gameplay code.
   group=z_GameplayParameters_DestructibleDamage
   parameterName=Code_PrecomputedHitVector
   primitiveDataIndex=23
   sortPriority=110
   defaultValue=<rgba 0 0 1 0>
  }
  n5 = VectorParameter() {
   bUseCustomPrimitiveData=true
   desc=Destructible Damage. This parameter is controlled by gameplay code.
   group=z_GameplayParameters_DestructibleDamage
   parameterName=Code_PrecomputedHitVector_Prev_Frame
   primitiveDataIndex=32
   sortPriority=120
   defaultValue=<rgba 0 0 1 0>
  }
  NamedRerouteDeclaration(None=PreviousFrameSwitch(Current Frame=@n4.RGB, Previous Frame=@n5.RGB)) {name=PrecomputedHitVector, nodeColor=<rgba 0 0 0 1>}
  n6 = StaticBoolParameter() {
   defaultValue=true
   desc="DestructibleDamage: Enables the cracked visuals && the animated wobble."
   group=- Effect Switches
   parameterName=Destructible Damage
   sortPriority=0
  }
  n7 = FunctionInput(Preview=@n6) {
   bUsePreviewValueAsDefault=true
   description="Default=True. Enables the cracked visuals && the WPO wobble. False will just pass the material through without modification."
   inputName=UseDestructibleDamage
   inputType=FunctionInput_StaticBool
   sortPriority=100
  }
  NamedRerouteDeclaration(None=@n7) {name=UseDestructibleDamage, nodeColor=<rgba 0.7005 0 1 1>}
  n8 = ScalarParameter() {
   bUseCustomPrimitiveData=true
   desc="Destructible Damage. This parameter is controlled by gameplay code.\r\nDefault=Custom Primitive Data Index 13"
   group=z_GameplayParameters_DestructibleDamage
   parameterName=Code_DestructionAmount
   primitiveDataIndex=13
   sortPriority=105
  }
  n9 = FunctionInput(Preview=@n8) {
   bUsePreviewValueAsDefault=true
   description=Default parameter is controlled by code. Custom Primitive Data Index 13
   inputName=DestructibleDamageAmount
   inputType=FunctionInput_Scalar
   sortPriority=200
   previewValue=<0 1 0 1>
  }
  NamedRerouteDeclaration(None=@n9) {name=DestructibleDamageAmount, nodeColor=<rgba 1 0 0.0911 1>}
  n10 = FunctionInput(Preview=StaticBool() {value=true}) {
   bUsePreviewValueAsDefault=true
   description="Default=True. Enables the damage wobble when hit."
   inputName=UseDamageWobbleWPO
   inputType=FunctionInput_StaticBool
   sortPriority=130
  }
  NamedRerouteDeclaration(None=@n10) {name=EnableWobbleWPO, nodeColor=<rgba 0.2791 1 0.2835 1>}
  n11 = Constant3Vector() {constant=<rgba 1 1 1 0>}
  n12 = Normalize(VectorInput=VertexNormalWS())
  n13 = Constant()
  n14 = GetMaterialAttributes(None=FunctionInput(Preview=SetMaterialAttributes!(MaterialAttributes, BaseColor=@n11, Normal=@n12, WorldPositionOffset=@n13) {attributeSetTypes [#1, #2, #3]}) {bUsePreviewValueAsDefault=true, inputName=Material_In, inputType=FunctionInput_MaterialAttributes}) {attributeGetTypes [#1, #2, #3]}
  NamedRerouteDeclaration(None=@n14.MaterialAttributes) {name=Material Input, nodeColor=<rgba 1 0.4035 0.3566 1>}
  NamedRerouteDeclaration(None=@n14.WorldPositionOffset) {name=WPO Input, nodeColor=<rgba 1 0 0.9844 1>}
  NamedRerouteDeclaration(None=@n14.Normal) {name=Normal Input, nodeColor=<rgba 0.875 1 0 1>}
  NamedRerouteDeclaration(None=Distance(A=ActorPositionWS() {originType=CameraRelative}, B=Constant())) {name=Distance To Actor, nodeColor=<rgba 0 0.0859 1 1>}
  n15 = WorldPosition() {worldPositionShaderOffset=WPT_CameraRelative}
  NamedRerouteDeclaration(None=Add(
    A=FunctionInput(Preview=@n15.XYZ) {bUsePreviewValueAsDefault=true, description="Default=Absolute World Position", inputName=WorldPosition, sortPriority=150}
    B=NamedRerouteUsage()
   )) {name=WorldPos, nodeColor=<rgba 1 0 0 1>}
  n16 = Constant() {r=12000}
  n17 = Constant() {r=3000}
  n18 = Constant() {r=6000}
  n19 = FunctionInput(Preview=ShadingPathSwitch(Default=QualitySwitch(Default=@n16, Low=@n17, High=@n16, Medium=@n18, Epic=@n16), Mobile=Constant() {r=800})) {
   bUsePreviewValueAsDefault=true
   description="Default=Absolute World Position"
   inputName=WPO Distance Cutoff
   inputType=FunctionInput_Scalar
   sortPriority=150
  }
  NamedRerouteDeclaration(None=@n19) {name=WPO Distance Cutoff, nodeColor=<rgba 0 0.0859 1 1>}
  n20 = Transform(None=LocalPosition().XYZ) {transformSourceType=TRANSFORMSOURCE_Local}
  n21 = Constant() {r=700}
  n22 = NamedRerouteDeclaration(None=VertexInterpolator(VS=Divide(A=@n20, B=@n21)).PS) {name=Local Coords, nodeColor=<rgba 0 0.7656 1 1>}
  MaterialExpressionAbs_0 = Abs(None=VertexNormalWS())
  n23 = Convert(None=@MaterialExpressionAbs_0) {
   nodeName=Break Float3
   convertInputs [{
     type=Vector3
     expressionInput {
      expression=&self:MaterialExpressionAbs_0
      inputName=None
      mask=0
      maskA=0
      maskB=0
      maskG=0
      maskR=0
      outputIndex=0
     }
    }]
   convertOutputs [{type=Scalar}, {type=Scalar}, {type=Scalar}]
  }
  n24 = AppendVector(A=Subtract(A=@n23, B=Max(A=@n23, B=@n23)), B=Subtract(A=@n23, B=Max(A=@n23, B=@n23)))
  n25 = NamedRerouteDeclaration(None=Ceil(None=Saturate(None=@n24))) {name=Triplanar Mask, nodeColor=<rgba 0 0.7891 1 1>}
  n26 = LinearInterpolate(
   A=LinearInterpolate(A=ComponentMask(None=@n22) {g=true, r=true}, B=ComponentMask(None=@n22) {b=true, r=true}, Alpha=ComponentMask(None=@n25) {g=true})
   B=ComponentMask(None=@n22) {b=true, g=true}
   Alpha=ComponentMask(None=@n25) {r=true}
  )
  NamedRerouteDeclaration(None=TextureSample(UVs=@n26) {texture=/Game/Packages/jl_tests/CrackMaterial/T_crack_FX_MASK}.RGB) {name=Texture Mask, nodeColor=<rgba 0.6875 0 1 1>}
  n27 = NamedRerouteUsage()
  n28 = LinearInterpolate(A=NamedRerouteUsage(), B=Constant3Vector() {constant=<rgba 0 0 1 0>}, Alpha=Saturate(None=Add(A=@n27) {constB=0.3}))
  n29 = ComponentMask(None=@n28) {r=true}
  n30 = NamedRerouteUsage()
  n31 = ComponentMask(None=@n30) {r=true}
  n32 = Add(A=@n29, B=@n31)
  n33 = ComponentMask(None=@n28) {g=true}
  n34 = ComponentMask(None=@n30) {g=true}
  n35 = Add(A=@n33, B=@n34)
  n36 = ComponentMask(None=@n28) {b=true}
  n37 = ComponentMask(None=@n30) {b=true}
  n38 = Multiply(A=@n36, B=@n37)
  n39 = Multiply(A=@n36, B=@n34)
  n40 = Add(A=@n33, B=@n37)
  n41 = NamedRerouteUsage()
  n42 = LinearInterpolate(
   A=LinearInterpolate(
    A=MaterialFunctionCall(R=@n32, G=@n35, B=@n38) {materialFunction=&AppendMany}.RGB
    B=MaterialFunctionCall(R=@n32, G=@n39, B=@n40) {materialFunction=&AppendMany}.RGB
    Alpha=ComponentMask(None=@n41) {g=true}
   )
   B=MaterialFunctionCall(R=Multiply(A=@n36, B=@n31), G=Add(A=@n29, B=@n34), B=@n40) {materialFunction=&AppendMany}.RGB
   Alpha=ComponentMask(None=@n41) {r=true}
  )
  n43 = NamedRerouteUsage()
  NamedRerouteDeclaration(None=ShadingPathSwitch(Default=QualitySwitch(Default=@n42, Low=@n43, High=@n42, Medium=@n43, Epic=@n42), Mobile=@n43)) {name=Normal Cracks, nodeColor=<rgba 0.1953 0 1 1>}
  n44 = Multiply(A=Add(A=NamedRerouteUsage(), B=Constant3Vector() {constant=<rgba -0.06 0 -0.22 0>})) {constB=4}
  MaterialExpressionMultiply_29 = Multiply(A=@n44, B=@n44)
  n45 = Convert(None=@MaterialExpressionMultiply_29) {
   nodeName=Break Float3
   convertInputs [{
     type=Vector3
     expressionInput {
      expression=&self:MaterialExpressionMultiply_29
      inputName=None
      mask=0
      maskA=0
      maskB=0
      maskG=0
      maskR=0
      outputIndex=0
     }
    }]
   convertOutputs [{type=Scalar}, {type=Scalar}, {type=Scalar}]
  }
  n46 = Add(A=Add(A=@n45, B=@n45), B=@n45)
  n47 = Multiply(A=NamedRerouteUsage()) {constB=2}
  NamedRerouteDeclaration(None=Saturate(None=Add(A=@n46, B=@n47))) {name=Radial Pattern, nodeColor=<rgba 1 0.5156 0 1>}
  NamedRerouteDeclaration(None=TextureSample(UVs=@n26) {samplerType=SAMPLERTYPE_Normal, texture=/Game/Packages/jl_tests/CrackMaterial/T_crack_Normal}.RGB) {name=Texture Normal, nodeColor=<rgba 0 0.1562 1 1>}
  n48 = FunctionInput(Preview=StaticBool()) {
   bUsePreviewValueAsDefault=true
   description="Default=False"
   inputName=UseAlternateHitGradientOutput
   inputType=FunctionInput_StaticBool
   sortPriority=180
  }
  NamedRerouteDeclaration(None=@n48) {name=EnableDamageCracks, nodeColor=<rgba 0.3511 0.6367 1 1>}
  n49 = NamedRerouteUsage()
  n50 = MaterialFunctionCall(
   R=LinearInterpolate(Alpha=@n49) {constB=4}
   G=LinearInterpolate(Alpha=@n49) {constB=1.2}
   B=LinearInterpolate(Alpha=@n49) {constA=-1.7, constB=1.5}
  ) {materialFunction=&AppendMany}
  n51 = NamedRerouteUsage()
  n52 = Constant3Vector() {constant=<rgba 0.8 0.7 1.5 0>}
  MaterialExpressionMultiply_26 = Multiply(A=Multiply(A=Saturate(None=@n50.RGB), B=NamedRerouteUsage()), B=OneMinus(None=Multiply(A=@n51, B=@n52)))
  n53 = Convert(None=@MaterialExpressionMultiply_26) {
   nodeName=Break Float3
   convertInputs [{
     type=Vector3
     expressionInput {
      expression=&self:MaterialExpressionMultiply_26
      inputName=None
      mask=0
      maskA=0
      maskB=0
      maskG=0
      maskR=0
      outputIndex=0
     }
    }]
   convertOutputs [{type=Scalar}, {type=Scalar}, {type=Scalar}]
  }
  n54 = Max(A=@n53, B=@n53)
  n55 = NamedRerouteDeclaration(None=OneMinus(None=Max(A=@n54, B=@n53))) {name=Normal Mask, nodeColor=<rgba 0.5703 0 1 1>}
  n56 = Saturate(None=Add(A=Multiply(A=@n55) {constB=1.4}) {constB=0.15})
  NamedRerouteDeclaration(None=Multiply(A=NamedRerouteDeclaration(None=@n56) {name=Cracks Mask, nodeColor=<rgba 0 0.1974 0.5156 1>}, B=NamedRerouteUsage())) {name=BaseColor Cracks, nodeColor=<rgba 0.6406 0 1 1>}
  n57 = ScalarParameter() {
   bUseCustomPrimitiveData=true
   defaultValue=1
   desc="Destructible Damage. This parameter is controlled by gameplay code.\r\nDefault=Custom Primitive Data Index 26"
   group=z_GameplayParameters_DestructibleDamage
   parameterName=WSDamageRadiusAlt
   primitiveDataIndex=27
   sortPriority=140
  }
  n58 = ScalarParameter() {
   bUseCustomPrimitiveData=true
   defaultValue=1
   desc="Destructible Damage. This parameter is controlled by gameplay code.\r\nDefault=Custom Primitive Data Index 26"
   group=z_GameplayParameters_DestructibleDamage
   parameterName=WSDamageRadius
   primitiveDataIndex=26
   sortPriority=140
  }
  n59 = FunctionInput(Preview=StaticBool()) {bUsePreviewValueAsDefault=true, inputName=UseAlternativeCustomPrimDataForDestruction, inputType=FunctionInput_StaticBool}
  n60 = FunctionInput(Preview=StaticSwitch(True=@n57, False=@n58, Value=@n59)) {
   bUsePreviewValueAsDefault=true
   description="Default parameter is controlled by code. Default=Custom Primitive Data Index 26"
   inputName=WSDamageRadius
   inputType=FunctionInput_Scalar
   sortPriority=220
  }
  NamedRerouteDeclaration(None=@n60) {name=WSDamageRadius, nodeColor=<rgba 0 0 0 1>}
  n61 = VectorParameter() {
   bUseCustomPrimitiveData=true
   desc="Destructible Damage. This parameter is controlled by gameplay code.\r\nDefault=Custom Primitive Data Index 20,21,22,23"
   group=z_GameplayParameters_DestructibleDamage
   parameterName=WSDeformationCenterAlt
   primitiveDataIndex=7
   sortPriority=130
  }
  n62 = VectorParameter() {
   bUseCustomPrimitiveData=true
   desc="Destructible Damage. This parameter is controlled by gameplay code.\r\nDefault=Custom Primitive Data Index 20,21,22,23"
   group=z_GameplayParameters_DestructibleDamage
   parameterName=WSDeformationCenter
   primitiveDataIndex=20
   sortPriority=130
  }
  n63 = FunctionInput(Preview=TransformPosition(None=StaticSwitch(True=@n61.RGB, False=@n62.RGB, Value=@n59)) {transformSourceType=TRANSFORMPOSSOURCE_World, transformType=TRANSFORMPOSSOURCE_TranslatedWorld}) {
   bUsePreviewValueAsDefault=true
   description="Default parameter is controlled by code. Default=Custom Primitive Data Index 20,21,22,23"
   inputName=WSDeformationCenter
   sortPriority=210
  }
  NamedRerouteDeclaration(None=@n63) {name=WSDeformationCenter, nodeColor=<rgba 1 0 0 1>}
  NamedRerouteDeclaration(None=@n14.BaseColor) {name=BaseColor Input, nodeColor=<rgba 0 0.5547 1 1>}
  n64 = NamedRerouteUsage()
  n65 = NamedRerouteUsage()
  n66 = Saturate(None=Multiply(A=Distance(A=@n64, B=@n65), B=NamedRerouteUsage()))
  n67 = NamedRerouteUsage()
  n68 = Add(A=Multiply(A=OneMinus(None=@n66), B=NamedRerouteUsage()), B=@n67)
  NamedRerouteDeclaration(None=ShadingPathSwitch(Default=@n68, Mobile=QualitySwitch(Default=@n68, Low=@n67))) {name=WPO Cracks, nodeColor=<rgba 1 0.8135 0 1>}
 ]}
