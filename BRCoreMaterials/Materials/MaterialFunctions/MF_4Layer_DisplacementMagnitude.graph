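# Material function that scales a material's displacement by separate magnitudes for up to four layers.
# MF_4Layer_ScalarMasking combines four per-layer magnitude parameters using the RGBA Layer Mask input, and the result multiplies the incoming Displacement attribute.
# A Displacement static switch writes either this product or a zero constant back into the material attributes.
# The unscaled displacement is also exposed as a Displacement Texture output.
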MaterialFunction /BRCoreMaterials/Materials/MaterialFunctions/MF_4Layer_DisplacementMagnitude
graph {nodes [
  n1 = GetMaterialAttributes(None=FunctionInput() {inputName=Material, inputType=FunctionInput_MaterialAttributes}) {attributeGetTypes [#1]}
  n2 = Multiply(
   A=FunctionOutput(None=@n1.Displacement) {outputName=Displacement Texture, sortPriority=1}
   B=MaterialFunctionCall!(
    RGBA Mask=NamedRerouteUsage()
    L1=ScalarParameter() {group=1 Layer, parameterName=L1 Displacement Magnitude, sliderMax=1, sortPriority=8}
    L2=ScalarParameter() {group=2 Layer, parameterName=L2 Displacement Magnitude, sliderMax=1, sortPriority=8}
    L3=ScalarParameter() {group=3 Layer, parameterName=L3 Displacement Magnitude, sliderMax=1, sortPriority=8}
    L4=ScalarParameter() {group=4 Layer, parameterName=L4 Displacement Magnitude, sliderMax=1, sortPriority=8}
   ) {materialFunction=&/MF_4Layer_ScalarMasking}.Result
  )
  n3 = Constant()
  n4 = StaticBoolParameter() {group=- Effect Switches, parameterName=Displacement}
  FunctionOutput(None=SetMaterialAttributes!(MaterialAttributes=@n1.MaterialAttributes, Displacement=StaticSwitch(True=@n2, False=@n3, Value=@n4)) {attributeSetTypes [#1]})
  n5 = FunctionInput() {bUsePreviewValueAsDefault=true, inputName=Layer Mask, inputType=FunctionInput_Vector4, sortPriority=1, previewValue=<1 1 1 1>}
  NamedRerouteDeclaration(None=@n5) {name=Layer Mask, nodeColor=<rgba 0 0.5312 1 1>}
 ]}
