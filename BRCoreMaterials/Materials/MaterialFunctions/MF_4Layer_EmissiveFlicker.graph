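# Material function that adds an optional emissive flicker to a four-layer material, passing attributes through unchanged when the Emissive Flicker static switch is off.
# Four triangle waves at separate speeds are summed, offset, floored at a minimum intensity and saturated to scale EmissiveColor.
# A time shift hashed from actor world position desynchronizes instances.
# Per-layer weights, combined through the Layer Mask input, control how much each layer flickers.
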MaterialFunction /BRCoreMaterials/Materials/MaterialFunctions/MF_4Layer_EmissiveFlicker
graph {nodes [
  n1 = FunctionInput() {inputName=Material, inputType=FunctionInput_MaterialAttributes}
  n2 = GetMaterialAttributes(None=@n1) {attributeGetTypes [#1]}
  n3 = VectorParameter() {
   group=Emissive Flicker
   parameterName=EmissiveFlicker Speed
   sortPriority=4
   channelNames {a=Speed 4, b=Speed 3, g=Speed 2, r=Speed 1}
   defaultValue=<rgba 0.64 1.775 4.745 3.115>
  }
  n4 = ActorPositionWS()
  n5 = Constant3Vector() {constant=<rgba 0.2457 0.1237 0.458 0>}
  n6 = MaterialFunctionCall(Float3=Frac(None=Multiply(A=@n4, B=@n5) {constB=0.15})) {materialFunction=&BreakOutFloat3Components}
  n7 = Add(A=Add(A=Add(A=@n6.R, B=@n6.R), B=@n6.B), B=Time())
  n8 = Add(A=Frac(None=Multiply(A=@n3.RGBA, B=@n7))) {constB=-0.5}
  n9 = DotProduct(A=Multiply(A=Abs(None=@n8)) {constB=2}, B=Constant4Vector() {constant=<rgba 1 1 1 1>}.RGBA)
  n10 = ScalarParameter() {defaultValue=0.2, group=Emissive Flicker, parameterName=EmissiveFlicker Minimum Intensity, sliderMax=1, sortPriority=6}
  n11 = Max(A=Add(A=Add(A=@n9) {constB=-1}, B=ScalarParameter() {group=Emissive Flicker, parameterName=EmissiveFlicker Offset, sortPriority=5}), B=@n10)
  n12 = LinearInterpolate(
   A=@n2.EmissiveColor
   B=Multiply(A=@n2.EmissiveColor, B=Saturate(None=@n11))
   Alpha=MaterialFunctionCall!(
    RGBA Mask=NamedRerouteUsage()
    L1=ScalarParameter() {defaultValue=1, group=Emissive Flicker, parameterName=L1 EmissiveFlicker Weight, sliderMax=1, sortPriority=0}
    L2=ScalarParameter() {defaultValue=1, group=Emissive Flicker, parameterName=L2 EmissiveFlicker Weight, sliderMax=1, sortPriority=1}
    L3=ScalarParameter() {defaultValue=1, group=Emissive Flicker, parameterName=L3 EmissiveFlicker Weight, sliderMax=1, sortPriority=2}
    L4=ScalarParameter() {defaultValue=1, group=Emissive Flicker, parameterName=L4 EmissiveFlicker Weight, sliderMax=1, sortPriority=3}
   ) {materialFunction=&/MF_4Layer_ScalarMasking}.Result
  )
  FunctionOutput(None=StaticSwitchParameter(
    True=SetMaterialAttributes!(MaterialAttributes=@n2.MaterialAttributes, EmissiveColor=@n12) {attributeSetTypes [#1]}
    False=@n1
   ) {group=- Effect Switches, parameterName=Emissive Flicker})
  n13 = FunctionInput() {bUsePreviewValueAsDefault=true, inputName=Layer Mask, inputType=FunctionInput_Vector4, sortPriority=1, previewValue=<1 1 1 1>}
  NamedRerouteDeclaration(None=@n13) {name=Layer Mask, nodeColor=<rgba 0 0.5312 1 1>}
 ]}
