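# Material function that builds one layer, apparently for a four-layer material, outputting material attributes plus a mask.
# It samples diffuse, normal, packed specular/metallic/roughness and displacement textures, with static switches selecting tiled or world-space UVs.
# World-space sampling uses explicit derivatives with a Nanite-specific alternative, while displacement blends three axis projections by mask and fades through a zipper mask.
# Emissive can derive from a normalized color and gets a Lumen boost under ray tracing.
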MaterialFunction /BRCoreMaterials/Materials/MaterialFunctions/MF_4Layer_Layer
graph {nodes [
  n1 = SetMaterialAttributes!(
   MaterialAttributes
   BaseColor=NamedRerouteUsage()
   Normal=NamedRerouteUsage()
   Specular=NamedRerouteUsage()
   Metallic=NamedRerouteUsage()
   Roughness=NamedRerouteUsage()
   OpacityMask=NamedRerouteUsage()
   EmissiveColor=NamedRerouteUsage()
   Displacement=NamedRerouteUsage()
  ) {attributeSetTypes [
    #1
    #2
    #3
    #4
    #5
    #6
    #7
    #8
   ]}
  FunctionOutput(None=@n1) {outputName=Material}
  FunctionOutput(None=NamedRerouteUsage()) {outputName=Mask, sortPriority=1}
  n2 = StaticSwitch(True=NamedRerouteUsage(), False=NamedRerouteUsage(), Value=StaticBoolParameter() {group=- Effect Switches, parameterName=World Space UVs})
  NamedRerouteDeclaration(None=ComponentMask(None=@n2) {b=true, g=true, r=true}) {name=Diffuse Color, nodeColor=<rgba 0.7344 1 0 1>}
  NamedRerouteDeclaration(None=StaticSwitch(True=NamedRerouteUsage(), False=NamedRerouteUsage(), Value=StaticBoolParameter() {group=- Effect Switches, parameterName=World Space UVs})) {name=Normals Color, nodeColor=<rgba 0.5494 0.0564 1 1>}
  n3 = FunctionInput() {bUsePreviewValueAsDefault=true, inputName=Emissive Color}
  n4 = Normalize(VectorInput=NamedRerouteUsage())
  n5 = FunctionInput() {bUsePreviewValueAsDefault=true, inputName=Emissive FromBaseColor, inputType=FunctionInput_Scalar}
  n6 = Multiply(
   A=Multiply(A=LinearInterpolate(A=@n3, B=@n4, Alpha=@n5) {constA=1}, B=NamedRerouteUsage())
   B=FunctionInput() {bUsePreviewValueAsDefault=true, inputName=Emissive Intensity, inputType=FunctionInput_Scalar}
  )
  n7 = FunctionInput() {bUsePreviewValueAsDefault=true, inputName=Emissive Lumen Boost, inputType=FunctionInput_Scalar, previewValue=<1 1 1 1>}
  NamedRerouteDeclaration(None=RayTracingQualitySwitch(Normal=@n6, RayTraced=Multiply(A=@n6, B=@n7))) {name=Emissive Color, nodeColor=<rgba 1 0.3281 0 1>}
  NamedRerouteDeclaration(None=ComponentMask(None=@n2) {a=true}) {name=Mask Value, nodeColor=<rgba 1 0.9163 0.7587 1>}
  NamedRerouteDeclaration(None=StaticSwitch(True=NamedRerouteUsage(), False=NamedRerouteUsage(), Value=StaticBoolParameter() {group=- Effect Switches, parameterName=World Space UVs})) {name=Displacement Value, nodeColor=<rgba 0 1 0.3672 1>}
  n8 = NamedRerouteUsage()
  n9 = NamedRerouteUsage()
  n10 = TextureSample!(UVs=@n8, Tex=@n9, "DDX(UVs)"=NamedRerouteUsage(), "DDY(UVs)"=NamedRerouteUsage(), Apply View MipBias) {mipValueMode=TMVM_Derivative, samplerSource=SSM_Wrap_WorldGroupSettings, texture=/BRCoreMaterials/Materials/Textures/T_BaseColor}
  NamedRerouteDeclaration(None=NaniteReplace!(
    Default=@n10.RGBA
    Nanite=TextureSample!(UVs=@n8, Tex=@n9, Apply View MipBias) {samplerSource=SSM_Wrap_WorldGroupSettings, texture=&T_BaseColor}.RGBA
   )) {name=Diffuse WS, nodeColor=<rgba 0.7344 1 0 1>}
  n11 = NamedRerouteUsage()
  n12 = FunctionInput() {bUsePreviewValueAsDefault=true, inputName=Displacement Zipper Mask, inputType=FunctionInput_Scalar, previewValue=<0 0 0 0>}
  n13 = FunctionInput() {bUsePreviewValueAsDefault=true, inputName=Displacement Zipper Exp, inputType=FunctionInput_Scalar, previewValue=<0 0 0 0>}
  n14 = OneMinus(None=Power(Base=@n12, Exp=@n13))
  NamedRerouteDeclaration(None=StaticSwitch(True=Multiply(A=@n11, B=@n14), False=Constant(), Value=StaticBoolParameter() {group=- Effect Switches, parameterName=Displacement})) {name=Displacement, nodeColor=<rgba 0 0.1562 1 1>}
  n15 = NamedRerouteUsage()
  n16 = NamedRerouteUsage()
  n17 = TextureSample!(UVs=@n15, Tex=@n16, "DDX(UVs)"=NamedRerouteUsage(), "DDY(UVs)"=NamedRerouteUsage(), Apply View MipBias) {
   mipValueMode=TMVM_Derivative
   samplerSource=SSM_Wrap_WorldGroupSettings
   samplerType=SAMPLERTYPE_Normal
   texture=/BRCoreMaterials/Materials/Textures/T_Normal
  }
  n18 = TextureSample!(UVs=@n15, Tex=@n16, Apply View MipBias) {samplerSource=SSM_Wrap_WorldGroupSettings, samplerType=SAMPLERTYPE_Normal, texture=&T_Normal}
  NamedRerouteDeclaration(None=NaniteReplace(Default=@n17.RGB, Nanite=@n18.RGB)) {name=Normals WS, nodeColor=<rgba 0.5494 0.0564 1 1>}
  n19 = NamedRerouteUsage()
  n20 = NamedRerouteUsage()
  n21 = TextureSample!(UVs=@n19, Tex=@n20, "DDX(UVs)"=NamedRerouteUsage(), "DDY(UVs)"=NamedRerouteUsage(), Apply View MipBias) {
   mipValueMode=TMVM_Derivative
   samplerSource=SSM_Wrap_WorldGroupSettings
   samplerType=SAMPLERTYPE_Masks
   texture=/BRCoreMaterials/Materials/Textures/T_Specular
  }
  n22 = TextureSample!(UVs=@n19, Tex=@n20, Apply View MipBias) {samplerSource=SSM_Wrap_WorldGroupSettings, samplerType=SAMPLERTYPE_Masks, texture=&T_Specular}
  NamedRerouteDeclaration(None=NaniteReplace(Default=@n21.R, Nanite=@n22.R)) {name=Specular WS, nodeColor=<rgba 1 0 0.0179 1>}
  NamedRerouteDeclaration(None=NaniteReplace(Default=@n21.G, Nanite=@n22.G)) {name=Metallic WS, nodeColor=<rgba 0.1298 1 0 1>}
  NamedRerouteDeclaration(None=NaniteReplace(Default=@n21.B, Nanite=@n22.B)) {name=Roughness WS, nodeColor=<rgba 0 0.0454 1 1>}
  n23 = FunctionInput() {bUsePreviewValueAsDefault=true, inputName=Displacement WorldSpace Mask, previewValue=<0.5 0.5 0.5 1>}
  n24 = ComponentMask(None=@n23) {r=true}
  n25 = FunctionInput() {bUsePreviewValueAsDefault=true, inputName=Displacement WorldSpace UVX, inputType=FunctionInput_Vector2, previewValue=<0.5 0.5 0.5 1>}
  n26 = NamedRerouteUsage()
  n27 = TextureSample(UVs=@n25, Tex=@n26) {samplerType=SAMPLERTYPE_LinearGrayscale, texture=/BRCoreMaterials/Materials/Textures/T_Displacement}
  n28 = ComponentMask(None=@n23) {g=true}
  n29 = FunctionInput() {bUsePreviewValueAsDefault=true, inputName=Displacement WorldSpace UVY, inputType=FunctionInput_Vector2, previewValue=<0.5 0.5 0.5 1>}
  n30 = TextureSample(UVs=@n29, Tex=@n26) {samplerType=SAMPLERTYPE_LinearGrayscale, texture=&T_Displacement}
  n31 = Add(
   A=IfThenElse!(Condition=@n24, True=Multiply(A=@n27.R, B=@n24), False=Constant())
   B=IfThenElse(Condition=@n28, True=Multiply(A=@n30.R, B=@n28), False=Constant())
  )
  n32 = ComponentMask(None=@n23) {b=true}
  n33 = FunctionInput() {bUsePreviewValueAsDefault=true, inputName=Displacement WorldSpace UVZ, inputType=FunctionInput_Vector2, previewValue=<0.5 0.5 0.5 1>}
  n34 = IfThenElse(
   Condition=@n32
   True=Multiply(
    A=TextureSample(UVs=@n33, Tex=@n26) {samplerType=SAMPLERTYPE_LinearGrayscale, texture=&T_Displacement}.R
    B=@n32
   )
   False=Constant()
  )
  NamedRerouteDeclaration(None=IfThenElse(Condition=NamedRerouteUsage(), True=Add(A=@n31, B=@n34), False=NamedRerouteUsage())) {name=Displacement WS, nodeColor=<rgba 1 0 0.2109 1>}
  n35 = FunctionInput() {bUsePreviewValueAsDefault=true, inputName=Texcoords, inputType=FunctionInput_Vector2, sortPriority=3, previewValue=<1 1 1 1>}
  n36 = Subtract(A=@n35, B=NamedRerouteUsage())
  n37 = FunctionInput() {bUsePreviewValueAsDefault=true, inputName=Tiling, inputType=FunctionInput_Scalar, sortPriority=3, previewValue=<1 1 1 1>}
  n38 = NamedRerouteDeclaration(None=Add(A=Multiply(A=@n36, B=@n37), B=NamedRerouteUsage())) {name=Texcoord UVs, nodeColor=<rgba 1 0.3047 0 1>}
  n39 = IfThenElse(
   Condition=NamedRerouteDeclaration(None=FunctionInput() {bUsePreviewValueAsDefault=true, inputName=WorldSpaceUVs Mode, inputType=FunctionInput_Scalar}) {name=WorldSpaceUVs Mode, nodeColor=<rgba 0.1016 0 1 1>}
   True=FunctionInput() {bUsePreviewValueAsDefault=true, inputName=WorldSpaceUVs, inputType=FunctionInput_Vector4}
   False=AppendVector(A=@n38, B=@n38)
  )
  NamedRerouteDeclaration(None=ComponentMask(None=@n39) {g=true, r=true}) {name=WorldSpace UVs, nodeColor=<rgba 1 0 0.004 1>}
  NamedRerouteDeclaration(None=TextureSample(UVs=NamedRerouteUsage(), Tex=NamedRerouteUsage()) {samplerType=SAMPLERTYPE_LinearGrayscale, texture=&T_Displacement}.R) {name=Displacement T0, nodeColor=<rgba 0 0.5312 1 1>}
  n40 = FunctionInput(Preview=TextureObject() {texture=&T_BaseColor}) {bUsePreviewValueAsDefault=true, inputName=Diffuse, inputType=FunctionInput_Texture2D, sortPriority=-4, previewValue=<0.5 0.5 0.5 1>}
  NamedRerouteDeclaration(None=@n40) {name=Input Diffuse, nodeColor=<rgba 0 0.7422 1 1>}
  n41 = FunctionInput(Preview=TextureObject() {samplerType=SAMPLERTYPE_Normal, texture=&T_Normal}) {bUsePreviewValueAsDefault=true, inputName=Normals, inputType=FunctionInput_Texture2D, sortPriority=-3, previewValue=<0 0 1 1>}
  NamedRerouteDeclaration(None=@n41) {name=Input Normals, nodeColor=<rgba 0.3035 0.2102 1 1>}
  n42 = FunctionInput(Preview=TextureObject() {samplerType=SAMPLERTYPE_Masks, texture=&T_Specular}) {bUsePreviewValueAsDefault=true, inputName=SpecularMasks, inputType=FunctionInput_Texture2D, sortPriority=-2, previewValue=<0.5 0.5 0.5 1>}
  NamedRerouteDeclaration(None=@n42) {name=Input Specular, nodeColor=<rgba 1 0.0474 0.3626 1>}
  n43 = ComponentMask(None=@n39) {a=true, b=true}
  NamedRerouteDeclaration(None=DDX(Value=@n43)) {name=DDX, nodeColor=<rgba 0 1 0.2969 1>}
  NamedRerouteDeclaration(None=DDY(Value=@n43)) {name=DDY, nodeColor=<rgba 1 0 0.0234 1>}
  NamedRerouteDeclaration(None=TextureSample(UVs=NamedRerouteUsage(), Tex=NamedRerouteUsage()) {samplerSource=SSM_Wrap_WorldGroupSettings, texture=&T_BaseColor}.RGBA) {name=Diffuse T0, nodeColor=<rgba 0.7344 1 0 1>}
  n44 = TextureSample(UVs=NamedRerouteUsage(), Tex=NamedRerouteUsage()) {samplerSource=SSM_Wrap_WorldGroupSettings, samplerType=SAMPLERTYPE_Normal, texture=&T_Normal}
  NamedRerouteDeclaration(None=@n44.RGB) {name=Normals T0, nodeColor=<rgba 0.5494 0.0564 1 1>}
  n45 = TextureSample(UVs=NamedRerouteUsage(), Tex=NamedRerouteUsage()) {samplerSource=SSM_Wrap_WorldGroupSettings, samplerType=SAMPLERTYPE_Masks, texture=&T_Specular}
  NamedRerouteDeclaration(None=@n45.R) {name=Specular T0, nodeColor=<rgba 1 0 0.0179 1>}
  NamedRerouteDeclaration(None=@n45.G) {name=Metallic T0, nodeColor=<rgba 0.1298 1 0 1>}
  NamedRerouteDeclaration(None=StaticSwitch(True=NamedRerouteUsage(), False=NamedRerouteUsage(), Value=StaticBoolParameter() {group=- Effect Switches, parameterName=World Space UVs})) {name=Specular Value, nodeColor=<rgba 1 0 0.0179 1>}
  NamedRerouteDeclaration(None=@n45.B) {name=Roughness T0, nodeColor=<rgba 0 0.0454 1 1>}
  NamedRerouteDeclaration(None=StaticSwitch(True=NamedRerouteUsage(), False=NamedRerouteUsage(), Value=StaticBoolParameter() {group=- Effect Switches, parameterName=World Space UVs})) {name=Metallic Value, nodeColor=<rgba 0.1298 1 0 1>}
  NamedRerouteDeclaration(None=StaticSwitch(True=NamedRerouteUsage(), False=NamedRerouteUsage(), Value=StaticBoolParameter() {group=- Effect Switches, parameterName=World Space UVs})) {name=Roughness Value, nodeColor=<rgba 0 0.0454 1 1>}
  n46 = FunctionInput(Preview=TextureObject() {samplerType=SAMPLERTYPE_LinearGrayscale, texture=&T_Displacement}) {bUsePreviewValueAsDefault=true, inputName=Displacement, inputType=FunctionInput_Texture2D, previewValue=<0.5 0.5 0.5 1>}
  NamedRerouteDeclaration(None=@n46) {name=Input Displacement, nodeColor=<rgba 0 1 0.1094 1>}
  n47 = FunctionInput() {bUsePreviewValueAsDefault=true, inputName=UV Offset, inputType=FunctionInput_Vector2, sortPriority=3, previewValue=<1 1 1 1>}
  NamedRerouteDeclaration(None=@n47) {name=UV Offset, nodeColor=<rgba 1 0 0.4453 1>}
  NamedRerouteDeclaration(None=LinearInterpolate(B=NamedRerouteUsage(), Alpha=FunctionInput() {bUsePreviewValueAsDefault=true, inputName=Opacity Mask, inputType=FunctionInput_Scalar}) {constA=1}) {name=Opacity Mask, nodeColor=<rgba 0 1 0.7422 1>}
 ]}
