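# Material function for four-layer materials that blends the input material into the landscape's runtime virtual textures near the terrain surface.
# World-height distance to the RVT height drives a falloff mask, with height-blend and per-layer masking functions refining the blend.
# Base color, specular and roughness blend with metallic zeroed, while normals lerp separately.
# Sample position shifts along the vertex normal to control stretching; a static switch and VT feature switch gate the effect.
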MaterialFunction /BRCoreMaterials/Materials/MaterialFunctions/MF_4Layer_RVT
graph {expressions {
  MaterialExpressionAbs_1 {class=MaterialExpressionAbs, inputs [None {from=MaterialExpressionDivide_0, output=""}], properties {}}
  MaterialExpressionBlendMaterialAttributes_0 {
   class=MaterialExpressionBlendMaterialAttributes
   inputs [
    A {from=MaterialExpressionNamedRerouteDeclaration_11, output=""}
    B {from=MaterialExpressionSetMaterialAttributes_0, output=""}
    Alpha {from=MaterialExpressionMultiply_0, output=""}
   ]
   properties {}
  }
  MaterialExpressionComponentMask_1 {class=MaterialExpressionComponentMask, inputs [None {from=MaterialExpressionTransformPosition_1}], properties {b=true}}
  MaterialExpressionConstant_0 {class=MaterialExpressionConstant}
  MaterialExpressionConvert_0 {
   class=MaterialExpressionConvert
   inputs [
    None {from=MaterialExpressionNamedRerouteUsage_0}
    None {from=MaterialExpressionNamedRerouteUsage_0}
    None {from=MaterialExpressionNamedRerouteUsage_0}
   ]
   properties {
    convertInputs [
     {
      type=Scalar
      expressionInput {
       expression=&self:MaterialExpressionNamedRerouteUsage_0
       inputName=None
       mask=0
       maskA=0
       maskB=0
       maskG=0
       maskR=0
       outputIndex=0
      }
     }
     {
      type=Scalar
      expressionInput {
       expression=&self:MaterialExpressionNamedRerouteUsage_0
       inputName=None
       mask=0
       maskA=0
       maskB=0
       maskG=0
       maskR=0
       outputIndex=0
      }
     }
     {
      type=Scalar
      expressionInput {
       expression=&self:MaterialExpressionNamedRerouteUsage_0
       inputName=None
       mask=0
       maskA=0
       maskB=0
       maskG=0
       maskR=0
       outputIndex=0
      }
     }
    ]
    convertOutputs [{type=Vector3}]
   }
  }
  MaterialExpressionDivide_0 {
   class=MaterialExpressionDivide
   inputs [
    A {from=MaterialExpressionSubtract_0}
    B {from=MaterialExpressionNamedRerouteDeclaration_10}
   ]
  }
  MaterialExpressionDivide_2 {
   class=MaterialExpressionDivide
   inputs [
    A {from=""}
    B {from=MaterialExpressionScalarParameter_2}
   ]
  }
  MaterialExpressionFunctionInput_0 {
   class=MaterialExpressionFunctionInput
   inputs [Preview {from="", output=""}]
   properties {inputName=Material, inputType=FunctionInput_MaterialAttributes}
  }
  MaterialExpressionFunctionInput_1 {
   class=MaterialExpressionFunctionInput
   inputs [Preview {from="", output=""}]
   properties {inputName=Layer Mask, inputType=FunctionInput_Vector4, sortPriority=1}
  }
  MaterialExpressionFunctionInput_2 {
   class=MaterialExpressionFunctionInput
   inputs [Preview {from="", output=""}]
   properties {inputName=Layer Texture Mask, inputType=FunctionInput_Scalar, sortPriority=2}
  }
  MaterialExpressionFunctionOutput_0 {class=MaterialExpressionFunctionOutput, inputs [None {from=MaterialExpressionVirtualTextureFeatureSwitch_0, output=""}]}
  MaterialExpressionGetMaterialAttributes_2 {
   class=MaterialExpressionGetMaterialAttributes
   inputs [None {from=MaterialExpressionBlendMaterialAttributes_0, output=""}]
   properties {attributeGetTypes [#1]}
  }
  MaterialExpressionLinearInterpolate_0 {
   class=MaterialExpressionLinearInterpolate
   inputs [
    A {from=MaterialExpressionGetMaterialAttributes_2, output=Normal}
    B {from=MaterialExpressionNamedRerouteUsage_6}
    Alpha {from=MaterialExpressionMultiply_4}
   ]
  }
  MaterialExpressionMaterialFunctionCall_0 {
   class=MaterialExpressionMaterialFunctionCall
   inputs [
    Background {from=MaterialExpressionNamedRerouteUsage_16}
    Background Low {from=MaterialExpressionScalarParameter_4}
    Background High {from=MaterialExpressionScalarParameter_5}
    Foreground {from=MaterialExpressionNamedRerouteUsage_11}
    Foreground Low {from=MaterialExpressionScalarParameter_7}
    Foreground High {from=MaterialExpressionScalarParameter_8}
    Contrast {from=MaterialExpressionScalarParameter_9}
    Mask {from=MaterialExpressionNamedRerouteUsage_7}
   ]
   properties {materialFunction=&/MF_4Layer_HeightBlend}
  }
  MaterialExpressionMaterialFunctionCall_4 {
   class=MaterialExpressionMaterialFunctionCall
   inputs [
    RGBA Mask {from=MaterialExpressionNamedRerouteUsage_12}
    L1 {from=MaterialExpressionScalarParameter_12}
    L2 {from=MaterialExpressionScalarParameter_25}
    L3 {from=MaterialExpressionScalarParameter_26}
    L4 {from=MaterialExpressionScalarParameter_27}
   ]
   properties {materialFunction=&/MF_4Layer_ScalarMasking}
  }
  MaterialExpressionMultiply_0 {
   class=MaterialExpressionMultiply
   inputs [
    A {from=MaterialExpressionMultiply_11}
    B {from=MaterialExpressionNamedRerouteUsage_14}
   ]
  }
  MaterialExpressionMultiply_11 {
   class=MaterialExpressionMultiply
   inputs [
    A {from=MaterialExpressionNamedRerouteUsage_13}
    B {from=MaterialExpressionScalarParameter_6}
   ]
  }
  MaterialExpressionMultiply_2 {
   class=MaterialExpressionMultiply
   inputs [
    A {from=MaterialExpressionNamedRerouteUsage_1}
    B {from=MaterialExpressionVertexNormalWS_0}
   ]
  }
  MaterialExpressionMultiply_3 {
   class=MaterialExpressionMultiply
   inputs [
    A {from=MaterialExpressionMultiply_2}
    B {from=MaterialExpressionMultiply_7}
   ]
   properties {constB=50}
  }
  MaterialExpressionMultiply_4 {
   class=MaterialExpressionMultiply
   inputs [
    A {from=MaterialExpressionMultiply_5}
    B {from=MaterialExpressionNamedRerouteUsage_15}
   ]
  }
  MaterialExpressionMultiply_5 {
   class=MaterialExpressionMultiply
   inputs [
    A {from=MaterialExpressionPower_0}
    B {from=MaterialExpressionScalarParameter_3}
   ]
  }
  MaterialExpressionMultiply_7 {
   class=MaterialExpressionMultiply
   inputs [
    A {from=MaterialExpressionNamedRerouteUsage_9}
    B {from=MaterialExpressionScalarParameter_1}
   ]
  }
  MaterialExpressionNamedRerouteDeclaration_0 {
   class=MaterialExpressionNamedRerouteDeclaration
   inputs [None {from=MaterialExpressionRuntimeVirtualTextureSampleParameter_2, output=WorldHeight}]
   properties {name=RVT World Height, nodeColor=<rgba 0 1 0.2031 1>}
  }
  MaterialExpressionNamedRerouteDeclaration_1 {
   class=MaterialExpressionNamedRerouteDeclaration
   inputs [None {from=MaterialExpressionSaturate_0}]
   properties {name=Height Gradient, nodeColor=<rgba 0.1953 0 1 1>}
  }
  MaterialExpressionNamedRerouteDeclaration_10 {
   class=MaterialExpressionNamedRerouteDeclaration
   inputs [None {from=MaterialExpressionScalarParameter_0}]
   properties {name=RVT Blend Height, nodeColor=<rgba 0.7344 0 1 1>}
  }
  MaterialExpressionNamedRerouteDeclaration_11 {
   class=MaterialExpressionNamedRerouteDeclaration
   inputs [None {from=MaterialExpressionFunctionInput_0}]
   properties {name=Material Input, nodeColor=<rgba 0.2891 0 1 1>}
  }
  MaterialExpressionNamedRerouteDeclaration_12 {
   class=MaterialExpressionNamedRerouteDeclaration
   inputs [None {from=MaterialExpressionMaterialFunctionCall_4, output=Result}]
   properties {name=Layer Weight, nodeColor=<rgba 1 0.9844 0 1>}
  }
  MaterialExpressionNamedRerouteDeclaration_13 {
   class=MaterialExpressionNamedRerouteDeclaration
   inputs [None {from=MaterialExpressionFunctionInput_2}]
   properties {name=Layer Texture Mask, nodeColor=<rgba 0 1 0.6967 1>}
  }
  MaterialExpressionNamedRerouteDeclaration_2 {
   class=MaterialExpressionNamedRerouteDeclaration
   inputs [None {from=MaterialExpressionFunctionInput_1}]
   properties {name=Layer Mask, nodeColor=<rgba 1 0.0168 0.8336 1>}
  }
  MaterialExpressionNamedRerouteDeclaration_3 {
   class=MaterialExpressionNamedRerouteDeclaration
   inputs [None {from=MaterialExpressionRuntimeVirtualTextureSampleParameter_1, output=BaseColor}]
   properties {name=RVT BaseColor, nodeColor=<rgba 0 0.6953 1 1>}
  }
  MaterialExpressionNamedRerouteDeclaration_4 {
   class=MaterialExpressionNamedRerouteDeclaration
   inputs [None {from=MaterialExpressionRuntimeVirtualTextureSampleParameter_1, output=Specular}]
   properties {name=RVT Specular, nodeColor=<rgba 1 0.9844 0 1>}
  }
  MaterialExpressionNamedRerouteDeclaration_5 {
   class=MaterialExpressionNamedRerouteDeclaration
   inputs [None {from=MaterialExpressionRuntimeVirtualTextureSampleParameter_1, output=Roughness}]
   properties {name=RVT Roughness, nodeColor=<rgba 0.6641 0 1 1>}
  }
  MaterialExpressionNamedRerouteDeclaration_6 {
   class=MaterialExpressionNamedRerouteDeclaration
   inputs [None {from=MaterialExpressionRuntimeVirtualTextureSampleParameter_1, output=Normal}]
   properties {name=RVT Normal, nodeColor=<rgba 0.6641 0 1 1>}
  }
  MaterialExpressionNamedRerouteDeclaration_7 {
   class=MaterialExpressionNamedRerouteDeclaration
   inputs [None {from=MaterialExpressionPower_1}]
   properties {name=RVT Mask, nodeColor=<rgba 1 0.0234 0 1>}
  }
  MaterialExpressionNamedRerouteDeclaration_8 {
   class=MaterialExpressionNamedRerouteDeclaration
   inputs [None {from=MaterialExpressionRuntimeVirtualTextureSampleParameter_0, output=Displacement}]
   properties {name=RVT Displacement, nodeColor=<rgba 0 0.7891 1 1>}
  }
  MaterialExpressionNamedRerouteDeclaration_9 {
   class=MaterialExpressionNamedRerouteDeclaration
   inputs [None {from=MaterialExpressionMaterialFunctionCall_0, output=Mask}]
   properties {name=RVT Mask After Heightmap, nodeColor=<rgba 1 0.8906 0 1>}
  }
  MaterialExpressionNamedRerouteUsage_0 {class=MaterialExpressionNamedRerouteUsage}
  MaterialExpressionNamedRerouteUsage_1 {class=MaterialExpressionNamedRerouteUsage}
  MaterialExpressionNamedRerouteUsage_10 {class=MaterialExpressionNamedRerouteUsage}
  MaterialExpressionNamedRerouteUsage_11 {class=MaterialExpressionNamedRerouteUsage}
  MaterialExpressionNamedRerouteUsage_12 {class=MaterialExpressionNamedRerouteUsage}
  MaterialExpressionNamedRerouteUsage_13 {class=MaterialExpressionNamedRerouteUsage}
  MaterialExpressionNamedRerouteUsage_14 {class=MaterialExpressionNamedRerouteUsage}
  MaterialExpressionNamedRerouteUsage_15 {class=MaterialExpressionNamedRerouteUsage}
  MaterialExpressionNamedRerouteUsage_16 {class=MaterialExpressionNamedRerouteUsage}
  MaterialExpressionNamedRerouteUsage_2 {class=MaterialExpressionNamedRerouteUsage}
  MaterialExpressionNamedRerouteUsage_3 {class=MaterialExpressionNamedRerouteUsage}
  MaterialExpressionNamedRerouteUsage_4 {class=MaterialExpressionNamedRerouteUsage}
  MaterialExpressionNamedRerouteUsage_5 {class=MaterialExpressionNamedRerouteUsage}
  MaterialExpressionNamedRerouteUsage_6 {class=MaterialExpressionNamedRerouteUsage}
  MaterialExpressionNamedRerouteUsage_7 {class=MaterialExpressionNamedRerouteUsage}
  MaterialExpressionNamedRerouteUsage_9 {class=MaterialExpressionNamedRerouteUsage}
  MaterialExpressionOneMinus_0 {class=MaterialExpressionOneMinus, inputs [None {from=MaterialExpressionNamedRerouteDeclaration_1}]}
  MaterialExpressionPower_0 {
   class=MaterialExpressionPower
   inputs [
    Base {from=MaterialExpressionNamedRerouteUsage_5}
    Exp {from=MaterialExpressionDivide_2}
   ]
  }
  MaterialExpressionPower_1 {
   class=MaterialExpressionPower
   inputs [
    Base {from=MaterialExpressionOneMinus_0}
    Exp {from=MaterialExpressionScalarParameter_10}
   ]
  }
  MaterialExpressionRuntimeVirtualTextureSampleParameter_0 {
   class=MaterialExpressionRuntimeVirtualTextureSampleParameter
   inputs [
    UVs {from="", output=""}
    Translated World Position {from=MaterialExpressionSubtract_1, output=""}
   ]
   properties {
    group=RVT
    materialType=Displacement
    parameterName=RVT Displacement
    sortPriority=-1
    virtualTexture=/Game/Athena/Apollo/Maps/Landscape/VT/RVT_Helios_Displacement
    worldPositionOriginType=CameraRelative
   }
  }
  MaterialExpressionRuntimeVirtualTextureSampleParameter_1 {
   class=MaterialExpressionRuntimeVirtualTextureSampleParameter
   inputs [
    UVs {from="", output=""}
    Translated World Position {from=MaterialExpressionSubtract_1, output=""}
   ]
   properties {
    group=RVT
    materialType=BaseColor_Normal_Specular_YCoCg
    parameterName=RVT Color
    sortPriority=-3
    virtualTexture=/Game/Athena/Apollo/Maps/Landscape/VT/RVT_Apollo_Landscape
    worldPositionOriginType=CameraRelative
   }
  }
  MaterialExpressionRuntimeVirtualTextureSampleParameter_2 {
   class=MaterialExpressionRuntimeVirtualTextureSampleParameter
   inputs [
    UVs {from="", output=""}
    Translated World Position {from="", output=""}
   ]
   properties {
    group=RVT
    materialType=WorldHeight
    parameterName=RVT Height
    sortPriority=-2
    virtualTexture=/Game/Athena/Apollo/Maps/Landscape/VT/RVT_Apollo_Landscape_Height
    worldPositionOriginType=CameraRelative
   }
  }
  MaterialExpressionSaturate_0 {class=MaterialExpressionSaturate, inputs [None {from=MaterialExpressionAbs_1}]}
  MaterialExpressionScalarParameter_0 {class=MaterialExpressionScalarParameter, properties {defaultValue=100, group=RVT, parameterName=RVT Distance, sortPriority=5}}
  MaterialExpressionScalarParameter_1 {
   class=MaterialExpressionScalarParameter
   properties {defaultValue=1, group=RVT, parameterName=RVT UV Stretching, sliderMax=2, sortPriority=10}
  }
  MaterialExpressionScalarParameter_10 {class=MaterialExpressionScalarParameter, properties {defaultValue=2, group=RVT, parameterName=RVT Falloff, sortPriority=6}}
  MaterialExpressionScalarParameter_12 {
   class=MaterialExpressionScalarParameter
   properties {defaultValue=1, group=RVT | RVT Layer Mask, parameterName=L1 RVT, sliderMax=1, sortPriority=0}
  }
  MaterialExpressionScalarParameter_2 {
   class=MaterialExpressionScalarParameter
   properties {defaultValue=0.5, group=RVT, parameterName=RVT Normals Distance, sliderMax=1, sortPriority=7}
  }
  MaterialExpressionScalarParameter_25 {
   class=MaterialExpressionScalarParameter
   properties {defaultValue=1, group=RVT | RVT Layer Mask, parameterName=L2 RVT, sliderMax=1, sortPriority=1}
  }
  MaterialExpressionScalarParameter_26 {
   class=MaterialExpressionScalarParameter
   properties {defaultValue=1, group=RVT | RVT Layer Mask, parameterName=L3 RVT, sliderMax=1, sortPriority=2}
  }
  MaterialExpressionScalarParameter_27 {
   class=MaterialExpressionScalarParameter
   properties {defaultValue=1, group=RVT | RVT Layer Mask, parameterName=L4 RVT, sliderMax=1, sortPriority=3}
  }
  MaterialExpressionScalarParameter_3 {
   class=MaterialExpressionScalarParameter
   properties {defaultValue=1, group=RVT, parameterName=RVT Normals Opacity, sliderMax=1, sortPriority=8}
  }
  MaterialExpressionScalarParameter_4 {
   class=MaterialExpressionScalarParameter
   properties {defaultValue=0.5, group=RVT | RVT Blend, parameterName=RVT Blend Foreground Low, sortPriority=0}
  }
  MaterialExpressionScalarParameter_5 {
   class=MaterialExpressionScalarParameter
   properties {defaultValue=0.5, group=RVT | RVT Blend, parameterName=RVT Blend Foreground High, sortPriority=1}
  }
  MaterialExpressionScalarParameter_6 {
   class=MaterialExpressionScalarParameter
   properties {defaultValue=1, group=RVT, parameterName=RVT Material Opacity, sliderMax=1, sortPriority=9}
  }
  MaterialExpressionScalarParameter_7 {class=MaterialExpressionScalarParameter, properties {group=RVT | RVT Blend, parameterName=RVT Blend Background Low, sortPriority=2}}
  MaterialExpressionScalarParameter_8 {
   class=MaterialExpressionScalarParameter
   properties {defaultValue=1, group=RVT | RVT Blend, parameterName=RVT Blend Background High, sortPriority=3}
  }
  MaterialExpressionScalarParameter_9 {
   class=MaterialExpressionScalarParameter
   properties {defaultValue=1, group=RVT | RVT Blend, parameterName=RVT Blend Contrast, sortPriority=4}
  }
  MaterialExpressionSetMaterialAttributes_0 {
   class=MaterialExpressionSetMaterialAttributes
   inputs [
    MaterialAttributes {from=MaterialExpressionNamedRerouteDeclaration_11, output=""}
    BaseColor {from=MaterialExpressionNamedRerouteUsage_2, output=""}
    Specular {from=MaterialExpressionNamedRerouteUsage_3, output=""}
    Roughness {from=MaterialExpressionNamedRerouteUsage_4, output=""}
    Metallic {from=MaterialExpressionConstant_0, output=""}
   ]
   properties {attributeSetTypes [
     #2
     #3
     #4
     #5
    ]}
  }
  MaterialExpressionSetMaterialAttributes_1 {
   class=MaterialExpressionSetMaterialAttributes
   inputs [
    MaterialAttributes {from=MaterialExpressionGetMaterialAttributes_2, output=MaterialAttributes}
    Normal {from=MaterialExpressionLinearInterpolate_0, output=""}
   ]
   properties {attributeSetTypes [#1]}
  }
  MaterialExpressionStaticSwitchParameter_0 {
   class=MaterialExpressionStaticSwitchParameter
   inputs [
    True {from=MaterialExpressionSetMaterialAttributes_1, output=""}
    False {from=MaterialExpressionNamedRerouteDeclaration_11, output=""}
   ]
   properties {group=- Effect Switches, parameterName=RVT}
  }
  MaterialExpressionSubtract_0 {
   class=MaterialExpressionSubtract
   inputs [
    A {from=MaterialExpressionWorldPosition_0, output=Z}
    B {from=MaterialExpressionComponentMask_1}
   ]
  }
  MaterialExpressionSubtract_1 {
   class=MaterialExpressionSubtract
   inputs [
    A {from=MaterialExpressionWorldPosition_1, output=XYZ}
    B {from=MaterialExpressionMultiply_3}
   ]
  }
  MaterialExpressionTransformPosition_1 {
   class=MaterialExpressionTransformPosition
   inputs [None {from=MaterialExpressionConvert_0, output=""}]
   properties {transformSourceType=TRANSFORMPOSSOURCE_World, transformType=TRANSFORMPOSSOURCE_TranslatedWorld}
  }
  MaterialExpressionVertexNormalWS_0 {class=MaterialExpressionVertexNormalWS, inputs [], properties {}}
  MaterialExpressionVirtualTextureFeatureSwitch_0 {
   class=MaterialExpressionVirtualTextureFeatureSwitch
   inputs [
    No {from=MaterialExpressionNamedRerouteUsage_10, output=""}
    Yes {from=MaterialExpressionStaticSwitchParameter_0, output=""}
   ]
   properties {}
  }
  MaterialExpressionWorldPosition_0 {class=MaterialExpressionWorldPosition, inputs [], properties {worldPositionShaderOffset=WPT_CameraRelativeNoOffsets}}
  MaterialExpressionWorldPosition_1 {class=MaterialExpressionWorldPosition, inputs [], properties {worldPositionShaderOffset=WPT_CameraRelativeNoOffsets}}
 }}
