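# Material function that adds a stylized sheen to BaseColor for a four-layer material, behind the static switch "Sheen".
# A sharp N·L ramp against the sky atmosphere light blends shadow and light tints, then a hard-edged half-vector highlight brightens the result.
# The highlight uses normals blended toward vertex normals and distorted by a local-position pattern.
# MF_4Layer_ScalarMasking selects its enum-driven size and intensity per layer.
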MaterialFunction /BRCoreMaterials/Materials/MaterialFunctions/MF_4Layer_Sheen
graph {nodes [
  n1 = NamedRerouteUsage()
  n2 = NamedRerouteUsage()
  n3 = VectorParameter() {group=Sheen, parameterName=Sheen Diffuse Shadow, sortPriority=33, defaultValue=<rgba 0.5781 0.5781 0.5781 1>}
  n4 = VectorParameter() {group=Sheen, parameterName=Sheen Diffuse Light, defaultValue=<rgba 1 1 1 1>}
  n5 = SkyAtmosphereLightDirection()
  n6 = NamedRerouteUsage()
  n7 = StaticSwitch(True=Transform(None=@n6), False=@n6, Value=NamedRerouteUsage())
  n8 = Add(A=Multiply(A=DotProduct(A=@n5, B=@n7)) {constB=20}) {constB=0.5}
  n9 = Add(A=SkyAtmosphereLightDirection(), B=CameraVectorWS())
  n10 = NamedRerouteUsage()
  n11 = Transform(None=@n10)
  n12 = NamedRerouteUsage()
  n13 = VertexNormalWS()
  n14 = NamedRerouteDeclaration(None=LinearInterpolate(
    A=StaticSwitch(True=@n11, False=@n10, Value=@n12)
    B=Normalize(VectorInput=@n13)
    Alpha=ScalarParameter() {defaultValue=0.2, group=Sheen, parameterName=Sheen Specular Vertex Normals, sliderMax=1, sortPriority=8}
   )) {name=Normals, nodeColor=<rgba 0 0.2031 1 1>}
  n15 = Frac(None=Divide(A=LocalPosition().XYZ, B=ScalarParameter() {defaultValue=30, group=Sheen, parameterName=Sheen Specular Distortion Scale, sortPriority=9}) {constB=300})
  n16 = Abs(None=Multiply(A=Add(A=@n15) {constB=-0.5}) {constB=2})
  n17 = OneMinus(None=Multiply(A=@n16, B=@n16) {constB=0.1})
  n18 = Multiply(A=@n17, B=@n17)
  n19 = Multiply(
   A=Add(A=OneMinus(None=@n18)) {constB=-0.5}
   B=ScalarParameter() {defaultValue=0.03, group=Sheen, parameterName=Sheen Specular Distortion Intensity, sortPriority=10}
  ) {constB=0.04}
  n20 = ScalarParameter() {
   controlType=Enumeration
   defaultValue=1
   enumeration=/BRCoreMaterials/Materials/Enums/EN_4Layer_SheenSize
   group=Sheen
   parameterName=L1 Sheen Specular Size
   sortPriority=4
  }
  n21 = ScalarParameter() {
   controlType=Enumeration
   defaultValue=1
   enumeration=&EN_4Layer_SheenSize
   group=Sheen
   parameterName=L2 Sheen Specular Size
   sortPriority=5
  }
  n22 = ScalarParameter() {
   controlType=Enumeration
   defaultValue=1
   enumeration=&EN_4Layer_SheenSize
   group=Sheen
   parameterName=L3 Sheen Specular Size
   sortPriority=6
  }
  n23 = ScalarParameter() {
   controlType=Enumeration
   defaultValue=1
   enumeration=&EN_4Layer_SheenSize
   group=Sheen
   parameterName=L4 Sheen Specular Size
   sortPriority=7
  }
  n24 = MaterialFunctionCall!(RGBA Mask=NamedRerouteUsage(), L1=@n20, L2=@n21, L3=@n22, L4=@n23) {materialFunction=&/MF_4Layer_ScalarMasking}
  n25 = Add(A=DotProduct(A=Normalize(VectorInput=@n9), B=Add(A=@n14, B=@n19)), B=Add(A=Multiply(A=@n24.Result) {constB=0.11}) {constB=-0.4}) {constB=0}
  n26 = Constant() {r=50}
  n27 = Multiply(
   A=Saturate(None=MaterialFunctionCall(In=@n25, Contrast=@n26) {materialFunction=&CheapContrast}.Result)
   B=NamedRerouteUsage()
  )
  n28 = ScalarParameter() {
   controlType=Enumeration
   enumeration=/BRCoreMaterials/Materials/Enums/EN_4Layer_SheenIntensity
   group=Sheen
   parameterName=L1 Sheen Specular Intensity
   sortPriority=4
  }
  n29 = ScalarParameter() {
   controlType=Enumeration
   enumeration=&EN_4Layer_SheenIntensity
   group=Sheen
   parameterName=L2 Sheen Specular Intensity
   sortPriority=5
  }
  n30 = ScalarParameter() {
   controlType=Enumeration
   enumeration=&EN_4Layer_SheenIntensity
   group=Sheen
   parameterName=L3 Sheen Specular Intensity
   sortPriority=6
  }
  n31 = ScalarParameter() {
   controlType=Enumeration
   enumeration=&EN_4Layer_SheenIntensity
   group=Sheen
   parameterName=L4 Sheen Specular Intensity
   sortPriority=7
  }
  n32 = Multiply(A=MaterialFunctionCall(RGBA Mask=NamedRerouteUsage(), L1=@n28, L2=@n29, L3=@n30, L4=@n31) {materialFunction=&MF_4Layer_ScalarMasking}.Result) {constB=1.5}
  n33 = Multiply(
   A=LinearInterpolate(A=Multiply(A=@n2, B=@n3.RGB), B=Multiply(A=@n2, B=@n4.RGB), Alpha=Saturate(None=@n8)) {constA=0.5}
   B=Add(A=Multiply(A=@n27, B=@n32))
  )
  FunctionOutput(None=StaticSwitchParameter(
    True=SetMaterialAttributes!(MaterialAttributes=@n1, BaseColor=@n33) {attributeSetTypes [#1]}
    False=NamedRerouteUsage()
   ) {group=- Effect Switches, parameterName=Sheen})
  n34 = FunctionInput() {bUsePreviewValueAsDefault=true, inputName=Layer Mask, inputType=FunctionInput_Vector4, sortPriority=1, previewValue=<1 1 1 1>}
  NamedRerouteDeclaration(None=@n34) {name=Layer Mask, nodeColor=<rgba 1 0 0.3471 1>}
  n35 = FunctionInput() {bUsePreviewValueAsDefault=true, inputName=Layer Texture Mask, inputType=FunctionInput_Scalar, sortPriority=2, previewValue=<1 1 1 1>}
  NamedRerouteDeclaration(None=@n35) {name=Layer Texture Mask, nodeColor=<rgba 0 0.5312 1 1>}
  NamedRerouteDeclaration(None=FunctionInput() {bUsePreviewValueAsDefault=true, inputName=TangentNormals, inputType=FunctionInput_StaticBool, sortPriority=3}) {name=Tangent Normals, nodeColor=<rgba 1 0.3984 0 1>}
  n36 = FunctionInput() {bUsePreviewValueAsDefault=true, inputName=Material, inputType=FunctionInput_MaterialAttributes, previewValue=<1 1 1 1>}
  n37 = GetMaterialAttributes(None=@n36) {attributeGetTypes [
    #1
    #2
    #3
    #4
   ]}
  NamedRerouteDeclaration(None=@n37.EmissiveColor) {name=Emissive Input, nodeColor=<rgba 1 0 0.1875 1>}
  NamedRerouteDeclaration(None=@n37.MaterialAttributes) {name=Material Input, nodeColor=<rgba 1 0.4922 0 1>}
  NamedRerouteDeclaration(None=@n37.BaseColor) {name=BaseColor Input, nodeColor=<rgba 1 0 0.2109 1>}
  NamedRerouteDeclaration(None=@n37.Normal) {name=Normal Input, nodeColor=<rgba 0 0.5312 1 1>}
  NamedRerouteDeclaration(None=@n37.Specular) {name=Specular Input, nodeColor=<rgba 0 0.0859 1 1>}
 ]}
