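# Material function that tints Base Color by multiplication for four-layer materials, using per-layer colors selected by a layer mask.
# Tint sources are color parameters, packed Custom Primitive Data values unpacked to color (L1–L3 only), and Color Replace colors blended in by a static switch.
# An If with a near-zero threshold picks the tint in the vertex stage through a VertexInterpolator.
# A mask lerp, also switched for Color Replace, limits where the multiply applies.
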MaterialFunction /BRCoreMaterials/Materials/MaterialFunctions/MF_4Layer_Tint_Multiply
graph {nodes [
  n1 = GetMaterialAttributes(None=FunctionInput() {inputName=Material, inputType=FunctionInput_MaterialAttributes}) {attributeGetTypes [#1]}
  n2 = NamedRerouteUsage()
  n3 = DotProduct(A=Constant3Vector() {constant=<rgba 1 1 1 0>}, B=@n2)
  n4 = StaticSwitchParameter(True=LinearInterpolate(A=NamedRerouteUsage(), B=NamedRerouteUsage(), Alpha=NamedRerouteUsage()), False=NamedRerouteUsage()) {group=- Effect Switches, parameterName=Color Replace}
  n5 = Multiply(A=@n1.BaseColor, B=VertexInterpolator(VS=If(A=@n3, "A > B"=@n2, "A < B"=@n4) {constB=0.01}).PS)
  n6 = StaticSwitchParameter(True=LinearInterpolate(A=NamedRerouteUsage(), B=NamedRerouteUsage(), Alpha=NamedRerouteUsage()), False=NamedRerouteUsage()) {group=- Effect Switches, parameterName=Color Replace}
  FunctionOutput(None=SetMaterialAttributes!(MaterialAttributes=@n1.MaterialAttributes, BaseColor=LinearInterpolate(A=@n1.BaseColor, B=@n5, Alpha=@n6)) {attributeSetTypes [#1]}) {outputName=Material}
  n7 = NamedRerouteUsage()
  n8 = ScalarParameter() {defaultValue=1, group=1 Layer, parameterName=L1 Tint Mask, sortPriority=1}
  n9 = ScalarParameter() {defaultValue=1, group=3 Layer, parameterName=L2 Tint Mask, sortPriority=1}
  n10 = ScalarParameter() {defaultValue=1, group=3 Layer, parameterName=L3 Tint Mask, sortPriority=1}
  n11 = ScalarParameter() {defaultValue=1, group=4 Layer, parameterName=L4 Tint Mask, sortPriority=1}
  NamedRerouteDeclaration(None=LinearInterpolate(
    B=FunctionInput() {inputName=Layer Texture Mask, inputType=FunctionInput_Scalar, sortPriority=4}
    Alpha=MaterialFunctionCall!(RGBA Mask=@n7, L1=@n8, L2=@n9, L3=@n10, L4=@n11) {materialFunction=&/MF_4Layer_ScalarMasking}.Result
   ) {constA=1}) {name=Layer Texture Mask, nodeColor=<rgba 0 1 0.7439 1>}
  NamedRerouteDeclaration(None=FunctionInput() {inputName=Layer Mask, inputType=FunctionInput_Vector4, sortPriority=3}) {name=Layer Mask, nodeColor=<rgba 0.7109 0 1 1>}
  NamedRerouteDeclaration(None=FunctionInput() {inputName=Color Replace Toggle, inputType=FunctionInput_Scalar, sortPriority=2}) {name=Color Replace Toggle, nodeColor=<rgba 0 1 0.2266 1>}
  NamedRerouteDeclaration(None=FunctionInput() {inputName=Color Replace Mask, inputType=FunctionInput_Scalar, sortPriority=1}) {name=Color Replace Mask, nodeColor=<rgba 0.9453 0 1 1>}
  n12 = ScalarParameter() {
   bUseCustomPrimitiveData=true
   desc="For now Scalar, this is possible alternative to Texture Swatch. Uses single ramp texture shared between all 4 layers and can have indepent row selected per layer. Enabled globally via function. Will disable use of color swatch if enabled. Diffuse tint still can be used if needed via separate mask "
   group=Y_CPD
   parameterName=CPD_PackedHSV_1
   primitiveDataIndex=30
   sliderMin=1
  }
  n13 = ScalarParameter() {
   bUseCustomPrimitiveData=true
   desc="For now Scalar, this is possible alternative to Texture Swatch. Uses single ramp texture shared between all 4 layers and can have indepent row selected per layer. Enabled globally via function. Will disable use of color swatch if enabled. Diffuse tint still can be used if needed via separate mask "
   group=Y_CPD
   parameterName=CPD_PackedHSV_2
   primitiveDataIndex=31
   sliderMin=1
  }
  n14 = ScalarParameter() {
   bUseCustomPrimitiveData=true
   desc="For now Scalar, this is possible alternative to Texture Swatch. Uses single ramp texture shared between all 4 layers and can have indepent row selected per layer. Enabled globally via function. Will disable use of color swatch if enabled. Diffuse tint still can be used if needed via separate mask "
   group=Y_CPD
   parameterName=CPD_PackedHSV_3
   primitiveDataIndex=35
   sliderMin=1
  }
  n15 = MaterialFunctionCall(RGBA Mask=NamedRerouteUsage(), L1=@n12, L2=@n13, L3=@n14, L4=Constant()) {materialFunction=&MF_4Layer_ScalarMasking}
  NamedRerouteDeclaration(None=MaterialFunctionCall!(In=@n15.Result) {materialFunction=/Game/Global/Base4Layers_2022/MaterialFunctions/MF_ColorTint_Unpack}.Result) {name=TextureData Color, nodeColor=<rgba 0.6172 1 0 1>}
  NamedRerouteDeclaration(None=MaterialFunctionCall!(
    Layer Mask=NamedRerouteUsage()
    L1=VectorParameter() {group=1 Layer, parameterName=L1 Tint Color, sortPriority=0, defaultValue=<rgba 1 1 1 1>}.RGB
    L2=VectorParameter() {group=2 Layer, parameterName=L2 Tint Color, sortPriority=0, defaultValue=<rgba 1 1 1 1>}.RGB
    L3=VectorParameter() {group=3 Layer, parameterName=L3 Tint Color, sortPriority=0, defaultValue=<rgba 1 1 1 1>}.RGB
    L4=VectorParameter() {group=4 Layer, parameterName=L4 Tint Color, sortPriority=0, defaultValue=<rgba 1 1 1 1>}.RGB
   ) {materialFunction=&/MF_4Layer_ColorMasking}.Result) {name=Color Tint, nodeColor=<rgba 1 0 0.9375 1>}
  n16 = VectorParameter() {group=Color Replace, parameterName=L1 ColorReplace New Color, sortPriority=2, defaultValue=<rgba 0.1866 0.3697 1 1>}
  n17 = VectorParameter() {group=Color Replace, parameterName=L2 ColorReplace New Color, sortPriority=12, defaultValue=<rgba 0.1866 0.3697 1 1>}
  n18 = VectorParameter() {group=Color Replace, parameterName=L3 ColorReplace New Color, sortPriority=22, defaultValue=<rgba 0.1866 0.3697 1 1>}
  NamedRerouteDeclaration(None=MaterialFunctionCall(
    Layer Mask=NamedRerouteUsage()
    L1=@n16.RGB
    L2=@n17.RGB
    L3=@n18.RGB
    L4=VectorParameter() {group=Color Replace, parameterName=L4 ColorReplace New Color, defaultValue=<rgba 0.1866 0.3697 1 1>}.RGB
   ) {materialFunction=&MF_4Layer_ColorMasking}.Result) {name=ColorReplace Tint, nodeColor=<rgba 0 1 0.7422 1>}
  NamedRerouteDeclaration(None=@n15.Result) {name=TextureData Scalar, nodeColor=<rgba 0 0.395 1 1>}
 ]}
