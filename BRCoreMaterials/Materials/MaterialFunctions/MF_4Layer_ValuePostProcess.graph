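# Material function that post-processes a Material attributes input, apparently for four-layer materials, adjusting values by render context, shading path and quality level.
# Normal passes through in realtime and converts from world to tangent space for material proxy baking.
# Mobile Low quality flattens specular and roughness to constants, while displacement survives only at Epic quality off mobile.
# Emissive gets partial inverse eye-adaptation compensation on the default shading path, staying untouched on mobile.
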MaterialFunction /BRCoreMaterials/Materials/MaterialFunctions/MF_4Layer_ValuePostProcess
graph {nodes [
  n1 = GetMaterialAttributes(None=FunctionInput() {inputName=Material, inputType=FunctionInput_MaterialAttributes}) {attributeGetTypes [#1]}
  n2 = Transform(None=@n1.Normal) {transformSourceType=TRANSFORMSOURCE_World, transformType=TRANSFORM_Tangent}
  n3 = GetMaterialAttributes(None=SetMaterialAttributes!(MaterialAttributes=@n1.MaterialAttributes, Normal=MaterialProxyReplace(Realtime=@n1.Normal, MaterialProxy=@n2)) {attributeSetTypes [#1]})
  n4 = SetMaterialAttributes!(MaterialAttributes=@n3.MaterialAttributes, Specular=Constant() {r=0.5}, Roughness=Constant() {r=0.8}) {attributeSetTypes [#2, #3]}
  n5 = GetMaterialAttributes(None=ShadingPathSwitch(
    Default=@n3.MaterialAttributes
    Mobile=QualitySwitch(Default=@n3.MaterialAttributes, Low=@n4, High=@n3.MaterialAttributes, Medium=@n3.MaterialAttributes, Epic=@n3.MaterialAttributes)
   )) {attributeGetTypes [#4]}
  n6 = Constant()
  n7 = QualitySwitch(Default=@n6, Low=@n6, High=@n6, Medium=@n6, Epic=@n5.Displacement)
  n8 = GetMaterialAttributes(None=SetMaterialAttributes!(MaterialAttributes=@n5.MaterialAttributes, Displacement=ShadingPathSwitch(Default=@n7, Mobile=@n6)) {attributeSetTypes [#4]}) {attributeGetTypes [#5]}
  n9 = EyeAdaptationInverse(LightValueInput=@n8.EmissiveColor, AlphaInput=Constant() {r=0.5})
  FunctionOutput(None=ShadingPathSwitch(
    Default=SetMaterialAttributes!(MaterialAttributes=@n8.MaterialAttributes, EmissiveColor=@n9.EyeAdaptationInverse) {attributeSetTypes [#5]}
    Mobile=@n8.MaterialAttributes
   ))
 ]}
