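# Material function that adds a masked variation to the material attributes of a four-layer material, behind a static switch.
# Builds masks from normal direction, local bounds, distance-field AO, vertex color and a world-space projected pattern texture, offset by that pattern.
# Combined masks, weighted per layer, drive a blend toward tinted base color and adjusted specular and roughness.
# Pattern UVs use a dithered projection mask with explicit derivatives; several masks have debug views.
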MaterialFunction /BRCoreMaterials/Materials/MaterialFunctions/MF_4Layer_Variation
graph {expressions {
  MaterialExpressionAdd_0 {
   class=MaterialExpressionAdd
   inputs [
    A {from=MaterialExpressionScalarParameter_0}
    B {from=""}
   ]
  }
  MaterialExpressionAdd_1 {
   class=MaterialExpressionAdd
   inputs [
    A {from=MaterialExpressionMultiply_1}
    B {from=MaterialExpressionMultiply_3}
   ]
  }
  MaterialExpressionAdd_10 {
   class=MaterialExpressionAdd
   inputs [
    A {from=MaterialExpressionNamedRerouteUsage_45}
    B {from=MaterialExpressionNamedRerouteUsage_72}
   ]
  }
  MaterialExpressionAdd_11 {
   class=MaterialExpressionAdd
   inputs [
    A {from=MaterialExpressionNamedRerouteUsage_73}
    B {from=MaterialExpressionNamedRerouteUsage_73}
   ]
  }
  MaterialExpressionAdd_12 {
   class=MaterialExpressionAdd
   inputs [
    A {from=MaterialExpressionAdd_24}
    B {from=MaterialExpressionMultiply_29}
   ]
  }
  MaterialExpressionAdd_13 {
   class=MaterialExpressionAdd
   inputs [
    A {from=MaterialExpressionDotProduct_0}
    B {from=MaterialExpressionMultiply_17}
   ]
  }
  MaterialExpressionAdd_15 {
   class=MaterialExpressionAdd
   inputs [
    A {from=MaterialExpressionNamedRerouteUsage_68}
    B {from=MaterialExpressionAdd_30}
   ]
  }
  MaterialExpressionAdd_17 {
   class=MaterialExpressionAdd
   inputs [
    A {from=MaterialExpressionNamedRerouteUsage_74}
    B {from=MaterialExpressionAdd_34}
   ]
  }
  MaterialExpressionAdd_19 {
   class=MaterialExpressionAdd
   inputs [
    A {from=MaterialExpressionNamedRerouteUsage_73}
    B {from=MaterialExpressionNamedRerouteUsage_1}
   ]
  }
  MaterialExpressionAdd_2 {
   class=MaterialExpressionAdd
   inputs [
    A {from=MaterialExpressionMultiply_6}
    B {from=MaterialExpressionMultiply_7}
   ]
  }
  MaterialExpressionAdd_20 {
   class=MaterialExpressionAdd
   inputs [
    A {from=MaterialExpressionNamedRerouteUsage_75}
    B {from=MaterialExpressionAdd_36}
   ]
  }
  MaterialExpressionAdd_21 {
   class=MaterialExpressionAdd
   inputs [
    A {from=MaterialExpressionPower_3}
    B {from=MaterialExpressionMultiply_16}
   ]
  }
  MaterialExpressionAdd_22 {
   class=MaterialExpressionAdd
   inputs [
    A {from=MaterialExpressionMultiply_33}
    B {from=MaterialExpressionMultiply_34}
   ]
  }
  MaterialExpressionAdd_23 {
   class=MaterialExpressionAdd
   inputs [
    A {from=MaterialExpressionConvert_2}
    B {from=MaterialExpressionConvert_2}
   ]
  }
  MaterialExpressionAdd_24 {
   class=MaterialExpressionAdd
   inputs [
    A {from=MaterialExpressionAdd_23}
    B {from=MaterialExpressionConvert_2}
   ]
  }
  MaterialExpressionAdd_25 {
   class=MaterialExpressionAdd
   inputs [
    A {from=MaterialExpressionDivide_0}
    B {from=MaterialExpressionConvert_8}
   ]
  }
  MaterialExpressionAdd_26 {
   class=MaterialExpressionAdd
   inputs [
    A {from=MaterialExpressionMultiply_22}
    B {from=MaterialExpressionConvert_9}
   ]
  }
  MaterialExpressionAdd_27 {
   class=MaterialExpressionAdd
   inputs [
    A {from=MaterialExpressionBounds_0, output=Half Extents}
    B {from=""}
   ]
   properties {constB=0.01}
  }
  MaterialExpressionAdd_3 {
   class=MaterialExpressionAdd
   inputs [
    A {from=MaterialExpressionMultiply_2}
    B {from=MaterialExpressionMultiply_5}
   ]
  }
  MaterialExpressionAdd_30 {
   class=MaterialExpressionAdd
   inputs [
    A {from=MaterialExpressionScalarParameter_36}
    B {from=""}
   ]
   properties {constB=-0.5}
  }
  MaterialExpressionAdd_34 {
   class=MaterialExpressionAdd
   inputs [
    A {from=MaterialExpressionScalarParameter_54}
    B {from=""}
   ]
   properties {constB=-0.5}
  }
  MaterialExpressionAdd_36 {
   class=MaterialExpressionAdd
   inputs [
    A {from=MaterialExpressionScalarParameter_60}
    B {from=""}
   ]
   properties {constB=-0.5}
  }
  MaterialExpressionAdd_4 {
   class=MaterialExpressionAdd
   inputs [
    A {from=MaterialExpressionAdd_1}
    B {from=MaterialExpressionMultiply_6}
   ]
  }
  MaterialExpressionAdd_5 {
   class=MaterialExpressionAdd
   inputs [
    A {from=MaterialExpressionAdd_1}
    B {from=MaterialExpressionAdd_2}
   ]
  }
  MaterialExpressionAdd_6 {
   class=MaterialExpressionAdd
   inputs [
    A {from=MaterialExpressionDotProduct_3}
    B {from=MaterialExpressionMaterialFunctionCall_3, output=Result}
   ]
  }
  MaterialExpressionAdd_7 {
   class=MaterialExpressionAdd
   inputs [
    A {from=MaterialExpressionAdd_6}
    B {from=""}
   ]
  }
  MaterialExpressionBlendMaterialAttributes_1 {
   class=MaterialExpressionBlendMaterialAttributes
   inputs [
    A {from=MaterialExpressionNamedRerouteUsage_85, output=""}
    B {from=MaterialExpressionSetMaterialAttributes_4, output=""}
    Alpha {from=MaterialExpressionSaturate_1, output=""}
   ]
   properties {}
  }
  MaterialExpressionBounds_0 {class=MaterialExpressionBounds, inputs [], properties {}}
  MaterialExpressionConstant_0 {class=MaterialExpressionConstant}
  MaterialExpressionConstant_1 {class=MaterialExpressionConstant, properties {r=1}}
  MaterialExpressionConstant_2 {class=MaterialExpressionConstant, properties {r=1}}
  MaterialExpressionConstant_3 {class=MaterialExpressionConstant, properties {r=1}}
  MaterialExpressionConstant_4 {class=MaterialExpressionConstant, properties {r=1}}
  MaterialExpressionConstant_5 {class=MaterialExpressionConstant}
  MaterialExpressionConvert_0 {
   class=MaterialExpressionConvert
   inputs [
    None {from=MaterialExpressionScalarParameter_18}
    None {from=MaterialExpressionScalarParameter_19}
    None {from=MaterialExpressionScalarParameter_17}
   ]
   properties {
    nodeName=Append
    convertInputs [
     {
      type=Scalar
      expressionInput {
       expression=&self:MaterialExpressionScalarParameter_18
       inputName=None
       mask=0
       maskA=0
       maskB=0
       maskG=0
       maskR=0
       outputIndex=0
      }
     }
     {
      type=Scalar
      expressionInput {
       expression=&self:MaterialExpressionScalarParameter_19
       inputName=None
       mask=0
       maskA=0
       maskB=0
       maskG=0
       maskR=0
       outputIndex=0
      }
     }
     {
      type=Scalar
      expressionInput {
       expression=&self:MaterialExpressionScalarParameter_17
       inputName=None
       mask=0
       maskA=0
       maskB=0
       maskG=0
       maskR=0
       outputIndex=0
      }
     }
    ]
    convertOutputs [{type=Vector3}]
   }
  }
  MaterialExpressionConvert_1 {
   class=MaterialExpressionConvert
   inputs [
    None {from=MaterialExpressionScalarParameter_23}
    None {from=MaterialExpressionScalarParameter_24}
    None {from=MaterialExpressionScalarParameter_22}
   ]
   properties {
    nodeName=Append
    convertInputs [
     {
      type=Scalar
      expressionInput {
       expression=&self:MaterialExpressionScalarParameter_23
       inputName=None
       mask=0
       maskA=0
       maskB=0
       maskG=0
       maskR=0
       outputIndex=0
      }
     }
     {
      type=Scalar
      expressionInput {
       expression=&self:MaterialExpressionScalarParameter_24
       inputName=None
       mask=0
       maskA=0
       maskB=0
       maskG=0
       maskR=0
       outputIndex=0
      }
     }
     {
      type=Scalar
      expressionInput {
       expression=&self:MaterialExpressionScalarParameter_22
       inputName=None
       mask=0
       maskA=0
       maskB=0
       maskG=0
       maskR=0
       outputIndex=0
      }
     }
    ]
    convertOutputs [{type=Vector3}]
   }
  }
  MaterialExpressionConvert_10 {
   class=MaterialExpressionConvert
   inputs [
    None {from=MaterialExpressionGetMaterialAttributes_1, output=Specular}
    None {from=MaterialExpressionGetMaterialAttributes_1, output=Specular}
    None {from=MaterialExpressionGetMaterialAttributes_1, output=Specular}
   ]
   properties {
    nodeName=Make Vector3
    convertInputs [
     {
      type=Scalar
      expressionInput {
       expression=&self:MaterialExpressionGetMaterialAttributes_1
       inputName=None
       mask=0
       maskA=0
       maskB=0
       maskG=0
       maskR=0
       outputIndex=1
      }
     }
     {
      type=Scalar
      expressionInput {
       expression=&self:MaterialExpressionGetMaterialAttributes_1
       inputName=None
       mask=0
       maskA=0
       maskB=0
       maskG=0
       maskR=0
       outputIndex=2
      }
     }
     {
      type=Scalar
      expressionInput {
       expression=&self:MaterialExpressionGetMaterialAttributes_1
       inputName=None
       mask=0
       maskA=0
       maskB=0
       maskG=0
       maskR=0
       outputIndex=3
      }
     }
    ]
    convertOutputs [{type=Vector3}]
   }
  }
  MaterialExpressionConvert_2 {
   class=MaterialExpressionConvert
   inputs [None {from=MaterialExpressionAdd_22}]
   properties {
    nodeName=Break Float3
    convertInputs [{
      type=Vector3
      expressionInput {
       expression=&self:MaterialExpressionAdd_22
       inputName=None
       mask=0
       maskA=0
       maskB=0
       maskG=0
       maskR=0
       outputIndex=0
      }
     }]
    convertOutputs [{type=Scalar}, {type=Scalar}, {type=Scalar}]
   }
  }
  MaterialExpressionConvert_3 {
   class=MaterialExpressionConvert
   inputs [
    None {from=MaterialExpressionScalarParameter_21}
    None {from=MaterialExpressionScalarParameter_45}
    None {from=MaterialExpressionScalarParameter_52}
   ]
   properties {
    nodeName=Append
    convertInputs [
     {
      type=Scalar
      expressionInput {
       expression=&self:MaterialExpressionScalarParameter_21
       inputName=None
       mask=0
       maskA=0
       maskB=0
       maskG=0
       maskR=0
       outputIndex=0
      }
     }
     {
      type=Scalar
      expressionInput {
       expression=&self:MaterialExpressionScalarParameter_45
       inputName=None
       mask=0
       maskA=0
       maskB=0
       maskG=0
       maskR=0
       outputIndex=0
      }
     }
     {
      type=Scalar
      expressionInput {
       expression=&self:MaterialExpressionScalarParameter_52
       inputName=None
       mask=0
       maskA=0
       maskB=0
       maskG=0
       maskR=0
       outputIndex=0
      }
     }
    ]
    convertOutputs [{type=Vector3}]
   }
  }
  MaterialExpressionConvert_4 {
   class=MaterialExpressionConvert
   inputs [None {from=MaterialExpressionMultiply_21}]
   properties {
    nodeName=Break Float3
    convertInputs [{
      type=Vector3
      expressionInput {
       expression=&self:MaterialExpressionMultiply_21
       inputName=None
       mask=0
       maskA=0
       maskB=0
       maskG=0
       maskR=0
       outputIndex=0
      }
     }]
    convertOutputs [{type=Scalar}, {type=Scalar}, {type=Scalar}]
   }
  }
  MaterialExpressionConvert_5 {
   class=MaterialExpressionConvert
   inputs [
    None {from=MaterialExpressionScalarParameter_47}
    None {from=MaterialExpressionScalarParameter_48}
    None {from=MaterialExpressionScalarParameter_46}
   ]
   properties {
    convertInputs [
     {
      type=Scalar
      expressionInput {
       expression=&self:MaterialExpressionScalarParameter_47
       inputName=None
       mask=0
       maskA=0
       maskB=0
       maskG=0
       maskR=0
       outputIndex=0
      }
     }
     {
      type=Scalar
      expressionInput {
       expression=&self:MaterialExpressionScalarParameter_48
       inputName=None
       mask=0
       maskA=0
       maskB=0
       maskG=0
       maskR=0
       outputIndex=0
      }
     }
     {
      type=Scalar
      expressionInput {
       expression=&self:MaterialExpressionScalarParameter_46
       inputName=None
       mask=0
       maskA=0
       maskB=0
       maskG=0
       maskR=0
       outputIndex=0
      }
     }
    ]
    convertOutputs [{type=Vector3}]
   }
  }
  MaterialExpressionConvert_6 {
   class=MaterialExpressionConvert
   inputs [
    None {from=MaterialExpressionScalarParameter_50}
    None {from=MaterialExpressionScalarParameter_51}
    None {from=MaterialExpressionScalarParameter_49}
   ]
   properties {
    nodeName=Append
    convertInputs [
     {
      type=Scalar
      expressionInput {
       expression=&self:MaterialExpressionScalarParameter_50
       inputName=None
       mask=0
       maskA=0
       maskB=0
       maskG=0
       maskR=0
       outputIndex=0
      }
     }
     {
      type=Scalar
      expressionInput {
       expression=&self:MaterialExpressionScalarParameter_51
       inputName=None
       mask=0
       maskA=0
       maskB=0
       maskG=0
       maskR=0
       outputIndex=0
      }
     }
     {
      type=Scalar
      expressionInput {
       expression=&self:MaterialExpressionScalarParameter_49
       inputName=None
       mask=0
       maskA=0
       maskB=0
       maskG=0
       maskR=0
       outputIndex=0
      }
     }
    ]
    convertOutputs [{type=Vector3}]
   }
  }
  MaterialExpressionConvert_7 {
   class=MaterialExpressionConvert
   inputs [None {from=MaterialExpressionNamedRerouteUsage_2}]
   properties {
    nodeName=Break Float4
    convertInputs [{
      type=Vector4
      defaultValue=<rgba 0 0 0 1>
      expressionInput {
       expression=&self:MaterialExpressionNamedRerouteUsage_2
       inputName=None
       mask=0
       maskA=0
       maskB=0
       maskG=0
       maskR=0
       outputIndex=0
      }
     }]
    convertOutputs [{type=Scalar}, {type=Scalar}, {type=Scalar}, {type=Scalar}]
   }
  }
  MaterialExpressionConvert_8 {
   class=MaterialExpressionConvert
   inputs [
    None {from=MaterialExpressionScalarParameter_74}
    None {from=MaterialExpressionScalarParameter_75}
    None {from=MaterialExpressionScalarParameter_73}
   ]
   properties {
    nodeName=Append
    convertInputs [
     {
      type=Scalar
      expressionInput {
       expression=&self:MaterialExpressionScalarParameter_74
       inputName=None
       mask=0
       maskA=0
       maskB=0
       maskG=0
       maskR=0
       outputIndex=0
      }
     }
     {
      type=Scalar
      expressionInput {
       expression=&self:MaterialExpressionScalarParameter_75
       inputName=None
       mask=0
       maskA=0
       maskB=0
       maskG=0
       maskR=0
       outputIndex=0
      }
     }
     {
      type=Scalar
      expressionInput {
       expression=&self:MaterialExpressionScalarParameter_73
       inputName=None
       mask=0
       maskA=0
       maskB=0
       maskG=0
       maskR=0
       outputIndex=0
      }
     }
    ]
    convertOutputs [{type=Vector3}]
   }
  }
  MaterialExpressionConvert_9 {
   class=MaterialExpressionConvert
   inputs [
    None {from=MaterialExpressionScalarParameter_77}
    None {from=MaterialExpressionScalarParameter_78}
    None {from=MaterialExpressionScalarParameter_76}
   ]
   properties {
    nodeName=Append
    convertInputs [
     {
      type=Scalar
      expressionInput {
       expression=&self:MaterialExpressionScalarParameter_77
       inputName=None
       mask=0
       maskA=0
       maskB=0
       maskG=0
       maskR=0
       outputIndex=0
      }
     }
     {
      type=Scalar
      expressionInput {
       expression=&self:MaterialExpressionScalarParameter_78
       inputName=None
       mask=0
       maskA=0
       maskB=0
       maskG=0
       maskR=0
       outputIndex=0
      }
     }
     {
      type=Scalar
      expressionInput {
       expression=&self:MaterialExpressionScalarParameter_76
       inputName=None
       mask=0
       maskA=0
       maskB=0
       maskG=0
       maskR=0
       outputIndex=0
      }
     }
    ]
    convertOutputs [{type=Vector3}]
   }
  }
  MaterialExpressionDDX_1 {class=MaterialExpressionDDX, inputs [Value {from=MaterialExpressionMaterialFunctionCall_10, output=UV Derivatives}], properties {}}
  MaterialExpressionDDY_1 {class=MaterialExpressionDDY, inputs [Value {from=MaterialExpressionMaterialFunctionCall_10, output=UV Derivatives}], properties {}}
  MaterialExpressionDistanceFieldApproxAO_0 {
   class=MaterialExpressionDistanceFieldApproxAO
   inputs [
    World Position {from="", output=""}
    Normal {from=MaterialExpressionLinearInterpolate_7, output=""}
    BaseDistance {from=MaterialExpressionScalarParameter_35, output=""}
    Radius {from="", output=""}
   ]
   properties {}
  }
  MaterialExpressionDivide_0 {
   class=MaterialExpressionDivide
   inputs [
    A {from=MaterialExpressionLocalPosition_0, output=XYZ}
    B {from=MaterialExpressionAdd_27}
   ]
  }
  MaterialExpressionDivide_1 {
   class=MaterialExpressionDivide
   inputs [
    A {from=MaterialExpressionDotProduct_1}
    B {from=MaterialExpressionDotProduct_2}
   ]
  }
  MaterialExpressionDotProduct_0 {
   class=MaterialExpressionDotProduct
   inputs [
    A {from=MaterialExpressionAdd_3, output=""}
    B {from=MaterialExpressionConstant_4, output=""}
   ]
   properties {}
  }
  MaterialExpressionDotProduct_1 {
   class=MaterialExpressionDotProduct
   inputs [
    A {from=MaterialExpressionTextureSample_0, output=RGBA}
    B {from=MaterialExpressionNamedRerouteUsage_3, output=""}
   ]
   properties {}
  }
  MaterialExpressionDotProduct_2 {
   class=MaterialExpressionDotProduct
   inputs [
    A {from=MaterialExpressionNamedRerouteUsage_3, output=""}
    B {from=MaterialExpressionConstant_1, output=""}
   ]
   properties {}
  }
  MaterialExpressionDotProduct_3 {
   class=MaterialExpressionDotProduct
   inputs [
    A {from=MaterialExpressionConvert_10, output=""}
    B {from=MaterialExpressionVectorParameter_6, output=RGB}
   ]
   properties {}
  }
  MaterialExpressionFunctionInput_0 {
   class=MaterialExpressionFunctionInput
   inputs [Preview {from="", output=""}]
   properties {inputName=Material, inputType=FunctionInput_MaterialAttributes}
  }
  MaterialExpressionFunctionInput_2 {
   class=MaterialExpressionFunctionInput
   inputs [Preview {from="", output=""}]
   properties {bUsePreviewValueAsDefault=true, inputName=Layer Mask, inputType=FunctionInput_Vector4, sortPriority=2, previewValue=<1 1 1 1>}
  }
  MaterialExpressionFunctionInput_3 {
   class=MaterialExpressionFunctionInput
   inputs [Preview {from="", output=""}]
   properties {bUsePreviewValueAsDefault=true, inputName=Layer Texture Mask, inputType=FunctionInput_Scalar, sortPriority=3, previewValue=<1 1 1 1>}
  }
  MaterialExpressionFunctionOutput_0 {class=MaterialExpressionFunctionOutput, inputs [None {from=MaterialExpressionStaticSwitchParameter_0, output=""}]}
  MaterialExpressionGetMaterialAttributes_0 {
   class=MaterialExpressionGetMaterialAttributes
   inputs [None {from=MaterialExpressionFunctionInput_0, output=""}]
   properties {attributeGetTypes [
     #1
     #2
     #3
     #4
    ]}
  }
  MaterialExpressionGetMaterialAttributes_1 {
   class=MaterialExpressionGetMaterialAttributes
   inputs [None {from=MaterialExpressionNamedRerouteUsage_4, output=""}]
   properties {attributeGetTypes [#3, #5, #6]}
  }
  MaterialExpressionGetMaterialAttributes_3 {
   class=MaterialExpressionGetMaterialAttributes
   inputs [None {from=MaterialExpressionNamedRerouteUsage_24, output=""}]
   properties {attributeGetTypes [#3, #6]}
  }
  MaterialExpressionGetMaterialAttributes_4 {
   class=MaterialExpressionGetMaterialAttributes
   inputs [None {from=MaterialExpressionSetMaterialAttributes_3, output=""}]
   properties {attributeGetTypes [#1]}
  }
  MaterialExpressionIf_0 {
   class=MaterialExpressionIf
   inputs [
    A {from=MaterialExpressionDotProduct_2, output=""}
    B {from="", output=""}
    "A > B" {from=MaterialExpressionDivide_1, output=""}
    "A == B" {from=MaterialExpressionConstant_2, output=""}
    "A < B" {from=MaterialExpressionConstant_2, output=""}
   ]
   properties {}
  }
  MaterialExpressionLinearInterpolate_11 {
   class=MaterialExpressionLinearInterpolate
   inputs [
    A {from=MaterialExpressionMultiply_27}
    B {from=MaterialExpressionMultiply_28}
    Alpha {from=MaterialExpressionNamedRerouteUsage_89}
   ]
  }
  MaterialExpressionLinearInterpolate_12 {
   class=MaterialExpressionLinearInterpolate
   inputs [
    A {from=MaterialExpressionScalarParameter_62}
    B {from=MaterialExpressionGetMaterialAttributes_3, output=Roughness}
    Alpha {from=MaterialExpressionMaterialFunctionCall_12, output=Result}
   ]
  }
  MaterialExpressionLinearInterpolate_3 {
   class=MaterialExpressionLinearInterpolate
   inputs [
    A {from=MaterialExpressionVertexNormalWS_1}
    B {from=MaterialExpressionNamedRerouteUsage_42}
    Alpha {from=MaterialExpressionScalarParameter_16}
   ]
  }
  MaterialExpressionLinearInterpolate_7 {
   class=MaterialExpressionLinearInterpolate
   inputs [
    A {from=MaterialExpressionVertexNormalWS_0}
    B {from=MaterialExpressionNamedRerouteUsage_53}
    Alpha {from=MaterialExpressionScalarParameter_58}
   ]
  }
  MaterialExpressionLinearInterpolate_9 {
   class=MaterialExpressionLinearInterpolate
   inputs [
    A {from=MaterialExpressionScalarParameter_37}
    B {from=MaterialExpressionGetMaterialAttributes_3, output=Specular}
    Alpha {from=MaterialExpressionMaterialFunctionCall_12, output=Result}
   ]
  }
  MaterialExpressionLocalPosition_0 {class=MaterialExpressionLocalPosition, inputs [], properties {}}
  MaterialExpressionMaterialFunctionCall_1 {
   class=MaterialExpressionMaterialFunctionCall
   inputs [
    "Xb1,PS4" {from=""}
    "PC-High&&Epic,XSX,PS5" {from=""}
    "PC-Perf,Android" {from=""}
    Switch {from=""}
    "PC-Low(NotPerf)&&Medium" {from=""}
   ]
   properties {materialFunction=&PlatformSwitch_1D}
  }
  MaterialExpressionMaterialFunctionCall_10 {
   class=MaterialExpressionMaterialFunctionCall
   inputs [
    UV Scale {from=MaterialExpressionScalarParameter_39}
    Mask {from=MaterialExpressionNamedRerouteDeclaration_11}
   ]
   properties {materialFunction=&/MF_4Layer_WorldSpaceUVs_UVs}
  }
  MaterialExpressionMaterialFunctionCall_11 {
   class=MaterialExpressionMaterialFunctionCall
   inputs [
    Enum {from=MaterialExpressionScalarParameter_79}
    Material {from=MaterialExpressionNamedRerouteUsage_77}
    Debug 1 {from=MaterialExpressionNamedRerouteUsage_81}
    Debug 2 {from=""}
    Debug 3 {from=""}
    Debug 4 {from=""}
    Debug 5 {from=""}
   ]
   properties {materialFunction=&/MF_4Layer_Debug}
  }
  MaterialExpressionMaterialFunctionCall_12 {
   class=MaterialExpressionMaterialFunctionCall
   inputs [
    Input {from=MaterialExpressionNamedRerouteUsage_90}
    Input Low {from=MaterialExpressionScalarParameter_64}
    Input High {from=MaterialExpressionScalarParameter_67}
   ]
   properties {materialFunction=&RemapValue_0-1}
  }
  MaterialExpressionMaterialFunctionCall_13 {
   class=MaterialExpressionMaterialFunctionCall
   inputs [
    Enum {from=MaterialExpressionScalarParameter_56}
    Material {from=MaterialExpressionNamedRerouteUsage_80}
    Debug 1 {from=MaterialExpressionNamedRerouteUsage_86}
    Debug 2 {from=""}
    Debug 3 {from=""}
    Debug 4 {from=""}
    Debug 5 {from=""}
   ]
   properties {materialFunction=&MF_4Layer_Debug}
  }
  MaterialExpressionMaterialFunctionCall_3 {
   class=MaterialExpressionMaterialFunctionCall
   inputs [
    Default {from=MaterialExpressionQualitySwitch_1}
    Low/Mobile {from=MaterialExpressionConstant_5}
   ]
   properties {materialFunction=/Game/Effects/Fort_Effects/Materials/_Functions/MF_LowMobile_PathingSwitch}
  }
  MaterialExpressionMaterialFunctionCall_4 {
   class=MaterialExpressionMaterialFunctionCall
   inputs [
    RGBA Mask {from=MaterialExpressionNamedRerouteUsage_9}
    L1 {from=MaterialExpressionScalarParameter_12}
    L2 {from=MaterialExpressionScalarParameter_25}
    L3 {from=MaterialExpressionScalarParameter_26}
    L4 {from=MaterialExpressionScalarParameter_27}
   ]
   properties {materialFunction=&/MF_4Layer_ScalarMasking}
  }
  MaterialExpressionMaterialFunctionCall_5 {
   class=MaterialExpressionMaterialFunctionCall
   inputs [
    Mode {from=MaterialExpressionAdd_0}
    Blend Radius {from=MaterialExpressionScalarParameter_6}
   ]
   properties {materialFunction=&/MF_4Layer_WorldSpaceUVs_Masks}
  }
  MaterialExpressionMaterialFunctionCall_8 {
   class=MaterialExpressionMaterialFunctionCall
   inputs [
    Enum {from=MaterialExpressionScalarParameter_55}
    Material {from=MaterialExpressionNamedRerouteUsage_76}
    Debug 1 {from=MaterialExpressionNamedRerouteUsage_65}
    Debug 2 {from=""}
    Debug 3 {from=""}
    Debug 4 {from=""}
    Debug 5 {from=""}
   ]
   properties {materialFunction=&MF_4Layer_Debug}
  }
  MaterialExpressionMaterialFunctionCall_9 {
   class=MaterialExpressionMaterialFunctionCall
   inputs [
    Enum {from=MaterialExpressionScalarParameter_20}
    Material {from=MaterialExpressionNamedRerouteUsage_44}
    Debug 1 {from=MaterialExpressionNamedRerouteUsage_47}
    Debug 2 {from=""}
    Debug 3 {from=""}
    Debug 4 {from=""}
    Debug 5 {from=""}
   ]
   properties {materialFunction=&MF_4Layer_Debug}
  }
  MaterialExpressionMax_0 {
   class=MaterialExpressionMax
   inputs [
    A {from=MaterialExpressionConvert_4}
    B {from=MaterialExpressionConvert_4}
   ]
  }
  MaterialExpressionMax_1 {
   class=MaterialExpressionMax
   inputs [
    A {from=MaterialExpressionMax_0}
    B {from=MaterialExpressionConvert_4}
   ]
  }
  MaterialExpressionMax_2 {
   class=MaterialExpressionMax
   inputs [
    A {from=MaterialExpressionAdd_7}
    B {from=""}
   ]
  }
  MaterialExpressionMeshPaintTextureObject_0 {class=MaterialExpressionMeshPaintTextureObject, inputs [], properties {}}
  MaterialExpressionMultiply_0 {
   class=MaterialExpressionMultiply
   inputs [
    A {from=MaterialExpressionNamedRerouteUsage_73}
    B {from=MaterialExpressionNamedRerouteUsage_0}
   ]
  }
  MaterialExpressionMultiply_1 {
   class=MaterialExpressionMultiply
   inputs [
    A {from=MaterialExpressionConvert_7}
    B {from=MaterialExpressionVectorParameter_9, output=RGBA}
   ]
  }
  MaterialExpressionMultiply_16 {
   class=MaterialExpressionMultiply
   inputs [
    A {from=MaterialExpressionAdd_20}
    B {from=MaterialExpressionScalarParameter_59}
   ]
  }
  MaterialExpressionMultiply_17 {
   class=MaterialExpressionMultiply
   inputs [
    A {from=MaterialExpressionAdd_15}
    B {from=MaterialExpressionScalarParameter_34}
   ]
  }
  MaterialExpressionMultiply_19 {
   class=MaterialExpressionMultiply
   inputs [
    A {from=MaterialExpressionGetMaterialAttributes_4, output=BaseColor}
    B {from=MaterialExpressionLinearInterpolate_11}
   ]
  }
  MaterialExpressionMultiply_2 {
   class=MaterialExpressionMultiply
   inputs [
    A {from=MaterialExpressionSaturate_0}
    B {from=MaterialExpressionConvert_0}
   ]
  }
  MaterialExpressionMultiply_21 {
   class=MaterialExpressionMultiply
   inputs [
    A {from=MaterialExpressionVertexColor_0}
    B {from=MaterialExpressionConvert_3}
   ]
  }
  MaterialExpressionMultiply_22 {
   class=MaterialExpressionMultiply
   inputs [
    A {from=MaterialExpressionDivide_0}
    B {from=""}
   ]
   properties {constB=-1}
  }
  MaterialExpressionMultiply_27 {
   class=MaterialExpressionMultiply
   inputs [
    A {from=MaterialExpressionVectorParameter_3, output=RGB}
    B {from=MaterialExpressionVectorParameter_3, output=RGB}
   ]
  }
  MaterialExpressionMultiply_28 {
   class=MaterialExpressionMultiply
   inputs [
    A {from=MaterialExpressionVectorParameter_2, output=RGB}
    B {from=MaterialExpressionVectorParameter_2, output=RGB}
   ]
  }
  MaterialExpressionMultiply_29 {
   class=MaterialExpressionMultiply
   inputs [
    A {from=MaterialExpressionAdd_17}
    B {from=MaterialExpressionScalarParameter_53}
   ]
  }
  MaterialExpressionMultiply_3 {
   class=MaterialExpressionMultiply
   inputs [
    A {from=MaterialExpressionConvert_7}
    B {from=MaterialExpressionVectorParameter_1, output=RGBA}
   ]
  }
  MaterialExpressionMultiply_33 {
   class=MaterialExpressionMultiply
   inputs [
    A {from=MaterialExpressionSaturate_12}
    B {from=MaterialExpressionConvert_5}
   ]
  }
  MaterialExpressionMultiply_34 {
   class=MaterialExpressionMultiply
   inputs [
    A {from=MaterialExpressionSaturate_13}
    B {from=MaterialExpressionConvert_6}
   ]
  }
  MaterialExpressionMultiply_36 {
   class=MaterialExpressionMultiply
   inputs [
    A {from=MaterialExpressionAdd_21}
    B {from=MaterialExpressionScalarParameter_82}
   ]
  }
  MaterialExpressionMultiply_4 {
   class=MaterialExpressionMultiply
   inputs [
    A {from=MaterialExpressionLinearInterpolate_3}
    B {from=""}
   ]
   properties {constB=-1}
  }
  MaterialExpressionMultiply_5 {
   class=MaterialExpressionMultiply
   inputs [
    A {from=MaterialExpressionSaturate_4}
    B {from=MaterialExpressionConvert_1}
   ]
  }
  MaterialExpressionMultiply_6 {
   class=MaterialExpressionMultiply
   inputs [
    A {from=MaterialExpressionConvert_7}
    B {from=MaterialExpressionVectorParameter_4, output=RGBA}
   ]
  }
  MaterialExpressionMultiply_7 {
   class=MaterialExpressionMultiply
   inputs [
    A {from=MaterialExpressionConvert_7}
    B {from=MaterialExpressionVectorParameter_5, output=RGBA}
   ]
  }
  MaterialExpressionMultiply_8 {
   class=MaterialExpressionMultiply
   inputs [
    A {from=MaterialExpressionScalarParameter_63}
    B {from=MaterialExpressionMax_2}
   ]
  }
  MaterialExpressionMultiply_9 {
   class=MaterialExpressionMultiply
   inputs [
    A {from=MaterialExpressionNamedRerouteUsage_6}
    B {from=MaterialExpressionScalarParameter_1}
   ]
  }
  MaterialExpressionNamedRerouteDeclaration_0 {
   class=MaterialExpressionNamedRerouteDeclaration
   inputs [None {from=MaterialExpressionGetMaterialAttributes_0, output=MaterialAttributes}]
   properties {name=Material Input, nodeColor=<rgba 1 0.4922 0 1>}
  }
  MaterialExpressionNamedRerouteDeclaration_1 {
   class=MaterialExpressionNamedRerouteDeclaration
   inputs [None {from=MaterialExpressionFunctionInput_2}]
   properties {name=Layer Mask, nodeColor=<rgba 1 0.001 0.4767 1>}
  }
  MaterialExpressionNamedRerouteDeclaration_11 {
   class=MaterialExpressionNamedRerouteDeclaration
   inputs [None {from=MaterialExpressionMaterialFunctionCall_5, output=Dithered}]
   properties {name=Dithered Mask, nodeColor=<rgba 0 1 0.3672 1>}
  }
  MaterialExpressionNamedRerouteDeclaration_12 {
   class=MaterialExpressionNamedRerouteDeclaration
   inputs [None {from=MaterialExpressionMaterialFunctionCall_5, output=Smooth}]
   properties {name=WorldSpace Smooth Mask, nodeColor=<rgba 1 0 0.3047 1>}
  }
  MaterialExpressionNamedRerouteDeclaration_18 {
   class=MaterialExpressionNamedRerouteDeclaration
   inputs [None {from=MaterialExpressionGetMaterialAttributes_0, output=Opacity}]
   properties {name=Opacity Input, nodeColor=<rgba 1 0.0107 0.3163 1>}
  }
  MaterialExpressionNamedRerouteDeclaration_2 {
   class=MaterialExpressionNamedRerouteDeclaration
   inputs [None {from=MaterialExpressionGetMaterialAttributes_0, output=BaseColor}]
   properties {name=BaseColor Input, nodeColor=<rgba 1 0 0.2109 1>}
  }
  MaterialExpressionNamedRerouteDeclaration_23 {
   class=MaterialExpressionNamedRerouteDeclaration
   inputs [None {from=MaterialExpressionMaterialFunctionCall_4, output=Result}]
   properties {name=Variation Layer Mask, nodeColor=<rgba 1 0 0.375 1>}
  }
  MaterialExpressionNamedRerouteDeclaration_29 {
   class=MaterialExpressionNamedRerouteDeclaration
   inputs [None {from=MaterialExpressionBlendMaterialAttributes_1}]
   properties {name=Variation Result, nodeColor=<rgba 1 0 0.8672 1>}
  }
  MaterialExpressionNamedRerouteDeclaration_3 {
   class=MaterialExpressionNamedRerouteDeclaration
   inputs [None {from=MaterialExpressionGetMaterialAttributes_0, output=Normal}]
   properties {name=Normal Input, nodeColor=<rgba 0 0.5312 1 1>}
  }
  MaterialExpressionNamedRerouteDeclaration_30 {
   class=MaterialExpressionNamedRerouteDeclaration
   inputs [None {from=MaterialExpressionMultiply_0}]
   properties {name=Mask Before Blend, nodeColor=<rgba 0 0.0156 1 1>}
  }
  MaterialExpressionNamedRerouteDeclaration_32 {
   class=MaterialExpressionNamedRerouteDeclaration
   inputs [None {from=MaterialExpressionMaterialFunctionCall_9, output=Result}]
   properties {name=Debug Mask Normals, nodeColor=<rgba 0.0312 0 1 1>}
  }
  MaterialExpressionNamedRerouteDeclaration_33 {
   class=MaterialExpressionNamedRerouteDeclaration
   inputs [None {from=MaterialExpressionAdd_13}]
   properties {name=Mask Normals, nodeColor=<rgba 1 0 0.7969 1>}
  }
  MaterialExpressionNamedRerouteDeclaration_35 {
   class=MaterialExpressionNamedRerouteDeclaration
   inputs [None {from=MaterialExpressionDDX_1}]
   properties {name=Mask DDX, nodeColor=<rgba 0.8047 1 0 1>}
  }
  MaterialExpressionNamedRerouteDeclaration_36 {
   class=MaterialExpressionNamedRerouteDeclaration
   inputs [None {from=MaterialExpressionDDY_1}]
   properties {name=Mask DDY, nodeColor=<rgba 0.875 0 1 1>}
  }
  MaterialExpressionNamedRerouteDeclaration_37 {
   class=MaterialExpressionNamedRerouteDeclaration
   inputs [None {from=MaterialExpressionMaterialFunctionCall_10, output=UVs}]
   properties {name=Mask UVs, nodeColor=<rgba 0 1 0.5547 1>}
  }
  MaterialExpressionNamedRerouteDeclaration_38 {
   class=MaterialExpressionNamedRerouteDeclaration
   inputs [None {from=MaterialExpressionPower_0}]
   properties {name=Variation Pattern Texture, nodeColor=<rgba 1 0.7734 0 1>}
  }
  MaterialExpressionNamedRerouteDeclaration_4 {
   class=MaterialExpressionNamedRerouteDeclaration
   inputs [None {from=MaterialExpressionStaticSwitchParameter_7}]
   properties {name=Layer Vector Selection, nodeColor=<rgba 0.4297 1 0 1>}
  }
  MaterialExpressionNamedRerouteDeclaration_41 {
   class=MaterialExpressionNamedRerouteDeclaration
   inputs [None {from=MaterialExpressionMaterialFunctionCall_10, output=UV X}]
   properties {name=Mask UVX, nodeColor=<rgba 1 0 0 1>}
  }
  MaterialExpressionNamedRerouteDeclaration_42 {
   class=MaterialExpressionNamedRerouteDeclaration
   inputs [None {from=MaterialExpressionMaterialFunctionCall_10, output=UV Y}]
   properties {name=Mask UVY, nodeColor=<rgba 0 1 0 1>}
  }
  MaterialExpressionNamedRerouteDeclaration_43 {
   class=MaterialExpressionNamedRerouteDeclaration
   inputs [None {from=MaterialExpressionMaterialFunctionCall_10, output=UV Z}]
   properties {name=Mask UVZ, nodeColor=<rgba 0 0 1 1>}
  }
  MaterialExpressionNamedRerouteDeclaration_47 {
   class=MaterialExpressionNamedRerouteDeclaration
   inputs [None {from=MaterialExpressionAdd_12}]
   properties {name=Bounds Mask, nodeColor=<rgba 1 0.3047 0 1>}
  }
  MaterialExpressionNamedRerouteDeclaration_48 {
   class=MaterialExpressionNamedRerouteDeclaration
   inputs [None {from=MaterialExpressionMaterialFunctionCall_8, output=Result}]
   properties {name=Debug DFAO Mask, nodeColor=<rgba 0.0312 0 1 1>}
  }
  MaterialExpressionNamedRerouteDeclaration_49 {
   class=MaterialExpressionNamedRerouteDeclaration
   inputs [None {from=MaterialExpressionMultiply_36}]
   properties {name=DFAO Mask, nodeColor=<rgba 0.5469 1 0 1>}
  }
  MaterialExpressionNamedRerouteDeclaration_5 {
   class=MaterialExpressionNamedRerouteDeclaration
   inputs [None {from=MaterialExpressionGetMaterialAttributes_0, output=Specular}]
   properties {name=Specular Input, nodeColor=<rgba 0 0.0859 1 1>}
  }
  MaterialExpressionNamedRerouteDeclaration_50 {
   class=MaterialExpressionNamedRerouteDeclaration
   inputs [None {from=MaterialExpressionMaterialFunctionCall_11, output=Result}]
   properties {name=Debug Bounds Mask, nodeColor=<rgba 0.0312 0 1 1>}
  }
  MaterialExpressionNamedRerouteDeclaration_52 {
   class=MaterialExpressionNamedRerouteDeclaration
   inputs [None {from=MaterialExpressionMax_1}]
   properties {name=Vertex Color Mask, nodeColor=<rgba 0.5938 0 1 1>}
  }
  MaterialExpressionNamedRerouteDeclaration_53 {
   class=MaterialExpressionNamedRerouteDeclaration
   inputs [None {from=MaterialExpressionMaterialFunctionCall_13, output=Result}]
   properties {name=Debug Vertex Color Mask, nodeColor=<rgba 0.0312 0 1 1>}
  }
  MaterialExpressionNamedRerouteDeclaration_54 {
   class=MaterialExpressionNamedRerouteDeclaration
   inputs [None {from=MaterialExpressionFunctionInput_3}]
   properties {name=Layer Texture Mask, nodeColor=<rgba 0 1 0.5078 1>}
  }
  MaterialExpressionNamedRerouteDeclaration_6 {
   class=MaterialExpressionNamedRerouteDeclaration
   inputs [None {from=MaterialExpressionPower_3}]
   properties {name=DFAO Raw, nodeColor=<rgba 1 0.9609 0 1>}
  }
  MaterialExpressionNamedRerouteUsage_0 {class=MaterialExpressionNamedRerouteUsage}
  MaterialExpressionNamedRerouteUsage_1 {class=MaterialExpressionNamedRerouteUsage}
  MaterialExpressionNamedRerouteUsage_2 {class=MaterialExpressionNamedRerouteUsage}
  MaterialExpressionNamedRerouteUsage_24 {class=MaterialExpressionNamedRerouteUsage}
  MaterialExpressionNamedRerouteUsage_29 {class=MaterialExpressionNamedRerouteUsage}
  MaterialExpressionNamedRerouteUsage_3 {class=MaterialExpressionNamedRerouteUsage}
  MaterialExpressionNamedRerouteUsage_4 {class=MaterialExpressionNamedRerouteUsage}
  MaterialExpressionNamedRerouteUsage_42 {class=MaterialExpressionNamedRerouteUsage}
  MaterialExpressionNamedRerouteUsage_44 {class=MaterialExpressionNamedRerouteUsage}
  MaterialExpressionNamedRerouteUsage_45 {class=MaterialExpressionNamedRerouteUsage}
  MaterialExpressionNamedRerouteUsage_47 {class=MaterialExpressionNamedRerouteUsage}
  MaterialExpressionNamedRerouteUsage_48 {class=MaterialExpressionNamedRerouteUsage}
  MaterialExpressionNamedRerouteUsage_50 {class=MaterialExpressionNamedRerouteUsage}
  MaterialExpressionNamedRerouteUsage_51 {class=MaterialExpressionNamedRerouteUsage}
  MaterialExpressionNamedRerouteUsage_52 {class=MaterialExpressionNamedRerouteUsage}
  MaterialExpressionNamedRerouteUsage_53 {class=MaterialExpressionNamedRerouteUsage}
  MaterialExpressionNamedRerouteUsage_59 {class=MaterialExpressionNamedRerouteUsage}
  MaterialExpressionNamedRerouteUsage_6 {class=MaterialExpressionNamedRerouteUsage}
  MaterialExpressionNamedRerouteUsage_65 {class=MaterialExpressionNamedRerouteUsage}
  MaterialExpressionNamedRerouteUsage_68 {class=MaterialExpressionNamedRerouteUsage}
  MaterialExpressionNamedRerouteUsage_72 {class=MaterialExpressionNamedRerouteUsage}
  MaterialExpressionNamedRerouteUsage_73 {class=MaterialExpressionNamedRerouteUsage}
  MaterialExpressionNamedRerouteUsage_74 {class=MaterialExpressionNamedRerouteUsage}
  MaterialExpressionNamedRerouteUsage_75 {class=MaterialExpressionNamedRerouteUsage}
  MaterialExpressionNamedRerouteUsage_76 {class=MaterialExpressionNamedRerouteUsage}
  MaterialExpressionNamedRerouteUsage_77 {class=MaterialExpressionNamedRerouteUsage}
  MaterialExpressionNamedRerouteUsage_80 {class=MaterialExpressionNamedRerouteUsage}
  MaterialExpressionNamedRerouteUsage_81 {class=MaterialExpressionNamedRerouteUsage}
  MaterialExpressionNamedRerouteUsage_82 {class=MaterialExpressionNamedRerouteUsage}
  MaterialExpressionNamedRerouteUsage_85 {class=MaterialExpressionNamedRerouteUsage}
  MaterialExpressionNamedRerouteUsage_86 {class=MaterialExpressionNamedRerouteUsage}
  MaterialExpressionNamedRerouteUsage_87 {class=MaterialExpressionNamedRerouteUsage}
  MaterialExpressionNamedRerouteUsage_88 {class=MaterialExpressionNamedRerouteUsage}
  MaterialExpressionNamedRerouteUsage_89 {class=MaterialExpressionNamedRerouteUsage}
  MaterialExpressionNamedRerouteUsage_9 {class=MaterialExpressionNamedRerouteUsage}
  MaterialExpressionNamedRerouteUsage_90 {class=MaterialExpressionNamedRerouteUsage}
  MaterialExpressionOneMinus_0 {class=MaterialExpressionOneMinus, inputs [None {from=MaterialExpressionSaturate_7}]}
  MaterialExpressionPower_0 {
   class=MaterialExpressionPower
   inputs [
    Base {from=MaterialExpressionIf_0}
    Exp {from=MaterialExpressionMultiply_8}
   ]
  }
  MaterialExpressionPower_3 {
   class=MaterialExpressionPower
   inputs [
    Base {from=MaterialExpressionOneMinus_0}
    Exp {from=MaterialExpressionScalarParameter_57}
   ]
  }
  MaterialExpressionQualitySwitch_0 {
   class=MaterialExpressionQualitySwitch
   inputs [
    Default {from="", output=""}
    Low {from="", output=""}
    High {from="", output=""}
    Medium {from="", output=""}
    Epic {from=MaterialExpressionAdd_19, output=""}
   ]
   properties {}
  }
  MaterialExpressionQualitySwitch_1 {
   class=MaterialExpressionQualitySwitch
   inputs [
    Default {from=MaterialExpressionConstant_5, output=""}
    Low {from="", output=""}
    High {from="", output=""}
    Medium {from="", output=""}
    Epic {from=MaterialExpressionMultiply_9, output=""}
   ]
   properties {}
  }
  MaterialExpressionReroute_0 {class=MaterialExpressionReroute, inputs [None {from=MaterialExpressionAdd_1}]}
  MaterialExpressionSaturate_0 {class=MaterialExpressionSaturate, inputs [None {from=MaterialExpressionLinearInterpolate_3}]}
  MaterialExpressionSaturate_1 {class=MaterialExpressionSaturate, inputs [None {from=MaterialExpressionNamedRerouteUsage_88}]}
  MaterialExpressionSaturate_12 {class=MaterialExpressionSaturate, inputs [None {from=MaterialExpressionAdd_25}]}
  MaterialExpressionSaturate_13 {class=MaterialExpressionSaturate, inputs [None {from=MaterialExpressionAdd_26}]}
  MaterialExpressionSaturate_4 {class=MaterialExpressionSaturate, inputs [None {from=MaterialExpressionMultiply_4}]}
  MaterialExpressionSaturate_7 {class=MaterialExpressionSaturate, inputs [None {from=MaterialExpressionDistanceFieldApproxAO_0}]}
  MaterialExpressionScalarParameter_0 {
   class=MaterialExpressionScalarParameter
   properties {
    controlType=Enumeration
    defaultValue=0.15
    enumeration=/BRCoreMaterials/Materials/Enums/EN_4Layer_Cover
    group=Variation | Variation World Space UVs
    parameterName=Variation WorldSpaceUVs Mode
    sliderMax=1
    sortPriority=0
   }
  }
  MaterialExpressionScalarParameter_1 {class=MaterialExpressionScalarParameter, properties {group=Variation | Variation DFAO Mask, parameterName=Variation DFAO Contrast Boost}}
  MaterialExpressionScalarParameter_12 {
   class=MaterialExpressionScalarParameter
   properties {defaultValue=1, group=Variation | Variation Layer Mask, parameterName=L1 Variation, sliderMax=1, sortPriority=0}
  }
  MaterialExpressionScalarParameter_16 {
   class=MaterialExpressionScalarParameter
   properties {defaultValue=0.5, group=Variation | Variation Normal Mask, parameterName=Variation Normal Mask Per Pixel, sliderMax=1, sortPriority=8}
  }
  MaterialExpressionScalarParameter_17 {
   class=MaterialExpressionScalarParameter
   properties {defaultValue=0.5, group=Variation | Variation Normal Mask, parameterName=Variation Normal Mask Z +, sortPriority=0}
  }
  MaterialExpressionScalarParameter_18 {
   class=MaterialExpressionScalarParameter
   properties {group=Variation | Variation Normal Mask, parameterName=Variation Normal Mask X +, sortPriority=2}
  }
  MaterialExpressionScalarParameter_19 {
   class=MaterialExpressionScalarParameter
   properties {group=Variation | Variation Normal Mask, parameterName=Variation Normal Mask Y +, sortPriority=4}
  }
  MaterialExpressionScalarParameter_20 {
   class=MaterialExpressionScalarParameter
   properties {
    controlType=Enumeration
    enumeration=/BRCoreMaterials/Materials/Enums/EN_4Layer_GenericToggle
    group=Cover | Cover Normal Mask
    parameterName=Variation Normal Mask Debug Mode
    sliderMax=1
   }
  }
  MaterialExpressionScalarParameter_21 {
   class=MaterialExpressionScalarParameter
   properties {group=Variation | Variation Vertex Color Mask, parameterName=Variation Vertex Color Mask Red, sortPriority=0}
  }
  MaterialExpressionScalarParameter_22 {
   class=MaterialExpressionScalarParameter
   properties {group=Variation | Variation Normal Mask, parameterName=Variation Normal Mask Z -, sortPriority=1}
  }
  MaterialExpressionScalarParameter_23 {
   class=MaterialExpressionScalarParameter
   properties {group=Variation | Variation Normal Mask, parameterName=Variation Normal Mask X -, sortPriority=3}
  }
  MaterialExpressionScalarParameter_24 {
   class=MaterialExpressionScalarParameter
   properties {group=Variation | Variation Normal Mask, parameterName=Variation Normal Mask Y -, sortPriority=5}
  }
  MaterialExpressionScalarParameter_25 {
   class=MaterialExpressionScalarParameter
   properties {defaultValue=1, group=Variation | Variation Layer Mask, parameterName=L2 Variation, sliderMax=1, sortPriority=0}
  }
  MaterialExpressionScalarParameter_26 {
   class=MaterialExpressionScalarParameter
   properties {defaultValue=1, group=Variation | Variation Layer Mask, parameterName=L3 Variation, sliderMax=1, sortPriority=0}
  }
  MaterialExpressionScalarParameter_27 {
   class=MaterialExpressionScalarParameter
   properties {defaultValue=1, group=Variation | Variation Layer Mask, parameterName=L4 Variation, sliderMax=1, sortPriority=0}
  }
  MaterialExpressionScalarParameter_34 {
   class=MaterialExpressionScalarParameter
   properties {group=Variation | Variation Normal Mask, parameterName=Variation Normal Mask Pattern Weight, sortPriority=10}
  }
  MaterialExpressionScalarParameter_35 {
   class=MaterialExpressionScalarParameter
   properties {defaultValue=150, group=Variation | Variation DFAO Mask, parameterName=Variation DFAO Mask Base Distance, sortPriority=0}
  }
  MaterialExpressionScalarParameter_36 {
   class=MaterialExpressionScalarParameter
   properties {group=Variation | Variation Normal Mask, parameterName=Variation Normal Mask Pattern Offset, sortPriority=11}
  }
  MaterialExpressionScalarParameter_37 {
   class=MaterialExpressionScalarParameter
   properties {
    defaultValue=0.5
    enumeration=/BRCoreMaterials/Materials/Enums/EN_4Layer_WorldSpaceUVs_Toggle
    group=Variation | Variation Color and Material Values
    parameterName=Variation Specular
    sortPriority=2
   }
  }
  MaterialExpressionScalarParameter_39 {
   class=MaterialExpressionScalarParameter
   properties {
    defaultValue=1500
    enumeration=&EN_4Layer_WorldSpaceUVs_Toggle
    group=Variation | Variation World Space UVs
    parameterName=Variation WorldSpaceUVs Scale Pattern
    sortPriority=2
   }
  }
  MaterialExpressionScalarParameter_45 {
   class=MaterialExpressionScalarParameter
   properties {group=Variation | Variation Vertex Color Mask, parameterName=Variation Vertex Color Mask Green, sortPriority=1}
  }
  MaterialExpressionScalarParameter_46 {
   class=MaterialExpressionScalarParameter
   properties {group=Variation | Variation Bounds Mask, parameterName=Variation Bounds Mask Z +, sortPriority=0}
  }
  MaterialExpressionScalarParameter_47 {
   class=MaterialExpressionScalarParameter
   properties {group=Variation | Variation Bounds Mask, parameterName=Variation Bounds Mask X +, sortPriority=2}
  }
  MaterialExpressionScalarParameter_48 {
   class=MaterialExpressionScalarParameter
   properties {group=Variation | Variation Bounds Mask, parameterName=Variation Bounds Mask Y +, sortPriority=4}
  }
  MaterialExpressionScalarParameter_49 {
   class=MaterialExpressionScalarParameter
   properties {group=Variation | Variation Bounds Mask, parameterName=Variation Bounds Mask Z -, sortPriority=1}
  }
  MaterialExpressionScalarParameter_50 {
   class=MaterialExpressionScalarParameter
   properties {group=Variation | Variation Bounds Mask, parameterName=Variation Bounds Mask X-, sortPriority=3}
  }
  MaterialExpressionScalarParameter_51 {
   class=MaterialExpressionScalarParameter
   properties {group=Variation | Variation Bounds Mask, parameterName=Variation Bounds Mask Y -, sortPriority=5}
  }
  MaterialExpressionScalarParameter_52 {
   class=MaterialExpressionScalarParameter
   properties {group=Variation | Variation Vertex Color Mask, parameterName=Variation Vertex Color Mask Blue, sortPriority=2}
  }
  MaterialExpressionScalarParameter_53 {
   class=MaterialExpressionScalarParameter
   properties {group=Variation | Variation Bounds Mask, parameterName=Variation Bounds Mask Pattern Weight, sortPriority=12}
  }
  MaterialExpressionScalarParameter_54 {
   class=MaterialExpressionScalarParameter
   properties {group=Variation | Variation Bounds Mask, parameterName=Variation Bounds Mask Pattern Offset, sortPriority=13}
  }
  MaterialExpressionScalarParameter_55 {
   class=MaterialExpressionScalarParameter
   properties {
    controlType=Enumeration
    enumeration=&EN_4Layer_GenericToggle
    group=Cover | Cover DFAO Mask
    parameterName=Variation DFAO Mask Debug Mode
    sliderMax=1
   }
  }
  MaterialExpressionScalarParameter_56 {
   class=MaterialExpressionScalarParameter
   properties {
    controlType=Enumeration
    enumeration=&EN_4Layer_GenericToggle
    group=Cover | Cover Vertex Color Mask
    parameterName=Variation Vertex Color Mask Debug Mode
    sliderMax=1
   }
  }
  MaterialExpressionScalarParameter_57 {
   class=MaterialExpressionScalarParameter
   properties {defaultValue=1, group=Variation | Variation DFAO Mask, parameterName=Variation DFAO Falloff, sortPriority=2}
  }
  MaterialExpressionScalarParameter_58 {
   class=MaterialExpressionScalarParameter
   properties {group=Variation | Variation DFAO Mask, parameterName=Variation DFAO Mask Pixel Normals, sliderMax=1, sortPriority=1}
  }
  MaterialExpressionScalarParameter_59 {
   class=MaterialExpressionScalarParameter
   properties {group=Variation | Variation DFAO Mask, parameterName=Variation DFAO Mask Pattern Weight, sortPriority=3}
  }
  MaterialExpressionScalarParameter_6 {
   class=MaterialExpressionScalarParameter
   properties {
    defaultValue=0.1
    enumeration=&EN_4Layer_WorldSpaceUVs_Toggle
    group=Variation | Variation World Space UVs
    parameterName=Variation WorldSpaceUVs Blend
    sliderMax=1
    sortPriority=3
   }
  }
  MaterialExpressionScalarParameter_60 {
   class=MaterialExpressionScalarParameter
   properties {group=Variation | Variation DFAO Mask, parameterName=Variation DFAO Mask Pattern Offset, sortPriority=4}
  }
  MaterialExpressionScalarParameter_62 {
   class=MaterialExpressionScalarParameter
   properties {
    defaultValue=0.6
    enumeration=&EN_4Layer_WorldSpaceUVs_Toggle
    group=Variation | Variation Color and Material Values
    parameterName=Variation Roughness
    sortPriority=5
   }
  }
  MaterialExpressionScalarParameter_63 {
   class=MaterialExpressionScalarParameter
   properties {
    defaultValue=1
    enumeration=&EN_4Layer_WorldSpaceUVs_Toggle
    group=Variation | Variation Color and Material Values
    parameterName=Variation Pattern Contrast
    sortPriority=4
   }
  }
  MaterialExpressionScalarParameter_64 {
   class=MaterialExpressionScalarParameter
   properties {defaultValue=0.3, group=Variation | Variation Color and Material Values, parameterName=Variation Pattern Spec/Rough Mask Low}
  }
  MaterialExpressionScalarParameter_67 {
   class=MaterialExpressionScalarParameter
   properties {defaultValue=0.6, group=Variation | Variation Color and Material Values, parameterName=Variation Pattern Spec/Rough Mask High}
  }
  MaterialExpressionScalarParameter_73 {
   class=MaterialExpressionScalarParameter
   properties {group=Variation | Variation Bounds Mask, parameterName=Variation Bounds Mask Offset Z +, sortPriority=6}
  }
  MaterialExpressionScalarParameter_74 {
   class=MaterialExpressionScalarParameter
   properties {group=Variation | Variation Bounds Mask, parameterName=Variation Bounds Mask Offset X +, sortPriority=8}
  }
  MaterialExpressionScalarParameter_75 {
   class=MaterialExpressionScalarParameter
   properties {group=Variation | Variation Bounds Mask, parameterName=Variation Bounds Mask Offset Y +, sortPriority=10}
  }
  MaterialExpressionScalarParameter_76 {
   class=MaterialExpressionScalarParameter
   properties {group=Variation | Variation Bounds Mask, parameterName=Variation Bounds Mask Offset Z -, sortPriority=7}
  }
  MaterialExpressionScalarParameter_77 {
   class=MaterialExpressionScalarParameter
   properties {group=Variation | Variation Bounds Mask, parameterName=Variation Bounds Mask Offset X-, sortPriority=9}
  }
  MaterialExpressionScalarParameter_78 {
   class=MaterialExpressionScalarParameter
   properties {group=Variation | Variation Bounds Mask, parameterName=Variation Bounds Mask Offset Y -, sortPriority=11}
  }
  MaterialExpressionScalarParameter_79 {
   class=MaterialExpressionScalarParameter
   properties {
    controlType=Enumeration
    enumeration=&EN_4Layer_GenericToggle
    group=Cover | Cover Bounds Mask
    parameterName=Variation Bounds Mask Debug Mode
    sliderMax=1
   }
  }
  MaterialExpressionScalarParameter_82 {
   class=MaterialExpressionScalarParameter
   properties {group=Variation | Variation DFAO Mask, parameterName=Variation DFAO Mask Weight, sortPriority=-1}
  }
  MaterialExpressionSetMaterialAttributes_3 {
   class=MaterialExpressionSetMaterialAttributes
   inputs [
    MaterialAttributes {from=MaterialExpressionGetMaterialAttributes_3, output=MaterialAttributes}
    Specular {from=MaterialExpressionLinearInterpolate_9, output=""}
    Roughness {from=MaterialExpressionLinearInterpolate_12, output=""}
   ]
   properties {attributeSetTypes [#3, #6]}
  }
  MaterialExpressionSetMaterialAttributes_4 {
   class=MaterialExpressionSetMaterialAttributes
   inputs [
    MaterialAttributes {from=MaterialExpressionGetMaterialAttributes_4, output=MaterialAttributes}
    BaseColor {from=MaterialExpressionMultiply_19, output=""}
   ]
   properties {attributeSetTypes [#1]}
  }
  MaterialExpressionStaticSwitchParameter_0 {
   class=MaterialExpressionStaticSwitchParameter
   inputs [
    True {from=MaterialExpressionStaticSwitchParameter_2, output=""}
    False {from=MaterialExpressionNamedRerouteUsage_29, output=""}
   ]
   properties {group=- Effect Switches, parameterName=Variation}
  }
  MaterialExpressionStaticSwitchParameter_2 {
   class=MaterialExpressionStaticSwitchParameter
   inputs [
    True {from=MaterialExpressionNamedRerouteUsage_48, output=""}
    False {from=MaterialExpressionStaticSwitchParameter_6, output=""}
   ]
   properties {group=Variation | Variation Normal Mask, parameterName=Variation Normal Mask Debug, sortPriority=31}
  }
  MaterialExpressionStaticSwitchParameter_4 {
   class=MaterialExpressionStaticSwitchParameter
   inputs [
    True {from=MaterialExpressionNamedRerouteUsage_82, output=""}
    False {from=MaterialExpressionStaticSwitchParameter_5, output=""}
   ]
   properties {group=Variation | Variation Bounds Mask, parameterName=Variation Bounds Mask Debug, sortPriority=31}
  }
  MaterialExpressionStaticSwitchParameter_5 {
   class=MaterialExpressionStaticSwitchParameter
   inputs [
    True {from=MaterialExpressionNamedRerouteUsage_59, output=""}
    False {from=MaterialExpressionStaticSwitchParameter_6, output=""}
   ]
   properties {group=Variation | Variation DFAO Mask, parameterName=Variation DFAO Mask Debug, sortPriority=31}
  }
  MaterialExpressionStaticSwitchParameter_6 {
   class=MaterialExpressionStaticSwitchParameter
   inputs [
    True {from=MaterialExpressionNamedRerouteUsage_87, output=""}
    False {from=MaterialExpressionNamedRerouteDeclaration_29, output=""}
   ]
   properties {group=Variation | Variation Vertex Color Mask, parameterName=Variation Vertex Color Mask Debug, sortPriority=31}
  }
  MaterialExpressionStaticSwitchParameter_7 {
   class=MaterialExpressionStaticSwitchParameter
   inputs [
    True {from=MaterialExpressionStaticSwitchParameter_8, output=""}
    False {from=MaterialExpressionMultiply_1, output=""}
   ]
   properties {group=- Layer Switches, parameterName=Use 2 Layers, sortPriority=0}
  }
  MaterialExpressionStaticSwitchParameter_8 {
   class=MaterialExpressionStaticSwitchParameter
   inputs [
    True {from=MaterialExpressionStaticSwitchParameter_9, output=""}
    False {from=MaterialExpressionReroute_0, output=""}
   ]
   properties {group=- Layer Switches, parameterName=Use 3 Layers, sortPriority=1}
  }
  MaterialExpressionStaticSwitchParameter_9 {
   class=MaterialExpressionStaticSwitchParameter
   inputs [
    True {from=MaterialExpressionAdd_5, output=""}
    False {from=MaterialExpressionAdd_4, output=""}
   ]
   properties {group=- Layer Switches, parameterName=Use 4 Layers, sortPriority=2}
  }
  MaterialExpressionTextureCoordinate_0 {class=MaterialExpressionTextureCoordinate}
  MaterialExpressionTextureObjectParameter_2 {
   class=MaterialExpressionTextureObjectParameter
   inputs [
    UVs {from="", output=""}
    Apply View MipBias {from="", output=""}
   ]
   properties {
    group=Variation
    parameterName=Variation Pattern
    samplerType=SAMPLERTYPE_Masks
    sortPriority=4
    texture=/BRCoreMaterials/Materials/Textures/T_Mask
   }
  }
  MaterialExpressionTextureSample_0 {
   class=MaterialExpressionTextureSample
   inputs [
    UVs {from=MaterialExpressionNamedRerouteUsage_50}
    Tex {from=MaterialExpressionTextureObjectParameter_2}
    "DDX(UVs)" {from=MaterialExpressionNamedRerouteUsage_51}
    "DDY(UVs)" {from=MaterialExpressionNamedRerouteUsage_52}
    Apply View MipBias {from=""}
   ]
   properties {mipValueMode=TMVM_Derivative, samplerType=SAMPLERTYPE_Masks, texture=&T_Mask}
  }
  MaterialExpressionTextureSample_1 {
   class=MaterialExpressionTextureSample
   inputs [
    UVs {from=MaterialExpressionTextureCoordinate_0}
    Tex {from=MaterialExpressionMeshPaintTextureObject_0}
    Apply View MipBias {from=""}
   ]
   properties {texture=/Engine/EngineMaterials/T_Default_MacroVariation}
  }
  MaterialExpressionVectorParameter_0 {class=MaterialExpressionVectorParameter, properties {group=Variation, parameterName=L1 VariationMaskChannel, defaultValue=<rgba 1 0 0 0>}}
  MaterialExpressionVectorParameter_1 {class=MaterialExpressionVectorParameter, properties {group=Variation, parameterName=L2 VariationMaskChannel, defaultValue=<rgba 1 0 0 0>}}
  MaterialExpressionVectorParameter_2 {
   class=MaterialExpressionVectorParameter
   properties {group=Variation | Variation Color and Material Values, parameterName=Variation Tint Color High, sortPriority=5, defaultValue=<rgba 1 1 1 1>}
  }
  MaterialExpressionVectorParameter_3 {
   class=MaterialExpressionVectorParameter
   properties {group=Variation | Variation Color and Material Values, parameterName=Variation Tint Color Low, sortPriority=5, defaultValue=<rgba 1 1 1 1>}
  }
  MaterialExpressionVectorParameter_4 {class=MaterialExpressionVectorParameter, properties {group=Variation, parameterName=L3 VariationMaskChannel, defaultValue=<rgba 1 0 0 0>}}
  MaterialExpressionVectorParameter_5 {class=MaterialExpressionVectorParameter, properties {group=Variation, parameterName=L4 VariationMaskChannel, defaultValue=<rgba 1 0 0 0>}}
  MaterialExpressionVectorParameter_6 {
   class=MaterialExpressionVectorParameter
   properties {
    group=Variation | Variation Color and Material Values
    parameterName=Variation Base Layer Spec Mask Contrast Boost
    defaultValue=<rgba 64 0 0 0>
   }
  }
  MaterialExpressionVectorParameter_9 {class=MaterialExpressionVectorParameter, properties {group=Variation, parameterName=L1 VariationMaskChannel, defaultValue=<rgba 1 0 0 0>}}
  MaterialExpressionVertexColor_0 {class=MaterialExpressionVertexColor, inputs [], properties {}}
  MaterialExpressionVertexNormalWS_0 {class=MaterialExpressionVertexNormalWS, inputs [], properties {}}
  MaterialExpressionVertexNormalWS_1 {class=MaterialExpressionVertexNormalWS, inputs [], properties {}}
 }}
