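# Material function that adds an optional velvet sheen to a four-layer material, toggled by the Velvet static switch.
# A tiled distortion texture perturbs the normal, feeding a repeating, triangle-wave view-angle mask for brightened, saturation-adjusted reflection color.
# The same texture's blue channel drives a specular multiplier, while metallic takes an override value.
# Per-layer weights through MF_4Layer_ScalarMasking, multiplied by a mask, blend the velvet attributes over the incoming material.
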MaterialFunction /BRCoreMaterials/Materials/MaterialFunctions/MF_4Layer_Velvet
graph {nodes [
  n1 = NamedRerouteUsage()
  n2 = Multiply(A=NamedRerouteUsage(), B=ScalarParameter() {defaultValue=1, group=Velvet, parameterName=Velvet BaseColor Multiplier, sortPriority=5})
  n3 = Multiply(A=@n2, B=ScalarParameter() {defaultValue=3, group=Velvet, parameterName=Velvet Reflection Multiplier, sortPriority=6}) {constB=8}
  n4 = Multiply(A=Add(A=ScalarParameter() {defaultValue=0.9, group=Velvet, parameterName=Velvet Reflection Saturation, sortPriority=7}) {constB=-1}) {constB=-1}
  n5 = SetMaterialAttributes!(
   MaterialAttributes=NamedRerouteUsage()
   BaseColor=LinearInterpolate(A=@n2, B=Desaturation(None=@n3, Fraction=@n4), Alpha=NamedRerouteUsage())
   Metallic=ScalarParameter() {defaultValue=0.2, group=Velvet, parameterName=Velvet Metallic Override, sortPriority=8}
   Normal=NamedRerouteUsage()
   Specular=Multiply(A=NamedRerouteUsage(), B=NamedRerouteUsage())
  ) {attributeSetTypes [
    #1
    #2
    #3
    #4
   ]}
  n6 = Multiply(
   A=MaterialFunctionCall!(
    RGBA Mask=NamedRerouteUsage()
    L1=ScalarParameter() {group=Velvet, parameterName=L1 Velvet Weight, sliderMax=1, sortPriority=0}
    L2=ScalarParameter() {group=Velvet, parameterName=L2 Velvet Weight, sliderMax=1, sortPriority=1}
    L3=ScalarParameter() {group=Velvet, parameterName=L3 Velvet Weight, sliderMax=1, sortPriority=2}
    L4=ScalarParameter() {group=Velvet, parameterName=L4 Velvet Weight, sliderMax=1, sortPriority=3}
   ) {materialFunction=&/MF_4Layer_ScalarMasking}.Result
   B=NamedRerouteUsage()
  )
  FunctionOutput(None=StaticSwitchParameter(True=BlendMaterialAttributes(A=@n1, B=@n5, Alpha=@n6), False=@n1) {group=- Effect Switches, parameterName=Velvet})
  n7 = GetMaterialAttributes(None=FunctionInput() {inputName=Material, inputType=FunctionInput_MaterialAttributes}) {attributeGetTypes [#1, #3, #4]}
  NamedRerouteDeclaration(None=@n7.MaterialAttributes) {name=Material Input, nodeColor=<rgba 1 0.4922 0 1>}
  n8 = FunctionInput() {bUsePreviewValueAsDefault=true, inputName=Layer Mask, inputType=FunctionInput_Vector4, sortPriority=1, previewValue=<1 1 1 1>}
  NamedRerouteDeclaration(None=@n8) {name=Layer Mask, nodeColor=<rgba 1 0 0.3471 1>}
  NamedRerouteDeclaration(None=@n7.BaseColor) {name=BaseColor Input, nodeColor=<rgba 1 0 0.2109 1>}
  NamedRerouteDeclaration(None=@n7.Normal) {name=Normal Input, nodeColor=<rgba 0 0.5312 1 1>}
  NamedRerouteDeclaration(None=@n7.Specular) {name=Specular Input, nodeColor=<rgba 0 0.0859 1 1>}
  n9 = TextureSample(UVs=Multiply(A=TextureCoordinate(), B=ScalarParameter() {defaultValue=4, group=Velvet, parameterName=Velvet Texture Tiling, sortPriority=4})) {samplerType=SAMPLERTYPE_LinearColor, texture=/Game/Global/Textures/Default/Masks_and_Gradients/T_Velvet_Distortion}
  NamedRerouteDeclaration(None=Power(Base=@n9.B, Exp=ScalarParameter() {defaultValue=1, group=Velvet, parameterName=Velvet Specular Power, sortPriority=9}) {constExponent=8}) {name=Specular Multiplier, nodeColor=<rgba 0.4062 1 0 1>}
  n10 = NamedRerouteUsage()
  n11 = Multiply(
   A=Add(A=@n9.RGB) {constB=-0.5}
   B=ScalarParameter() {defaultValue=0.5, group=Velvet, parameterName=Velvet Distortion Multiplier, sortPriority=10}
  ) {constB=1.6}
  n12 = AppendVector(A=Add(A=ComponentMask(None=@n10) {g=true, r=true}, B=ComponentMask(None=@n11) {g=true, r=true}), B=ComponentMask(None=@n10) {b=true})
  n13 = DotProduct(A=Transform(None=NamedRerouteDeclaration(None=@n12) {name=Normal Distorted, nodeColor=<rgba 1 0.3281 0 1>}), B=CameraVectorWS())
  n14 = Frac(None=Multiply(A=Saturate(None=@n13), B=ScalarParameter() {defaultValue=2, group=Velvet, parameterName=Velvet Fresnel Tiling, sortPriority=11}) {constB=1.2})
  n15 = Multiply(A=Abs(None=Add(A=@n14) {constB=-0.5})) {constB=2}
  n16 = ScalarParameter() {defaultValue=2, group=Velvet, parameterName=Velvet Fresnel Power, sortPriority=12}
  NamedRerouteDeclaration(None=Multiply(A=Power(Base=@n15, Exp=@n16), B=NamedRerouteUsage())) {name=Fresnel Mask, nodeColor=<rgba 0 0.5781 1 1>}
  n17 = FunctionInput() {bUsePreviewValueAsDefault=true, inputName=Layer Texture Mask, inputType=FunctionInput_Scalar, sortPriority=2, previewValue=<1 1 1 1>}
  NamedRerouteDeclaration(None=@n17) {name=Layer Texture Mask, nodeColor=<rgba 0 0.5312 1 1>}
 ]}
