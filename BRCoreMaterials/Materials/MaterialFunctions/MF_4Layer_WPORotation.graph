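# Material function that adds optional four-layer animated rotation to materials through World Position Offset, gated by the WPO Rotation static switch.
# Layer Mask selects per-layer speed, blade count, axis and pivot offset, with axis and pivot transformed from instance space.
# Time-driven rotation wraps within one blade interval via Frac and Divide.
# RotateVertexAboutAxis receives these values, and its WPO output feeds both World Position Offset and Normal.
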MaterialFunction /BRCoreMaterials/Materials/MaterialFunctions/MF_4Layer_WPORotation
graph {
 properties {description="This function is used to give assets rotation on the X, Y, and Z-axis. Normalized rotation along with fixed normals. "}
 nodes [
  n1 = FunctionInput() {inputName=Material, inputType=FunctionInput_MaterialAttributes}
  n2 = GetMaterialAttributes(None=@n1) {attributeGetTypes [#1]}
  n3 = MaterialFunctionCall!(
   RotationAngle=NamedRerouteUsage()
   NormalizedRotationAxis=NamedRerouteUsage()
   PivotPoint=NamedRerouteUsage()
   Position=WorldPosition() {worldPositionShaderOffset=WPT_CameraRelative}.XYZ
   Normal=@n2.Normal
   TangentU
   TangentV
  ) {materialFunction=/Game/Characters/Player/Male/Medium/Bodies/M_MED_Inferno/FX/Mat/RotateVertexAboutAxis}
  FunctionOutput(None=StaticSwitchParameter(
    True=SetMaterialAttributes!(MaterialAttributes=@n2.MaterialAttributes, WorldPositionOffset=@n3.WPO, Normal=@n3.WPO) {attributeSetTypes [#2, #1]}
    False=@n1
   ) {group=- Effect Switches, parameterName=WPO Rotation})
  n4 = NamedRerouteUsage()
  n5 = ScalarParameter() {group=WPO Rotation, parameterName=L1 WPORotation Speed, sortPriority=0}
  n6 = ScalarParameter() {group=WPO Rotation, parameterName=L2 WPORotation Speed, sortPriority=4}
  n7 = ScalarParameter() {group=WPO Rotation, parameterName=L3 WPORotation Speed, sortPriority=8}
  n8 = ScalarParameter() {group=WPO Rotation, parameterName=L4 WPORotation Speed, sortPriority=12}
  n9 = MaterialFunctionCall!(
   RGBA Mask=NamedRerouteUsage()
   L1=ScalarParameter() {defaultValue=1, group=WPO Rotation, parameterName=L1 WPORotation Looping Blades, sortPriority=1}
   L2=ScalarParameter() {defaultValue=1, group=WPO Rotation, parameterName=L2 WPORotation Looping Blades, sortPriority=5}
   L3=ScalarParameter() {defaultValue=1, group=WPO Rotation, parameterName=L3 WPORotation Looping Blades, sortPriority=9}
   L4=ScalarParameter() {defaultValue=1, group=WPO Rotation, parameterName=L4 WPORotation Looping Blades, sortPriority=13}
  ) {materialFunction=&/MF_4Layer_ScalarMasking}
  n10 = Multiply(
   A=Multiply(
    A=MaterialFunctionCall(RGBA Mask=@n4, L1=@n5, L2=@n6, L3=@n7, L4=@n8) {materialFunction=&MF_4Layer_ScalarMasking}.Result
    B=Time() {period=1}
   )
   B=@n9.Result
  ) {constB=2}
  NamedRerouteDeclaration(None=Divide(A=Frac(None=@n10), B=@n9.Result)) {name=RotationAmount, nodeColor=<rgba 0 1 0.2031 1>}
  n11 = VectorParameter() {group=WPO Rotation, parameterName=L1 WPORotation RotationAxis, sortPriority=2, defaultValue=<rgba 1 0 0 1>}
  n12 = VectorParameter() {group=WPO Rotation, parameterName=L2 WPORotation RotationAxis, sortPriority=6, defaultValue=<rgba 1 0 0 1>}
  n13 = VectorParameter() {group=WPO Rotation, parameterName=L3 WPORotation RotationAxis, sortPriority=10, defaultValue=<rgba 1 0 0 1>}
  n14 = VectorParameter() {group=WPO Rotation, parameterName=L4 WPORotation RotationAxis, sortPriority=14, defaultValue=<rgba 1 0 0 1>}
  n15 = MaterialFunctionCall!(Layer Mask=NamedRerouteUsage(), L1=@n11.RGB, L2=@n12.RGB, L3=@n13.RGB, L4=@n14.RGB) {materialFunction=&/MF_4Layer_ColorMasking}
  NamedRerouteDeclaration(None=Normalize(VectorInput=Transform(None=@n15.Result) {transformSourceType=TRANSFORMSOURCE_Instance})) {name=RotationAxis, nodeColor=<rgba 0.5469 0 1 1>}
  NamedRerouteDeclaration(None=FunctionInput() {inputName=Layer Mask, inputType=FunctionInput_Vector4, sortPriority=1}) {name=Layer Mask, nodeColor=<rgba 0 1 0.3438 1>}
  n16 = NamedRerouteUsage()
  n17 = VectorParameter() {group=WPO Rotation, parameterName=L1 WPORotation Pivot Offset, sortPriority=3, defaultValue=<rgba 0 0 0 1>}
  n18 = VectorParameter() {group=WPO Rotation, parameterName=L2 WPORotation Pivot Offset, sortPriority=7, defaultValue=<rgba 0 0 0 1>}
  n19 = VectorParameter() {group=WPO Rotation, parameterName=L3 WPORotation Pivot Offset, sortPriority=11, defaultValue=<rgba 0 0 0 1>}
  n20 = VectorParameter() {group=WPO Rotation, parameterName=L4 WPORotation Pivot Offset, sortPriority=15, defaultValue=<rgba 0 0 0 1>}
  NamedRerouteDeclaration(None=TransformPosition(None=MaterialFunctionCall(Layer Mask=@n16, L1=@n17.RGB, L2=@n18.RGB, L3=@n19.RGB, L4=@n20.RGB) {materialFunction=&MF_4Layer_ColorMasking}.Result) {transformSourceType=TRANSFORMPOSSOURCE_Instance, transformType=TRANSFORMPOSSOURCE_TranslatedWorld}) {name=Pivot Offset, nodeColor=<rgba 0.2891 1 0 1>}
 ]
}
