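# Material function that samples diffuse, specular mask and normal textures with world-space biplanar projection on XZ and YZ planes.
# WorldPosition is Z-flipped, offset and scaled into both UV sets.
# Normal samples are swizzled and combined with the vertex normal per plane.
# A contrast-powered vertex-normal weight blends each pair, but its denominator adds R twice instead of R+G, so YZ never exceeds half weight.
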MaterialFunction /BRCoreMaterials/Materials/MaterialFunctions/MF_Biplanar_Sampling
graph {nodes [
  n1 = FunctionInput(Preview=TextureObject() {texture=/Game/Environments/Helios/Terrain/Textures/T_Helios_RustBucket_Cliff_Strata_01_D}) {inputName=T_Diffuse, inputType=FunctionInput_Texture2D, sortPriority=10}
  n2 = FunctionInput(Preview=WorldPosition().XYZ) {bUsePreviewValueAsDefault=true, inputName=WorldPosition, sortPriority=-1}
  n3 = Constant3Vector() {constant=<rgba 1 1 -1 0>}
  n4 = Constant3Vector()
  n5 = Divide(
   A=Add(A=Multiply(A=@n2, B=@n3), B=FunctionInput(Preview=@n4) {bUsePreviewValueAsDefault=true, inputName=Biplanar - Position Offset})
   B=FunctionInput(Preview=Constant() {r=2550}) {bUsePreviewValueAsDefault=true, inputName=Biplanar - Scale, inputType=FunctionInput_Scalar, sortPriority=1}
  )
  n6 = ComponentMask(None=@n5) {b=true, r=true}
  n7 = ComponentMask(None=@n5) {b=true, g=true}
  n8 = VertexNormalWS()
  n9 = ComponentMask(None=Power(Base=Abs(None=@n8), Exp=FunctionInput() {inputName=Biplanar - Fade Contrast, inputType=FunctionInput_Scalar, sortPriority=2}) {constExponent=10}) {g=true, r=true}
  n10 = MaterialFunctionCall(Float2=@n9) {materialFunction=&BreakOutFloat2Components}
  n11 = Add(A=@n10.R, B=@n10.R)
  n12 = ComponentMask(None=Divide(A=@n9, B=Add(A=@n11) {constB=0.001, desc=fix division by zero for upward faces})) {r=true}
  FunctionOutput(None=LinearInterpolate(A=TextureSample(UVs=@n6, Tex=@n1).RGBA, B=TextureSample(UVs=@n7, Tex=@n1).RGBA, Alpha=@n12)) {outputName=Biplanar - Diffuse, sortPriority=5}
  n13 = FunctionInput(Preview=TextureObject() {samplerType=SAMPLERTYPE_Masks, texture=/Game/Environments/Helios/Terrain/Textures/T_Helios_RustBucket_Cliff_Strata_01_S}) {inputName=T_SpecularMasks, inputType=FunctionInput_Texture2D, sortPriority=11}
  FunctionOutput(None=LinearInterpolate(A=TextureSample(UVs=@n6, Tex=@n13).RGB, B=TextureSample(UVs=@n7, Tex=@n13).RGB, Alpha=@n12)) {outputName=Biplanar - SpecularMasks, sortPriority=6}
  n14 = FunctionInput(Preview=TextureObject() {samplerType=SAMPLERTYPE_Normal, texture=/Game/Environments/Helios/Terrain/Textures/T_Helios_RustBucket_Cliff_Strata_01_N}) {inputName=T_Normals, inputType=FunctionInput_Texture2D, sortPriority=12}
  n15 = MaterialFunctionCall(Color=TextureSample(UVs=@n6, Tex=@n14).RGB) {materialFunction=&SplitComponents}
  n16 = VertexNormalWS()
  n17 = ComponentMask(None=@n16) {r=true}
  n18 = Add(A=@n15.R, B=@n17)
  n19 = ComponentMask(None=@n16) {g=true}
  n20 = Multiply(A=@n15.B, B=@n19)
  n21 = ComponentMask(None=@n16) {b=true}
  n22 = Add(A=Multiply(A=@n15.G) {constB=-1}, B=@n21)
  n23 = MaterialFunctionCall(Color=TextureSample(UVs=@n7, Tex=@n14).RGB) {materialFunction=&SplitComponents}
  n24 = Multiply(A=@n23.B, B=@n17)
  n25 = Add(A=@n23.R, B=@n19)
  n26 = Add(A=Multiply(A=@n23.G) {constB=-1}, B=@n21)
  FunctionOutput(None=LinearInterpolate(
    A=MaterialFunctionCall(R=@n18, G=@n20, B=@n22) {materialFunction=&AppendMany}.RGB
    B=MaterialFunctionCall(R=@n24, G=@n25, B=@n26) {materialFunction=&AppendMany}.RGB
    Alpha=@n12
   )) {outputName=Biplanar - Normals, sortPriority=7}
 ]}
