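# Rock material function that tints the BaseColor of incoming blended-layer MaterialAttributes using baked curvature and ambient occlusion.
# Curvature, shaped by bias and contrast, adds a color.
# Inverted AO, optionally min/max-combined with DFAO, then masks a second color that multiplies or replaces BaseColor according to that color's alpha.
# Static switch inputs gate each stage, and each can be limited to tintable layers through the Tintable Mask.
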MaterialFunction /BRCoreMaterials/Materials/MaterialFunctions/MF_Rocks_Curvature_AO
graph {nodes [
  FunctionInput() {inputName="\\ --- Masks --- /", inputType=FunctionInput_Scalar, sortPriority=20}
  FunctionInput() {inputName="\\ --- Dependencies --- /", inputType=FunctionInput_Scalar, sortPriority=800}
  n1 = VectorParameter() {
   desc=AO Color is masked by the other AO params and then Multiplied or Replaces the color of the blended L1 through L4 layers
   group=--- F1 | Baked Curvature/AO
   parameterName=* AO - Color
   sortPriority=460
   channelNames {a=Linear vs Mult, b="", g="", r=""}
   defaultValue=<rgba 0.2691 0.1783 0.1117 0.5>
  }
  n2 = QualitySwitch(Default=@n1, Epic=@n1)
  n3 = FunctionInput(Preview=SetMaterialAttributes!(MaterialAttributes)) {bUsePreviewValueAsDefault=true, inputName=Input MaterialAttributes, inputType=FunctionInput_MaterialAttributes, sortPriority=-100}
  n4 = GetMaterialAttributes(None=@n3) {attributeGetTypes [#1]}
  n5 = VectorParameter() {
   desc=Curvature Color is masked by the other Curvature params and then added to the blended L1 through L4 layers
   group=--- F1 | Baked Curvature/AO
   parameterName=* Curvature - Color
   sortPriority=401
   defaultValue=<rgba 0.4115 0.4115 0.4115 1>
  }
  n6 = ScalarParameter() {
   desc="Uses Baked Curvature data coming from L1 SpecularMasks or VTX color\r\nIncrease/decrease Bias to increase/decrease curvature coverage"
   group=--- F1 | Baked Curvature/AO
   parameterName=* Curvature - Bias
   sliderMax=1
   sortPriority=402
  }
  n7 = NamedRerouteUsage()
  n8 = ScalarParameter() {
   defaultValue=1
   desc="Uses Baked Curvature data coming from L1 SpecularMasks or VTX color\r\nAdd Soft Curvature 1 = Use Full range Curvature \r\nAdd Soft Curvature 0 = Reduce range of Curvature data"
   group=--- F1 | Baked Curvature/AO
   parameterName=* Curvature - Contrast
   sortPriority=403
  }
  n9 = Saturate(None=Add(A=@n6, B=Multiply(A=@n7, B=@n8)))
  n10 = ScalarParameter() {
   defaultValue=1
   desc="Uses Tintable Mask from Layers to Mask Curvature\r\n1 - On - Mask Curvature effect to Layers with \"Is Tintable Layer\" = True\r\n0 - Off - Ignore Tintable Mask and affect all layers"
   group=--- F1 | Baked Curvature/AO
   parameterName=* Curvature - Use Tintable Mask
   sliderMax=1
   sortPriority=400
  }
  n11 = If(A=@n10, "A > B"=NamedRerouteUsage(), "A < B"=Constant() {r=1}) {constB=0.5, equalsThreshold=0.001}
  n12 = Add(A=@n4.BaseColor, B=Multiply(A=@n5.RGB, B=Multiply(A=@n9, B=@n11)))
  n13 = GetMaterialAttributes(None=StaticSwitch(
    True=SetMaterialAttributes!(MaterialAttributes=@n4.MaterialAttributes, BaseColor=@n12) {attributeSetTypes [#1]}
    False=@n4.MaterialAttributes
    Value=NamedRerouteUsage()
   )) {attributeGetTypes [#1]}
  n14 = ScalarParameter() {
   desc="1 - Results in Max blend of DFAO and Baked AO (Generally results in the AO Coloring being more limited)\r\n0 - Results in a Min blend of DFAO and Baked AO"
   group=--- F1 | Baked Curvature/AO
   parameterName="• AO - DFAO - Blend (1=Max 0=Min)"
   sliderMax=1
   sortPriority=451
  }
  n15 = NamedRerouteUsage()
  n16 = NamedRerouteUsage()
  n17 = StaticBoolParameter() {
   desc="Off - AO Color uses the Baked AO coming from L1 SpecularMasks or VTX color\r\nO - AO Color uses the DFAO values set in the \"Rock/Cliff Shared settings\" parameter group"
   group=--- F1 | Baked Curvature/AO
   parameterName="• AO - Use DFAO"
   sortPriority=450
  }
  n18 = StaticSwitch(
   True=If(A=@n14, "A > B"=Max(A=@n15, B=@n16), "A < B"=Min(A=@n15, B=@n16)) {constB=0.5, equalsThreshold=0.001}
   False=NamedRerouteUsage()
   Value=@n17
  )
  n19 = ScalarParameter() {group=--- F1 | Baked Curvature/AO, parameterName=* AO - Bias, sortPriority=465}
  n20 = ScalarParameter() {defaultValue=1, group=--- F1 | Baked Curvature/AO, parameterName=* AO - Contrast, sortPriority=466}
  n21 = Multiply(A=Add(A=OneMinus(None=@n18), B=QualitySwitch(Default=@n19, Epic=@n19)), B=@n20)
  n22 = ScalarParameter() {
   defaultValue=1
   desc="Uses Tintable Mask from Layers to Mask Dirt\r\n1 - On - Mask Dirt  effect to Layers with \"Is Tintable Layer\" = True\r\n0 - Off - Ignore Tintable Mask and affect all layers"
   group=--- F1 | Baked Curvature/AO
   parameterName=* AO - Use Tintable Mask
   sliderMax=1
   sortPriority=450
  }
  n23 = NamedRerouteUsage()
  n24 = Constant() {r=1}
  n25 = LinearInterpolate(
   A=@n13.BaseColor
   B=LinearInterpolate(A=Multiply(A=@n13.BaseColor, B=@n2), B=@n1, Alpha=@n1.Linear vs Mult)
   Alpha=Multiply(A=Saturate(None=@n21), B=If(A=@n22, "A > B"=@n23, "A < B"=@n24) {constB=0.5, equalsThreshold=0.001})
  )
  FunctionOutput(None=StaticSwitch(
    True=SetMaterialAttributes(MaterialAttributes=@n13.MaterialAttributes, BaseColor=@n25) {attributeSetTypes [#1]}
    False=@n13.MaterialAttributes
    Value=NamedRerouteUsage()
   )) {outputName=Blended MaterialAttributes, sortPriority=-100}
  NamedRerouteDeclaration(None=FunctionInput() {inputName=DFAO, inputType=FunctionInput_Scalar, sortPriority=812}) {name=_DFAO, nodeColor=<rgba 0 0 1 1>}
  NamedRerouteDeclaration() {name=_Dirt Mask, nodeColor=<rgba 1 1 1 1>}
  NamedRerouteDeclaration() {name=_Curvature Mask, nodeColor=<rgba 1 1 1 1>}
  NamedRerouteDeclaration(None=FunctionInput() {inputName=Baked Curvature, inputType=FunctionInput_Scalar, sortPriority=810}) {name=_Curvature, nodeColor=<rgba 0 0 1 1>}
  NamedRerouteDeclaration(None=FunctionInput(Preview=StaticBool()) {bUsePreviewValueAsDefault=true, inputName=Use Baked AO, inputType=FunctionInput_StaticBool, sortPriority=-109}) {name=Use Dirt, nodeColor=<rgba 0.1979 0.0084 0 1>}
  NamedRerouteDeclaration(None=FunctionInput() {bUsePreviewValueAsDefault=true, inputName=Tintable Mask, inputType=FunctionInput_Scalar, sortPriority=101}) {name=_Tintable Mask, nodeColor=<rgba 1 1 1 1>}
  NamedRerouteDeclaration(None=FunctionInput(Preview=StaticBool()) {bUsePreviewValueAsDefault=true, inputName=Use Baked Curvature, inputType=FunctionInput_StaticBool, sortPriority=-110}) {name=Use Curvature, nodeColor=<rgba 0.1979 0.0084 0 1>}
  n26 = FunctionInput(Preview=Constant() {r=1}) {bUsePreviewValueAsDefault=true, inputName=Baked AO, inputType=FunctionInput_Scalar, sortPriority=811, previewValue=<1 0 0 1>}
  NamedRerouteDeclaration(None=@n26) {name=_Baked AO, nodeColor=<rgba 0 0 1 1>}
  NamedRerouteUsage()
  NamedRerouteUsage()
  QualitySwitch(Default=@n20, Epic=@n20)
 ]}
