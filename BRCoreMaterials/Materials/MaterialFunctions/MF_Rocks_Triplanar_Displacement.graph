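# Rock material function for world-space triplanar texturing, with static switches for realtime displacement and expensive triplanar that select the sampled RGBA.
# World position makes three planar UV sets, weighted by vertex normals raised to a fade power.
# It builds full three-sample, flat hard-projection and temporally dithered single-sample variants, the latter using explicit derivatives.
# Alongside RGB and alpha, it outputs flat and dithered UVs, derivatives, dither mask and seam mask.
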MaterialFunction /BRCoreMaterials/Materials/MaterialFunctions/MF_Rocks_Triplanar_Displacement
graph {nodes [
  n1 = Power(Base=NamedRerouteUsage(), Exp=NamedRerouteUsage()) {constExponent=10}
  n2 = MaterialFunctionCall(Float3=Saturate(None=@n1)) {materialFunction=&BreakOutFloat3Components}
  n3 = Divide(A=@n1, B=Add(A=Add(A=@n2.R, B=@n2.R), B=@n2.B))
  ComponentMask(None=@n3) {g=true}
  n4 = LinearInterpolate(
   A=LinearInterpolate(A=NamedRerouteUsage(), B=NamedRerouteUsage(), Alpha=NamedRerouteUsage())
   B=NamedRerouteUsage()
   Alpha=NamedRerouteUsage()
  )
  FunctionOutput(None=@n4) {outputName=Triplanar Flat - UVs, sortPriority=30}
  FunctionOutput(None=StaticSwitch(True=NamedRerouteUsage(), False=NamedRerouteUsage(), Value=NamedRerouteUsage())) {outputName=Triplanar - Seam Mask, sortPriority=35}
  n5 = StaticSwitch(
   True=StaticSwitch(True=NamedRerouteUsage(), False=NamedRerouteUsage(), Value=NamedRerouteUsage())
   False=NamedRerouteUsage()
   Value=NamedRerouteUsage()
  )
  FunctionOutput(None=ComponentMask(None=@n5) {b=true, g=true, r=true}) {outputName=RGB, sortPriority=-20}
  FunctionOutput(None=ComponentMask(None=@n5) {a=true}) {outputName=A, sortPriority=-10}
  n6 = FunctionInput(Preview=VertexNormalWS()) {bUsePreviewValueAsDefault=true, description=VertexNormalWS as Default, inputName="VertexNormal (WS)", sortPriority=6}
  n7 = FunctionInput() {inputName=Triplanar - Fade, inputType=FunctionInput_Scalar, sortPriority=10, previewValue=<1 0 0 1>}
  n8 = Power(
   Base=NamedRerouteDeclaration(None=Abs(None=@n6)) {name="VertexNormal WS (Abs)", nodeColor=<rgba 0 1 0.6016 1>}
   Exp=NamedRerouteDeclaration(None=@n7) {name=Triplanar - Fade, nodeColor=<rgba 1 0.2812 0 1>}
  ) {constExponent=10}
  n9 = MaterialFunctionCall(Float3=Saturate(None=@n8)) {materialFunction=&BreakOutFloat3Components}
  n10 = Divide(A=@n8, B=Add(A=Add(A=@n9.R, B=@n9.R), B=@n9.B))
  n11 = ViewProperty() {property=MEVP_TemporalSampleIndex}
  MaterialExpressionAdd_4 = Add(A=ScreenPosition().PixelPosition, B=AppendVector(A=@n11.Property, B=@n11.Property))
  n12 = Custom!(p=@MaterialExpressionAdd_4) {
   code="Mod( ((uint)(p.x) + 2 * (uint)(p.y)) , 5 )"
   outputType=CMOT_Float1
   inputs [{inputName=p, input {expression=&self:MaterialExpressionAdd_4}}]
  }
  n13 = TextureSample(UVs=MaterialFunctionCall!(TextureResolution=Constant2Vector() {g=64, r=64}) {materialFunction=/Engine/Functions/Engine_MaterialFunctions02/Texturing/ScreenAlignedPixelToPixelUVs}.Result) {samplerType=SAMPLERTYPE_LinearGrayscale, texture=/Engine/EngineMaterials/Good64x64TilingNoiseHighFreq}
  n14 = Max(A=Multiply(A=Add(A=@n12, B=@n13.R) {constB=0}) {constB=0.1667}) {constB=0.001}
  n15 = OneMinus(None=@n14)
  n16 = Floor(None=Add(A=ComponentMask(None=@n10) {g=true, r=true}, B=AppendVector(A=@n14, B=@n15)))
  n17 = ComponentMask(None=@n16) {r=true}
  n18 = ComponentMask(None=@n16) {g=true}
  n19 = OneMinus(None=Max(A=@n17, B=@n18))
  FunctionOutput(None=AppendVector(A=AppendVector(A=@n17, B=@n18), B=@n19)) {outputName=Triplanar Dithered - Mask, sortPriority=27}
  MaterialFunctionCall!(UV Scale, Seam Sharpness) {materialFunction=/Game/Athena/Environments/MaterialFunctions/MF_TriplanarFlat}
  n20 = FunctionInput(Preview=WorldPosition() {worldPositionShaderOffset=WPT_ExcludeAllShaderOffsets}.XYZ) {
   bUsePreviewValueAsDefault=true
   description="Absolute World Position (Excluding Material Offsets) as Default"
   inputName=Absolute World Position
   sortPriority=5
  }
  n21 = FunctionInput() {bUsePreviewValueAsDefault=true, inputName=Texture Scale, sortPriority=1, previewValue=<120 120 120 120>}
  n22 = Divide(A=Multiply(A=@n20, B=Constant3Vector() {constant=<rgba 1 1 -1 0>}), B=@n21) {constB=120}
  NamedRerouteDeclaration(None=ComponentMask(None=@n22) {b=true, r=true}) {name=UV_XZ, nodeColor=<rgba 0 0.4 0 1>}
  NamedRerouteDeclaration(None=ComponentMask(None=@n22) {b=true, g=true}) {name=UV_YZ, nodeColor=<rgba 0.4 0 0 1>}
  n23 = NamedRerouteUsage()
  n24 = TextureSample(UVs=NamedRerouteUsage(), Tex=@n23) {samplerSource=SSM_Wrap_WorldGroupSettings}
  n25 = TextureSample(UVs=NamedRerouteUsage(), Tex=@n23) {samplerSource=SSM_Wrap_WorldGroupSettings}
  n26 = ComponentMask(None=@n3) {r=true}
  n27 = NamedRerouteUsage()
  NamedRerouteDeclaration(None=LinearInterpolate(
    A=LinearInterpolate(A=@n24.RGBA, B=@n25.RGBA, Alpha=@n26)
    B=TextureSample(UVs=@n27, Tex=@n23) {samplerSource=SSM_Wrap_WorldGroupSettings}.RGBA
    Alpha=ComponentMask(None=@n3) {b=true}
   )) {name=Triplanar Full - Sample, nodeColor=<rgba 0.2957 0.2957 1 1>}
  NamedRerouteDeclaration(None=Constant() {desc=Triplanar Full - Seam Mask}) {name=Triplanar Full - Seam Mask, nodeColor=<rgba 0.2957 0.2957 1 1>}
  n28 = NamedRerouteUsage()
  n29 = ComponentMask(None=@n28) {b=true}
  n30 = ComponentMask(None=@n28) {g=true, r=true}
  n31 = Subtract(A=@n29, B=@n30)
  n32 = ComponentMask(None=@n31) {r=true}
  n33 = ComponentMask(None=@n31) {g=true}
  n34 = Saturate(None=Subtract(A=@n30, B=@n29))
  n35 = ComponentMask(None=@n34) {r=true}
  n36 = ComponentMask(None=@n34) {g=true}
  n37 = Subtract(A=@n35, B=@n36)
  n38 = Max(A=Saturate(None=Min(A=@n32, B=@n33)), B=Abs(None=@n37))
  n39 = FunctionInput(Preview=Constant() {r=64}) {bUsePreviewValueAsDefault=true, inputName=Triplanar - Flat - Seam Sharpness, inputType=FunctionInput_Scalar, sortPriority=20}
  NamedRerouteDeclaration(None=Power(Base=OneMinus(None=@n38), Exp=@n39)) {name=Triplanar Flat - Seam Mask, nodeColor=<rgba 0 1 0.9531 1>}
  n40 = Add(A=@n37) {constB=0.5}
  n41 = NamedRerouteDeclaration(None=Saturate(None=Round(None=@n40))) {name=Flat R, nodeColor=<rgba 1 0 0.8906 1>}
  n42 = Add(A=Subtract(A=@n36, B=@n35)) {constB=0.5}
  n43 = NamedRerouteDeclaration(None=Saturate(None=Round(None=@n42))) {name=Flat G, nodeColor=<rgba 0 1 0.7891 1>}
  n44 = Max(A=@n41, B=@n43)
  NamedRerouteDeclaration(None=AppendVector(A=AppendVector(A=@n41, B=@n43), B=OneMinus(None=@n44))) {name=Triplanar Flat - Mask, nodeColor=<rgba 0.7344 1 0 1>}
  NamedRerouteDeclaration(None=FunctionInput(Preview=TextureObject() {texture=/Game/Global/Textures/Default/Blanks/T_FlatGrey}) {bUsePreviewValueAsDefault=true, inputName=Diffuse Texture Input, inputType=FunctionInput_Texture2D, sortPriority=-10}) {name=Diffuse TextureObject, nodeColor=<rgba 1 0.9427 0.9427 1>}
  NamedRerouteDeclaration(None=ComponentMask(None=@n22) {g=true, r=true}) {name=UV_XY, nodeColor=<rgba 0 0 0.4 1>}
  NamedRerouteDeclaration(None=FunctionInput(Preview=StaticBool()) {bUsePreviewValueAsDefault=true, inputName=Use Realtime Displacement, inputType=FunctionInput_StaticBool, sortPriority=-25}) {name="Use Realtime Displacement(SB)", nodeColor=<rgba 0.1979 0.0084 0 1>}
  NamedRerouteDeclaration(None=FunctionInput(Preview=StaticBool()) {bUsePreviewValueAsDefault=true, inputName=Use Expensive Triplanar, inputType=FunctionInput_StaticBool, sortPriority=-20}) {name="Use Costly Triplanar Override(SB)", nodeColor=<rgba 0.1979 0.0084 0 1>}
  NamedRerouteDeclaration(None=TextureSample(UVs=@n4, Tex=NamedRerouteUsage()).RGBA) {name=Triplanar Flat - Sample, nodeColor=<rgba 0.2957 0.2957 1 1>}
  n45 = NamedRerouteUsage()
  n46 = NamedRerouteUsage()
  n47 = Add(A=Add(A=Multiply(A=@n45, B=@n17), B=Multiply(A=@n46, B=@n18)), B=Multiply(A=NamedRerouteUsage(), B=@n19))
  n48 = FunctionInput() {bUsePreviewValueAsDefault=true, inputName=Derivatives Scale, inputType=FunctionInput_Scalar, sortPriority=15, previewValue=<1 1 1 1>}
  n49 = Multiply(A=TextureCoordinate(), B=@n48)
  n50 = DDX(Value=@n49)
  n51 = DDY(Value=@n49)
  NamedRerouteDeclaration(None=TextureSample!(
    UVs=FunctionOutput(None=@n47) {outputName=Triplanar Dithered - UVs, sortPriority=20}
    Tex=NamedRerouteUsage()
    "DDX(UVs)"=FunctionOutput(None=@n50) {outputName=Triplanar Dithered - DDX, sortPriority=25}
    "DDY(UVs)"=FunctionOutput(None=@n51) {outputName=Triplanar Dithered - DDY, sortPriority=26}
    Apply View MipBias
   ) {mipValueMode=TMVM_Derivative}.RGBA) {name=Triplanar Dithered - Sample, nodeColor=<rgba 0.2957 0.2957 1 1>}
  n52 = ComponentMask(None=@n3) {b=true}
  n53 = ComponentMask(None=@n3) {g=true, r=true}
  n54 = Subtract(A=@n52, B=@n53)
  n55 = Saturate(None=Min(A=ComponentMask(None=@n54) {r=true}, B=ComponentMask(None=@n54) {g=true}))
  n56 = Saturate(None=Subtract(A=@n53, B=@n52))
  n57 = Abs(None=Subtract(A=ComponentMask(None=@n56) {r=true}, B=ComponentMask(None=@n56) {g=true}))
  Power(Base=OneMinus(None=Max(A=@n55, B=@n57)), Exp=Constant() {desc=Triplanar - Full - Seam Sharpness, r=64})
 ]}
