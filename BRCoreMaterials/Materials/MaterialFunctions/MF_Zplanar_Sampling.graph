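# Material function that projects diffuse, specular mask and normal textures along world Z, with Helios cliff strata textures as previews.
# World position XY divided by a scale input gives shared UVs for all three samples.
# Diffuse RGB and alpha set BaseColor and Opacity, the mask's red channel drives Specular, Metallic and Roughness alike, and normals pass straight through.
# It outputs material attributes plus raw samples, labelled "Biplanar" despite the single projection.
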MaterialFunction /BRCoreMaterials/Materials/MaterialFunctions/MF_Zplanar_Sampling
graph {nodes [
  n1 = Constant() {r=100}
  n2 = ComponentMask(None=Divide(
    A=WorldPosition().XYZ
    B=FunctionInput(Preview=@n1) {bUsePreviewValueAsDefault=true, inputName=Zplanar - Scale, inputType=FunctionInput_Scalar, sortPriority=-10}
   )) {g=true, r=true}
  n3 = TextureSample(
   UVs=@n2
   Tex=FunctionInput(Preview=TextureObject() {texture=/Game/Environments/Helios/Terrain/Textures/T_Helios_RustBucket_Cliff_Strata_01_D}) {inputName=T_Diffuse, inputType=FunctionInput_Texture2D, sortPriority=4}
  )
  n4 = TextureSample(
   UVs=@n2
   Tex=FunctionInput(Preview=TextureObject() {samplerType=SAMPLERTYPE_Masks, texture=/Game/Environments/Helios/Terrain/Textures/T_Helios_RustBucket_Cliff_Strata_01_S}) {inputName=T_SpecularMasks, inputType=FunctionInput_Texture2D, sortPriority=5}
  )
  n5 = TextureSample(
   UVs=@n2
   Tex=FunctionInput(Preview=TextureObject() {samplerType=SAMPLERTYPE_Normal, texture=/Game/Environments/Helios/Terrain/Textures/T_Helios_RustBucket_Cliff_Strata_01_N}) {inputName=T_Normals, inputType=FunctionInput_Texture2D, sortPriority=6}
  )
  n6 = SetMaterialAttributes!(
   MaterialAttributes
   BaseColor=ComponentMask(None=@n3.RGBA) {b=true, g=true, r=true}
   Opacity=ComponentMask(None=@n3.RGBA) {a=true}
   Specular=@n4.R
   Metallic=@n4.R
   Roughness=@n4.R
   Normal=@n5.RGB
  ) {attributeSetTypes [
    #1
    #2
    #3
    #4
    #5
    #6
   ]}
  FunctionOutput(None=@n6) {outputName=SetMaterialAttributes, sortPriority=-10}
  FunctionOutput(None=@n3.RGBA) {outputName=Biplanar - Diffuse, sortPriority=5}
  FunctionOutput(None=@n4.RGB) {outputName=Biplanar - SpecularMasks, sortPriority=6}
  FunctionOutput(None=@n5.RGB) {outputName=Biplanar - Normals, sortPriority=7}
 ]}
