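# MF_Rocks_Layer builds one configurable rock material layer, outputting material attributes, a tintable mask, and baked AO, curvature and painted masks.
# Diffuse, SpecularMasks and Normal textures are sampled through static-switched UV, wavy-distorted dithered triplanar, biplanar or Z-planar projections.
# Static inputs toggle detail, fuzzy Fresnel shading and displacement contribution, including an optional costlier triplanar.
# Quality switches and Nanite replace nodes vary behavior per platform.
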MaterialFunction /BRCoreMaterials/Materials/MaterialFunctions/MF_Rocks_Layer
graph {expressions {
  MaterialExpressionAdd_0 {
   class=MaterialExpressionAdd
   inputs [
    A {from=MaterialExpressionNamedRerouteUsage_27}
    B {from=""}
   ]
  }
  MaterialExpressionAdd_1 {
   class=MaterialExpressionAdd
   inputs [
    A {from=MaterialExpressionGetMaterialAttributes_1, output=Roughness}
    B {from=MaterialExpressionNamedRerouteUsage_84}
   ]
  }
  MaterialExpressionAdd_2 {
   class=MaterialExpressionAdd
   inputs [
    A {from=MaterialExpressionReroute_93}
    B {from=MaterialExpressionReroute_94}
   ]
  }
  MaterialExpressionAppendVector_0 {
   class=MaterialExpressionAppendVector
   inputs [
    A {from=MaterialExpressionMultiply_1, output=""}
    B {from=MaterialExpressionComponentMask_3, output=""}
   ]
  }
  MaterialExpressionAppendVector_1 {
   class=MaterialExpressionAppendVector
   inputs [
    A {from=MaterialExpressionVectorParameter_2, output=""}
    B {from=MaterialExpressionVectorParameter_2, output=""}
   ]
  }
  MaterialExpressionAppendVector_4 {
   class=MaterialExpressionAppendVector
   inputs [
    A {from=MaterialExpressionVectorParameter_11, output=""}
    B {from=MaterialExpressionVectorParameter_11, output=""}
   ]
  }
  MaterialExpressionComponentMask_0 {class=MaterialExpressionComponentMask, inputs [None {from=MaterialExpressionReroute_38}], properties {a=true}}
  MaterialExpressionComponentMask_1 {
   class=MaterialExpressionComponentMask
   inputs [None {from=MaterialExpressionGetMaterialAttributes_11, output=BaseColor}]
   properties {r=true}
  }
  MaterialExpressionComponentMask_10 {class=MaterialExpressionComponentMask, inputs [None {from=MaterialExpressionNamedRerouteUsage_132}], properties {b=true, g=true, r=true}}
  MaterialExpressionComponentMask_11 {class=MaterialExpressionComponentMask, inputs [None {from=MaterialExpressionNamedRerouteUsage_103}], properties {a=true}}
  MaterialExpressionComponentMask_12 {class=MaterialExpressionComponentMask, inputs [None {from=MaterialExpressionReroute_38}], properties {b=true, g=true, r=true}}
  MaterialExpressionComponentMask_2 {class=MaterialExpressionComponentMask, inputs [None {from=MaterialExpressionReroute_10}], properties {g=true, r=true}}
  MaterialExpressionComponentMask_3 {class=MaterialExpressionComponentMask, inputs [None {from=MaterialExpressionReroute_10}], properties {b=true}}
  MaterialExpressionComponentMask_4 {class=MaterialExpressionComponentMask, inputs [None {from=MaterialExpressionNamedRerouteUsage_30}], properties {a=true}}
  MaterialExpressionComponentMask_5 {class=MaterialExpressionComponentMask, inputs [None {from=MaterialExpressionNaniteReplace_1}], properties {g=true}}
  MaterialExpressionComponentMask_6 {class=MaterialExpressionComponentMask, inputs [None {from=MaterialExpressionNaniteReplace_1}], properties {r=true}}
  MaterialExpressionComponentMask_7 {class=MaterialExpressionComponentMask, inputs [None {from=MaterialExpressionNaniteReplace_1}], properties {b=true}}
  MaterialExpressionComponentMask_8 {class=MaterialExpressionComponentMask, inputs [None {from=MaterialExpressionNamedRerouteUsage_115}], properties {a=true}}
  MaterialExpressionComponentMask_9 {class=MaterialExpressionComponentMask, inputs [None {from=MaterialExpressionNamedRerouteUsage_132}], properties {a=true}}
  MaterialExpressionConstant3Vector_0 {class=MaterialExpressionConstant3Vector, properties {constant=<rgba 1 0.5 0 0>}}
  MaterialExpressionConstant3Vector_2 {class=MaterialExpressionConstant3Vector}
  MaterialExpressionConstant4Vector_0 {class=MaterialExpressionConstant4Vector, inputs [], properties {}}
  MaterialExpressionConstant4Vector_1 {class=MaterialExpressionConstant4Vector, inputs [], properties {constant=<rgba 0 0 0 0.5>}}
  MaterialExpressionConstant_0 {class=MaterialExpressionConstant, properties {r=700}}
  MaterialExpressionConstant_1 {class=MaterialExpressionConstant, properties {r=3}}
  MaterialExpressionConstant_10 {class=MaterialExpressionConstant, properties {r=1}}
  MaterialExpressionConstant_11 {class=MaterialExpressionConstant, properties {r=1}}
  MaterialExpressionConstant_12 {class=MaterialExpressionConstant, properties {r=1}}
  MaterialExpressionConstant_15 {class=MaterialExpressionConstant, properties {r=10}}
  MaterialExpressionConstant_16 {class=MaterialExpressionConstant, properties {desc="No Displacmenet, Neutral Height", r=0.5}}
  MaterialExpressionConstant_17 {class=MaterialExpressionConstant, properties {r=1}}
  MaterialExpressionConstant_18 {class=MaterialExpressionConstant, properties {r=1}}
  MaterialExpressionConstant_20 {class=MaterialExpressionConstant, properties {r=1}}
  MaterialExpressionConstant_21 {class=MaterialExpressionConstant, properties {r=1}}
  MaterialExpressionConstant_3 {class=MaterialExpressionConstant, properties {r=700}}
  MaterialExpressionConstant_4 {class=MaterialExpressionConstant, properties {r=10}}
  MaterialExpressionConstant_5 {class=MaterialExpressionConstant, properties {r=75}}
  MaterialExpressionConstant_6 {class=MaterialExpressionConstant, properties {r=1}}
  MaterialExpressionConstant_7 {class=MaterialExpressionConstant, properties {r=700}}
  MaterialExpressionConstant_8 {class=MaterialExpressionConstant, properties {r=700}}
  MaterialExpressionConstant_9 {class=MaterialExpressionConstant}
  MaterialExpressionConvert_0 {
   class=MaterialExpressionConvert
   inputs [None {from=MaterialExpressionStaticSwitch_12}]
   properties {
    nodeName=Break Float3
    convertInputs [{
      type=Vector3
      expressionInput {
       expression=&self:MaterialExpressionStaticSwitch_12
       inputName=None
       mask=0
       maskA=0
       maskB=0
       maskG=0
       maskR=0
       outputIndex=0
      }
     }]
    convertOutputs [{type=Scalar}, {type=Scalar}, {type=Scalar}]
   }
  }
  MaterialExpressionConvert_1 {
   class=MaterialExpressionConvert
   inputs [
    None {from=MaterialExpressionNamedRerouteUsage_19}
    None {from=MaterialExpressionNamedRerouteUsage_18}
    None {from=MaterialExpressionNamedRerouteUsage_20}
   ]
   properties {
    nodeName=Make Vector3
    convertInputs [
     {
      type=Scalar
      expressionInput {
       expression=&self:MaterialExpressionNamedRerouteUsage_19
       inputName=None
       mask=0
       maskA=0
       maskB=0
       maskG=0
       maskR=0
       outputIndex=0
      }
     }
     {
      type=Scalar
      expressionInput {
       expression=&self:MaterialExpressionNamedRerouteUsage_18
       inputName=None
       mask=0
       maskA=0
       maskB=0
       maskG=0
       maskR=0
       outputIndex=0
      }
     }
     {
      type=Scalar
      expressionInput {
       expression=&self:MaterialExpressionNamedRerouteUsage_20
       inputName=None
       mask=0
       maskA=0
       maskB=0
       maskG=0
       maskR=0
       outputIndex=0
      }
     }
    ]
    convertOutputs [{type=Vector3}]
   }
  }
  MaterialExpressionFresnel_0 {
   class=MaterialExpressionFresnel
   inputs [
    ExponentIn {from=MaterialExpressionNamedRerouteUsage_127, output=""}
    BaseReflectFractionIn {from="", output=""}
    Normal {from="", output=""}
   ]
   properties {baseReflectFraction=0}
  }
  MaterialExpressionFunctionInput_0 {
   class=MaterialExpressionFunctionInput
   inputs [Preview {from=MaterialExpressionConstant4Vector_0, output=RGBA}]
   properties {inputName=Triplanar - Offset, inputType=FunctionInput_Vector4, sortPriority=53}
  }
  MaterialExpressionFunctionInput_1 {
   class=MaterialExpressionFunctionInput
   inputs [Preview {from=MaterialExpressionTextureObject_0, output=""}]
   properties {bUsePreviewValueAsDefault=true, inputName=Diffuse Texture Input, inputType=FunctionInput_Texture2D, sortPriority=320}
  }
  MaterialExpressionFunctionInput_10 {
   class=MaterialExpressionFunctionInput
   inputs [Preview {from=MaterialExpressionStaticBool_10, output=""}]
   properties {bUsePreviewValueAsDefault=true, inputName=WS - Use Wavy Distortion, inputType=FunctionInput_StaticBool, sortPriority=50}
  }
  MaterialExpressionFunctionInput_11 {
   class=MaterialExpressionFunctionInput
   inputs [Preview {from=MaterialExpressionTextureObject_1, output=""}]
   properties {inputName=SpecularMasks Texture Input, inputType=FunctionInput_Texture2D, sortPriority=370}
  }
  MaterialExpressionFunctionInput_12 {
   class=MaterialExpressionFunctionInput
   inputs [Preview {from=MaterialExpressionTextureObject_2, output=""}]
   properties {inputName=Normal Texture Input, inputType=FunctionInput_Texture2D, sortPriority=410}
  }
  MaterialExpressionFunctionInput_13 {
   class=MaterialExpressionFunctionInput
   inputs [Preview {from=MaterialExpressionConstant_4, output=""}]
   properties {bUsePreviewValueAsDefault=true, inputName=Triplanar - Fade, inputType=FunctionInput_Scalar, sortPriority=56}
  }
  MaterialExpressionFunctionInput_14 {
   class=MaterialExpressionFunctionInput
   inputs [Preview {from=MaterialExpressionConstant_5, output=""}]
   properties {bUsePreviewValueAsDefault=true, inputName="Triplanar - Fade (Low/Mobile)", inputType=FunctionInput_Scalar, sortPriority=57}
  }
  MaterialExpressionFunctionInput_15 {
   class=MaterialExpressionFunctionInput
   inputs [Preview {from=MaterialExpressionConstant_9, output=""}]
   properties {bUsePreviewValueAsDefault=true, inputName="Roughness - Bias (Mobile)", inputType=FunctionInput_Scalar, sortPriority=391}
  }
  MaterialExpressionFunctionInput_16 {
   class=MaterialExpressionFunctionInput
   inputs [Preview {from=MaterialExpressionConstant_18, output=""}]
   properties {bUsePreviewValueAsDefault=true, inputName=Is Tintable Layer, inputType=FunctionInput_Scalar}
  }
  MaterialExpressionFunctionInput_17 {
   class=MaterialExpressionFunctionInput
   inputs [Preview {from=MaterialExpressionStaticBool_7, output=""}]
   properties {bUsePreviewValueAsDefault=true, inputName=Layer Contributes to Displacement, inputType=FunctionInput_StaticBool, sortPriority=2}
  }
  MaterialExpressionFunctionInput_18 {
   class=MaterialExpressionFunctionInput
   inputs [Preview {from=MaterialExpressionConstant_3, output=""}]
   properties {bUsePreviewValueAsDefault=true, inputName=Triplanar - Scale, inputType=FunctionInput_Scalar, sortPriority=55}
  }
  MaterialExpressionFunctionInput_19 {
   class=MaterialExpressionFunctionInput
   inputs [Preview {from="", output=""}]
   properties {inputName="Normals - Value (non-Nanite)", inputType=FunctionInput_Scalar, sortPriority=416}
  }
  MaterialExpressionFunctionInput_2 {
   class=MaterialExpressionFunctionInput
   inputs [Preview {from="", output=""}]
   properties {bUsePreviewValueAsDefault=true, inputName=WS - Wavy Distortion - Offset, sortPriority=51}
  }
  MaterialExpressionFunctionInput_20 {
   class=MaterialExpressionFunctionInput
   inputs [Preview {from="", output=""}]
   properties {inputName="Normals - Value (Nanite)", inputType=FunctionInput_Scalar, sortPriority=415}
  }
  MaterialExpressionFunctionInput_21 {class=MaterialExpressionFunctionInput, inputs [Preview {from="", output=""}], properties {inputName=Fuzzy Shaded Mask, sortPriority=1006}}
  MaterialExpressionFunctionInput_22 {
   class=MaterialExpressionFunctionInput
   inputs [Preview {from=MaterialExpressionConstant3Vector_2, output=""}]
   properties {inputName=Biplanar - Position Offset, sortPriority=62}
  }
  MaterialExpressionFunctionInput_23 {
   class=MaterialExpressionFunctionInput
   inputs [Preview {from=MaterialExpressionConstant_6, output=""}]
   properties {inputName=UV - Tiling, inputType=FunctionInput_Scalar, sortPriority=81, previewValue=<1 0 0 1>}
  }
  MaterialExpressionFunctionInput_24 {
   class=MaterialExpressionFunctionInput
   inputs [Preview {from=MaterialExpressionConstant_8, output=""}]
   properties {inputName=Biplanar - Fade Contrast, inputType=FunctionInput_Scalar, sortPriority=63}
  }
  MaterialExpressionFunctionInput_25 {
   class=MaterialExpressionFunctionInput
   inputs [Preview {from=MaterialExpressionStaticBool_5, output=""}]
   properties {bUsePreviewValueAsDefault=true, inputName=Is Fuzzy Layer, inputType=FunctionInput_StaticBool, sortPriority=1}
  }
  MaterialExpressionFunctionInput_27 {
   class=MaterialExpressionFunctionInput
   inputs [Preview {from=MaterialExpressionStaticBool_6, output=""}]
   properties {bUsePreviewValueAsDefault=true, inputName=Is Detail Layer, inputType=FunctionInput_StaticBool, sortPriority=1}
  }
  MaterialExpressionFunctionInput_28 {
   class=MaterialExpressionFunctionInput
   inputs [Preview {from=MaterialExpressionStaticBool_13, output=""}]
   properties {inputName=Use SpecularMasks as AO/Curvature/Painted MSK, inputType=FunctionInput_StaticBool, sortPriority=3}
  }
  MaterialExpressionFunctionInput_29 {
   class=MaterialExpressionFunctionInput
   inputs [Preview {from=MaterialExpressionStaticBool_8, output=""}]
   properties {inputName=Use Second UV Channel, inputType=FunctionInput_StaticBool, sortPriority=80}
  }
  MaterialExpressionFunctionInput_3 {
   class=MaterialExpressionFunctionInput
   inputs [Preview {from=MaterialExpressionStaticBool_14, output=""}]
   properties {bUsePreviewValueAsDefault=true, inputName=Layer Is Used, inputType=FunctionInput_StaticBool, sortPriority=-1}
  }
  MaterialExpressionFunctionInput_30 {
   class=MaterialExpressionFunctionInput
   inputs [Preview {from="", output=""}]
   properties {inputName=Height - Min, inputType=FunctionInput_Scalar, sortPriority=331}
  }
  MaterialExpressionFunctionInput_31 {
   class=MaterialExpressionFunctionInput
   inputs [Preview {from="", output=""}]
   properties {inputName=Height - Max, inputType=FunctionInput_Scalar, sortPriority=332}
  }
  MaterialExpressionFunctionInput_32 {
   class=MaterialExpressionFunctionInput
   inputs [Preview {from=MaterialExpressionStaticBool_0, output=""}]
   properties {bUsePreviewValueAsDefault=true, inputName=Use WS, inputType=FunctionInput_StaticBool, sortPriority=35}
  }
  MaterialExpressionFunctionInput_33 {
   class=MaterialExpressionFunctionInput
   inputs [Preview {from="", output=""}]
   properties {inputName=Specular, inputType=FunctionInput_Scalar, sortPriority=380}
  }
  MaterialExpressionFunctionInput_34 {
   class=MaterialExpressionFunctionInput
   inputs [Preview {from="", output=""}]
   properties {inputName=Metallic, inputType=FunctionInput_Scalar, sortPriority=385}
  }
  MaterialExpressionFunctionInput_35 {
   class=MaterialExpressionFunctionInput
   inputs [Preview {from="", output=""}]
   properties {inputName=Roughness, inputType=FunctionInput_Scalar, sortPriority=390}
  }
  MaterialExpressionFunctionInput_36 {
   class=MaterialExpressionFunctionInput
   inputs [Preview {from=MaterialExpressionConstant_7, output=""}]
   properties {inputName=Zplanar - Scale, inputType=FunctionInput_Scalar, sortPriority=71}
  }
  MaterialExpressionFunctionInput_37 {
   class=MaterialExpressionFunctionInput
   inputs [Preview {from=MaterialExpressionStaticBool_9, output=""}]
   properties {bUsePreviewValueAsDefault=true, inputName=Use Zplanar WS, inputType=FunctionInput_StaticBool, sortPriority=70}
  }
  MaterialExpressionFunctionInput_38 {
   class=MaterialExpressionFunctionInput
   inputs [Preview {from=MaterialExpressionStaticBool_11, output=""}]
   properties {bUsePreviewValueAsDefault=true, inputName=Use DEBUG UVs, inputType=FunctionInput_StaticBool, sortPriority=30}
  }
  MaterialExpressionFunctionInput_39 {
   class=MaterialExpressionFunctionInput
   inputs [Preview {from=MaterialExpressionConstant_10, output=""}]
   properties {inputName=Detail - Color - Intensity, inputType=FunctionInput_Scalar, sortPriority=340}
  }
  MaterialExpressionFunctionInput_4 {
   class=MaterialExpressionFunctionInput
   inputs [Preview {from=MaterialExpressionConstant_0, output=""}]
   properties {inputName=Biplanar - Scale, inputType=FunctionInput_Scalar, sortPriority=61}
  }
  MaterialExpressionFunctionInput_40 {
   class=MaterialExpressionFunctionInput
   inputs [Preview {from=MaterialExpressionConstant_11, output=""}]
   properties {inputName=Detail - Specular - Intensity, inputType=FunctionInput_Scalar, sortPriority=341}
  }
  MaterialExpressionFunctionInput_41 {
   class=MaterialExpressionFunctionInput
   inputs [Preview {from=MaterialExpressionConstant_17, output=""}]
   properties {bUsePreviewValueAsDefault=true, inputName=Triplanar - Derivatives Scale, inputType=FunctionInput_Scalar, sortPriority=54}
  }
  MaterialExpressionFunctionInput_42 {
   class=MaterialExpressionFunctionInput
   inputs [Preview {from=MaterialExpressionStaticBool_12, output=""}]
   properties {bUsePreviewValueAsDefault=true, inputName=Use VTX AO/Curvature/Painted on Nanite, inputType=FunctionInput_StaticBool, sortPriority=4}
  }
  MaterialExpressionFunctionInput_43 {
   class=MaterialExpressionFunctionInput
   inputs [Preview {from="", output=""}]
   properties {inputName=Color - Multiply, sortPriority=321, previewValue=<1 1 1 1>}
  }
  MaterialExpressionFunctionInput_44 {
   class=MaterialExpressionFunctionInput
   inputs [Preview {from="", output=""}]
   properties {inputName=Realtime Displacment - Min, inputType=FunctionInput_Scalar, sortPriority=715}
  }
  MaterialExpressionFunctionInput_45 {
   class=MaterialExpressionFunctionInput
   inputs [Preview {from="", output=""}]
   properties {inputName=Realtime Displacment - Max, inputType=FunctionInput_Scalar, sortPriority=716}
  }
  MaterialExpressionFunctionInput_46 {
   class=MaterialExpressionFunctionInput
   inputs [Preview {from="", output=""}]
   properties {inputName=Realtime Displacment - Triplanar Seam, inputType=FunctionInput_Scalar, sortPriority=720}
  }
  MaterialExpressionFunctionInput_47 {
   class=MaterialExpressionFunctionInput
   inputs [Preview {from=MaterialExpressionConstant_1, output=""}]
   properties {bUsePreviewValueAsDefault=true, inputName=Fuzzy - Exponent, inputType=FunctionInput_Scalar, sortPriority=455}
  }
  MaterialExpressionFunctionInput_48 {
   class=MaterialExpressionFunctionInput
   inputs [Preview {from="", output=""}]
   properties {inputName="\\ --- Dependencies --- /", inputType=FunctionInput_Scalar, sortPriority=1000}
  }
  MaterialExpressionFunctionInput_49 {
   class=MaterialExpressionFunctionInput
   inputs [Preview {from="", output=""}]
   properties {inputName=Wavy Distortion World Position, sortPriority=1005}
  }
  MaterialExpressionFunctionInput_5 {
   class=MaterialExpressionFunctionInput
   inputs [Preview {from=MaterialExpressionStaticBool_1, output=""}]
   properties {bUsePreviewValueAsDefault=true, inputName=Use Biplanar WS, inputType=FunctionInput_StaticBool, sortPriority=60}
  }
  MaterialExpressionFunctionInput_50 {
   class=MaterialExpressionFunctionInput
   inputs [Preview {from=MaterialExpressionConstant_20, output=""}]
   properties {bUsePreviewValueAsDefault=true, inputName=Fuzzy - Intensity, inputType=FunctionInput_Scalar, sortPriority=450}
  }
  MaterialExpressionFunctionInput_51 {
   class=MaterialExpressionFunctionInput
   inputs [Preview {from=MaterialExpressionConstant_21, output=""}]
   properties {bUsePreviewValueAsDefault=true, inputName=Fuzzy - Roughness, inputType=FunctionInput_Scalar, sortPriority=465}
  }
  MaterialExpressionFunctionInput_53 {
   class=MaterialExpressionFunctionInput
   inputs [Preview {from=MaterialExpressionConstant4Vector_1, output=RGBA}]
   properties {bUsePreviewValueAsDefault=true, inputName=Fuzzy - Color, inputType=FunctionInput_Vector4, sortPriority=460, previewValue=<1 1 1 1>}
  }
  MaterialExpressionFunctionInput_6 {
   class=MaterialExpressionFunctionInput
   inputs [Preview {from=MaterialExpressionConstant_12, output=""}]
   properties {inputName=Detail - Specular - Contrast, inputType=FunctionInput_Scalar, sortPriority=342}
  }
  MaterialExpressionFunctionInput_7 {
   class=MaterialExpressionFunctionInput
   inputs [Preview {from="", output=""}]
   properties {inputName=Detail - Color - Multiply, sortPriority=322, previewValue=<1 1 1 1>}
  }
  MaterialExpressionFunctionInput_8 {
   class=MaterialExpressionFunctionInput
   inputs [Preview {from=MaterialExpressionConstant_15, output=""}]
   properties {bUsePreviewValueAsDefault=true, inputName=Realtime Displacement - Flat - Seam Sharpness, inputType=FunctionInput_Scalar, sortPriority=705}
  }
  MaterialExpressionFunctionInput_9 {
   class=MaterialExpressionFunctionInput
   inputs [Preview {from=MaterialExpressionStaticBool_15, output=""}]
   properties {bUsePreviewValueAsDefault=true, inputName=Realtime Displacement - Use Costly Triplanar, inputType=FunctionInput_StaticBool, sortPriority=700}
  }
  MaterialExpressionFunctionOutput_0 {
   class=MaterialExpressionFunctionOutput
   inputs [None {from=MaterialExpressionNamedRerouteUsage_64, output=""}]
   properties {outputName=Tintable Mask, sortPriority=20}
  }
  MaterialExpressionFunctionOutput_1 {
   class=MaterialExpressionFunctionOutput
   inputs [None {from=MaterialExpressionReroute_27, output=""}]
   properties {outputName=MaterialAttributes, sortPriority=-10}
  }
  MaterialExpressionFunctionOutput_2 {
   class=MaterialExpressionFunctionOutput
   inputs [None {from=MaterialExpressionNamedRerouteUsage_73, output=""}]
   properties {outputName=Baked Curvature, sortPriority=50}
  }
  MaterialExpressionFunctionOutput_3 {
   class=MaterialExpressionFunctionOutput
   inputs [None {from=MaterialExpressionNamedRerouteUsage_74, output=""}]
   properties {outputName=Baked AO, sortPriority=51}
  }
  MaterialExpressionFunctionOutput_4 {
   class=MaterialExpressionFunctionOutput
   inputs [None {from=MaterialExpressionNamedRerouteUsage_75, output=""}]
   properties {outputName=Painted Mask, sortPriority=52}
  }
  MaterialExpressionGetMaterialAttributes_0 {
   class=MaterialExpressionGetMaterialAttributes
   inputs [None {from=MaterialExpressionReroute_14, output=""}]
   properties {attributeGetTypes [#1]}
  }
  MaterialExpressionGetMaterialAttributes_1 {
   class=MaterialExpressionGetMaterialAttributes
   inputs [None {from=MaterialExpressionReroute_14, output=""}]
   properties {attributeGetTypes [#2]}
  }
  MaterialExpressionGetMaterialAttributes_11 {
   class=MaterialExpressionGetMaterialAttributes
   inputs [None {from=MaterialExpressionReroute_14, output=""}]
   properties {attributeGetTypes [#3]}
  }
  MaterialExpressionGetMaterialAttributes_2 {
   class=MaterialExpressionGetMaterialAttributes
   inputs [None {from=MaterialExpressionSetMaterialAttributes_2, output=""}]
   properties {attributeGetTypes [#4]}
  }
  MaterialExpressionGetMaterialAttributes_3 {
   class=MaterialExpressionGetMaterialAttributes
   inputs [None {from=MaterialExpressionReroute_85, output=""}]
   properties {attributeGetTypes [#3, #2, #1]}
  }
  MaterialExpressionGetMaterialAttributes_4 {
   class=MaterialExpressionGetMaterialAttributes
   inputs [None {from=MaterialExpressionReroute_15, output=""}]
   properties {attributeGetTypes [#3]}
  }
  MaterialExpressionGetMaterialAttributes_9 {
   class=MaterialExpressionGetMaterialAttributes
   inputs [None {from=MaterialExpressionReroute_14, output=""}]
   properties {attributeGetTypes [#5]}
  }
  MaterialExpressionLinearInterpolate_0 {
   class=MaterialExpressionLinearInterpolate
   inputs [
    A {from=MaterialExpressionSubtract_0}
    B {from=MaterialExpressionAdd_0}
    Alpha {from=MaterialExpressionReroute_44}
   ]
   properties {constA=-1}
  }
  MaterialExpressionLinearInterpolate_1 {
   class=MaterialExpressionLinearInterpolate
   inputs [
    A {from=MaterialExpressionNamedRerouteUsage_16}
    B {from=MaterialExpressionNamedRerouteUsage_15}
    Alpha {from=MaterialExpressionReroute_39}
   ]
  }
  MaterialExpressionLinearInterpolate_11 {
   class=MaterialExpressionLinearInterpolate
   inputs [
    A {from=""}
    B {from=MaterialExpressionSaturate_0}
    Alpha {from=MaterialExpressionNamedRerouteUsage_70}
   ]
   properties {constA=1}
  }
  MaterialExpressionLinearInterpolate_13 {
   class=MaterialExpressionLinearInterpolate
   inputs [
    A {from=""}
    B {from=MaterialExpressionMultiply_3}
    Alpha {from=MaterialExpressionNamedRerouteUsage_69}
   ]
   properties {constA=1}
  }
  MaterialExpressionLinearInterpolate_2 {
   class=MaterialExpressionLinearInterpolate
   inputs [
    A {from=MaterialExpressionNamedRerouteUsage_117}
    B {from=MaterialExpressionNamedRerouteUsage_118}
    Alpha {from=MaterialExpressionReroute_74}
   ]
  }
  MaterialExpressionLinearInterpolate_3 {
   class=MaterialExpressionLinearInterpolate
   inputs [
    A {from=MaterialExpressionLinearInterpolate_2}
    B {from=MaterialExpressionNamedRerouteUsage_119}
    Alpha {from=MaterialExpressionNamedRerouteUsage_114}
   ]
  }
  MaterialExpressionLinearInterpolate_4 {
   class=MaterialExpressionLinearInterpolate
   inputs [
    A {from=MaterialExpressionReroute_78}
    B {from=MaterialExpressionNamedRerouteUsage_130}
    Alpha {from=MaterialExpressionSaturate_2}
   ]
  }
  MaterialExpressionLinearInterpolate_5 {
   class=MaterialExpressionLinearInterpolate
   inputs [
    A {from=MaterialExpressionReroute_92}
    B {from=MaterialExpressionLinearInterpolate_6}
    Alpha {from=MaterialExpressionMultiply_6}
   ]
  }
  MaterialExpressionLinearInterpolate_6 {
   class=MaterialExpressionLinearInterpolate
   inputs [
    A {from=MaterialExpressionMultiply_7}
    B {from=MaterialExpressionAdd_2}
    Alpha {from=MaterialExpressionReroute_91}
   ]
  }
  MaterialExpressionMaterialFunctionCall_0 {
   class=MaterialExpressionMaterialFunctionCall
   inputs [
    WorldPosition {from=MaterialExpressionStaticSwitch_3}
    Biplanar - Position Offset {from=MaterialExpressionNamedRerouteUsage_47}
    Biplanar - Scale {from=MaterialExpressionNamedRerouteUsage_56}
    Biplanar - Fade Contrast {from=MaterialExpressionNamedRerouteUsage_58}
    T_Diffuse {from=MaterialExpressionReroute_4}
    T_SpecularMasks {from=MaterialExpressionReroute_5}
    T_Normals {from=MaterialExpressionReroute_6}
   ]
   properties {materialFunction=&/MF_Biplanar_Sampling}
  }
  MaterialExpressionMaterialFunctionCall_12 {
   class=MaterialExpressionMaterialFunctionCall
   inputs [
    Tangent Normals {from=MaterialExpressionTextureSample_4, output=RGB}
    Vertex Normals {from=""}
    Triplanar Mask {from=MaterialExpressionReroute_32}
   ]
   properties {materialFunction=/Game/Athena/Environments/MaterialFunctions/MF_TriplanarNormalTangentToWorld}
  }
  MaterialExpressionMaterialFunctionCall_2 {
   class=MaterialExpressionMaterialFunctionCall
   inputs [
    UVs {from=MaterialExpressionReroute_95}
    Normal {from=MaterialExpressionTextureSample_9, output=RGB}
   ]
   properties {materialFunction=/Engine/Functions/Engine_MaterialFunctions01/Texturing/DeriveTangentBasis}
  }
  MaterialExpressionMaterialFunctionCall_26 {
   class=MaterialExpressionMaterialFunctionCall
   inputs [
    Absolute World Position {from=MaterialExpressionReroute_76}
    "Normal (WS)" {from=""}
    Texture Scale {from=MaterialExpressionNamedRerouteUsage_9}
    Triplanar Fade {from=MaterialExpressionMaterialFunctionCall_36, output=Result}
    Derivatives Scale {from=MaterialExpressionNamedRerouteUsage_109}
   ]
   properties {materialFunction=&/MF_TriplanarDithered_WavyDistortion}
  }
  MaterialExpressionMaterialFunctionCall_3 {
   class=MaterialExpressionMaterialFunctionCall
   inputs [
    Zplanar - Scale {from=MaterialExpressionNamedRerouteUsage_40}
    T_Diffuse {from=MaterialExpressionReroute_25}
    T_SpecularMasks {from=MaterialExpressionReroute_26}
    T_Normals {from=MaterialExpressionReroute_28}
   ]
   properties {materialFunction=&/MF_Zplanar_Sampling}
  }
  MaterialExpressionMaterialFunctionCall_36 {
   class=MaterialExpressionMaterialFunctionCall
   inputs [
    Low {from=MaterialExpressionReroute_110}
    High {from=MaterialExpressionReroute_109}
    Epic {from=MaterialExpressionReroute_109}
    -------- {from=MaterialExpressionReroute_109}
    Mobile {from=MaterialExpressionReroute_110}
    Medium {from=MaterialExpressionReroute_109}
   ]
   properties {materialFunction=/FortniteBaseMaterials/Functions/MF_QualitySwitch_Scalar}
  }
  MaterialExpressionMaterialFunctionCall_4 {
   class=MaterialExpressionMaterialFunctionCall
   inputs [
    Use Realtime Displacement {from=MaterialExpressionNamedRerouteUsage_63}
    Use Expensive Triplanar {from=MaterialExpressionNamedRerouteUsage_107}
    Diffuse Texture Input {from=MaterialExpressionNamedRerouteUsage_111}
    Texture Scale {from=MaterialExpressionNamedRerouteUsage_12}
    Absolute World Position {from=MaterialExpressionReroute_69}
    "VertexNormal (WS)" {from=""}
    Triplanar - Fade {from=MaterialExpressionMaterialFunctionCall_36, output=Result}
    Derivatives Scale {from=MaterialExpressionNamedRerouteUsage_110}
    Triplanar - Flat - Seam Sharpness {from=MaterialExpressionNamedRerouteUsage_112}
   ]
   properties {
    desc=A Triplanar with options just for getting around the screen space issues
    materialFunction=&/MF_Rocks_Triplanar_Displacement
   }
  }
  MaterialExpressionMultiply_0 {
   class=MaterialExpressionMultiply
   inputs [
    A {from=MaterialExpressionSaturate_2}
    B {from=MaterialExpressionNamedRerouteUsage_21}
   ]
  }
  MaterialExpressionMultiply_1 {
   class=MaterialExpressionMultiply
   inputs [
    A {from=MaterialExpressionComponentMask_2}
    B {from=MaterialExpressionNaniteReplace_2}
   ]
  }
  MaterialExpressionMultiply_2 {
   class=MaterialExpressionMultiply
   inputs [
    A {from=MaterialExpressionNamedRerouteUsage_24}
    B {from=MaterialExpressionStaticSwitch_11}
   ]
  }
  MaterialExpressionMultiply_27 {
   class=MaterialExpressionMultiply
   inputs [
    A {from=MaterialExpressionGetMaterialAttributes_4, output=BaseColor}
    B {from=MaterialExpressionNamedRerouteUsage_76}
   ]
  }
  MaterialExpressionMultiply_3 {
   class=MaterialExpressionMultiply
   inputs [
    A {from=MaterialExpressionReroute_42}
    B {from=MaterialExpressionNamedRerouteUsage_39}
   ]
   properties {constB=2}
  }
  MaterialExpressionMultiply_4 {
   class=MaterialExpressionMultiply
   inputs [
    A {from=MaterialExpressionMultiply_6}
    B {from=MaterialExpressionNamedRerouteUsage_21}
   ]
  }
  MaterialExpressionMultiply_6 {
   class=MaterialExpressionMultiply
   inputs [
    A {from=MaterialExpressionFresnel_0}
    B {from=MaterialExpressionNamedRerouteUsage_128}
   ]
  }
  MaterialExpressionMultiply_7 {
   class=MaterialExpressionMultiply
   inputs [
    A {from=MaterialExpressionReroute_93}
    B {from=MaterialExpressionReroute_94}
   ]
  }
  MaterialExpressionNamedRerouteDeclaration_0 {
   class=MaterialExpressionNamedRerouteDeclaration
   inputs [None {from=MaterialExpressionFunctionInput_3}]
   properties {name="Layer Is Used(SB)", nodeColor=<rgba 0.1979 0.0084 0 1>}
  }
  MaterialExpressionNamedRerouteDeclaration_1 {
   class=MaterialExpressionNamedRerouteDeclaration
   inputs [None {from=MaterialExpressionFunctionInput_21}]
   properties {name=_Fuzzy Shaded Mask, nodeColor=<rgba 0 0 0.5469 1>}
  }
  MaterialExpressionNamedRerouteDeclaration_10 {
   class=MaterialExpressionNamedRerouteDeclaration
   inputs [None {from=MaterialExpressionTextureSample_5, output=RGBA}]
   properties {name=Diffuse - UV Sample, nodeColor=<rgba 0.099 0.1719 0.0972 1>}
  }
  MaterialExpressionNamedRerouteDeclaration_11 {
   class=MaterialExpressionNamedRerouteDeclaration
   inputs [None {from=MaterialExpressionFunctionInput_7}]
   properties {name=Detail - Color - Multiply, nodeColor=<rgba 0.5043 1 0 1>}
  }
  MaterialExpressionNamedRerouteDeclaration_12 {
   class=MaterialExpressionNamedRerouteDeclaration
   inputs [None {from=MaterialExpressionFunctionInput_2}]
   properties {name=WS - Wavy Distortion - Offset, nodeColor=<rgba 0.5043 1 0 1>}
  }
  MaterialExpressionNamedRerouteDeclaration_13 {
   class=MaterialExpressionNamedRerouteDeclaration
   inputs [None {from=MaterialExpressionFunctionInput_24}]
   properties {name=Biplanar - Fade Contrast, nodeColor=<rgba 0.5043 1 0 1>}
  }
  MaterialExpressionNamedRerouteDeclaration_14 {
   class=MaterialExpressionNamedRerouteDeclaration
   inputs [None {from=MaterialExpressionFunctionInput_1}]
   properties {name=Diffuse TextureObject, nodeColor=<rgba 1 0.9427 0.9427 1>}
  }
  MaterialExpressionNamedRerouteDeclaration_15 {
   class=MaterialExpressionNamedRerouteDeclaration
   inputs [None {from=MaterialExpressionFunctionInput_11}]
   properties {name=SpecularMasks TextureObject, nodeColor=<rgba 1 0.9427 0.9427 1>}
  }
  MaterialExpressionNamedRerouteDeclaration_16 {
   class=MaterialExpressionNamedRerouteDeclaration
   inputs [None {from=MaterialExpressionFunctionInput_12}]
   properties {name=Normal TextureObject, nodeColor=<rgba 1 0.9427 0.9427 1>}
  }
  MaterialExpressionNamedRerouteDeclaration_17 {
   class=MaterialExpressionNamedRerouteDeclaration
   inputs [None {from=MaterialExpressionFunctionInput_10}]
   properties {name="WS - Use Wavy Distortion(SB)", nodeColor=<rgba 0.1979 0.0084 0 1>}
  }
  MaterialExpressionNamedRerouteDeclaration_18 {
   class=MaterialExpressionNamedRerouteDeclaration
   inputs [None {from=MaterialExpressionTextureSample_8, output=RGB}]
   properties {name=SpecularMasks - UV Sample, nodeColor=<rgba 0.099 0.1719 0.0972 1>}
  }
  MaterialExpressionNamedRerouteDeclaration_19 {
   class=MaterialExpressionNamedRerouteDeclaration
   inputs [None {from=MaterialExpressionFunctionInput_18}]
   properties {name=Triplanar - Scale, nodeColor=<rgba 0.5043 1 0 1>}
  }
  MaterialExpressionNamedRerouteDeclaration_2 {
   class=MaterialExpressionNamedRerouteDeclaration
   inputs [None {from=MaterialExpressionFunctionInput_4}]
   properties {name=Biplanar - Scale, nodeColor=<rgba 0.5043 1 0 1>}
  }
  MaterialExpressionNamedRerouteDeclaration_20 {
   class=MaterialExpressionNamedRerouteDeclaration
   inputs [None {from=MaterialExpressionFunctionInput_17}]
   properties {name="Layer Contributes to Displacement(SB)", nodeColor=<rgba 0.1979 0.0084 0 1>}
  }
  MaterialExpressionNamedRerouteDeclaration_21 {
   class=MaterialExpressionNamedRerouteDeclaration
   inputs [None {from=MaterialExpressionFunctionInput_9}]
   properties {name="Realtime Displacement - Use Costly Triplanar(SB)", nodeColor=<rgba 0.1979 0.0084 0 1>}
  }
  MaterialExpressionNamedRerouteDeclaration_22 {
   class=MaterialExpressionNamedRerouteDeclaration
   inputs [None {from=MaterialExpressionFunctionInput_13}]
   properties {name=Triplanar - Fade, nodeColor=<rgba 0.5043 1 0 1>}
  }
  MaterialExpressionNamedRerouteDeclaration_23 {
   class=MaterialExpressionNamedRerouteDeclaration
   inputs [None {from=MaterialExpressionFunctionInput_14}]
   properties {name="Triplanar - Fade (Low/Mobile)", nodeColor=<rgba 0.5043 1 0 1>}
  }
  MaterialExpressionNamedRerouteDeclaration_24 {
   class=MaterialExpressionNamedRerouteDeclaration
   inputs [None {from=MaterialExpressionFunctionInput_19}]
   properties {name="Normals - Value (non-Nanite)", nodeColor=<rgba 0.5043 1 0 1>}
  }
  MaterialExpressionNamedRerouteDeclaration_25 {
   class=MaterialExpressionNamedRerouteDeclaration
   inputs [None {from=MaterialExpressionTextureSample_1, output=RGB}]
   properties {name=Triplanar WS - DEBUG, nodeColor=<rgba 0.4072 0.4072 0 1>}
  }
  MaterialExpressionNamedRerouteDeclaration_26 {
   class=MaterialExpressionNamedRerouteDeclaration
   inputs [None {from=MaterialExpressionFunctionInput_20}]
   properties {name="Normals - Value (Nanite)", nodeColor=<rgba 0.5043 1 0 1>}
  }
  MaterialExpressionNamedRerouteDeclaration_27 {
   class=MaterialExpressionNamedRerouteDeclaration
   inputs [None {from=MaterialExpressionFunctionInput_15}]
   properties {name="Roughness - Bias (Mobile)", nodeColor=<rgba 0.5043 1 0 1>}
  }
  MaterialExpressionNamedRerouteDeclaration_28 {
   class=MaterialExpressionNamedRerouteDeclaration
   inputs [None {from=MaterialExpressionFunctionInput_28}]
   properties {name="Use SpecularMasks as AO/Curvature/Painted MSK(SB)", nodeColor=<rgba 0.1979 0.0084 0 1>}
  }
  MaterialExpressionNamedRerouteDeclaration_29 {
   class=MaterialExpressionNamedRerouteDeclaration
   inputs [None {from=MaterialExpressionTextureSample_10, output=RGB}]
   properties {name=UV - DEBUG, nodeColor=<rgba 0.4072 0.4072 0 1>}
  }
  MaterialExpressionNamedRerouteDeclaration_3 {
   class=MaterialExpressionNamedRerouteDeclaration
   inputs [None {from=MaterialExpressionFunctionInput_5}]
   properties {name=Biplanar - Use Biplanar WS, nodeColor=<rgba 0.1979 0.0084 0 1>}
  }
  MaterialExpressionNamedRerouteDeclaration_30 {
   class=MaterialExpressionNamedRerouteDeclaration
   inputs [None {from=MaterialExpressionFunctionInput_29}]
   properties {name="Use Second UV Channel(SB)", nodeColor=<rgba 0.1979 0.0084 0 1>}
  }
  MaterialExpressionNamedRerouteDeclaration_31 {
   class=MaterialExpressionNamedRerouteDeclaration
   inputs [None {from=MaterialExpressionFunctionInput_30}]
   properties {name=Height - Min, nodeColor=<rgba 0.5043 1 0 1>}
  }
  MaterialExpressionNamedRerouteDeclaration_32 {
   class=MaterialExpressionNamedRerouteDeclaration
   inputs [None {from=MaterialExpressionFunctionInput_31}]
   properties {name=Height - Max, nodeColor=<rgba 0.5043 1 0 1>}
  }
  MaterialExpressionNamedRerouteDeclaration_33 {
   class=MaterialExpressionNamedRerouteDeclaration
   inputs [None {from=MaterialExpressionFunctionInput_32}]
   properties {name="Use WS(SB)", nodeColor=<rgba 0.1979 0.0084 0 1>}
  }
  MaterialExpressionNamedRerouteDeclaration_34 {
   class=MaterialExpressionNamedRerouteDeclaration
   inputs [None {from=MaterialExpressionFunctionInput_33}]
   properties {name=Specular, nodeColor=<rgba 0.5043 1 0 1>}
  }
  MaterialExpressionNamedRerouteDeclaration_35 {
   class=MaterialExpressionNamedRerouteDeclaration
   inputs [None {from=MaterialExpressionFunctionInput_34}]
   properties {name=Metallic, nodeColor=<rgba 0.5043 1 0 1>}
  }
  MaterialExpressionNamedRerouteDeclaration_36 {
   class=MaterialExpressionNamedRerouteDeclaration
   inputs [None {from=MaterialExpressionFunctionInput_35}]
   properties {name=Roughness, nodeColor=<rgba 0.5043 1 0 1>}
  }
  MaterialExpressionNamedRerouteDeclaration_37 {
   class=MaterialExpressionNamedRerouteDeclaration
   inputs [None {from=MaterialExpressionFunctionInput_16}]
   properties {name=Is Tintable Layer, nodeColor=<rgba 0.5043 1 0 1>}
  }
  MaterialExpressionNamedRerouteDeclaration_38 {
   class=MaterialExpressionNamedRerouteDeclaration
   inputs [None {from=MaterialExpressionTransform_2}]
   properties {name=Normals - Biplanar WS Sample, nodeColor=<rgba 0.099 0.1719 0.0972 1>}
  }
  MaterialExpressionNamedRerouteDeclaration_39 {
   class=MaterialExpressionNamedRerouteDeclaration
   inputs [None {from=MaterialExpressionFunctionInput_23}]
   properties {name=UV - Tiling, nodeColor=<rgba 0.5043 1 0 1>}
  }
  MaterialExpressionNamedRerouteDeclaration_4 {
   class=MaterialExpressionNamedRerouteDeclaration
   inputs [None {from=MaterialExpressionFunctionInput_22}]
   properties {name=Biplanar - Position Offset, nodeColor=<rgba 0.5043 1 0 1>}
  }
  MaterialExpressionNamedRerouteDeclaration_40 {
   class=MaterialExpressionNamedRerouteDeclaration
   inputs [None {from=MaterialExpressionFunctionInput_25}]
   properties {name="Is Fuzzy Layer(SB)", nodeColor=<rgba 0.1979 0.0084 0 1>}
  }
  MaterialExpressionNamedRerouteDeclaration_41 {
   class=MaterialExpressionNamedRerouteDeclaration
   inputs [None {from=MaterialExpressionFunctionInput_41}]
   properties {name=Triplanar - Derivatives Scale, nodeColor=<rgba 0.5043 1 0 1>}
  }
  MaterialExpressionNamedRerouteDeclaration_42 {
   class=MaterialExpressionNamedRerouteDeclaration
   inputs [None {from=MaterialExpressionFunctionInput_44}]
   properties {name=Realtime Displacment - Min, nodeColor=<rgba 0.5043 1 0 1>}
  }
  MaterialExpressionNamedRerouteDeclaration_43 {
   class=MaterialExpressionNamedRerouteDeclaration
   inputs [None {from=MaterialExpressionFunctionInput_45}]
   properties {name=Realtime Displacment - Max, nodeColor=<rgba 0.5043 1 0 1>}
  }
  MaterialExpressionNamedRerouteDeclaration_44 {
   class=MaterialExpressionNamedRerouteDeclaration
   inputs [None {from=MaterialExpressionFunctionInput_36}]
   properties {name=Zplanar - Scale, nodeColor=<rgba 0.5043 1 0 1>}
  }
  MaterialExpressionNamedRerouteDeclaration_45 {
   class=MaterialExpressionNamedRerouteDeclaration
   inputs [None {from=MaterialExpressionFunctionInput_37}]
   properties {name="Zplanar - Use Zplanar WS(SB)", nodeColor=<rgba 0.1979 0.0084 0 1>}
  }
  MaterialExpressionNamedRerouteDeclaration_46 {
   class=MaterialExpressionNamedRerouteDeclaration
   inputs [None {from=MaterialExpressionFunctionInput_46}]
   properties {name=Realtime Displacment - Triplanar Seam, nodeColor=<rgba 0.5043 1 0 1>}
  }
  MaterialExpressionNamedRerouteDeclaration_47 {
   class=MaterialExpressionNamedRerouteDeclaration
   inputs [None {from=MaterialExpressionTransform_1}]
   properties {name=Normals - Zplanar WS Sample, nodeColor=<rgba 0.099 0.1719 0.0972 1>}
  }
  MaterialExpressionNamedRerouteDeclaration_48 {
   class=MaterialExpressionNamedRerouteDeclaration
   inputs [None {from=MaterialExpressionMaterialFunctionCall_3, output=Biplanar - Diffuse}]
   properties {name=Diffuse - Zplanar WS Sample, nodeColor=<rgba 0.099 0.1719 0.0972 1>}
  }
  MaterialExpressionNamedRerouteDeclaration_49 {
   class=MaterialExpressionNamedRerouteDeclaration
   inputs [None {from=MaterialExpressionMaterialFunctionCall_3, output=Biplanar - SpecularMasks}]
   properties {name=SpecularMasks - Zplanar WS Sample, nodeColor=<rgba 0.099 0.1719 0.0972 1>}
  }
  MaterialExpressionNamedRerouteDeclaration_5 {
   class=MaterialExpressionNamedRerouteDeclaration
   inputs [None {from=MaterialExpressionFunctionInput_0}]
   properties {name=Triplanar - Offset, nodeColor=<rgba 0.5043 1 0 1>}
  }
  MaterialExpressionNamedRerouteDeclaration_50 {
   class=MaterialExpressionNamedRerouteDeclaration
   inputs [None {from=MaterialExpressionMaterialFunctionCall_4, output=Triplanar - Seam Mask}]
   properties {name=Triplanar - Seam Mask, nodeColor=<rgba 0 1 0.2031 1>}
  }
  MaterialExpressionNamedRerouteDeclaration_51 {
   class=MaterialExpressionNamedRerouteDeclaration
   inputs [None {from=MaterialExpressionFunctionInput_47}]
   properties {name=Fuzzy - Exponent, nodeColor=<rgba 0.5043 1 0 1>}
  }
  MaterialExpressionNamedRerouteDeclaration_52 {
   class=MaterialExpressionNamedRerouteDeclaration
   inputs [None {from=MaterialExpressionMaterialFunctionCall_0, output=Biplanar - Diffuse}]
   properties {name=Diffuse - Biplanar WS Sample, nodeColor=<rgba 0.099 0.1719 0.0972 1>}
  }
  MaterialExpressionNamedRerouteDeclaration_53 {
   class=MaterialExpressionNamedRerouteDeclaration
   inputs [None {from=MaterialExpressionMaterialFunctionCall_0, output=Biplanar - SpecularMasks}]
   properties {name=SpecularMasks - Biplanar WS Sample, nodeColor=<rgba 0.099 0.1719 0.0972 1>}
  }
  MaterialExpressionNamedRerouteDeclaration_54 {
   class=MaterialExpressionNamedRerouteDeclaration
   inputs [None {from=MaterialExpressionFunctionInput_50}]
   properties {name=Fuzzy - Intensity, nodeColor=<rgba 0.5043 1 0 1>}
  }
  MaterialExpressionNamedRerouteDeclaration_55 {
   class=MaterialExpressionNamedRerouteDeclaration
   inputs [None {from=MaterialExpressionFunctionInput_51}]
   properties {name=Fuzzy - Roughness, nodeColor=<rgba 0.5043 1 0 1>}
  }
  MaterialExpressionNamedRerouteDeclaration_56 {
   class=MaterialExpressionNamedRerouteDeclaration
   inputs [None {from=MaterialExpressionFunctionInput_49}]
   properties {name=_Wavy Distortion World Position, nodeColor=<rgba 0 0 0.5469 1>}
  }
  MaterialExpressionNamedRerouteDeclaration_57 {
   class=MaterialExpressionNamedRerouteDeclaration
   inputs [None {from=""}]
   properties {name=Zplanar WS - DEBUG, nodeColor=<rgba 0.4072 0.4072 0 1>}
  }
  MaterialExpressionNamedRerouteDeclaration_58 {
   class=MaterialExpressionNamedRerouteDeclaration
   inputs [None {from=MaterialExpressionFunctionInput_38}]
   properties {name="Use DEBUG UVs(SB)", nodeColor=<rgba 0.1979 0.0084 0 1>}
  }
  MaterialExpressionNamedRerouteDeclaration_59 {
   class=MaterialExpressionNamedRerouteDeclaration
   inputs [None {from=MaterialExpressionFunctionInput_39}]
   properties {name=Detail - Color - Intensity, nodeColor=<rgba 0.5043 1 0 1>}
  }
  MaterialExpressionNamedRerouteDeclaration_6 {
   class=MaterialExpressionNamedRerouteDeclaration
   inputs [None {from=MaterialExpressionFunctionInput_6}]
   properties {name=Detail - Specular - Contrast, nodeColor=<rgba 0.5043 1 0 1>}
  }
  MaterialExpressionNamedRerouteDeclaration_60 {
   class=MaterialExpressionNamedRerouteDeclaration
   inputs [None {from=MaterialExpressionFunctionInput_40}]
   properties {name=Detail - Specular - Intensity, nodeColor=<rgba 0.5043 1 0 1>}
  }
  MaterialExpressionNamedRerouteDeclaration_61 {
   class=MaterialExpressionNamedRerouteDeclaration
   inputs [None {from=MaterialExpressionComponentMask_5}]
   properties {name=_Baked Curvature, nodeColor=<rgba 0 0 1 1>}
  }
  MaterialExpressionNamedRerouteDeclaration_62 {
   class=MaterialExpressionNamedRerouteDeclaration
   inputs [None {from=MaterialExpressionComponentMask_6}]
   properties {name=_Baked AO, nodeColor=<rgba 0 0 1 1>}
  }
  MaterialExpressionNamedRerouteDeclaration_63 {
   class=MaterialExpressionNamedRerouteDeclaration
   inputs [None {from=MaterialExpressionFunctionInput_42}]
   properties {name="Use VTX Curvature/AO/Painted on Nanite(SB)", nodeColor=<rgba 0.1979 0.0084 0 1>}
  }
  MaterialExpressionNamedRerouteDeclaration_64 {
   class=MaterialExpressionNamedRerouteDeclaration
   inputs [None {from=MaterialExpressionFunctionInput_43}]
   properties {name=Color - Multiply, nodeColor=<rgba 0.5043 1 0 1>}
  }
  MaterialExpressionNamedRerouteDeclaration_65 {
   class=MaterialExpressionNamedRerouteDeclaration
   inputs [None {from=MaterialExpressionComponentMask_7}]
   properties {name=_Painted Mask, nodeColor=<rgba 0 0 1 1>}
  }
  MaterialExpressionNamedRerouteDeclaration_66 {
   class=MaterialExpressionNamedRerouteDeclaration
   inputs [None {from=MaterialExpressionFunctionInput_53}]
   properties {name=Fuzzy - Color, nodeColor=<rgba 0.5043 1 0 1>}
  }
  MaterialExpressionNamedRerouteDeclaration_7 {
   class=MaterialExpressionNamedRerouteDeclaration
   inputs [None {from=MaterialExpressionStaticSwitch_24}]
   properties {name=Normals - UV Sample, nodeColor=<rgba 0.099 0.1719 0.0972 1>}
  }
  MaterialExpressionNamedRerouteDeclaration_8 {
   class=MaterialExpressionNamedRerouteDeclaration
   inputs [None {from=MaterialExpressionFunctionInput_8}]
   properties {name=Realtime Displacement - Flat - Seam Sharpness, nodeColor=<rgba 0.5043 1 0 1>}
  }
  MaterialExpressionNamedRerouteDeclaration_9 {
   class=MaterialExpressionNamedRerouteDeclaration
   inputs [None {from=MaterialExpressionFunctionInput_27}]
   properties {name="Is Detail Layer(SB)", nodeColor=<rgba 0.1979 0.0084 0 1>}
  }
  MaterialExpressionNamedRerouteUsage_0 {class=MaterialExpressionNamedRerouteUsage}
  MaterialExpressionNamedRerouteUsage_1 {class=MaterialExpressionNamedRerouteUsage}
  MaterialExpressionNamedRerouteUsage_10 {class=MaterialExpressionNamedRerouteUsage}
  MaterialExpressionNamedRerouteUsage_100 {class=MaterialExpressionNamedRerouteUsage}
  MaterialExpressionNamedRerouteUsage_101 {class=MaterialExpressionNamedRerouteUsage}
  MaterialExpressionNamedRerouteUsage_102 {class=MaterialExpressionNamedRerouteUsage}
  MaterialExpressionNamedRerouteUsage_103 {class=MaterialExpressionNamedRerouteUsage}
  MaterialExpressionNamedRerouteUsage_104 {class=MaterialExpressionNamedRerouteUsage}
  MaterialExpressionNamedRerouteUsage_105 {class=MaterialExpressionNamedRerouteUsage}
  MaterialExpressionNamedRerouteUsage_106 {class=MaterialExpressionNamedRerouteUsage}
  MaterialExpressionNamedRerouteUsage_107 {class=MaterialExpressionNamedRerouteUsage}
  MaterialExpressionNamedRerouteUsage_108 {class=MaterialExpressionNamedRerouteUsage}
  MaterialExpressionNamedRerouteUsage_109 {class=MaterialExpressionNamedRerouteUsage}
  MaterialExpressionNamedRerouteUsage_11 {class=MaterialExpressionNamedRerouteUsage}
  MaterialExpressionNamedRerouteUsage_110 {class=MaterialExpressionNamedRerouteUsage}
  MaterialExpressionNamedRerouteUsage_111 {class=MaterialExpressionNamedRerouteUsage}
  MaterialExpressionNamedRerouteUsage_112 {class=MaterialExpressionNamedRerouteUsage}
  MaterialExpressionNamedRerouteUsage_113 {class=MaterialExpressionNamedRerouteUsage}
  MaterialExpressionNamedRerouteUsage_114 {class=MaterialExpressionNamedRerouteUsage}
  MaterialExpressionNamedRerouteUsage_115 {class=MaterialExpressionNamedRerouteUsage}
  MaterialExpressionNamedRerouteUsage_116 {class=MaterialExpressionNamedRerouteUsage}
  MaterialExpressionNamedRerouteUsage_117 {class=MaterialExpressionNamedRerouteUsage}
  MaterialExpressionNamedRerouteUsage_118 {class=MaterialExpressionNamedRerouteUsage}
  MaterialExpressionNamedRerouteUsage_119 {class=MaterialExpressionNamedRerouteUsage}
  MaterialExpressionNamedRerouteUsage_12 {class=MaterialExpressionNamedRerouteUsage}
  MaterialExpressionNamedRerouteUsage_120 {class=MaterialExpressionNamedRerouteUsage}
  MaterialExpressionNamedRerouteUsage_121 {class=MaterialExpressionNamedRerouteUsage}
  MaterialExpressionNamedRerouteUsage_122 {class=MaterialExpressionNamedRerouteUsage}
  MaterialExpressionNamedRerouteUsage_123 {class=MaterialExpressionNamedRerouteUsage}
  MaterialExpressionNamedRerouteUsage_124 {class=MaterialExpressionNamedRerouteUsage}
  MaterialExpressionNamedRerouteUsage_125 {class=MaterialExpressionNamedRerouteUsage}
  MaterialExpressionNamedRerouteUsage_126 {class=MaterialExpressionNamedRerouteUsage}
  MaterialExpressionNamedRerouteUsage_127 {class=MaterialExpressionNamedRerouteUsage}
  MaterialExpressionNamedRerouteUsage_128 {class=MaterialExpressionNamedRerouteUsage}
  MaterialExpressionNamedRerouteUsage_129 {class=MaterialExpressionNamedRerouteUsage}
  MaterialExpressionNamedRerouteUsage_13 {class=MaterialExpressionNamedRerouteUsage}
  MaterialExpressionNamedRerouteUsage_130 {class=MaterialExpressionNamedRerouteUsage}
  MaterialExpressionNamedRerouteUsage_131 {class=MaterialExpressionNamedRerouteUsage}
  MaterialExpressionNamedRerouteUsage_132 {class=MaterialExpressionNamedRerouteUsage}
  MaterialExpressionNamedRerouteUsage_133 {class=MaterialExpressionNamedRerouteUsage}
  MaterialExpressionNamedRerouteUsage_134 {class=MaterialExpressionNamedRerouteUsage}
  MaterialExpressionNamedRerouteUsage_135 {class=MaterialExpressionNamedRerouteUsage}
  MaterialExpressionNamedRerouteUsage_136 {class=MaterialExpressionNamedRerouteUsage}
  MaterialExpressionNamedRerouteUsage_137 {class=MaterialExpressionNamedRerouteUsage}
  MaterialExpressionNamedRerouteUsage_138 {class=MaterialExpressionNamedRerouteUsage}
  MaterialExpressionNamedRerouteUsage_139 {class=MaterialExpressionNamedRerouteUsage}
  MaterialExpressionNamedRerouteUsage_14 {class=MaterialExpressionNamedRerouteUsage}
  MaterialExpressionNamedRerouteUsage_140 {class=MaterialExpressionNamedRerouteUsage}
  MaterialExpressionNamedRerouteUsage_15 {class=MaterialExpressionNamedRerouteUsage}
  MaterialExpressionNamedRerouteUsage_16 {class=MaterialExpressionNamedRerouteUsage}
  MaterialExpressionNamedRerouteUsage_17 {class=MaterialExpressionNamedRerouteUsage}
  MaterialExpressionNamedRerouteUsage_18 {class=MaterialExpressionNamedRerouteUsage}
  MaterialExpressionNamedRerouteUsage_19 {class=MaterialExpressionNamedRerouteUsage}
  MaterialExpressionNamedRerouteUsage_2 {class=MaterialExpressionNamedRerouteUsage}
  MaterialExpressionNamedRerouteUsage_20 {class=MaterialExpressionNamedRerouteUsage}
  MaterialExpressionNamedRerouteUsage_21 {class=MaterialExpressionNamedRerouteUsage}
  MaterialExpressionNamedRerouteUsage_22 {class=MaterialExpressionNamedRerouteUsage}
  MaterialExpressionNamedRerouteUsage_23 {class=MaterialExpressionNamedRerouteUsage}
  MaterialExpressionNamedRerouteUsage_24 {class=MaterialExpressionNamedRerouteUsage}
  MaterialExpressionNamedRerouteUsage_25 {class=MaterialExpressionNamedRerouteUsage}
  MaterialExpressionNamedRerouteUsage_26 {class=MaterialExpressionNamedRerouteUsage}
  MaterialExpressionNamedRerouteUsage_27 {class=MaterialExpressionNamedRerouteUsage}
  MaterialExpressionNamedRerouteUsage_28 {class=MaterialExpressionNamedRerouteUsage}
  MaterialExpressionNamedRerouteUsage_29 {class=MaterialExpressionNamedRerouteUsage}
  MaterialExpressionNamedRerouteUsage_3 {class=MaterialExpressionNamedRerouteUsage}
  MaterialExpressionNamedRerouteUsage_30 {class=MaterialExpressionNamedRerouteUsage}
  MaterialExpressionNamedRerouteUsage_31 {class=MaterialExpressionNamedRerouteUsage}
  MaterialExpressionNamedRerouteUsage_32 {class=MaterialExpressionNamedRerouteUsage}
  MaterialExpressionNamedRerouteUsage_33 {class=MaterialExpressionNamedRerouteUsage}
  MaterialExpressionNamedRerouteUsage_34 {class=MaterialExpressionNamedRerouteUsage}
  MaterialExpressionNamedRerouteUsage_35 {class=MaterialExpressionNamedRerouteUsage}
  MaterialExpressionNamedRerouteUsage_36 {class=MaterialExpressionNamedRerouteUsage}
  MaterialExpressionNamedRerouteUsage_37 {class=MaterialExpressionNamedRerouteUsage}
  MaterialExpressionNamedRerouteUsage_38 {class=MaterialExpressionNamedRerouteUsage}
  MaterialExpressionNamedRerouteUsage_39 {class=MaterialExpressionNamedRerouteUsage}
  MaterialExpressionNamedRerouteUsage_4 {class=MaterialExpressionNamedRerouteUsage}
  MaterialExpressionNamedRerouteUsage_40 {class=MaterialExpressionNamedRerouteUsage}
  MaterialExpressionNamedRerouteUsage_41 {class=MaterialExpressionNamedRerouteUsage}
  MaterialExpressionNamedRerouteUsage_42 {class=MaterialExpressionNamedRerouteUsage}
  MaterialExpressionNamedRerouteUsage_43 {class=MaterialExpressionNamedRerouteUsage}
  MaterialExpressionNamedRerouteUsage_44 {class=MaterialExpressionNamedRerouteUsage}
  MaterialExpressionNamedRerouteUsage_45 {class=MaterialExpressionNamedRerouteUsage}
  MaterialExpressionNamedRerouteUsage_46 {class=MaterialExpressionNamedRerouteUsage}
  MaterialExpressionNamedRerouteUsage_47 {class=MaterialExpressionNamedRerouteUsage}
  MaterialExpressionNamedRerouteUsage_48 {class=MaterialExpressionNamedRerouteUsage}
  MaterialExpressionNamedRerouteUsage_49 {class=MaterialExpressionNamedRerouteUsage}
  MaterialExpressionNamedRerouteUsage_5 {class=MaterialExpressionNamedRerouteUsage}
  MaterialExpressionNamedRerouteUsage_50 {class=MaterialExpressionNamedRerouteUsage}
  MaterialExpressionNamedRerouteUsage_51 {class=MaterialExpressionNamedRerouteUsage}
  MaterialExpressionNamedRerouteUsage_52 {class=MaterialExpressionNamedRerouteUsage}
  MaterialExpressionNamedRerouteUsage_53 {class=MaterialExpressionNamedRerouteUsage}
  MaterialExpressionNamedRerouteUsage_54 {class=MaterialExpressionNamedRerouteUsage}
  MaterialExpressionNamedRerouteUsage_55 {class=MaterialExpressionNamedRerouteUsage}
  MaterialExpressionNamedRerouteUsage_56 {class=MaterialExpressionNamedRerouteUsage}
  MaterialExpressionNamedRerouteUsage_57 {class=MaterialExpressionNamedRerouteUsage}
  MaterialExpressionNamedRerouteUsage_58 {class=MaterialExpressionNamedRerouteUsage}
  MaterialExpressionNamedRerouteUsage_59 {class=MaterialExpressionNamedRerouteUsage}
  MaterialExpressionNamedRerouteUsage_6 {class=MaterialExpressionNamedRerouteUsage}
  MaterialExpressionNamedRerouteUsage_60 {class=MaterialExpressionNamedRerouteUsage}
  MaterialExpressionNamedRerouteUsage_61 {class=MaterialExpressionNamedRerouteUsage}
  MaterialExpressionNamedRerouteUsage_62 {class=MaterialExpressionNamedRerouteUsage}
  MaterialExpressionNamedRerouteUsage_63 {class=MaterialExpressionNamedRerouteUsage}
  MaterialExpressionNamedRerouteUsage_64 {class=MaterialExpressionNamedRerouteUsage}
  MaterialExpressionNamedRerouteUsage_65 {class=MaterialExpressionNamedRerouteUsage}
  MaterialExpressionNamedRerouteUsage_66 {class=MaterialExpressionNamedRerouteUsage}
  MaterialExpressionNamedRerouteUsage_67 {class=MaterialExpressionNamedRerouteUsage}
  MaterialExpressionNamedRerouteUsage_68 {class=MaterialExpressionNamedRerouteUsage}
  MaterialExpressionNamedRerouteUsage_69 {class=MaterialExpressionNamedRerouteUsage}
  MaterialExpressionNamedRerouteUsage_7 {class=MaterialExpressionNamedRerouteUsage}
  MaterialExpressionNamedRerouteUsage_70 {class=MaterialExpressionNamedRerouteUsage}
  MaterialExpressionNamedRerouteUsage_71 {class=MaterialExpressionNamedRerouteUsage}
  MaterialExpressionNamedRerouteUsage_72 {class=MaterialExpressionNamedRerouteUsage}
  MaterialExpressionNamedRerouteUsage_73 {class=MaterialExpressionNamedRerouteUsage}
  MaterialExpressionNamedRerouteUsage_74 {class=MaterialExpressionNamedRerouteUsage}
  MaterialExpressionNamedRerouteUsage_75 {class=MaterialExpressionNamedRerouteUsage}
  MaterialExpressionNamedRerouteUsage_76 {class=MaterialExpressionNamedRerouteUsage}
  MaterialExpressionNamedRerouteUsage_77 {class=MaterialExpressionNamedRerouteUsage}
  MaterialExpressionNamedRerouteUsage_78 {class=MaterialExpressionNamedRerouteUsage}
  MaterialExpressionNamedRerouteUsage_79 {class=MaterialExpressionNamedRerouteUsage}
  MaterialExpressionNamedRerouteUsage_8 {class=MaterialExpressionNamedRerouteUsage}
  MaterialExpressionNamedRerouteUsage_80 {class=MaterialExpressionNamedRerouteUsage}
  MaterialExpressionNamedRerouteUsage_81 {class=MaterialExpressionNamedRerouteUsage}
  MaterialExpressionNamedRerouteUsage_82 {class=MaterialExpressionNamedRerouteUsage}
  MaterialExpressionNamedRerouteUsage_83 {class=MaterialExpressionNamedRerouteUsage}
  MaterialExpressionNamedRerouteUsage_84 {class=MaterialExpressionNamedRerouteUsage}
  MaterialExpressionNamedRerouteUsage_85 {class=MaterialExpressionNamedRerouteUsage}
  MaterialExpressionNamedRerouteUsage_87 {class=MaterialExpressionNamedRerouteUsage}
  MaterialExpressionNamedRerouteUsage_88 {class=MaterialExpressionNamedRerouteUsage}
  MaterialExpressionNamedRerouteUsage_89 {class=MaterialExpressionNamedRerouteUsage}
  MaterialExpressionNamedRerouteUsage_9 {class=MaterialExpressionNamedRerouteUsage}
  MaterialExpressionNamedRerouteUsage_90 {class=MaterialExpressionNamedRerouteUsage}
  MaterialExpressionNamedRerouteUsage_91 {class=MaterialExpressionNamedRerouteUsage}
  MaterialExpressionNamedRerouteUsage_92 {class=MaterialExpressionNamedRerouteUsage}
  MaterialExpressionNamedRerouteUsage_93 {class=MaterialExpressionNamedRerouteUsage}
  MaterialExpressionNamedRerouteUsage_94 {class=MaterialExpressionNamedRerouteUsage}
  MaterialExpressionNamedRerouteUsage_95 {class=MaterialExpressionNamedRerouteUsage}
  MaterialExpressionNamedRerouteUsage_96 {class=MaterialExpressionNamedRerouteUsage}
  MaterialExpressionNamedRerouteUsage_97 {class=MaterialExpressionNamedRerouteUsage}
  MaterialExpressionNamedRerouteUsage_98 {class=MaterialExpressionNamedRerouteUsage}
  MaterialExpressionNamedRerouteUsage_99 {class=MaterialExpressionNamedRerouteUsage}
  MaterialExpressionNaniteReplace_1 {
   class=MaterialExpressionNaniteReplace
   inputs [
    Default {from=MaterialExpressionReroute_62, output=""}
    Nanite {from=MaterialExpressionStaticSwitch_30, output=""}
   ]
   properties {}
  }
  MaterialExpressionNaniteReplace_2 {
   class=MaterialExpressionNaniteReplace
   inputs [
    Default {from=MaterialExpressionNamedRerouteUsage_10, output=""}
    Nanite {from=MaterialExpressionNamedRerouteUsage_11, output=""}
   ]
   properties {}
  }
  MaterialExpressionNormalize_1 {class=MaterialExpressionNormalize, inputs [VectorInput {from=MaterialExpressionAppendVector_0, output=""}]}
  MaterialExpressionReroute_0 {class=MaterialExpressionReroute, inputs [None {from=MaterialExpressionNamedRerouteUsage_28}]}
  MaterialExpressionReroute_1 {class=MaterialExpressionReroute, inputs [None {from=MaterialExpressionNamedRerouteUsage_138}]}
  MaterialExpressionReroute_10 {class=MaterialExpressionReroute, inputs [None {from=MaterialExpressionGetMaterialAttributes_2, output=Normal}]}
  MaterialExpressionReroute_100 {class=MaterialExpressionReroute, inputs [None {from=MaterialExpressionStaticSwitch_13}]}
  MaterialExpressionReroute_101 {class=MaterialExpressionReroute, inputs [None {from=MaterialExpressionStaticSwitch_18}]}
  MaterialExpressionReroute_102 {class=MaterialExpressionReroute, inputs [None {from=MaterialExpressionLinearInterpolate_2}]}
  MaterialExpressionReroute_103 {class=MaterialExpressionReroute, inputs [None {from=MaterialExpressionStaticSwitch_26}]}
  MaterialExpressionReroute_104 {class=MaterialExpressionReroute, inputs [None {from=MaterialExpressionReroute_105}]}
  MaterialExpressionReroute_105 {class=MaterialExpressionReroute, inputs [None {from=MaterialExpressionReroute_102}]}
  MaterialExpressionReroute_106 {class=MaterialExpressionReroute, inputs [None {from=MaterialExpressionNamedRerouteUsage_23}]}
  MaterialExpressionReroute_107 {class=MaterialExpressionReroute, inputs [None {from=MaterialExpressionGetMaterialAttributes_1, output=MaterialAttributes}]}
  MaterialExpressionReroute_108 {class=MaterialExpressionReroute, inputs [None {from=MaterialExpressionShadingPathSwitch_10}]}
  MaterialExpressionReroute_109 {class=MaterialExpressionReroute, inputs [None {from=MaterialExpressionNamedRerouteUsage_6}]}
  MaterialExpressionReroute_11 {class=MaterialExpressionReroute, inputs [None {from=MaterialExpressionReroute_31}]}
  MaterialExpressionReroute_110 {class=MaterialExpressionReroute, inputs [None {from=MaterialExpressionNamedRerouteUsage_7}]}
  MaterialExpressionReroute_12 {class=MaterialExpressionReroute, inputs [None {from=MaterialExpressionNamedRerouteUsage_61}]}
  MaterialExpressionReroute_13 {class=MaterialExpressionReroute, inputs [None {from=MaterialExpressionReroute_30}]}
  MaterialExpressionReroute_14 {class=MaterialExpressionReroute, inputs [None {from=MaterialExpressionSetMaterialAttributes_3}]}
  MaterialExpressionReroute_15 {class=MaterialExpressionReroute, inputs [None {from=MaterialExpressionSetMaterialAttributes_8}]}
  MaterialExpressionReroute_16 {class=MaterialExpressionReroute, inputs [None {from=MaterialExpressionTextureSample_3, output=RGB}]}
  MaterialExpressionReroute_17 {class=MaterialExpressionReroute, inputs [None {from=MaterialExpressionGetMaterialAttributes_9, output=MaterialAttributes}]}
  MaterialExpressionReroute_18 {class=MaterialExpressionReroute, inputs [None {from=MaterialExpressionTransform_6}]}
  MaterialExpressionReroute_19 {class=MaterialExpressionReroute, inputs [None {from=MaterialExpressionTextureSample_2, output=RGBA}]}
  MaterialExpressionReroute_2 {class=MaterialExpressionReroute, inputs [None {from=MaterialExpressionNamedRerouteUsage_31}]}
  MaterialExpressionReroute_20 {class=MaterialExpressionReroute, inputs [None {from=MaterialExpressionNamedRerouteUsage_52}]}
  MaterialExpressionReroute_21 {class=MaterialExpressionReroute, inputs [None {from=MaterialExpressionNamedRerouteUsage_54}]}
  MaterialExpressionReroute_22 {class=MaterialExpressionReroute, inputs [None {from=MaterialExpressionNamedRerouteUsage_66}]}
  MaterialExpressionReroute_23 {class=MaterialExpressionReroute, inputs [None {from=MaterialExpressionNamedRerouteUsage_79}]}
  MaterialExpressionReroute_24 {class=MaterialExpressionReroute, inputs [None {from=MaterialExpressionNamedRerouteUsage_59}]}
  MaterialExpressionReroute_25 {class=MaterialExpressionReroute, inputs [None {from=MaterialExpressionNamedRerouteUsage_49}]}
  MaterialExpressionReroute_26 {class=MaterialExpressionReroute, inputs [None {from=MaterialExpressionNamedRerouteUsage_51}]}
  MaterialExpressionReroute_27 {class=MaterialExpressionReroute, inputs [None {from=MaterialExpressionStaticSwitch_4}]}
  MaterialExpressionReroute_28 {class=MaterialExpressionReroute, inputs [None {from=MaterialExpressionNamedRerouteUsage_50}]}
  MaterialExpressionReroute_29 {class=MaterialExpressionReroute, inputs [None {from=MaterialExpressionNamedRerouteUsage_77}]}
  MaterialExpressionReroute_3 {class=MaterialExpressionReroute, inputs [None {from=MaterialExpressionNamedRerouteUsage_48}]}
  MaterialExpressionReroute_30 {class=MaterialExpressionReroute, inputs [None {from=MaterialExpressionReroute_18}], properties {desc=Triplanar Normals}}
  MaterialExpressionReroute_31 {class=MaterialExpressionReroute, inputs [None {from=MaterialExpressionReroute_16}], properties {desc=Triplanar SpecularMasks}}
  MaterialExpressionReroute_32 {class=MaterialExpressionReroute, inputs [None {from=MaterialExpressionReroute_34}]}
  MaterialExpressionReroute_33 {class=MaterialExpressionReroute, inputs [None {from=MaterialExpressionReroute_19}], properties {desc=Triplanar Diffuse}}
  MaterialExpressionReroute_34 {class=MaterialExpressionReroute, inputs [None {from=MaterialExpressionMaterialFunctionCall_26, output=Mask}]}
  MaterialExpressionReroute_35 {class=MaterialExpressionReroute, inputs [None {from=MaterialExpressionReroute_36}]}
  MaterialExpressionReroute_36 {class=MaterialExpressionReroute, inputs [None {from=MaterialExpressionStaticSwitch_17}]}
  MaterialExpressionReroute_37 {class=MaterialExpressionReroute, inputs [None {from=MaterialExpressionNamedRerouteUsage_37}]}
  MaterialExpressionReroute_38 {class=MaterialExpressionReroute, inputs [None {from=MaterialExpressionStaticSwitch_2}]}
  MaterialExpressionReroute_39 {class=MaterialExpressionReroute, inputs [None {from=MaterialExpressionGetMaterialAttributes_9, output=Opacity}]}
  MaterialExpressionReroute_4 {class=MaterialExpressionReroute, inputs [None {from=MaterialExpressionNamedRerouteUsage_38}]}
  MaterialExpressionReroute_40 {class=MaterialExpressionReroute, inputs [None {from=MaterialExpressionNamedRerouteUsage_81}]}
  MaterialExpressionReroute_41 {class=MaterialExpressionReroute, inputs [None {from=MaterialExpressionNamedRerouteUsage_80}]}
  MaterialExpressionReroute_42 {
   class=MaterialExpressionReroute
   inputs [None {from=MaterialExpressionGetMaterialAttributes_11, output=BaseColor}]
   properties {desc=L2 BaseColor}
  }
  MaterialExpressionReroute_43 {class=MaterialExpressionReroute, inputs [None {from=MaterialExpressionNamedRerouteUsage_82}]}
  MaterialExpressionReroute_44 {class=MaterialExpressionReroute, inputs [None {from=MaterialExpressionComponentMask_1}], properties {desc=L2 difuse_R}}
  MaterialExpressionReroute_45 {class=MaterialExpressionReroute, inputs [None {from=MaterialExpressionNamedRerouteUsage_83}]}
  MaterialExpressionReroute_46 {class=MaterialExpressionReroute, inputs [None {from=MaterialExpressionLinearInterpolate_11}]}
  MaterialExpressionReroute_47 {class=MaterialExpressionReroute, inputs [None {from=MaterialExpressionLinearInterpolate_13}]}
  MaterialExpressionReroute_48 {class=MaterialExpressionReroute, inputs [None {from=MaterialExpressionNamedRerouteUsage_78}]}
  MaterialExpressionReroute_49 {class=MaterialExpressionReroute, inputs [None {from=MaterialExpressionNamedRerouteUsage_85}]}
  MaterialExpressionReroute_5 {class=MaterialExpressionReroute, inputs [None {from=MaterialExpressionNamedRerouteUsage_46}]}
  MaterialExpressionReroute_50 {class=MaterialExpressionReroute, inputs [None {from=MaterialExpressionMultiply_2}]}
  MaterialExpressionReroute_51 {class=MaterialExpressionReroute, inputs [None {from=""}]}
  MaterialExpressionReroute_52 {class=MaterialExpressionReroute, inputs [None {from=MaterialExpressionNamedRerouteUsage_87}]}
  MaterialExpressionReroute_53 {class=MaterialExpressionReroute, inputs [None {from=MaterialExpressionNamedRerouteUsage_88}]}
  MaterialExpressionReroute_54 {class=MaterialExpressionReroute, inputs [None {from=MaterialExpressionNamedRerouteUsage_97}]}
  MaterialExpressionReroute_55 {class=MaterialExpressionReroute, inputs [None {from=MaterialExpressionNamedRerouteUsage_98}]}
  MaterialExpressionReroute_56 {class=MaterialExpressionReroute, inputs [None {from=MaterialExpressionNamedRerouteUsage_96}]}
  MaterialExpressionReroute_57 {class=MaterialExpressionReroute, inputs [None {from=MaterialExpressionNamedRerouteUsage_94}]}
  MaterialExpressionReroute_58 {class=MaterialExpressionReroute, inputs [None {from=MaterialExpressionNamedRerouteUsage_95}]}
  MaterialExpressionReroute_59 {class=MaterialExpressionReroute, inputs [None {from=MaterialExpressionNamedRerouteUsage_92}]}
  MaterialExpressionReroute_6 {class=MaterialExpressionReroute, inputs [None {from=MaterialExpressionNamedRerouteUsage_45}]}
  MaterialExpressionReroute_60 {class=MaterialExpressionReroute, inputs [None {from=MaterialExpressionNamedRerouteUsage_93}]}
  MaterialExpressionReroute_61 {class=MaterialExpressionReroute, inputs [None {from=MaterialExpressionNamedRerouteUsage_35}]}
  MaterialExpressionReroute_62 {class=MaterialExpressionReroute, inputs [None {from=MaterialExpressionStaticSwitch_9}]}
  MaterialExpressionReroute_63 {class=MaterialExpressionReroute, inputs [None {from=MaterialExpressionNamedRerouteUsage_91}]}
  MaterialExpressionReroute_64 {class=MaterialExpressionReroute, inputs [None {from=MaterialExpressionNamedRerouteUsage_89}]}
  MaterialExpressionReroute_65 {class=MaterialExpressionReroute, inputs [None {from=MaterialExpressionNamedRerouteUsage_90}]}
  MaterialExpressionReroute_66 {class=MaterialExpressionReroute, inputs [None {from=MaterialExpressionNamedRerouteUsage_36}]}
  MaterialExpressionReroute_67 {class=MaterialExpressionReroute, inputs [None {from=MaterialExpressionNamedRerouteUsage_106}]}
  MaterialExpressionReroute_68 {class=MaterialExpressionReroute, inputs [None {from=MaterialExpressionNamedRerouteUsage_26}]}
  MaterialExpressionReroute_69 {class=MaterialExpressionReroute, inputs [None {from=MaterialExpressionReroute_76}]}
  MaterialExpressionReroute_7 {class=MaterialExpressionReroute, inputs [None {from=MaterialExpressionNamedRerouteUsage_57}]}
  MaterialExpressionReroute_70 {class=MaterialExpressionReroute, inputs [None {from=MaterialExpressionStaticSwitch_23}], properties {desc=Diffuse Alpha for Displacement}}
  MaterialExpressionReroute_71 {
   class=MaterialExpressionReroute
   inputs [None {from=MaterialExpressionMaterialFunctionCall_4, output=A}]
   properties {desc=Triplanar Alpha for Displacement}
  }
  MaterialExpressionReroute_72 {class=MaterialExpressionReroute, inputs [None {from=MaterialExpressionReroute_71}]}
  MaterialExpressionReroute_73 {class=MaterialExpressionReroute, inputs [None {from=MaterialExpressionComponentMask_0}], properties {desc=Alpha of Diffuse for Height}}
  MaterialExpressionReroute_74 {class=MaterialExpressionReroute, inputs [None {from=MaterialExpressionGetMaterialAttributes_0, output=OpacityMask}]}
  MaterialExpressionReroute_75 {class=MaterialExpressionReroute, inputs [None {from=MaterialExpressionReroute_103}]}
  MaterialExpressionReroute_76 {class=MaterialExpressionReroute, inputs [None {from=MaterialExpressionStaticSwitch_15}]}
  MaterialExpressionReroute_77 {class=MaterialExpressionReroute, inputs [None {from=MaterialExpressionStaticSwitch_20}]}
  MaterialExpressionReroute_78 {class=MaterialExpressionReroute, inputs [None {from=MaterialExpressionGetMaterialAttributes_3, output=Roughness}]}
  MaterialExpressionReroute_79 {class=MaterialExpressionReroute, inputs [None {from=MaterialExpressionNamedRerouteUsage_121}]}
  MaterialExpressionReroute_8 {class=MaterialExpressionReroute, inputs [None {from=MaterialExpressionSetMaterialAttributes_0}]}
  MaterialExpressionReroute_80 {class=MaterialExpressionReroute, inputs [None {from=MaterialExpressionNamedRerouteUsage_120}]}
  MaterialExpressionReroute_81 {class=MaterialExpressionReroute, inputs [None {from=MaterialExpressionNamedRerouteUsage_113}]}
  MaterialExpressionReroute_82 {class=MaterialExpressionReroute, inputs [None {from=MaterialExpressionNamedRerouteUsage_125}]}
  MaterialExpressionReroute_83 {class=MaterialExpressionReroute, inputs [None {from=MaterialExpressionStaticSwitch_0}]}
  MaterialExpressionReroute_84 {class=MaterialExpressionReroute, inputs [None {from=MaterialExpressionNamedRerouteUsage_123}]}
  MaterialExpressionReroute_85 {class=MaterialExpressionReroute, inputs [None {from=MaterialExpressionSetMaterialAttributes_1}]}
  MaterialExpressionReroute_86 {class=MaterialExpressionReroute, inputs [None {from=MaterialExpressionNamedRerouteUsage_136}]}
  MaterialExpressionReroute_87 {class=MaterialExpressionReroute, inputs [None {from=MaterialExpressionNamedRerouteUsage_134}]}
  MaterialExpressionReroute_88 {class=MaterialExpressionReroute, inputs [None {from=MaterialExpressionNamedRerouteUsage_133}]}
  MaterialExpressionReroute_89 {class=MaterialExpressionReroute, inputs [None {from=MaterialExpressionNamedRerouteUsage_135}]}
  MaterialExpressionReroute_9 {class=MaterialExpressionReroute, inputs [None {from=MaterialExpressionReroute_33}]}
  MaterialExpressionReroute_90 {class=MaterialExpressionReroute, inputs [None {from=MaterialExpressionNamedRerouteUsage_29}]}
  MaterialExpressionReroute_91 {class=MaterialExpressionReroute, inputs [None {from=MaterialExpressionComponentMask_9}]}
  MaterialExpressionReroute_92 {class=MaterialExpressionReroute, inputs [None {from=MaterialExpressionGetMaterialAttributes_3, output=BaseColor}]}
  MaterialExpressionReroute_93 {class=MaterialExpressionReroute, inputs [None {from=MaterialExpressionGetMaterialAttributes_3, output=BaseColor}]}
  MaterialExpressionReroute_94 {class=MaterialExpressionReroute, inputs [None {from=MaterialExpressionComponentMask_10}]}
  MaterialExpressionReroute_95 {class=MaterialExpressionReroute, inputs [None {from=MaterialExpressionReroute_50}]}
  MaterialExpressionReroute_96 {class=MaterialExpressionReroute, inputs [None {from=MaterialExpressionNamedRerouteUsage_101}]}
  MaterialExpressionReroute_97 {class=MaterialExpressionReroute, inputs [None {from=MaterialExpressionNamedRerouteUsage_102}]}
  MaterialExpressionReroute_98 {class=MaterialExpressionReroute, inputs [None {from=MaterialExpressionStaticSwitch_7}]}
  MaterialExpressionReroute_99 {class=MaterialExpressionReroute, inputs [None {from=MaterialExpressionStaticSwitch_16}]}
  MaterialExpressionSaturate_0 {class=MaterialExpressionSaturate, inputs [None {from=MaterialExpressionLinearInterpolate_0}]}
  MaterialExpressionSaturate_1 {
   class=MaterialExpressionSaturate
   inputs [None {from=MaterialExpressionLinearInterpolate_1}]
   properties {desc=Do we need to keep range in 0 to 1?}
  }
  MaterialExpressionSaturate_2 {class=MaterialExpressionSaturate, inputs [None {from=MaterialExpressionMultiply_6}]}
  MaterialExpressionScalarParameter_0 {class=MaterialExpressionScalarParameter, properties {defaultValue=0.5, group=Layer, parameterName=Specular, sortPriority=380}}
  MaterialExpressionScalarParameter_1 {class=MaterialExpressionScalarParameter, properties {group=Layer, parameterName=Metallic, sortPriority=385}}
  MaterialExpressionScalarParameter_10 {
   class=MaterialExpressionScalarParameter
   properties {defaultValue=1, group=Layer, parameterName=Realtime Displacment - Max, sortPriority=716}
  }
  MaterialExpressionScalarParameter_11 {
   class=MaterialExpressionScalarParameter
   properties {defaultValue=1, group=Layer, parameterName=Realtime Displacment - Triplanar Seam, sortPriority=720}
  }
  MaterialExpressionScalarParameter_12 {
   class=MaterialExpressionScalarParameter
   properties {defaultValue=10, group=Layer, parameterName=Realtime Displacement - Flat - Seam Sharpness, sortPriority=705}
  }
  MaterialExpressionScalarParameter_13 {
   class=MaterialExpressionScalarParameter
   properties {
    desc="Features that use the Tintable values are \"Tinting\" \"Baked Curvature/AO\" \"Biome Tint\"\r\n1 - This Layer can be fully affected by features using Tintable values\r\n0 - This Layer is not affected by features using Tintable values"
    group=Layer
    parameterName=Is Tintable Layer
    sortPriority=0
   }
  }
  MaterialExpressionScalarParameter_14 {class=MaterialExpressionScalarParameter, properties {defaultValue=3, group=Layer, parameterName=Fuzzy - Exponent, sortPriority=455}}
  MaterialExpressionScalarParameter_16 {
   class=MaterialExpressionScalarParameter
   properties {
    defaultValue=1
    desc=Controls this layers normal strength on nanite assets
    group=Layer
    parameterName="Normals - Value (Nanite)"
    sortPriority=415
   }
  }
  MaterialExpressionScalarParameter_17 {
   class=MaterialExpressionScalarParameter
   properties {
    defaultValue=10
    desc=Controls the contrast of the fade between the 2 projections. A value too high will remove the fade and create a harsh seam
    group=Layer
    parameterName=Biplanar - Fade Contrast
    sortPriority=63
   }
  }
  MaterialExpressionScalarParameter_18 {
   class=MaterialExpressionScalarParameter
   properties {
    defaultValue=2
    desc=Controls this layers normal strength on non-nanite assets
    group=Layer
    parameterName="Normals - Value (non-Nanite)"
    sortPriority=416
   }
  }
  MaterialExpressionScalarParameter_19 {class=MaterialExpressionScalarParameter, properties {group=Layer, parameterName="Roughness - Bias (Mobile)", sortPriority=391}}
  MaterialExpressionScalarParameter_2 {class=MaterialExpressionScalarParameter, properties {defaultValue=0.5, group=Layer, parameterName=Roughness, sortPriority=390}}
  MaterialExpressionScalarParameter_23 {class=MaterialExpressionScalarParameter, properties {defaultValue=1, group=Layer, parameterName=Fuzzy - Intensity, sortPriority=450}}
  MaterialExpressionScalarParameter_24 {
   class=MaterialExpressionScalarParameter
   properties {
    defaultValue=0.6
    desc=This Roughness value will be blended to where the Fuzzy shading effect occurs.
    group=Layer
    parameterName=Fuzzy - Roughness
    sortPriority=465
   }
  }
  MaterialExpressionScalarParameter_26 {
   class=MaterialExpressionScalarParameter
   properties {
    defaultValue=700
    desc="Scale is the World Space size that the Triplanar projects, \r\nSmaller sizes = more tiling\r\nLarger sizes = less tiling"
    group=Layer
    parameterName=Triplanar - Scale
    sortPriority=55
   }
  }
  MaterialExpressionScalarParameter_29 {class=MaterialExpressionScalarParameter, properties {defaultValue=1, group=Layer, parameterName=UV - Tiling, sortPriority=105}}
  MaterialExpressionScalarParameter_3 {
   class=MaterialExpressionScalarParameter
   properties {
    defaultValue=700
    desc="Scale is the World Space size that the  Zplanar projects, \r\nSmaller sizes = more tiling\r\nLarger sizes = less tiling"
    group=Layer
    parameterName=Zplanar - Scale
    sortPriority=85
   }
  }
  MaterialExpressionScalarParameter_32 {
   class=MaterialExpressionScalarParameter
   properties {
    defaultValue=1
    desc="Controls the intensity of the Detail Color mix, uses the Red channel of the Diffuse\r\nGenerally want this to be a lower value like 0.3"
    group=Layer
    parameterName=Detail - Color - Intensity
    sortPriority=340
   }
  }
  MaterialExpressionScalarParameter_33 {
   class=MaterialExpressionScalarParameter
   properties {
    defaultValue=1
    desc="Controls the intensity of the Detail Specular mix, uses the Red channel of the Diffuse\r\nIf using with a low Detail - Specular - Contrast, something around 2.0 works well for this param"
    group=Layer
    parameterName=Detail - Specular - Intensity
    sortPriority=341
   }
  }
  MaterialExpressionScalarParameter_4 {
   class=MaterialExpressionScalarParameter
   properties {
    defaultValue=700
    desc="Scale is the World Space size that the Biplanar projects, \r\nSmaller sizes = more tiling\r\nLarger sizes = less tiling"
    group=Layer
    parameterName=Biplanar - Scale
    sortPriority=61
   }
  }
  MaterialExpressionScalarParameter_5 {
   class=MaterialExpressionScalarParameter
   properties {
    defaultValue=1
    desc="Controls the contrast of the Detail Specular mix\r\nA low value usually works best like -0.3"
    group=Layer
    parameterName=Detail - Specular - Contrast
    sortPriority=342
   }
  }
  MaterialExpressionScalarParameter_6 {
   class=MaterialExpressionScalarParameter
   properties {
    defaultValue=1
    desc=Use values less than 1 if noticing low res mips where not expected
    group=Layer
    parameterName=Triplanar - Derivatives Scale
    sortPriority=54
   }
  }
  MaterialExpressionScalarParameter_68 {
   class=MaterialExpressionScalarParameter
   properties {
    defaultValue=75
    desc="Used to control the fade between projections on Low/Mobile, \r\nHigher values = sharper transition\r\nSmaller values = softer transition"
    group=Layer
    parameterName="Triplanar - Fade (Low/Mobile)"
    sortPriority=57
   }
  }
  MaterialExpressionScalarParameter_7 {class=MaterialExpressionScalarParameter, properties {group=Layer, parameterName=Realtime Displacment - Min, sortPriority=715}}
  MaterialExpressionScalarParameter_70 {
   class=MaterialExpressionScalarParameter
   properties {
    defaultValue=10
    desc="Used to control the fade between projections, \r\nHigher values = sharper transition\r\nSmaller values = softer transition\r\n"
    group=Layer
    parameterName=Triplanar - Fade
    sortPriority=56
   }
  }
  MaterialExpressionScalarParameter_8 {
   class=MaterialExpressionScalarParameter
   properties {desc="Values affects Height Masking option\r\nSets the lowest Height value", group=Layer, parameterName=Height - Min, sortPriority=331}
  }
  MaterialExpressionScalarParameter_9 {
   class=MaterialExpressionScalarParameter
   properties {
    defaultValue=1
    desc="Values affects Height Masking option\r\nSets the highest Height value"
    group=Layer
    parameterName=Height - Max
    sortPriority=332
   }
  }
  MaterialExpressionSetMaterialAttributes_0 {
   class=MaterialExpressionSetMaterialAttributes
   inputs [
    MaterialAttributes {from=MaterialExpressionGetMaterialAttributes_4, output=MaterialAttributes}
    BaseColor {from=MaterialExpressionMultiply_27, output=""}
   ]
   properties {attributeSetTypes [#3]}
  }
  MaterialExpressionSetMaterialAttributes_1 {
   class=MaterialExpressionSetMaterialAttributes
   inputs [
    MaterialAttributes {from=MaterialExpressionReroute_8, output=""}
    Roughness {from=MaterialExpressionReroute_108, output=""}
   ]
   properties {attributeSetTypes [#2]}
  }
  MaterialExpressionSetMaterialAttributes_10 {
   class=MaterialExpressionSetMaterialAttributes
   inputs [
    MaterialAttributes {from=MaterialExpressionGetMaterialAttributes_3, output=MaterialAttributes}
    BaseColor {from=MaterialExpressionLinearInterpolate_5, output=""}
    Roughness {from=MaterialExpressionLinearInterpolate_4, output=""}
   ]
   properties {attributeSetTypes [#3, #2]}
  }
  MaterialExpressionSetMaterialAttributes_14 {
   class=MaterialExpressionSetMaterialAttributes
   inputs [
    MaterialAttributes {from=MaterialExpressionReroute_35, output=""}
    BaseColor {from=MaterialExpressionStaticSwitch_14, output=""}
   ]
   properties {attributeSetTypes [#3]}
  }
  MaterialExpressionSetMaterialAttributes_2 {
   class=MaterialExpressionSetMaterialAttributes
   inputs [
    MaterialAttributes {from="", output=""}
    BaseColor {from=MaterialExpressionComponentMask_12, output=""}
    Opacity {from=MaterialExpressionReroute_73, output=""}
    Specular {from=MaterialExpressionConvert_0, output=""}
    Metallic {from=MaterialExpressionConvert_0, output=""}
    Roughness {from=MaterialExpressionConvert_0, output=""}
    Normal {from=MaterialExpressionReroute_83, output=""}
    OpacityMask {from=MaterialExpressionReroute_70, output=""}
   ]
   properties {
    desc="TextureName --- Funtionality --- ChannelUsage --- Input\r\nDiffuse --- Base Color --- RGB --- Base Color\r\nDiffuse --- Displacement --- A --- Opacity\r\nSpecularMasks --- Specular Variation --- R --- Specular\r\nSpecularMasks --- Metallic or Tint/Specular Mask --- G --- Metallic\r\nSpecularMasks --- Roughness --- B --- Roughness\r\nNormals --- Normal Map --- RGB --- Normal"
    attributeSetTypes [
     #3
     #5
     #6
     #7
     #2
     #4
     #1
    ]
   }
  }
  MaterialExpressionSetMaterialAttributes_3 {
   class=MaterialExpressionSetMaterialAttributes
   inputs [
    MaterialAttributes {from=MaterialExpressionGetMaterialAttributes_2, output=MaterialAttributes}
    Normal {from=MaterialExpressionNormalize_1, output=""}
   ]
   properties {attributeSetTypes [#4]}
  }
  MaterialExpressionSetMaterialAttributes_5 {
   class=MaterialExpressionSetMaterialAttributes
   inputs [
    MaterialAttributes {from=MaterialExpressionReroute_77, output=""}
    OpacityMask {from=MaterialExpressionReroute_75, output=""}
   ]
   properties {attributeSetTypes [#1]}
  }
  MaterialExpressionSetMaterialAttributes_6 {
   class=MaterialExpressionSetMaterialAttributes
   inputs [
    MaterialAttributes {from=MaterialExpressionGetMaterialAttributes_11, output=MaterialAttributes}
    BaseColor {from=MaterialExpressionReroute_47, output=""}
    Specular {from=MaterialExpressionReroute_46, output=""}
   ]
   properties {attributeSetTypes [#3, #6]}
  }
  MaterialExpressionSetMaterialAttributes_7 {
   class=MaterialExpressionSetMaterialAttributes
   inputs [
    MaterialAttributes {from=MaterialExpressionReroute_77, output=""}
    OpacityMask {from=MaterialExpressionConstant_16, output=""}
   ]
   properties {attributeSetTypes [#1]}
  }
  MaterialExpressionSetMaterialAttributes_8 {
   class=MaterialExpressionSetMaterialAttributes
   inputs [
    MaterialAttributes {from=MaterialExpressionReroute_17, output=""}
    Opacity {from=MaterialExpressionLinearInterpolate_1, output=""}
   ]
   properties {attributeSetTypes [#5]}
  }
  MaterialExpressionShadingPathSwitch_10 {
   class=MaterialExpressionShadingPathSwitch
   inputs [
    Default {from=MaterialExpressionGetMaterialAttributes_1, output=Roughness}
    Deferred {from="", output=""}
    Forward {from="", output=""}
    Mobile {from=MaterialExpressionAdd_1, output=""}
   ]
   properties {}
  }
  MaterialExpressionStaticBoolParameter_0 {
   class=MaterialExpressionStaticBoolParameter
   inputs []
   properties {
    desc="Enable if the SpecularMasks texture on this layer stores the following ( R = Curvature | G = AO | B = Painted Mask )"
    group=Layer
    parameterName=Use SpecualarMasks as Curvatuve/AO/Painted MSK
    sortPriority=3
   }
  }
  MaterialExpressionStaticBoolParameter_1 {
   class=MaterialExpressionStaticBoolParameter
   inputs []
   properties {
    desc="OFF - Layer acts like a default layer\r\nOn - Layer is now a \"Fuzzy\" layer and has additional params (Example. Moss)"
    group=Layer
    parameterName=Is Fuzzy Layer
    sortPriority=1
   }
  }
  MaterialExpressionStaticBoolParameter_10 {
   class=MaterialExpressionStaticBoolParameter
   inputs []
   properties {
    desc="OFF = Nanite uses VTX and Default uses Textures (Prop Rocks are the only meshes that gets tesselated and displaced on nanite platforms, by default they won`t use VTX)\r\nON = Force VTX for all (VTX masks are only used on big custom sculpts that have a custom nanite mesh with no UVs)\r\n"
    group=Layer
    parameterName=Use VTX Curvature/AO/Painted on Nanite
    sortPriority=4
   }
  }
  MaterialExpressionStaticBoolParameter_11 {class=MaterialExpressionStaticBoolParameter, inputs [], properties {group=Layer, parameterName=Use Second UV Channel, sortPriority=100}}
  MaterialExpressionStaticBoolParameter_13 {
   class=MaterialExpressionStaticBoolParameter
   inputs []
   properties {desc=Enable to use Z Projected planar mapping, group=Layer, parameterName=Use Zplanar WS, sortPriority=80}
  }
  MaterialExpressionStaticBoolParameter_14 {
   class=MaterialExpressionStaticBoolParameter
   inputs []
   properties {
    desc="Off - If Realtime Displacement is enabled this layer doesn't displace inwards or outwards (Value of 0.5, ie no displacement)\r\nOn - Layer's height affects Realtime Displacement, additional params are exposed as well"
    group=Layer
    parameterName=Layer Contributes to Displacement
    sortPriority=2
   }
  }
  MaterialExpressionStaticBoolParameter_15 {class=MaterialExpressionStaticBoolParameter, inputs [], properties {group=Layer, parameterName=Use DEBUG UVs, sortPriority=30}}
  MaterialExpressionStaticBoolParameter_2 {
   class=MaterialExpressionStaticBoolParameter
   inputs []
   properties {
    desc="OFF - Layer acts like a default layer\r\nON - Layer is now a \"detail\" layer and has the following changes\r\nSpecularMasks texture isn't used, \r\nDiffuse texture Red channel is used along with Detail params to overlay Color and Specular info to all layers below"
    group=Layer
    parameterName=Is Detail Layer
    sortPriority=1
   }
  }
  MaterialExpressionStaticBoolParameter_22 {
   class=MaterialExpressionStaticBoolParameter
   inputs []
   properties {
    desc="Enables Wavy Distortion on WS\r\nWavy Distortion shared settings will be in \"Rock/Cliff Shared settings\" parameter group"
    group=Layer
    parameterName=WS - Use Wavy Distortion
    sortPriority=50
   }
  }
  MaterialExpressionStaticBoolParameter_3 {
   class=MaterialExpressionStaticBoolParameter
   inputs []
   properties {desc="Enable to use side projected planar mapping(Sediment Cliffs etc)", group=Layer, parameterName=Use Biplanar WS, sortPriority=60}
  }
  MaterialExpressionStaticBoolParameter_4 {class=MaterialExpressionStaticBoolParameter, inputs [], properties {group=Layer, parameterName=Layer Is Used, sortPriority=-1}}
  MaterialExpressionStaticBoolParameter_5 {class=MaterialExpressionStaticBoolParameter, inputs [], properties {group=Layer, parameterName=Use WS, sortPriority=35}}
  MaterialExpressionStaticBoolParameter_7 {
   class=MaterialExpressionStaticBoolParameter
   inputs []
   properties {group=Layer, parameterName=Realtime Displacement - Use Costly Triplanar, sortPriority=700}
  }
  MaterialExpressionStaticBool_0 {class=MaterialExpressionStaticBool}
  MaterialExpressionStaticBool_1 {class=MaterialExpressionStaticBool}
  MaterialExpressionStaticBool_10 {class=MaterialExpressionStaticBool}
  MaterialExpressionStaticBool_11 {class=MaterialExpressionStaticBool}
  MaterialExpressionStaticBool_12 {class=MaterialExpressionStaticBool}
  MaterialExpressionStaticBool_13 {class=MaterialExpressionStaticBool}
  MaterialExpressionStaticBool_14 {class=MaterialExpressionStaticBool}
  MaterialExpressionStaticBool_15 {class=MaterialExpressionStaticBool}
  MaterialExpressionStaticBool_5 {class=MaterialExpressionStaticBool}
  MaterialExpressionStaticBool_6 {class=MaterialExpressionStaticBool}
  MaterialExpressionStaticBool_7 {class=MaterialExpressionStaticBool}
  MaterialExpressionStaticBool_8 {class=MaterialExpressionStaticBool}
  MaterialExpressionStaticBool_9 {class=MaterialExpressionStaticBool}
  MaterialExpressionStaticSwitch_0 {
   class=MaterialExpressionStaticSwitch
   inputs [
    True {from=MaterialExpressionReroute_100, output=""}
    False {from=MaterialExpressionNamedRerouteUsage_33, output=""}
    Value {from=MaterialExpressionNamedRerouteUsage_17, output=""}
   ]
   properties {}
  }
  MaterialExpressionStaticSwitch_1 {
   class=MaterialExpressionStaticSwitch
   inputs [
    True {from=MaterialExpressionReroute_99, output=""}
    False {from=MaterialExpressionNamedRerouteUsage_44, output=""}
    Value {from=MaterialExpressionNamedRerouteUsage_17, output=""}
   ]
   properties {}
  }
  MaterialExpressionStaticSwitch_10 {
   class=MaterialExpressionStaticSwitch
   inputs [
    True {from=MaterialExpressionSetMaterialAttributes_10, output=""}
    False {from=MaterialExpressionReroute_85, output=""}
    Value {from=MaterialExpressionNamedRerouteUsage_126, output=""}
   ]
   properties {}
  }
  MaterialExpressionStaticSwitch_11 {
   class=MaterialExpressionStaticSwitch
   inputs [
    True {from=MaterialExpressionTextureCoordinate_1, output=""}
    False {from=MaterialExpressionTextureCoordinate_0, output=""}
    Value {from=MaterialExpressionNamedRerouteUsage_13, output=""}
   ]
   properties {}
  }
  MaterialExpressionStaticSwitch_12 {
   class=MaterialExpressionStaticSwitch
   inputs [
    True {from=MaterialExpressionConvert_1, output=""}
    False {from=MaterialExpressionStaticSwitch_1, output=""}
    Value {from=MaterialExpressionNamedRerouteUsage_100, output=""}
   ]
   properties {}
  }
  MaterialExpressionStaticSwitch_13 {
   class=MaterialExpressionStaticSwitch
   inputs [
    True {from=MaterialExpressionNamedRerouteUsage_105, output=""}
    False {from=MaterialExpressionReroute_13, output=""}
    Value {from=MaterialExpressionNamedRerouteUsage_122, output=""}
   ]
   properties {}
  }
  MaterialExpressionStaticSwitch_14 {
   class=MaterialExpressionStaticSwitch
   inputs [
    True {from=MaterialExpressionStaticSwitch_19, output=""}
    False {from=MaterialExpressionNamedRerouteUsage_41, output=""}
    Value {from=MaterialExpressionNamedRerouteUsage_68, output=""}
   ]
   properties {}
  }
  MaterialExpressionStaticSwitch_15 {
   class=MaterialExpressionStaticSwitch
   inputs [
    True {from=MaterialExpressionNamedRerouteUsage_131, output=""}
    False {from=MaterialExpressionWorldPosition_6, output=XYZ}
    Value {from=MaterialExpressionNamedRerouteUsage_129, output=""}
   ]
   properties {}
  }
  MaterialExpressionStaticSwitch_16 {
   class=MaterialExpressionStaticSwitch
   inputs [
    True {from=MaterialExpressionNamedRerouteUsage_104, output=""}
    False {from=MaterialExpressionReroute_11, output=""}
    Value {from=MaterialExpressionNamedRerouteUsage_122, output=""}
   ]
   properties {}
  }
  MaterialExpressionStaticSwitch_17 {
   class=MaterialExpressionStaticSwitch
   inputs [
    True {from=MaterialExpressionSetMaterialAttributes_5, output=""}
    False {from=MaterialExpressionSetMaterialAttributes_7, output=""}
    Value {from=MaterialExpressionNamedRerouteUsage_116, output=""}
   ]
   properties {}
  }
  MaterialExpressionStaticSwitch_18 {
   class=MaterialExpressionStaticSwitch
   inputs [
    True {from=MaterialExpressionComponentMask_11, output=""}
    False {from=MaterialExpressionReroute_72, output=""}
    Value {from=MaterialExpressionNamedRerouteUsage_122, output=""}
   ]
   properties {}
  }
  MaterialExpressionStaticSwitch_19 {
   class=MaterialExpressionStaticSwitch
   inputs [
    True {from=MaterialExpressionNamedRerouteUsage_62, output=""}
    False {from=MaterialExpressionNamedRerouteUsage_32, output=""}
    Value {from=MaterialExpressionNamedRerouteUsage_67, output=""}
   ]
   properties {}
  }
  MaterialExpressionStaticSwitch_2 {
   class=MaterialExpressionStaticSwitch
   inputs [
    True {from=MaterialExpressionReroute_98, output=""}
    False {from=MaterialExpressionNamedRerouteUsage_8, output=""}
    Value {from=MaterialExpressionNamedRerouteUsage_17, output=""}
   ]
   properties {}
  }
  MaterialExpressionStaticSwitch_20 {
   class=MaterialExpressionStaticSwitch
   inputs [
    True {from=MaterialExpressionSetMaterialAttributes_6, output=""}
    False {from=MaterialExpressionStaticSwitch_10, output=""}
    Value {from=MaterialExpressionNamedRerouteUsage_71, output=""}
   ]
   properties {}
  }
  MaterialExpressionStaticSwitch_21 {
   class=MaterialExpressionStaticSwitch
   inputs [
    True {from=MaterialExpressionComponentMask_8, output=""}
    False {from=MaterialExpressionReroute_71, output=""}
    Value {from=MaterialExpressionNamedRerouteUsage_43, output=""}
   ]
   properties {}
  }
  MaterialExpressionStaticSwitch_22 {
   class=MaterialExpressionStaticSwitch
   inputs [
    True {from=MaterialExpressionReroute_102, output=""}
    False {from=MaterialExpressionLinearInterpolate_3, output=""}
    Value {from=MaterialExpressionNamedRerouteUsage_140, output=""}
   ]
   properties {}
  }
  MaterialExpressionStaticSwitch_23 {
   class=MaterialExpressionStaticSwitch
   inputs [
    True {from=MaterialExpressionReroute_101, output=""}
    False {from=MaterialExpressionComponentMask_4, output=""}
    Value {from=MaterialExpressionNamedRerouteUsage_17, output=""}
   ]
   properties {}
  }
  MaterialExpressionStaticSwitch_24 {
   class=MaterialExpressionStaticSwitch
   inputs [
    True {from=MaterialExpressionTransform_0, output=""}
    False {from=MaterialExpressionTextureSample_9, output=RGB}
    Value {from=MaterialExpressionNamedRerouteUsage_108, output=""}
   ]
   properties {}
  }
  MaterialExpressionStaticSwitch_25 {
   class=MaterialExpressionStaticSwitch
   inputs [
    True {from=MaterialExpressionReroute_105, output=""}
    False {from=MaterialExpressionLinearInterpolate_3, output=""}
    Value {from=MaterialExpressionNamedRerouteUsage_139, output=""}
   ]
   properties {}
  }
  MaterialExpressionStaticSwitch_26 {
   class=MaterialExpressionStaticSwitch
   inputs [
    True {from=MaterialExpressionStaticSwitch_25, output=""}
    False {from=MaterialExpressionReroute_104, output=""}
    Value {from=MaterialExpressionNamedRerouteUsage_137, output=""}
   ]
   properties {}
  }
  MaterialExpressionStaticSwitch_3 {
   class=MaterialExpressionStaticSwitch
   inputs [
    True {from=MaterialExpressionNamedRerouteUsage_34, output=""}
    False {from=MaterialExpressionWorldPosition_0, output=XYZ}
    Value {from=MaterialExpressionNamedRerouteUsage_25, output=""}
   ]
   properties {}
  }
  MaterialExpressionStaticSwitch_30 {
   class=MaterialExpressionStaticSwitch
   inputs [
    True {from=MaterialExpressionVertexColor_5, output=""}
    False {from=MaterialExpressionReroute_62, output=""}
    Value {from=MaterialExpressionNamedRerouteUsage_72, output=""}
   ]
   properties {desc="OFF = Nanite uses VTX and Default uses Textures (Prop Rocks are the only meshes that gets tesselated and displaced on nanite platforms, by default they won`t use VTX)\r\nON = Force VTX for all (VTX masks are only used on big custom sculpts that have a custom nanite mesh with no UVs)\r\n"}
  }
  MaterialExpressionStaticSwitch_4 {
   class=MaterialExpressionStaticSwitch
   inputs [
    True {from=MaterialExpressionSetMaterialAttributes_14, output=""}
    False {from=MaterialExpressionReroute_35, output=""}
    Value {from=MaterialExpressionNamedRerouteUsage_65, output=""}
   ]
   properties {}
  }
  MaterialExpressionStaticSwitch_5 {
   class=MaterialExpressionStaticSwitch
   inputs [
    True {from=MaterialExpressionNamedRerouteUsage_53, output=""}
    False {from=MaterialExpressionReroute_33, output=""}
    Value {from=MaterialExpressionNamedRerouteUsage_43, output=""}
   ]
   properties {}
  }
  MaterialExpressionStaticSwitch_6 {
   class=MaterialExpressionStaticSwitch
   inputs [
    True {from=MaterialExpressionNamedRerouteUsage_55, output=""}
    False {from=MaterialExpressionReroute_30, output=""}
    Value {from=MaterialExpressionNamedRerouteUsage_43, output=""}
   ]
   properties {}
  }
  MaterialExpressionStaticSwitch_7 {
   class=MaterialExpressionStaticSwitch
   inputs [
    True {from=MaterialExpressionNamedRerouteUsage_124, output=""}
    False {from=MaterialExpressionReroute_9, output=""}
    Value {from=MaterialExpressionNamedRerouteUsage_122, output=""}
   ]
   properties {}
  }
  MaterialExpressionStaticSwitch_8 {
   class=MaterialExpressionStaticSwitch
   inputs [
    True {from=MaterialExpressionNamedRerouteUsage_60, output=""}
    False {from=MaterialExpressionReroute_31, output=""}
    Value {from=MaterialExpressionNamedRerouteUsage_43, output=""}
   ]
   properties {}
  }
  MaterialExpressionStaticSwitch_9 {
   class=MaterialExpressionStaticSwitch
   inputs [
    True {from=MaterialExpressionNamedRerouteUsage_14, output=""}
    False {from=MaterialExpressionConstant3Vector_0, output=""}
    Value {from=MaterialExpressionNamedRerouteUsage_99, output=""}
   ]
   properties {}
  }
  MaterialExpressionSubtract_0 {
   class=MaterialExpressionSubtract
   inputs [
    A {from=""}
    B {from=MaterialExpressionNamedRerouteUsage_27}
   ]
   properties {constA=0}
  }
  MaterialExpressionTextureCoordinate_0 {class=MaterialExpressionTextureCoordinate}
  MaterialExpressionTextureCoordinate_1 {class=MaterialExpressionTextureCoordinate, properties {coordinateIndex=1}}
  MaterialExpressionTextureObject_0 {class=MaterialExpressionTextureObject, inputs [], properties {texture=/Game/Global/Textures/Default/Blanks/T_FlatGrey}}
  MaterialExpressionTextureObject_1 {
   class=MaterialExpressionTextureObject
   inputs []
   properties {samplerType=SAMPLERTYPE_Masks, texture=&T_Fortnite_Default_S}
  }
  MaterialExpressionTextureObject_2 {class=MaterialExpressionTextureObject, inputs [], properties {samplerType=SAMPLERTYPE_Normal, texture=/Engine/EngineMaterials/FlatNormal}}
  MaterialExpressionTextureSample_1 {
   class=MaterialExpressionTextureSample
   inputs [
    UVs {from=MaterialExpressionMaterialFunctionCall_26, output=UVs}
    Tex {from=""}
    "DDX(UVs)" {from=MaterialExpressionMaterialFunctionCall_26, output=UVs}
    "DDY(UVs)" {from=MaterialExpressionMaterialFunctionCall_26, output=UVs}
    Apply View MipBias {from=""}
   ]
   properties {
    mipValueMode=TMVM_Derivative
    samplerSource=SSM_Wrap_WorldGroupSettings
    texture=/Game/FN_PublicTurntable/Turntable_Lowpoly/Content/BlueprintAssets/Checker
   }
  }
  MaterialExpressionTextureSample_10 {
   class=MaterialExpressionTextureSample
   inputs [
    UVs {from=MaterialExpressionReroute_50}
    Tex {from=""}
    Apply View MipBias {from=""}
   ]
   properties {samplerSource=SSM_Wrap_WorldGroupSettings, texture=&Checker}
  }
  MaterialExpressionTextureSample_2 {
   class=MaterialExpressionTextureSample
   inputs [
    UVs {from=MaterialExpressionMaterialFunctionCall_26, output=UVs}
    Tex {from=MaterialExpressionNamedRerouteUsage_0}
    "DDX(UVs)" {from=MaterialExpressionMaterialFunctionCall_26, output=UVs}
    "DDY(UVs)" {from=MaterialExpressionMaterialFunctionCall_26, output=UVs}
    Apply View MipBias {from=""}
   ]
   properties {mipValueMode=TMVM_Derivative, samplerSource=SSM_Wrap_WorldGroupSettings}
  }
  MaterialExpressionTextureSample_3 {
   class=MaterialExpressionTextureSample
   inputs [
    UVs {from=MaterialExpressionMaterialFunctionCall_26, output=UVs}
    Tex {from=MaterialExpressionNamedRerouteUsage_2}
    "DDX(UVs)" {from=MaterialExpressionMaterialFunctionCall_26, output=UVs}
    "DDY(UVs)" {from=MaterialExpressionMaterialFunctionCall_26, output=UVs}
    Apply View MipBias {from=""}
   ]
   properties {mipValueMode=TMVM_Derivative, samplerSource=SSM_Wrap_WorldGroupSettings, samplerType=SAMPLERTYPE_Masks}
  }
  MaterialExpressionTextureSample_4 {
   class=MaterialExpressionTextureSample
   inputs [
    UVs {from=MaterialExpressionMaterialFunctionCall_26, output=UVs}
    Tex {from=MaterialExpressionNamedRerouteUsage_1}
    "DDX(UVs)" {from=MaterialExpressionMaterialFunctionCall_26, output=UVs}
    "DDY(UVs)" {from=MaterialExpressionMaterialFunctionCall_26, output=UVs}
    Apply View MipBias {from=""}
   ]
   properties {mipValueMode=TMVM_Derivative, samplerSource=SSM_Wrap_WorldGroupSettings, samplerType=SAMPLERTYPE_Normal}
  }
  MaterialExpressionTextureSample_5 {
   class=MaterialExpressionTextureSample
   inputs [
    UVs {from=MaterialExpressionReroute_50}
    Tex {from=MaterialExpressionNamedRerouteUsage_3}
    Apply View MipBias {from=""}
   ]
   properties {samplerSource=SSM_Wrap_WorldGroupSettings}
  }
  MaterialExpressionTextureSample_8 {
   class=MaterialExpressionTextureSample
   inputs [
    UVs {from=MaterialExpressionReroute_50}
    Tex {from=MaterialExpressionNamedRerouteUsage_5}
    Apply View MipBias {from=""}
   ]
   properties {samplerSource=SSM_Wrap_WorldGroupSettings}
  }
  MaterialExpressionTextureSample_9 {
   class=MaterialExpressionTextureSample
   inputs [
    UVs {from=MaterialExpressionReroute_95}
    Tex {from=MaterialExpressionNamedRerouteUsage_4}
    Apply View MipBias {from=""}
   ]
   properties {samplerSource=SSM_Wrap_WorldGroupSettings}
  }
  MaterialExpressionTransform_0 {
   class=MaterialExpressionTransform
   inputs [None {from=MaterialExpressionMaterialFunctionCall_2, output=World Space Normal}]
   properties {transformSourceType=TRANSFORMSOURCE_World, transformType=TRANSFORM_Tangent}
  }
  MaterialExpressionTransform_1 {
   class=MaterialExpressionTransform
   inputs [None {from=MaterialExpressionMaterialFunctionCall_3, output=Biplanar - Normals}]
   properties {transformSourceType=TRANSFORMSOURCE_World, transformType=TRANSFORM_Tangent}
  }
  MaterialExpressionTransform_2 {
   class=MaterialExpressionTransform
   inputs [None {from=MaterialExpressionMaterialFunctionCall_0, output=Biplanar - Normals}]
   properties {transformSourceType=TRANSFORMSOURCE_World, transformType=TRANSFORM_Tangent}
  }
  MaterialExpressionTransform_6 {
   class=MaterialExpressionTransform
   inputs [None {from=MaterialExpressionMaterialFunctionCall_12, output=World Normals}]
   properties {transformSourceType=TRANSFORMSOURCE_World, transformType=TRANSFORM_Tangent}
  }
  MaterialExpressionVectorParameter_0 {
   class=MaterialExpressionVectorParameter
   properties {desc=Multiply with the BaseColor, group=Layer, parameterName=Color - Multiply, sortPriority=321, defaultValue=<rgba 1 1 1 0>}
  }
  MaterialExpressionVectorParameter_1 {
   class=MaterialExpressionVectorParameter
   properties {
    desc="Multiply with the Detail Basecolor to mix, uses the Red channel of the Diffuse\r\nGenerally want the value of this to be between .5 and 1.5"
    group=Layer
    parameterName=Detail - Color - Multiply
    sortPriority=321
    defaultValue=<rgba 1 1 1 0>
   }
  }
  MaterialExpressionVectorParameter_11 {
   class=MaterialExpressionVectorParameter
   properties {
    group=Layer
    parameterName=WS - Wavy Distortion - Offset
    sortPriority=51
    channelNames {a="", b=Z, g=Y, r=X}
    defaultValue=<rgba 0 0 0 100000>
   }
  }
  MaterialExpressionVectorParameter_12 {
   class=MaterialExpressionVectorParameter
   properties {group=Layer, parameterName=Triplanar - Offset, sortPriority=53, channelNames {a="", b=Z, g=Y, r=X}}
  }
  MaterialExpressionVectorParameter_2 {
   class=MaterialExpressionVectorParameter
   properties {
    desc="This Color is either Multiplied or added to Layer's Color based on the Alpha channel value.\r\nNote that this effect happens after the Layer's Color Multiply"
    group=Layer
    parameterName=Fuzzy - Color
    sortPriority=460
    channelNames {a="0=Multiply 1=Add", b="", g="", r=""}
    defaultValue=<rgba 1.1 1.1 1.1 0.5>
   }
  }
  MaterialExpressionVectorParameter_3 {
   class=MaterialExpressionVectorParameter
   properties {group=Layer, parameterName=Biplanar - Position Offset, sortPriority=62, defaultValue=<rgba 0 0 0 1>}
  }
  MaterialExpressionVertexColor_5 {class=MaterialExpressionVertexColor, inputs [], properties {}}
  MaterialExpressionWorldPosition_0 {class=MaterialExpressionWorldPosition, inputs [], properties {}}
  MaterialExpressionWorldPosition_6 {class=MaterialExpressionWorldPosition, inputs [], properties {}}
 }}
