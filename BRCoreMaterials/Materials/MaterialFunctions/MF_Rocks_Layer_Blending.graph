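# Material function that blends rock material layer B onto layer A, outputting blended attributes and a tintable mask, with static switches for layer options.
# Standard layers lerp attributes by mask, optionally adding B's normal XY.
# Detail layers instead multiply base color and specular, with a simpler mobile path and optional normal adding only at Epic quality.
# OpacityMask carries displacement, interpolated separately, apparently using the realtime displacement mask.
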MaterialFunction /BRCoreMaterials/Materials/MaterialFunctions/MF_Rocks_Layer_Blending
graph {nodes [
  Constant()
  FunctionInput() {inputName="\\ --- Blend Options --- /", inputType=FunctionInput_StaticBool, sortPriority=100}
  n1 = NamedRerouteUsage()
  n2 = NamedRerouteUsage()
  n3 = NamedRerouteUsage()
  FunctionOutput(None=StaticSwitch(True=LinearInterpolate(A=@n1, B=@n2, Alpha=@n3), False=NamedRerouteUsage(), Value=NamedRerouteUsage())) {outputName=Tintable Mask, sortPriority=-50}
  n4 = NamedRerouteUsage()
  n5 = GetMaterialAttributes(None=NamedRerouteUsage()) {attributeGetTypes [
    #1
    #2
    #3
    #4
    #5
    #6
   ]}
  n6 = GetMaterialAttributes(None=NamedRerouteUsage()) {attributeGetTypes [
    #1
    #2
    #3
    #6
   ]}
  n7 = Multiply(A=Multiply(A=@n6.BaseColor, B=@n5.BaseColor), B=Constant() {r=1.5})
  n8 = Multiply(A=Multiply(B=@n5.Specular) {constA=2}, B=@n6.Specular)
  n9 = ShadingPathSwitch(
   Default=SetMaterialAttributes!(MaterialAttributes=@n4, BaseColor=Saturate(None=@n7), Specular=Saturate(None=@n8)) {attributeSetTypes [#1, #3]}
   Mobile=SetMaterialAttributes!(MaterialAttributes=@n4, Specular=@n6.Specular) {attributeSetTypes [#3]}
  )
  n10 = Multiply(A=@n5.Normal, B=Constant3Vector() {constant=<rgba 1 1 0 0>})
  n11 = StaticSwitch(True=Normalize(VectorInput=Add(A=@n6.Normal, B=@n10)), False=@n6.Normal, Value=NamedRerouteUsage())
  n12 = GetMaterialAttributes(None=NamedRerouteUsage()) {attributeGetTypes [
    #1
    #2
    #3
    #4
    #5
    #6
    #7
   ]}
  n13 = GetMaterialAttributes(None=NamedRerouteUsage()) {attributeGetTypes [
    #1
    #2
    #3
    #4
    #5
    #6
    #7
   ]}
  n14 = BlendMaterialAttributes(A=@n12.MaterialAttributes, B=@n13.MaterialAttributes, Alpha=NamedRerouteUsage()) {desc=Use Input A Vertex Attributes, vertexAttributeBlendType=UseA}
  n15 = Multiply(A=@n13.Normal, B=Constant3Vector() {constant=<rgba 1 1 0 0>})
  n16 = LinearInterpolate(A=@n12.Normal, B=Normalize(VectorInput=Add(A=@n12.Normal, B=@n15)), Alpha=NamedRerouteUsage())
  n17 = StaticSwitch(
   True=QualitySwitch(
    Default=@n9
    Low=@n9
    High=@n9
    Medium=@n9
    Epic=SetMaterialAttributes!(MaterialAttributes=@n9, Normal=@n11) {attributeSetTypes [#6]}
   )
   False=StaticSwitch(
    True=SetMaterialAttributes(MaterialAttributes=@n14, Normal=@n16) {attributeSetTypes [#6]}
    False=@n14
    Value=NamedRerouteUsage()
   )
   Value=NamedRerouteUsage()
  )
  n18 = LinearInterpolate(A=NamedRerouteUsage(), B=NamedRerouteUsage(), Alpha=NamedRerouteUsage())
  FunctionOutput(None=StaticSwitch(
    True=SetMaterialAttributes!(MaterialAttributes=@n17, OpacityMask=@n18) {attributeSetTypes [#7]}
    False=NamedRerouteUsage()
    Value=NamedRerouteUsage()
   )) {outputName=Blended MaterialAttributes, sortPriority=-100}
  NamedRerouteDeclaration(None=FunctionInput(Preview=StaticBool()) {inputName=Layer B - Is Used, inputType=FunctionInput_StaticBool, sortPriority=101}) {name="Layer B - Is Used(SB)", nodeColor=<rgba 0.1979 0.0084 0 1>}
  NamedRerouteDeclaration(None=FunctionInput() {inputName=B Tintable Mask, inputType=FunctionInput_Scalar, sortPriority=-49}) {name=_B Tintable Mask, nodeColor=<rgba 0.0021 1 0.4002 1>}
  n19 = FunctionInput() {bUsePreviewValueAsDefault=true, inputName=Layer B Realtime Displacement Mask, inputType=FunctionInput_Scalar, sortPriority=-18}
  NamedRerouteDeclaration(None=@n19) {name=_Layer B Realtime Displacement Mask, nodeColor=<rgba 1 1 1 1>}
  NamedRerouteDeclaration(None=FunctionInput() {bUsePreviewValueAsDefault=true, inputName=Layer B Mask, inputType=FunctionInput_Scalar, sortPriority=-20}) {name=_Layer B Mask, nodeColor=<rgba 1 1 1 1>}
  NamedRerouteDeclaration(None=FunctionInput(Preview=StaticBool()) {inputName=Layer B - Add Normals, inputType=FunctionInput_StaticBool, sortPriority=120}) {name="Layer B - Add Normals(SB)", nodeColor=<rgba 0.1979 0.0084 0 1>}
  NamedRerouteDeclaration(None=@n12.OpacityMask) {name=_A MaterialDisplacement, nodeColor=<rgba 1 0 0.039 1>}
  NamedRerouteDeclaration(None=FunctionInput() {inputName=A MaterialAttributes, inputType=FunctionInput_MaterialAttributes, sortPriority=-100}) {name=_A MaterialAttributes, nodeColor=<rgba 1 0 0.039 1>}
  NamedRerouteDeclaration(None=@n13.OpacityMask) {name=_B MaterialDisplacement, nodeColor=<rgba 0.0021 1 0.4002 1>}
  NamedRerouteDeclaration(None=FunctionInput() {inputName=B MaterialAttributes, inputType=FunctionInput_MaterialAttributes, sortPriority=-50}) {name=_B MaterialAttributes, nodeColor=<rgba 0.0021 1 0.4002 1>}
  NamedRerouteDeclaration(None=FunctionInput(Preview=StaticBool()) {inputName=Layer B - Is Detail Layer, inputType=FunctionInput_StaticBool, sortPriority=110}) {name="Layer B - Is Detail Layer(SB)", nodeColor=<rgba 0.1979 0.0084 0 1>}
  NamedRerouteDeclaration(None=FunctionInput() {inputName=A Tintable Mask, inputType=FunctionInput_Scalar, sortPriority=-99}) {name=_A Tintable Mask, nodeColor=<rgba 1 0 0.039 1>}
 ]}
