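# Material function that builds the blend mask for layering rock materials, outputting a combined mask and a realtime displacement mask.
# Static switches toggle world-Z normal, vertex color, DFAO, Z gradient, baked AO/curvature/painted, waterline, world-space noise and height masks.
# Masks get bias and intensity controls, some with inversion, and combine by selectable max, min or blend.
# If nodes replace lerps to save cost, and vertex gradients are compensated for less dense mobile meshes.
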MaterialFunction /BRCoreMaterials/Materials/MaterialFunctions/MF_Rocks_Layer_Masking
graph {expressions {
  MaterialExpressionAdd_0 {
   class=MaterialExpressionAdd
   inputs [
    A {from=MaterialExpressionIf_9}
    B {from=MaterialExpressionNamedRerouteUsage_21}
   ]
  }
  MaterialExpressionAdd_1 {
   class=MaterialExpressionAdd
   inputs [
    A {from=MaterialExpressionIf_13}
    B {from=MaterialExpressionNamedRerouteUsage_41}
   ]
  }
  MaterialExpressionAdd_2 {
   class=MaterialExpressionAdd
   inputs [
    A {from=MaterialExpressionOneMinus_2}
    B {from=MaterialExpressionAdd_6}
   ]
   properties {constB=30}
  }
  MaterialExpressionAdd_3 {
   class=MaterialExpressionAdd
   inputs [
    A {from=MaterialExpressionIf_4}
    B {from=MaterialExpressionNamedRerouteUsage_86}
   ]
  }
  MaterialExpressionAdd_4 {
   class=MaterialExpressionAdd
   inputs [
    A {from=MaterialExpressionMultiply_5}
    B {from=MaterialExpressionSubtract_0}
   ]
  }
  MaterialExpressionAdd_43 {
   class=MaterialExpressionAdd
   inputs [
    A {from=MaterialExpressionStaticSwitch_6}
    B {from=MaterialExpressionNamedRerouteUsage_10}
   ]
  }
  MaterialExpressionAdd_44 {
   class=MaterialExpressionAdd
   inputs [
    A {from=MaterialExpressionStaticSwitch_7}
    B {from=MaterialExpressionNamedRerouteUsage_31}
   ]
  }
  MaterialExpressionAdd_46 {
   class=MaterialExpressionAdd
   inputs [
    A {from=MaterialExpressionIf_1}
    B {from=MaterialExpressionNamedRerouteUsage_27}
   ]
  }
  MaterialExpressionAdd_5 {
   class=MaterialExpressionAdd
   inputs [
    A {from=MaterialExpressionMultiply_1}
    B {from=MaterialExpressionSubtract_1}
   ]
  }
  MaterialExpressionAdd_51 {
   class=MaterialExpressionAdd
   inputs [
    A {from=MaterialExpressionMultiply_78}
    B {from=MaterialExpressionSubtract_19}
   ]
  }
  MaterialExpressionAdd_6 {
   class=MaterialExpressionAdd
   inputs [
    A {from=MaterialExpressionScalarParameter_5}
    B {from=MaterialExpressionNamedRerouteUsage_43}
   ]
   properties {constB=30}
  }
  MaterialExpressionAdd_7 {
   class=MaterialExpressionAdd
   inputs [
    A {from=MaterialExpressionIf_10}
    B {from=MaterialExpressionNamedRerouteUsage_22}
   ]
  }
  MaterialExpressionComponentMask_12 {class=MaterialExpressionComponentMask, inputs [None {from=MaterialExpressionTransform_4}], properties {b=true}}
  MaterialExpressionComponentMask_14 {class=MaterialExpressionComponentMask, inputs [None {from=MaterialExpressionVertexNormalWS_0}], properties {b=true}}
  MaterialExpressionConstant_0 {class=MaterialExpressionConstant, properties {r=1}}
  MaterialExpressionConstant_1 {class=MaterialExpressionConstant, properties {r=-0.2}}
  MaterialExpressionConstant_10 {class=MaterialExpressionConstant, properties {r=1}}
  MaterialExpressionConstant_11 {class=MaterialExpressionConstant, properties {r=1}}
  MaterialExpressionConstant_12 {class=MaterialExpressionConstant, properties {r=1}}
  MaterialExpressionConstant_13 {class=MaterialExpressionConstant, properties {r=1}}
  MaterialExpressionConstant_14 {class=MaterialExpressionConstant}
  MaterialExpressionConstant_15 {class=MaterialExpressionConstant, properties {r=1}}
  MaterialExpressionConstant_16 {class=MaterialExpressionConstant, properties {r=2}}
  MaterialExpressionConstant_17 {class=MaterialExpressionConstant, properties {r=1}}
  MaterialExpressionConstant_18 {class=MaterialExpressionConstant}
  MaterialExpressionConstant_19 {class=MaterialExpressionConstant, properties {r=1}}
  MaterialExpressionConstant_2 {class=MaterialExpressionConstant, properties {r=1.5}}
  MaterialExpressionConstant_20 {class=MaterialExpressionConstant, properties {r=1}}
  MaterialExpressionConstant_21 {class=MaterialExpressionConstant, properties {r=1}}
  MaterialExpressionConstant_22 {class=MaterialExpressionConstant, properties {r=1}}
  MaterialExpressionConstant_23 {class=MaterialExpressionConstant}
  MaterialExpressionConstant_24 {class=MaterialExpressionConstant, properties {r=1}}
  MaterialExpressionConstant_25 {class=MaterialExpressionConstant, properties {r=1}}
  MaterialExpressionConstant_26 {class=MaterialExpressionConstant, properties {r=1}}
  MaterialExpressionConstant_27 {class=MaterialExpressionConstant}
  MaterialExpressionConstant_28 {class=MaterialExpressionConstant, properties {r=1}}
  MaterialExpressionConstant_29 {class=MaterialExpressionConstant, properties {r=1}}
  MaterialExpressionConstant_3 {class=MaterialExpressionConstant}
  MaterialExpressionConstant_30 {class=MaterialExpressionConstant, properties {r=1}}
  MaterialExpressionConstant_31 {class=MaterialExpressionConstant}
  MaterialExpressionConstant_32 {class=MaterialExpressionConstant, properties {r=1}}
  MaterialExpressionConstant_33 {class=MaterialExpressionConstant}
  MaterialExpressionConstant_34 {class=MaterialExpressionConstant}
  MaterialExpressionConstant_35 {class=MaterialExpressionConstant, properties {r=1}}
  MaterialExpressionConstant_36 {class=MaterialExpressionConstant, properties {r=1}}
  MaterialExpressionConstant_37 {class=MaterialExpressionConstant, properties {r=1}}
  MaterialExpressionConstant_38 {class=MaterialExpressionConstant, properties {r=1}}
  MaterialExpressionConstant_39 {class=MaterialExpressionConstant}
  MaterialExpressionConstant_4 {class=MaterialExpressionConstant}
  MaterialExpressionConstant_40 {class=MaterialExpressionConstant}
  MaterialExpressionConstant_5 {class=MaterialExpressionConstant, properties {r=1}}
  MaterialExpressionConstant_6 {class=MaterialExpressionConstant}
  MaterialExpressionConstant_7 {class=MaterialExpressionConstant, properties {r=1}}
  MaterialExpressionConstant_8 {class=MaterialExpressionConstant}
  MaterialExpressionConstant_9 {class=MaterialExpressionConstant, properties {r=1}}
  MaterialExpressionConvert_1 {
   class=MaterialExpressionConvert
   inputs [
    None {from=MaterialExpressionNamedRerouteUsage_90}
    None {from=MaterialExpressionNamedRerouteUsage_91}
    None {from=MaterialExpressionNamedRerouteUsage_92}
   ]
   properties {
    nodeName=Make Vector3
    convertInputs [
     {
      type=Scalar
      expressionInput {
       expression=&self:MaterialExpressionNamedRerouteUsage_90
       inputName=None
       mask=0
       maskA=0
       maskB=0
       maskG=0
       maskR=0
       outputIndex=0
      }
     }
     {
      type=Scalar
      expressionInput {
       expression=&self:MaterialExpressionNamedRerouteUsage_91
       inputName=None
       mask=0
       maskA=0
       maskB=0
       maskG=0
       maskR=0
       outputIndex=0
      }
     }
     {
      type=Scalar
      expressionInput {
       expression=&self:MaterialExpressionNamedRerouteUsage_92
       inputName=None
       mask=0
       maskA=0
       maskB=0
       maskG=0
       maskR=0
       outputIndex=0
      }
     }
    ]
    convertOutputs [{type=Vector3}]
   }
  }
  MaterialExpressionDotProduct_0 {
   class=MaterialExpressionDotProduct
   inputs [
    A {from=MaterialExpressionNamedRerouteUsage_48, output=""}
    B {from=MaterialExpressionNamedRerouteUsage_89, output=""}
   ]
   properties {}
  }
  MaterialExpressionDotProduct_1 {
   class=MaterialExpressionDotProduct
   inputs [
    A {from=MaterialExpressionNamedRerouteUsage_39, output=""}
    B {from=MaterialExpressionNamedRerouteUsage_93, output=""}
   ]
   properties {}
  }
  MaterialExpressionDotProduct_2 {
   class=MaterialExpressionDotProduct
   inputs [
    A {from=MaterialExpressionNamedRerouteUsage_13, output=""}
    B {from=MaterialExpressionVertexColor_1, output=""}
   ]
   properties {}
  }
  MaterialExpressionFunctionInput_0 {
   class=MaterialExpressionFunctionInput
   inputs [Preview {from=MaterialExpressionConstant_0, output=""}]
   properties {bUsePreviewValueAsDefault=true, inputName="• VTX Color - Intensity", inputType=FunctionInput_Scalar, sortPriority=176}
  }
  MaterialExpressionFunctionInput_1 {
   class=MaterialExpressionFunctionInput
   inputs [Preview {from=MaterialExpressionStaticBool_0, output=""}]
   properties {bUsePreviewValueAsDefault=true, inputName=--- Use HEIGHT MASKING ---, inputType=FunctionInput_StaticBool, sortPriority=500}
  }
  MaterialExpressionFunctionInput_10 {
   class=MaterialExpressionFunctionInput
   inputs [Preview {from="", output=""}]
   properties {inputName=DFAO, inputType=FunctionInput_Scalar, sortPriority=813}
  }
  MaterialExpressionFunctionInput_11 {
   class=MaterialExpressionFunctionInput
   inputs [Preview {from=MaterialExpressionSetMaterialAttributes_0, output=""}]
   properties {
    description=MaterialAttributes from previous layers
    inputName=Layer A MaterialAttributes
    inputType=FunctionInput_MaterialAttributes
    sortPriority=-100
   }
  }
  MaterialExpressionFunctionInput_12 {
   class=MaterialExpressionFunctionInput
   inputs [Preview {from="", output=""}]
   properties {inputName=Baked Curvature, inputType=FunctionInput_Scalar, sortPriority=810}
  }
  MaterialExpressionFunctionInput_13 {
   class=MaterialExpressionFunctionInput
   inputs [Preview {from=MaterialExpressionConstant_12, output=""}]
   properties {bUsePreviewValueAsDefault=true, inputName="• Height - Blend (1=Blend 0=Min)", inputType=FunctionInput_Scalar, sortPriority=510}
  }
  MaterialExpressionFunctionInput_14 {
   class=MaterialExpressionFunctionInput
   inputs [Preview {from="", output=""}]
   properties {inputName=Z Gradient, inputType=FunctionInput_Scalar, sortPriority=814}
  }
  MaterialExpressionFunctionInput_15 {
   class=MaterialExpressionFunctionInput
   inputs [Preview {from="", output=""}]
   properties {inputName=Baked AO, inputType=FunctionInput_Scalar, sortPriority=811}
  }
  MaterialExpressionFunctionInput_16 {
   class=MaterialExpressionFunctionInput
   inputs [Preview {from="", output=""}]
   properties {inputName=Painted Mask, inputType=FunctionInput_Scalar, sortPriority=812}
  }
  MaterialExpressionFunctionInput_17 {
   class=MaterialExpressionFunctionInput
   inputs [Preview {from=MaterialExpressionSetMaterialAttributes_0, output=""}]
   properties {description=_Input B MaterialAttributes, inputName=Layer B MaterialAttributes, inputType=FunctionInput_MaterialAttributes, sortPriority=-99}
  }
  MaterialExpressionFunctionInput_18 {
   class=MaterialExpressionFunctionInput
   inputs [Preview {from=MaterialExpressionConstant_13, output=""}]
   properties {
    bUsePreviewValueAsDefault=true
    inputName="• Height - Selection (1=Current Layer 0=Previous Layers)"
    inputType=FunctionInput_Scalar
    sortPriority=509
   }
  }
  MaterialExpressionFunctionInput_19 {
   class=MaterialExpressionFunctionInput
   inputs [Preview {from=MaterialExpressionStaticBool_1, output=""}]
   properties {bUsePreviewValueAsDefault=true, inputName=--- Use WATERLINE MASKING ---, inputType=FunctionInput_StaticBool, sortPriority=400}
  }
  MaterialExpressionFunctionInput_2 {
   class=MaterialExpressionFunctionInput
   inputs [Preview {from=MaterialExpressionConstant_3, output=""}]
   properties {bUsePreviewValueAsDefault=true, inputName="• Height - Bias", inputType=FunctionInput_Scalar, sortPriority=515}
  }
  MaterialExpressionFunctionInput_20 {
   class=MaterialExpressionFunctionInput
   inputs [Preview {from=MaterialExpressionConstant_27, output=""}]
   properties {bUsePreviewValueAsDefault=true, inputName=* Waterline - Bias, inputType=FunctionInput_Scalar, sortPriority=415}
  }
  MaterialExpressionFunctionInput_21 {
   class=MaterialExpressionFunctionInput
   inputs [Preview {from=MaterialExpressionConstant_28, output=""}]
   properties {bUsePreviewValueAsDefault=true, inputName="* Waterline - Blend (1=Max 0=Min)", inputType=FunctionInput_Scalar, sortPriority=410}
  }
  MaterialExpressionFunctionInput_22 {
   class=MaterialExpressionFunctionInput
   inputs [Preview {from=MaterialExpressionConstant_29, output=""}]
   properties {bUsePreviewValueAsDefault=true, inputName=* Waterline - Intensity, inputType=FunctionInput_Scalar, sortPriority=416}
  }
  MaterialExpressionFunctionInput_23 {
   class=MaterialExpressionFunctionInput
   inputs [Preview {from=MaterialExpressionConstant_30, output=""}]
   properties {bUsePreviewValueAsDefault=true, inputName="• Baked AO/Curvature/Painted - Invert", inputType=FunctionInput_Scalar, sortPriority=361}
  }
  MaterialExpressionFunctionInput_24 {
   class=MaterialExpressionFunctionInput
   inputs [Preview {from=MaterialExpressionConstant_31, output=""}]
   properties {bUsePreviewValueAsDefault=true, inputName="• WS Noise - Invert", inputType=FunctionInput_Scalar, sortPriority=461}
  }
  MaterialExpressionFunctionInput_25 {
   class=MaterialExpressionFunctionInput
   inputs [Preview {from=MaterialExpressionConstant_32, output=""}]
   properties {bUsePreviewValueAsDefault=true, inputName="• VTX Color - Invert", inputType=FunctionInput_Scalar, sortPriority=161}
  }
  MaterialExpressionFunctionInput_26 {class=MaterialExpressionFunctionInput, inputs [Preview {from="", output=""}], properties {inputName=WS Noise Masks, sortPriority=815}}
  MaterialExpressionFunctionInput_27 {
   class=MaterialExpressionFunctionInput
   inputs [Preview {from=MaterialExpressionConstant_33, output=""}]
   properties {bUsePreviewValueAsDefault=true, inputName="• VTX Color - Bias", inputType=FunctionInput_Scalar, sortPriority=175}
  }
  MaterialExpressionFunctionInput_28 {
   class=MaterialExpressionFunctionInput
   inputs [Preview {from=MaterialExpressionConstant_34, output=""}]
   properties {bUsePreviewValueAsDefault=true, inputName="• Height - Min", inputType=FunctionInput_Scalar, sortPriority=512}
  }
  MaterialExpressionFunctionInput_29 {
   class=MaterialExpressionFunctionInput
   inputs [Preview {from=MaterialExpressionConstant_35, output=""}]
   properties {bUsePreviewValueAsDefault=true, inputName="• Height - Max", inputType=FunctionInput_Scalar, sortPriority=513}
  }
  MaterialExpressionFunctionInput_3 {
   class=MaterialExpressionFunctionInput
   inputs [Preview {from=MaterialExpressionStaticBool_8, output=""}]
   properties {bUsePreviewValueAsDefault=true, inputName=--- Use VTX COLOR MASKING ---, inputType=FunctionInput_StaticBool, sortPriority=150}
  }
  MaterialExpressionFunctionInput_30 {
   class=MaterialExpressionFunctionInput
   inputs [Preview {from=MaterialExpressionConstant_36, output=""}]
   properties {
    bUsePreviewValueAsDefault=true
    inputName=* Realtime Displacement - WS Noise - Contribution
    inputType=FunctionInput_Scalar
    sortPriority=560
   }
  }
  MaterialExpressionFunctionInput_31 {
   class=MaterialExpressionFunctionInput
   inputs [Preview {from=MaterialExpressionConstant_37, output=""}]
   properties {bUsePreviewValueAsDefault=true, inputName=* Realtime Displacement - Height - Contribution, inputType=FunctionInput_Scalar, sortPriority=565}
  }
  MaterialExpressionFunctionInput_32 {
   class=MaterialExpressionFunctionInput
   inputs [Preview {from=MaterialExpressionConstant_38, output=""}]
   properties {bUsePreviewValueAsDefault=true, inputName=* Realtime Displacement - Masking - Intensity, inputType=FunctionInput_Scalar, sortPriority=555}
  }
  MaterialExpressionFunctionInput_33 {
   class=MaterialExpressionFunctionInput
   inputs [Preview {from=MaterialExpressionConstant_39, output=""}]
   properties {bUsePreviewValueAsDefault=true, inputName=* Realtime Displacement - Masking - Bias, inputType=FunctionInput_Scalar, sortPriority=550}
  }
  MaterialExpressionFunctionInput_34 {
   class=MaterialExpressionFunctionInput
   inputs [Preview {from=MaterialExpressionConstant_26, output=""}]
   properties {bUsePreviewValueAsDefault=true, inputName="• VTX Color - Channel Weights", sortPriority=160}
  }
  MaterialExpressionFunctionInput_4 {
   class=MaterialExpressionFunctionInput
   inputs [Preview {from=MaterialExpressionConstant_11, output=""}]
   properties {bUsePreviewValueAsDefault=true, inputName="• Height - Intensity", inputType=FunctionInput_Scalar, sortPriority=516}
  }
  MaterialExpressionFunctionInput_40 {
   class=MaterialExpressionFunctionInput
   inputs [Preview {from=MaterialExpressionStaticBool_19, output=""}]
   properties {bUsePreviewValueAsDefault=true, inputName=* DFAO - Force Baked AO on Nanite, inputType=FunctionInput_StaticBool, sortPriority=210}
  }
  MaterialExpressionFunctionInput_41 {
   class=MaterialExpressionFunctionInput
   inputs [Preview {from=MaterialExpressionStaticBool_20, output=""}]
   properties {bUsePreviewValueAsDefault=true, inputName=--- Use DFAO MASKING ---, inputType=FunctionInput_StaticBool, sortPriority=200}
  }
  MaterialExpressionFunctionInput_42 {
   class=MaterialExpressionFunctionInput
   inputs [Preview {from=MaterialExpressionConstant_4, output=""}]
   properties {bUsePreviewValueAsDefault=true, inputName="• DFAO - Bias", inputType=FunctionInput_Scalar, sortPriority=225}
  }
  MaterialExpressionFunctionInput_43 {
   class=MaterialExpressionFunctionInput
   inputs [Preview {from=MaterialExpressionConstant_5, output=""}]
   properties {bUsePreviewValueAsDefault=true, inputName="• DFAO - Intensity", inputType=FunctionInput_Scalar, sortPriority=226}
  }
  MaterialExpressionFunctionInput_48 {
   class=MaterialExpressionFunctionInput
   inputs [Preview {from=MaterialExpressionStaticBool_25, output=""}]
   properties {bUsePreviewValueAsDefault=true, inputName=--- Use Z GRADIENT MASKING ---, inputType=FunctionInput_StaticBool, sortPriority=250}
  }
  MaterialExpressionFunctionInput_49 {
   class=MaterialExpressionFunctionInput
   inputs [Preview {from=MaterialExpressionConstant_6, output=""}]
   properties {bUsePreviewValueAsDefault=true, inputName="• Z Gradient - Height - Bottom", inputType=FunctionInput_Scalar, sortPriority=265}
  }
  MaterialExpressionFunctionInput_5 {
   class=MaterialExpressionFunctionInput
   inputs [Preview {from=MaterialExpressionConstant_1, output=""}]
   properties {bUsePreviewValueAsDefault=true, inputName=* World Z Normal - Bias, inputType=FunctionInput_Scalar, sortPriority=120}
  }
  MaterialExpressionFunctionInput_50 {
   class=MaterialExpressionFunctionInput
   inputs [Preview {from=MaterialExpressionConstant_7, output=""}]
   properties {bUsePreviewValueAsDefault=true, inputName="• Z Gradient - Height - Top", inputType=FunctionInput_Scalar, sortPriority=266}
  }
  MaterialExpressionFunctionInput_54 {
   class=MaterialExpressionFunctionInput
   inputs [Preview {from=MaterialExpressionStaticBool_29, output=""}]
   properties {
    bUsePreviewValueAsDefault=true
    inputName=--- Use BAKED AO/CURVATURE/PAINTED MASKING ---
    inputType=FunctionInput_StaticBool
    sortPriority=350
   }
  }
  MaterialExpressionFunctionInput_55 {
   class=MaterialExpressionFunctionInput
   inputs [Preview {from=MaterialExpressionConstant_8, output=""}]
   properties {bUsePreviewValueAsDefault=true, inputName="• Baked AO/Curvature/Painted - Bias", inputType=FunctionInput_Scalar, sortPriority=375}
  }
  MaterialExpressionFunctionInput_56 {
   class=MaterialExpressionFunctionInput
   inputs [Preview {from=MaterialExpressionConstant_9, output=""}]
   properties {bUsePreviewValueAsDefault=true, inputName="• Baked AO/Curvature/Painted - Intensity", inputType=FunctionInput_Scalar, sortPriority=376}
  }
  MaterialExpressionFunctionInput_57 {
   class=MaterialExpressionFunctionInput
   inputs [Preview {from=MaterialExpressionConstant_10, output=""}]
   properties {bUsePreviewValueAsDefault=true, inputName="• Baked AO/Curvature/Painted - Channel Weights", sortPriority=360}
  }
  MaterialExpressionFunctionInput_6 {
   class=MaterialExpressionFunctionInput
   inputs [Preview {from=MaterialExpressionStaticBool_9, output=""}]
   properties {bUsePreviewValueAsDefault=true, inputName="• World Z Normal - Use VTX Normals", inputType=FunctionInput_StaticBool, sortPriority=110}
  }
  MaterialExpressionFunctionInput_61 {
   class=MaterialExpressionFunctionInput
   inputs [Preview {from=MaterialExpressionStaticBool_33, output=""}]
   properties {bUsePreviewValueAsDefault=true, inputName=--- Use WS NOISE MASKING ---, inputType=FunctionInput_StaticBool, sortPriority=450}
  }
  MaterialExpressionFunctionInput_65 {
   class=MaterialExpressionFunctionInput
   inputs [Preview {from=MaterialExpressionConstant_14, output=""}]
   properties {bUsePreviewValueAsDefault=true, inputName="• WS Noise - Bias", inputType=FunctionInput_Scalar, sortPriority=475}
  }
  MaterialExpressionFunctionInput_66 {
   class=MaterialExpressionFunctionInput
   inputs [Preview {from=MaterialExpressionConstant_15, output=""}]
   properties {bUsePreviewValueAsDefault=true, inputName="• WS Noise - Intensity", inputType=FunctionInput_Scalar, sortPriority=476}
  }
  MaterialExpressionFunctionInput_67 {
   class=MaterialExpressionFunctionInput
   inputs [Preview {from=MaterialExpressionConstant_17, output=""}]
   properties {bUsePreviewValueAsDefault=true, inputName="• Masking - Intensity", inputType=FunctionInput_Scalar, sortPriority=61}
  }
  MaterialExpressionFunctionInput_68 {
   class=MaterialExpressionFunctionInput
   inputs [Preview {from=MaterialExpressionConstant_18, output=""}]
   properties {bUsePreviewValueAsDefault=true, inputName="• Masking - Bias", inputType=FunctionInput_Scalar, sortPriority=60}
  }
  MaterialExpressionFunctionInput_69 {
   class=MaterialExpressionFunctionInput
   inputs [Preview {from=MaterialExpressionConstant_40, output=""}]
   properties {bUsePreviewValueAsDefault=true, inputName=* Masking - Invert, inputType=FunctionInput_Scalar, sortPriority=50}
  }
  MaterialExpressionFunctionInput_7 {
   class=MaterialExpressionFunctionInput
   inputs [Preview {from=MaterialExpressionStaticBool_11, output=""}]
   properties {bUsePreviewValueAsDefault=true, inputName=--- Use WORLD Z NORMAL MASKING ---, inputType=FunctionInput_StaticBool, sortPriority=100}
  }
  MaterialExpressionFunctionInput_70 {
   class=MaterialExpressionFunctionInput
   inputs [Preview {from=MaterialExpressionConstant_19, output=""}]
   properties {bUsePreviewValueAsDefault=true, inputName="• VTX Color - Blend (1=Max 0=Min)", inputType=FunctionInput_Scalar, sortPriority=170}
  }
  MaterialExpressionFunctionInput_71 {
   class=MaterialExpressionFunctionInput
   inputs [Preview {from=MaterialExpressionConstant_20, output=""}]
   properties {bUsePreviewValueAsDefault=true, inputName="• DFAO - Blend (1=Max 0=Min)", inputType=FunctionInput_Scalar, sortPriority=220}
  }
  MaterialExpressionFunctionInput_72 {
   class=MaterialExpressionFunctionInput
   inputs [Preview {from=MaterialExpressionConstant_21, output=""}]
   properties {bUsePreviewValueAsDefault=true, inputName="• Z Gradient - Blend (1=Max 0=Min)", inputType=FunctionInput_Scalar, sortPriority=260}
  }
  MaterialExpressionFunctionInput_73 {
   class=MaterialExpressionFunctionInput
   inputs [Preview {from=MaterialExpressionConstant_22, output=""}]
   properties {
    bUsePreviewValueAsDefault=true
    inputName="• Baked AO/Curvature/Painted - Blend (1=Max 0=Min)"
    inputType=FunctionInput_Scalar
    sortPriority=370
   }
  }
  MaterialExpressionFunctionInput_74 {
   class=MaterialExpressionFunctionInput
   inputs [Preview {from=MaterialExpressionConstant_24, output=""}]
   properties {bUsePreviewValueAsDefault=true, inputName="• WS Noise - Blend (1=Blend 0=Min)", inputType=FunctionInput_Scalar, sortPriority=470}
  }
  MaterialExpressionFunctionInput_75 {
   class=MaterialExpressionFunctionInput
   inputs [Preview {from=MaterialExpressionConstant_25, output=""}]
   properties {bUsePreviewValueAsDefault=true, inputName="• WS Noise - Channel Weights", sortPriority=460}
  }
  MaterialExpressionFunctionInput_8 {
   class=MaterialExpressionFunctionInput
   inputs [Preview {from="", output=""}]
   properties {inputName="\\ --- Dependencies --- /", inputType=FunctionInput_Scalar, sortPriority=800}
  }
  MaterialExpressionFunctionInput_9 {
   class=MaterialExpressionFunctionInput
   inputs [Preview {from=MaterialExpressionConstant_2, output=""}]
   properties {bUsePreviewValueAsDefault=true, inputName=* World Z Normal - Intensity, inputType=FunctionInput_Scalar, sortPriority=121}
  }
  MaterialExpressionFunctionOutput_0 {
   class=MaterialExpressionFunctionOutput
   inputs [None {from=MaterialExpressionNamedRerouteUsage_105, output=""}]
   properties {outputName=Combined Realtime Displacement Mask, sortPriority=10}
  }
  MaterialExpressionFunctionOutput_1 {
   class=MaterialExpressionFunctionOutput
   inputs [None {from=MaterialExpressionNamedRerouteUsage_45, output=""}]
   properties {outputName=Combined Mask}
  }
  MaterialExpressionGetMaterialAttributes_16 {
   class=MaterialExpressionGetMaterialAttributes
   inputs [None {from=MaterialExpressionNamedRerouteUsage_1, output=""}]
   properties {attributeGetTypes [#1]}
  }
  MaterialExpressionGetMaterialAttributes_2 {
   class=MaterialExpressionGetMaterialAttributes
   inputs [None {from=MaterialExpressionNamedRerouteUsage_100, output=""}]
   properties {attributeGetTypes [#2]}
  }
  MaterialExpressionGetMaterialAttributes_9 {
   class=MaterialExpressionGetMaterialAttributes
   inputs [None {from=MaterialExpressionNamedRerouteUsage_303, output=""}]
   properties {attributeGetTypes [#2]}
  }
  MaterialExpressionIf_0 {
   class=MaterialExpressionIf
   inputs [
    A {from=MaterialExpressionNamedRerouteUsage_37, output=""}
    B {from="", output=""}
    "A > B" {from=MaterialExpressionMax_3, output=""}
    "A == B" {from="", output=""}
    "A < B" {from=MaterialExpressionMin_2, output=""}
   ]
   properties {constB=0.5, desc="cheaper than lerp without \"==\" condition", equalsThreshold=0.001}
  }
  MaterialExpressionIf_1 {
   class=MaterialExpressionIf
   inputs [
    A {from=MaterialExpressionNamedRerouteUsage_94, output=""}
    B {from="", output=""}
    "A > B" {from=MaterialExpressionOneMinus_0, output=""}
    "A == B" {from="", output=""}
    "A < B" {from=MaterialExpressionDotProduct_1, output=""}
   ]
   properties {constB=0.5, desc="cheaper than lerp without \"==\" condition", equalsThreshold=0.001}
  }
  MaterialExpressionIf_10 {
   class=MaterialExpressionIf
   inputs [
    A {from=MaterialExpressionNamedRerouteUsage_25, output=""}
    B {from="", output=""}
    "A > B" {from=MaterialExpressionOneMinus_1, output=""}
    "A == B" {from="", output=""}
    "A < B" {from=MaterialExpressionReroute_51, output=""}
   ]
   properties {constB=0.5, desc="cheaper than lerp without \"==\" condition", equalsThreshold=0.001}
  }
  MaterialExpressionIf_11 {
   class=MaterialExpressionIf
   inputs [
    A {from=MaterialExpressionNamedRerouteUsage_8, output=""}
    B {from="", output=""}
    "A > B" {from=MaterialExpressionAdd_5, output=""}
    "A == B" {from="", output=""}
    "A < B" {from="", output=""}
   ]
   properties {constB=0.5, desc="cheaper than lerp without \"==\" condition", equalsThreshold=0.001}
  }
  MaterialExpressionIf_13 {
   class=MaterialExpressionIf
   inputs [
    A {from=MaterialExpressionNamedRerouteUsage_95, output=""}
    B {from="", output=""}
    "A > B" {from=MaterialExpressionOneMinus_3, output=""}
    "A == B" {from="", output=""}
    "A < B" {from=MaterialExpressionDotProduct_0, output=""}
   ]
   properties {constB=0.5, desc="cheaper than lerp without \"==\" condition", equalsThreshold=0.001}
  }
  MaterialExpressionIf_2 {
   class=MaterialExpressionIf
   inputs [
    A {from=MaterialExpressionNamedRerouteUsage_36, output=""}
    B {from="", output=""}
    "A > B" {from=MaterialExpressionMax_22, output=""}
    "A == B" {from="", output=""}
    "A < B" {from=MaterialExpressionMin_21, output=""}
   ]
   properties {constB=0.5, desc="cheaper than lerp without \"==\" condition", equalsThreshold=0.001}
  }
  MaterialExpressionIf_3 {
   class=MaterialExpressionIf
   inputs [
    A {from=MaterialExpressionNamedRerouteUsage_38, output=""}
    B {from="", output=""}
    "A > B" {from=MaterialExpressionMax_19, output=""}
    "A == B" {from="", output=""}
    "A < B" {from=MaterialExpressionMin_18, output=""}
   ]
   properties {constB=0.5, desc="cheaper than lerp without \"==\" condition", equalsThreshold=0.001}
  }
  MaterialExpressionIf_4 {
   class=MaterialExpressionIf
   inputs [
    A {from=MaterialExpressionNamedRerouteUsage_96, output=""}
    B {from="", output=""}
    "A > B" {from=MaterialExpressionOneMinus_4, output=""}
    "A == B" {from="", output=""}
    "A < B" {from=MaterialExpressionDotProduct_2, output=""}
   ]
   properties {constB=0.5, desc="cheaper than lerp without \"==\" condition", equalsThreshold=0.001}
  }
  MaterialExpressionIf_5 {
   class=MaterialExpressionIf
   inputs [
    A {from=MaterialExpressionNamedRerouteUsage_33, output=""}
    B {from="", output=""}
    "A > B" {from=MaterialExpressionMax_20, output=""}
    "A == B" {from="", output=""}
    "A < B" {from=MaterialExpressionMin_19, output=""}
   ]
   properties {constB=0.5, desc="cheaper than lerp without \"==\" condition", equalsThreshold=0.001}
  }
  MaterialExpressionIf_6 {
   class=MaterialExpressionIf
   inputs [
    A {from=MaterialExpressionNamedRerouteUsage_34, output=""}
    B {from="", output=""}
    "A > B" {from=MaterialExpressionMax_0, output=""}
    "A == B" {from="", output=""}
    "A < B" {from=MaterialExpressionMin_0, output=""}
   ]
   properties {constB=0.5, desc="cheaper than lerp without \"==\" condition", equalsThreshold=0.001}
  }
  MaterialExpressionIf_7 {
   class=MaterialExpressionIf
   inputs [
    A {from=MaterialExpressionNamedRerouteUsage_35, output=""}
    B {from="", output=""}
    "A > B" {from=MaterialExpressionAdd_4, output=""}
    "A == B" {from="", output=""}
    "A < B" {from=MaterialExpressionMin_24, output=""}
   ]
   properties {constB=0.5, desc="cheaper than lerp without \"==\" condition", equalsThreshold=0.001}
  }
  MaterialExpressionIf_8 {
   class=MaterialExpressionIf
   inputs [
    A {from=MaterialExpressionNamedRerouteUsage_306, output=""}
    B {from="", output=""}
    "A > B" {from=MaterialExpressionAdd_51, output=""}
    "A == B" {from="", output=""}
    "A < B" {from=MaterialExpressionMin_26, output=""}
   ]
   properties {constB=0.5, desc="cheaper than lerp without \"==\" condition", equalsThreshold=0.001}
  }
  MaterialExpressionIf_9 {
   class=MaterialExpressionIf
   inputs [
    A {from=MaterialExpressionNamedRerouteUsage_47, output=""}
    B {from="", output=""}
    "A > B" {from=MaterialExpressionOneMinus_31, output=""}
    "A == B" {from="", output=""}
    "A < B" {from=MaterialExpressionReroute_235, output=""}
   ]
   properties {constB=0.5, desc="cheaper than lerp without \"==\" condition", equalsThreshold=0.001}
  }
  MaterialExpressionLinearInterpolate_11 {
   class=MaterialExpressionLinearInterpolate
   inputs [
    A {from=MaterialExpressionReroute_9}
    B {from=MaterialExpressionIf_11}
    Alpha {from=MaterialExpressionNamedRerouteUsage_26}
   ]
  }
  MaterialExpressionLinearInterpolate_31 {
   class=MaterialExpressionLinearInterpolate
   inputs [
    A {from=MaterialExpressionNamedRerouteUsage_87}
    B {from=MaterialExpressionNamedRerouteUsage_88}
    Alpha {from=MaterialExpressionLinearInterpolate_37}
   ]
  }
  MaterialExpressionLinearInterpolate_37 {
   class=MaterialExpressionLinearInterpolate
   inputs [
    A {from=MaterialExpressionReroute_214}
    B {from=MaterialExpressionReroute_65}
    Alpha {from=MaterialExpressionNamedRerouteUsage_304}
   ]
  }
  MaterialExpressionLinearInterpolate_5 {
   class=MaterialExpressionLinearInterpolate
   inputs [
    A {from=MaterialExpressionReroute_50}
    B {from=MaterialExpressionIf_7}
    Alpha {from=MaterialExpressionNamedRerouteUsage_28}
   ]
  }
  MaterialExpressionMax_0 {
   class=MaterialExpressionMax
   inputs [
    A {from=MaterialExpressionReroute_4}
    B {from=MaterialExpressionReroute_11}
   ]
  }
  MaterialExpressionMax_19 {
   class=MaterialExpressionMax
   inputs [
    A {from=MaterialExpressionReroute_237}
    B {from=MaterialExpressionReroute_239}
   ]
  }
  MaterialExpressionMax_20 {
   class=MaterialExpressionMax
   inputs [
    A {from=MaterialExpressionReroute_241}
    B {from=MaterialExpressionReroute_242}
   ]
  }
  MaterialExpressionMax_22 {
   class=MaterialExpressionMax
   inputs [
    A {from=MaterialExpressionReroute_249}
    B {from=MaterialExpressionReroute_250}
   ]
  }
  MaterialExpressionMax_3 {
   class=MaterialExpressionMax
   inputs [
    A {from=MaterialExpressionReroute_143}
    B {from=MaterialExpressionReroute_146}
   ]
  }
  MaterialExpressionMin_0 {
   class=MaterialExpressionMin
   inputs [
    A {from=MaterialExpressionReroute_4, output=""}
    B {from=MaterialExpressionReroute_11, output=""}
   ]
   properties {}
  }
  MaterialExpressionMin_1 {
   class=MaterialExpressionMin
   inputs [
    A {from=MaterialExpressionReroute_38, output=""}
    B {from=MaterialExpressionReroute_48, output=""}
   ]
   properties {}
  }
  MaterialExpressionMin_18 {
   class=MaterialExpressionMin
   inputs [
    A {from=MaterialExpressionReroute_237, output=""}
    B {from=MaterialExpressionReroute_239, output=""}
   ]
   properties {}
  }
  MaterialExpressionMin_19 {
   class=MaterialExpressionMin
   inputs [
    A {from=MaterialExpressionReroute_241, output=""}
    B {from=MaterialExpressionReroute_242, output=""}
   ]
   properties {}
  }
  MaterialExpressionMin_2 {
   class=MaterialExpressionMin
   inputs [
    A {from=MaterialExpressionReroute_143, output=""}
    B {from=MaterialExpressionReroute_146, output=""}
   ]
   properties {}
  }
  MaterialExpressionMin_21 {
   class=MaterialExpressionMin
   inputs [
    A {from=MaterialExpressionReroute_249, output=""}
    B {from=MaterialExpressionReroute_250, output=""}
   ]
   properties {}
  }
  MaterialExpressionMin_24 {
   class=MaterialExpressionMin
   inputs [
    A {from=MaterialExpressionReroute_261, output=""}
    B {from=MaterialExpressionReroute_262, output=""}
   ]
   properties {}
  }
  MaterialExpressionMin_26 {
   class=MaterialExpressionMin
   inputs [
    A {from=MaterialExpressionReroute_210, output=""}
    B {from=MaterialExpressionReroute_211, output=""}
   ]
   properties {}
  }
  MaterialExpressionMultiply_0 {
   class=MaterialExpressionMultiply
   inputs [
    A {from=MaterialExpressionAdd_0}
    B {from=MaterialExpressionNamedRerouteUsage_19}
   ]
  }
  MaterialExpressionMultiply_1 {
   class=MaterialExpressionMultiply
   inputs [
    A {from=MaterialExpressionReroute_38}
    B {from=""}
   ]
   properties {constB=2, desc=2}
  }
  MaterialExpressionMultiply_2 {
   class=MaterialExpressionMultiply
   inputs [
    A {from=MaterialExpressionMultiply_71}
    B {from=MaterialExpressionShadingPathSwitch_14}
   ]
  }
  MaterialExpressionMultiply_3 {
   class=MaterialExpressionMultiply
   inputs [
    A {from=MaterialExpressionAdd_1}
    B {from=MaterialExpressionNamedRerouteUsage_360}
   ]
  }
  MaterialExpressionMultiply_4 {
   class=MaterialExpressionMultiply
   inputs [
    A {from=MaterialExpressionAdd_2}
    B {from=MaterialExpressionNamedRerouteUsage_79}
   ]
   properties {constB=0.015}
  }
  MaterialExpressionMultiply_5 {
   class=MaterialExpressionMultiply
   inputs [
    A {from=MaterialExpressionReroute_261}
    B {from=""}
   ]
   properties {constB=2, desc=2}
  }
  MaterialExpressionMultiply_6 {
   class=MaterialExpressionMultiply
   inputs [
    A {from=MaterialExpressionAdd_7}
    B {from=MaterialExpressionNamedRerouteUsage_0}
   ]
  }
  MaterialExpressionMultiply_69 {
   class=MaterialExpressionMultiply
   inputs [
    A {from=MaterialExpressionAdd_43}
    B {from=MaterialExpressionNamedRerouteUsage_11}
   ]
  }
  MaterialExpressionMultiply_70 {
   class=MaterialExpressionMultiply
   inputs [
    A {from=MaterialExpressionReroute_178}
    B {from=MaterialExpressionConstant_16}
   ]
   properties {desc=compensates for the less dense meshes on mobile creating bigger vertex gradients}
  }
  MaterialExpressionMultiply_71 {
   class=MaterialExpressionMultiply
   inputs [
    A {from=MaterialExpressionAdd_3}
    B {from=""}
   ]
   properties {constB=3}
  }
  MaterialExpressionMultiply_72 {
   class=MaterialExpressionMultiply
   inputs [
    A {from=MaterialExpressionAdd_44}
    B {from=MaterialExpressionNamedRerouteUsage_32}
   ]
  }
  MaterialExpressionMultiply_74 {
   class=MaterialExpressionMultiply
   inputs [
    A {from=MaterialExpressionAdd_46}
    B {from=MaterialExpressionNamedRerouteUsage_40}
   ]
  }
  MaterialExpressionMultiply_78 {
   class=MaterialExpressionMultiply
   inputs [
    A {from=MaterialExpressionReroute_210}
    B {from=""}
   ]
   properties {constB=2, desc=2}
  }
  MaterialExpressionNamedRerouteDeclaration_0 {
   class=MaterialExpressionNamedRerouteDeclaration
   inputs [None {from=MaterialExpressionFunctionInput_41}]
   properties {name=--- Use DFAO MASKING ---, nodeColor=<rgba 0.1979 0.0084 0 1>}
  }
  MaterialExpressionNamedRerouteDeclaration_1 {
   class=MaterialExpressionNamedRerouteDeclaration
   inputs [None {from=MaterialExpressionFunctionInput_1}]
   properties {name=--- Use HEIGHT MASKING ---, nodeColor=<rgba 0.1979 0.0084 0 1>}
  }
  MaterialExpressionNamedRerouteDeclaration_10 {
   class=MaterialExpressionNamedRerouteDeclaration
   inputs [None {from=MaterialExpressionFunctionInput_13}]
   properties {name="• Height - Blend (1=Blend 0=Min)", nodeColor=<rgba 0.5043 1 0 1>}
  }
  MaterialExpressionNamedRerouteDeclaration_11 {
   class=MaterialExpressionNamedRerouteDeclaration
   inputs [None {from=MaterialExpressionFunctionInput_10}]
   properties {name=_DFAO, nodeColor=<rgba 1 0 0.0525 1>}
  }
  MaterialExpressionNamedRerouteDeclaration_12 {
   class=MaterialExpressionNamedRerouteDeclaration
   inputs [None {from=MaterialExpressionFunctionInput_12}]
   properties {name=_Baked Curvature, nodeColor=<rgba 1 0 0.0525 1>}
  }
  MaterialExpressionNamedRerouteDeclaration_13 {
   class=MaterialExpressionNamedRerouteDeclaration
   inputs [None {from=MaterialExpressionFunctionInput_16}]
   properties {name=_Painted Mask, nodeColor=<rgba 1 0 0.0525 1>}
  }
  MaterialExpressionNamedRerouteDeclaration_14 {
   class=MaterialExpressionNamedRerouteDeclaration
   inputs [None {from=MaterialExpressionSaturate_4}]
   properties {name=Height, nodeColor=<rgba 0.5 0 1 1>}
  }
  MaterialExpressionNamedRerouteDeclaration_15 {
   class=MaterialExpressionNamedRerouteDeclaration
   inputs [None {from=MaterialExpressionReroute_54}]
   properties {name=_Input B MaterialAttributes, desc=MaterialAttributes from layer to be blended, nodeColor=<rgba 0.1016 0 1 1>}
  }
  MaterialExpressionNamedRerouteDeclaration_16 {
   class=MaterialExpressionNamedRerouteDeclaration
   inputs [None {from=MaterialExpressionFunctionInput_18}]
   properties {name="• Height - Selection (1=Current Layer 0=Previous Layers)", nodeColor=<rgba 0.5043 1 0 1>}
  }
  MaterialExpressionNamedRerouteDeclaration_17 {
   class=MaterialExpressionNamedRerouteDeclaration
   inputs [None {from=MaterialExpressionSaturate_2}]
   properties {name=Waterline, nodeColor=<rgba 1 0.8438 0 1>}
  }
  MaterialExpressionNamedRerouteDeclaration_18 {
   class=MaterialExpressionNamedRerouteDeclaration
   inputs [None {from=MaterialExpressionFunctionInput_19}]
   properties {name=--- Use WATERLINE MASKING ---, nodeColor=<rgba 0.1979 0.0084 0 1>}
  }
  MaterialExpressionNamedRerouteDeclaration_19 {
   class=MaterialExpressionNamedRerouteDeclaration
   inputs [None {from=MaterialExpressionFunctionInput_20}]
   properties {name=* Waterline - Bias, nodeColor=<rgba 0.5043 1 0 1>}
  }
  MaterialExpressionNamedRerouteDeclaration_2 {
   class=MaterialExpressionNamedRerouteDeclaration
   inputs [None {from=MaterialExpressionFunctionInput_4}]
   properties {name="• Height - Intensity", nodeColor=<rgba 0.5043 1 0 1>}
  }
  MaterialExpressionNamedRerouteDeclaration_20 {
   class=MaterialExpressionNamedRerouteDeclaration
   inputs [None {from=MaterialExpressionFunctionInput_21}]
   properties {name="* Waterline - Blend (1=Max 0=Min)", nodeColor=<rgba 0.5043 1 0 1>}
  }
  MaterialExpressionNamedRerouteDeclaration_21 {
   class=MaterialExpressionNamedRerouteDeclaration
   inputs [None {from=MaterialExpressionFunctionInput_22}]
   properties {name=* Waterline - Intensity, nodeColor=<rgba 0.5043 1 0 1>}
  }
  MaterialExpressionNamedRerouteDeclaration_22 {
   class=MaterialExpressionNamedRerouteDeclaration
   inputs [None {from=MaterialExpressionFunctionInput_26}]
   properties {name=_WS Noise Masks, nodeColor=<rgba 1 0 0.0525 1>}
  }
  MaterialExpressionNamedRerouteDeclaration_23 {
   class=MaterialExpressionNamedRerouteDeclaration
   inputs [None {from=MaterialExpressionConvert_1}]
   properties {name=Baked AO/Curvature/Painted Texture, nodeColor=<rgba 1 1 1 1>}
  }
  MaterialExpressionNamedRerouteDeclaration_24 {
   class=MaterialExpressionNamedRerouteDeclaration
   inputs [None {from=MaterialExpressionFunctionInput_23}]
   properties {name="• Baked AO/Curvature/Painted - Invert", nodeColor=<rgba 0.5043 1 0 1>}
  }
  MaterialExpressionNamedRerouteDeclaration_25 {
   class=MaterialExpressionNamedRerouteDeclaration
   inputs [None {from=MaterialExpressionFunctionInput_24}]
   properties {name="• WS Noise - Invert", nodeColor=<rgba 0.5043 1 0 1>}
  }
  MaterialExpressionNamedRerouteDeclaration_26 {
   class=MaterialExpressionNamedRerouteDeclaration
   inputs [None {from=MaterialExpressionFunctionInput_25}]
   properties {name="• VTX Color - Invert", nodeColor=<rgba 0.5043 1 0 1>}
  }
  MaterialExpressionNamedRerouteDeclaration_27 {
   class=MaterialExpressionNamedRerouteDeclaration
   inputs [None {from=MaterialExpressionFunctionInput_5}]
   properties {name=* World Z Normal - Bias, nodeColor=<rgba 0.5043 1 0 1>}
  }
  MaterialExpressionNamedRerouteDeclaration_28 {
   class=MaterialExpressionNamedRerouteDeclaration
   inputs [None {from=MaterialExpressionFunctionInput_7}]
   properties {name=--- Use WORLD Z NORMAL MASKING ---, nodeColor=<rgba 0.1979 0.0084 0 1>}
  }
  MaterialExpressionNamedRerouteDeclaration_29 {
   class=MaterialExpressionNamedRerouteDeclaration
   inputs [None {from=MaterialExpressionFunctionInput_9}]
   properties {name=* World Z Normal - Intensity, nodeColor=<rgba 0.5043 1 0 1>}
  }
  MaterialExpressionNamedRerouteDeclaration_3 {
   class=MaterialExpressionNamedRerouteDeclaration
   inputs [None {from=MaterialExpressionFunctionInput_48}]
   properties {name=--- Use Z GRADIENT MASKING ---, nodeColor=<rgba 0.1979 0.0084 0 1>}
  }
  MaterialExpressionNamedRerouteDeclaration_30 {
   class=MaterialExpressionNamedRerouteDeclaration
   inputs [None {from=MaterialExpressionFunctionInput_6}]
   properties {name="• World Z Normal - Use VTX Normals", desc="T= VTX Normals only\r\nF= Combined Material Normals", nodeColor=<rgba 0.1979 0.0084 0 1>}
  }
  MaterialExpressionNamedRerouteDeclaration_31 {
   class=MaterialExpressionNamedRerouteDeclaration
   inputs [None {from=MaterialExpressionFunctionInput_0}]
   properties {name="• VTX Color - Intensity", nodeColor=<rgba 0.5043 1 0 1>}
  }
  MaterialExpressionNamedRerouteDeclaration_32 {
   class=MaterialExpressionNamedRerouteDeclaration
   inputs [None {from=MaterialExpressionFunctionInput_3}]
   properties {name=--- Use VTX COLOR MASKING ---, nodeColor=<rgba 0.1979 0.0084 0 1>}
  }
  MaterialExpressionNamedRerouteDeclaration_33 {
   class=MaterialExpressionNamedRerouteDeclaration
   inputs [None {from=MaterialExpressionFunctionInput_27}]
   properties {name="• VTX Color - Bias", nodeColor=<rgba 0.5043 1 0 1>}
  }
  MaterialExpressionNamedRerouteDeclaration_34 {
   class=MaterialExpressionNamedRerouteDeclaration
   inputs [None {from=MaterialExpressionFunctionInput_34}]
   properties {name="• VTX Color - Channel Weights", nodeColor=<rgba 0.5043 1 0 1>}
  }
  MaterialExpressionNamedRerouteDeclaration_35 {
   class=MaterialExpressionNamedRerouteDeclaration
   inputs [None {from=MaterialExpressionFunctionInput_28}]
   properties {name="• Height - Min", nodeColor=<rgba 0.5043 1 0 1>}
  }
  MaterialExpressionNamedRerouteDeclaration_36 {
   class=MaterialExpressionNamedRerouteDeclaration
   inputs [None {from=MaterialExpressionFunctionInput_69}]
   properties {name=* Masking - Invert, nodeColor=<rgba 0.5043 1 0 1>}
  }
  MaterialExpressionNamedRerouteDeclaration_37 {
   class=MaterialExpressionNamedRerouteDeclaration
   inputs [None {from=MaterialExpressionFunctionInput_29}]
   properties {name="• Height - Max", nodeColor=<rgba 0.5043 1 0 1>}
  }
  MaterialExpressionNamedRerouteDeclaration_38 {
   class=MaterialExpressionNamedRerouteDeclaration
   inputs [None {from=MaterialExpressionFunctionInput_68}]
   properties {name="• Masking - Bias", nodeColor=<rgba 0.5043 1 0 1>}
  }
  MaterialExpressionNamedRerouteDeclaration_39 {
   class=MaterialExpressionNamedRerouteDeclaration
   inputs [None {from=MaterialExpressionFunctionInput_67}]
   properties {name="• Masking - Intensity", nodeColor=<rgba 0.5043 1 0 1>}
  }
  MaterialExpressionNamedRerouteDeclaration_4 {
   class=MaterialExpressionNamedRerouteDeclaration
   inputs [None {from=MaterialExpressionFunctionInput_54}]
   properties {name=--- Use BAKED AO/CURVATURE/PAINTED MASKING ---, nodeColor=<rgba 0.1979 0.0084 0 1>}
  }
  MaterialExpressionNamedRerouteDeclaration_40 {
   class=MaterialExpressionNamedRerouteDeclaration
   inputs [None {from=MaterialExpressionFunctionInput_66}]
   properties {name="• WS Noise - Intensity", nodeColor=<rgba 0.5043 1 0 1>}
  }
  MaterialExpressionNamedRerouteDeclaration_41 {
   class=MaterialExpressionNamedRerouteDeclaration
   inputs [None {from=MaterialExpressionFunctionInput_65}]
   properties {name="• WS Noise - Bias", nodeColor=<rgba 0.5043 1 0 1>}
  }
  MaterialExpressionNamedRerouteDeclaration_42 {
   class=MaterialExpressionNamedRerouteDeclaration
   inputs [None {from=MaterialExpressionFunctionInput_30}]
   properties {name=* Realtime Displacement - WS Noise - Contribution, nodeColor=<rgba 0.5043 1 0 1>}
  }
  MaterialExpressionNamedRerouteDeclaration_43 {
   class=MaterialExpressionNamedRerouteDeclaration
   inputs [None {from=MaterialExpressionFunctionInput_31}]
   properties {name=* Realtime Displacement - Height - Contribution, nodeColor=<rgba 0.5043 1 0 1>}
  }
  MaterialExpressionNamedRerouteDeclaration_44 {
   class=MaterialExpressionNamedRerouteDeclaration
   inputs [None {from=MaterialExpressionFunctionInput_56}]
   properties {name="• Baked AO/Curvature/Painted - Intensity", nodeColor=<rgba 0.5043 1 0 1>}
  }
  MaterialExpressionNamedRerouteDeclaration_45 {
   class=MaterialExpressionNamedRerouteDeclaration
   inputs [None {from=MaterialExpressionFunctionInput_55}]
   properties {name="• Baked AO/Curvature/Painted - Bias", nodeColor=<rgba 0.5043 1 0 1>}
  }
  MaterialExpressionNamedRerouteDeclaration_46 {
   class=MaterialExpressionNamedRerouteDeclaration
   inputs [None {from=MaterialExpressionFunctionInput_50}]
   properties {name="• Z Gradient - Height - Top", nodeColor=<rgba 0.5043 1 0 1>}
  }
  MaterialExpressionNamedRerouteDeclaration_47 {
   class=MaterialExpressionNamedRerouteDeclaration
   inputs [None {from=MaterialExpressionFunctionInput_49}]
   properties {name="• Z Gradient - Height - Bottom", nodeColor=<rgba 0.5043 1 0 1>}
  }
  MaterialExpressionNamedRerouteDeclaration_48 {
   class=MaterialExpressionNamedRerouteDeclaration
   inputs [None {from=MaterialExpressionFunctionInput_43}]
   properties {name="• DFAO - Intensity", nodeColor=<rgba 0.5043 1 0 1>}
  }
  MaterialExpressionNamedRerouteDeclaration_49 {
   class=MaterialExpressionNamedRerouteDeclaration
   inputs [None {from=MaterialExpressionFunctionInput_42}]
   properties {name="• DFAO - Bias", nodeColor=<rgba 0.5043 1 0 1>}
  }
  MaterialExpressionNamedRerouteDeclaration_5 {
   class=MaterialExpressionNamedRerouteDeclaration
   inputs [None {from=MaterialExpressionFunctionInput_2}]
   properties {name="• Height - Bias", nodeColor=<rgba 0.5043 1 0 1>}
  }
  MaterialExpressionNamedRerouteDeclaration_50 {
   class=MaterialExpressionNamedRerouteDeclaration
   inputs [None {from=MaterialExpressionFunctionInput_33}]
   properties {name=* Realtime Displacement - Masking - Bias, nodeColor=<rgba 0.5043 1 0 1>}
  }
  MaterialExpressionNamedRerouteDeclaration_51 {
   class=MaterialExpressionNamedRerouteDeclaration
   inputs [None {from=MaterialExpressionReroute_20}]
   properties {name=Combined Mask, nodeColor=<rgba 0.5 0 1 1>}
  }
  MaterialExpressionNamedRerouteDeclaration_52 {
   class=MaterialExpressionNamedRerouteDeclaration
   inputs [None {from=MaterialExpressionFunctionInput_32}]
   properties {name=* Realtime Displacement - Masking - Intensity, nodeColor=<rgba 0.5043 1 0 1>}
  }
  MaterialExpressionNamedRerouteDeclaration_53 {
   class=MaterialExpressionNamedRerouteDeclaration
   inputs [None {from=MaterialExpressionFunctionInput_40}]
   properties {name=* DFAO - Force Baked AO on Nanite, nodeColor=<rgba 0.1979 0.0084 0 1>}
  }
  MaterialExpressionNamedRerouteDeclaration_54 {
   class=MaterialExpressionNamedRerouteDeclaration
   inputs [None {from=MaterialExpressionReroute_52}]
   properties {name=Combined Mask - Realtime Displacement, nodeColor=<rgba 0.5 0 1 1>}
  }
  MaterialExpressionNamedRerouteDeclaration_6 {
   class=MaterialExpressionNamedRerouteDeclaration
   inputs [None {from=MaterialExpressionFunctionInput_61}]
   properties {name=--- Use WS NOISE MASKING ---, nodeColor=<rgba 0.1979 0.0084 0 1>}
  }
  MaterialExpressionNamedRerouteDeclaration_68 {
   class=MaterialExpressionNamedRerouteDeclaration
   inputs [None {from=MaterialExpressionFunctionInput_57}]
   properties {name="• Baked AO/Curvature/Painted - Channel Weights", nodeColor=<rgba 0.5043 1 0 1>}
  }
  MaterialExpressionNamedRerouteDeclaration_7 {
   class=MaterialExpressionNamedRerouteDeclaration
   inputs [None {from=MaterialExpressionFunctionInput_14}]
   properties {name=_Z Gradient, nodeColor=<rgba 1 0 0.0525 1>}
  }
  MaterialExpressionNamedRerouteDeclaration_74 {
   class=MaterialExpressionNamedRerouteDeclaration
   inputs [None {from=MaterialExpressionFunctionInput_70}]
   properties {name="• VTX Color - Blend (1=Max 0=Min)", nodeColor=<rgba 0.5043 1 0 1>}
  }
  MaterialExpressionNamedRerouteDeclaration_75 {
   class=MaterialExpressionNamedRerouteDeclaration
   inputs [None {from=MaterialExpressionFunctionInput_71}]
   properties {name="• DFAO - Blend (1=Max 0=Min)", nodeColor=<rgba 0.5043 1 0 1>}
  }
  MaterialExpressionNamedRerouteDeclaration_76 {
   class=MaterialExpressionNamedRerouteDeclaration
   inputs [None {from=MaterialExpressionFunctionInput_72}]
   properties {name="• Z Gradient - Blend (1=Max 0=Min)", nodeColor=<rgba 0.5043 1 0 1>}
  }
  MaterialExpressionNamedRerouteDeclaration_77 {
   class=MaterialExpressionNamedRerouteDeclaration
   inputs [None {from=MaterialExpressionFunctionInput_73}]
   properties {name="• Baked AO/Curvature/Painted - Blend (1=Max 0=Min)", nodeColor=<rgba 0.5043 1 0 1>}
  }
  MaterialExpressionNamedRerouteDeclaration_78 {
   class=MaterialExpressionNamedRerouteDeclaration
   inputs [None {from=MaterialExpressionFunctionInput_74}]
   properties {name="• WS Noise - Blend (1=Blend 0=Min)", nodeColor=<rgba 0.5043 1 0 1>}
  }
  MaterialExpressionNamedRerouteDeclaration_79 {
   class=MaterialExpressionNamedRerouteDeclaration
   inputs [None {from=MaterialExpressionFunctionInput_75}]
   properties {name="• WS Noise - Channel Weights", nodeColor=<rgba 0.5043 1 0 1>}
  }
  MaterialExpressionNamedRerouteDeclaration_8 {
   class=MaterialExpressionNamedRerouteDeclaration
   inputs [None {from=MaterialExpressionFunctionInput_15}]
   properties {name=_Baked AO, nodeColor=<rgba 1 0 0.0525 1>}
  }
  MaterialExpressionNamedRerouteDeclaration_80 {
   class=MaterialExpressionNamedRerouteDeclaration
   inputs [None {from=MaterialExpressionSaturate_0}]
   properties {name=WS Noise, nodeColor=<rgba 0.5 0 1 1>}
  }
  MaterialExpressionNamedRerouteDeclaration_81 {
   class=MaterialExpressionNamedRerouteDeclaration
   inputs [None {from=MaterialExpressionStaticBoolParameter_48}]
   properties {name="Normals - Add Normals(SB)", nodeColor=<rgba 0.1979 0.0084 0 1>}
  }
  MaterialExpressionNamedRerouteDeclaration_83 {
   class=MaterialExpressionNamedRerouteDeclaration
   inputs [None {from=MaterialExpressionSaturate_32}]
   properties {name=Normal Z, nodeColor=<rgba 0.5 0 1 1>}
  }
  MaterialExpressionNamedRerouteDeclaration_84 {
   class=MaterialExpressionNamedRerouteDeclaration
   inputs [None {from=MaterialExpressionSaturate_33}]
   properties {name=Vertex Color, nodeColor=<rgba 0.5 0 1 1>}
  }
  MaterialExpressionNamedRerouteDeclaration_85 {
   class=MaterialExpressionNamedRerouteDeclaration
   inputs [None {from=MaterialExpressionSaturate_34}]
   properties {name=DFAO - Mask, nodeColor=<rgba 0.5 0 1 1>}
  }
  MaterialExpressionNamedRerouteDeclaration_87 {
   class=MaterialExpressionNamedRerouteDeclaration
   inputs [None {from=MaterialExpressionSaturate_36}]
   properties {name=Baked AO/Curvature/Painted, nodeColor=<rgba 0.5 0 1 1>}
  }
  MaterialExpressionNamedRerouteDeclaration_88 {
   class=MaterialExpressionNamedRerouteDeclaration
   inputs [None {from=MaterialExpressionSmoothStep_0}]
   properties {name=Z Gradient, nodeColor=<rgba 0.3828 0 1 1>}
  }
  MaterialExpressionNamedRerouteDeclaration_9 {
   class=MaterialExpressionNamedRerouteDeclaration
   inputs [None {from=MaterialExpressionReroute_15}]
   properties {name=_Input A MaterialAttributes, desc=MaterialAttributes from previous layers, nodeColor=<rgba 0.1016 0 1 1>}
  }
  MaterialExpressionNamedRerouteUsage_0 {class=MaterialExpressionNamedRerouteUsage}
  MaterialExpressionNamedRerouteUsage_1 {class=MaterialExpressionNamedRerouteUsage}
  MaterialExpressionNamedRerouteUsage_10 {class=MaterialExpressionNamedRerouteUsage}
  MaterialExpressionNamedRerouteUsage_100 {class=MaterialExpressionNamedRerouteUsage}
  MaterialExpressionNamedRerouteUsage_101 {class=MaterialExpressionNamedRerouteUsage}
  MaterialExpressionNamedRerouteUsage_102 {class=MaterialExpressionNamedRerouteUsage}
  MaterialExpressionNamedRerouteUsage_103 {class=MaterialExpressionNamedRerouteUsage}
  MaterialExpressionNamedRerouteUsage_104 {class=MaterialExpressionNamedRerouteUsage}
  MaterialExpressionNamedRerouteUsage_105 {class=MaterialExpressionNamedRerouteUsage}
  MaterialExpressionNamedRerouteUsage_106 {class=MaterialExpressionNamedRerouteUsage}
  MaterialExpressionNamedRerouteUsage_107 {class=MaterialExpressionNamedRerouteUsage}
  MaterialExpressionNamedRerouteUsage_108 {class=MaterialExpressionNamedRerouteUsage}
  MaterialExpressionNamedRerouteUsage_109 {class=MaterialExpressionNamedRerouteUsage}
  MaterialExpressionNamedRerouteUsage_11 {class=MaterialExpressionNamedRerouteUsage}
  MaterialExpressionNamedRerouteUsage_12 {class=MaterialExpressionNamedRerouteUsage}
  MaterialExpressionNamedRerouteUsage_13 {class=MaterialExpressionNamedRerouteUsage}
  MaterialExpressionNamedRerouteUsage_14 {class=MaterialExpressionNamedRerouteUsage, properties {desc=Modulates VTX color transition intensity}}
  MaterialExpressionNamedRerouteUsage_15 {class=MaterialExpressionNamedRerouteUsage}
  MaterialExpressionNamedRerouteUsage_16 {class=MaterialExpressionNamedRerouteUsage}
  MaterialExpressionNamedRerouteUsage_168 {class=MaterialExpressionNamedRerouteUsage}
  MaterialExpressionNamedRerouteUsage_17 {class=MaterialExpressionNamedRerouteUsage}
  MaterialExpressionNamedRerouteUsage_170 {class=MaterialExpressionNamedRerouteUsage}
  MaterialExpressionNamedRerouteUsage_171 {class=MaterialExpressionNamedRerouteUsage}
  MaterialExpressionNamedRerouteUsage_173 {class=MaterialExpressionNamedRerouteUsage}
  MaterialExpressionNamedRerouteUsage_174 {class=MaterialExpressionNamedRerouteUsage}
  MaterialExpressionNamedRerouteUsage_18 {class=MaterialExpressionNamedRerouteUsage}
  MaterialExpressionNamedRerouteUsage_19 {class=MaterialExpressionNamedRerouteUsage}
  MaterialExpressionNamedRerouteUsage_2 {class=MaterialExpressionNamedRerouteUsage}
  MaterialExpressionNamedRerouteUsage_20 {class=MaterialExpressionNamedRerouteUsage}
  MaterialExpressionNamedRerouteUsage_21 {class=MaterialExpressionNamedRerouteUsage}
  MaterialExpressionNamedRerouteUsage_22 {class=MaterialExpressionNamedRerouteUsage}
  MaterialExpressionNamedRerouteUsage_23 {class=MaterialExpressionNamedRerouteUsage}
  MaterialExpressionNamedRerouteUsage_25 {class=MaterialExpressionNamedRerouteUsage}
  MaterialExpressionNamedRerouteUsage_26 {class=MaterialExpressionNamedRerouteUsage}
  MaterialExpressionNamedRerouteUsage_27 {class=MaterialExpressionNamedRerouteUsage}
  MaterialExpressionNamedRerouteUsage_28 {class=MaterialExpressionNamedRerouteUsage}
  MaterialExpressionNamedRerouteUsage_29 {class=MaterialExpressionNamedRerouteUsage}
  MaterialExpressionNamedRerouteUsage_3 {class=MaterialExpressionNamedRerouteUsage}
  MaterialExpressionNamedRerouteUsage_30 {class=MaterialExpressionNamedRerouteUsage}
  MaterialExpressionNamedRerouteUsage_303 {class=MaterialExpressionNamedRerouteUsage}
  MaterialExpressionNamedRerouteUsage_304 {class=MaterialExpressionNamedRerouteUsage}
  MaterialExpressionNamedRerouteUsage_305 {class=MaterialExpressionNamedRerouteUsage}
  MaterialExpressionNamedRerouteUsage_306 {class=MaterialExpressionNamedRerouteUsage}
  MaterialExpressionNamedRerouteUsage_31 {class=MaterialExpressionNamedRerouteUsage}
  MaterialExpressionNamedRerouteUsage_32 {class=MaterialExpressionNamedRerouteUsage}
  MaterialExpressionNamedRerouteUsage_33 {class=MaterialExpressionNamedRerouteUsage}
  MaterialExpressionNamedRerouteUsage_34 {class=MaterialExpressionNamedRerouteUsage}
  MaterialExpressionNamedRerouteUsage_347 {class=MaterialExpressionNamedRerouteUsage}
  MaterialExpressionNamedRerouteUsage_348 {class=MaterialExpressionNamedRerouteUsage}
  MaterialExpressionNamedRerouteUsage_349 {class=MaterialExpressionNamedRerouteUsage}
  MaterialExpressionNamedRerouteUsage_35 {class=MaterialExpressionNamedRerouteUsage}
  MaterialExpressionNamedRerouteUsage_350 {class=MaterialExpressionNamedRerouteUsage}
  MaterialExpressionNamedRerouteUsage_351 {class=MaterialExpressionNamedRerouteUsage}
  MaterialExpressionNamedRerouteUsage_36 {class=MaterialExpressionNamedRerouteUsage}
  MaterialExpressionNamedRerouteUsage_360 {class=MaterialExpressionNamedRerouteUsage}
  MaterialExpressionNamedRerouteUsage_37 {class=MaterialExpressionNamedRerouteUsage}
  MaterialExpressionNamedRerouteUsage_38 {class=MaterialExpressionNamedRerouteUsage}
  MaterialExpressionNamedRerouteUsage_39 {class=MaterialExpressionNamedRerouteUsage}
  MaterialExpressionNamedRerouteUsage_4 {class=MaterialExpressionNamedRerouteUsage}
  MaterialExpressionNamedRerouteUsage_40 {class=MaterialExpressionNamedRerouteUsage}
  MaterialExpressionNamedRerouteUsage_41 {class=MaterialExpressionNamedRerouteUsage}
  MaterialExpressionNamedRerouteUsage_42 {class=MaterialExpressionNamedRerouteUsage}
  MaterialExpressionNamedRerouteUsage_43 {class=MaterialExpressionNamedRerouteUsage}
  MaterialExpressionNamedRerouteUsage_44 {class=MaterialExpressionNamedRerouteUsage}
  MaterialExpressionNamedRerouteUsage_45 {class=MaterialExpressionNamedRerouteUsage}
  MaterialExpressionNamedRerouteUsage_46 {class=MaterialExpressionNamedRerouteUsage}
  MaterialExpressionNamedRerouteUsage_47 {class=MaterialExpressionNamedRerouteUsage}
  MaterialExpressionNamedRerouteUsage_48 {class=MaterialExpressionNamedRerouteUsage}
  MaterialExpressionNamedRerouteUsage_49 {class=MaterialExpressionNamedRerouteUsage}
  MaterialExpressionNamedRerouteUsage_5 {class=MaterialExpressionNamedRerouteUsage}
  MaterialExpressionNamedRerouteUsage_50 {class=MaterialExpressionNamedRerouteUsage}
  MaterialExpressionNamedRerouteUsage_51 {class=MaterialExpressionNamedRerouteUsage}
  MaterialExpressionNamedRerouteUsage_52 {class=MaterialExpressionNamedRerouteUsage}
  MaterialExpressionNamedRerouteUsage_53 {class=MaterialExpressionNamedRerouteUsage}
  MaterialExpressionNamedRerouteUsage_54 {class=MaterialExpressionNamedRerouteUsage}
  MaterialExpressionNamedRerouteUsage_55 {class=MaterialExpressionNamedRerouteUsage}
  MaterialExpressionNamedRerouteUsage_56 {class=MaterialExpressionNamedRerouteUsage}
  MaterialExpressionNamedRerouteUsage_57 {class=MaterialExpressionNamedRerouteUsage}
  MaterialExpressionNamedRerouteUsage_58 {class=MaterialExpressionNamedRerouteUsage}
  MaterialExpressionNamedRerouteUsage_59 {class=MaterialExpressionNamedRerouteUsage}
  MaterialExpressionNamedRerouteUsage_6 {class=MaterialExpressionNamedRerouteUsage}
  MaterialExpressionNamedRerouteUsage_60 {class=MaterialExpressionNamedRerouteUsage}
  MaterialExpressionNamedRerouteUsage_61 {class=MaterialExpressionNamedRerouteUsage}
  MaterialExpressionNamedRerouteUsage_62 {class=MaterialExpressionNamedRerouteUsage}
  MaterialExpressionNamedRerouteUsage_63 {class=MaterialExpressionNamedRerouteUsage}
  MaterialExpressionNamedRerouteUsage_64 {class=MaterialExpressionNamedRerouteUsage}
  MaterialExpressionNamedRerouteUsage_65 {class=MaterialExpressionNamedRerouteUsage}
  MaterialExpressionNamedRerouteUsage_66 {class=MaterialExpressionNamedRerouteUsage}
  MaterialExpressionNamedRerouteUsage_67 {class=MaterialExpressionNamedRerouteUsage}
  MaterialExpressionNamedRerouteUsage_68 {class=MaterialExpressionNamedRerouteUsage}
  MaterialExpressionNamedRerouteUsage_69 {class=MaterialExpressionNamedRerouteUsage}
  MaterialExpressionNamedRerouteUsage_7 {class=MaterialExpressionNamedRerouteUsage}
  MaterialExpressionNamedRerouteUsage_70 {class=MaterialExpressionNamedRerouteUsage}
  MaterialExpressionNamedRerouteUsage_71 {class=MaterialExpressionNamedRerouteUsage}
  MaterialExpressionNamedRerouteUsage_72 {class=MaterialExpressionNamedRerouteUsage}
  MaterialExpressionNamedRerouteUsage_73 {class=MaterialExpressionNamedRerouteUsage}
  MaterialExpressionNamedRerouteUsage_74 {class=MaterialExpressionNamedRerouteUsage}
  MaterialExpressionNamedRerouteUsage_75 {class=MaterialExpressionNamedRerouteUsage}
  MaterialExpressionNamedRerouteUsage_76 {class=MaterialExpressionNamedRerouteUsage}
  MaterialExpressionNamedRerouteUsage_77 {class=MaterialExpressionNamedRerouteUsage}
  MaterialExpressionNamedRerouteUsage_78 {class=MaterialExpressionNamedRerouteUsage}
  MaterialExpressionNamedRerouteUsage_79 {class=MaterialExpressionNamedRerouteUsage}
  MaterialExpressionNamedRerouteUsage_8 {class=MaterialExpressionNamedRerouteUsage}
  MaterialExpressionNamedRerouteUsage_80 {class=MaterialExpressionNamedRerouteUsage}
  MaterialExpressionNamedRerouteUsage_81 {class=MaterialExpressionNamedRerouteUsage}
  MaterialExpressionNamedRerouteUsage_82 {class=MaterialExpressionNamedRerouteUsage}
  MaterialExpressionNamedRerouteUsage_83 {class=MaterialExpressionNamedRerouteUsage}
  MaterialExpressionNamedRerouteUsage_84 {class=MaterialExpressionNamedRerouteUsage}
  MaterialExpressionNamedRerouteUsage_85 {class=MaterialExpressionNamedRerouteUsage}
  MaterialExpressionNamedRerouteUsage_86 {class=MaterialExpressionNamedRerouteUsage}
  MaterialExpressionNamedRerouteUsage_87 {class=MaterialExpressionNamedRerouteUsage}
  MaterialExpressionNamedRerouteUsage_88 {class=MaterialExpressionNamedRerouteUsage}
  MaterialExpressionNamedRerouteUsage_89 {class=MaterialExpressionNamedRerouteUsage}
  MaterialExpressionNamedRerouteUsage_9 {class=MaterialExpressionNamedRerouteUsage}
  MaterialExpressionNamedRerouteUsage_90 {class=MaterialExpressionNamedRerouteUsage}
  MaterialExpressionNamedRerouteUsage_91 {class=MaterialExpressionNamedRerouteUsage}
  MaterialExpressionNamedRerouteUsage_92 {class=MaterialExpressionNamedRerouteUsage}
  MaterialExpressionNamedRerouteUsage_93 {class=MaterialExpressionNamedRerouteUsage}
  MaterialExpressionNamedRerouteUsage_94 {class=MaterialExpressionNamedRerouteUsage}
  MaterialExpressionNamedRerouteUsage_95 {class=MaterialExpressionNamedRerouteUsage}
  MaterialExpressionNamedRerouteUsage_96 {class=MaterialExpressionNamedRerouteUsage}
  MaterialExpressionNamedRerouteUsage_97 {class=MaterialExpressionNamedRerouteUsage}
  MaterialExpressionNamedRerouteUsage_98 {class=MaterialExpressionNamedRerouteUsage}
  MaterialExpressionNamedRerouteUsage_99 {class=MaterialExpressionNamedRerouteUsage}
  MaterialExpressionOneMinus_0 {class=MaterialExpressionOneMinus, inputs [None {from=MaterialExpressionDotProduct_1}]}
  MaterialExpressionOneMinus_1 {class=MaterialExpressionOneMinus, inputs [None {from=MaterialExpressionReroute_51}]}
  MaterialExpressionOneMinus_2 {
   class=MaterialExpressionOneMinus
   inputs [None {from=MaterialExpressionWorldPosition_0, output=Z}]
   properties {desc=flip so positive values raise the waterline}
  }
  MaterialExpressionOneMinus_3 {class=MaterialExpressionOneMinus, inputs [None {from=MaterialExpressionDotProduct_0}]}
  MaterialExpressionOneMinus_31 {class=MaterialExpressionOneMinus, inputs [None {from=MaterialExpressionReroute_235}]}
  MaterialExpressionOneMinus_4 {class=MaterialExpressionOneMinus, inputs [None {from=MaterialExpressionDotProduct_2}]}
  MaterialExpressionReroute_0 {class=MaterialExpressionReroute, inputs [None {from=MaterialExpressionNamedRerouteUsage_77}]}
  MaterialExpressionReroute_1 {class=MaterialExpressionReroute, inputs [None {from=MaterialExpressionNamedRerouteUsage_76}]}
  MaterialExpressionReroute_10 {class=MaterialExpressionReroute, inputs [None {from=MaterialExpressionNamedRerouteUsage_71}]}
  MaterialExpressionReroute_11 {class=MaterialExpressionReroute, inputs [None {from=MaterialExpressionNamedRerouteUsage_55}]}
  MaterialExpressionReroute_12 {class=MaterialExpressionReroute, inputs [None {from=MaterialExpressionStaticSwitch_10}]}
  MaterialExpressionReroute_13 {class=MaterialExpressionReroute, inputs [None {from=MaterialExpressionNamedRerouteUsage_72}]}
  MaterialExpressionReroute_14 {class=MaterialExpressionReroute, inputs [None {from=MaterialExpressionNamedRerouteUsage_81}]}
  MaterialExpressionReroute_142 {class=MaterialExpressionReroute, inputs [None {from=MaterialExpressionReroute_143}]}
  MaterialExpressionReroute_143 {class=MaterialExpressionReroute, inputs [None {from=MaterialExpressionReroute_6}]}
  MaterialExpressionReroute_146 {class=MaterialExpressionReroute, inputs [None {from=MaterialExpressionNamedRerouteUsage_173}]}
  MaterialExpressionReroute_15 {class=MaterialExpressionReroute, inputs [None {from=MaterialExpressionFunctionInput_11}]}
  MaterialExpressionReroute_16 {class=MaterialExpressionReroute, inputs [None {from=MaterialExpressionNamedRerouteUsage_69}]}
  MaterialExpressionReroute_17 {class=MaterialExpressionReroute, inputs [None {from=MaterialExpressionNamedRerouteUsage_68}]}
  MaterialExpressionReroute_178 {class=MaterialExpressionReroute, inputs [None {from=MaterialExpressionNamedRerouteUsage_14}]}
  MaterialExpressionReroute_18 {class=MaterialExpressionReroute, inputs [None {from=MaterialExpressionStaticSwitch_0}]}
  MaterialExpressionReroute_19 {class=MaterialExpressionReroute, inputs [None {from=MaterialExpressionStaticSwitch_1}]}
  MaterialExpressionReroute_2 {class=MaterialExpressionReroute, inputs [None {from=MaterialExpressionNamedRerouteUsage_74}]}
  MaterialExpressionReroute_20 {class=MaterialExpressionReroute, inputs [None {from=MaterialExpressionSaturate_39}]}
  MaterialExpressionReroute_209 {class=MaterialExpressionReroute, inputs [None {from=MaterialExpressionReroute_210}]}
  MaterialExpressionReroute_21 {class=MaterialExpressionReroute, inputs [None {from=MaterialExpressionNamedRerouteUsage_70}]}
  MaterialExpressionReroute_210 {class=MaterialExpressionReroute, inputs [None {from=MaterialExpressionReroute_28}]}
  MaterialExpressionReroute_211 {class=MaterialExpressionReroute, inputs [None {from=MaterialExpressionNamedRerouteUsage_2}]}
  MaterialExpressionReroute_212 {class=MaterialExpressionReroute, inputs [None {from=MaterialExpressionStaticSwitch_42}]}
  MaterialExpressionReroute_214 {
   class=MaterialExpressionReroute
   inputs [None {from=MaterialExpressionGetMaterialAttributes_9, output=Opacity}]
   properties {desc=Previous Layer Height}
  }
  MaterialExpressionReroute_22 {class=MaterialExpressionReroute, inputs [None {from=MaterialExpressionNamedRerouteUsage_66}]}
  MaterialExpressionReroute_228 {class=MaterialExpressionReroute, inputs [None {from=MaterialExpressionReroute_241}]}
  MaterialExpressionReroute_229 {class=MaterialExpressionReroute, inputs [None {from=MaterialExpressionReroute_261}]}
  MaterialExpressionReroute_23 {class=MaterialExpressionReroute, inputs [None {from=MaterialExpressionNamedRerouteUsage_67}]}
  MaterialExpressionReroute_230 {class=MaterialExpressionReroute, inputs [None {from=MaterialExpressionReroute_237}]}
  MaterialExpressionReroute_231 {class=MaterialExpressionReroute, inputs [None {from=MaterialExpressionReroute_249}]}
  MaterialExpressionReroute_235 {class=MaterialExpressionReroute, inputs [None {from=MaterialExpressionReroute_212}]}
  MaterialExpressionReroute_237 {class=MaterialExpressionReroute, inputs [None {from=MaterialExpressionReroute_19}]}
  MaterialExpressionReroute_239 {class=MaterialExpressionReroute, inputs [None {from=MaterialExpressionNamedRerouteUsage_171}]}
  MaterialExpressionReroute_24 {class=MaterialExpressionReroute, inputs [None {from=MaterialExpressionNamedRerouteUsage_64}]}
  MaterialExpressionReroute_241 {class=MaterialExpressionReroute, inputs [None {from=MaterialExpressionReroute_18}]}
  MaterialExpressionReroute_242 {class=MaterialExpressionReroute, inputs [None {from=MaterialExpressionNamedRerouteUsage_170}]}
  MaterialExpressionReroute_249 {class=MaterialExpressionReroute, inputs [None {from=MaterialExpressionReroute_47}]}
  MaterialExpressionReroute_25 {class=MaterialExpressionReroute, inputs [None {from=MaterialExpressionStaticSwitch_2}]}
  MaterialExpressionReroute_250 {class=MaterialExpressionReroute, inputs [None {from=MaterialExpressionNamedRerouteUsage_174}]}
  MaterialExpressionReroute_26 {class=MaterialExpressionReroute, inputs [None {from=MaterialExpressionStaticSwitch_3}]}
  MaterialExpressionReroute_260 {class=MaterialExpressionReroute, inputs [None {from=MaterialExpressionNamedRerouteUsage_350}]}
  MaterialExpressionReroute_261 {class=MaterialExpressionReroute, inputs [None {from=MaterialExpressionReroute_12}]}
  MaterialExpressionReroute_262 {class=MaterialExpressionReroute, inputs [None {from=MaterialExpressionNamedRerouteUsage_44}]}
  MaterialExpressionReroute_263 {class=MaterialExpressionReroute, inputs [None {from=MaterialExpressionNamedRerouteUsage_347}]}
  MaterialExpressionReroute_264 {class=MaterialExpressionReroute, inputs [None {from=MaterialExpressionNamedRerouteUsage_349}]}
  MaterialExpressionReroute_265 {class=MaterialExpressionReroute, inputs [None {from=MaterialExpressionNamedRerouteUsage_348}]}
  MaterialExpressionReroute_266 {class=MaterialExpressionReroute, inputs [None {from=MaterialExpressionNamedRerouteUsage_351}]}
  MaterialExpressionReroute_27 {class=MaterialExpressionReroute, inputs [None {from=MaterialExpressionStaticSwitch_4}]}
  MaterialExpressionReroute_28 {class=MaterialExpressionReroute, inputs [None {from=MaterialExpressionStaticSwitch_5}]}
  MaterialExpressionReroute_29 {class=MaterialExpressionReroute, inputs [None {from=MaterialExpressionNamedRerouteUsage_65}]}
  MaterialExpressionReroute_3 {class=MaterialExpressionReroute, inputs [None {from=MaterialExpressionReroute_4}]}
  MaterialExpressionReroute_30 {class=MaterialExpressionReroute, inputs [None {from=MaterialExpressionNamedRerouteUsage_61}]}
  MaterialExpressionReroute_31 {class=MaterialExpressionReroute, inputs [None {from=MaterialExpressionNamedRerouteUsage_57}]}
  MaterialExpressionReroute_32 {class=MaterialExpressionReroute, inputs [None {from=MaterialExpressionNamedRerouteUsage_58}]}
  MaterialExpressionReroute_33 {class=MaterialExpressionReroute, inputs [None {from=MaterialExpressionNamedRerouteUsage_52}]}
  MaterialExpressionReroute_34 {class=MaterialExpressionReroute, inputs [None {from=MaterialExpressionNamedRerouteUsage_56}]}
  MaterialExpressionReroute_35 {class=MaterialExpressionReroute, inputs [None {from=MaterialExpressionNamedRerouteUsage_53}]}
  MaterialExpressionReroute_36 {class=MaterialExpressionReroute, inputs [None {from=MaterialExpressionNamedRerouteUsage_54}]}
  MaterialExpressionReroute_37 {class=MaterialExpressionReroute, inputs [None {from=MaterialExpressionNamedRerouteUsage_78}]}
  MaterialExpressionReroute_38 {class=MaterialExpressionReroute, inputs [None {from=MaterialExpressionStaticSwitch_11}]}
  MaterialExpressionReroute_39 {class=MaterialExpressionReroute, inputs [None {from=MaterialExpressionNamedRerouteUsage_51}]}
  MaterialExpressionReroute_4 {class=MaterialExpressionReroute, inputs [None {from=MaterialExpressionReroute_27}]}
  MaterialExpressionReroute_40 {class=MaterialExpressionReroute, inputs [None {from=MaterialExpressionNamedRerouteUsage_62}]}
  MaterialExpressionReroute_41 {class=MaterialExpressionReroute, inputs [None {from=MaterialExpressionNamedRerouteUsage_63}]}
  MaterialExpressionReroute_42 {class=MaterialExpressionReroute, inputs [None {from=MaterialExpressionNamedRerouteUsage_50}]}
  MaterialExpressionReroute_43 {class=MaterialExpressionReroute, inputs [None {from=MaterialExpressionNamedRerouteUsage_60}]}
  MaterialExpressionReroute_44 {class=MaterialExpressionReroute, inputs [None {from=MaterialExpressionNamedRerouteUsage_49}]}
  MaterialExpressionReroute_45 {class=MaterialExpressionReroute, inputs [None {from=MaterialExpressionNamedRerouteUsage_59}]}
  MaterialExpressionReroute_46 {class=MaterialExpressionReroute, inputs [None {from=MaterialExpressionReroute_25}]}
  MaterialExpressionReroute_47 {class=MaterialExpressionReroute, inputs [None {from=MaterialExpressionReroute_46}]}
  MaterialExpressionReroute_48 {class=MaterialExpressionReroute, inputs [None {from=MaterialExpressionNamedRerouteUsage_20}]}
  MaterialExpressionReroute_49 {class=MaterialExpressionReroute, inputs [None {from=MaterialExpressionStaticSwitch_12}]}
  MaterialExpressionReroute_5 {class=MaterialExpressionReroute, inputs [None {from=MaterialExpressionNamedRerouteUsage_75}]}
  MaterialExpressionReroute_50 {class=MaterialExpressionReroute, inputs [None {from=MaterialExpressionReroute_261}]}
  MaterialExpressionReroute_51 {class=MaterialExpressionReroute, inputs [None {from=MaterialExpressionReroute_49}]}
  MaterialExpressionReroute_52 {class=MaterialExpressionReroute, inputs [None {from=MaterialExpressionSaturate_3}]}
  MaterialExpressionReroute_53 {class=MaterialExpressionReroute, inputs [None {from=MaterialExpressionNamedRerouteUsage_109}]}
  MaterialExpressionReroute_54 {class=MaterialExpressionReroute, inputs [None {from=MaterialExpressionFunctionInput_17}]}
  MaterialExpressionReroute_55 {class=MaterialExpressionReroute, inputs [None {from=MaterialExpressionNamedRerouteUsage_107}]}
  MaterialExpressionReroute_56 {class=MaterialExpressionReroute, inputs [None {from=MaterialExpressionNamedRerouteUsage_106}]}
  MaterialExpressionReroute_57 {class=MaterialExpressionReroute, inputs [None {from=MaterialExpressionNamedRerouteUsage_108}]}
  MaterialExpressionReroute_58 {class=MaterialExpressionReroute, inputs [None {from=MaterialExpressionNamedRerouteUsage_82}]}
  MaterialExpressionReroute_59 {class=MaterialExpressionReroute, inputs [None {from=MaterialExpressionNamedRerouteUsage_83}]}
  MaterialExpressionReroute_6 {class=MaterialExpressionReroute, inputs [None {from=MaterialExpressionReroute_26}]}
  MaterialExpressionReroute_60 {class=MaterialExpressionReroute, inputs [None {from=MaterialExpressionNamedRerouteUsage_80}]}
  MaterialExpressionReroute_61 {class=MaterialExpressionReroute, inputs [None {from=MaterialExpressionNamedRerouteUsage_97}]}
  MaterialExpressionReroute_62 {class=MaterialExpressionReroute, inputs [None {from=MaterialExpressionNamedRerouteUsage_98}]}
  MaterialExpressionReroute_63 {class=MaterialExpressionReroute, inputs [None {from=MaterialExpressionNamedRerouteUsage_99}]}
  MaterialExpressionReroute_64 {class=MaterialExpressionReroute, inputs [None {from=MaterialExpressionNamedRerouteUsage_42}]}
  MaterialExpressionReroute_65 {
   class=MaterialExpressionReroute
   inputs [None {from=MaterialExpressionGetMaterialAttributes_2, output=Opacity}]
   properties {desc=Current Layer Height}
  }
  MaterialExpressionReroute_66 {class=MaterialExpressionReroute, inputs [None {from=MaterialExpressionNamedRerouteUsage_102}]}
  MaterialExpressionReroute_67 {class=MaterialExpressionReroute, inputs [None {from=MaterialExpressionNamedRerouteUsage_101}]}
  MaterialExpressionReroute_7 {class=MaterialExpressionReroute, inputs [None {from=MaterialExpressionNamedRerouteUsage_84}]}
  MaterialExpressionReroute_8 {class=MaterialExpressionReroute, inputs [None {from=MaterialExpressionNamedRerouteUsage_73}]}
  MaterialExpressionReroute_9 {class=MaterialExpressionReroute, inputs [None {from=MaterialExpressionReroute_38}]}
  MaterialExpressionSaturate_0 {class=MaterialExpressionSaturate, inputs [None {from=MaterialExpressionMultiply_3}]}
  MaterialExpressionSaturate_1 {class=MaterialExpressionSaturate, inputs [None {from=MaterialExpressionComponentMask_14}]}
  MaterialExpressionSaturate_17 {class=MaterialExpressionSaturate, inputs [None {from=MaterialExpressionComponentMask_12}]}
  MaterialExpressionSaturate_2 {class=MaterialExpressionSaturate, inputs [None {from=MaterialExpressionMultiply_4}]}
  MaterialExpressionSaturate_3 {class=MaterialExpressionSaturate, inputs [None {from=MaterialExpressionMultiply_6}]}
  MaterialExpressionSaturate_32 {class=MaterialExpressionSaturate, inputs [None {from=MaterialExpressionMultiply_69}]}
  MaterialExpressionSaturate_33 {class=MaterialExpressionSaturate, inputs [None {from=MaterialExpressionMultiply_2}]}
  MaterialExpressionSaturate_34 {class=MaterialExpressionSaturate, inputs [None {from=MaterialExpressionMultiply_72}]}
  MaterialExpressionSaturate_36 {class=MaterialExpressionSaturate, inputs [None {from=MaterialExpressionMultiply_74}]}
  MaterialExpressionSaturate_39 {class=MaterialExpressionSaturate, inputs [None {from=MaterialExpressionMultiply_0}]}
  MaterialExpressionSaturate_4 {class=MaterialExpressionSaturate, inputs [None {from=MaterialExpressionLinearInterpolate_31}]}
  MaterialExpressionScalarParameter_0 {class=MaterialExpressionScalarParameter, properties {defaultValue=1, group=Mask, parameterName="• Height - Intensity", sortPriority=516}}
  MaterialExpressionScalarParameter_1 {
   class=MaterialExpressionScalarParameter
   properties {defaultValue=1, group=Mask, parameterName="• Height - Blend (1=Max 0=Min)", sortPriority=510}
  }
  MaterialExpressionScalarParameter_10 {class=MaterialExpressionScalarParameter, properties {defaultValue=1, group=Mask, parameterName=* Waterline - Intensity, sortPriority=416}}
  MaterialExpressionScalarParameter_106 {class=MaterialExpressionScalarParameter, properties {group=Mask, parameterName="• WS Noise - Bias", sortPriority=475}}
  MaterialExpressionScalarParameter_107 {class=MaterialExpressionScalarParameter, properties {defaultValue=1, group=Mask, parameterName="• WS Noise - Intensity", sortPriority=476}}
  MaterialExpressionScalarParameter_11 {class=MaterialExpressionScalarParameter, properties {group=Mask, parameterName="• VTX Color - Bias", sortPriority=175}}
  MaterialExpressionScalarParameter_12 {
   class=MaterialExpressionScalarParameter
   properties {group=Mask, parameterName=* Realtime Displacement - WS Noise - Contribution, sortPriority=560}
  }
  MaterialExpressionScalarParameter_13 {
   class=MaterialExpressionScalarParameter
   properties {
    defaultValue=1
    desc="1= object top bound\r\n0= object bottom bound"
    group=Mask
    parameterName="• Z Gradient - Height - Top"
    sortPriority=266
   }
  }
  MaterialExpressionScalarParameter_14 {
   class=MaterialExpressionScalarParameter
   properties {desc="1= object top bound\r\n0= object bottom bound", group=Mask, parameterName="• Z Gradient - Height - Bottom", sortPriority=265}
  }
  MaterialExpressionScalarParameter_15 {
   class=MaterialExpressionScalarParameter
   properties {defaultValue=1, group=Mask, parameterName="• VTX Color - Blend (1=Max 0=Min)", sortPriority=170}
  }
  MaterialExpressionScalarParameter_16 {
   class=MaterialExpressionScalarParameter
   properties {defaultValue=1, group=Mask, parameterName="• DFAO - Blend (1=Max 0=Min)", sortPriority=220}
  }
  MaterialExpressionScalarParameter_17 {
   class=MaterialExpressionScalarParameter
   properties {defaultValue=1, group=Mask, parameterName="• Z Gradient - Blend (1=Max 0=Min)", sortPriority=260}
  }
  MaterialExpressionScalarParameter_18 {class=MaterialExpressionScalarParameter, properties {group=Mask, parameterName="• Masking - Invert", sliderMax=1, sortPriority=50}}
  MaterialExpressionScalarParameter_19 {
   class=MaterialExpressionScalarParameter
   properties {defaultValue=1, group=Mask, parameterName="• WS Noise - Blend (1=Blend 0=Min)", sortPriority=470}
  }
  MaterialExpressionScalarParameter_2 {class=MaterialExpressionScalarParameter, properties {defaultValue=1, group=Mask, parameterName="• VTX Color - Intensity", sortPriority=176}}
  MaterialExpressionScalarParameter_20 {class=MaterialExpressionScalarParameter, properties {group=Mask, parameterName="• WS Noise - Invert", sortPriority=461}}
  MaterialExpressionScalarParameter_21 {class=MaterialExpressionScalarParameter, properties {defaultValue=1, group=Mask, parameterName="• VTX Color - Invert", sortPriority=161}}
  MaterialExpressionScalarParameter_22 {class=MaterialExpressionScalarParameter, properties {defaultValue=1, group=Mask, parameterName="• Height - Max", sortPriority=513}}
  MaterialExpressionScalarParameter_23 {class=MaterialExpressionScalarParameter, properties {group=Mask, parameterName="• Height - Min", sortPriority=512}}
  MaterialExpressionScalarParameter_24 {
   class=MaterialExpressionScalarParameter
   properties {group=Mask, parameterName=* Realtime Displacement - Height - Contribution, sortPriority=565}
  }
  MaterialExpressionScalarParameter_25 {class=MaterialExpressionScalarParameter, properties {group=Mask, parameterName=* Realtime Displacement - Masking - Bias, sortPriority=550}}
  MaterialExpressionScalarParameter_26 {
   class=MaterialExpressionScalarParameter
   properties {defaultValue=1, group=Mask, parameterName=* Realtime Displacement - Masking - Intensity, sortPriority=555}
  }
  MaterialExpressionScalarParameter_3 {class=MaterialExpressionScalarParameter, properties {group=Mask, parameterName="• Height - Bias", sortPriority=515}}
  MaterialExpressionScalarParameter_4 {
   class=MaterialExpressionScalarParameter
   properties {defaultValue=1, group=Mask, parameterName="• Height - Selection (1=Current Layer 0=Previous Layers)", sortPriority=509}
  }
  MaterialExpressionScalarParameter_5 {
   class=MaterialExpressionScalarParameter
   properties {
    bUseCustomPrimitiveData=true
    desc="Primitive Data Only, Enable per instance in the details panel."
    group=Waterline
    parameterName=PD_WaterlineHeightOffset
    sortPriority=0
   }
  }
  MaterialExpressionScalarParameter_54 {
   class=MaterialExpressionScalarParameter
   properties {defaultValue=1, group=Mask, parameterName="• Baked AO/Curvature/Painted - Intensity", sortPriority=376}
  }
  MaterialExpressionScalarParameter_55 {class=MaterialExpressionScalarParameter, properties {group=Mask, parameterName="• Baked AO/Curvature/Painted - Bias", sortPriority=375}}
  MaterialExpressionScalarParameter_6 {
   class=MaterialExpressionScalarParameter
   properties {defaultValue=1, group=Mask, parameterName="* Waterline - Blend (1=Max 0=Min)", sortPriority=410}
  }
  MaterialExpressionScalarParameter_68 {
   class=MaterialExpressionScalarParameter
   properties {defaultValue=1, group=Mask, parameterName="• Baked AO/Curvature/Painted - Blend (1=Max 0=Min)", sortPriority=370}
  }
  MaterialExpressionScalarParameter_7 {class=MaterialExpressionScalarParameter, properties {group=Mask, parameterName="• Masking - Bias", sortPriority=60}}
  MaterialExpressionScalarParameter_77 {class=MaterialExpressionScalarParameter, properties {group=Mask, parameterName="• Baked AO/Curvature/Painted - Invert", sortPriority=361}}
  MaterialExpressionScalarParameter_8 {class=MaterialExpressionScalarParameter, properties {defaultValue=1, group=Mask, parameterName="• Masking - Intensity", sortPriority=61}}
  MaterialExpressionScalarParameter_9 {class=MaterialExpressionScalarParameter, properties {group=Mask, parameterName=* Waterline - Bias, sortPriority=415}}
  MaterialExpressionScalarParameter_92 {class=MaterialExpressionScalarParameter, properties {defaultValue=-0.2, group=Mask, parameterName=* World Z Normal - Bias, sortPriority=120}}
  MaterialExpressionScalarParameter_93 {
   class=MaterialExpressionScalarParameter
   properties {defaultValue=1.5, group=Mask, parameterName=* World Z Normal - Intensity, sortPriority=121}
  }
  MaterialExpressionScalarParameter_95 {class=MaterialExpressionScalarParameter, properties {defaultValue=1, group=Mask, parameterName="• DFAO - Intensity", sortPriority=226}}
  MaterialExpressionScalarParameter_96 {class=MaterialExpressionScalarParameter, properties {group=Mask, parameterName="• DFAO - Bias", sortPriority=225}}
  MaterialExpressionSetMaterialAttributes_0 {class=MaterialExpressionSetMaterialAttributes, inputs [MaterialAttributes {from="", output=""}]}
  MaterialExpressionShadingPathSwitch_14 {
   class=MaterialExpressionShadingPathSwitch
   inputs [
    Default {from=MaterialExpressionReroute_178, output=""}
    Deferred {from="", output=""}
    Forward {from="", output=""}
    Mobile {from=MaterialExpressionMultiply_70, output=""}
   ]
   properties {}
  }
  MaterialExpressionSmoothStep_0 {
   class=MaterialExpressionSmoothStep
   inputs [
    Min {from=MaterialExpressionNamedRerouteUsage_29, output=""}
    Max {from=MaterialExpressionNamedRerouteUsage_30, output=""}
    Value {from=MaterialExpressionNamedRerouteUsage_4, output=""}
   ]
   properties {}
  }
  MaterialExpressionStaticBoolParameter_0 {class=MaterialExpressionStaticBoolParameter, inputs [], properties {group=Mask, parameterName=--- Use HEIGHT MASKING ---, sortPriority=500}}
  MaterialExpressionStaticBoolParameter_1 {
   class=MaterialExpressionStaticBoolParameter
   inputs []
   properties {group=Mask, parameterName=--- Use WATERLINE MASKING ---, sortPriority=400}
  }
  MaterialExpressionStaticBoolParameter_11 {
   class=MaterialExpressionStaticBoolParameter
   inputs []
   properties {desc="T= VTX Normals only\r\nF= Combined Material Normals", group=Mask, parameterName="• World Z Normal - Use VTX Normals", sortPriority=110}
  }
  MaterialExpressionStaticBoolParameter_18 {
   class=MaterialExpressionStaticBoolParameter
   inputs []
   properties {group=Mask, parameterName=--- Use Z GRADIENT MASKING ---, sortPriority=250}
  }
  MaterialExpressionStaticBoolParameter_20 {
   class=MaterialExpressionStaticBoolParameter
   inputs []
   properties {group=Mask, parameterName=--- Use WS NOISE MASKING ---, sortPriority=450}
  }
  MaterialExpressionStaticBoolParameter_24 {
   class=MaterialExpressionStaticBoolParameter
   inputs []
   properties {group=Mask, parameterName=--- Use BAKED AO/CURVATURE/PAINTED MASKING ---, sortPriority=350}
  }
  MaterialExpressionStaticBoolParameter_25 {
   class=MaterialExpressionStaticBoolParameter
   inputs []
   properties {defaultValue=true, group=Mask, parameterName=* DFAO - Force Baked AO on Nanite, sortPriority=210}
  }
  MaterialExpressionStaticBoolParameter_31 {
   class=MaterialExpressionStaticBoolParameter
   inputs []
   properties {group=Mask, parameterName=--- Use WORLD Z NORMAL MASKING ---, sortPriority=100}
  }
  MaterialExpressionStaticBoolParameter_42 {
   class=MaterialExpressionStaticBoolParameter
   inputs []
   properties {group=Mask, parameterName=--- Use VTX COLOR MASKING ---, sortPriority=150}
  }
  MaterialExpressionStaticBoolParameter_43 {class=MaterialExpressionStaticBoolParameter, inputs [], properties {group=Mask, parameterName=--- Use DFAO MASKING ---, sortPriority=200}}
  MaterialExpressionStaticBoolParameter_48 {class=MaterialExpressionStaticBoolParameter, inputs [], properties {group=Mask, parameterName=Normals - Add Normals, sortPriority=325}}
  MaterialExpressionStaticBool_0 {class=MaterialExpressionStaticBool}
  MaterialExpressionStaticBool_1 {class=MaterialExpressionStaticBool}
  MaterialExpressionStaticBool_11 {class=MaterialExpressionStaticBool}
  MaterialExpressionStaticBool_19 {class=MaterialExpressionStaticBool}
  MaterialExpressionStaticBool_20 {class=MaterialExpressionStaticBool}
  MaterialExpressionStaticBool_25 {class=MaterialExpressionStaticBool}
  MaterialExpressionStaticBool_29 {class=MaterialExpressionStaticBool}
  MaterialExpressionStaticBool_33 {class=MaterialExpressionStaticBool}
  MaterialExpressionStaticBool_8 {class=MaterialExpressionStaticBool}
  MaterialExpressionStaticBool_9 {class=MaterialExpressionStaticBool}
  MaterialExpressionStaticSwitch_0 {
   class=MaterialExpressionStaticSwitch
   inputs [
    True {from=MaterialExpressionNamedRerouteUsage_168, output=""}
    False {from=MaterialExpressionConstant_23, output=""}
    Value {from=MaterialExpressionNamedRerouteUsage_9, output=""}
   ]
   properties {}
  }
  MaterialExpressionStaticSwitch_1 {
   class=MaterialExpressionStaticSwitch
   inputs [
    True {from=MaterialExpressionIf_5, output=""}
    False {from=MaterialExpressionReroute_228, output=""}
    Value {from=MaterialExpressionNamedRerouteUsage_85, output=""}
   ]
   properties {}
  }
  MaterialExpressionStaticSwitch_10 {
   class=MaterialExpressionStaticSwitch
   inputs [
    True {from=MaterialExpressionIf_6, output=""}
    False {from=MaterialExpressionReroute_3, output=""}
    Value {from=MaterialExpressionNamedRerouteUsage_23, output=""}
   ]
   properties {}
  }
  MaterialExpressionStaticSwitch_11 {
   class=MaterialExpressionStaticSwitch
   inputs [
    True {from=MaterialExpressionLinearInterpolate_5, output=""}
    False {from=MaterialExpressionReroute_50, output=""}
    Value {from=MaterialExpressionNamedRerouteUsage_3, output=""}
   ]
   properties {}
  }
  MaterialExpressionStaticSwitch_12 {
   class=MaterialExpressionStaticSwitch
   inputs [
    True {from=MaterialExpressionLinearInterpolate_11, output=""}
    False {from=MaterialExpressionReroute_9, output=""}
    Value {from=MaterialExpressionNamedRerouteUsage_5, output=""}
   ]
   properties {}
  }
  MaterialExpressionStaticSwitch_2 {
   class=MaterialExpressionStaticSwitch
   inputs [
    True {from=MaterialExpressionIf_3, output=""}
    False {from=MaterialExpressionReroute_230, output=""}
    Value {from=MaterialExpressionNamedRerouteUsage_15, output=""}
   ]
   properties {}
  }
  MaterialExpressionStaticSwitch_3 {
   class=MaterialExpressionStaticSwitch
   inputs [
    True {from=MaterialExpressionIf_2, output=""}
    False {from=MaterialExpressionReroute_231, output=""}
    Value {from=MaterialExpressionNamedRerouteUsage_16, output=""}
   ]
   properties {}
  }
  MaterialExpressionStaticSwitch_4 {
   class=MaterialExpressionStaticSwitch
   inputs [
    True {from=MaterialExpressionIf_0, output=""}
    False {from=MaterialExpressionReroute_142, output=""}
    Value {from=MaterialExpressionNamedRerouteUsage_17, output=""}
   ]
   properties {}
  }
  MaterialExpressionStaticSwitch_42 {
   class=MaterialExpressionStaticSwitch
   inputs [
    True {from=MaterialExpressionIf_8, output=""}
    False {from=MaterialExpressionReroute_209, output=""}
    Value {from=MaterialExpressionNamedRerouteUsage_305, output=""}
   ]
   properties {}
  }
  MaterialExpressionStaticSwitch_5 {
   class=MaterialExpressionStaticSwitch
   inputs [
    True {from=MaterialExpressionIf_7, output=""}
    False {from=MaterialExpressionReroute_229, output=""}
    Value {from=MaterialExpressionNamedRerouteUsage_18, output=""}
   ]
   properties {}
  }
  MaterialExpressionStaticSwitch_6 {
   class=MaterialExpressionStaticSwitch
   inputs [
    True {from=MaterialExpressionSaturate_1, output=""}
    False {from=MaterialExpressionSaturate_17, output=""}
    Value {from=MaterialExpressionNamedRerouteUsage_12, output=""}
   ]
   properties {desc="T= VTX Normals only\r\nF= Combined Material Normals"}
  }
  MaterialExpressionStaticSwitch_7 {
   class=MaterialExpressionStaticSwitch
   inputs [
    True {from=MaterialExpressionNamedRerouteUsage_7, output=""}
    False {from=MaterialExpressionNamedRerouteUsage_6, output=""}
    Value {from=MaterialExpressionNamedRerouteUsage_46, output=""}
   ]
   properties {desc="** DF AO is for Epic Settings Only!\r\nTrue = AO Baked from VTX or Texture\r\nFalse = AO Distance Field"}
  }
  MaterialExpressionSubtract_0 {
   class=MaterialExpressionSubtract
   inputs [
    A {from=MaterialExpressionNamedRerouteUsage_44}
    B {from=""}
   ]
   properties {desc=1}
  }
  MaterialExpressionSubtract_1 {
   class=MaterialExpressionSubtract
   inputs [
    A {from=MaterialExpressionReroute_48}
    B {from=""}
   ]
   properties {desc=1}
  }
  MaterialExpressionSubtract_19 {
   class=MaterialExpressionSubtract
   inputs [
    A {from=MaterialExpressionReroute_211}
    B {from=""}
   ]
   properties {desc=1}
  }
  MaterialExpressionTransform_4 {class=MaterialExpressionTransform, inputs [None {from=MaterialExpressionGetMaterialAttributes_16, output=Normal}]}
  MaterialExpressionVectorParameter_0 {
   class=MaterialExpressionVectorParameter
   properties {
    group=Mask
    parameterName="• VTX Color - Channel Weights"
    sortPriority=160
    channelNames {a="", b=Blue, g=Green, r=Red}
    defaultValue=<rgba 1 0 0 0>
   }
  }
  MaterialExpressionVectorParameter_2 {
   class=MaterialExpressionVectorParameter
   properties {
    group=Mask
    parameterName=* WS Noise - Channel Weights
    sortPriority=460
    channelNames {a="", b=Blue, g=Green, r=Red}
    defaultValue=<rgba 1 0 0 0>
   }
  }
  MaterialExpressionVectorParameter_5 {
   class=MaterialExpressionVectorParameter
   properties {
    group=Mask
    parameterName=* Baked AO/Curvature/Painted - Channel Weights
    sortPriority=360
    channelNames {a="", b=Blue, g=Green, r=Red}
    defaultValue=<rgba 1 0 0 0>
   }
  }
  MaterialExpressionVertexColor_1 {class=MaterialExpressionVertexColor, inputs [], properties {}}
  MaterialExpressionVertexInterpolator_5 {class=MaterialExpressionVertexInterpolator, inputs [VS {from=MaterialExpressionSaturate_33, output=""}], properties {}}
  MaterialExpressionVertexNormalWS_0 {class=MaterialExpressionVertexNormalWS, inputs [], properties {}}
  MaterialExpressionWorldPosition_0 {class=MaterialExpressionWorldPosition, inputs [], properties {}}
 }}
