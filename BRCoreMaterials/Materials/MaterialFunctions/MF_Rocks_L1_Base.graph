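# Material function for the base layer (L1) of a layered rock material.
# It exposes per-layer parameters for textures, UV or world-space triplanar, biplanar and Z-planar projection, fuzzy shading and displacement, feeding MF_Rocks_Layer.
# MF_Rocks_Layer_Blending applies the layer at full mask; detail-layer behavior is hard-coded off and the MF_Rocks_Layer_Masking call is unwired.
# Outputs are layer attributes, tintable mask and baked curvature, AO and painted masks.
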MaterialFunction /BRCoreMaterials/Materials/MaterialFunctions/MF_Rocks_L1_Base
graph {nodes [
  FunctionInput() {inputName="\\ --- Dependencies --- /", inputType=FunctionInput_Scalar, sortPriority=800}
  n1 = Constant() {r=1}
  n2 = StaticBool()
  n3 = MaterialFunctionCall!(
   A MaterialAttributes=SetMaterialAttributes!(MaterialAttributes)
   A Tintable Mask=Constant()
   B MaterialAttributes=NamedRerouteUsage()
   B Tintable Mask=NamedRerouteUsage()
   Layer B Mask=@n1
   Layer B Realtime Displacement Mask=@n1
   "\\ --- Blend Options --- /"
   Layer B - Is Used=NamedRerouteUsage()
   Layer B - Is Detail Layer=@n2
   Layer B - Add Normals=@n2
  ) {materialFunction=&/MF_Rocks_Layer_Blending}
  FunctionOutput(None=@n3.Blended MaterialAttributes) {outputName=Layer MaterialAttributes, sortPriority=-100}
  FunctionOutput(None=NamedRerouteUsage()) {outputName=Baked Curvature, sortPriority=15}
  FunctionOutput(None=NamedRerouteUsage()) {outputName=Painted Mask, sortPriority=17}
  FunctionOutput(None=NamedRerouteUsage()) {outputName=Baked AO, sortPriority=16}
  FunctionOutput(None=@n3.Tintable Mask) {outputName=Layer Tintable Mask, sortPriority=-50}
  MaterialFunctionCall!(
   Layer A MaterialAttributes
   Layer B MaterialAttributes
   * Masking - Invert
   "• Masking - Bias"
   "• Masking - Intensity"
   --- Use WORLD Z NORMAL MASKING ---
   "• World Z Normal - Use VTX Normals"
   * World Z Normal - Bias
   * World Z Normal - Intensity
   --- Use VTX COLOR MASKING ---
   "• VTX Color - Channel Weights"
   "• VTX Color - Invert"
   "• VTX Color - Blend (1=Max 0=Min)"
   "• VTX Color - Bias"
   "• VTX Color - Intensity"
   --- Use DFAO MASKING ---
   * DFAO - Force Baked AO on Nanite
   "• DFAO - Blend (1=Max 0=Min)"
   "• DFAO - Bias"
   "• DFAO - Intensity"
   --- Use Z GRADIENT MASKING ---
   "• Z Gradient - Blend (1=Max 0=Min)"
   "• Z Gradient - Height - Bottom"
   "• Z Gradient - Height - Top"
   --- Use BAKED AO/CURVATURE/PAINTED MASKING ---
   "• Baked AO/Curvature/Painted - Channel Weights"
   "• Baked AO/Curvature/Painted - Invert"
   "• Baked AO/Curvature/Painted - Blend (1=Max 0=Min)"
   "• Baked AO/Curvature/Painted - Bias"
   "• Baked AO/Curvature/Painted - Intensity"
   --- Use WATERLINE MASKING ---
   "* Waterline - Blend (1=Max 0=Min)"
   * Waterline - Bias
   * Waterline - Intensity
   --- Use WS NOISE MASKING ---
   "• WS Noise - Channel Weights"
   "• WS Noise - Invert"
   "• WS Noise - Blend (1=Blend 0=Min)"
   "• WS Noise - Bias"
   "• WS Noise - Intensity"
   --- Use HEIGHT MASKING ---
   "• Height - Selection (1=Current Layer 0=Previous Layers)"
   "• Height - Blend (1=Blend 0=Min)"
   "• Height - Min"
   "• Height - Max"
   "• Height - Bias"
   "• Height - Intensity"
   * Realtime Displacement - Masking - Bias
   * Realtime Displacement - Masking - Intensity
   * Realtime Displacement - WS Noise - Contribution
   * Realtime Displacement - Height - Contribution
   "\\ --- Dependencies --- /"
   Baked Curvature
   Baked AO
   Painted Mask
   DFAO
   Z Gradient
   WS Noise Masks
  ) {materialFunction=&/MF_Rocks_Layer_Masking}
  NamedRerouteDeclaration(None=StaticBoolParameter() {group="--  L1 | Base (TD1)", parameterName=* L1 | Base - Realtime Displacement - Use Costly Triplanar, sortPriority=700}) {name="Realtime Displacement - Use Costly Triplanar(SB)", nodeColor=<rgba 0.1979 0.0084 0 1>}
  n4 = MaterialFunctionCall!(
   Layer Is Used=NamedRerouteUsage()
   Is Tintable Layer=NamedRerouteUsage()
   Is Fuzzy Layer=NamedRerouteUsage()
   Is Detail Layer=StaticBool() {desc=Remove Detail Layer Functionality}
   Layer Contributes to Displacement=NamedRerouteUsage()
   Use SpecularMasks as AO/Curvature/Painted MSK=NamedRerouteUsage()
   Use VTX AO/Curvature/Painted on Nanite=NamedRerouteUsage()
   Use DEBUG UVs=NamedRerouteUsage()
   Use WS=NamedRerouteUsage()
   WS - Use Wavy Distortion=NamedRerouteUsage()
   WS - Wavy Distortion - Offset=NamedRerouteUsage()
   Triplanar - Offset=NamedRerouteUsage()
   Triplanar - Derivatives Scale=NamedRerouteUsage()
   Triplanar - Scale=NamedRerouteUsage()
   Triplanar - Fade=NamedRerouteUsage()
   "Triplanar - Fade (Low/Mobile)"=NamedRerouteUsage()
   Use Biplanar WS=NamedRerouteUsage()
   Biplanar - Scale=NamedRerouteUsage()
   Biplanar - Position Offset=NamedRerouteUsage()
   Biplanar - Fade Contrast=NamedRerouteUsage()
   Use Zplanar WS=NamedRerouteUsage()
   Zplanar - Scale=NamedRerouteUsage()
   Use Second UV Channel=NamedRerouteUsage()
   UV - Tiling=NamedRerouteUsage()
   Diffuse Texture Input=NamedRerouteUsage()
   Color - Multiply=NamedRerouteUsage()
   Detail - Color - Multiply
   Height - Min=NamedRerouteUsage()
   Height - Max=NamedRerouteUsage()
   Detail - Color - Intensity=NamedRerouteUsage()
   Detail - Specular - Intensity=NamedRerouteUsage()
   Detail - Specular - Contrast=NamedRerouteUsage()
   SpecularMasks Texture Input=NamedRerouteUsage()
   Specular=NamedRerouteUsage()
   Metallic=NamedRerouteUsage()
   Roughness=NamedRerouteUsage()
   "Roughness - Bias (Mobile)"=NamedRerouteUsage()
   Normal Texture Input=NamedRerouteUsage()
   "Normals - Value (Nanite)"=NamedRerouteUsage()
   "Normals - Value (non-Nanite)"=NamedRerouteUsage()
   Fuzzy - Intensity=NamedRerouteUsage()
   Fuzzy - Exponent=NamedRerouteUsage()
   Fuzzy - Color=NamedRerouteUsage()
   Fuzzy - Roughness=NamedRerouteUsage()
   Realtime Displacement - Use Costly Triplanar=NamedRerouteUsage()
   Realtime Displacement - Flat - Seam Sharpness=NamedRerouteUsage()
   Realtime Displacment - Min=NamedRerouteUsage()
   Realtime Displacment - Max=NamedRerouteUsage()
   Realtime Displacment - Triplanar Seam=NamedRerouteUsage()
   "\\ --- Dependencies --- /"
   Wavy Distortion World Position=NamedRerouteUsage()
   Fuzzy Shaded Mask
  ) {materialFunction=&/MF_Rocks_Layer}
  NamedRerouteDeclaration(None=@n4.Baked Curvature) {name=_Baked Curvature, nodeColor=<rgba 0 0 1 1>}
  n5 = ScalarParameter() {
   desc=Controls the contrast of the fade between the 2 projections. A value too high will remove the fade and create a harsh seam
   group="--  L1 | Base (TD1)"
   parameterName=* L1 | Base - Biplanar - Fade Contrast
   sortPriority=63
  }
  NamedRerouteDeclaration(None=@n5) {name=Biplanar - Fade Contrast, nodeColor=<rgba 0.5043 1 0 1>}
  NamedRerouteDeclaration(None=StaticBoolParameter() {group="--  L1 | Base (TD1)", parameterName=* L1 | Base - Is Detail Layer, sortPriority=1}) {name="Is Detail Layer(SB)", nodeColor=<rgba 0.1979 0.0084 0 1>}
  NamedRerouteDeclaration(None=StaticBoolParameter() {group="--  L1 | Base (TD1)", parameterName=* L1 | Base - Layer Contributes to Displacement, sortPriority=2}) {name="Layer Contributes to Displacement(SB)", nodeColor=<rgba 0.1979 0.0084 0 1>}
  NamedRerouteDeclaration(None=StaticBoolParameter() {defaultValue=true, group="--  L1 | Base (TD1)", parameterName=* L1 | Base - Use Normals WS, sortPriority=42}) {name="Use Normals WS(SB)", nodeColor=<rgba 0.1979 0.0084 0 1>}
  NamedRerouteDeclaration(None=StaticBoolParameter() {defaultValue=true, group="--  L1 | Base (TD1)", parameterName=* L1 | Base - Use Diffuse WS, sortPriority=40}) {name="Use Diffuse WS(SB)", nodeColor=<rgba 0.1979 0.0084 0 1>}
  NamedRerouteDeclaration(None=StaticBoolParameter() {defaultValue=true, group="--  L1 | Base (TD1)", parameterName=* L1 | Base - Use SpecularMasks WS, sortPriority=41}) {name="Use SpecularMasks WS(SB)", nodeColor=<rgba 0.1979 0.0084 0 1>}
  NamedRerouteDeclaration(None=@n4.Baked AO) {name=_Baked AO, nodeColor=<rgba 0 0 1 1>}
  n6 = TextureObjectParameter() {group="--  L1 | Base (TD1)", parameterName=Diffuse, sortPriority=320, texture=/Game/Global/Textures/Default/Blanks/T_FlatGrey}
  NamedRerouteDeclaration(None=@n6) {name=Diffuse TextureObject, nodeColor=<rgba 1 0.9427 0.9427 1>}
  n7 = TextureObjectParameter() {
   desc="Actually Just a Mask texture, named Spec_2 for TD to work"
   group="--  L1 | Base (TD1)"
   parameterName=SpecularMasks
   samplerType=SAMPLERTYPE_Masks
   sortPriority=370
   texture=&T_Fortnite_Default_S
  }
  NamedRerouteDeclaration(None=@n7) {name=SpecularMasks TextureObject, nodeColor=<rgba 1 0.9427 0.9427 1>}
  n8 = TextureObjectParameter() {
   group="--  L1 | Base (TD1)"
   parameterName=Normals
   samplerType=SAMPLERTYPE_Normal
   sortPriority=410
   texture=/Engine/EngineMaterials/FlatNormal
  }
  NamedRerouteDeclaration(None=@n8) {name=Normal TextureObject, nodeColor=<rgba 1 0.9427 0.9427 1>}
  NamedRerouteDeclaration(None=ScalarParameter() {defaultValue=700, group="--  L1 | Base (TD1)", parameterName=* L1 | Base - Triplanar - Scale, sortPriority=55}) {name=Triplanar - Scale, nodeColor=<rgba 0.5043 1 0 1>}
  NamedRerouteDeclaration(None=ScalarParameter() {defaultValue=10, group="--  L1 | Base (TD1)", parameterName=* L1 | Base - Triplanar - Fade, sortPriority=56}) {name=Triplanar - Fade, nodeColor=<rgba 0.5043 1 0 1>}
  NamedRerouteDeclaration(None=ScalarParameter() {defaultValue=75, group="--  L1 | Base (TD1)", parameterName="* L1 | Base - Triplanar - Fade (Low/Mobile)", sortPriority=57}) {name="Triplanar - Fade (Low/Mobile)", nodeColor=<rgba 0.5043 1 0 1>}
  NamedRerouteDeclaration(None=@n4.Painted Mask) {name=_Painted Mask, nodeColor=<rgba 0 0 1 1>}
  NamedRerouteDeclaration(None=ScalarParameter() {defaultValue=2, group="--  L1 | Base (TD1)", parameterName="* L1 | Base - Normals - Value (non-Nanite)", sortPriority=416}) {name="Normals - Value (non-Nanite)", nodeColor=<rgba 0.5043 1 0 1>}
  NamedRerouteDeclaration(None=ScalarParameter() {defaultValue=1, group="--  L1 | Base (TD1)", parameterName="* L1 | Base - Normals - Value (Nanite)", sortPriority=415}) {name="Normals - Value (Nanite)", nodeColor=<rgba 0.5043 1 0 1>}
  n9 = StaticBoolParameter() {
   desc="Does the SpecularMasks texture store | R = Curvature | G = AO | B = Painted Mask"
   group="--  L1 | Base (TD1)"
   parameterName=* L1 | Base - Use SpecularMasks as AO/Curvature/Painted MSK
   sortPriority=3
  }
  NamedRerouteDeclaration(None=@n9) {name="Use SpecularMasks as AO/Curvature/Painted MSK(SB)", nodeColor=<rgba 0.1979 0.0084 0 1>}
  NamedRerouteDeclaration(None=StaticBoolParameter() {group="--  L1 | Base (TD1)", parameterName=* L1 | Base - Use Second UV Channel, sortPriority=100}) {name="Use Second UV Channel(SB)", nodeColor=<rgba 0.1979 0.0084 0 1>}
  NamedRerouteDeclaration(None=ScalarParameter() {group="--  L1 | Base (TD1)", parameterName=* L1 | Base - Height - Min, sortPriority=331}) {name=Height - Min, nodeColor=<rgba 0.5043 1 0 1>}
  NamedRerouteDeclaration(None=ScalarParameter() {defaultValue=1, group="--  L1 | Base (TD1)", parameterName=* L1 | Base - Height - Max, sortPriority=332}) {name=Height - Max, nodeColor=<rgba 0.5043 1 0 1>}
  NamedRerouteDeclaration(None=StaticBoolParameter() {group="--  L1 | Base (TD1)", parameterName=* L1 | Base - Use WS, sortPriority=35}) {name="Use WS(SB)", nodeColor=<rgba 0.1979 0.0084 0 1>}
  NamedRerouteDeclaration(None=ScalarParameter() {defaultValue=0.5, group="--  L1 | Base (TD1)", parameterName=* L1 | Base - Specular, sortPriority=380}) {name=Specular, nodeColor=<rgba 0.5043 1 0 1>}
  NamedRerouteDeclaration(None=ScalarParameter() {group="--  L1 | Base (TD1)", parameterName=* L1 | Base - Metallic, sortPriority=385}) {name=Metallic, nodeColor=<rgba 0.5043 1 0 1>}
  NamedRerouteDeclaration(None=ScalarParameter() {defaultValue=0.5, group="--  L1 | Base (TD1)", parameterName=* L1 | Base - Roughness, sortPriority=390}) {name=Roughness, nodeColor=<rgba 0.5043 1 0 1>}
  n10 = ScalarParameter() {
   defaultValue=2550
   desc="Scale is the World Space size that the Biplanar projects, \r\nSmaller sizes = more tiling\r\nLarger sizes = less tiling"
   group="--  L1 | Base (TD1)"
   parameterName=* L1 | Base - Biplanar - Scale
   sortPriority=61
  }
  NamedRerouteDeclaration(None=@n10) {name=Biplanar - Scale, nodeColor=<rgba 0.5043 1 0 1>}
  NamedRerouteDeclaration(None=StaticBoolParameter() {group="--  L1 | Base (TD1)", parameterName=* L1 | Base - Is Fuzzy Layer, sortPriority=1}) {name="Is Fuzzy Layer(SB)", nodeColor=<rgba 0.1979 0.0084 0 1>}
  NamedRerouteDeclaration(None=ScalarParameter() {defaultValue=1, group="--  L1 | Base (TD1)", parameterName=* L1 | Base - Triplanar - Derivatives Scale, sortPriority=54}) {name=Triplanar - Derivatives Scale, nodeColor=<rgba 0.5043 1 0 1>}
  NamedRerouteDeclaration(None=ScalarParameter() {group="--  L1 | Base (TD1)", parameterName=* L1 | Base - Realtime Displacment - Min, sortPriority=715}) {name=Realtime Displacment - Min, nodeColor=<rgba 0.5043 1 0 1>}
  NamedRerouteDeclaration(None=ScalarParameter() {defaultValue=1, group="--  L1 | Base (TD1)", parameterName=* L1 | Base - Realtime Displacment - Max, sortPriority=716}) {name=Realtime Displacment - Max, nodeColor=<rgba 0.5043 1 0 1>}
  n11 = VectorParameter() {group="--  L1 | Base (TD1)", parameterName=* L1 | Base - Triplanar - Offset, sortPriority=54, channelNames {a="", b=Z, g=Y, r=X}}
  NamedRerouteDeclaration(None=@n11) {name=Triplanar - Offset, nodeColor=<rgba 0.5043 1 0 1>}
  n12 = VectorParameter() {
   group="--  L1 | Base (TD1)"
   parameterName=* L1 | Base - WS - Wavy Distortion - Offset
   sortPriority=51
   channelNames {a="", b=Z, g=Y, r=X}
   defaultValue=<rgba 0 0 0 100000>
  }
  NamedRerouteDeclaration(None=@n12) {name=Triplanar - Wavy Distortion - Offset, nodeColor=<rgba 0.5043 1 0 1>}
  n13 = ScalarParameter() {defaultValue=0.5, group="--  L1 | Base (TD1)", parameterName=* L1 | Base - Realtime Displacment - Triplanar Seam, sortPriority=720}
  NamedRerouteDeclaration(None=@n13) {name=Realtime Displacment - Triplanar Seam, nodeColor=<rgba 0.5043 1 0 1>}
  NamedRerouteDeclaration(None=@n4.MaterialAttributes) {name=Layer MateriaAttributes, nodeColor=<rgba 1 0 0.1875 1>}
  NamedRerouteDeclaration(None=ScalarParameter() {defaultValue=3, group="--  L1 | Base (TD1)", parameterName=* L1 | Base - Fuzzy - Exponent, sortPriority=455}) {name=Fuzzy - Exponent, nodeColor=<rgba 0.5043 1 0 1>}
  NamedRerouteDeclaration(None=ScalarParameter() {defaultValue=1, group="--  L1 | Base (TD1)", parameterName=* L1 | Base - Fuzzy - Intensity, sortPriority=450}) {name=Fuzzy - Intensity, nodeColor=<rgba 0.5043 1 0 1>}
  n14 = ScalarParameter() {
   defaultValue=0.6
   desc=This Roughness value will be blended to where the Fuzzy shading effect occurs.
   group="--  L1 | Base (TD1)"
   parameterName=* L1 | Base - Fuzzy - Roughness
   sortPriority=465
  }
  NamedRerouteDeclaration(None=@n14) {name=Fuzzy - Roughness, nodeColor=<rgba 0.5043 1 0 1>}
  NamedRerouteDeclaration(None=ScalarParameter() {group="--  L1 | Base (TD1)", parameterName="* L1 | Base - Roughness - Bias (Mobile)", sortPriority=391}) {name="Roughness - Bias (Mobile)", nodeColor=<rgba 0.5043 1 0 1>}
  NamedRerouteDeclaration(None=StaticBoolParameter() {group="--  L1 | Base (TD1)", parameterName=* L1 | Base - WS - Use Wavy Distortion, sortPriority=50}) {name="Triplanar - Use Wavy Distortion(SB)", nodeColor=<rgba 0.1979 0.0084 0 1>}
  NamedRerouteDeclaration(None=@n4.Tintable Mask) {name=Layer Tintable Mask, nodeColor=<rgba 1 0 0.1875 1>}
  NamedRerouteDeclaration(None=ScalarParameter() {defaultValue=1, group="--  L1 | Base (TD1)", parameterName=* L1 | Base - Is Tintable Layer, sortPriority=0}) {name=Is Tintable Layer, nodeColor=<rgba 0.5043 1 0 1>}
  NamedRerouteDeclaration(None=StaticBoolParameter() {group="--  L1 | Base (TD1)", parameterName=* L1 | Base - Use Diffuse Alpha as Height, sortPriority=325}) {name="Use Diffuse Alpha as Height(SB)", nodeColor=<rgba 0.1979 0.0084 0 1>}
  NamedRerouteDeclaration(None=ScalarParameter() {defaultValue=1, group="--  L1 | Base (TD1)", parameterName=* L1 | Base - UV - Tiling, sortPriority=105}) {name=UV - Tiling, nodeColor=<rgba 0.5043 1 0 1>}
  n15 = VectorParameter() {
   desc="This Color is either Multiplied or added to Layer's Color based on the Alpha channel value.\r\nNote that this effect happens after the Layer's Color Multiply"
   group="--  L1 | Base (TD1)"
   parameterName=* L1 | Base - Fuzzy - Color
   sortPriority=460
   channelNames {a="0=Multiply 1=Add", b="", g="", r=""}
   defaultValue=<rgba 1.1 1.1 1.1 0>
  }
  NamedRerouteDeclaration(None=AppendVector(A=@n15, B=@n15)) {name=Fuzzy - Color, nodeColor=<rgba 0.5043 1 0 1>}
  n16 = StaticBoolParameter() {
   desc="Enable to use side projected planar mapping(Sediment Cliffs etc)"
   group="--  L1 | Base (TD1)"
   parameterName=* L1 | Base - Use Biplanar WS
   sortPriority=60
  }
  NamedRerouteDeclaration(None=@n16) {name=Biplanar - Use Biplanar WS, nodeColor=<rgba 0.1979 0.0084 0 1>}
  NamedRerouteDeclaration(None=FunctionInput() {inputName=Wavy Distortion World Position, sortPriority=1005}) {name=_Wavy Distortion World Position, nodeColor=<rgba 1 0 0.0525 1>}
  NamedRerouteDeclaration(None=ScalarParameter() {defaultValue=700, group="--  L1 | Base (TD1)", parameterName=* L1 | Base - Zplanar - Scale, sortPriority=85}) {name=Zplanar - Scale, nodeColor=<rgba 0.5043 1 0 1>}
  NamedRerouteDeclaration(None=StaticBoolParameter() {group="--  L1 | Base (TD1)", parameterName=* L1 | Base - Use Zplanar WS, sortPriority=80}) {name="Zplanar - Use Zplanar WS(SB)", nodeColor=<rgba 0.1979 0.0084 0 1>}
  n17 = ScalarParameter() {defaultValue=10, group="--  L1 | Base (TD1)", parameterName=* L1 | Base - Realtime Displacement - Flat - Seam Sharpness, sortPriority=705}
  NamedRerouteDeclaration(None=@n17) {name=Realtime Displacement - Flat - Seam Sharpness, nodeColor=<rgba 0.5043 1 0 1>}
  NamedRerouteDeclaration(None=StaticBoolParameter() {group="--  L1 | Base (TD1)", parameterName=* L1 | Base - Use DEBUG UVs, sortPriority=30}) {name="Use DEBUG UVs(SB)", nodeColor=<rgba 0.1979 0.0084 0 1>}
  NamedRerouteDeclaration(None=ScalarParameter() {defaultValue=1, group="--  L1 | Base (TD1)", parameterName=* L1 | Base - Detail - Color - Intensity, sortPriority=340}) {name=Detail - Color - Intensity, nodeColor=<rgba 0.5043 1 0 1>}
  NamedRerouteDeclaration(None=ScalarParameter() {defaultValue=1, group="--  L1 | Base (TD1)", parameterName=* L1 | Base - Detail - Specular - Intensity, sortPriority=341}) {name=Detail - Specular - Intensity, nodeColor=<rgba 0.5043 1 0 1>}
  n18 = StaticBoolParameter() {
   desc="OFF = Nanite uses VTX and Default uses Textures (Prop Rocks are the only meshes that gets tesselated and displaced on nanite platforms, by default they won`t use VTX)\r\nON = Force VTX for all (VTX masks are only used on big custom sculpts that have a custom nanite mesh with no UVs)\r\n"
   group="--  L1 | Base (TD1)"
   parameterName=* L1 | Base - Use VTX AO/Curvature/Painted on Nanite
   sortPriority=10
  }
  NamedRerouteDeclaration(None=@n18) {name="Use VTX AO/Curvature/Painted on Nanite(SB)", nodeColor=<rgba 0.1979 0.0084 0 1>}
  n19 = VectorParameter() {group="--  L1 | Base (TD1)", parameterName=* L1 | Base - Color - Multiply, sortPriority=321, defaultValue=<rgba 1 1 1 0>}
  NamedRerouteDeclaration(None=@n19) {name=Color - Multiply, nodeColor=<rgba 0.5043 1 0 1>}
  NamedRerouteDeclaration(None=ScalarParameter() {defaultValue=1, group="--  L1 | Base (TD1)", parameterName=* L1 | Base - Detail - Specular - Contrast, sortPriority=342}) {name=Detail - Specular - Contrast, nodeColor=<rgba 0.5043 1 0 1>}
  NamedRerouteDeclaration(None=FunctionInput(Preview=StaticBool() {value=true}) {bUsePreviewValueAsDefault=true, inputName=Layer Is Used, inputType=FunctionInput_StaticBool, sortPriority=-200}) {name="Layer Is Used(SB)", nodeColor=<rgba 0.1979 0.0084 0 1>}
  n20 = VectorParameter() {group="--  L1 | Base (TD1)", parameterName=* L1 | Base - Biplanar - Position Offset, sortPriority=62, defaultValue=<rgba 0 0 0 1>}
  NamedRerouteDeclaration(None=@n20.RGB) {name=Biplanar - Position Offset, nodeColor=<rgba 0.5043 1 0 1>}
  NamedRerouteUsage()
  NamedRerouteUsage()
  Reroute(None=NamedRerouteUsage())
  Reroute(None=NamedRerouteUsage())
  Reroute(None=Reroute())
  Reroute(None=NamedRerouteUsage())
  Reroute(None=NamedRerouteUsage())
  Reroute(None=NamedRerouteUsage())
  StaticBool() {desc=Remove Baked Masks Support from this layer}
 ]}
