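# Material function for rock layer 4 (Extra/TD4), exposing its parameters and blending it over the incoming material attributes and tintable mask.
# MF_Rocks_Layer builds the layer with optional triplanar, biplanar or Z-planar projection, with its baked-mask support hard-wired off.
# MF_Rocks_Layer_Masking combines switchable normal, vertex color, DFAO, gradient, baked, waterline, noise and height masks.
# MF_Rocks_Layer_Blending applies that mask plus a separate realtime displacement mask.
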MaterialFunction /BRCoreMaterials/Materials/MaterialFunctions/MF_Rocks_L4_Extra
graph {nodes [
  FunctionInput() {inputName="\\ --- Dependencies --- /", inputType=FunctionInput_Scalar, sortPriority=800}
  n1 = MaterialFunctionCall!(
   A MaterialAttributes=NamedRerouteUsage()
   A Tintable Mask=NamedRerouteUsage()
   B MaterialAttributes=NamedRerouteUsage()
   B Tintable Mask=NamedRerouteUsage()
   Layer B Mask=NamedRerouteUsage()
   Layer B Realtime Displacement Mask=NamedRerouteUsage()
   "\\ --- Blend Options --- /"
   Layer B - Is Used=NamedRerouteUsage()
   Layer B - Is Detail Layer=NamedRerouteUsage()
   Layer B - Add Normals=NamedRerouteUsage()
  ) {materialFunction=&/MF_Rocks_Layer_Blending}
  FunctionOutput(None=@n1.Blended MaterialAttributes) {outputName=Blended MaterialAttributes, sortPriority=-100}
  FunctionOutput(None=NamedRerouteUsage()) {outputName=Blend Mask, sortPriority=-25}
  FunctionOutput(None=NamedRerouteUsage()) {outputName=Layer MaterialAttributes}
  FunctionOutput(None=@n1.Tintable Mask) {outputName=Blended Tintable Mask, sortPriority=-50}
  NamedRerouteDeclaration(None=FunctionInput() {bUsePreviewValueAsDefault=true, inputName=DFAO, inputType=FunctionInput_Scalar, sortPriority=813}) {name=_DFAO, nodeColor=<rgba 1 0 0.0525 1>}
  n2 = StaticBoolParameter() {defaultValue=true, group="--  L4 | Extra (TD4) Blending", parameterName=* L4 | Blending - Normals - Add Normals, sortPriority=0}
  NamedRerouteDeclaration(None=@n2) {name="* Blending - Normals - Add Normals(SB)", nodeColor=<rgba 0.1979 0.0084 0 1>}
  NamedRerouteDeclaration(None=StaticBoolParameter() {group="--  L4 | Extra (TD4) Blending", parameterName=* L4 | Masking --- Use HEIGHT MASKING ---, sortPriority=500}) {name=--- Use HEIGHT MASKING ---, nodeColor=<rgba 0.1979 0.0084 0 1>}
  NamedRerouteDeclaration(None=ScalarParameter() {defaultValue=1, group="--  L4 | Extra (TD4) Blending", parameterName=* L4 | Masking - Height - Intensity, sortPriority=516}) {name="• Height - Intensity", nodeColor=<rgba 0.5043 1 0 1>}
  NamedRerouteDeclaration(None=FunctionInput() {bUsePreviewValueAsDefault=true, inputName=Baked Curvature, inputType=FunctionInput_Scalar, sortPriority=810}) {name=_Baked Curvature, nodeColor=<rgba 1 0 0.0525 1>}
  NamedRerouteDeclaration(None=FunctionInput() {bUsePreviewValueAsDefault=true, inputName=Painted Mask, inputType=FunctionInput_Scalar, sortPriority=812}) {name=_Painted Mask, nodeColor=<rgba 1 0 0.0525 1>}
  NamedRerouteDeclaration(None=StaticBoolParameter() {group="--  L4 | Extra (TD4)", parameterName=* L4 | Extra - Is Detail Layer, sortPriority=1}) {name="Is Detail Layer(SB)", nodeColor=<rgba 0.1979 0.0084 0 1>}
  NamedRerouteDeclaration(None=StaticBoolParameter() {group="--  L4 | Extra (TD4)", parameterName=* L4 | Extra - Layer Contributes to Displacement, sortPriority=2}) {name="Layer Contributes to Displacement(SB)", nodeColor=<rgba 0.1979 0.0084 0 1>}
  NamedRerouteDeclaration(None=StaticBoolParameter() {defaultValue=true, group="--  L4 | Extra (TD4)", parameterName=* L4 | Extra - Use Normals WS, sortPriority=42}) {name="Use Normals WS(SB)", nodeColor=<rgba 0.1979 0.0084 0 1>}
  NamedRerouteDeclaration(None=StaticBoolParameter() {defaultValue=true, group="--  L4 | Extra (TD4)", parameterName=* L4 | Extra - Use Diffuse WS, sortPriority=40}) {name="Use Diffuse WS(SB)", nodeColor=<rgba 0.1979 0.0084 0 1>}
  n3 = ScalarParameter() {defaultValue=-0.2, group="--  L4 | Extra (TD4) Blending", parameterName=* L4 | Masking - World Z Normal - Bias, sortPriority=120}
  NamedRerouteDeclaration(None=@n3) {name=* World Z Normal - Bias, nodeColor=<rgba 0.5043 1 0 1>}
  NamedRerouteDeclaration(None=ScalarParameter() {group="--  L4 | Extra (TD4) Blending", parameterName=* L4 | Masking - Height - Bias, sortPriority=515}) {name="• Height - Bias", nodeColor=<rgba 0.5043 1 0 1>}
  NamedRerouteDeclaration(None=StaticBoolParameter() {group="--  L4 | Extra (TD4) Blending", parameterName=* L4 | Masking --- Use WORLD Z NORMAL MASKING ---, sortPriority=100}) {name="== WORLD Z NORMAL MASKING ==", nodeColor=<rgba 0.1979 0.0084 0 1>}
  n4 = ScalarParameter() {defaultValue=1.5, group="--  L4 | Extra (TD4) Blending", parameterName=* L4 | Masking - World Z Normal - Intensity, sortPriority=121}
  NamedRerouteDeclaration(None=@n4) {name=* World Z Normal - Intensity, nodeColor=<rgba 0.5043 1 0 1>}
  n5 = StaticBoolParameter() {
   desc="T= VTX Normals only\r\nF= Combined Material Normals"
   group="--  L4 | Extra (TD4) Blending"
   parameterName=* L4 | Masking - World Z Normal - Use VTX Normals
   sortPriority=110
  }
  NamedRerouteDeclaration(None=@n5) {name="• World Z Normal - Use VTX Normals", desc="T= VTX Normals only\r\nF= Combined Material Normals", nodeColor=<rgba 0.1979 0.0084 0 1>}
  n6 = ScalarParameter() {defaultValue=1, group="--  L4 | Extra (TD4) Blending", parameterName=* L4 | Masking - VTX Color - Intensity, sortPriority=176}
  NamedRerouteDeclaration(None=@n6) {name="• VTX Color - Intensity", nodeColor=<rgba 0.5043 1 0 1>}
  NamedRerouteDeclaration(None=StaticBoolParameter() {group="--  L4 | Extra (TD4) Blending", parameterName=* L4 | Masking --- Use VTX COLOR MASKING ---, sortPriority=150}) {name=--- Use VTX COLOR MASKING ---, nodeColor=<rgba 0.1979 0.0084 0 1>}
  NamedRerouteDeclaration(None=StaticBoolParameter() {group="--  L4 | Extra (TD4) Blending", parameterName=* L4 | Masking --- Use DFAO MASKING ---, sortPriority=200}) {name=--- Use DFAO MASKING ---, nodeColor=<rgba 0.1979 0.0084 0 1>}
  NamedRerouteDeclaration(None=StaticBoolParameter() {group="--  L4 | Extra (TD4) Blending", parameterName=* L4 | Masking --- Use Z GRADIENT MASKING ---, sortPriority=250}) {name="== Z GRADIENT MASKING ==", nodeColor=<rgba 0.1979 0.0084 0 1>}
  n7 = StaticBoolParameter() {group="--  L4 | Extra (TD4) Blending", parameterName=* L4 | Masking --- Use BAKED AO/CURVATURE/PAINTED MASKING ---, sortPriority=350}
  NamedRerouteDeclaration(None=@n7) {name=--- Use BAKED AO/CURVATURE/PAINTED MASKING ---, nodeColor=<rgba 0.1979 0.0084 0 1>}
  NamedRerouteDeclaration(None=StaticBoolParameter() {group="--  L4 | Extra (TD4) Blending", parameterName=* L4 | Masking --- Use WS NOISE MASKING ---, sortPriority=450}) {name="== WS NOISE MASKING ==", nodeColor=<rgba 0.1979 0.0084 0 1>}
  NamedRerouteDeclaration(None=StaticBoolParameter() {defaultValue=true, group="--  L4 | Extra (TD4)", parameterName=* L4 | Extra - Use SpecularMasks WS, sortPriority=41}) {name="Use SpecularMasks WS(SB)", nodeColor=<rgba 0.1979 0.0084 0 1>}
  n8 = ScalarParameter() {
   desc="Inverts the combined layer masking options\r\n*Adjust these masking options last for best results"
   group="--  L4 | Extra (TD4) Blending"
   parameterName=* L4 | Masking - Invert
   sortPriority=50
  }
  NamedRerouteDeclaration(None=@n8) {name=* Masking - Invert, nodeColor=<rgba 0.5043 1 0 1>}
  n9 = ScalarParameter() {
   desc="Bias for tweaking the combined layer masking options\r\n*Adjust these masking options last for best results"
   group="--  L4 | Extra (TD4) Blending"
   parameterName=* L4 | Masking - Bias
   sortPriority=60
  }
  NamedRerouteDeclaration(None=@n9) {name="• Masking - Bias", nodeColor=<rgba 0.5043 1 0 1>}
  n10 = ScalarParameter() {
   defaultValue=1
   desc="Intensity for tweaking the combined layer masking options\r\n*Adjust these masking options last for best results"
   group="--  L4 | Extra (TD4) Blending"
   parameterName=* L4 | Masking - Intensity
   sortPriority=61
  }
  NamedRerouteDeclaration(None=@n10) {name="• Masking - Intensity", nodeColor=<rgba 0.5043 1 0 1>}
  n11 = ScalarParameter() {defaultValue=1, group="--  L4 | Extra (TD4) Blending", parameterName=* L4 | Masking - WS Noise - Intensity, sortPriority=476}
  NamedRerouteDeclaration(None=@n11) {name="• WS Noise - Intensity", nodeColor=<rgba 0.5043 1 0 1>}
  NamedRerouteDeclaration(None=ScalarParameter() {group="--  L4 | Extra (TD4) Blending", parameterName=* L4 | Masking - WS Noise - Bias, sortPriority=475}) {name="• WS Noise - Bias", nodeColor=<rgba 0.5043 1 0 1>}
  n12 = ScalarParameter() {
   defaultValue=0.1
   group="--  L4 | Extra (TD4) Blending"
   parameterName=* L4 | Masking - Baked AO/Curvature/Painted - Intensity
   sortPriority=376
  }
  NamedRerouteDeclaration(None=@n12) {name="• Baked AO/Curvature/Painted - Intensity", nodeColor=<rgba 0.5043 1 0 1>}
  NamedRerouteDeclaration(None=ScalarParameter() {group="--  L4 | Extra (TD4) Blending", parameterName=* L4 | Masking - Baked AO/Curvature/Painted - Bias, sortPriority=375}) {name="• Baked AO/Curvature/Painted - Bias", nodeColor=<rgba 0.5043 1 0 1>}
  n13 = ScalarParameter() {
   defaultValue=1
   desc="1= object top bound\r\n0= object bottom bound"
   group="--  L4 | Extra (TD4) Blending"
   parameterName=* L4 | Masking - Z Gradient - Height - Top
   sortPriority=266
  }
  NamedRerouteDeclaration(None=@n13) {name="• Z Gradient - Height - Top", nodeColor=<rgba 0.5043 1 0 1>}
  n14 = ScalarParameter() {
   desc="1= object top bound\r\n0= object bottom bound"
   group="--  L4 | Extra (TD4) Blending"
   parameterName=* L4 | Masking - Z Gradient - Height - Bottom
   sortPriority=265
  }
  NamedRerouteDeclaration(None=@n14) {name="• Z Gradient - Height - Bottom", nodeColor=<rgba 0.5043 1 0 1>}
  NamedRerouteDeclaration(None=ScalarParameter() {defaultValue=1, group="--  L4 | Extra (TD4) Blending", parameterName=* L4 | Masking - DFAO - Intensity, sortPriority=226}) {name="• DFAO - Intensity", nodeColor=<rgba 0.5043 1 0 1>}
  NamedRerouteDeclaration(None=FunctionInput(Preview=SetMaterialAttributes!(MaterialAttributes)) {inputName=Input MaterialAttributes, inputType=FunctionInput_MaterialAttributes, sortPriority=-200}) {name=Input MaterialAttributes, nodeColor=<rgba 0 1 0 1>}
  NamedRerouteDeclaration(None=FunctionInput(Preview=Constant3Vector()) {bUsePreviewValueAsDefault=true, inputName=WS Noise Masks, sortPriority=815}) {name=_WS Noise Masks, nodeColor=<rgba 1 0 0.0525 1>}
  NamedRerouteDeclaration(None=ScalarParameter() {group="--  L4 | Extra (TD4) Blending", parameterName=* L4 | Masking - DFAO - Bias, sortPriority=225}) {name="• DFAO - Bias", nodeColor=<rgba 0.5043 1 0 1>}
  n15 = StaticBoolParameter() {defaultValue=true, group="--  L4 | Extra (TD4) Blending", parameterName=* L4 | Masking - DFAO - Force Baked AO on Nanite, sortPriority=210}
  NamedRerouteDeclaration(None=@n15) {name=* DFAO - Force Baked AO on Nanite, nodeColor=<rgba 0.1979 0.0084 0 1>}
  n16 = ScalarParameter() {
   defaultValue=1
   group="--  L4 | Extra (TD4) Blending"
   parameterName="* L4 | Masking - VTX Color - Blend (1=Max 0=Min)"
   sliderMax=1
   sortPriority=170
  }
  NamedRerouteDeclaration(None=@n16) {name="• VTX Color - Blend (1=Max 0=Min)", nodeColor=<rgba 0.5043 1 0 1>}
  n17 = ScalarParameter() {
   defaultValue=1
   group="--  L4 | Extra (TD4) Blending"
   parameterName="* L4 | Masking - DFAO - Blend (1=Max 0=Min)"
   sliderMax=1
   sortPriority=220
  }
  NamedRerouteDeclaration(None=@n17) {name="• DFAO - Blend (1=Max 0=Min)", nodeColor=<rgba 0.5043 1 0 1>}
  n18 = ScalarParameter() {
   defaultValue=1
   group="--  L4 | Extra (TD4) Blending"
   parameterName="* L4 | Masking - Z Gradient - Blend (1=Max 0=Min)"
   sliderMax=1
   sortPriority=260
  }
  NamedRerouteDeclaration(None=@n18) {name="• Z Gradient - Blend (1=Max 0=Min)", nodeColor=<rgba 0.5043 1 0 1>}
  n19 = ScalarParameter() {
   defaultValue=1
   group="--  L4 | Extra (TD4) Blending"
   parameterName="* L4 | Masking - Baked AO/Curvature/Painted - Blend (1=Max 0=Min)"
   sliderMax=1
   sortPriority=370
  }
  NamedRerouteDeclaration(None=@n19) {name="• Baked AO/Curvature/Painted - Blend (1=Max 0=Min)", nodeColor=<rgba 0.5043 1 0 1>}
  n20 = ScalarParameter() {
   defaultValue=1
   group="--  L4 | Extra (TD4) Blending"
   parameterName="* L4 | Masking - WS Noise - Blend (1=Blend 0=Min)"
   sliderMax=1
   sortPriority=470
  }
  NamedRerouteDeclaration(None=@n20) {name="• WS Noise - Blend (1=Blend 0=Min)", nodeColor=<rgba 0.5043 1 0 1>}
  n21 = TextureObjectParameter() {group="--  L4 | Extra (TD4)", parameterName=Diffuse_Texture_4, sortPriority=320, texture=/Game/Global/Textures/Default/Blanks/T_FlatGrey}
  NamedRerouteDeclaration(None=@n21) {name=Diffuse TextureObject, nodeColor=<rgba 1 0.9427 0.9427 1>}
  n22 = TextureObjectParameter() {
   desc="Actually Just a Mask texture, named Spec_2 for TD to work"
   group="--  L4 | Extra (TD4)"
   parameterName=SpecularMasks_4
   samplerType=SAMPLERTYPE_Masks
   sortPriority=370
   texture=&T_Fortnite_Default_S
  }
  NamedRerouteDeclaration(None=@n22) {name=SpecularMasks TextureObject, nodeColor=<rgba 1 0.9427 0.9427 1>}
  n23 = TextureObjectParameter() {
   group="--  L4 | Extra (TD4)"
   parameterName=Normals_Texture_4
   samplerType=SAMPLERTYPE_Normal
   sortPriority=410
   texture=/Engine/EngineMaterials/FlatNormal
  }
  NamedRerouteDeclaration(None=@n23) {name=Normal TextureObject, nodeColor=<rgba 1 0.9427 0.9427 1>}
  NamedRerouteDeclaration(None=ScalarParameter() {defaultValue=700, group="--  L4 | Extra (TD4)", parameterName=* L4 | Extra - Triplanar - Scale, sortPriority=55}) {name=Triplanar - Scale, nodeColor=<rgba 0.5043 1 0 1>}
  n24 = VectorParameter() {group="--  L4 | Extra (TD4)", parameterName=* L4 | Extra - Detail - Color - Multiply, sortPriority=321, defaultValue=<rgba 1 1 1 0>}
  NamedRerouteDeclaration(None=@n24) {name=Detail - Color - Multiply, nodeColor=<rgba 0.5043 1 0 1>}
  NamedRerouteDeclaration(None=ScalarParameter() {defaultValue=10, group="--  L4 | Extra (TD4)", parameterName=* L4 | Extra - Triplanar - Fade, sortPriority=56}) {name=Triplanar - Fade, nodeColor=<rgba 0.5043 1 0 1>}
  n25 = ScalarParameter() {
   defaultValue=1
   group="--  L4 | Extra (TD4) Blending"
   parameterName="* L4 | Masking - Height - Selection (1=Current Layer 0=Previous Layers)"
   sortPriority=509
  }
  NamedRerouteDeclaration(None=@n25) {name="• Height - Selection (1=Current Layer 0=Previous Layers)", nodeColor=<rgba 0.5043 1 0 1>}
  NamedRerouteDeclaration(None=ScalarParameter() {defaultValue=75, group="--  L4 | Extra (TD4)", parameterName="* L4 | Extra - Triplanar - Fade (Low/Mobile)", sortPriority=57}) {name="Triplanar - Fade (Low/Mobile)", nodeColor=<rgba 0.5043 1 0 1>}
  NamedRerouteDeclaration(None=StaticBoolParameter() {defaultValue=true, group="--  L4 | Extra (TD4)", parameterName=* L4 | Extra - Is Masked Layer, sortPriority=0}) {name="Is Masked Layer (SB)", nodeColor=<rgba 0.1979 0.0084 0 1>}
  NamedRerouteDeclaration() {name=Input Mobile MaterialAttributes, nodeColor=<rgba 0 0.5 1 1>}
  NamedRerouteDeclaration(None=ScalarParameter() {defaultValue=2, group="--  L4 | Extra (TD4)", parameterName="* L4 | Extra - Normals - Value (non-Nanite)", sortPriority=416}) {name="Normals - Value (non-Nanite)", nodeColor=<rgba 0.5043 1 0 1>}
  NamedRerouteDeclaration(None=ScalarParameter() {defaultValue=1, group="--  L4 | Extra (TD4)", parameterName="* L4 | Extra - Normals - Value (Nanite)", sortPriority=415}) {name="Normals - Value (Nanite)", nodeColor=<rgba 0.5043 1 0 1>}
  n26 = StaticBoolParameter() {
   desc="Does the SpecularMasks texture store | R = Curvature | G = AO | B = Painted Mask"
   group="--  L4 | Extra (TD4)"
   parameterName=* L4 | Extra - Use SpecularMasks as AO/Curvature/Painted MSK
   sortPriority=3
  }
  NamedRerouteDeclaration(None=@n26) {name="Use SpecularMasks as AO/Curvature/Painted MSK(SB)", nodeColor=<rgba 0.1979 0.0084 0 1>}
  NamedRerouteDeclaration(None=StaticBoolParameter() {group="--  L4 | Extra (TD4)", parameterName=* L4 | Extra - Use Second UV Channel, sortPriority=100}) {name="Use Second UV Channel(SB)", nodeColor=<rgba 0.1979 0.0084 0 1>}
  NamedRerouteDeclaration(None=ScalarParameter() {group="--  L4 | Extra (TD4)", parameterName=* L4 | Extra - Height - Min, sortPriority=331}) {name=Height - Min, nodeColor=<rgba 0.5043 1 0 1>}
  NamedRerouteDeclaration(None=ScalarParameter() {defaultValue=1, group="--  L4 | Extra (TD4)", parameterName=* L4 | Extra - Height - Max, sortPriority=332}) {name=Height - Max, nodeColor=<rgba 0.5043 1 0 1>}
  NamedRerouteDeclaration(None=StaticBoolParameter() {group="--  L4 | Extra (TD4)", parameterName=* L4 | Extra - Use WS, sortPriority=35}) {name="Use WS(SB)", nodeColor=<rgba 0.1979 0.0084 0 1>}
  NamedRerouteDeclaration(None=ScalarParameter() {defaultValue=0.5, group="--  L4 | Extra (TD4)", parameterName=* L4 | Extra - Specular, sortPriority=380}) {name=Specular, nodeColor=<rgba 0.5043 1 0 1>}
  NamedRerouteDeclaration(None=ScalarParameter() {group="--  L4 | Extra (TD4)", parameterName=* L4 | Extra - Metallic, sortPriority=385}) {name=Metallic, nodeColor=<rgba 0.5043 1 0 1>}
  NamedRerouteDeclaration(None=ScalarParameter() {defaultValue=0.5, group="--  L4 | Extra (TD4)", parameterName=* L4 | Extra - Roughness, sortPriority=390}) {name=Roughness, nodeColor=<rgba 0.5043 1 0 1>}
  n27 = FunctionInput() {inputName=Input Tintable Mask, inputType=FunctionInput_Scalar, sortPriority=-199, previewValue=<1 0 0 1>}
  NamedRerouteDeclaration(None=@n27) {name=Input Tintable Mask, nodeColor=<rgba 0 1 0 1>}
  NamedRerouteDeclaration(None=StaticBoolParameter() {group="--  L4 | Extra (TD4) Blending", parameterName=* L4 | Masking --- Use WATERLINE MASKING ---, sortPriority=400}) {name=--- Use WATERLINE MASKING ---, nodeColor=<rgba 0.1979 0.0084 0 1>}
  NamedRerouteDeclaration(None=ScalarParameter() {group="--  L4 | Extra (TD4) Blending", parameterName=* L4 | Masking - Waterline - Bias, sortPriority=415}) {name=* Waterline - Bias, nodeColor=<rgba 0.5043 1 0 1>}
  n28 = ScalarParameter() {group="--  L4 | Extra (TD4) Blending", parameterName="* L4 | Masking - Waterline - Blend (1=Max 0=Min)", sliderMax=1, sortPriority=410}
  NamedRerouteDeclaration(None=@n28) {name="* Waterline - Blend (1=Max 0=Min)", nodeColor=<rgba 0.5043 1 0 1>}
  n29 = ScalarParameter() {defaultValue=1, group="--  L4 | Extra (TD4) Blending", parameterName=* L4 | Masking - Waterline - Intensity, sortPriority=416}
  NamedRerouteDeclaration(None=@n29) {name=* Waterline - Intensity, nodeColor=<rgba 0.5043 1 0 1>}
  n30 = VectorParameter() {group="--  L4 | Extra (TD4)", parameterName=* L4 | Extra - Triplanar - Offset, sortPriority=53, channelNames {a="", b=Z, g=Y, r=X}}
  NamedRerouteDeclaration(None=@n30) {name=Triplanar - Offset, nodeColor=<rgba 0.5043 1 0 1>}
  n31 = VectorParameter() {
   group="--  L4 | Extra (TD4)"
   parameterName=* L4 | Extra - WS - Wavy Distortion - Offset
   sortPriority=51
   channelNames {a="", b=Z, g=Y, r=X}
   defaultValue=<rgba 0 0 0 100000>
  }
  NamedRerouteDeclaration(None=@n31) {name=Triplanar - Wavy Distortion - Offset, nodeColor=<rgba 0.5043 1 0 1>}
  NamedRerouteDeclaration(None=ScalarParameter() {group="--  L4 | Extra (TD4) Blending", parameterName=* L4 | Masking - Baked AO/Curvature/Painted - Invert, sortPriority=361}) {name="• Baked AO/Curvature/Painted - Invert", nodeColor=<rgba 0.5043 1 0 1>}
  n32 = VectorParameter() {
   group="--  L4 | Extra (TD4) Blending"
   parameterName=* L4 | Masking - WS Noise - Channel Weights
   sortPriority=460
   channelNames {a="", b=Blue, g=Green, r=Red}
   defaultValue=<rgba 1 0 0 0>
  }
  NamedRerouteDeclaration(None=@n32.RGB) {name="• WS Noise - Channel Weights", nodeColor=<rgba 0.5043 1 0 1>}
  NamedRerouteDeclaration(None=ScalarParameter() {group="--  L4 | Extra (TD4) Blending", parameterName=* L4 | Masking - WS Noise - Invert, sortPriority=461}) {name="• WS Noise - Invert", nodeColor=<rgba 0.5043 1 0 1>}
  n33 = VectorParameter() {
   desc="Sets amount that each VTX color contributes to layer masking\r\nTotal weight value over 1 will increase sharpness of the resulting masking"
   group="--  L4 | Extra (TD4) Blending"
   parameterName=* L4 | Masking - VTX Color - Channel Weights
   sortPriority=160
   channelNames {a="", b=Blue, g=Green, r=Red}
   defaultValue=<rgba 1 0 0 0>
  }
  NamedRerouteDeclaration(None=@n33.RGB) {name="• VTX Color - Channel Weights", nodeColor=<rgba 0.5043 1 0 1>}
  n34 = StaticBool() {desc=Remove Baked Masks Support from this layer}
  n35 = MaterialFunctionCall!(
   Layer Is Used=NamedRerouteUsage()
   Is Tintable Layer=NamedRerouteUsage()
   Is Fuzzy Layer=NamedRerouteUsage()
   Is Detail Layer=NamedRerouteUsage()
   Layer Contributes to Displacement=NamedRerouteUsage()
   Use SpecularMasks as AO/Curvature/Painted MSK=@n34
   Use VTX AO/Curvature/Painted on Nanite=@n34
   Use DEBUG UVs=NamedRerouteUsage()
   Use WS=NamedRerouteUsage()
   WS - Use Wavy Distortion=NamedRerouteUsage()
   WS - Wavy Distortion - Offset=NamedRerouteUsage()
   Triplanar - Offset=NamedRerouteUsage()
   Triplanar - Derivatives Scale=NamedRerouteUsage()
   Triplanar - Scale=NamedRerouteUsage()
   Triplanar - Fade=NamedRerouteUsage()
   "Triplanar - Fade (Low/Mobile)"=NamedRerouteUsage()
   Use Biplanar WS=NamedRerouteUsage()
   Biplanar - Scale=NamedRerouteUsage()
   Biplanar - Position Offset=NamedRerouteUsage()
   Biplanar - Fade Contrast=NamedRerouteUsage()
   Use Zplanar WS=NamedRerouteUsage()
   Zplanar - Scale=NamedRerouteUsage()
   Use Second UV Channel=NamedRerouteUsage()
   UV - Tiling=NamedRerouteUsage()
   Diffuse Texture Input=NamedRerouteUsage()
   Color - Multiply=NamedRerouteUsage()
   Detail - Color - Multiply=NamedRerouteUsage()
   Height - Min=NamedRerouteUsage()
   Height - Max=NamedRerouteUsage()
   Detail - Color - Intensity=NamedRerouteUsage()
   Detail - Specular - Intensity=NamedRerouteUsage()
   Detail - Specular - Contrast=NamedRerouteUsage()
   SpecularMasks Texture Input=NamedRerouteUsage()
   Specular=NamedRerouteUsage()
   Metallic=NamedRerouteUsage()
   Roughness=NamedRerouteUsage()
   "Roughness - Bias (Mobile)"
   Normal Texture Input=NamedRerouteUsage()
   "Normals - Value (Nanite)"=NamedRerouteUsage()
   "Normals - Value (non-Nanite)"=NamedRerouteUsage()
   Fuzzy - Intensity=NamedRerouteUsage()
   Fuzzy - Exponent=NamedRerouteUsage()
   Fuzzy - Color=NamedRerouteUsage()
   Fuzzy - Roughness=NamedRerouteUsage()
   Realtime Displacement - Use Costly Triplanar=NamedRerouteUsage()
   Realtime Displacement - Flat - Seam Sharpness=NamedRerouteUsage()
   Realtime Displacment - Min=NamedRerouteUsage()
   Realtime Displacment - Max=NamedRerouteUsage()
   Realtime Displacment - Triplanar Seam=NamedRerouteUsage()
   "\\ --- Dependencies --- /"
   Wavy Distortion World Position=NamedRerouteUsage()
   Fuzzy Shaded Mask
  ) {materialFunction=&/MF_Rocks_Layer}
  NamedRerouteDeclaration(None=@n35.MaterialAttributes) {name=Layer B MateriaAttributes, nodeColor=<rgba 1 0 0.1875 1>}
  n36 = ScalarParameter() {
   defaultValue=1
   desc="1 = VTX Channel Weights Inverted\r\n0 = VTX Channel Weights Default"
   group="--  L4 | Extra (TD4) Blending"
   parameterName=* L4 | Masking - VTX Color - Invert
   sortPriority=161
  }
  NamedRerouteDeclaration(None=@n36) {name="• VTX Color - Invert", nodeColor=<rgba 0.5043 1 0 1>}
  NamedRerouteDeclaration(None=ScalarParameter() {group="--  L4 | Extra (TD4) Blending", parameterName=* L4 | Masking - VTX Color - Bias, sortPriority=175}) {name="• VTX Color - Bias", nodeColor=<rgba 0.5043 1 0 1>}
  NamedRerouteDeclaration(None=ScalarParameter() {group="--  L4 | Extra (TD4) Blending", parameterName=* L4 | Masking - Height - Min, sortPriority=512}) {name="• Height - Min", nodeColor=<rgba 0.5043 1 0 1>}
  NamedRerouteDeclaration(None=ScalarParameter() {defaultValue=1, group="--  L4 | Extra (TD4) Blending", parameterName=* L4 | Masking - Height - Max, sortPriority=513}) {name="• Height - Max", nodeColor=<rgba 0.5043 1 0 1>}
  n37 = ScalarParameter() {
   defaultValue=1
   group="--  L4 | Extra (TD4) Blending"
   parameterName="* L4 | Masking - Height - Blend (1=Blend 0=Min)"
   sliderMax=1
   sortPriority=510
  }
  NamedRerouteDeclaration(None=@n37) {name="• Height - Blend (1=Blend 0=Min)", nodeColor=<rgba 0.5043 1 0 1>}
  n38 = MaterialFunctionCall!(
   Layer A MaterialAttributes=NamedRerouteUsage()
   Layer B MaterialAttributes=NamedRerouteUsage()
   * Masking - Invert=NamedRerouteUsage()
   "• Masking - Bias"=NamedRerouteUsage()
   "• Masking - Intensity"=NamedRerouteUsage()
   --- Use WORLD Z NORMAL MASKING ---=NamedRerouteUsage()
   "• World Z Normal - Use VTX Normals"=NamedRerouteUsage()
   * World Z Normal - Bias=NamedRerouteUsage()
   * World Z Normal - Intensity=NamedRerouteUsage()
   --- Use VTX COLOR MASKING ---=NamedRerouteUsage()
   "• VTX Color - Channel Weights"=NamedRerouteUsage()
   "• VTX Color - Invert"=NamedRerouteUsage()
   "• VTX Color - Blend (1=Max 0=Min)"=NamedRerouteUsage()
   "• VTX Color - Bias"=NamedRerouteUsage()
   "• VTX Color - Intensity"=NamedRerouteUsage()
   --- Use DFAO MASKING ---=NamedRerouteUsage()
   * DFAO - Force Baked AO on Nanite=NamedRerouteUsage()
   "• DFAO - Blend (1=Max 0=Min)"=NamedRerouteUsage()
   "• DFAO - Bias"=NamedRerouteUsage()
   "• DFAO - Intensity"=NamedRerouteUsage()
   --- Use Z GRADIENT MASKING ---=NamedRerouteUsage()
   "• Z Gradient - Blend (1=Max 0=Min)"=NamedRerouteUsage()
   "• Z Gradient - Height - Bottom"=NamedRerouteUsage()
   "• Z Gradient - Height - Top"=NamedRerouteUsage()
   --- Use BAKED AO/CURVATURE/PAINTED MASKING ---=NamedRerouteUsage()
   "• Baked AO/Curvature/Painted - Channel Weights"=NamedRerouteUsage()
   "• Baked AO/Curvature/Painted - Invert"=NamedRerouteUsage()
   "• Baked AO/Curvature/Painted - Blend (1=Max 0=Min)"=NamedRerouteUsage()
   "• Baked AO/Curvature/Painted - Bias"=NamedRerouteUsage()
   "• Baked AO/Curvature/Painted - Intensity"=NamedRerouteUsage()
   --- Use WATERLINE MASKING ---=NamedRerouteUsage()
   "* Waterline - Blend (1=Max 0=Min)"=NamedRerouteUsage()
   * Waterline - Bias=NamedRerouteUsage()
   * Waterline - Intensity=NamedRerouteUsage()
   --- Use WS NOISE MASKING ---=NamedRerouteUsage()
   "• WS Noise - Channel Weights"=NamedRerouteUsage()
   "• WS Noise - Invert"=NamedRerouteUsage()
   "• WS Noise - Blend (1=Blend 0=Min)"=NamedRerouteUsage()
   "• WS Noise - Bias"=NamedRerouteUsage()
   "• WS Noise - Intensity"=NamedRerouteUsage()
   --- Use HEIGHT MASKING ---=NamedRerouteUsage()
   "• Height - Selection (1=Current Layer 0=Previous Layers)"=NamedRerouteUsage()
   "• Height - Blend (1=Blend 0=Min)"=NamedRerouteUsage()
   "• Height - Min"=NamedRerouteUsage()
   "• Height - Max"=NamedRerouteUsage()
   "• Height - Bias"=NamedRerouteUsage()
   "• Height - Intensity"=NamedRerouteUsage()
   * Realtime Displacement - Masking - Bias=NamedRerouteUsage()
   * Realtime Displacement - Masking - Intensity=NamedRerouteUsage()
   * Realtime Displacement - WS Noise - Contribution=NamedRerouteUsage()
   * Realtime Displacement - Height - Contribution=NamedRerouteUsage()
   "\\ --- Dependencies --- /"
   Baked Curvature=NamedRerouteUsage()
   Baked AO=NamedRerouteUsage()
   Painted Mask=NamedRerouteUsage()
   DFAO=NamedRerouteUsage()
   Z Gradient=NamedRerouteUsage()
   WS Noise Masks=NamedRerouteUsage()
  ) {materialFunction=&/MF_Rocks_Layer_Masking}
  NamedRerouteDeclaration(None=@n38.Combined Realtime Displacement Mask) {name=Layer Combined Blend Mask - Realtime Displacement, nodeColor=<rgba 1 0 0.1875 1>}
  n39 = ScalarParameter() {group="--  L4 | Extra (TD4) Blending", parameterName=* L4 | Masking - Realtime Displacement - WS Noise - Contribution, sortPriority=560}
  NamedRerouteDeclaration(None=@n39) {name=* Realtime Displacement - WS Noise - Contribution, nodeColor=<rgba 0.5043 1 0 1>}
  n40 = ScalarParameter() {group="--  L4 | Extra (TD4) Blending", parameterName=* L4 | Masking - Realtime Displacement - Height - Contribution, sortPriority=565}
  NamedRerouteDeclaration(None=@n40) {name=* Realtime Displacement - Height - Contribution, nodeColor=<rgba 0.5043 1 0 1>}
  NamedRerouteDeclaration(None=StaticBoolParameter() {group="--  L4 | Extra (TD4)", parameterName=* L4 | Extra - WS - Use Wavy Distortion, sortPriority=50}) {name="Triplanar - Use Wavy Distortion(SB)", nodeColor=<rgba 0.1979 0.0084 0 1>}
  NamedRerouteDeclaration(None=@n35.Tintable Mask) {name=Layer B Tintable Mask, nodeColor=<rgba 1 0 0.1875 1>}
  NamedRerouteDeclaration(None=ScalarParameter() {defaultValue=1, group="--  L4 | Extra (TD4)", parameterName=* L4 | Extra - Is Tintable Layer, sortPriority=0}) {name=Is Tintable Layer, nodeColor=<rgba 0.5043 1 0 1>}
  NamedRerouteDeclaration(None=ScalarParameter() {group="--  L4 | Extra (TD4) Blending", parameterName=* L4 | Masking - Realtime Displacement - Masking - Bias, sortPriority=550}) {name=* Realtime Displacement - Masking - Bias, nodeColor=<rgba 0.5043 1 0 1>}
  n41 = ScalarParameter() {
   defaultValue=1
   group="--  L4 | Extra (TD4) Blending"
   parameterName=* L4 | Masking - Realtime Displacement - Masking - Intensity
   sortPriority=555
  }
  NamedRerouteDeclaration(None=@n41) {name=* Realtime Displacement - Masking - Intensity, nodeColor=<rgba 0.5043 1 0 1>}
  NamedRerouteDeclaration(None=StaticBoolParameter() {group="--  L4 | Extra (TD4)", parameterName=* L4 | Extra - Use Diffuse Alpha as Height, sortPriority=325}) {name="Use Diffuse Alpha as Height(SB)", nodeColor=<rgba 0.1979 0.0084 0 1>}
  NamedRerouteDeclaration(None=ScalarParameter() {defaultValue=1, group="--  L4 | Extra (TD4)", parameterName=* L4 | Extra - UV - Tiling, sortPriority=105}) {name=UV - Tiling, nodeColor=<rgba 0.5043 1 0 1>}
  n42 = ScalarParameter() {defaultValue=10, group="--  L4 | Extra (TD4)", parameterName=* L4 | Extra - Realtime Displacement - Flat - Seam Sharpness, sortPriority=705}
  NamedRerouteDeclaration(None=@n42) {name=Realtime Displacement - Flat - Seam Sharpness, nodeColor=<rgba 0.5043 1 0 1>}
  NamedRerouteDeclaration(None=StaticBoolParameter() {group="--  L4 | Extra (TD4)", parameterName=* L4 | Extra - Realtime Displacement - Use Costly Triplanar, sortPriority=700}) {name="Realtime Displacement - Use Costly Triplanar(SB)", nodeColor=<rgba 0.1979 0.0084 0 1>}
  NamedRerouteDeclaration(None=ScalarParameter() {group="--  L4 | Extra (TD4)", parameterName=* L4 | Extra - Realtime Displacment - Min, sortPriority=715}) {name=Realtime Displacment - Min, nodeColor=<rgba 0.5043 1 0 1>}
  n43 = VectorParameter() {
   group="--  L4 | Extra (TD4) Blending"
   parameterName=* L4 | Masking - Baked AO/Curvature/Painted - Channel Weights
   sortPriority=360
   channelNames {a="", b=Blue, g=Green, r=Red}
   defaultValue=<rgba 1 0 0 0>
  }
  NamedRerouteDeclaration(None=@n43.RGB) {name="• Baked AO/Curvature/Painted - Channel Weights", nodeColor=<rgba 0.5043 1 0 1>}
  NamedRerouteDeclaration(None=ScalarParameter() {defaultValue=1, group="--  L4 | Extra (TD4)", parameterName=* L4 | Extra - Realtime Displacment - Max, sortPriority=716}) {name=Realtime Displacment - Max, nodeColor=<rgba 0.5043 1 0 1>}
  NamedRerouteDeclaration(None=FunctionInput() {bUsePreviewValueAsDefault=true, inputName=Z Gradient, inputType=FunctionInput_Scalar, sortPriority=814}) {name=_Z Gradient, nodeColor=<rgba 1 0 0.0525 1>}
  n44 = ScalarParameter() {defaultValue=1, group="--  L4 | Extra (TD4)", parameterName=* L4 | Extra - Realtime Displacment - Triplanar Seam, sortPriority=720}
  NamedRerouteDeclaration(None=@n44) {name=Realtime Displacment - Triplanar Seam, nodeColor=<rgba 0.5043 1 0 1>}
  NamedRerouteDeclaration(None=ScalarParameter() {defaultValue=1, group="--  L4 | Extra (TD4)", parameterName=* L4 | Extra - Triplanar - Derivatives Scale, sortPriority=54}) {name=Triplanar - Derivatives Scale, nodeColor=<rgba 0.5043 1 0 1>}
  NamedRerouteDeclaration(None=ScalarParameter() {defaultValue=3, group="--  L4 | Extra (TD4)", parameterName=* L4 | Extra - Fuzzy - Exponent, sortPriority=455}) {name=Fuzzy - Exponent, nodeColor=<rgba 0.5043 1 0 1>}
  NamedRerouteDeclaration(None=ScalarParameter() {defaultValue=1, group="--  L4 | Extra (TD4)", parameterName=* L4 | Extra - Fuzzy - Intensity, sortPriority=450}) {name=Fuzzy - Intensity, nodeColor=<rgba 0.5043 1 0 1>}
  n45 = ScalarParameter() {
   defaultValue=1
   desc=This Roughness value will be blended to where the Fuzzy shading effect occurs.
   group="--  L4 | Extra (TD4)"
   parameterName=* L4 | Extra - Fuzzy - Roughness
   sortPriority=465
  }
  NamedRerouteDeclaration(None=@n45) {name=Fuzzy - Roughness, nodeColor=<rgba 0.5043 1 0 1>}
  n46 = VectorParameter() {
   desc="This Color is either Multiplied or added to Layer's Color based on the Alpha channel value.\r\nNote that this effect happens after the Layer's Color Multiply"
   group="--  L4 | Extra (TD4)"
   parameterName=* L4 | Extra - Fuzzy - Color
   sortPriority=460
   channelNames {a="0=Multiply 1=Add", b="", g="", r=""}
   defaultValue=<rgba 1.1 1.1 1.1 0>
  }
  NamedRerouteDeclaration(None=AppendVector(A=@n46, B=@n46)) {name=Fuzzy - Color, nodeColor=<rgba 0.5043 1 0 1>}
  NamedRerouteDeclaration(None=StaticBoolParameter() {group="--  L4 | Extra (TD4)", parameterName=* L4 | Extra - Is Fuzzy Layer, sortPriority=1}) {name="Is Fuzzy Layer(SB)", nodeColor=<rgba 0.1979 0.0084 0 1>}
  NamedRerouteDeclaration(None=FunctionInput() {inputName=Wavy Distortion World Position, sortPriority=1005}) {name=_Wavy Distortion World Position, nodeColor=<rgba 1 0 0.0525 1>}
  NamedRerouteDeclaration(None=ScalarParameter() {defaultValue=700, group="--  L4 | Extra (TD4)", parameterName=* L4 | Extra - Zplanar - Scale, sortPriority=85}) {name=Zplanar - Scale, nodeColor=<rgba 0.5043 1 0 1>}
  NamedRerouteDeclaration(None=StaticBoolParameter() {group="--  L4 | Extra (TD4)", parameterName=* L4 | Extra - Use Zplanar WS, sortPriority=80}) {name="Zplanar - Use Zplanar WS(SB)", nodeColor=<rgba 0.1979 0.0084 0 1>}
  NamedRerouteDeclaration(None=FunctionInput() {bUsePreviewValueAsDefault=true, inputName=Baked AO, inputType=FunctionInput_Scalar, sortPriority=811}) {name=_Baked AO, nodeColor=<rgba 1 0 0.0525 1>}
  NamedRerouteDeclaration(None=StaticBoolParameter() {group="--  L4 | Extra (TD4)", parameterName=* L4 | Extra - Use DEBUG UVs, sortPriority=30}) {name="Use DEBUG UVs(SB)", nodeColor=<rgba 0.1979 0.0084 0 1>}
  NamedRerouteDeclaration(None=ScalarParameter() {defaultValue=0.6, group="--  L4 | Extra (TD4)", parameterName=* L4 | Extra - Detail - Color - Intensity, sortPriority=340}) {name=Detail - Color - Intensity, nodeColor=<rgba 0.5043 1 0 1>}
  n47 = ScalarParameter() {defaultValue=2.06, group="--  L4 | Extra (TD4)", parameterName=* L4 | Extra - Detail - Specular - Intensity, sortPriority=341}
  NamedRerouteDeclaration(None=@n47) {name=Detail - Specular - Intensity, nodeColor=<rgba 0.5043 1 0 1>}
  n48 = StaticBoolParameter() {
   desc="OFF = Nanite uses VTX and Default uses Textures (Prop Rocks are the only meshes that gets tesselated and displaced on nanite platforms, by default they won`t use VTX)\r\nON = Force VTX for all (VTX masks are only used on big custom sculpts that have a custom nanite mesh with no UVs)\r\n"
   group="--  L4 | Extra (TD4)"
   parameterName=* L4 | Extra - Use VTX AO/Curvature/Painted on Nanite
   sortPriority=49
  }
  NamedRerouteDeclaration(None=@n48) {name="Use VTX AO/Curvature/Painted on Nanite(SB)", nodeColor=<rgba 0.1979 0.0084 0 1>}
  n49 = VectorParameter() {group="--  L4 | Extra (TD4)", parameterName=* L4 | Extra - Color - Multiply, sortPriority=321, defaultValue=<rgba 1 1 1 0>}
  NamedRerouteDeclaration(None=@n49) {name=Color - Multiply, nodeColor=<rgba 0.5043 1 0 1>}
  n50 = ScalarParameter() {defaultValue=-0.3, group="--  L4 | Extra (TD4)", parameterName=* L4 | Extra - Detail - Specular - Contrast, sortPriority=342}
  NamedRerouteDeclaration(None=@n50) {name=Detail - Specular - Contrast, nodeColor=<rgba 0.5043 1 0 1>}
  NamedRerouteDeclaration(None=FunctionInput(Preview=StaticBool()) {bUsePreviewValueAsDefault=true, inputName=Layer Is Used, inputType=FunctionInput_StaticBool, sortPriority=-200}) {name="Layer Is Used(SB)", nodeColor=<rgba 0.1979 0.0084 0 1>}
  n51 = ScalarParameter() {
   defaultValue=700
   desc="Scale is the World Space size that the Biplanar projects, \r\nSmaller sizes = more tiling\r\nLarger sizes = less tiling"
   group="--  L4 | Extra (TD4)"
   parameterName=* L4 | Extra - Biplanar - Scale
   sortPriority=61
  }
  NamedRerouteDeclaration(None=@n51) {name=Biplanar - Scale, nodeColor=<rgba 0.5043 1 0 1>}
  n52 = StaticBoolParameter() {
   desc="Enable to use side projected planar mapping(Sediment Cliffs etc)"
   group="--  L4 | Extra (TD4)"
   parameterName=* L4 | Extra - Use Biplanar WS
   sortPriority=60
  }
  NamedRerouteDeclaration(None=@n52) {name=Biplanar - Use Biplanar WS, nodeColor=<rgba 0.1979 0.0084 0 1>}
  NamedRerouteDeclaration(None=@n38.Combined Mask) {name=Layer Combined Blend Mask, nodeColor=<rgba 1 0 0.1875 1>}
  n53 = VectorParameter() {group="--  L4 | Extra (TD4)", parameterName=* L4 | Extra - Biplanar - Position Offset, sortPriority=62, defaultValue=<rgba 0 0 0 1>}
  NamedRerouteDeclaration(None=@n53.RGB) {name=Biplanar - Position Offset, nodeColor=<rgba 0.5043 1 0 1>}
  n54 = ScalarParameter() {
   defaultValue=10
   desc=Controls the contrast of the fade between the 2 projections. A value too high will remove the fade and create a harsh seam
   group="--  L4 | Extra (TD4)"
   parameterName=* L4 | Extra - Biplanar - Fade Contrast
   sortPriority=63
  }
  NamedRerouteDeclaration(None=@n54) {name=Biplanar - Fade Contrast, nodeColor=<rgba 0.5043 1 0 1>}
  NamedRerouteDeclaration(None=ScalarParameter() {group="--  L4 | Extra (TD4)", parameterName="* L4 | Extra - Roughness - Bias (Mobile)", sortPriority=391}) {name="Roughness - Bias (Mobile)", nodeColor=<rgba 0.5043 1 0 1>}
  NamedRerouteUsage()
  NamedRerouteUsage()
  Reroute(None=StaticBool() {desc=Remove Detail Layer Functionality})
  Reroute(None=NamedRerouteUsage())
  Reroute(None=NamedRerouteUsage())
  Reroute(None=NamedRerouteUsage())
  Reroute(None=NamedRerouteUsage())
  Reroute(None=NamedRerouteUsage())
  Reroute(None=NamedRerouteUsage())
  Reroute(None=NamedRerouteUsage())
  StaticBool()
 ]}
