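# Material function that blends rock mesh materials into terrain using Runtime Virtual Texture color, specular, roughness, normal and height inputs.
# Separate material and normal masks fade over top/side distances chosen by vertex normal, eroded by world-space noise and optionally limited by vertex color.
# Optional top snow tints the RVT color, then HeightLerp blends the layers.
# A static switch, RVT switch and MaterialProxyReplace control where blending applies.
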MaterialFunction /BRCoreMaterials/Materials/MaterialFunctions/MF_Rocks_RVT_Blending
graph {nodes [
  n1 = WorldPosition()
  n2 = Divide(A=@n1.XY, B=ScalarParameter() {defaultValue=1000, group=--- F5 | RVT, parameterName=* RVT - Edge Noise - Scale, sortPriority=70})
  n3 = Subtract(A=@n2) {constB=0.5}
  n4 = Multiply(A=@n3, B=@n3)
  ComponentMask(None=@n4) {r=true}
  FunctionInput() {inputName="\\ --- RVT --- /", inputType=FunctionInput_Scalar, sortPriority=900}
  n5 = NamedRerouteUsage()
  n6 = NamedRerouteUsage()
  n7 = SetMaterialAttributes!(MaterialAttributes, BaseColor=NamedRerouteUsage(), Specular=NamedRerouteUsage(), Roughness=NamedRerouteUsage(), Normal=NamedRerouteUsage()) {attributeSetTypes [
    #1
    #2
    #3
    #4
   ]}
  n8 = GetMaterialAttributes(None=MaterialFunctionCall!(
    ScalarWeightMask
    UseSnowOcclusionTextureMask=StaticBoolParameter() {defaultValue=true, group=--- F3 | Landscape, parameterName=TopSnow_UseLayerMaskTexture, sortPriority=-3}
    ------
    UseMaterialInput
    AdditionalMaskToRemoveSnow
    OutputTangentNormals
    UseAdditionalMaskToRemoveSnow
    UseSnowNormals
    OuputVertexNormalsAsTheSnowNormals
    UseSparkle
    LQSpark
    AllowSnowMicroNormals
    CliffMaterialOnly
    Material=@n7
    CliffLandscapeLayerMask
    IsLandscape=StaticBool()
    AreInputNormalsTangentSpace=StaticBool()
    UseInputNormalsForFakeSnowShadow=StaticBool() {value=true}
   ) {materialFunction=/Hera/Environment/Terrain/Landscape/Materials/Layers/QualityLevel/MF_Hera_Snow_Default}.SnowMaterialOnly) {attributeGetTypes [#1]}
  n9 = ComponentMask(None=VertexNormalWS()) {b=true}
  n10 = Saturate(None=Multiply(A=Power(Base=@n9) {constExponent=4}) {constB=2})
  n11 = OneMinus(None=NamedRerouteUsage())
  n12 = ScalarParameter() {
   desc=Set this to 1 for full snow meshes like snow drifts. This overrides vertex color and vertex normal Z masking.
   group=--- F3 | Landscape
   parameterName=TopSnow_FULL_Snow
   sortPriority=-2
  }
  n13 = Saturate(None=Add(A=Multiply(A=@n10, B=@n11), B=@n12))
  n14 = Constant()
  n15 = SetMaterialAttributes!(
   MaterialAttributes=NamedRerouteUsage()
   BaseColor=StaticSwitch(
    True=LinearInterpolate(A=@n6, B=@n8.BaseColor, Alpha=@n13)
    False=@n6
    Value=StaticBoolParameter() {group=--- F3 | Landscape, parameterName=* Landscape - Use Snow, sortPriority=-4}
   )
   Specular=NamedRerouteUsage()
   Roughness=NamedRerouteUsage()
   EmissiveColor=@n14
   Metallic=@n14
   SubsurfaceColor=@n14
   Opacity=Constant() {r=1}
  ) {attributeSetTypes [
    #1
    #2
    #3
    #5
    #6
    #7
    #8
   ]}
  n16 = OneMinus(None=NamedRerouteUsage())
  n17 = GetMaterialAttributes(None=NamedRerouteUsage()) {attributeGetTypes [#8]}
  n18 = ScalarParameter() {defaultValue=2, group=--- F5 | RVT, parameterName=* RVT - Contrast, sortPriority=80}
  n19 = MaterialProxyReplace(
   Realtime=BlendMaterialAttributes(
    A=@n15
    B=NamedRerouteUsage()
    Alpha=MaterialFunctionCall(Transition Phase=@n16, Height Texture=@n17.Opacity, Contrast=@n18) {materialFunction=&HeightLerp}.Alpha
   ) {vertexAttributeBlendType=UseA}
   MaterialProxy=@n5
  )
  FunctionOutput(None=StaticSwitch(
    True=MaterialFunctionCall(NoRVT=@n5, RVT=@n19) {materialFunction=&FN_RVT_Switch_Material}.Result
    False=NamedRerouteUsage()
    Value=NamedRerouteUsage()
   )) {outputName=Blended MaterialAttributes, sortPriority=-100}
  MaterialFunctionCall!(NoRVT=Constant() {r=1}, RVT=NamedRerouteUsage()) {materialFunction=/Game/Packages/Fortress_SharedMaterials/Base_Material_Functions/FN_RVT_Switch_Color}
  Multiply(A=Divide(A=@n4) {constB=0.045}) {constB=-1}
  n20 = NamedRerouteUsage()
  n21 = NamedRerouteUsage()
  n22 = ScalarParameter() {
   defaultValue=10
   desc="Controls the distance of the RVT Mask on top(upward facing surfaces)"
   group=--- F5 | RVT
   parameterName=* RVT - Distance - Top
   sortPriority=55
  }
  n23 = ScalarParameter() {
   defaultValue=120
   desc=Controls the distance of the RVT Mask for normals on side surfaces. Distance is from where the mesh meets the terrain and up
   group=--- F5 | RVT
   parameterName=* RVT - Distance - Sides
   sortPriority=56
  }
  n24 = DotProduct(A=VertexNormalWS(), B=Constant3Vector() {constant=<rgba 0 0 1 0>})
  n25 = OneMinus(None=Saturate(None=Multiply(A=@n24, B=@n24) {constB=2, desc=Increase strength of mask to better match VTX color values}))
  n26 = LinearInterpolate(A=@n22, B=@n23, Alpha=@n25)
  n27 = OneMinus(None=Saturate(None=Divide(A=@n21, B=@n26)))
  n28 = TextureSampleParameter2D(UVs=@n2) {group=--- F5 | RVT, parameterName=* RVT - Edge Noise, sortPriority=71, texture=/Game/Effects/Fort_Effects/Textures/Tile/T_GaussianNoise_01_M}
  n29 = Multiply(A=@n28.R, B=ScalarParameter() {defaultValue=0.75, group=--- F5 | RVT, parameterName=* RVT - Edge Noise - Amount, sortPriority=72})
  n30 = OneMinus(None=@n29)
  n31 = Saturate(None=Divide(A=Subtract(A=@n27, B=@n29), B=@n30))
  NamedRerouteDeclaration(None=StaticSwitch(True=Min(A=@n20, B=@n31), False=@n31, Value=NamedRerouteUsage())) {name=RVT BlendMask, nodeColor=<rgba 0.9922 1 0 1>}
  NamedRerouteDeclaration(None=FunctionInput() {inputName=RVT Height, inputType=FunctionInput_Scalar, sortPriority=920}) {name=_RVT Height, nodeColor=<rgba 0.3749 0.6555 1 0.5>}
  NamedRerouteDeclaration(None=StaticBoolParameter() {group=--- F5 | RVT, parameterName=* RVT - Use VTX Color Mask, sortPriority=150}) {name=--- Use VTX COLOR MASKING ---, nodeColor=<rgba 0.1979 0.0084 0 1>}
  NamedRerouteDeclaration(None=FunctionInput() {inputName=RVT Base Color, sortPriority=905}) {name=_RVT Base Color, nodeColor=<rgba 0.3749 0.6555 1 0.5>}
  NamedRerouteDeclaration(None=FunctionInput() {inputName=RVT Specular, inputType=FunctionInput_Scalar, sortPriority=906}) {name=_RVT Specular, nodeColor=<rgba 0.3749 0.6555 1 0.5>}
  NamedRerouteDeclaration(None=FunctionInput() {inputName=RVT Roughness, inputType=FunctionInput_Scalar, sortPriority=907}) {name=_RVT Roughness, nodeColor=<rgba 0.3749 0.6555 1 0.5>}
  NamedRerouteDeclaration(None=FunctionInput() {inputName=RVT Normal, sortPriority=909}) {name=_RVT Normals, nodeColor=<rgba 0.3749 0.6555 1 0.5>}
  n32 = VertexColor()
  NamedRerouteDeclaration(None=@n32.R) {name=VertexColor-Red, nodeColor=<rgba 0.9062 0.9451 1 1>}
  n33 = VectorParameter() {
   desc="Sets amount that each VTX color contributes to layer masking\r\nTotal weight value over 1 will increase sharpness of the resulting masking"
   group=--- F5 | RVT
   parameterName=* RVT - VTX Color Mask - Channel Weights
   sortPriority=160
   channelNames {a=Alpha, b=Blue, g=Green, r=Red}
   defaultValue=<rgba 1 0 0 0>
  }
  NamedRerouteDeclaration(None=@n33.RGBA) {name="• VTX Color - Channel Weights", nodeColor=<rgba 0.5043 1 0 1>}
  NamedRerouteDeclaration(None=FunctionInput(Preview=SetMaterialAttributes!(MaterialAttributes)) {inputName=Input MaterialAttributes, inputType=FunctionInput_MaterialAttributes, sortPriority=-100}) {name=_Input MaterialAttributes, nodeColor=<rgba 0.1016 0 1 1>}
  n34 = NamedRerouteUsage()
  n35 = NamedRerouteUsage()
  n36 = ScalarParameter() {
   defaultValue=10
   desc="Controls the distance of the RVT Mask for normals on top(upward facing surfaces)"
   group=--- F5 | RVT
   parameterName="* RVT - Distance - Top (Normals)"
   sortPriority=50
  }
  n37 = ScalarParameter() {
   defaultValue=120
   desc=Controls the distance of the RVT Mask for normals on side surfaces. Distance is from where the mesh meets the terrain and up
   group=--- F5 | RVT
   parameterName="* RVT - Distance - Sides (Normals)"
   sortPriority=51
  }
  n38 = LinearInterpolate(A=@n36, B=@n37, Alpha=@n25)
  n39 = OneMinus(None=Saturate(None=Divide(A=@n35, B=@n38)))
  n40 = Saturate(None=Divide(A=Subtract(A=@n39, B=@n29), B=@n30))
  NamedRerouteDeclaration(None=StaticSwitch(True=Min(A=@n34, B=@n40), False=@n40, Value=NamedRerouteUsage())) {name=RVT BlendMask - Normals, nodeColor=<rgba 0.9922 1 0 1>}
  NamedRerouteDeclaration(None=ScalarParameter() {group=--- F5 | RVT, parameterName=* RVT - VTX Color Mask - Invert, sortPriority=161}) {name="• VTX Color - Invert", nodeColor=<rgba 0.5043 1 0 1>}
  NamedRerouteDeclaration() {name=UseTopSnow, nodeColor=<rgba 0.3594 0 1 1>}
  NamedRerouteDeclaration(None=FunctionInput(Preview=StaticBool()) {bUsePreviewValueAsDefault=true, inputName=Use RVT, inputType=FunctionInput_StaticBool, sortPriority=-110}) {name=Use RVT, nodeColor=<rgba 1 0.4219 0 1>}
  n41 = NamedRerouteUsage()
  n42 = NamedRerouteUsage()
  NamedRerouteDeclaration(None=BlendMaterialAttributes(
    A=@n42
    B=SetMaterialAttributes!(MaterialAttributes=@n42, Normal=@n41) {attributeSetTypes [#4]}
    Alpha=NamedRerouteUsage()
   ) {vertexAttributeBlendType=UseA}) {name=_Input MaterialAttributes - RVT ModifiedNormals, nodeColor=<rgba 1 1 0 1>}
  n43 = Saturate(None=DotProduct(A=NamedRerouteUsage(), B=VertexColor()))
  NamedRerouteDeclaration(None=LinearInterpolate(A=@n43, B=OneMinus(None=@n43), Alpha=NamedRerouteUsage())) {name=VTX Color Mask, nodeColor=<rgba 0 0.6484 1 1>}
  NamedRerouteUsage()
  n44 = DotProduct(A=Constant4Vector() {constant=<rgba 9.654 9.256 0 0>}.RGB, B=Divide(A=@n1.XY) {constB=2000})
  Sobol!(Cell=Frac(None=Sine(None=@n44)), Index, Seed)
  n45 = StaticBoolParameter() {
   desc="Affects how the initial masking is created for the Top and Sides Distance values\r\nTrue = Uses the Green VTX channel that has white on the sides and black on top\r\nFalse = Uses a mask(Dot Product) based on the upward facing surfaces"
   group=--- F5 | RVT
   parameterName=* RVT - Use Green VTX Color for Top/Sides
   sortPriority=10
  }
  StaticSwitch(True=@n32.G, False=@n25, Value=@n45)
  VectorNoise()
 ]}
