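# Material function that adds optional real-time displacement to rock material attributes, toggled by a static bool input.
# It reads height from the incoming OpacityMask attribute and writes it to Displacement, passing attributes through unchanged when disabled.
# An optional vertex-color channel mask, with invert and contrast controls, pulls displacement toward a neutral 0.5.
# The Intensity parameter's blend is not connected to the output.
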MaterialFunction /BRCoreMaterials/Materials/MaterialFunctions/MF_Rocks_Realtime_Displacement
graph {nodes [
  FunctionOutput(None=NamedRerouteUsage()) {outputName=Blended MaterialAttributes, sortPriority=-100}
  n1 = GetMaterialAttributes(None=NamedRerouteUsage()) {attributeGetTypes [#1]}
  LinearInterpolate(
   A=Constant() {r=0.5}
   B=@n1.OpacityMask
   Alpha=ScalarParameter() {defaultValue=1, group="-      | Realtime Displacement", parameterName=* Realtime Displacement - Intensity, sortPriority=10}
  )
  n2 = ChannelMaskParameter(None=VertexColor()) {group="-      | Realtime Displacement", parameterName=* Realtime Displacement - VTX Masking - Channel, sortPriority=30}
  n3 = ScalarParameter() {
   defaultValue=1
   group="-      | Realtime Displacement"
   parameterName=* Realtime Displacement - VTX Masking - Channel - Invert
   sortPriority=35
  }
  n4 = ScalarParameter() {
   defaultValue=1
   group="-      | Realtime Displacement"
   parameterName=* Realtime Displacement - VTX Masking - Channel - Contrast
   sortPriority=35
  }
  n5 = Power(Base=LinearInterpolate(A=@n2.R, B=OneMinus(None=@n2.R), Alpha=@n3), Exp=@n4) {constExponent=8}
  n6 = ScalarParameter() {group="-      | Realtime Displacement", parameterName=* Realtime Displacement - Use VTX Masking, sliderMax=1, sortPriority=40}
  n7 = LinearInterpolate(A=@n1.OpacityMask, B=LinearInterpolate(A=@n1.OpacityMask, Alpha=Saturate(None=@n5)) {constB=0.5}, Alpha=@n6)
  NamedRerouteDeclaration(None=StaticSwitch(
    True=SetMaterialAttributes!(MaterialAttributes=@n1.MaterialAttributes, Displacement=@n7) {attributeSetTypes [#2]}
    False=@n1.MaterialAttributes
    Value=NamedRerouteUsage()
   )) {name=RealtimeDisplacement Blended MaterialAttributes, nodeColor=<rgba 1 0 0.5859 1>}
  NamedRerouteDeclaration(None=FunctionInput(Preview=SetMaterialAttributes!(MaterialAttributes)) {inputName=Input MaterialAttributes, inputType=FunctionInput_MaterialAttributes, sortPriority=-100}) {name=_Input MaterialAttributes, nodeColor=<rgba 0.1016 0 1 1>}
  n8 = FunctionInput(Preview=StaticBool()) {bUsePreviewValueAsDefault=true, inputName=Use Realtime Displacement, inputType=FunctionInput_StaticBool, sortPriority=-110}
  NamedRerouteDeclaration(None=@n8) {name=Use Realtime Displacement, nodeColor=<rgba 1 0.4219 0 1>}
 ]}
