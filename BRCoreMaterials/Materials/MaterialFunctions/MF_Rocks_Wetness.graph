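# Material function for rocks that applies waterline-based wetness to input material attributes, behind a static switch.
# A porosity estimate from specular, roughness and metallic, an input mask and a world-height waterline mask with per-primitive offset supply the wetness terms.
# Wetness darkens and tints base color, less so on upward faces, and lowers specular and roughness.
# Tint varies by platform, and a separate ocean-level height band is built for roughness override.
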MaterialFunction /BRCoreMaterials/Materials/MaterialFunctions/MF_Rocks_Wetness
graph {nodes [
  FunctionInput() {inputName="\\ --- Masks --- /", inputType=FunctionInput_Scalar, sortPriority=50}
  n1 = GetMaterialAttributes(None=NamedRerouteUsage()) {attributeGetTypes [
    #1
    #2
    #3
    #4
    #5
    #6
   ]}
  n2 = VectorParameter() {group=--- F4 | Wetness, parameterName=* Wetness - Color Tint, sortPriority=-1, defaultValue=<rgba 0.25 0.25 0.25 0>}
  n3 = VectorParameter() {group=--- F4 | Wetness, parameterName=* Wetness - Color Tint Low, sortPriority=-1, defaultValue=<rgba 0.25 0.25 0.25 0>}
  n4 = MaterialFunctionCall!("Xb1,PS4"=@n2, "PC-High&&Epic,XSX,PS5"=@n2, "PC-Perf,Android"=@n3, Switch=@n3, "PC-Low(NotPerf)&&Medium"=@n3) {materialFunction=/Game/Effects/Niagara/MasterMaterials/Functions/Common/PlatformSwitch_3D}
  n5 = ScalarParameter() {
   desc=This controls how much the Color Tint is used vs the default wetness coloring
   group=--- F4 | Wetness
   parameterName=* Wetness - Color Tint - Low Override
   sortPriority=15
  }
  n6 = ScalarParameter() {
   desc=This controls how much the Color Tint is used vs the default wetness coloring
   group=--- F4 | Wetness
   parameterName=* Wetness - Color Tint - Strength
   sortPriority=15
  }
  n7 = ScalarParameter() {
   desc=This controls how much the Color Tint is used vs the default wetness coloring
   group=--- F4 | Wetness
   parameterName=* Wetness - Color Tint - Strength Low
   sortPriority=15
  }
  n8 = LinearInterpolate(
   A=Multiply(A=@n1.BaseColor, B=@n1.BaseColor) {desc=Mimic Saturation and darkening from Porosity}
   B=Multiply(A=LinearInterpolate(A=@n2, B=@n4.Result, Alpha=@n5), B=@n1.BaseColor)
   Alpha=LinearInterpolate(
    A=@n6
    B=MaterialFunctionCall("Xb1,PS4"=@n6, "PC-High&&Epic,XSX,PS5"=@n6, "PC-Perf,Android"=@n7, Switch=@n7, "PC-Low(NotPerf)&&Medium"=@n7) {materialFunction=&PlatformSwitch_1D}.Result
    Alpha=@n5
   )
  )
  n9 = DotProduct(A=VertexNormalWS(), B=Constant3Vector() {constant=<rgba 0 0 1 0>})
  n10 = LinearInterpolate(A=@n1.BaseColor, B=LinearInterpolate(A=@n8, B=Multiply(A=@n8) {constB=1.5}, Alpha=Saturate(None=@n9)), Alpha=NamedRerouteUsage())
  n11 = ScalarParameter() {
   defaultValue=0.224
   desc="This is the fully Wet Specular value, \r\nSpecular should usually be slightly lower on wet surfaces (0.224 instead of closer to 0.5)"
   group=--- F4 | Wetness
   parameterName=* Wetness - Specular
   sortPriority=15
  }
  n12 = LinearInterpolate(A=@n1.Specular, B=@n11, Alpha=NamedRerouteUsage()) {constB=0.224}
  n13 = ScalarParameter() {defaultValue=0.1, desc=This is the fully Wet Roughness value, group=--- F4 | Wetness, parameterName=* Wetness - Roughness, sortPriority=20}
  n14 = Max(A=NamedRerouteUsage(), B=LinearInterpolate(A=@n1.Roughness, B=@n13, Alpha=NamedRerouteUsage()))
  FunctionOutput(None=StaticSwitch(
    True=SetMaterialAttributes!(MaterialAttributes=@n1.MaterialAttributes, BaseColor=@n10, Specular=@n12, Roughness=@n14) {attributeSetTypes [#1, #3, #5]}
    False=NamedRerouteUsage()
    Value=NamedRerouteUsage()
   )) {outputName=Blended MaterialAttributes, sortPriority=-100}
  n15 = ScalarParameter() {
   bUseCustomPrimitiveData=true
   desc="Primitive Data Only, Enable per instance in the details panel."
   group=Waterline
   parameterName=PD_WaterlineHeightOffset
   sortPriority=0
  }
  n16 = WorldPosition()
  n17 = Add(
   A=Add(A=@n15, B=ScalarParameter() {defaultValue=-155, group=--- F4 | Wetness, parameterName=* Wetness - Waterline - Offset, sortPriority=75})
   B=ShadingPathSwitch(Default=@n16.Z, Mobile=TruncateLWC!(None=@n16.Z))
  )
  n18 = ScalarParameter() {
   defaultValue=-640
   desc=This offsets the OceanLevel Roughness override
   group=--- F4 | Wetness
   parameterName=* Wetness - Waterline - OceanLevel - Offset
  }
  n19 = ScalarParameter() {
   defaultValue=20
   desc=The range that the OceanLevel Roughness override blends in
   group=--- F4 | Wetness
   parameterName=* Wetness - Waterline - OceanLevel - Falloff
  }
  n20 = Divide(A=Add(A=@n17, B=@n18), B=@n19)
  NamedRerouteDeclaration(None=OneMinus(None=Saturate(None=@n20))) {name=Waterline - OceanLevel, nodeColor=<rgba 1 0.2109 0 1>}
  NamedRerouteDeclaration(None=FunctionInput(Preview=Constant() {r=1}) {bUsePreviewValueAsDefault=true, inputName=Wetness Mask, inputType=FunctionInput_Scalar, sortPriority=55}) {name=Wetness Mask, nodeColor=<rgba 0.1016 0 1 1>}
  n21 = ScalarParameter() {defaultValue=1100, group=--- F4 | Wetness, parameterName=* Wetness - Waterline - Blur, sortPriority=76}
  NamedRerouteDeclaration(None=OneMinus(None=Divide(A=@n17, B=@n21))) {name=Waterline - Height Mask, nodeColor=<rgba 0 0.8594 1 1>}
  NamedRerouteDeclaration(None=FunctionInput(Preview=SetMaterialAttributes!(MaterialAttributes)) {inputName=Input MaterialAttributes, inputType=FunctionInput_MaterialAttributes, sortPriority=-100}) {name=_Input MaterialAttributes, nodeColor=<rgba 0.1016 0 1 1>}
  n22 = GetMaterialAttributes(None=NamedRerouteUsage()) {attributeGetTypes [
    #1
    #2
    #3
    #4
    #5
    #6
   ]}
  n23 = Multiply(
   A=Multiply(A=@n22.Specular) {constB=2, desc="Less light response = Higher Porosity < 0.5 < More light response = Lower Porosity"}
   B=@n22.Roughness
  )
  n24 = Ceil(None=@n22.Metallic) {desc="If Metal -> 0 Porosity"}
  n25 = ScalarParameter() {
   desc="Less Wetness response < Bias < More Wetness response\r\nPorosity is a Combination of Specular and Roughness that affects how wetness absorbtion happens."
   group=--- F4 | Wetness
   parameterName=* Wetness - Porosity - Bias
  }
  n26 = Add(A=Min(A=Saturate(None=@n23), B=OneMinus(None=@n24)), B=@n25)
  n27 = ScalarParameter() {
   defaultValue=1
   desc="Less Wetness response/contrast < Intensity < More Wetness response/contrast\r\nPorosity is a Combination of Specular and Roughness that affects how wetness absorbtion happens."
   group=--- F4 | Wetness
   parameterName=* Wetness - Porosity - Intensity
  }
  NamedRerouteDeclaration(None=Saturate(None=Multiply(A=@n26, B=@n27))) {name=Porosity, nodeColor=<rgba 1 0 0.4687 1>}
  n28 = Subtract(A=NamedRerouteUsage(), B=Constant() {r=1})
  n29 = SmoothStep(
   Min=ScalarParameter() {group=--- F4 | Wetness, parameterName=* Wetness - Waterline - Min, sortPriority=85}
   Max=ScalarParameter() {defaultValue=0.5, group=--- F4 | Wetness, parameterName=* Wetness - Waterline - Max, sortPriority=86}
   Value=NamedRerouteUsage()
  )
  n30 = Constant()
  n31 = NamedRerouteUsage()
  n32 = NamedRerouteUsage()
  n33 = Multiply(A=Multiply(A=StaticSwitch(True=@n29, False=@n30, Value=@n31), B=Saturate(None=@n32))) {constB=2}
  NamedRerouteDeclaration(None=Saturate(None=Add(A=@n28, B=@n33))) {name=Wetness, nodeColor=<rgba 0 0.7891 1 1>}
  NamedRerouteDeclaration() {name=UseTopSnow, nodeColor=<rgba 0.3594 0 1 1>}
  NamedRerouteDeclaration(None=FunctionInput(Preview=StaticBool()) {bUsePreviewValueAsDefault=true, inputName=Use Wetness, inputType=FunctionInput_StaticBool, sortPriority=-110}) {name=Use Wetness, nodeColor=<rgba 1 0.4219 0 1>}
  NamedRerouteDeclaration(None=StaticBoolParameter() {defaultValue=true, group=--- F4 | Wetness, parameterName=* Wetness - Use Waterline, sortPriority=70}) {name=Use Waterline, nodeColor=<rgba 1 0.4219 0 1>}
  n34 = NamedRerouteUsage()
  SetMaterialAttributes!(MaterialAttributes, BaseColor=@n34, EmissiveColor=@n34) {attributeSetTypes [#1, #7]}
 ]}
