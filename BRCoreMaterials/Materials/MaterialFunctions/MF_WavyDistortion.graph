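# Material function that bends world position vertically in periodic waves, with a second output named Sediment Wave Distortion.
# It offsets world position, divides by the Frequency input, then shapes the fraction f into smooth repeating bumps as (f·(1−f))².
# The bumps sum as X twice plus Z, leaving Y unused, and scale by Intensity.
# The result is output directly and subtracted from Z, with XY unchanged.
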MaterialFunction /BRCoreMaterials/Materials/MaterialFunctions/MF_WavyDistortion
graph {nodes [
  n1 = FunctionInput(Preview=WorldPosition().XYZ) {bUsePreviewValueAsDefault=true, inputName=WorldPosition, sortPriority=-50}
  n2 = ComponentMask(None=@n1) {b=true}
  n3 = ComponentMask(None=@n1) {b=true, g=true, r=true}
  n4 = FunctionInput(Preview=Constant3Vector()) {bUsePreviewValueAsDefault=true, inputName=Wavy Distortion - Position Offset, sortPriority=15}
  n5 = Constant() {r=6000}
  n6 = Frac(None=Divide(
    A=Add(A=@n3, B=@n4)
    B=FunctionInput(Preview=@n5) {bUsePreviewValueAsDefault=true, inputName=Wavy Distortion - Frequency, inputType=FunctionInput_Scalar, sortPriority=20}
   ))
  n7 = Multiply(A=OneMinus(None=@n6), B=@n6)
  n8 = MaterialFunctionCall(Color=Multiply(A=@n7, B=@n7)) {materialFunction=&SplitComponents}
  n9 = Multiply(
   A=Add(A=Add(A=@n8.R, B=@n8.R), B=@n8.B)
   B=FunctionInput(Preview=Constant() {r=5000}) {bUsePreviewValueAsDefault=true, inputName=Wavy Distortion - Intensity, inputType=FunctionInput_Scalar, sortPriority=10}
  )
  FunctionOutput(None=AppendVector(A=ComponentMask(None=@n1) {g=true, r=true}, B=Subtract(A=@n2, B=@n9))) {outputName=Wavy World Position Distortion}
  FunctionOutput(None=@n9) {outputName=Sediment Wave Distortion, sortPriority=10}
 ]}
